# Copyright Epic Games, Inc. All Rights Reserved.

using { /Fortnite.com/Devices }
using { /Fortnite.com/Teams }
using { /Verse.org/Simulation }
using { ModulePlayerClasses }

ctf_manager := class(creative_device):

	@editable PlayerClassesDevice : ?player_classes_manager_device = false
	@editable TeamDeathmatchDevice : ?fn_game_tdm = false

    # Capture Zone
	@editable Intel_Team1 : capture_item_spawner_device = capture_item_spawner_device{}
	@editable Intel_Team2 : capture_item_spawner_device = capture_item_spawner_device{}
	#@editable Intel_Team1 : carryable_spawner_device = carryable_spawner_device{}
	#@editable Intel_Team2 : carryable_spawner_device = carryable_spawner_device{}

	@editable PlayerMarker : player_marker_device = player_marker_device{}

	#@editable VFXPowerIntel1 : visual_effect_powerup_device = visual_effect_powerup_device{}
	#@editable VFXPowerIntel2 : visual_effect_powerup_device = visual_effect_powerup_device{}

    # Audio Players
    @editable
	@editable VO_IntelCaptured : audio_player_device = audio_player_device{}
	@editable VO_IntelLost : audio_player_device = audio_player_device{}
	@editable VO_IntelPickedUp : audio_player_device = audio_player_device{}
	@editable VO_IntelReturned : audio_player_device = audio_player_device{}
	@editable VO_IntelDropped : audio_player_device = audio_player_device{}

	@editable SupplyDropSpawner : supply_drop_spawner_device = supply_drop_spawner_device{}

    @editable XP_Accolade : accolades_device = accolades_device{}

    OnBegin<override>()<suspends>:void=
	{
		Intel_Team1.ItemCapturedEvent.Subscribe(OnTeamCaptured)
		Intel_Team2.ItemCapturedEvent.Subscribe(OnTeamCaptured)
		Intel_Team1.ItemDroppedEvent.Subscribe(OnTeamDroppedEvent)
		Intel_Team2.ItemDroppedEvent.Subscribe(OnTeamDroppedEvent)
		Intel_Team1.ItemPickedUpEvent.Subscribe(OnTeamPickedUp)
		Intel_Team2.ItemPickedUpEvent.Subscribe(OnTeamPickedUp)
		Intel_Team1.ItemReturnedEvent.Subscribe(OnTeamReturnedEvent)
		Intel_Team2.ItemReturnedEvent.Subscribe(OnTeamReturnedEvent)
	}

    OnTeamCaptured(Agent: agent):void=
	{
		SupplyDropSpawner.Spawn()
		PlayVO_IntelCapturedOrLost(Agent)
		OnIntelDropped(Agent)
	}

    OnTeamDroppedEvent(Agent: agent):void=
	{
		PlayVO_IntelDropped(Agent)
		OnIntelDropped(Agent)
	}

    OnTeamPickedUp(Agent: agent):void=
	{
		PlayVO_IntelPickedUp(Agent)
		OnIntelPickedUp(Agent)
	}

    OnTeamReturnedEvent(Agent: agent):void=
	{
		PlayVO_IntelReturned(Agent)
		OnIntelDropped(Agent)
	}

    OnIntelPickedUp(Agent: agent):void=
	{
		PlayerMarker.Attach(Agent)
		if ( PlayerClasses := PlayerClassesDevice? )
		{
			PlayerClasses.OnIntelPickedUp(Agent)
		}
	}

    OnIntelDropped(Agent: agent):void=
	{
		PlayerMarker.Detach(Agent)
		if ( PlayerClasses := PlayerClassesDevice? )
		{
			PlayerClasses.OnIntelDropped(Agent)
		}
	}

    PlayVO_IntelCapturedOrLost<private>(MyAgent:agent):void=
	{
		if ( PlayerClasses := PlayerClassesDevice? )
		{
			spawn{PlayerClasses.AwardScore(MyAgent, 5.0)}
			spawn{PlayerClasses.DamageBuffPlayer(MyAgent, 10.0, 1.4)}
		}

		TeamCollection := GetPlayspace().GetTeamCollection()
		if ( TeamOwner := TeamCollection.GetTeam[MyAgent] )
		{
			if ( TeamDeathMatch := TeamDeathmatchDevice? )
			{
				TeamDeathMatch.AddTeamKillCount(TeamOwner, 10)
			}

			# For each player in the game, play the appropriate audio & associated HUD message
			AllPlayers := GetPlayspace().GetPlayers()
			for (Player : AllPlayers)
			{
				if ( TeamCollection.IsOnTeam[Player, TeamOwner] )
				{
					PlayVOForPlayer(Player, VO_IntelCaptured)
					XP_Accolade.Award(Player)
					if ( PlayerClasses := PlayerClassesDevice? )
					{
						PlayerClasses.AddUltimateMeter(Player, 25.0)
						PlayerClasses.PlayClassVO(Player, FNVOType.Idle)
					}
				}
				else
				{
					PlayVOForPlayer(Player, VO_IntelLost)
					if ( PlayerClasses := PlayerClassesDevice? )
					{
						PlayerClasses.PlayClassVO(Player, FNVOType.FullAlertLost)
					}
				}
			}
		}
	}

	PlayVO_IntelDropped<private>(MyAgent:agent):void=
	{
		TeamCollection := GetPlayspace().GetTeamCollection()
		if ( TeamOwner := TeamCollection.GetTeam[MyAgent] )
		{
			AllPlayers := GetPlayspace().GetPlayers()
			for (Player : AllPlayers, TeamCollection.IsOnTeam[Player, TeamOwner] )
			{
				PlayVOForPlayer(Player, VO_IntelDropped)
				if ( PlayerClasses := PlayerClassesDevice? )
				{
					PlayerClasses.PlayClassVO(Player, FNVOType.ReturnIdle)
				}
			}
		}
	}

    PlayVO_IntelPickedUp<private>(MyAgent:agent):void=
	{
		TeamCollection := GetPlayspace().GetTeamCollection()
		if ( TeamOwner := TeamCollection.GetTeam[MyAgent] )
		{
			AllPlayers := GetPlayspace().GetPlayers()
			for (Player : AllPlayers)
			{
				PlayVOForPlayer(Player, VO_IntelPickedUp)
				if ( not TeamCollection.IsOnTeam[Player, TeamOwner] )
				{
					if ( PlayerClasses := PlayerClassesDevice? )
					{
						PlayerClasses.PlayClassVO(Player, FNVOType.Sus)
					}
				}
			}
		}
	}

	PlayVO_IntelReturned<private>(MyAgent:agent):void=
	{
		TeamCollection := GetPlayspace().GetTeamCollection()
		if ( TeamOwner := TeamCollection.GetTeam[MyAgent] )
		{
			AllPlayers := GetPlayspace().GetPlayers()
			for (Player : AllPlayers)
			{
				PlayVOForPlayer(Player, VO_IntelReturned)
				if ( TeamCollection.IsOnTeam[Player, TeamOwner] )
				{
					if ( PlayerClasses := PlayerClassesDevice? )
					{
						PlayerClasses.PlayClassVO(Player, FNVOType.ReturnIdle)
					}
				}
			}
		}
	}

    PlayVOForPlayer<private>(MyAgent:agent, AudioDevice:audio_player_device):void=
	{
		AudioDevice.Play(MyAgent)
	}
