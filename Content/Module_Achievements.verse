<#
	2AGENCY
	Achievements
#>
using { /Verse.org/Simulation }

ModuleAchievements<public> := module:

	FNAchievements<public> := enum<persistable> # DEPRECATED
	{
		FirstMatch,					#
		UseDootDevice,				#
		EventCube,					#
		EventZP,					#
		EventDevice,				#
		KillMidasJules,				# Once upon a time.
		KillMidasHades,				# Once upon a time.
		KillFoundationSlone,		# I got over it.
		KillRazSpireAssassin,		# The Spire Quests
		KillAresGods,				# God of War
		#UltMidas,					# Use Ult
		#UltSlone,					# Use Ult
		UltCubeQueen,				# Use Ult
		UltCubeBlue,				# Use Ult
		#UltHerald,					# Use Ult
		#UltAgeless,					# Use Ult
		#UltRelik,					# Use Ult
		#UltKado,					# Use Ult
		#UltValeria,					# Use Ult
		#UltZeus,					# Use Ult
		#UltMixSandstorm,			# Use Ults: Pandora's Box + Olympus
		<#SinnersFinale,			# Kill dancing player with a sword
		#>

		Last
	}