<#
	2AGENCY
	Player Classes Stuff
#>
using { /Fortnite.com/Devices }
using { /Verse.org/Simulation }
using { /Verse.org/Assets }

ModulePlayerClasses<public> := module:

	FNClasses<public> := enum
	{
		None,

		Midas,
		TNTina,
		Brutus,
		Skye,
		Meowscles,
		Kit,
		Jules,
		Ocean,
		ShadowMidas,
		Shadow,
		SpireAssassin,
		Raz,
		Slone,
		ZygAndChoppy,
		CubeAssassin,
		CubeQueen,
		SgtWinter,
		TheFoundation,
		Gunnar,
		TheIceKing,
		HuntmasterSaber,
		TheHerald,
		TheAgelessChampion,
		WildguardRelik,
		KadoThorne,
		TheDealer,
		Oscar,
		Nisha,
		Montague,
		Valeria,
		Ares,
		Hades,
		Cerberus,
		Zeus,
		MegaloDon,
		RingmasterScarr,
		TheMachinist,
		UltimaCarver,
		ShogunX,
		NightRose,
		Puppeteer,
		FletcherKane,
		AlphaHeavy,
		Daigo,
		Kor,

		CubeBlue,
		Jones,
		DoctorDoom,
		DoctorDoomMythic,
		Caretaker,
		Hope,

		Last
	}

	FNFactions<public> := enum
	{
		None,

		Spies,
		Seven,
		IO,
		LastReality,
		Society,
		Pantheon,
		Wasteland,

		Last
	}

	FNVOType<public> := enum
	{
		Idle,
		Sus,
		ReturnIdle,
		FullAlert,
		#FullAlertIdle,
		FullAlertLost,
		DBNO,
		Interrogation,
		Ability,
		AbilityEnd,
		Ult
	}

	FNBGMState<public> := enum
	{
		None,
		Idle,
		Alerted,
		AlertedPhase2
	}

	FNAbility<public> := enum
	{
		None,
		DoubleJump,
		Warp,
		ShieldRegen,
		Shadow,
		LifeDrain,
		Rock,
		Charge,
		SpinSlash,
		AlienParasite,
		MirrorImage,
		Lockdown,
		Phase2,
		Rebirth,
		DamageBuff,
		HealingRune,
		Laserbeam,
		RiftGun,
		AirRaid,
		LastResort,
		Engineer,
		PersonalShield,
		FootDive,
		PartingGift,
		Snap,
		GrindVines,
		TherapeuticSwim,
		Bombardment,
		DemonClaw,
		Puppeteer,
		Enragement,
		HealthRegen,
		ShortBeam
	}

	FNAbilityData<public> := struct
	{
		#ID<public>: FNAbility = FNAbility.None
		Title<public>: string = ""
		#Desc<public>: string = ""
		Icon<public>: texture = Textures.ui_ability_no
		Cooldown<public>: float = 10.0
		ShowOnHud<public>: logic = true
	}

	FNCond<public> := enum
	{
		Burning,
		SlurpHeal
	}

	<#fn_cond_s<public> := struct
	{
		ExpireTime<public> : float = 0.0
		Provider : ?fort_character = false
	}#>

	CategoryTeamAlt<localizes>:message="Team Alt"

	fn_class_data<public> := class<concrete>
	{
		FNClassID <public>: FNClasses = FNClasses.None
		FNClassFaction <public>: FNFactions = FNFactions.None

		ClassImage <public>: texture = Textures.midas_cursed1
		#@editable ClassImage : texture = Textures.midas_cursed1

		FNClassName<public>:message = TextLocalization.Class_Default
		FNClassTitle<public>:message = TextLocalization.ClassTitle_Default

		MaxHealth<public>:float = 400.0
		MaxShield<public>:float = 100.0
		MaxUlt<public>:float = 100.0

		MainAbility<public> :FNAbility = FNAbility.None
		MainAbility2<public> :FNAbility = FNAbility.None
		PassiveList<public> :[]FNAbility = array{}

		@editable ClassSelector<public> : class_and_team_selector_device = class_and_team_selector_device{}

		@editable BGMIdle<public> : audio_player_device = audio_player_device{}
		@editable BGMPhase1<public> : audio_player_device = audio_player_device{}
		@editable BGMPhase2<public> : audio_player_device = audio_player_device{}

		@editable VFXPowerup<public> : visual_effect_powerup_device = visual_effect_powerup_device{}

		@editable HasTeamAlt<public>:logic = false
		Team2_ClassImage <public>: texture = Textures.midas_cursed1
		Team2_FNClassName<public>:message = TextLocalization.Class_Default
		@editable{Categories:=array{CategoryTeamAlt}} Team2_BGMIdle<public> : audio_player_device = audio_player_device{}
		@editable{Categories:=array{CategoryTeamAlt}} Team2_BGMPhase1<public> : audio_player_device = audio_player_device{}
		@editable{Categories:=array{CategoryTeamAlt}} Team2_VFXPowerup<public> : visual_effect_powerup_device = visual_effect_powerup_device{}
	}
