<#
	2AGENCY
	Player Ex
#>
using { /Fortnite.com/Devices }
using { /Verse.org/Simulation }
using { /Verse.org/Assets }
using { /Verse.org/SceneGraph }
using { /Fortnite.com/Characters }
using { /Verse.org/Simulation/Tags }
using { /UnrealEngine.com/Temporary/SpatialMath }

ModulePlayerEx<public> := module:

	(Agent : agent).IsInSpawnRoom<public>(CreativeDevice:creative_device)<transacts><decides>: void =
	{
		var IsItTrue : logic = false
		TaggedDevice := CreativeDevice.FindCreativeObjectsWithTag(spawnroom_volume_tag{})
		for (FoundDevice: TaggedDevice, SpawnZone := mutator_zone_device[FoundDevice], SpawnZone.IsInVolume[Agent])
		{
			set IsItTrue = true
			IsItTrue?
		}

		IsItTrue?
	}

	(Agent : agent).IsOnTeam<public>(CreativeDevice:creative_device, TeamID:int)<transacts><decides>: void =
	{
		var IsItTrue : logic = false
		TeamCollection := CreativeDevice.GetPlayspace().GetTeamCollection()
		Teams := TeamCollection.GetTeams()
		if ( TeamCollection.IsOnTeam[Agent, Teams[TeamID]] )
		{
			set IsItTrue = true
			IsItTrue?
		}

		IsItTrue?
	}

	ChairAnim<public>(CreativeDevice:creative_device, Agent:agent, PropAnim:creative_prop_asset, AnimDuration:float)<suspends>:void=
	{
		if (FortCharacter := Agent.GetFortCharacter[], FortCharacter.IsActive[] )
		{
			race
			{
				block
				{
					FortCharacter.EliminatedEvent().Await()
					#Print("Chair Anim: died.")
				}
				block
				{
					PlayerPos := FortCharacter.GetTransform().Translation
					PlayerRot := FortCharacter.GetViewRotation()
					if (PlayerRotYaw := PlayerRot.GetYawPitchRollDegrees()[0], PlayerRotRoll := PlayerRot.GetYawPitchRollDegrees()[2])
					{
						AnimPos := PlayerPos
						AnimRot := MakeRotationFromYawPitchRollDegrees(PlayerRotYaw, 0.0, PlayerRotRoll) 
						NewTransform := transform{Scale := vector3{X:= 0.1, Y:= 0.1, Z:= 0.1}, Rotation := AnimRot, Translation := AnimPos} # newly spawned chair device still has chair mesh so we just have to make it not so visible.
						SpawnVFX := SpawnProp(PropAnim, NewTransform)
						if (SpawnedVFX:= SpawnVFX(0)?)
						{
							defer
							{
								FortCharacter.ReleaseFromStasis()
								SpawnedVFX.Dispose()
								#if ( FortCharacter.TeleportTo[PlayerPos, PlayerRot] ) {}
							}
							PropPos := SpawnedVFX.GetTransform().Translation
							TaggedDevice := CreativeDevice.FindCreativeObjectsWithTag(chairanim_tag{})
							for (FoundDevice: TaggedDevice, ChairDevice := chair_device[FoundDevice])
							{
								DevicePos := FoundDevice.GetTransform().Translation
								if ( IsAlmostEqual[PropPos,DevicePos,0.1] ) # 0.0 is bad. 0.1 just right.
								{
									ChairDevice.Seat(Agent)
									ChairDevice.DisableExit()
									FortCharacter.PutInStasis(stasis_args{AllowTurning := false, AllowFalling := false, AllowEmotes := true})
									Sleep(AnimDuration)
									ChairDevice.Eject()
								}
							}
						}
					}
				}
			}
		}
	}

	ChairAnimFollow<public>(CreativeDevice:creative_device, Agent:agent, PropAnim:creative_prop_asset, AnimDuration:float, ?IgnoreZ:logic = false)<suspends>:void=
	{
		if (FortCharacter := Agent.GetFortCharacter[], FortCharacter.IsActive[] )
		{
			race
			{
				block
				{
					FortCharacter.EliminatedEvent().Await()
					#Print("Chair Anim: died.")
				}
				block
				{
					PlayerPos := FortCharacter.GetTransform().Translation
					PlayerRot := FortCharacter.GetViewRotation()
					if (PlayerRotYaw := PlayerRot.GetYawPitchRollDegrees()[0], PlayerRotRoll := PlayerRot.GetYawPitchRollDegrees()[2])
					{
						AnimPos := PlayerPos
						AnimRot := MakeRotationFromYawPitchRollDegrees(PlayerRotYaw, 0.0, PlayerRotRoll) 
						NewTransform := transform{Scale := vector3{X:= 0.1, Y:= 0.1, Z:= 0.1}, Rotation := AnimRot, Translation := AnimPos} # newly spawned chair device still has chair mesh so we just have to make it not so visible.
						SpawnVFX := SpawnProp(PropAnim, NewTransform)
						if (SpawnedVFX:= SpawnVFX(0)?)
						{
							defer
							{
								FortCharacter.ReleaseFromStasis()
								SpawnedVFX.Dispose()
								#if ( FortCharacter.TeleportTo[PlayerPos, PlayerRot] ) {}
							}
							PropPos := SpawnedVFX.GetTransform().Translation
							TaggedDevice := CreativeDevice.FindCreativeObjectsWithTag(chairanim_tag{})
							for (FoundDevice: TaggedDevice, ChairDevice := chair_device[FoundDevice])
							{
								DevicePos := FoundDevice.GetTransform().Translation
								if ( IsAlmostEqual[PropPos,DevicePos,0.1] ) # 0.0 is bad. 0.1 just right.
								{
									ChairDevice.Seat(Agent)
									ChairDevice.DisableExit()
									FortCharacter.PutInStasis(stasis_args{AllowTurning := false, AllowFalling := false, AllowEmotes := true})

									race
									{
										block
										{
											loop
											{
												if (not SpawnedVFX.IsValid[])
												{
													#Print("ChairFollow: Chair not valid")
													break
												}

												if (not FortCharacter.IsActive[])
												{
													#Print("ChairFollow: Player died")
													break
												}

												Sleep(0.1)
												#Print("ChairFollow: Looping...")
												PlayerNewPos := FortCharacter.GetTransform().Translation
												PlayerNewRotation := FortCharacter.GetViewRotation()

												var NewRot : rotation = PlayerNewRotation
												if (IgnoreZ?, PlayerNewRotYaw := PlayerNewRotation.GetYawPitchRollDegrees()[0], PlayerNewRotPitch := PlayerNewRotation.GetYawPitchRollDegrees()[1])
												{
													set NewRot = MakeRotationFromYawPitchRollDegrees(PlayerNewRotYaw, PlayerNewRotPitch, 0.0) 
												}

												ChairDevice.MoveTo(PlayerNewPos, NewRot, 0.15)
											}
										}
										block
										{
											Sleep(AnimDuration)
											if ( ChairDevice.TeleportTo[AnimPos, AnimRot] ) {}
											ChairDevice.Eject()
										}
									}
								}
							}
						}
					}
				}
			}
		}
	}

	ChairAnimInvulnerable<public>(CreativeDevice:creative_device, Agent:agent, PropAnim:creative_prop_asset, AnimDuration:float)<suspends>:void=
	{
		if (FortCharacter := Agent.GetFortCharacter[], FortCharacter.IsActive[] )
		{
			FortCharacter.SetVulnerability(false)
			spawn{ChairAnim(CreativeDevice,Agent,PropAnim,AnimDuration)}
			Sleep(AnimDuration)
			FortCharacter.SetVulnerability(true)
		}
	}

	(Agent:agent).GetActivePlayer<public>()<decides><transacts>:player=
	{
		Player:player = player[Agent]
		Player.IsActive[]
		return Player
	}

	(Device : creative_device).GetTaggedObjects<public>(Tag:tag, cast_type:castable_subtype(creative_object_interface))<transacts>:[]cast_type =
	{
		for (Obj: Device.FindCreativeObjectsWithTag(Tag), Casted := cast_type[Obj])
		{
			Casted
		}
	}

	(Agent:agent).SlowHeal<public>(HealAmount:float)<suspends>:void=
	{
		if (FortCharacter := Agent.GetFortCharacter[], FortCharacter.IsActive[], HealInt := Int[HealAmount] )
		{
			race
			{
				block
				{
					var NumberOfHeal : int = 0
					loop
					{
						if ( not FortCharacter.IsActive[] )
						{
							break
						}
						FortCharacter.Heal(1.0)

						set NumberOfHeal += 1
						if ( NumberOfHeal >= HealInt )
						{
							break
						}

						Sleep(0.1)
					}
				}
				block
				{
					FortCharacter.EliminatedEvent().Await()
				}
			}
		}
	}

	(Agent:agent).MoveMod<public>(MoveModDevice:movement_modulator_device):void=
	{
		if ( FortCharacter := Agent.GetFortCharacter[], FortCharacter.IsActive[] )
		{
			MoveModDevice.Activate(Agent)
		}
	}

	<#HideAllHudElements(): void=
	{
		MinimapHUD: creative_hud_identifier_minimap = creative_hud_identifier_minimap{}
		StormTimerHUD: creative_hud_identifier_storm_timer = creative_hud_identifier_storm_timer{}
		StormNotiHUD: creative_hud_identifier_storm_notifications = creative_hud_identifier_storm_notifications{}
		PlayerCountHUD: creative_hud_identifier_player_count = creative_hud_identifier_player_count{}
		ElimCounterHUD: creative_hud_identifier_elimination_counter = creative_hud_identifier_elimination_counter{}
		TopLeftTeamInfoHUD: creative_hud_identifier_team_info = creative_hud_identifier_team_info{}

		ElementsToHide :[] hud_element_identifier = array{
			MinimapHUD,
			StormTimerHUD,
			StormNotiHUD,
			PlayerCountHUD,
			ElimCounterHUD,
			TopLeftTeamInfoHUD
		}

		HudController := GetPlayspace().GetHUDController()
		HudController.HideElements(ElementsToHide)
	}#>

	(Entity:entity).PlaySFX<public>(SoundAsset:sound_wave, Duration:float)<suspends>:void=
	{
		if ( SimulationEntity := Entity.GetSimulationEntity[] )
		{
			SpawnTransform := Entity.GetGlobalTransform()
			NE := entity{}
			TransComp := transform_component{ Entity := NE }
			SFXComp := Audio.MSS_Oneshot{ Entity := NE, Sounds := array{SoundAsset}, VolumeBase := 0.0, PitchBase := 0.0, AutoPlay := true }
			NE.AddComponents(array{TransComp,SFXComp})
			SimulationEntity.AddEntities(array{NE})
			NE.SetGlobalTransform(SpawnTransform)
			Sleep(Duration)
			NE.RemoveFromParent()
		}
	}
