<#
	Modules
	Creative Prop Stuff
#>
using { /Fortnite.com/Devices }
using { /Fortnite.com/Devices/CreativeAnimation }
using { /Fortnite.com/Devices/CreativeAnimation/InterpolationTypes }
using { /Fortnite.com/Characters }
using { /Verse.org/Assets }
using { /Verse.org/Simulation }
using { /Verse.org/SceneGraph }
using { /Verse.org/SpatialMath }
using { /UnrealEngine.com/Assets }
using { /UnrealEngine.com/Temporary/SpatialMath }

ModulePropExtra<public> := module:

	SpawnPropVFX<public>(PropVFX:creative_prop_asset,SpawnPos:(/UnrealEngine.com/Temporary/SpatialMath:)vector3,SpawnRotation:(/UnrealEngine.com/Temporary/SpatialMath:)rotation,LifeTime:float)<suspends>:void=
	{
		SpawnVFX := SpawnProp(PropVFX, SpawnPos, SpawnRotation)
		Sleep(LifeTime)
		if (SpawnedVFX:= SpawnVFX(0)?) {SpawnedVFX.Dispose() }
	}

	SpawnParticleVFX<public>(ParticleVFX:particle_system, SpawnPos:(/UnrealEngine.com/Temporary/SpatialMath:)vector3,SpawnRotation:(/UnrealEngine.com/Temporary/SpatialMath:)rotation,LifeTime:float)<suspends>:void=
	{
		MyVFX:cancelable = SpawnParticleSystem(ParticleVFX, SpawnPos, ?Rotation := SpawnRotation, ?StartDelay := 0.0)
		Sleep(LifeTime)
		MyVFX.Cancel()


	}

	<#(CreativeDevice:creative_device).SpawnFXComponent(VFXCompo:particle_system_component, SpawnPos:(/Verse.org/SpatialMath:)vector3, LifeTime:float)<suspends>:void=
	{
		if ( SimulationEntity := CreativeDevice.GetSimulationEntity[] )
		{
			SpawnTransform := (/Verse.org/SpatialMath:)transform{Translation:=SpawnPos}
			NE := entity{}
			TransComp := transform_component{ Entity := NE }
			VFXComp := VFXCompo{ Entity := NE, AutoPlay := true }
			NE.AddComponents(array{TransComp,VFXComp})
			SimulationEntity.AddEntities(array{NE})
			NE.SetGlobalTransform(SpawnTransform)
			Sleep(LifeTime)
			NE.RemoveFromParent()
		}
	}#>

	SpawnPropVFXFollowPlayer<public>(PropVFX:creative_prop_asset, FortCharacter:fort_character, LifeTime:float, ?IgnoreZ:logic = false)<suspends>:void=
	{
		PlayerPos := FortCharacter.GetTransform().Translation
		#PlayerRotation := FortCharacter.GetViewRotation()
		SpawnVFX := SpawnProp(PropVFX, PlayerPos, (/UnrealEngine.com/Temporary/SpatialMath:)rotation{})

		if (SpawnedVFX:= SpawnVFX(0)?)
		{
			#if ( set mPlayerVFX[Agent] += array{SpawnedVFX} ) {}

			branch
			{
				block
				{
					loop
					{
						if (not SpawnedVFX.IsValid[])
						{
							SpawnedVFX.Dispose()
							break
						}

						if (not FortCharacter.IsActive[])
						{
							SpawnedVFX.Dispose()
							break
						}

						Sleep(0.1)
						PlayerNewPos := FortCharacter.GetTransform().Translation
						PlayerNewRotation := FortCharacter.GetViewRotation()

						var NewRot : (/UnrealEngine.com/Temporary/SpatialMath:)rotation = PlayerNewRotation
						if (IgnoreZ?, PlayerRotYaw := PlayerNewRotation.GetYawPitchRollDegrees()[0], PlayerRotPitch := PlayerNewRotation.GetYawPitchRollDegrees()[1])
						{
							set NewRot = (/UnrealEngine.com/Temporary/SpatialMath:)MakeRotationFromYawPitchRollDegrees(PlayerRotYaw, PlayerRotPitch, 0.0) 
						}

						#SpawnedVFX.MoveTo(PlayerNewPos, PlayerNewRotation, 0.15)

						if (AController := SpawnedVFX.GetAnimationController[])
						{
							AnimKeyframes := array
							{
								keyframe_delta
								{
									DeltaLocation := PlayerNewPos - SpawnedVFX.GetTransform().Translation
									DeltaRotation := MakeComponentWiseDeltaRotation(NewRot, SpawnedVFX.GetTransform().Rotation)
									Time := 0.1
									Interpolation := Linear
								}
							}
							AController.SetAnimation(AnimKeyframes, ?Mode:=animation_mode.OneShot)
							AController.Play()
						}
					}
				}
			}
			block
			{
				if ( LifeTime <> 0.0 )
				{
					Sleep(LifeTime)
					SpawnedVFX.Dispose()
				}
			}
		}
	}

	(CreativeProp : creative_prop).SmoothMoveTo<public>(DestinationPos:(/UnrealEngine.com/Temporary/SpatialMath:)vector3, Duration:float)<suspends>:void=
	{
		if (AController := CreativeProp.GetAnimationController[])
		{
			AnimKeyframes := array
			{
				keyframe_delta
				{
					DeltaLocation := DestinationPos - CreativeProp.GetTransform().Translation
					DeltaRotation := (/UnrealEngine.com/Temporary/SpatialMath:)rotation{}
					Time := Duration
					Interpolation := Linear
				}
			}
			AController.SetAnimation(AnimKeyframes, ?Mode:=animation_mode.OneShot)
			AController.Play()
			Sleep(Duration)
		}
	}

	<#StopPropVFXFollowPlayer<public>(FortCharacter:fort_character)<suspends>:void=
	{
		if (set mPlayerVFX[Agent] = array{}) {}

		AllVFX := mPlayerVFX[Agent]
		for ( VFX : AllVFX )
		{
			VFX.Dispose()
		}
	}#>