using. /Verse.org/Simulation

TextLocalization := module:

	UIPlayerName<public><localizes>(Agent:agent)<computes>:message = "{Agent}"

	Class_Default<public><localizes>:message			= "ClassName"
	ClassTitle_Default<public><localizes>:message		= "ClassTitle"
	Class_Midas<public><localizes>:message 				= "Midas"
	ClassTitle_Midas<public><localizes>:message 		= "Mastermind of the Spies"
	Class_TNTina<public><localizes>:message 			= "TNTina"
	Class_DynamoTNTina<public><localizes>:message 		= "Dynamo TNTina"
	ClassTitle_TNTina<public><localizes>:message		= "Explosive of the Spies"
	Class_Brutus<public><localizes>:message				= "Brutus"
	ClassTitle_Brutus<public><localizes>:message		= "Muscle of the Spies"
	Class_Skye<public><localizes>:message				= "Skye"
	ClassTitle_Skye<public><localizes>:message			= "Adventurer of the Spies"
	Class_Meowdas<public><localizes>:message			= "Meowscles"
	Class_Meowscles<public><localizes>:message			= "Meowdas"
	ClassTitle_Meowscles<public><localizes>:message		= "Muscle Cat of the Spies"
	Class_Kit<public><localizes>:message				= "Kit"
	ClassTitle_Kit<public><localizes>:message			= "Kitten of the Spies"
	Class_Jules<public><localizes>:message				= "Jules"
	ClassTitle_Jules<public><localizes>:message			= "Engineer of the Spies"
	Class_Ocean<public><localizes>:message				= "Ocean"
	ClassTitle_Ocean<public><localizes>:message			= "Ocean of the Spies"
	Class_ShadowMidas<public><localizes>:message		= "Shadow Midas"
	ClassTitle_ShadowMidas<public><localizes>:message	= "Mastermind of the Dead"
	Class_Shadow<public><localizes>:message				= "Shadow"
	ClassTitle_Shadow<public><localizes>:message		= "Shadow"
	Class_SpireAssassin<public><localizes>:message		= "Spire Assassin"
	ClassTitle_SpireAssassin<public><localizes>:message	= "Guardian of the Imagined Order"
	Class_Raz<public><localizes>:message				= "Glyph Master Raz"
	ClassTitle_Raz<public><localizes>:message			= "Thief of the Last Reality"
	Class_Slone<public><localizes>:message				= "Doctor Slone"
	ClassTitle_Slone<public><localizes>:message			= "Leader of the Imagined Order"
	Class_ZygAndChoppy<public><localizes>:message		= "Zyg and Choppy"
	ClassTitle_ZygAndChoppy<public><localizes>:message	= "Droid of the Last Reality"
	Class_CubeBlue<public><localizes>:message			= "Blue Cube"
	ClassTitle_CubeBlue<public><localizes>:message		= "Defector of the Last Reality"
	Class_CubeAssassin<public><localizes>:message		= "Cube Assassin"
	ClassTitle_CubeAssassin<public><localizes>:message	= "Assassin of the Last Reality"
	Class_CubeQueen<public><localizes>:message			= "The Cube Queen"
	ClassTitle_CubeQueen<public><localizes>:message		= "Ruler of the Last Reality"
	Class_SgtWinter<public><localizes>:message			= "Sgt. Winter"
	ClassTitle_SgtWinter<public><localizes>:message		= "Santa Clause"
	Class_Foundation<public><localizes>:message			= "The Foundation"
	ClassTitle_Foundation<public><localizes>:message	= "Leader of the Seven"
	Class_Gunnar<public><localizes>:message				= "Gunnar"
	ClassTitle_Gunnar<public><localizes>:message		= "Enforcer of the Imagined Order"
	Class_HuntmasterSaber<public><localizes>:message	= "Huntmaster Saber"
	ClassTitle_HuntmasterSaber<public><localizes>:message	= "Masterkey of the Imagined Order"
	Class_Herald<public><localizes>:message				= "The Herald"
	ClassTitle_Herald<public><localizes>:message		= "Herald of the Last Reality"
	Class_Ageless<public><localizes>:message			= "The Ageless Champion"
	ClassTitle_Ageless<public><localizes>:message		= "Predecessor of Imagined Order"
	Class_Relik<public><localizes>:message				= "Wildguard Relik"
	ClassTitle_Relik<public><localizes>:message			= "The Intrepid Hunter"
	Class_KadoThorne<public><localizes>:message			= "Kado Thorne"
	ClassTitle_KadoThorne<public><localizes>:message	= "Eclipsed Time Traveller"
	Class_Dealer<public><localizes>:message				= "The Dealer"
	ClassTitle_Dealer<public><localizes>:message		= "Eclipsed Dealer"
	Class_Oscar<public><localizes>:message				= "Oscar"
	ClassTitle_Oscar<public><localizes>:message			= "Hunter of the Society"
	Class_Nisha<public><localizes>:message				= "Nisha"
	ClassTitle_Nisha<public><localizes>:message			= "Fencer of the Society"
	Class_Montague<public><localizes>:message			= "Montague"
	ClassTitle_Montague<public><localizes>:message		= "Founder of the Society"
	Class_Valeria<public><localizes>:message			= "Valeria"
	ClassTitle_Valeria<public><localizes>:message		= "Leader of the Society"
	Class_Zeus<public><localizes>:message				= "Zeus"
	ClassTitle_Zeus<public><localizes>:message			= "God of Thunder"
	Class_Ares<public><localizes>:message				= "Ares"
	ClassTitle_Ares<public><localizes>:message			= "God of War"
	Class_Cerberus<public><localizes>:message			= "Cerberus"
	ClassTitle_Cerberus<public><localizes>:message		= "Dog of the Gods"
	Class_Hades<public><localizes>:message				= "Hades"
	ClassTitle_Hades<public><localizes>:message			= "God of the Dead"
	Class_MegaloDon<public><localizes>:message			= "Megalo Don"
	ClassTitle_MegaloDon<public><localizes>:message		= "Lord of the Wasteland"
	Class_RingmasterScarr<public><localizes>:message	= "Ringmaster Scarr"
	ClassTitle_RingmasterScarr<public><localizes>:message	= "Ringmaster of the Wasteland"
	Class_Machinist<public><localizes>:message			= "The Machinist"
	ClassTitle_Machinist<public><localizes>:message		= "Mechanic of the Wasteland"
	Class_UltimaCarver<public><localizes>:message		= "Ultima Carver"
	ClassTitle_UltimaCarver<public><localizes>:message	= "Brawler's Patch"
	Class_ShogunX<public><localizes>:message			= "Shogun X"
	ClassTitle_ShogunX<public><localizes>:message		= "Demon Shogun"
	Class_NightRose<public><localizes>:message			= "The Night Rose"
	ClassTitle_NightRose<public><localizes>:message		= "Demon's Puppet"
	Class_Puppeteer<public><localizes>:message			= "The Puppeteer"
	ClassTitle_Puppeteer<public><localizes>:message		= "Demon Puppeteer"
	Class_FletcherKane<public><localizes>:message		= "Fletcher Kane"
	ClassTitle_FletcherKane<public><localizes>:message	= "Alpha Baron"
	Class_AlphaHeavy<public><localizes>:message			= "Alpha Heavy"
	ClassTitle_AlphaHeavy<public><localizes>:message	= "Alpha Elite"
	Class_Daigo<public><localizes>:message				= "Daigo"
	ClassTitle_Daigo<public><localizes>:message			= "Demon Mask Maker"
	Class_Kor<public><localizes>:message				= "Kor"
	ClassTitle_Kor<public><localizes>:message			= "Demon Lieutenant"
	Class_IceKing<public><localizes>:message			= "The Ice King"
	ClassTitle_IceKing<public><localizes>:message		= "Ruler of the Ice Moon"
	Class_Jones<public><localizes>:message				= "Jones"
	ClassTitle_Jones<public><localizes>:message			= "The Legend"
	Class_DoctorDoom<public><localizes>:message			= "The Wanderer"
	ClassTitle_DoctorDoom<public><localizes>:message	= "Friend of the Wasteland"
	Class_DoctorDoomMythic<public><localizes>:message	= "Doom"
	ClassTitle_DoctorDoomMythic<public><localizes>:message	= "Absolute Doom"
	Class_Caretaker<public><localizes>:message			= "Caretaker"
	ClassTitle_Caretaker<public><localizes>:message		= "Caretaker of the Last Reality"
	Class_Hope<public><localizes>:message				= "Hope"
	ClassTitle_Hope<public><localizes>:message			= "Pandora's Chosen"

	DefeatedBy<public><localizes>:message = "You were defeated by"

	SelectClass<public><localizes>:message = "Select Your Class"
	ChangeTeam<public><localizes>:message = "Change Team"
	MainMenu<public><localizes>:message = "👑Main Menu"
	Settings<public><localizes>:message = "Settings"
	Achievements<public><localizes>:message = "Achievements"

	Setting_MusicON<public><localizes>:message = "🎵Music: ON"
	Setting_MusicOFF<public><localizes>:message = "🎵Music: OFF"
	Setting_VoiceMenuON<public><localizes>:message = "💬 Voice Menu: ON"
	Setting_VoiceMenuOFF<public><localizes>:message = "💬 Voice Menu: OFF"
	Setting_MouseCursorON<public><localizes>:message = "🖱️ Mouse Cursor on Class Selection (PC ONLY): ON"
	Setting_MouseCursorOFF<public><localizes>:message = "🖱️ Mouse Cursor on Class Selection (PC ONLY): OFF"
	Setting_HatON<public><localizes>:message = "🎩Boss Hat: ON"
	Setting_HatOFF<public><localizes>:message = "🎩Boss Hat: OFF"
	Setting_FirstPersonON<public><localizes>:message = "🎥First Person Mode: ON"
	Setting_FirstPersonOFF<public><localizes>:message = "🎥First Person Mode: OFF"
	Exit<public><localizes>:message = "⬅️ Exit"

	AbilityDesc_DoubleJump<public><localizes>:message = "ABILITY: Double Jump"

	MSG_UseUlt<public><localizes>:message = "<Em>EMOTE</> to unleash <Em>Ultimate!</>"
	MSG_AbilityCool<public><localizes>:message = "Ability on cooldown..."
	MSG_NoAbilityInSpawn<public><localizes>:message = "Cannot use Ability in spawn room."
	MSG_NoTeammateFound<public><localizes>:message = "No teammate found."