<#
	Modules
	UI Stuff
#>
using { /Verse.org/Assets }
using { /Verse.org/Simulation }
using { /UnrealEngine.com/Assets }
using { /UnrealEngine.com/Temporary/SpatialMath }
using { /UnrealEngine.com/Temporary/UI }

ModuleUIExtra<public> := module:

(Listenable : listenable(widget_message)).SubscribeWidget(OutputFunc : tuple(widget_message, t)->void, ExtraData : t where t:type) : cancelable =
{
    Wrapper := wrapper_widgetmsg(t) {ExtraData := ExtraData, OutputFunc := OutputFunc}
    Listenable.Subscribe(Wrapper.InputFunc)
}
 
wrapper_widgetmsg(t : type) := class()
{
    ExtraData : t;
    OutputFunc : tuple(widget_message, t) -> void
    InputFunc(Message : widget_message):void = OutputFunc(Message, ExtraData)
}

(Listenable : listenable(tuple())).SubscribeEmpty(OutputFunc : t -> void, ExtraData : t where t:type) : cancelable =
{
    Wrapper := wrapper_empty(t) {ExtraData := ExtraData, OutputFunc := OutputFunc}
    Listenable.Subscribe(Wrapper.InputFunc)
}
 
wrapper_empty(t : type) := class()
{
    ExtraData : t;
    OutputFunc : t -> void
    InputFunc():void = OutputFunc(ExtraData)
}
