<#
	2AGENCY
	Player Classes Manager
#>
using { /Fortnite.com/Devices }
using { /Fortnite.com/Devices/CreativeAnimation }
using { /Fortnite.com/Devices/CreativeAnimation/InterpolationTypes }
using { /Fortnite.com/Characters }
using { /Fortnite.com/Teams }
using { /Fortnite.com/UI }
using { /Fortnite.com/Game }
using { /Fortnite.com/FortPlayerUtilities }
using { /Fortnite.com/AI }
using { /Fortnite.com/Animation/PlayAnimation }
using { /Verse.org/Assets }
using { /Verse.org/SceneGraph }
using { /Verse.org/Simulation }
using { /Verse.org/Simulation/Tags }
using { /Verse.org/Random }
using { /Verse.org/Colors }
using { /UnrealEngine.com/Assets }
using { /UnrealEngine.com/Temporary/Diagnostics }
using { /UnrealEngine.com/Temporary/UI }
using { /UnrealEngine.com/Temporary/SpatialMath }
using { ModulePlayerClasses }
using { ModulePlayerEx }
using { ModulePropExtra }
using { ModuleAchievements }

<# super useful notes
	"movement_modulator_device" on npc = verse internal error
	"movement_modulator_device" on dead player = verse internal error
	"vfx_creator" on npc = verse internal error
	check if player is ded or not before using damage_args.
	Mythic Zeus Thunderbolts + DBNO device.Revive = verse internal error.
	Commanding Hired guard will fucking crash the ENTIRE game.
	-cardboard boxes cause verse internal errors
	UI Hud disappeared after opening PopUp Menu
	DO NOT use multiple input trigger with the same key. it will trigger the subscribed codes multiple time.
	big bang event music is copyrighted. cant use.
	widget can get stuck forever for some reason.
	when import skeletal mesh to uefn, scale up to 100!
#>

<# Todo
	make ability hud into stack_box....... hsr status effects
	add bottom xp bar?
	make ice king cyro ability.
	make "apply nitro when shield break" ability.
	make damage reduction status for kado thorne.
#>


spawnroom_volume_tag := class(tag){}
vfx_hat_tag := class(tag){}

Audio := module:
	VO<public> := module:
		AlphaHeavy<public> := module{}
		Ares<public> := module{}
		Brutus<public> := module{}
		Caretaker<public> := module{}
		Cerberus<public> := module{}
		CubeAssassin<public> := module{}
		CubeBlue<public> := module{}
		CubeQueen<public> := module{}
		Daigo<public> := module{}
		Doom<public> := module{}
		Doombot<public> := module{}
		FletcherKane<public> := module{}
		Gunnar<public> := module{}
		Hades<public> := module{}
		Hope<public> := module{}
		HuntmasterSaber<public> := module{}
		IceKing<public> := module{}
		Jones<public> := module{}
		Jules<public> := module{}
		KadoThorne<public> := module{}
		Kit<public> := module{}
		Kor<public> := module{}
		MegaloDon<public> := module{}
		Meowscles<public> := module{}
		Midas<public> := module{}
		Montague<public> := module{}
		NightRose<public> := module:
			Puppeteer<public> := module{}
		Nisha<public> := module{}
		Ocean<public> := module{}
		Oscar<public> := module{}
		Raz<public> := module{}
		Relik<public> := module{}
		RingmasterScarr<public> := module{}
		SgtWinter<public> := module{}
		Shadow<public> := module{}
		ShadowMidas<public> := module{}
		ShogunX<public> := module{}
		Skye<public> := module{}
		Slone<public> := module{}
		SpireAssassin<public> := module{}
		TheAgeless<public> := module{}
		TheDealer<public> := module{}
		TheFoundation<public> := module{}
		TheHerald<public> := module{}
		TheMachinist<public> := module{}
		TNTina<public> := module{}
		UltimaCarver<public> := module{}
		Valeria<public> := module{}
		Zeus<public> := module{}
		ZygAndChoppy<public> := module{}

FNCondData<public> := struct
{
	#ID<public>: FNCond = FNCond.None
	AgentProvider<public>: ?agent = false
	ExpireTime<public>: float = 10.0
}

fn_player_class := class()
{
    Agent<public> : agent
    LinkedCreativeDevice<public> : player_classes_manager_device

	var AgentClass<public> : FNClasses = FNClasses.None
	var ClassData	: fn_class_data = fn_class_data{FNClassID := FNClasses.None, FNClassFaction := FNFactions.None}

	var ActiveBGM : ?audio_player_device = false
	var ActiveSummon : ?fort_character = false
	var ActiveBuilding : ?creative_prop = false
	var ActiveVO : ?entity = false

	var MainAbility: FNAbilityData = FNAbilityData{Icon := Textures.ui_ability_no, Title := ""}
	var MainAbility2: FNAbilityData = FNAbilityData{Icon := Textures.ui_ability_no, Title := ""}
	var PassiveList: []FNAbilityData = array{}
	var flAbilityCooldown : float = 0.0
	var flAbilityCooldownMax : float = 10.0
	var flAbilityCooldown2 : float = 0.0
	var flAbilityCooldownMax2 : float = 10.0

	var flAbilityUsedTime : float = 0.0
	var flVOUsedTime : float = 0.0

	#var AbilityCooldown: []float = array{}

	var flPlayerUltimateMeter : float = 0.0
	var DamageMult : float = 1.0
	var HealthMult : float = 1.0
	var ShieldMult : float = 1.0

	var CondMap: [FNCond]FNCondData = map{}

	var bFirstSpawn : logic = true
	var bUltReady : logic = false
	var bAltTransform : logic = false
	var bDoubleJumped : logic = false
	var bGroundPounding : logic = false
	var bRevived<public> : logic = false
	var bShieldBroken : logic = false
	var bIsEmoting : logic = false
	var bHasTheFlag : logic = false
	var bIsStunned : logic = false

	var eAggro : FNBGMState = FNBGMState.None

	# for XPs
	var flHealingTracker : float = 0.0
	var KillsTracker : int = 0
	var AssistAgent : ?agent = false
	var DamageTakenTracker : float = 0.0

	var RoundScore : int = 0

    GetAgentClass<public>()<transacts><decides>: FNClasses =
	{
		ClassData.FNClassID
	}

    GetClassFaction<public>()<transacts><decides>: FNFactions =
	{
		ClassData.FNClassFaction
	}

    GetClassName<public>()<transacts><decides>: message =
	{
		ClassData.FNClassName
	}

    GetClassTitle<public>()<transacts><decides>: message =
	{
		ClassData.FNClassTitle
	}

	# main ability
    GetMainAbility<public>()<transacts><decides>:FNAbility=
	{
		ClassData.MainAbility <> FNAbility.None
		ClassData.MainAbility
	}

	# main ability
    GetMainAbility2<public>()<transacts><decides>:FNAbility=
	{
		ClassData.MainAbility2 <> FNAbility.None
		ClassData.MainAbility2
	}

	# passive abilities
    GetPassiveList<public>()<transacts><decides>:[]FNAbility=
	{
		ClassData.PassiveList
	}

    GetClassImage<public>()<transacts><decides>: texture =
	{
		if ( Agent.IsOnTeam[LinkedCreativeDevice,1], ClassData.HasTeamAlt? )
		{
			ClassData.Team2_ClassImage
		}
		else
		{
			ClassData.ClassImage
		}
	}

    IsEmoting<public>()<transacts><decides>: void =
	{
		bIsEmoting?
	}

    HasTheFlag<public>()<transacts><decides>: void =
	{
		bHasTheFlag?
	}

    IsStunned<public>()<transacts><decides>: void =
	{
		bIsStunned?
	}

    IsShieldBroken<public>()<transacts><decides>: void =
	{
		bShieldBroken?
	}

    GetUltimateMeter<public>()<transacts><decides>: float =
	{
		flPlayerUltimateMeter
	}

	<#AddUltimateMeter(Amount:float):void=
	{
		set flPlayerUltimateMeter += Amount

		<#if ( flPlayerUltimateMeter >= 50.0 )
		{
			Print("Ult Ready.")
		}#>
	}#>
}

# Returns Number converted to a string, truncated to Decimals places.
ToTruncatedString<public>(Number:float, Decimals:int):[]char=
{
    var Str:[]char = ToString(Number)
    if:
        DotIndex := Str.Find['.']
        StopIndex := if (Decimals > 0) then Min(DotIndex+Decimals+1,Str.Length) else DotIndex
        SliceStr := Str.Slice[0, StopIndex]
        set Str = SliceStr
    Str
}

Materials := module:
    UI<public> := module{}

ui_hudgameplay := class()
{
    LinkedCreativeDevice<public> : player_classes_manager_device
    LinkedPlayer<public> : player

    var IsInitialized<private> : logic = false
    var MainCanvas<private> : canvas = canvas{}

	UITextMessage<localizes>(Value : string) : message = "{Value}"
	#UIPlayerName<localizes>(Agent:agent)<computes>:message = "{Agent}"
	#UIKillerName<localizes>(InString : string, Agent:agent):message = "{InString}{Agent}"

	UIBG<private>:color_block = color_block{DefaultColor := NamedColors.Black, DefaultOpacity := 0.4}
	UIClassImage<private>: texture_block = texture_block{DefaultImage := Textures.midas_cursed1, DefaultDesiredSize := vector2{X := 125.0, Y := 125.0}}

	#UIUltimateText<private>: text_block = text_block{DefaultTextColor := NamedColors.White, DefaultShadowColor := NamedColors.Black, DefaultShadowOffset := option{vector2{X := 0.0, Y := 0.0}}}
	#UIUltimateMeter<private>:color_block = color_block{DefaultColor := NamedColors.Yellow, DefaultDesiredSize := vector2{X:= 125.0, Y := 10.0} }

    var UIDmgFloat : float = 0.0
	UIDmgText<private>: text_block = text_block{DefaultTextColor := NamedColors.White, DefaultShadowColor := NamedColors.Black, DefaultShadowOffset := option{vector2{X := 2.0, Y := 2.0}}}

	#var AbilityBlock: []material_block = array{}

	UIAbilityBlock:material_block = material_block{DefaultImage := Materials.UI.MI_ProgressBar_Icon{}, DefaultDesiredSize := vector2{X := 125.0, Y := 125.0} }
	UIAbilityCooldownText<private>: text_block = text_block{DefaultTextColor := NamedColors.White, DefaultShadowColor := NamedColors.Black, DefaultShadowOffset := option{vector2{X := 0.0, Y := 0.0}}}

	UIAbilityBlock2:material_block = material_block{DefaultImage := Materials.UI.MI_ProgressBar_Icon{}, DefaultDesiredSize := vector2{X := 125.0, Y := 125.0}}
	UIAbilityCooldownText2<private>: text_block = text_block{DefaultTextColor := NamedColors.White, DefaultShadowColor := NamedColors.Black, DefaultShadowOffset := option{vector2{X := 0.0, Y := 0.0}}}

	UIKillerBlock<private>:material_block = material_block{DefaultImage := Materials.UI.MI_UI_BG_White_Trans{}, DefaultDesiredSize := vector2{X:= 500.0, Y := 50.0} }
	UIKillerClassImage<private>: texture_block = texture_block{DefaultImage := Textures.midas_cursed1, DefaultDesiredSize := vector2{X := 125.0, Y := 125.0}}
	UIKillerTitleText<private>: text_block = text_block{DefaultTextColor := NamedColors.Silver, DefaultShadowColor := NamedColors.Black, DefaultShadowOffset := option{vector2{X := 0.0, Y := 0.0}}, DefaultJustification := text_justification.Left}
	UIKillerNameText<private>: text_block = text_block{DefaultTextColor := NamedColors.White, DefaultShadowColor := NamedColors.Black, DefaultShadowOffset := option{vector2{X := 0.0, Y := 0.0}}, DefaultJustification := text_justification.Center}
	UIKillerLevelText<private>: text_block = text_block{DefaultTextColor := NamedColors.White, DefaultShadowColor := NamedColors.Black, DefaultShadowOffset := option{vector2{X := 2.0, Y := 2.0}}, DefaultJustification := text_justification.Left}

    Init<public>() : void =
	{
        if (IsInitialized?)
		{
            return
		}

		set MainCanvas = CreateUI()
		set IsInitialized = true
    }

    OpenUI<public>() : void =
	{
        if (PlayerUI := GetPlayerUI[LinkedPlayer])
		{
            PlayerUI.AddWidget(GetCanvas(), player_ui_slot{InputMode := ui_input_mode.None})
		}
	}

    UpdateUI<public>() : void =
	{
		if ( Agent := agent[LinkedPlayer] )
		{
			<#if (FortCharacter := Agent.GetFortCharacter[])
			{
				if ( FortCharacter.IsActive[] )
				{
					MainCanvas.SetVisibility(widget_visibility.Visible)
				}
				else
				{
					MainCanvas.SetVisibility(widget_visibility.Hidden)
				}
			}#>

			ResetAbility()
			UpdateKillerInfo()

			if (CurrentClassImage := LinkedCreativeDevice.PlayersMap[Agent].GetClassImage[])
			{
				UIClassImage.SetImage( CurrentClassImage )
			}

			UpdateAbility()

			TeamCollection := LinkedCreativeDevice.GetPlayspace().GetTeamCollection()
			Teams := TeamCollection.GetTeams()
			if ( TeamCollection.IsOnTeam[Agent, Teams[0]] )
			{
				UIBG.SetColor(NamedColors.CornflowerBlue)
			}
			else
			{
				UIBG.SetColor(NamedColors.DarkRed)
			}
		}
	}

    UpdateAbility<public>() : void =
	{
		MatAbilityRaw := UIAbilityBlock.GetImage()
		MatAbilityRaw2 := UIAbilityBlock2.GetImage()
		if ( Agent := agent[LinkedPlayer], MatAbility := Materials.UI.MI_ProgressBar_Icon[MatAbilityRaw] )
		{
			if ( MainAbility := LinkedCreativeDevice.GetMainAbility[Agent], MainAbilityData := LinkedCreativeDevice.GetAbilityData[MainAbility] )
			{
				set MatAbility.IconTexture = MainAbilityData.Icon

				case( LinkedCreativeDevice.GetPlayerClassIndex(Agent) )
				{
					FNClasses.Raz =>
					{
						set MatAbility.IconTexture = Textures.ui_ability_raz_lifedrain
					}
					FNClasses.KadoThorne =>
					{
						set MatAbility.IconTexture = Textures.ui_ability_kado_lifedrain
					}
					FNClasses.Ares =>
					{
						set MatAbility.IconTexture = Textures.ui_ability_damagebuff_ares
					}
						FNClasses.Hades =>
					{
						set MatAbility.IconTexture = Textures.ui_ability_hades_lifedrain
					}
					FNClasses.RingmasterScarr =>
					{
						set MatAbility.IconTexture = Textures.ui_ability_damagebuff_scarr
					}
					FNClasses.DoctorDoom =>
					{
						set MatAbility.IconTexture = Textures.ui_ability_mysticalbomb
					}
					FNClasses.DoctorDoomMythic =>
					{
						set MatAbility.IconTexture = Textures.ui_ability_mysticalbomb
					}
					FNClasses.Kor =>
					{
						set MatAbility.IconTexture = Textures.ui_ability_demongrunt
					}
					_ =>
					{
					}
				}
			}
			else
			{
				set MatAbility.IconTexture = Textures.ui_ability_no
			}

			if ( MatAbility2 := Materials.UI.MI_ProgressBar_Icon[MatAbilityRaw2] )
			{
				if ( MainAbility2 := LinkedCreativeDevice.GetMainAbility2[Agent], MainAbility2Data := LinkedCreativeDevice.GetAbilityData[MainAbility2] )
				{
					set MatAbility2.IconTexture = MainAbility2Data.Icon

					<#case( LinkedCreativeDevice.GetPlayerClassIndex(Agent) )
					{
						FNClasses.Daigo =>
						{
							set MatAbility2.IconTexture = Textures.ui_ability_daigomedal
						}
						FNClasses.Kor =>
						{
							set MatAbility2.IconTexture = Textures.ui_ability_kormedal
						}
						_ =>
						{
						}
					}#>
				}
				else if ( PassiveList := LinkedCreativeDevice.GetPassiveList[Agent], PassiveList.Length <> 0, FirstAbility := PassiveList[0], AbilityData := LinkedCreativeDevice.GetAbilityData[FirstAbility] )
				{
					set MatAbility2.IconTexture = AbilityData.Icon
				}
				else
				{
					set MatAbility2.IconTexture = Textures.ui_ability_no
				}
			}

			TeamCollection := LinkedCreativeDevice.GetPlayspace().GetTeamCollection()
			Teams := TeamCollection.GetTeams()
			if ( TeamCollection.IsOnTeam[Agent, Teams[0]] )
			{
				UIAbilityBlock.SetTint(NamedColors.CornflowerBlue)
				UIAbilityBlock2.SetTint(NamedColors.CornflowerBlue)
			}
			else
			{
				UIAbilityBlock.SetTint(NamedColors.DarkRed)
				UIAbilityBlock2.SetTint(NamedColors.DarkRed)
			}
		}
	}

	<#UpdateAbilityList<public>() : void =
	{
		RowStackBoxes : []stack_box = for ( I := 0..iMaxColumn ) {stack_box{Orientation := orientation.Horizontal}}
		var BtnArray : []button_fnclass = array{}

		# create button for each classes
		AllAbilityList := LinkedCreativeDevice.GetAbilityList(Agent)
		for (NewAbility : AllAbilityList, ClassImage := NewAbility.ClassImage, ClassID := NewAbility.FNClassID)
		{
			NewBtn :=	button_quiet{}
			NewMat := Materials.UI.M_UI_ClassPortrait{IconTexture:=ClassImage}
			NewImage :=	material_block{DefaultImage := NewMat, DefaultDesiredSize := vector2{X := 85.0, Y := 125.0}}
			NewOverlay :=		overlay
								{
									Slots := array
									{
										overlay_slot
										{
											VerticalAlignment := vertical_alignment.Center
											Widget := NewImage
										}
										<#VerticalAlignment := vertical_alignment.Bottom
										HorizontalAlignment := horizontal_alignment.Right
										Padding := margin{Top := 0.0, Left := 0.0, Right := 0.0, Bottom := 0.0}
										Widget := text_block
										{
											DefaultText := UITextMessage("0")
											DefaultTextColor := NamedColors.White
											DefaultTextOpacity := 0.9
											DefaultShadowColor := NamedColors.Black
											DefaultShadowOffset := option{vector2{X := 0.0, Y := 0.0}}
											DefaultJustification := text_justification.Right
										}#>
										overlay_slot
										{
											VerticalAlignment := vertical_alignment.Fill
											HorizontalAlignment := horizontal_alignment.Fill
											Padding := margin{Top := 0.0, Left := 0.0, Right := 0.0, Bottom := 0.0}
											Widget := NewBtn
										}
									}
								}

			if (StackBox := RowStackBoxes[StackColumn])
			{
				StackBox.AddWidget(
								stack_box_slot
								{
									VerticalAlignment := vertical_alignment.Top
									HorizontalAlignment := horizontal_alignment.Left
									Padding := margin{Top := 0.0, Left := 0.0, Right := 1.0, Bottom := 0.0}
									Widget := NewOverlay
								}
				)
			}

			NewBtn.OnClick().SubscribeWidget(OnClassSelectedBtn, ClassID)
			NewFNBtn := button_fnclass{ ButtonQuiet := NewBtn, ButtonMat := NewImage, ClassData := NewAbility}

			set BtnArray += array{NewFNBtn}
			if { set ClassBtnMap[StackColumn] = BtnArray }
		}
	}#>

	UpdateAbilityCooldown<public>(AbilityID:FNAbility): void =
	{
		if (Agent := agent[LinkedPlayer])
		{
			if ( MainAbility := LinkedCreativeDevice.GetMainAbility[Agent], MainAbility = AbilityID )
			{
				MeterAmount := LinkedCreativeDevice.GetAbilityCooldown(Agent, MainAbility)
				MaxMeterAmount := LinkedCreativeDevice.GetAbilityMaxCooldown(Agent, MainAbility)
				if ( MeterAmount <= 0.0 )
				{
					UIAbilityCooldownText.SetText(UITextMessage(""))
				}
				else
				{
					UIAbilityCooldownText.SetText(UITextMessage("{ToTruncatedString(MeterAmount, 0)}"))
				}

				MatAbilityRaw := UIAbilityBlock.GetImage()
				if ( MatAbility := Materials.UI.MI_ProgressBar_Icon[MatAbilityRaw] )
				{
					set MatAbility.Progress = (MeterAmount / MaxMeterAmount) * 1.0
				}
			}

			if ( MainAbility2 := LinkedCreativeDevice.GetMainAbility2[Agent], MainAbility2 = AbilityID )
			{
				MeterAmount2 := LinkedCreativeDevice.GetAbilityCooldown(Agent, MainAbility2)
				MaxMeterAmount2 := LinkedCreativeDevice.GetAbilityMaxCooldown(Agent, MainAbility2)
				if ( MeterAmount2 <= 0.0 )
				{
					UIAbilityCooldownText2.SetText(UITextMessage(""))
				}
				else
				{
					UIAbilityCooldownText2.SetText(UITextMessage("{ToTruncatedString(MeterAmount2, 0)}"))
				}

				MatAbilityRaw2 := UIAbilityBlock2.GetImage()
				if ( MatAbility2 := Materials.UI.MI_ProgressBar_Icon[MatAbilityRaw2] )
				{
					set MatAbility2.Progress = (MeterAmount2 / MaxMeterAmount2) * 1.0
				}
			}
		}
	}

	DisableAbility<public>(AbilityID:FNAbility): void =
	{
		if  (Agent := agent[LinkedPlayer])
		{
			if ( MainAbility := LinkedCreativeDevice.GetMainAbility[Agent], MainAbility = AbilityID )
			{
				MaxMeterAmount := LinkedCreativeDevice.GetAbilityMaxCooldown(Agent, MainAbility)
				UIAbilityCooldownText.SetText(UITextMessage(""))

				MatAbilityRaw := UIAbilityBlock.GetImage()
				if ( MatAbility := Materials.UI.MI_ProgressBar_Icon[MatAbilityRaw] )
				{
					set MatAbility.Progress = (MaxMeterAmount / MaxMeterAmount) * 1.0
				}
			}

			if ( MainAbility2 := LinkedCreativeDevice.GetMainAbility2[Agent], MainAbility2 = AbilityID )
			{
				MaxMeterAmount := LinkedCreativeDevice.GetAbilityMaxCooldown(Agent, MainAbility2)
				UIAbilityCooldownText2.SetText(UITextMessage(""))

				MatAbilityRaw2 := UIAbilityBlock2.GetImage()
				if ( MatAbility2 := Materials.UI.MI_ProgressBar_Icon[MatAbilityRaw2] )
				{
					set MatAbility2.Progress = (MaxMeterAmount / MaxMeterAmount) * 1.0
				}
			}
		}
	}

	ResetAbility<public>(): void =
	{
		if  (Agent := agent[LinkedPlayer])
		{
			UIAbilityCooldownText.SetText(UITextMessage(""))

			MatAbilityRaw := UIAbilityBlock.GetImage()
			if ( MatAbility := Materials.UI.MI_ProgressBar_Icon[MatAbilityRaw] )
			{
				set MatAbility.Progress = 0.0
			}

			UIAbilityCooldownText2.SetText(UITextMessage(""))

			MatAbilityRaw2 := UIAbilityBlock2.GetImage()
			if ( MatAbility2 := Materials.UI.MI_ProgressBar_Icon[MatAbilityRaw2] )
			{
				set MatAbility2.Progress = 0.0
			}
		}
	}

	UpdateKillerInfo<public>(): void =
	{
		if ( Agent := agent[LinkedPlayer] )
		{
			if ( FortCharacter := Agent.GetFortCharacter[], UIKillerInfo := UIKillerBlock.GetParentWidget[] )
			{
				UIKillerInfo.SetVisibility(widget_visibility.Hidden)
			}
		}
	}

	ShowKillerInfo<public>(KillerAgent:agent)<suspends>: void =
	{
		Sleep(0.2)
		if ( Agent := agent[LinkedPlayer] )
		{
			if ( UIKillerInfo := UIKillerBlock.GetParentWidget[] )
			{
				#Print("ShowKillerInfo()")

				var BossTitleText : message = TextLocalization.DefeatedBy
				var BossIcon : texture = Textures.midas_cursed1

				if (CurrentClassImage := LinkedCreativeDevice.PlayersMap[KillerAgent].GetClassImage[])
				{
					set BossIcon = CurrentClassImage
				}

				if (CurrentClassTitle := LinkedCreativeDevice.PlayersMap[KillerAgent].GetClassTitle[])
				{
					set BossTitleText = CurrentClassTitle
				}

				UIKillerClassImage.SetImage( BossIcon )

				UIKillerNameText.SetText(TextLocalization.UIPlayerName(KillerAgent))
				KillerString := UIKillerNameText.GetText()
				Print("Killer Name: {KillerString}")

				UIKillerInfo.SetVisibility(widget_visibility.Visible)
				if ( KillerAgent = Agent )
				{
					var DeathText : string = "You Died"
					case( GetRandomInt(0,2) )
					{
						0 =>	set DeathText = "You Died"
						1 =>	set DeathText = "Ouch"
						2 =>	set DeathText = ":("
						_ =>	set DeathText = "You Died"
					}

					UIKillerTitleText.SetText(UITextMessage(""))
					UIKillerNameText.SetText(UITextMessage(DeathText))
				}
				else
				{
					if ( KillerString = "ISPuddy" )
					{
						UIKillerTitleText.SetText(UITextMessage("Creator of [BE MYTHIC BOSS]"))
					}
					else
					{
						UIKillerTitleText.SetText(BossTitleText)
					}
				}

				TeamCollection := LinkedCreativeDevice.GetPlayspace().GetTeamCollection()
				Teams := TeamCollection.GetTeams()
				if ( TeamCollection.IsOnTeam[KillerAgent, Teams[0]] )
				{
					UIKillerBlock.SetTint(NamedColors.CornflowerBlue)
				}
				else
				{
					UIKillerBlock.SetTint(NamedColors.DarkRed)
				}

				if ( SaveManager := LinkedCreativeDevice.SaveManagerDevice? )
				{
					PlayerLevel := SaveManager.GetPlayerLevel(Agent)
					UIKillerLevelText.SetText(UITextMessage("Lv. {PlayerLevel}"))
				}

				Sleep(3.8)

				UIKillerInfo.SetVisibility(widget_visibility.Hidden)
			}
		}
	}

	UpdateDamageText<public>(DmgAmount:float)<suspends>: void =
	{
		set UIDmgFloat += DmgAmount
		UIDmgText.SetText(UITextMessage("{ToTruncatedString(UIDmgFloat, 0)}"))
		Sleep(1.0)
		set UIDmgFloat = 0.0
		UIDmgText.SetText(UITextMessage(""))
	}

    CloseUI<public>()<suspends> : void =
	{
        if (PlayerUI := GetPlayerUI[LinkedPlayer])
		{
			PlayerUI.RemoveWidget(GetCanvas())
		}
    }

    GetCanvas<public>() : canvas = MainCanvas

	CreateUI<private>() : canvas =
	{
		UICanvas : canvas = canvas{Slots := array{
		canvas_slot
		{
			ZOrder := {0}
			Anchors := anchors{Minimum := vector2{X := 0.0, Y := 0.95}, Maximum := vector2{X := 0.0, Y := 0.95}}
			Offsets := margin{ Top:=0.0, Left:= 0.0, Right:=0.0, Bottom := 0.0 }
			Alignment := vector2{X := 0.0, Y := 0.95}
			SizeToContent := true
			Widget := stack_box
			{
				Orientation := orientation.Horizontal
				Slots := array
				{
					stack_box_slot
					{
						Widget := overlay
						{
							Slots := array
							{
								<#overlay_slot
								{
									VerticalAlignment := vertical_alignment.Fill
									HorizontalAlignment := horizontal_alignment.Fill
									Widget := UIBG
								}#>
								overlay_slot
								{
									VerticalAlignment := vertical_alignment.Center
									Widget := UIClassImage
								}
								<#overlay_slot
								{
									VerticalAlignment := vertical_alignment.Bottom
									HorizontalAlignment := horizontal_alignment.Left
									#Padding := margin{Top := 0.0, Left := 0.0, Right := 0.0, Bottom := 50.0}
									Widget := UIUltimateMeter
								}
								overlay_slot
								{
									VerticalAlignment := vertical_alignment.Bottom
									HorizontalAlignment := horizontal_alignment.Fill
									Widget := UIUltimateText
								}#>
							}
						}
					}
				}
			}
        }
		canvas_slot
		{
			ZOrder := {0}
			Anchors := anchors{Minimum := vector2{X := 0.75, Y := 0.85}, Maximum := vector2{X := 0.75, Y := 0.85}}
			Offsets := margin{Top := 0.0, Left := 0.0, Right := 0.0, Bottom := 0.0}
			Alignment := vector2{X := 0.0, Y := 0.0}
			SizeToContent := true
			Widget := stack_box
			{
				Orientation := orientation.Horizontal
				Slots := array
				{
					stack_box_slot
					{
						Padding := margin{Top := 0.0, Left := 0.0, Right := 0.0, Bottom := 0.0}
						Widget := overlay
						{
							Slots := array
							{
								overlay_slot
								{
									VerticalAlignment := vertical_alignment.Fill
									HorizontalAlignment := horizontal_alignment.Fill
									Widget := UIAbilityBlock
								}
								overlay_slot
								{
									VerticalAlignment := vertical_alignment.Bottom
									HorizontalAlignment := horizontal_alignment.Fill
									Widget := UIAbilityCooldownText
								}
							}
						}
					}
					stack_box_slot
					{
						Padding := margin{Top := 0.0, Left := 5.0, Right := 0.0, Bottom := 0.0}
						Widget := overlay
						{
							Slots := array
							{
								overlay_slot
								{
									VerticalAlignment := vertical_alignment.Fill
									HorizontalAlignment := horizontal_alignment.Fill
									Widget := UIAbilityBlock2
								}
								overlay_slot
								{
									VerticalAlignment := vertical_alignment.Bottom
									HorizontalAlignment := horizontal_alignment.Fill
									Widget := UIAbilityCooldownText2
								}
							}
						}
					}
				}
			}
        }
		<#canvas_slot
		{
			ZOrder := {0}
			Anchors := anchors{Minimum := vector2{X := 0.5, Y := 1.0}, Maximum := vector2{X := 0.5, Y := 1.0}}
			Offsets := margin{Top := 0.0, Left := 0.0, Right := 0.0, Bottom := 0.0}
			Alignment := vector2{X := 0.0, Y := 1.0}
			SizeToContent := true
			Widget := overlay
			{
				Slots := array
				{
					overlay_slot
					{
						VerticalAlignment := vertical_alignment.Bottom
						HorizontalAlignment := horizontal_alignment.Fill
						Widget := UIXPText
					}
				}
			}
        }#>
		canvas_slot
		{
			ZOrder := {0}
			Anchors := anchors{Minimum := vector2{X := 0.5, Y := 0.7}, Maximum := vector2{X := 0.5, Y := 0.7}}
			Offsets := margin{Top := 0.0, Left := 0.0, Right := 0.0, Bottom := 0.0}
			Alignment := vector2{X := 0.5, Y := 0.0}
			SizeToContent := true
			Widget := overlay
			{
				Slots := array
				{
					overlay_slot
					{
						VerticalAlignment := vertical_alignment.Fill
						HorizontalAlignment := horizontal_alignment.Fill
						Widget := UIKillerBlock
					}
					overlay_slot
					{
						VerticalAlignment := vertical_alignment.Center
						HorizontalAlignment := horizontal_alignment.Left
						Widget := color_block
						{
							DefaultColor := NamedColors.White
							DefaultDesiredSize := vector2{X := 125.0, Y := 125.0}
							DefaultOpacity := 0.2
						}
					}
					overlay_slot
					{
						VerticalAlignment := vertical_alignment.Center
						HorizontalAlignment := horizontal_alignment.Left
						Widget := UIKillerClassImage
					}
					overlay_slot
					{
						VerticalAlignment := vertical_alignment.Bottom
						HorizontalAlignment := horizontal_alignment.Left
						Padding := margin{Top := 0.0, Left := 0.0, Right := 0.0, Bottom := 0.0}
						Widget := UIKillerLevelText
					}
					overlay_slot
					{
						VerticalAlignment := vertical_alignment.Fill
						HorizontalAlignment := horizontal_alignment.Fill
						Padding := margin{Top := 0.0, Left := 135.0, Right := 0.0, Bottom := 0.0}
						Widget := stack_box
						{
							Orientation := orientation.Vertical
							Slots := array
							{
								stack_box_slot
								{
									VerticalAlignment := vertical_alignment.Fill
									HorizontalAlignment := horizontal_alignment.Fill
									Padding := margin{Top := 5.0, Left := 0.0, Right := 0.0, Bottom := 0.0}
									Widget := UIKillerTitleText
								}
								stack_box_slot
								{
									VerticalAlignment := vertical_alignment.Fill
									HorizontalAlignment := horizontal_alignment.Fill
									Padding := margin{Top := 0.0, Left := 0.0, Right := 0.0, Bottom := 5.0}
									Widget := UIKillerNameText
								}
							}
						}
					}
				}
			}
        }
		canvas_slot
		{
			ZOrder := {1}
			Anchors := anchors{Minimum := vector2{X := 0.5, Y := 0.5}, Maximum := vector2{X := 0.5, Y := 0.5} }
			Offsets := margin{Top := 0.0, Left := 0.0, Right := 0.0, Bottom := 0.0}
			Alignment := vector2{X := 0.5, Y := 0.5}
			SizeToContent := true
			Widget := UIDmgText
        }
		}}

		NewMat : Materials.UI.MI_ProgressBar_Icon = Materials.UI.MI_ProgressBar_Icon{}
		UIAbilityBlock.SetImage(NewMat)

		NewMat2 : Materials.UI.MI_ProgressBar_Icon = Materials.UI.MI_ProgressBar_Icon{}
		UIAbilityBlock2.SetImage(NewMat2)

		#UIUltimateText.SetText(UITextMessage(""))
		#UIUltimateText.SetShadowOpacity(1.0)
		UIDmgText.SetText(UITextMessage(""))
		UIKillerTitleText.SetText(UITextMessage(""))
		UIKillerTitleText.SetShadowOpacity(1.0)
		UIKillerNameText.SetText(UITextMessage(""))
		UIKillerNameText.SetShadowOpacity(1.0)
		UIKillerLevelText.SetText(UITextMessage(""))
		UIKillerLevelText.SetShadowOpacity(1.0)

		return UICanvas
	}
}

(Listenable : listenable(agent)).SubscribeAgent(OutputFunc : tuple(agent, t)->void, ExtraData : t where t:type) : cancelable =
{
    Wrapper := wrapper_agent(t){ExtraData := ExtraData, OutputFunc := OutputFunc}
    Listenable.Subscribe(Wrapper.InputFunc)
}
 
wrapper_agent(t : type) := class()
{
    ExtraData : t;
    OutputFunc : tuple(agent, t) -> void
    InputFunc(Agent : agent):void = OutputFunc(Agent, ExtraData)
}

#CategoryClasses<localizes>:message="Classes"

# A Verse-authored creative device that can be placed in a level
player_classes_manager_device := class(creative_device):

	@editable PlayerAbilityDevice : ?player_ability_device = false
	@editable PlayerUltDevice : ?player_ability_ultimate_device = false
	#@editable FactionWarDevice : ?faction_war_device = false
	@editable TeamDeathmatchDevice : ?fn_game_tdm = false
	@editable SaveManagerDevice : ?save_manager_device = false
	@editable UIMainMenuShowDevice : ?ui_mainmenu_device = false
	@editable UIAchievementDevice : ?ui_achievements_device = false
	#@editable UIClassMenuManagerDevice : ?ui_classmenu_manager_device = false

	@editable{Categories:=array{CategorySFX}} SFXAbilityCooldown	: audio_player_device = audio_player_device{}

	@editable InputUlt				: input_trigger_device = input_trigger_device{}
	@editable StatUlt				: stat_creator_device = stat_creator_device{}

	var PlayersMap: [agent]fn_player_class = map{}

	@editable TriggerClassMenuOpen : trigger_device = trigger_device{}
	#@editable ClassMenuOpener : hiding_prop_device = hiding_prop_device{}

	@editable ClassSelectorNONE		: class_and_team_selector_device = class_and_team_selector_device{}

	@editable ClassMidas			: fn_class_data = fn_class_data{
		FNClassID := FNClasses.Midas,
		FNClassFaction := FNFactions.Spies,
		FNClassName := TextLocalization.Class_Midas,
		FNClassTitle := TextLocalization.ClassTitle_Midas,
		ClassImage := Textures.midas_cursed1,
		MainAbility := FNAbility.DoubleJump,
		MainAbility2 := FNAbility.None
	}

	@editable ClassTNTina			: fn_class_data = fn_class_data{
		FNClassID := FNClasses.TNTina,
		FNClassFaction := FNFactions.Spies,
		FNClassName := TextLocalization.Class_TNTina,
		Team2_FNClassName := TextLocalization.Class_DynamoTNTina,
		FNClassTitle := TextLocalization.ClassTitle_TNTina,
		ClassImage := Textures.tntina_cursed,
		MainAbility := FNAbility.DoubleJump,
		MainAbility2 := FNAbility.PartingGift,
		Team2_ClassImage := Textures.class_dynamotntina
	}

	@editable ClassBrutus			: fn_class_data = fn_class_data{
		FNClassID := FNClasses.Brutus,
		FNClassFaction := FNFactions.Spies,
		FNClassName := TextLocalization.Class_Brutus,
		FNClassTitle := TextLocalization.ClassTitle_Brutus,
		ClassImage := Textures.brutus_cursed,
		MainAbility := FNAbility.DoubleJump,
		MainAbility2 := FNAbility.None
	}

	@editable ClassSkye				: fn_class_data = fn_class_data{
		FNClassID := FNClasses.Skye,
		FNClassFaction := FNFactions.Spies,
		FNClassName := TextLocalization.Class_Midas,
		FNClassTitle := TextLocalization.ClassTitle_Midas,
		ClassImage := Textures.skye_cursed,
		MainAbility := FNAbility.DoubleJump,
		MainAbility2 := FNAbility.None
	}

	@editable ClassMeowscles		: fn_class_data = fn_class_data{
		FNClassID := FNClasses.Meowscles,
		FNClassFaction := FNFactions.Spies,
		FNClassName := TextLocalization.Class_Meowscles,
		Team2_FNClassName := TextLocalization.Class_Meowdas,
		FNClassTitle := TextLocalization.ClassTitle_Meowscles,
		ClassImage := Textures.class_meowscles,
		MainAbility := FNAbility.DoubleJump,
		MainAbility2 := FNAbility.None,
		Team2_ClassImage := Textures.class_meowdas
		}

	@editable ClassKit				: fn_class_data = fn_class_data{
		FNClassID := FNClasses.Kit,
		FNClassFaction := FNFactions.Spies,
		FNClassName := TextLocalization.Class_Kit,
		FNClassTitle := TextLocalization.ClassTitle_Kit,
		ClassImage := Textures.class_kit,
		MainAbility := FNAbility.AirRaid,
		MainAbility2 := FNAbility.TherapeuticSwim,
		Team2_ClassImage := Textures.class_kit
	}

	@editable ClassJules			: fn_class_data = fn_class_data{
		FNClassID := FNClasses.Jules,
		FNClassFaction := FNFactions.Spies,
		FNClassName := TextLocalization.Class_Jules,
		FNClassTitle := TextLocalization.ClassTitle_Jules,
		ClassImage := Textures.jules_cursed,
		MainAbility := FNAbility.None,
		MainAbility2 := FNAbility.TherapeuticSwim
	}

	@editable ClassOcean			: fn_class_data = fn_class_data{
		FNClassID := FNClasses.Ocean,
		FNClassFaction := FNFactions.Spies,
		FNClassName := TextLocalization.Class_Ocean,
		FNClassTitle := TextLocalization.ClassTitle_Ocean,
		ClassImage := Textures.class_ocean,
		MainAbility := FNAbility.None,
		MainAbility2 := FNAbility.TherapeuticSwim
	}
	
	@editable ClassShadowMidas		: fn_class_data = fn_class_data{
		FNClassID := FNClasses.ShadowMidas,
		FNClassFaction := FNFactions.Spies,
		FNClassName := TextLocalization.Class_ShadowMidas,
		FNClassTitle := TextLocalization.ClassTitle_ShadowMidas,
		ClassImage := Textures.class_shadowmidas,
		MainAbility := FNAbility.Shadow,
		MainAbility2 := FNAbility.None
	}

	@editable ClassShadow			: fn_class_data = fn_class_data{
		FNClassID := FNClasses.Shadow,
		FNClassFaction := FNFactions.Spies,
		FNClassName := TextLocalization.Class_Shadow,
		FNClassTitle := TextLocalization.ClassTitle_Shadow,
		ClassImage := Textures.class_shadow,
		MainAbility := FNAbility.Warp,
		MainAbility2 := FNAbility.None
	}

	@editable ClassSpireAssassin	: fn_class_data = fn_class_data{
		FNClassID := FNClasses.SpireAssassin,
		FNClassFaction := FNFactions.IO,
		FNClassName := TextLocalization.Class_SpireAssassin,
		FNClassTitle := TextLocalization.ClassTitle_SpireAssassin,
		ClassImage := Textures.class_spireassassin,
		MainAbility := FNAbility.Warp,
		MainAbility2 := FNAbility.None
	}

	@editable ClassRaz				: fn_class_data = fn_class_data{
		FNClassID := FNClasses.Raz,
		FNClassFaction := FNFactions.LastReality,
		FNClassName := TextLocalization.Class_Raz,
		FNClassTitle := TextLocalization.ClassTitle_Raz,
		ClassImage := Textures.class_glyphmasterraz,
		MainAbility := FNAbility.LifeDrain,
		MainAbility2 := FNAbility.Warp
	}

	@editable ClassSlone			: fn_class_data = fn_class_data{
		FNClassID := FNClasses.Slone,
		FNClassFaction := FNFactions.IO,
		FNClassName := TextLocalization.Class_Slone,
		FNClassTitle := TextLocalization.ClassTitle_Slone,
		ClassImage := Textures.class_doctorslone,
		MainAbility := FNAbility.MirrorImage,
		MainAbility2 := FNAbility.None
	}

	@editable ClassZygAndChoppy		: fn_class_data = fn_class_data{
		FNClassID := FNClasses.ZygAndChoppy,
		FNClassFaction := FNFactions.LastReality,
		FNClassName := TextLocalization.Class_ZygAndChoppy,
		FNClassTitle := TextLocalization.ClassTitle_ZygAndChoppy,
		ClassImage := Textures.class_zygandchoppy,
		MainAbility := FNAbility.AlienParasite,
		MainAbility2 := FNAbility.None
	}

	@editable ClassCubeBlue			: fn_class_data = fn_class_data{
		FNClassID := FNClasses.CubeBlue,
		FNClassFaction := FNFactions.None,
		FNClassName := TextLocalization.Class_CubeBlue,
		FNClassTitle := TextLocalization.ClassTitle_CubeBlue,
		ClassImage := Textures.class_cubeblue,
		MainAbility := FNAbility.HealingRune,
		MainAbility2 := FNAbility.None,
		MaxHealth := 100.0,
		MaxShield := 400.0
	}

	@editable ClassCubeAssassin		: fn_class_data = fn_class_data{
		FNClassID := FNClasses.CubeAssassin,
		FNClassFaction := FNFactions.LastReality,
		FNClassName := TextLocalization.Class_CubeAssassin,
		FNClassTitle := TextLocalization.ClassTitle_CubeAssassin,
		ClassImage := Textures.class_cubeassassin,
		MainAbility := FNAbility.PersonalShield,
		MainAbility2 := FNAbility.Warp
	}

	@editable ClassCubeQueen		: fn_class_data = fn_class_data{
		FNClassID := FNClasses.CubeQueen,
		FNClassFaction := FNFactions.LastReality,
		FNClassName := TextLocalization.Class_CubeQueen,
		FNClassTitle := TextLocalization.ClassTitle_CubeQueen,
		ClassImage := Textures.class_cubequeen,
		MainAbility := FNAbility.Laserbeam,
		MainAbility2 := FNAbility.None
	}

	@editable ClassSgtWinter		: fn_class_data = fn_class_data{
		FNClassID := FNClasses.SgtWinter,
		FNClassFaction := FNFactions.None,
		FNClassName := TextLocalization.Class_SgtWinter,
		FNClassTitle := TextLocalization.ClassTitle_SgtWinter,
		ClassImage := Textures.class_sgtwinter,
		MainAbility := FNAbility.None,
		MainAbility2 := FNAbility.None
	}

	@editable ClassTheFoundation	: fn_class_data = fn_class_data{
		FNClassID := FNClasses.TheFoundation,
		FNClassFaction := FNFactions.Seven,
		FNClassName := TextLocalization.Class_Foundation,
		FNClassTitle := TextLocalization.ClassTitle_Foundation,
		ClassImage := Textures.class_thefoundation,
		MainAbility := FNAbility.Rock,
		MainAbility2 := FNAbility.DoubleJump
	}

	@editable ClassGunnar			: fn_class_data = fn_class_data{
		FNClassID := FNClasses.Gunnar,
		FNClassFaction := FNFactions.IO,
		FNClassName := TextLocalization.Class_Gunnar,
		FNClassTitle := TextLocalization.ClassTitle_Gunnar,
		ClassImage := Textures.class_gunnar,
		MainAbility := FNAbility.Charge,
		MainAbility2 := FNAbility.None
	}

	@editable ClassHuntmasterSaber	: fn_class_data = fn_class_data{
		FNClassID := FNClasses.HuntmasterSaber,
		FNClassFaction := FNFactions.IO,
		FNClassName := TextLocalization.Class_HuntmasterSaber,
		FNClassTitle := TextLocalization.ClassTitle_HuntmasterSaber,
		ClassImage := Textures.class_huntmastersaber,
		MainAbility := FNAbility.SpinSlash,
		MainAbility2 := FNAbility.Warp
	}

	@editable ClassTheHerald		: fn_class_data = fn_class_data{
		FNClassID := FNClasses.TheHerald,
		FNClassFaction := FNFactions.LastReality,
		FNClassName := TextLocalization.Class_Herald,
		FNClassTitle := TextLocalization.ClassTitle_Herald,
		ClassImage := Textures.class_theherald,
		MainAbility := FNAbility.Lockdown,
		MainAbility2 := FNAbility.Warp
	}

	@editable ClassTheAgeless		: fn_class_data = fn_class_data{
		FNClassID := FNClasses.TheAgelessChampion,
		FNClassFaction := FNFactions.IO,
		FNClassName := TextLocalization.Class_Ageless,
		FNClassTitle := TextLocalization.ClassTitle_Ageless,
		ClassImage := Textures.class_theageless,
		MainAbility := FNAbility.Laserbeam,
		MainAbility2 := FNAbility.None
	}

	@editable ClassWildguardRelik	: fn_class_data = fn_class_data{
		FNClassID := FNClasses.WildguardRelik,
		FNClassFaction := FNFactions.None,
		FNClassName := TextLocalization.Class_Relik,
		FNClassTitle := TextLocalization.ClassTitle_Relik,
		ClassImage := Textures.class_relik,
		MainAbility := FNAbility.Warp,
		MainAbility2 := FNAbility.None
	}

	@editable ClassKadoThorne		: fn_class_data = fn_class_data{
		FNClassID := FNClasses.KadoThorne,
		FNClassFaction := FNFactions.None,
		FNClassName := TextLocalization.Class_KadoThorne,
		FNClassTitle := TextLocalization.ClassTitle_KadoThorne,
		ClassImage := Textures.class_kadothorne,
		MainAbility := FNAbility.LifeDrain,
		MainAbility2 := FNAbility.DoubleJump
	}

	@editable ClassTheDealer		: fn_class_data = fn_class_data{
		FNClassID := FNClasses.TheDealer,
		FNClassFaction := FNFactions.None,
		FNClassName := TextLocalization.Class_Dealer,
		FNClassTitle := TextLocalization.ClassTitle_Dealer,
		ClassImage := Textures.class_thedealer,
		MainAbility := FNAbility.None,
		MainAbility2 := FNAbility.None,
		PassiveList := array{FNAbility.LastResort}
	}

	@editable ClassOscar			: fn_class_data = fn_class_data{
		FNClassID := FNClasses.Oscar,
		FNClassFaction := FNFactions.Society,
		FNClassName := TextLocalization.Class_Oscar,
		FNClassTitle := TextLocalization.ClassTitle_Oscar,
		ClassImage := Textures.class_oscar,
		MainAbility := FNAbility.Phase2,
		MainAbility2 := FNAbility.ShieldRegen
	}

	@editable ClassNisha			: fn_class_data = fn_class_data{
		FNClassID := FNClasses.Nisha,
		FNClassFaction := FNFactions.Society,
		FNClassName := TextLocalization.Class_Nisha,
		FNClassTitle := TextLocalization.ClassTitle_Nisha,
		ClassImage := Textures.class_nisha,
		MainAbility := FNAbility.Phase2,
		MainAbility2 := FNAbility.ShieldRegen
	}

	@editable ClassMontague			: fn_class_data = fn_class_data{
		FNClassID := FNClasses.Montague,
		FNClassFaction := FNFactions.Society,
		FNClassName := TextLocalization.Class_Montague,
		FNClassTitle := TextLocalization.ClassTitle_Montague,
		ClassImage := Textures.class_montague,
		MainAbility := FNAbility.Phase2,
		MainAbility2 := FNAbility.ShieldRegen
	}

	@editable ClassValeria			: fn_class_data = fn_class_data{
		FNClassID := FNClasses.Valeria,
		FNClassFaction := FNFactions.Society,
		FNClassName := TextLocalization.Class_Valeria,
		FNClassTitle := TextLocalization.ClassTitle_Valeria,
		ClassImage := Textures.class_valeria,
		MainAbility := FNAbility.Phase2,
		MainAbility2 := FNAbility.ShieldRegen
	}

	@editable ClassZeus				: fn_class_data = fn_class_data{
		FNClassID := FNClasses.Zeus,
		FNClassFaction := FNFactions.Pantheon,
		FNClassName := TextLocalization.Class_Zeus,
		FNClassTitle := TextLocalization.ClassTitle_Zeus,
		ClassImage := Textures.class_zeus,
		MainAbility := FNAbility.Rebirth,
		MainAbility2 := FNAbility.DoubleJump
	}

	@editable ClassAres				: fn_class_data = fn_class_data{
		FNClassID := FNClasses.Ares,
		FNClassFaction := FNFactions.Pantheon,
		FNClassName := TextLocalization.Class_Ares,
		FNClassTitle := TextLocalization.ClassTitle_Ares,
		ClassImage := Textures.class_ares,
		MainAbility := FNAbility.DamageBuff,
		MainAbility2 := FNAbility.None
	}

	@editable ClassCerberus			: fn_class_data = fn_class_data{
		FNClassID := FNClasses.Cerberus,
		FNClassFaction := FNFactions.Pantheon,
		FNClassName := TextLocalization.Class_Cerberus,
		FNClassTitle := TextLocalization.ClassTitle_Cerberus,
		ClassImage := Textures.class_cerberus,
		MainAbility := FNAbility.SpinSlash,
		MainAbility2 := FNAbility.Warp
	}

	@editable ClassHades			: fn_class_data = fn_class_data{
		FNClassID := FNClasses.Hades,
		FNClassFaction := FNFactions.Pantheon,
		FNClassName := TextLocalization.Class_Hades,
		FNClassTitle := TextLocalization.ClassTitle_Hades,
		ClassImage := Textures.class_hades,
		MainAbility := FNAbility.LifeDrain,
		MainAbility2 := FNAbility.None
	}

	@editable ClassMegaloDon		: fn_class_data = fn_class_data{
		FNClassID := FNClasses.MegaloDon,
		FNClassFaction := FNFactions.Wasteland,
		FNClassName := TextLocalization.Class_MegaloDon,
		FNClassTitle := TextLocalization.ClassTitle_MegaloDon,
		ClassImage := Textures.class_megalodon,
		MainAbility := FNAbility.Charge,
		MainAbility2 := FNAbility.None
	}

	@editable ClassRingmasterScarr	: fn_class_data = fn_class_data{
		FNClassID := FNClasses.RingmasterScarr,
		FNClassFaction := FNFactions.Wasteland,
		FNClassName := TextLocalization.Class_RingmasterScarr,
		FNClassTitle := TextLocalization.ClassTitle_RingmasterScarr,
		ClassImage := Textures.class_ringmasterscarr,
		MainAbility := FNAbility.DamageBuff,
		MainAbility2 := FNAbility.None
	}

	@editable ClassTheMachinist		: fn_class_data = fn_class_data{
		FNClassID := FNClasses.TheMachinist,
		FNClassFaction := FNFactions.Wasteland,
		FNClassName := TextLocalization.Class_Machinist,
		FNClassTitle := TextLocalization.ClassTitle_Machinist,
		ClassImage := Textures.class_themachinist,
		MainAbility := FNAbility.Engineer,
		MainAbility2 := FNAbility.ShieldRegen
	}

	@editable ClassUltimaCarver		: fn_class_data = fn_class_data{
		FNClassID := FNClasses.UltimaCarver,
		FNClassFaction := FNFactions.None,
		FNClassName := TextLocalization.Class_UltimaCarver,
		FNClassTitle := TextLocalization.ClassTitle_UltimaCarver,
		ClassImage := Textures.class_ultimacarver,
		MainAbility := FNAbility.None,
		MainAbility2 := FNAbility.None
	}

	@editable ClassShogunX			: fn_class_data = fn_class_data{
		FNClassID := FNClasses.ShogunX,
		FNClassFaction := FNFactions.None,
		FNClassName := TextLocalization.Class_ShogunX,
		FNClassTitle := TextLocalization.ClassTitle_ShogunX,
		ClassImage := Textures.class_shogunx,
		MainAbility := FNAbility.Bombardment,
		MainAbility2 := FNAbility.None
	}

	@editable ClassNightRose		: fn_class_data = fn_class_data{
		FNClassID := FNClasses.NightRose,
		FNClassFaction := FNFactions.None,
		FNClassName := TextLocalization.Class_NightRose,
		FNClassTitle := TextLocalization.ClassTitle_NightRose,
		ClassImage := Textures.class_nightrose,
		MainAbility := FNAbility.DemonClaw,
		MainAbility2 := FNAbility.Puppeteer
	}

	@editable ClassPuppeteer		: fn_class_data = fn_class_data{
		FNClassID := FNClasses.Puppeteer,
		FNClassFaction := FNFactions.None,
		FNClassName := TextLocalization.Class_Puppeteer,
		FNClassTitle := TextLocalization.ClassTitle_Puppeteer,
		ClassImage := Textures.class_puppeteer,
		MainAbility := FNAbility.None,
		MainAbility2 := FNAbility.None
	}

	@editable ClassFletcherKane		: fn_class_data = fn_class_data{
		FNClassID := FNClasses.FletcherKane,
		FNClassFaction := FNFactions.None,
		FNClassName := TextLocalization.Class_FletcherKane,
		FNClassTitle := TextLocalization.ClassTitle_FletcherKane,
		ClassImage := Textures.class_fletcherkane,
		MainAbility := FNAbility.DemonClaw,
		MainAbility2 := FNAbility.None
	}

	@editable ClassAlphaHeavy		: fn_class_data = fn_class_data{
		FNClassID := FNClasses.AlphaHeavy,
		FNClassFaction := FNFactions.None,
		FNClassName := TextLocalization.Class_AlphaHeavy,
		FNClassTitle := TextLocalization.ClassTitle_AlphaHeavy,
		ClassImage := Textures.class_alphaheavy,
		MainAbility := FNAbility.None,
		MainAbility2 := FNAbility.None
	}

	@editable ClassDaigo			: fn_class_data = fn_class_data{
		FNClassID := FNClasses.Daigo,
		FNClassFaction := FNFactions.None,
		FNClassName := TextLocalization.Class_Daigo,
		FNClassTitle := TextLocalization.ClassTitle_Daigo,
		ClassImage := Textures.class_daigo,
		MainAbility := FNAbility.ShortBeam,
		MainAbility2 := FNAbility.DoubleJump,
		PassiveList := array{FNAbility.ShieldRegen}
	}

	@editable ClassKor				: fn_class_data = fn_class_data{
		FNClassID := FNClasses.Kor,
		FNClassFaction := FNFactions.None,
		FNClassName := TextLocalization.Class_Kor,
		FNClassTitle := TextLocalization.ClassTitle_Kor,
		ClassImage := Textures.class_kor,
		MainAbility := FNAbility.MirrorImage,
		MainAbility2 := FNAbility.DoubleJump
	}

	@editable ClassTheIceKing		: fn_class_data = fn_class_data{
		FNClassID := FNClasses.TheIceKing,
		FNClassFaction := FNFactions.None,
		FNClassName := TextLocalization.Class_IceKing,
		FNClassTitle := TextLocalization.ClassTitle_IceKing,
		ClassImage := Textures.class_iceking,
		MainAbility := FNAbility.None,
		MainAbility2 := FNAbility.None
	}
	@editable ClassJones			: fn_class_data = fn_class_data{
		FNClassID := FNClasses.Jones,
		FNClassFaction := FNFactions.Seven,
		FNClassName := TextLocalization.Class_Jones,
		FNClassTitle := TextLocalization.ClassTitle_Jones,
		ClassImage := Textures.class_jones,
		MainAbility := FNAbility.RiftGun,
		MainAbility2 := FNAbility.None
	}

	@editable ClassDoctorDoom		: fn_class_data = fn_class_data{
		FNClassID := FNClasses.DoctorDoom
		FNClassFaction := FNFactions.None
		FNClassName := TextLocalization.Class_DoctorDoom
		FNClassTitle := TextLocalization.ClassTitle_DoctorDoom
		ClassImage := Textures.class_doctordoom
		MainAbility := FNAbility.Rock
		MainAbility2 := FNAbility.FootDive
		MaxUlt := 120.0
	}

	@editable ClassDoctorDoomMythic		: fn_class_data = fn_class_data{
		FNClassID := FNClasses.DoctorDoomMythic
		FNClassFaction := FNFactions.None
		FNClassName := TextLocalization.Class_DoctorDoomMythic
		FNClassTitle := TextLocalization.ClassTitle_DoctorDoomMythic
		ClassImage := Textures.class_doctordoom
		MainAbility := FNAbility.Rock
		MainAbility2 := FNAbility.FootDive
		MaxHealth := 400.0
		MaxShield := 400.0
		MaxUlt := 120.0
	}

	@editable ClassCaretaker		: fn_class_data = fn_class_data{
		FNClassID := FNClasses.Caretaker
		FNClassFaction := FNFactions.LastReality
		FNClassName := TextLocalization.Class_Caretaker
		FNClassTitle := TextLocalization.ClassTitle_Caretaker
		ClassImage := Textures.class_caretaker
		MainAbility := FNAbility.Laserbeam
		MainAbility2 := FNAbility.Snap
		MaxHealth := 1000.0
		MaxShield := 0.0
		MaxUlt := 100.0
	}

	@editable ClassHope: fn_class_data = fn_class_data{
		FNClassID := FNClasses.Hope
		FNClassFaction := FNFactions.None
		FNClassName := TextLocalization.Class_Hope
		FNClassTitle := TextLocalization.ClassTitle_Hope
		ClassImage := Textures.class_hope
		MainAbility := FNAbility.None
		MainAbility2 := FNAbility.None
	}

	@editable XP_Assist : accolades_device = accolades_device{}
	@editable XP_Healer : accolades_device = accolades_device{}
	@editable XP_4Kills : accolades_device = accolades_device{}
	@editable XP_Furina : accolades_device = accolades_device{}
	@editable XP_SgtWinterKill : accolades_device = accolades_device{}

	@editable{Categories:=array{CategoryVFXPowerup}} VFXPowerUlt : visual_effect_powerup_device = visual_effect_powerup_device{}
	@editable{Categories:=array{CategoryVFXPowerup}} VFXPowerCubeHeal : visual_effect_powerup_device = visual_effect_powerup_device{}

	#@editable{Categories:=array{CategorySFX}} SFXUltReady		: audio_player_device = audio_player_device{}

	@editable{Categories:=array{CategorySFX}} SFXAbilityLaserBeamHit : audio_player_device = audio_player_device{}
	@editable{Categories:=array{CategorySFX}} SFXAbilityCubeBeamEnd : audio_player_device = audio_player_device{}
	@editable InputCubeBlueFire : input_trigger_device = input_trigger_device{}
	@editable{Categories:=array{CategoryProp}} VFXPropCubeBlueFire : creative_prop_asset = DefaultCreativePropAsset
	@editable InputTargetCubeBlue : input_trigger_device = input_trigger_device{}
	@editable{Categories:=array{CategoryProp}} VFXPropCubeBlueHeal : creative_prop_asset = DefaultCreativePropAsset

	@editable InputTargetCaretaker : input_trigger_device = input_trigger_device{}
	@editable{Categories:=array{CategoryVFXPowerup}} VFXPowerCaretakerMelee : visual_effect_powerup_device = visual_effect_powerup_device{}

	@editable InputTargetDoom : input_trigger_device = input_trigger_device{}

	@editable DMGAmplifier			: damage_amplifier_powerup_device = damage_amplifier_powerup_device{}
	@editable{Categories:=array{CategoryVFXPowerup}} VFXDMGAmplifier		: visual_effect_powerup_device = visual_effect_powerup_device{}

	@editable ExplosiveDeviceLow : explosive_device  = explosive_device{}
	@editable ExplosiveDeviceHigh : explosive_device  = explosive_device{}

	UITextMessage<localizes>(Value : string) : message = "{Value}"
	@editable HudMsg_Shadow : hud_message_device = hud_message_device{}
	@editable HudMsg_Status : hud_message_device = hud_message_device{}

	@editable{Categories:=array{CategoryAnim}} AnimCaretakerMelee : creative_prop_asset = DefaultCreativePropAsset

	#@editable ConditionalItem : conditional_button_device = conditional_button_device{}

	@editable ConditionalDaigoBlade : conditional_button_device = conditional_button_device{}

	var UIHudMap : [player]ui_hudgameplay = map{}

	# Voice Menu UI
	@editable InputVoiceMenu : input_trigger_device = input_trigger_device{}

	@editable ScoreManager : score_manager_device = score_manager_device{}

	#ClassSwitchedEvent:event(agent) = event(agent){}
	#AddCondEvent:event(agent) = event(agent){}
	#RemoveCondEvent:event(agent) = event(agent){}

    # Runs when the device is started in a running game
    OnBegin<override>()<suspends>:void=
	{
		Print("Initialize Player Classes Manager...")

		ClassMidas.ClassSelector.ClassSwitchedEvent.SubscribeAgent(OnClassSwitched, FNClasses.Midas)
		ClassTNTina.ClassSelector.ClassSwitchedEvent.SubscribeAgent(OnClassSwitched, FNClasses.TNTina)
		ClassBrutus.ClassSelector.ClassSwitchedEvent.SubscribeAgent(OnClassSwitched, FNClasses.Brutus)
		ClassSkye.ClassSelector.ClassSwitchedEvent.SubscribeAgent(OnClassSwitched, FNClasses.Skye)
		ClassMeowscles.ClassSelector.ClassSwitchedEvent.SubscribeAgent(OnClassSwitched, FNClasses.Meowscles)
		ClassJules.ClassSelector.ClassSwitchedEvent.SubscribeAgent(OnClassSwitched, FNClasses.Jules)
		ClassKit.ClassSelector.ClassSwitchedEvent.SubscribeAgent(OnClassSwitched, FNClasses.Kit)
		ClassOcean.ClassSelector.ClassSwitchedEvent.SubscribeAgent(OnClassSwitched, FNClasses.Ocean)
		ClassShadowMidas.ClassSelector.ClassSwitchedEvent.SubscribeAgent(OnClassSwitched, FNClasses.ShadowMidas)
		ClassShadow.ClassSelector.ClassSwitchedEvent.SubscribeAgent(OnClassSwitched, FNClasses.Shadow)
		ClassWildguardRelik.ClassSelector.ClassSwitchedEvent.SubscribeAgent(OnClassSwitched, FNClasses.WildguardRelik)
		ClassKadoThorne.ClassSelector.ClassSwitchedEvent.SubscribeAgent(OnClassSwitched, FNClasses.KadoThorne)
		ClassTheDealer.ClassSelector.ClassSwitchedEvent.SubscribeAgent(OnClassSwitched, FNClasses.TheDealer)
		ClassTheAgeless.ClassSelector.ClassSwitchedEvent.SubscribeAgent(OnClassSwitched, FNClasses.TheAgelessChampion)
		ClassRaz.ClassSelector.ClassSwitchedEvent.SubscribeAgent(OnClassSwitched, FNClasses.Raz)
		ClassSlone.ClassSelector.ClassSwitchedEvent.SubscribeAgent(OnClassSwitched, FNClasses.Slone)
		ClassTheFoundation.ClassSelector.ClassSwitchedEvent.SubscribeAgent(OnClassSwitched, FNClasses.TheFoundation)
		ClassGunnar.ClassSelector.ClassSwitchedEvent.SubscribeAgent(OnClassSwitched, FNClasses.Gunnar)
		ClassHuntmasterSaber.ClassSelector.ClassSwitchedEvent.SubscribeAgent(OnClassSwitched, FNClasses.HuntmasterSaber)
		ClassTheHerald.ClassSelector.ClassSwitchedEvent.SubscribeAgent(OnClassSwitched, FNClasses.TheHerald)
		ClassTheIceKing.ClassSelector.ClassSwitchedEvent.SubscribeAgent(OnClassSwitched, FNClasses.TheIceKing)
		ClassCubeBlue.ClassSelector.ClassSwitchedEvent.SubscribeAgent(OnClassSwitched, FNClasses.CubeBlue)
		ClassCubeQueen.ClassSelector.ClassSwitchedEvent.SubscribeAgent(OnClassSwitched, FNClasses.CubeQueen)
		ClassValeria.ClassSelector.ClassSwitchedEvent.SubscribeAgent(OnClassSwitched, FNClasses.Valeria)
		ClassZeus.ClassSelector.ClassSwitchedEvent.SubscribeAgent(OnClassSwitched, FNClasses.Zeus)
		ClassAres.ClassSelector.ClassSwitchedEvent.SubscribeAgent(OnClassSwitched, FNClasses.Ares)
		ClassCerberus.ClassSelector.ClassSwitchedEvent.SubscribeAgent(OnClassSwitched, FNClasses.Cerberus)
		ClassHades.ClassSelector.ClassSwitchedEvent.SubscribeAgent(OnClassSwitched, FNClasses.Hades)
		ClassOscar.ClassSelector.ClassSwitchedEvent.SubscribeAgent(OnClassSwitched, FNClasses.Oscar)
		ClassNisha.ClassSelector.ClassSwitchedEvent.SubscribeAgent(OnClassSwitched, FNClasses.Nisha)
		ClassMontague.ClassSelector.ClassSwitchedEvent.SubscribeAgent(OnClassSwitched, FNClasses.Montague)
		ClassSgtWinter.ClassSelector.ClassSwitchedEvent.SubscribeAgent(OnClassSwitched, FNClasses.SgtWinter)
		ClassCubeAssassin.ClassSelector.ClassSwitchedEvent.SubscribeAgent(OnClassSwitched, FNClasses.CubeAssassin)
		ClassSpireAssassin.ClassSelector.ClassSwitchedEvent.SubscribeAgent(OnClassSwitched, FNClasses.SpireAssassin)
		ClassZygAndChoppy.ClassSelector.ClassSwitchedEvent.SubscribeAgent(OnClassSwitched, FNClasses.ZygAndChoppy)
		ClassMegaloDon.ClassSelector.ClassSwitchedEvent.SubscribeAgent(OnClassSwitched, FNClasses.MegaloDon)
		ClassRingmasterScarr.ClassSelector.ClassSwitchedEvent.SubscribeAgent(OnClassSwitched, FNClasses.RingmasterScarr)
		ClassTheMachinist.ClassSelector.ClassSwitchedEvent.SubscribeAgent(OnClassSwitched, FNClasses.TheMachinist)
		ClassUltimaCarver.ClassSelector.ClassSwitchedEvent.SubscribeAgent(OnClassSwitched, FNClasses.UltimaCarver)
		ClassShogunX.ClassSelector.ClassSwitchedEvent.SubscribeAgent(OnClassSwitched, FNClasses.ShogunX)
		ClassNightRose.ClassSelector.ClassSwitchedEvent.SubscribeAgent(OnClassSwitched, FNClasses.NightRose)
		ClassPuppeteer.ClassSelector.ClassSwitchedEvent.SubscribeAgent(OnClassSwitched, FNClasses.Puppeteer)
		ClassFletcherKane.ClassSelector.ClassSwitchedEvent.SubscribeAgent(OnClassSwitched, FNClasses.FletcherKane)
		ClassAlphaHeavy.ClassSelector.ClassSwitchedEvent.SubscribeAgent(OnClassSwitched, FNClasses.AlphaHeavy)
		ClassDaigo.ClassSelector.ClassSwitchedEvent.SubscribeAgent(OnClassSwitched, FNClasses.Daigo)
		ClassKor.ClassSelector.ClassSwitchedEvent.SubscribeAgent(OnClassSwitched, FNClasses.Kor)

		ClassJones.ClassSelector.ClassSwitchedEvent.SubscribeAgent(OnClassSwitched, FNClasses.Jones)
		ClassDoctorDoom.ClassSelector.ClassSwitchedEvent.SubscribeAgent(OnClassSwitched, FNClasses.DoctorDoom)
		ClassDoctorDoomMythic.ClassSelector.ClassSwitchedEvent.SubscribeAgent(OnClassSwitched, FNClasses.DoctorDoomMythic)
		ClassCaretaker.ClassSelector.ClassSwitchedEvent.SubscribeAgent(OnClassSwitched, FNClasses.Caretaker)
		ClassHope.ClassSelector.ClassSwitchedEvent.SubscribeAgent(OnClassSwitched, FNClasses.Hope)

		InputTargetCubeBlue.PressedEvent.Subscribe(InputPressedTargetCubeBlue)
		InputTargetCaretaker.PressedEvent.Subscribe(InputPressedTargetCaretaker)
		InputTargetDoom.PressedEvent.Subscribe(InputPressedTargetDoom)
	}

	GetAllClasses():[]fn_class_data=
	{
		var AllClassesArray : []fn_class_data = array{}

		set AllClassesArray += array{ClassMidas}
		set AllClassesArray += array{ClassTNTina}
		set AllClassesArray += array{ClassBrutus}
		set AllClassesArray += array{ClassSkye}
		set AllClassesArray += array{ClassMeowscles}
		set AllClassesArray += array{ClassKit}

		set AllClassesArray += array{ClassShadowMidas}
		set AllClassesArray += array{ClassJules}
		set AllClassesArray += array{ClassOcean}
		set AllClassesArray += array{ClassWildguardRelik}
		set AllClassesArray += array{ClassKadoThorne}
		set AllClassesArray += array{ClassTheDealer}

		set AllClassesArray += array{ClassRaz}
		set AllClassesArray += array{ClassZygAndChoppy}
		set AllClassesArray += array{ClassCubeAssassin}
		set AllClassesArray += array{ClassTheHerald}
		set AllClassesArray += array{ClassTheFoundation}

		set AllClassesArray += array{ClassTheAgeless}
		set AllClassesArray += array{ClassSlone}
		set AllClassesArray += array{ClassGunnar}
		set AllClassesArray += array{ClassHuntmasterSaber}
		set AllClassesArray += array{ClassSpireAssassin}

		set AllClassesArray += array{ClassSgtWinter}
		set AllClassesArray += array{ClassOscar}
		set AllClassesArray += array{ClassNisha}
		set AllClassesArray += array{ClassMontague}
		set AllClassesArray += array{ClassValeria}

		set AllClassesArray += array{ClassZeus}
		set AllClassesArray += array{ClassAres}
		set AllClassesArray += array{ClassCerberus}
		set AllClassesArray += array{ClassHades}

		set AllClassesArray += array{ClassMegaloDon}
		set AllClassesArray += array{ClassRingmasterScarr}
		set AllClassesArray += array{ClassTheMachinist}
		set AllClassesArray += array{ClassUltimaCarver}
		set AllClassesArray += array{ClassShogunX}
		set AllClassesArray += array{ClassNightRose}
		set AllClassesArray += array{ClassFletcherKane}
		set AllClassesArray += array{ClassAlphaHeavy}
		set AllClassesArray += array{ClassKor}
		set AllClassesArray += array{ClassDaigo}

		set AllClassesArray += array{ClassTheIceKing}
		set AllClassesArray += array{ClassCubeQueen}
		set AllClassesArray += array{ClassCubeBlue}
		set AllClassesArray += array{ClassJones}
		set AllClassesArray += array{ClassDoctorDoom}
		set AllClassesArray += array{ClassCaretaker}
		#set AllClassesArray += array{ClassHope}

		return AllClassesArray
	}

	OnPlayerJoined(Agent: agent): void =
	{
		spawn{PlayerEnterGame(Agent)}
	}

	# Removes an element from the given map and returns a new map without that element
	(Map:[keytype]valuetype where keytype:subtype(comparable), valuetype:type).RemoveKey(ElementToRemove:keytype):[keytype]valuetype=
	{
		var NewMap:[keytype]valuetype = map{}
		for (Key -> Value : Map, Key <> ElementToRemove)
		{
			set NewMap = ConcatenateMaps(NewMap, map{Key => Value})
		}

		return NewMap
	}

    OnPlayerRemoved(Agent: agent): void =
	{
		ClassSwitchedCleanup(Agent)
		if (AgentPlayer := player[Agent])
		{
			set UIHudMap = UIHudMap.RemoveKey(AgentPlayer)
		}
		set PlayersMap = PlayersMap.RemoveKey(Agent)
	}

    OnClassSwitched(Agent: agent, ClassID:FNClasses):void=
	{
		case( ClassID )
		{
			FNClasses.Midas =>				if{ set PlayersMap[Agent].ClassData = ClassMidas }
			FNClasses.TNTina =>				if{ set PlayersMap[Agent].ClassData = ClassTNTina }
			FNClasses.Brutus =>				if{ set PlayersMap[Agent].ClassData = ClassBrutus }
			FNClasses.Skye =>				if{ set PlayersMap[Agent].ClassData = ClassSkye }
			FNClasses.Meowscles =>			if{ set PlayersMap[Agent].ClassData = ClassMeowscles }
			FNClasses.Jules =>				if{ set PlayersMap[Agent].ClassData = ClassJules }
			FNClasses.Kit =>				if{ set PlayersMap[Agent].ClassData = ClassKit }
			FNClasses.Ocean =>				if{ set PlayersMap[Agent].ClassData = ClassOcean }
			FNClasses.ShadowMidas =>		if{ set PlayersMap[Agent].ClassData = ClassShadowMidas }
			FNClasses.Shadow =>				if{ set PlayersMap[Agent].ClassData = ClassShadow }
			FNClasses.Raz =>				if{ set PlayersMap[Agent].ClassData = ClassRaz }
			FNClasses.Slone =>				if{ set PlayersMap[Agent].ClassData = ClassSlone }
			FNClasses.TheFoundation =>		if{ set PlayersMap[Agent].ClassData = ClassTheFoundation }
			FNClasses.Gunnar =>				if{ set PlayersMap[Agent].ClassData = ClassGunnar }
			FNClasses.HuntmasterSaber =>	if{ set PlayersMap[Agent].ClassData = ClassHuntmasterSaber }
			FNClasses.TheHerald =>			if{ set PlayersMap[Agent].ClassData = ClassTheHerald }
			FNClasses.TheIceKing =>			if{ set PlayersMap[Agent].ClassData = ClassTheIceKing }
			FNClasses.TheAgelessChampion => if{ set PlayersMap[Agent].ClassData = ClassTheAgeless }
			FNClasses.WildguardRelik =>		if{ set PlayersMap[Agent].ClassData = ClassWildguardRelik }
			FNClasses.KadoThorne =>			if{ set PlayersMap[Agent].ClassData = ClassKadoThorne }
			FNClasses.TheDealer =>			if{ set PlayersMap[Agent].ClassData = ClassTheDealer }
			FNClasses.Valeria =>			if{ set PlayersMap[Agent].ClassData = ClassValeria }
			FNClasses.Oscar =>				if{ set PlayersMap[Agent].ClassData = ClassOscar }
			FNClasses.Nisha =>				if{ set PlayersMap[Agent].ClassData = ClassNisha }
			FNClasses.Montague =>			if{ set PlayersMap[Agent].ClassData = ClassMontague }
			FNClasses.SgtWinter =>			if{ set PlayersMap[Agent].ClassData = ClassSgtWinter }
			FNClasses.SpireAssassin =>		if{ set PlayersMap[Agent].ClassData = ClassSpireAssassin }
			FNClasses.CubeAssassin =>		if{ set PlayersMap[Agent].ClassData = ClassCubeAssassin }
			FNClasses.CubeBlue =>			if{ set PlayersMap[Agent].ClassData = ClassCubeBlue }
			FNClasses.CubeQueen =>			if{ set PlayersMap[Agent].ClassData = ClassCubeQueen }
			FNClasses.ZygAndChoppy =>		if{ set PlayersMap[Agent].ClassData = ClassZygAndChoppy }
			FNClasses.Zeus =>				if{ set PlayersMap[Agent].ClassData = ClassZeus }
			FNClasses.Ares =>				if{ set PlayersMap[Agent].ClassData = ClassAres }
			FNClasses.Cerberus =>			if{ set PlayersMap[Agent].ClassData = ClassCerberus }
			FNClasses.Hades =>				if{ set PlayersMap[Agent].ClassData = ClassHades }
			FNClasses.MegaloDon =>			if{ set PlayersMap[Agent].ClassData = ClassMegaloDon }
			FNClasses.RingmasterScarr =>	if{ set PlayersMap[Agent].ClassData = ClassRingmasterScarr }
			FNClasses.TheMachinist =>		if{ set PlayersMap[Agent].ClassData = ClassTheMachinist }
			FNClasses.UltimaCarver =>		if{ set PlayersMap[Agent].ClassData = ClassUltimaCarver }
			FNClasses.ShogunX =>			if{ set PlayersMap[Agent].ClassData = ClassShogunX }
			FNClasses.NightRose =>			if{ set PlayersMap[Agent].ClassData = ClassNightRose }
			FNClasses.Puppeteer =>			if{ set PlayersMap[Agent].ClassData = ClassPuppeteer }
			FNClasses.FletcherKane =>		if{ set PlayersMap[Agent].ClassData = ClassFletcherKane }
			FNClasses.AlphaHeavy =>			if{ set PlayersMap[Agent].ClassData = ClassAlphaHeavy }
			FNClasses.Jones =>				if{ set PlayersMap[Agent].ClassData = ClassJones }
			FNClasses.DoctorDoom =>			if{ set PlayersMap[Agent].ClassData = ClassDoctorDoom }
			FNClasses.DoctorDoomMythic =>	if{ set PlayersMap[Agent].ClassData = ClassDoctorDoomMythic }
			FNClasses.Caretaker =>			if{ set PlayersMap[Agent].ClassData = ClassCaretaker }
			FNClasses.Hope =>				if{ set PlayersMap[Agent].ClassData = ClassHope }
			FNClasses.Daigo =>				if{ set PlayersMap[Agent].ClassData = ClassDaigo }
			FNClasses.Kor =>				if{ set PlayersMap[Agent].ClassData = ClassKor }
			_ =>							if{ set PlayersMap[Agent].ClassData = ClassMidas }
		}

		ClassSwitchedCleanup(Agent)
	}

	InputPressedFire(Agent : agent) : void =
	{
		if (FortCharacter := Agent.GetFortCharacter[], FortCharacter.IsActive[] )
		{
			case( GetPlayerClassIndex(Agent) )
			{
				FNClasses.TheIceKing =>			PlayClassVO(Agent, FNVOType.FullAlert)
				FNClasses.Zeus =>				PlayClassVO(Agent, FNVOType.Ability)
				#FNClasses.ShogunX =>			PlayClassVO(Agent, FNVOType.ATK)
				FNClasses.Puppeteer =>			PlayClassVO(Agent, FNVOType.Ability)
				FNClasses.CubeBlue =>			UsePrimaryAttack(Agent)
				FNClasses.Caretaker =>			UsePrimaryAttack(Agent)
				FNClasses.DoctorDoomMythic =>
				{
					UsePrimaryAttack(Agent)
				}
				FNClasses.Daigo =>
				{
					if ( ConditionalDaigoBlade.IsHoldingItem[Agent] )
					{
						PlayClassVO(Agent, FNVOType.Ability)
					}
				}
				_ =>							#nothing
			}
		}
	}

	UsePrimaryAttack(Agent : agent) : void =
	{
		if (FortCharacter := Agent.GetFortCharacter[], FortCharacter.IsActive[] )
		{
			if ( Agent.IsInSpawnRoom[Self] )
			{
				#AddHudStatusMessage(Agent, "Cannot use attack in spawn room.")
				SFXAbilityCooldown.Play(Agent)
				return
			}

			CurrentTime := GetSimulationElapsedTime()
			if ( AbilityUsedTime := PlayersMap[Agent].flAbilityUsedTime, ElapsedTime := CurrentTime - PlayersMap[Agent].flAbilityUsedTime )
			{
				case( GetPlayerClassIndex(Agent) )
				{
					FNClasses.CubeBlue =>
					{
						if (ElapsedTime < 2.0)
						{
							Print("Ability on cooldown... {ElapsedTime}")
							SFXAbilityCooldown.Play(Agent)
							return
						}

						spawn{ BlueCubePrimaryAttack(Agent) }
					}
					FNClasses.Caretaker =>
					{
						if (ElapsedTime < 3.0)
						{
							Print("Ability on cooldown... {ElapsedTime}")
							SFXAbilityCooldown.Play(Agent)
							return
						}

						spawn{ CaretakerPrimaryAttack(Agent) }
					}
					FNClasses.DoctorDoomMythic =>
					{
						if (ElapsedTime < 0.6)
						{
							Print("Ability on cooldown... {ElapsedTime}")
							SFXAbilityCooldown.Play(Agent)
							return
						}

						spawn{ DoomPrimaryAttack(Agent) }
					}
					_ =>							#nothing
				}
			}
		}
	}

	AbilityCooldown(Agent : agent, AbilityID:FNAbility)<suspends>:void =
	{
		if ( AbilityData := GetAbilityData[AbilityID] )
		{
			var TimerMax : float = AbilityData.Cooldown

			if ( MainAbility := GetMainAbility[Agent], AbilityID = MainAbility )
			{
				if
				{
					set PlayersMap[Agent].flAbilityCooldownMax = TimerMax
					set PlayersMap[Agent].flAbilityCooldown = TimerMax
				}
				then
				{
					UpdateHudAbilityCooldown(Agent, AbilityID)
					loop
					{
						Sleep(1.0)
						if ( set PlayersMap[Agent].flAbilityCooldown -= 1.0 )
						{
							UpdateHudAbilityCooldown(Agent, AbilityID)
						}

						if ( PlayersMap[Agent].flAbilityCooldown <= 0.0 )
						{
							break
						}
					}
				}
			}
			else
			{
				case( GetPlayerClassIndex(Agent) )
				{
					FNClasses.Zeus =>				set TimerMax = 7.0
					FNClasses.Daigo =>				set TimerMax = 10.0
					FNClasses.Kor =>				set TimerMax = 7.0
					_ =>							#nothing
				}

				if
				{
					set PlayersMap[Agent].flAbilityCooldownMax2 = TimerMax
					set PlayersMap[Agent].flAbilityCooldown2 = TimerMax
				}
				then
				{
					UpdateHudAbilityCooldown(Agent, AbilityID)
					loop
					{
						Sleep(1.0)
						if ( set PlayersMap[Agent].flAbilityCooldown2 -= 1.0 )
						{
							UpdateHudAbilityCooldown(Agent, AbilityID)
						}

						if ( PlayersMap[Agent].flAbilityCooldown2 <= 0.0 )
						{
							break
						}
					}
				}
			}
		}
	}

    UpdateHudAbilityCooldown(Agent : agent, AbilityID:FNAbility) : void =
	{
		if ( AgentPlayer := player[Agent], PlayerHUD := UIHudMap[AgentPlayer] )
		{
			PlayerHUD.UpdateAbilityCooldown(AbilityID)
		}
	}

    UpdateHudDisableAbility(Agent : agent, AbilityID:FNAbility) : void =
	{
		if ( AgentPlayer := player[Agent], PlayerHUD := UIHudMap[AgentPlayer] )
		{
			PlayerHUD.DisableAbility(AbilityID)
		}
	}

    AddHudStatusMessage(Agent:agent, StatusText:message) : void =
	{
		if ( AgentPlayer := player[Agent], PlayerHUD := UIHudMap[AgentPlayer] )
		{
			HudMsg_Status.Show(Agent, StatusText, ?DisplayTime := 4.0)
		}
	}

	UpdateHudDamageText(Agent:agent, Amount:float)<suspends>: void =
	{
		if ( AgentPlayer := player[Agent], PlayerHUD := UIHudMap[AgentPlayer] )
		{
			PlayerHUD.UpdateDamageText(Amount)
		}
	}

    (Agent : agent).IsRevived<public>()<transacts><decides>:void =
	{
		PlayersMap[Agent].bRevived?
	}

    (Agent : agent).IsAltTransform<public>()<transacts><decides>:void =
	{
		PlayersMap[Agent].bAltTransform?
	}

    AgentIsAltTransform<public>(Agent : agent)<transacts><decides>:void =
	{
		PlayersMap[Agent].bAltTransform?
	}

	GetAllPlayersInVectorDistance( CenterPos : vector3, DistanceRange: float)<transacts>:[]fort_character=
	{
		var ClosestPlayers: []fort_character = array{}

		if ( Entity := GetSimulationEntity[] )
		{
			CollisionSphere := collision_sphere{ Radius := DistanceRange <#, CollisionProfile := collision_profiles.StationaryOverlapAll#> }
			CollisionTransform := FromTransform(transform{Translation := CenterPos})

			for
			{
				Overlap : Entity.FindOverlapHits(CollisionTransform, CollisionSphere)
				PlayerMesh := Meshes.PlayerBox[Overlap.TargetComponent]
				TargetEntity := PlayerMesh.Entity
			}
			do
			{
				if ( PlayerComp := TargetEntity.GetComponent[fn_player_component], Agent := PlayerComp.Agent, FortChar := Agent.GetFortCharacter[] )
				{
					set ClosestPlayers += array{FortChar}
				}

				if ( NPCComp := TargetEntity.GetComponent[fn_npc_component], AgentNPC := NPCComp.Agent, FortNPC := AgentNPC.GetFortCharacter[] )
				{
					set ClosestPlayers += array{FortNPC}
				}
			}
		}


		Drawer := debug_draw{}
		Drawer.DrawSphere(CenterPos, ?Radius := DistanceRange, ?DrawDurationPolicy := debug_draw_duration_policy.FiniteDuration, ?Duration:= float(1.5))
        return ClosestPlayers
	}

	GetAllPlayersInDistance(PlayerCenter : fort_character, DistanceRange: float)<transacts>:[]fort_character=
	{
		CenterCurrentPos := PlayerCenter.GetTransform().Translation
        return GetAllPlayersInVectorDistance(CenterCurrentPos, DistanceRange)
	}

	GetAllPlayersInPropDistance(PropCenter : creative_prop, DistanceRange: float)<transacts>:[]fort_character=
	{
		CenterCurrentPos := PropCenter.GetTransform().Translation
        return GetAllPlayersInVectorDistance(CenterCurrentPos, DistanceRange)
	}

	<#GetRandomVector3(Low : vector3, High : vector3) : vector3 =
	{
		vector3{X := GetRandomFloat(Low.X, High.X),Y := GetRandomFloat(Low.Y, High.Y), Z := GetRandomFloat(Low.Z, High.Z) }
	}#>

	BlueCubePrimaryAttack(Agent:agent)<suspends>:void=
	{
		if ( FortCharacter := Agent.GetFortCharacter[] )
		{
			Print("Blue Cube: Primary Attack")
			PlayClassVO(Agent, FNVOType.Ability)

			CurrentTime := GetSimulationElapsedTime()
			if {set PlayersMap[Agent].flAbilityUsedTime = CurrentTime}

			PlayerPos := FortCharacter.GetTransform().Translation
			PlayerRotation := FortCharacter.GetViewRotation()
			spawn{ SpawnParticleVFX(FX.NS_CubeBlue_ATK_asset, PlayerPos, PlayerRotation, 0.4) }

			#spawn{ SpawnPropVFXFollowPlayer(VFXPropCubeBlueFire,FortCharacter,0.4) }

			TeamCollection := GetPlayspace().GetTeamCollection()
			if ( TeamOwner := TeamCollection.GetTeam[Agent] )
			{
				PlayerForwardPos := FortCharacter.GetTransform().Translation + FortCharacter.GetViewRotation().GetLocalForward() * 350.0
				AllVictim := GetAllPlayersInVectorDistance(PlayerForwardPos, 250.0)
				for ( ClosestVictim : AllVictim, ClosestAgent := ClosestVictim.GetAgent[], ClosestAgent <> Agent and ( not TeamCollection.IsOnTeam[ClosestAgent, TeamOwner] ) )
				{
					Print("Lightning: Damage Victim")
					var DmgBonus : float = 0.0
					if {DamageTaken := PlayersMap[Agent].DamageTakenTracker, set DmgBonus = DamageTaken}
					if ( FortCharacter.IsActive[], DamageInstigator := Agent.GetInstigator(), DamageSource := game_action_causer[FortCharacter] )
					{
						BeamDmg : damage_args = damage_args {Instigator := option{DamageInstigator}, Source := option{DamageSource}, Amount:=(50.0*GetDamageBonus(Agent))+DmgBonus}
						ClosestVictim.Damage(BeamDmg)
						if {set PlayersMap[Agent].DamageTakenTracker = 0.0}
					}
					SFXAbilityLaserBeamHit.Play(ClosestAgent)
				}

				if ( ExplosiveDeviceLow.TeleportTo[PlayerForwardPos, rotation{}] ) {}
				ExplosiveDeviceLow.Explode(Agent)

			}

			SFXAbilityCubeBeamEnd.Play(Agent)
		}
	}

	CaretakerPrimaryAttack(Agent:agent)<suspends>:void=
	{
		if ( FortCharacter := Agent.GetFortCharacter[] )
		{
			Print("Caretaker: Primary Attack")

			CurrentTime := GetSimulationElapsedTime()
			if {set PlayersMap[Agent].flAbilityUsedTime = CurrentTime}

			spawn{ChairAnim(Self,Agent,AnimCaretakerMelee,1.45)}
			#spawn{ChairAnimFollow(Self,Agent,AnimCaretakerMelee,1.45,?IgnoreZ:=true)}
			FortCharacter.SetVulnerability(true) # prevent PutInStasis invincible bug!
			VFXPowerCaretakerMelee.Pickup(Agent)

			PlayerForwardPos := FortCharacter.GetTransform().Translation + FortCharacter.GetViewRotation().GetLocalForward() * 350.0

			Sleep(1.3)

			TeamCollection := GetPlayspace().GetTeamCollection()
			if ( TeamOwner := TeamCollection.GetTeam[Agent] )
			{
				spawn{ SpawnParticleVFX(FX.NS_Caretaker_Melee_Impact_asset, PlayerForwardPos, rotation{}, 1.0) }

				AllVictim := GetAllPlayersInVectorDistance(PlayerForwardPos, 600.0)
				for ( ClosestVictim : AllVictim, ClosestAgent := ClosestVictim.GetAgent[], ClosestAgent <> Agent and ( not TeamCollection.IsOnTeam[ClosestAgent, TeamOwner] ) )
				{
					Print("Caretaker: Smash Victim")

					var DmgMultiplier : float = GetDamageBonus(Agent)

					if ( FortCharacter.IsActive[], DamageInstigator := Agent.GetInstigator(), DamageSource := game_action_causer[FortCharacter] )
					{
						BeamDmg : damage_args = damage_args {Instigator := option{DamageInstigator}, Source := option{DamageSource}, Amount:=120.0*DmgMultiplier}
						ClosestVictim.Damage(BeamDmg)
					}
				}

				if ( ExplosiveDeviceHigh.TeleportTo[PlayerForwardPos, rotation{}] )
				{
					ExplosiveDeviceHigh.Explode(Agent)
				}
			}
		}
	}

	DoomPrimaryAttack(Agent:agent)<suspends>:void=
	{
		if ( PlayerAbility := PlayerAbilityDevice? )
		{
			PlayerAbility.DoomPrimaryAttack(Agent)
		}
	}

    InputPressedTargetCubeBlue(Agent : agent) : void =
	{
		if (FortCharacter := Agent.GetFortCharacter[], FortCharacter.IsActive[] )
		{
			PlayerCooldown := GetAbilityCooldown(Agent, FNAbility.HealingRune)
			if ( PlayerCooldown > 0.0 )
			{
				Print("Ability on cooldown...")
				SFXAbilityCooldown.Play(Agent)
				return
			}

			spawn{ BlueCubeHeal(Agent) }
			spawn{ AbilityCooldown(Agent,FNAbility.HealingRune) }
		}
	}

	BlueCubeHeal(Agent:agent)<suspends>:void=
	{
		if ( FortCharacter := Agent.GetFortCharacter[] )
		{
			Print("Blue Cube: Heal AoE")

			PlayerPos := FortCharacter.GetTransform().Translation
			StartVFX := SpawnProp(VFXPropCubeBlueHeal, PlayerPos, rotation{})
			if (StartedVFX := StartVFX(0)?)
			{
				var bDoneHeal : logic = false
				branch
				{
					loop
					{
						if ( not FortCharacter.IsActive[] )
						{
							break
						}

						if ( bDoneHeal? )
						{
							break
						}

						TeamCollection := GetPlayspace().GetTeamCollection()
						if ( TeamOwner := TeamCollection.GetTeam[Agent] )
						{
							PropPos := StartedVFX.GetTransform().Translation
							AllVictim := GetAllPlayersInVectorDistance(PropPos, 1280.0)
							for ( ClosestVictim : AllVictim, ClosestAgent := ClosestVictim.GetAgent[], TeamCollection.IsOnTeam[ClosestAgent, TeamOwner], ClosestPlayer := player[ClosestAgent] )
							{
								Print("Blue Cube: Healing...")
								if ( FortCharacter.IsActive[], DamageInstigator := Agent.GetInstigator(), DamageSource := game_action_causer[FortCharacter] )
								{
									<#if ( FortCharacter.IsDownButNotOut[] )
									{
										DBNODevice.Revive(Agent)
									}#>
									HealDmg : healing_args = healing_args{Instigator := option{DamageInstigator}, Source := option{DamageSource}, Amount:=15.0}
									ClosestVictim.Heal(HealDmg)
								}
								ClosestVictim.SetShield(ClosestVictim.GetShield() + 15.0)
								VFXPowerCubeHeal.Pickup(Agent)
							}
						}
						
						Sleep(2.0)
					}
				}
				Sleep(10.0)
				set bDoneHeal = true
				StartedVFX.Dispose()
			}
		}
	}

    InputPressedTargetCaretaker(Agent : agent) : void =
	{
		if (FortCharacter := Agent.GetFortCharacter[], FortCharacter.IsActive[], PlayerAbility := PlayerAbilityDevice? )
		{
			PlayerCooldown := GetAbilityCooldown(Agent, FNAbility.Snap)
			if ( PlayerCooldown > 0.0 )
			{
				Print("Ability on cooldown...")
				SFXAbilityCooldown.Play(Agent)
				return
			}

			spawn{ PlayerAbility.AbilitySnap(Agent) }
			spawn{ AbilityCooldown(Agent,FNAbility.Snap) }
		}
	}

    InputPressedTargetDoom(Agent : agent) : void =
	{
		if (FortCharacter := Agent.GetFortCharacter[], FortCharacter.IsActive[], PlayerAbility := PlayerAbilityDevice? )
		{
			PlayerCooldown := GetAbilityCooldown(Agent, FNAbility.Rock)
			if ( PlayerCooldown > 0.0 )
			{
				Print("Ability on cooldown...")
				SFXAbilityCooldown.Play(Agent)
				return
			}

			spawn{ PlayerAbility.AbilityLaserBeam(Agent) }
			spawn{ AbilityCooldown(Agent,FNAbility.Rock) }
		}
	}

    ClassSwitchedCleanup(Agent:agent):void=
	{
		SetPlayerAggro(Agent, FNBGMState.Idle)

		<#if ( ClassPassiveList := PlayersMap[Agent].ClassData.PassiveList )
		{
			if {set PlayersMap[Agent].PassiveList = ClassPassiveList}
		}#>

		spawn{RemoveAllCond(Agent)}
		if {set PlayersMap[Agent].bRevived = false}
		if {set PlayersMap[Agent].bGroundPounding = false}
		if {set PlayersMap[Agent].bDoubleJumped = false}
		SetUltimateMeter(Agent, 0.0)
		if {set PlayersMap[Agent].bUltReady = false}
		if {set PlayersMap[Agent].bAltTransform = false}
		if {set PlayersMap[Agent].bShieldBroken = false}
		if {set PlayersMap[Agent].bIsEmoting = false}
		if {set PlayersMap[Agent].bHasTheFlag = false}
		if {set PlayersMap[Agent].bIsStunned = false}
		if {set PlayersMap[Agent].KillsTracker = 0 }
		if {set PlayersMap[Agent].DamageTakenTracker = 0.0}
		if {set PlayersMap[Agent].AssistAgent = false}

		CurrentTime := GetSimulationElapsedTime()
		if {set PlayersMap[Agent].flAbilityUsedTime = CurrentTime}
		if {set PlayersMap[Agent].flVOUsedTime = CurrentTime}

		if ( SaveManager := SaveManagerDevice?, SaveManager.IsVOMenuEnabled[Agent] )
		{
			InputVoiceMenu.Register(Agent)
		}
		else
		{
			InputVoiceMenu.Unregister(Agent)
		}
	
		if ( FortCharacter := Agent.GetFortCharacter[] )
		{
			spawn{RemoveVO(Agent)}

			if ( FortSummon := PlayersMap[Agent].ActiveSummon?, FortSummon.IsActive[] )
			{
				FortSummon.Damage(5000.0)
				if {set PlayersMap[Agent].ActiveSummon = false}
			}

			if ( PropBuilding := PlayersMap[Agent].ActiveBuilding?, PropBuilding.IsValid[] )
			{
				PropBuilding.Dispose()
				if {set PlayersMap[Agent].ActiveBuilding = false}
			}

			FortCharacter.SetMaxHealth(GetMaxHealth(Agent))
			FortCharacter.SetHealth(GetMaxHealth(Agent))
			FortCharacter.SetMaxShield(GetMaxShield(Agent))
			FortCharacter.SetShield(GetMaxShield(Agent))
		}

		if ( Agent.IsOnTeam[Self,1], PlayersMap[Agent].ClassData.HasTeamAlt? )
		{
			if ( ClassVFXPowerup := PlayersMap[Agent].ClassData.Team2_VFXPowerup )
			{
				ClassVFXPowerup.Pickup(Agent)
			}
		}
		else
		{
			if ( ClassVFXPowerup := PlayersMap[Agent].ClassData.VFXPowerup )
			{
				ClassVFXPowerup.Pickup(Agent)
			}
		}

		if { set PlayersMap[Agent].flAbilityCooldown = 0.0 }
		if { set PlayersMap[Agent].flAbilityCooldown2 = 0.0 }

		# debuggin
		#AddUltimateMeter(Agent, 100.0)

		if ( PlayerAbility := PlayerAbilityDevice? )
		{
			spawn{PlayerAbility.MoveNormal(Agent)}
		}
	}

    GetPlayerClassIndex(Agent:agent):FNClasses=
	{
		if (ClassIndex := PlayersMap[Agent].GetAgentClass[])
		{
			return ClassIndex
		}
		
		return FNClasses.None
	}

    GetPlayerClassFaction(Agent:agent):FNFactions=
	{
		if (ClassFaction := PlayersMap[Agent].GetClassFaction[])
		{
			return ClassFaction
		}
		
		return FNFactions.None
	}

	SetPlayerClass<public>(Agent:agent, eClass:FNClasses):void=
	{
		ClassSelectorNONE.ChangeClass( Agent )

		case( eClass )
		{
			FNClasses.Midas =>
			{
				ClassMidas.ClassSelector.ChangeClass( Agent )
			}
			FNClasses.TNTina =>
			{
				ClassTNTina.ClassSelector.ChangeClass( Agent )
			}
			FNClasses.Brutus =>
			{
				ClassBrutus.ClassSelector.ChangeClass( Agent )
			}
			FNClasses.Skye =>
			{
				ClassSkye.ClassSelector.ChangeClass( Agent )
			}
			FNClasses.Meowscles =>
			{
				ClassMeowscles.ClassSelector.ChangeClass( Agent )
			}
			FNClasses.Jules =>
			{
				ClassJules.ClassSelector.ChangeClass( Agent )
			}
			FNClasses.Kit =>
			{
				ClassKit.ClassSelector.ChangeClass( Agent )
			}
			FNClasses.Ocean =>
			{
				ClassOcean.ClassSelector.ChangeClass( Agent )
			}
			FNClasses.ShadowMidas =>
			{
				ClassShadowMidas.ClassSelector.ChangeClass( Agent )
			}
			FNClasses.Shadow =>
			{
				ClassShadow.ClassSelector.ChangeClass( Agent )
				PlayClassVO(Agent, FNVOType.FullAlertLost)
				HudMsg_Shadow.Show(Agent)
				<#if ( FortCharacter := Agent.GetFortCharacter[] )
				{
					FortCharacter.Hide()
				}#>
			}
			FNClasses.WildguardRelik =>
			{
				ClassWildguardRelik.ClassSelector.ChangeClass( Agent )
			}
			FNClasses.TheAgelessChampion =>
			{
				ClassTheAgeless.ClassSelector.ChangeClass( Agent )
			}
			FNClasses.KadoThorne =>
			{
				ClassKadoThorne.ClassSelector.ChangeClass( Agent )
			}
			FNClasses.TheDealer =>
			{
				ClassTheDealer.ClassSelector.ChangeClass( Agent )
			}
			FNClasses.Raz =>
			{
				ClassRaz.ClassSelector.ChangeClass( Agent )
			}
			FNClasses.Slone =>
			{
				ClassSlone.ClassSelector.ChangeClass( Agent )
			}
			FNClasses.TheFoundation =>
			{
				ClassTheFoundation.ClassSelector.ChangeClass( Agent )
			}
			FNClasses.Gunnar =>
			{
				ClassGunnar.ClassSelector.ChangeClass( Agent )
			}
			FNClasses.HuntmasterSaber =>
			{
				ClassHuntmasterSaber.ClassSelector.ChangeClass( Agent )
			}
			FNClasses.TheHerald =>
			{
				ClassTheHerald.ClassSelector.ChangeClass( Agent )
			}
			FNClasses.TheIceKing =>
			{
				ClassTheIceKing.ClassSelector.ChangeClass( Agent )
			}
			FNClasses.CubeQueen =>
			{
				ClassCubeQueen.ClassSelector.ChangeClass( Agent )
			}
			FNClasses.Valeria =>
			{
				ClassValeria.ClassSelector.ChangeClass( Agent )
			}
			FNClasses.Oscar =>
			{
				ClassOscar.ClassSelector.ChangeClass( Agent )
			}
			FNClasses.Nisha =>
			{
				ClassNisha.ClassSelector.ChangeClass( Agent )
			}
			FNClasses.Montague =>
			{
				ClassMontague.ClassSelector.ChangeClass( Agent )
			}
			FNClasses.SgtWinter =>
			{
				ClassSgtWinter.ClassSelector.ChangeClass( Agent )
			}
			FNClasses.CubeAssassin =>
			{
				ClassCubeAssassin.ClassSelector.ChangeClass( Agent )
			}
			FNClasses.SpireAssassin =>
			{
				ClassSpireAssassin.ClassSelector.ChangeClass( Agent )
			}
			FNClasses.ZygAndChoppy =>
			{
				ClassZygAndChoppy.ClassSelector.ChangeClass( Agent )
			}
			FNClasses.CubeBlue =>
			{
				ClassCubeBlue.ClassSelector.ChangeClass( Agent )
			}
			FNClasses.Zeus =>
			{
				ClassZeus.ClassSelector.ChangeClass( Agent )
			}
			FNClasses.Ares =>
			{
				ClassAres.ClassSelector.ChangeClass( Agent )
			}
			FNClasses.Cerberus =>
			{
				ClassCerberus.ClassSelector.ChangeClass( Agent )
			}
			FNClasses.Hades =>
			{
				ClassHades.ClassSelector.ChangeClass( Agent )
			}
			FNClasses.MegaloDon =>
			{
				ClassMegaloDon.ClassSelector.ChangeClass( Agent )
			}
			FNClasses.RingmasterScarr =>
			{
				ClassRingmasterScarr.ClassSelector.ChangeClass( Agent )
			}
			FNClasses.TheMachinist =>
			{
				ClassTheMachinist.ClassSelector.ChangeClass( Agent )
			}
			FNClasses.UltimaCarver =>
			{
				ClassUltimaCarver.ClassSelector.ChangeClass( Agent )
			}
			FNClasses.ShogunX =>
			{
				ClassShogunX.ClassSelector.ChangeClass( Agent )
			}
			FNClasses.NightRose =>
			{
				ClassNightRose.ClassSelector.ChangeClass( Agent )
			}
			FNClasses.Puppeteer =>
			{
				ClassPuppeteer.ClassSelector.ChangeClass( Agent )
			}
			FNClasses.FletcherKane =>
			{
				ClassFletcherKane.ClassSelector.ChangeClass( Agent )
			}
			FNClasses.AlphaHeavy =>
			{
				ClassAlphaHeavy.ClassSelector.ChangeClass( Agent )
			}
			FNClasses.Jones =>
			{
				ClassJones.ClassSelector.ChangeClass( Agent )
			}
			FNClasses.DoctorDoom =>
			{
				ClassDoctorDoom.ClassSelector.ChangeClass( Agent )
			}
			FNClasses.DoctorDoomMythic =>
			{
				ClassDoctorDoomMythic.ClassSelector.ChangeClass( Agent )
			}
			FNClasses.Caretaker =>
			{
				ClassCaretaker.ClassSelector.ChangeClass( Agent )
			}
			FNClasses.Hope =>
			{
				ClassHope.ClassSelector.ChangeClass( Agent )
			}
			FNClasses.Daigo =>
			{
				ClassDaigo.ClassSelector.ChangeClass( Agent )
			}
			FNClasses.Kor =>
			{
				ClassKor.ClassSelector.ChangeClass( Agent )
			}
			_ =>
			{
				ClassMidas.ClassSelector.ChangeClass( Agent )
			}
		}

		spawn{UpdatePlayerHUD(Agent)}
		#PlayClassVO(Agent, FNVOType.Idle)
		#ClassSwitchedEvent.Signal(Agent
	}

	# all sound files
    GetClassVO(FNClass:FNClasses, VOType:FNVOType):[]sound_wave=
	{
		case( FNClass )
		{
			FNClasses.Midas =>
			{
				case( VOType )
				{
					FNVOType.Idle => 
					{ return array{
							Audio.VO.Midas.Henchman_Midas_Idle_010_asset,
							Audio.VO.Midas.Henchman_Midas_Idle_020_asset,
							Audio.VO.Midas.Henchman_Midas_Idle_030_asset,
							Audio.VO.Midas.Henchman_Midas_Idle_040_asset,
							Audio.VO.Midas.Henchman_Midas_Idle_050_asset,
							Audio.VO.Midas.Henchman_Midas_Idle_060_asset
						}
					}
					FNVOType.Sus =>
					{ return array{
							Audio.VO.Midas.Henchman_Midas_Suspicious_010_asset,
							Audio.VO.Midas.Henchman_Midas_Suspicious_020_asset,
							Audio.VO.Midas.Henchman_Midas_Suspicious_030_asset,
							Audio.VO.Midas.Henchman_Midas_Suspicious_040_asset,
							Audio.VO.Midas.Henchman_Midas_Suspicious_050_asset,
							Audio.VO.Midas.Henchman_Midas_Suspicious_060_asset
						}
					}
					FNVOType.ReturnIdle =>
					{ return array{
							Audio.VO.Midas.Henchman_Midas_SuspiciousReturnToIdle_010_asset,
							Audio.VO.Midas.Henchman_Midas_SuspiciousReturnToIdle_020_asset,
							Audio.VO.Midas.Henchman_Midas_SuspiciousReturnToIdle_030_asset,
							Audio.VO.Midas.Henchman_Midas_SuspiciousReturnToIdle_040_asset,
							Audio.VO.Midas.Henchman_Midas_SuspiciousReturnToIdle_050_asset,
							Audio.VO.Midas.Henchman_Midas_SuspiciousReturnToIdle_060_asset
						}
					}
					FNVOType.FullAlert =>
					{ return array{
							Audio.VO.Midas.Henchman_Midas_FullAlert_010_asset,
							Audio.VO.Midas.Henchman_Midas_FullAlert_020_asset,
							Audio.VO.Midas.Henchman_Midas_FullAlert_030_asset,
							Audio.VO.Midas.Henchman_Midas_FullAlert_040_asset,
							Audio.VO.Midas.Henchman_Midas_FullAlert_050_asset,
							Audio.VO.Midas.Henchman_Midas_FullAlert_060_asset,
							Audio.VO.Midas.Henchman_Midas_FullAlert_070_asset
						}
					}
					FNVOType.FullAlertLost =>
					{ return array{
							Audio.VO.Midas.Henchman_Midas_FullAlertLostTarget_010_asset,
							Audio.VO.Midas.Henchman_Midas_FullAlertLostTarget_020_asset,
							Audio.VO.Midas.Henchman_Midas_FullAlertLostTarget_030_asset,
							Audio.VO.Midas.Henchman_Midas_FullAlertLostTarget_040_asset,
							Audio.VO.Midas.Henchman_Midas_FullAlertLostTarget_050_asset,
							Audio.VO.Midas.Henchman_Midas_FullAlertLostTarget_060_asset
						}
					}
					FNVOType.DBNO =>
					{ return array{
							Audio.VO.Midas.Henchman_Midas_DBNO_Enter_010_asset,
							Audio.VO.Midas.Henchman_Midas_DBNO_Enter_020_asset,
							Audio.VO.Midas.Henchman_Midas_DBNO_Enter_030_asset,
							Audio.VO.Midas.Henchman_Midas_DBNO_Enter_040_asset,
							Audio.VO.Midas.Henchman_Midas_DBNO_Enter_050_asset
						}
					}
					FNVOType.Interrogation =>
					{ return array{
							Audio.VO.Midas.Henchman_Midas_Interrogation_010_asset,
							Audio.VO.Midas.Henchman_Midas_Interrogation_020_asset,
							Audio.VO.Midas.Henchman_Midas_Interrogation_030_asset,
							Audio.VO.Midas.Henchman_Midas_Interrogation_040_asset,
							Audio.VO.Midas.Henchman_Midas_Interrogation_050_asset,
							Audio.VO.Midas.Henchman_Midas_Interrogation_060_asset

						}
					}
					FNVOType.Ability =>	
					{ return array{
							Audio.VO.Midas.Henchman_Midas_FullAlert_010_asset,
							Audio.VO.Midas.Henchman_Midas_FullAlert_020_asset,
							Audio.VO.Midas.Henchman_Midas_FullAlert_030_asset,
							Audio.VO.Midas.Henchman_Midas_FullAlert_040_asset,
							Audio.VO.Midas.Henchman_Midas_FullAlert_050_asset,
							Audio.VO.Midas.Henchman_Midas_FullAlert_060_asset,
							Audio.VO.Midas.Henchman_Midas_FullAlert_070_asset
						}
					}
					FNVOType.AbilityEnd =>
					{ return array{
							Audio.VO.Midas.Henchman_Midas_FullAlert_010_asset,
							Audio.VO.Midas.Henchman_Midas_FullAlert_020_asset,
							Audio.VO.Midas.Henchman_Midas_FullAlert_030_asset,
							Audio.VO.Midas.Henchman_Midas_FullAlert_040_asset,
							Audio.VO.Midas.Henchman_Midas_FullAlert_050_asset,
							Audio.VO.Midas.Henchman_Midas_FullAlert_060_asset,
							Audio.VO.Midas.Henchman_Midas_FullAlert_070_asset
						}
					}
					FNVOType.Ult =>
					{ return array{
							Audio.VO.Midas.VO_Midas_Ult_asset
						}
					}
				}
			}
			FNClasses.TNTina =>
			{
				case( VOType )
				{
					FNVOType.Idle => 
					{ return array{
							Audio.VO.TNTina.Henchman_TnTina_FullAlert_010_asset,
							Audio.VO.TNTina.Henchman_TnTina_FullAlert_020_asset,
							Audio.VO.TNTina.Henchman_TnTina_FullAlert_030_asset,
							Audio.VO.TNTina.Henchman_TnTina_FullAlert_040_asset,
							Audio.VO.TNTina.Henchman_TnTina_FullAlert_050_asset,
							Audio.VO.TNTina.Henchman_TnTina_FullAlert_060_asset,
							Audio.VO.TNTina.Henchman_TnTina_FullAlert_070_asset
						}
					}
					FNVOType.Sus =>
					{ return array{
							Audio.VO.TNTina.Henchman_TnTina_Suspicious_010_asset,
							Audio.VO.TNTina.Henchman_TnTina_Suspicious_020_asset,
							Audio.VO.TNTina.Henchman_TnTina_Suspicious_030_asset,
							Audio.VO.TNTina.Henchman_TnTina_Suspicious_040_asset,
							Audio.VO.TNTina.Henchman_TnTina_Suspicious_050_asset,
							Audio.VO.TNTina.Henchman_TnTina_Suspicious_060_asset

						}
					}
					FNVOType.ReturnIdle =>
					{ return array{
							Audio.VO.TNTina.Henchman_TnTina_SuspiciousReturnToIdle_010_asset,
							Audio.VO.TNTina.Henchman_TnTina_SuspiciousReturnToIdle_020_asset,
							Audio.VO.TNTina.Henchman_TnTina_SuspiciousReturnToIdle_030_asset,
							Audio.VO.TNTina.Henchman_TnTina_SuspiciousReturnToIdle_040_asset,
							Audio.VO.TNTina.Henchman_TnTina_SuspiciousReturnToIdle_050_asset,
							Audio.VO.TNTina.Henchman_TnTina_SuspiciousReturnToIdle_060_asset
						}
					}
					FNVOType.FullAlert =>
					{ return array{
							Audio.VO.TNTina.Henchman_TnTina_FullAlert_010_asset,
							Audio.VO.TNTina.Henchman_TnTina_FullAlert_020_asset,
							Audio.VO.TNTina.Henchman_TnTina_FullAlert_030_asset,
							Audio.VO.TNTina.Henchman_TnTina_FullAlert_040_asset,
							Audio.VO.TNTina.Henchman_TnTina_FullAlert_050_asset,
							Audio.VO.TNTina.Henchman_TnTina_FullAlert_060_asset,
							Audio.VO.TNTina.Henchman_TnTina_FullAlert_070_asset

						}
					}
					FNVOType.FullAlertLost =>
					{ return array{
							Audio.VO.TNTina.Henchman_TnTina_FullAlertLostTarget_010_asset,
							Audio.VO.TNTina.Henchman_TnTina_FullAlertLostTarget_020_asset,
							Audio.VO.TNTina.Henchman_TnTina_FullAlertLostTarget_030_asset,
							Audio.VO.TNTina.Henchman_TnTina_FullAlertLostTarget_040_asset,
							Audio.VO.TNTina.Henchman_TnTina_FullAlertLostTarget_050_asset,
							Audio.VO.TNTina.Henchman_TnTina_FullAlertLostTarget_060_asset,
							Audio.VO.TNTina.Henchman_TnTina_FullAlertLostTarget_070_asset
						}
					}
					FNVOType.DBNO =>
					{ return array{
							Audio.VO.TNTina.Henchman_TnTina_DBNO_Enter_010_asset,
							Audio.VO.TNTina.Henchman_TnTina_DBNO_Enter_020_asset,
							Audio.VO.TNTina.Henchman_TnTina_DBNO_Enter_030_asset,
							Audio.VO.TNTina.Henchman_TnTina_DBNO_Enter_040_asset,
							Audio.VO.TNTina.Henchman_TnTina_DBNO_Enter_050_asset,
							Audio.VO.TNTina.Henchman_TnTina_DBNO_Enter_060_asset
						}
					}
					FNVOType.Interrogation =>
					{ return array{
							Audio.VO.TNTina.Henchman_TnTina_Interrogation_010_asset,
							Audio.VO.TNTina.Henchman_TnTina_Interrogation_020_asset,
							Audio.VO.TNTina.Henchman_TnTina_Interrogation_030_asset,
							Audio.VO.TNTina.Henchman_TnTina_Interrogation_040_asset,
							Audio.VO.TNTina.Henchman_TnTina_Interrogation_050_asset,
							Audio.VO.TNTina.Henchman_TnTina_Interrogation_060_asset
						}
					}
					FNVOType.Ability =>	
					{ return array{
							Audio.VO.TNTina.Henchman_TnTina_FullAlert_010_asset,
							Audio.VO.TNTina.Henchman_TnTina_FullAlert_020_asset,
							Audio.VO.TNTina.Henchman_TnTina_FullAlert_030_asset,
							Audio.VO.TNTina.Henchman_TnTina_FullAlert_040_asset,
							Audio.VO.TNTina.Henchman_TnTina_FullAlert_050_asset,
							Audio.VO.TNTina.Henchman_TnTina_FullAlert_060_asset,
							Audio.VO.TNTina.Henchman_TnTina_FullAlert_070_asset
						}
					}
					FNVOType.AbilityEnd =>
					{ return array{
							Audio.VO.TNTina.Henchman_TnTina_FullAlert_010_asset,
							Audio.VO.TNTina.Henchman_TnTina_FullAlert_020_asset,
							Audio.VO.TNTina.Henchman_TnTina_FullAlert_030_asset,
							Audio.VO.TNTina.Henchman_TnTina_FullAlert_040_asset,
							Audio.VO.TNTina.Henchman_TnTina_FullAlert_050_asset,
							Audio.VO.TNTina.Henchman_TnTina_FullAlert_060_asset,
							Audio.VO.TNTina.Henchman_TnTina_FullAlert_070_asset
						}
					}
					FNVOType.Ult =>
					{ return array{
							Audio.VO.TNTina.Henchman_TnTina_FullAlert_010_asset,
							Audio.VO.TNTina.Henchman_TnTina_FullAlert_020_asset,
							Audio.VO.TNTina.Henchman_TnTina_FullAlert_030_asset,
							Audio.VO.TNTina.Henchman_TnTina_FullAlert_040_asset,
							Audio.VO.TNTina.Henchman_TnTina_FullAlert_050_asset,
							Audio.VO.TNTina.Henchman_TnTina_FullAlert_060_asset,
							Audio.VO.TNTina.Henchman_TnTina_FullAlert_070_asset
						}
					}
				}
			}
			FNClasses.Brutus =>
			{
				case( VOType )
				{
					FNVOType.Idle => 
					{ return array{
							Audio.VO.Brutus.HenchmanHeavy_Idle_010_asset,
							Audio.VO.Brutus.HenchmanHeavy_Idle_020_asset,
							Audio.VO.Brutus.HenchmanHeavy_Idle_030_asset,
							Audio.VO.Brutus.HenchmanHeavy_Idle_040_asset,
							Audio.VO.Brutus.HenchmanHeavy_Idle_050_asset,
							Audio.VO.Brutus.HenchmanHeavy_Idle_060_asset
						}
					}
					FNVOType.Sus =>
					{ return array{
							Audio.VO.Brutus.Henchman_Heavy_Suspicious_010_asset,
							Audio.VO.Brutus.Henchman_Heavy_Suspicious_020_asset,
							Audio.VO.Brutus.Henchman_Heavy_Suspicious_030_asset,
							Audio.VO.Brutus.Henchman_Heavy_Suspicious_040_asset,
							Audio.VO.Brutus.Henchman_Heavy_Suspicious_050_asset,
							Audio.VO.Brutus.Henchman_Heavy_Suspicious_060_asset
						}
					}
					FNVOType.ReturnIdle =>
					{ return array{
							Audio.VO.Brutus.HenchmanHeavy_SuspiciousReturnToIdle_010_asset,
							Audio.VO.Brutus.HenchmanHeavy_SuspiciousReturnToIdle_020_asset,
							Audio.VO.Brutus.HenchmanHeavy_SuspiciousReturnToIdle_030_asset,
							Audio.VO.Brutus.HenchmanHeavy_SuspiciousReturnToIdle_040_asset,
							Audio.VO.Brutus.HenchmanHeavy_SuspiciousReturnToIdle_050_asset,
							Audio.VO.Brutus.HenchmanHeavy_SuspiciousReturnToIdle_060_asset
						}
					}
					FNVOType.FullAlert =>
					{ return array{
							Audio.VO.Brutus.HenchmanHeavy_FullAlert_010_asset,
							Audio.VO.Brutus.HenchmanHeavy_FullAlert_020_asset,
							Audio.VO.Brutus.HenchmanHeavy_FullAlert_030_asset,
							Audio.VO.Brutus.HenchmanHeavy_FullAlert_040_asset,
							Audio.VO.Brutus.HenchmanHeavy_FullAlert_050_asset,
							Audio.VO.Brutus.HenchmanHeavy_FullAlert_060_asset
						}
					}
					FNVOType.FullAlertLost =>
					{ return array{
							Audio.VO.Brutus.HenchmanHeavy_FullAlertLostTarget_010_asset,
							Audio.VO.Brutus.HenchmanHeavy_FullAlertLostTarget_020_asset,
							Audio.VO.Brutus.HenchmanHeavy_FullAlertLostTarget_030_asset,
							Audio.VO.Brutus.HenchmanHeavy_FullAlertLostTarget_040_asset,
							Audio.VO.Brutus.HenchmanHeavy_FullAlertLostTarget_050_asset,
							Audio.VO.Brutus.HenchmanHeavy_FullAlertLostTarget_060_asset
						}
					}
					FNVOType.DBNO =>
					{ return array{
							Audio.VO.Brutus.HenchmanHeavy_DBNO_Enter_010_asset,
							Audio.VO.Brutus.HenchmanHeavy_DBNO_Enter_020_asset,
							Audio.VO.Brutus.HenchmanHeavy_DBNO_Enter_030_asset,
							Audio.VO.Brutus.HenchmanHeavy_DBNO_Enter_040_asset,
							Audio.VO.Brutus.HenchmanHeavy_DBNO_Enter_050_asset,
							Audio.VO.Brutus.HenchmanHeavy_DBNO_Enter_060_asset,
							Audio.VO.Brutus.HenchmanHeavy_DBNO_Enter_070_asset
						}
					}
					FNVOType.Interrogation =>
					{ return array{
							Audio.VO.Brutus.HenchmanHeavy_Interrogation_010_asset,
							Audio.VO.Brutus.HenchmanHeavy_Interrogation_020_asset,
							Audio.VO.Brutus.HenchmanHeavy_Interrogation_030_asset,
							Audio.VO.Brutus.HenchmanHeavy_Interrogation_040_asset,
							Audio.VO.Brutus.HenchmanHeavy_Interrogation_050_asset,
							Audio.VO.Brutus.HenchmanHeavy_Interrogation_060_asset,
							Audio.VO.Brutus.HenchmanHeavy_Interrogation_070_asset
						}
					}
					FNVOType.Ability =>	
					{ return array{
							Audio.VO.Brutus.HenchmanHeavy_FullAlert_010_asset,
							Audio.VO.Brutus.HenchmanHeavy_FullAlert_020_asset,
							Audio.VO.Brutus.HenchmanHeavy_FullAlert_030_asset,
							Audio.VO.Brutus.HenchmanHeavy_FullAlert_040_asset,
							Audio.VO.Brutus.HenchmanHeavy_FullAlert_050_asset,
							Audio.VO.Brutus.HenchmanHeavy_FullAlert_060_asset
						}
					}
					FNVOType.AbilityEnd =>
					{ return array{
							Audio.VO.Brutus.HenchmanHeavy_FullAlert_010_asset,
							Audio.VO.Brutus.HenchmanHeavy_FullAlert_020_asset,
							Audio.VO.Brutus.HenchmanHeavy_FullAlert_030_asset,
							Audio.VO.Brutus.HenchmanHeavy_FullAlert_040_asset,
							Audio.VO.Brutus.HenchmanHeavy_FullAlert_050_asset,
							Audio.VO.Brutus.HenchmanHeavy_FullAlert_060_asset
						}
					}
					FNVOType.Ult =>
					{ return array{
							Audio.VO.Brutus.HenchmanHeavy_FullAlert_010_asset,
							Audio.VO.Brutus.HenchmanHeavy_FullAlert_020_asset,
							Audio.VO.Brutus.HenchmanHeavy_FullAlert_030_asset,
							Audio.VO.Brutus.HenchmanHeavy_FullAlert_040_asset,
							Audio.VO.Brutus.HenchmanHeavy_FullAlert_050_asset,
							Audio.VO.Brutus.HenchmanHeavy_FullAlert_060_asset
						}
					}
				}
			}
			FNClasses.Skye =>
			{
				case( VOType )
				{
					FNVOType.Idle => 
					{ return array{
							Audio.VO.Skye.Henchman_AdvGrl_Idle_010_asset,
							Audio.VO.Skye.Henchman_AdvGrl_Idle_020_asset,
							Audio.VO.Skye.Henchman_AdvGrl_Idle_030_asset,
							Audio.VO.Skye.Henchman_AdvGrl_Idle_040_asset,
							Audio.VO.Skye.Henchman_AdvGrl_Idle_050_asset,
							Audio.VO.Skye.Henchman_AdvGrl_Idle_060_asset
						}
					}
					FNVOType.Sus =>
					{ return array{
							Audio.VO.Skye.Henchman_AdvGrl_Suspicious_010_asset,
							Audio.VO.Skye.Henchman_AdvGrl_Suspicious_020_asset,
							Audio.VO.Skye.Henchman_AdvGrl_Suspicious_030_asset,
							Audio.VO.Skye.Henchman_AdvGrl_Suspicious_040_asset,
							Audio.VO.Skye.Henchman_AdvGrl_Suspicious_050_asset,
							Audio.VO.Skye.Henchman_AdvGrl_Suspicious_060_asset
						}
					}
					FNVOType.ReturnIdle =>
					{ return array{
							Audio.VO.Skye.Henchman_AdvGrl_SuspiciousReturnToIdle_010_asset,
							Audio.VO.Skye.Henchman_AdvGrl_SuspiciousReturnToIdle_020_asset,
							Audio.VO.Skye.Henchman_AdvGrl_SuspiciousReturnToIdle_030_asset,
							Audio.VO.Skye.Henchman_AdvGrl_SuspiciousReturnToIdle_040_asset,
							Audio.VO.Skye.Henchman_AdvGrl_SuspiciousReturnToIdle_050_asset,
							Audio.VO.Skye.Henchman_AdvGrl_SuspiciousReturnToIdle_060_asset
						}
					}
					FNVOType.FullAlert =>
					{ return array{
							Audio.VO.Skye.Henchman_AdvGrl_FullAlert_010_asset,
							Audio.VO.Skye.Henchman_AdvGrl_FullAlert_020_asset,
							Audio.VO.Skye.Henchman_AdvGrl_FullAlert_030_asset,
							Audio.VO.Skye.Henchman_AdvGrl_FullAlert_040_asset,
							Audio.VO.Skye.Henchman_AdvGrl_FullAlert_050_asset,
							Audio.VO.Skye.Henchman_AdvGrl_FullAlert_060_asset
						}
					}
					FNVOType.FullAlertLost =>
					{ return array{
							Audio.VO.Skye.Henchman_AdvGrl_FullAlertLostTarget_010_asset,
							Audio.VO.Skye.Henchman_AdvGrl_FullAlertLostTarget_020_asset,
							Audio.VO.Skye.Henchman_AdvGrl_FullAlertLostTarget_030_asset,
							Audio.VO.Skye.Henchman_AdvGrl_FullAlertLostTarget_040_asset,
							Audio.VO.Skye.Henchman_AdvGrl_FullAlertLostTarget_050_asset,
							Audio.VO.Skye.Henchman_AdvGrl_FullAlertLostTarget_060_asset,
							Audio.VO.Skye.Henchman_AdvGrl_FullAlertLostTarget_070_asset
						}
					}
					FNVOType.DBNO =>
					{ return array{
							Audio.VO.Skye.Henchman_AdvGrl_DBNO_Enter_010_asset,
							Audio.VO.Skye.Henchman_AdvGrl_DBNO_Enter_020_asset,
							Audio.VO.Skye.Henchman_AdvGrl_DBNO_Enter_030_asset,
							Audio.VO.Skye.Henchman_AdvGrl_DBNO_Enter_040_asset,
							Audio.VO.Skye.Henchman_AdvGrl_DBNO_Enter_050_asset,
							Audio.VO.Skye.Henchman_AdvGrl_DBNO_Enter_060_asset
						}
					}
					FNVOType.Interrogation =>
					{ return array{
							Audio.VO.Skye.Henchman_AdvGrl_Interrogation_010_asset,
							Audio.VO.Skye.Henchman_AdvGrl_Interrogation_020_asset,
							Audio.VO.Skye.Henchman_AdvGrl_Interrogation_030_asset,
							Audio.VO.Skye.Henchman_AdvGrl_Interrogation_040_asset,
							Audio.VO.Skye.Henchman_AdvGrl_Interrogation_050_asset,
							Audio.VO.Skye.Henchman_AdvGrl_Interrogation_060_asset
						}
					}
					FNVOType.Ability =>	
					{ return array{
							Audio.VO.Skye.Henchman_AdvGrl_FullAlert_010_asset,
							Audio.VO.Skye.Henchman_AdvGrl_FullAlert_020_asset,
							Audio.VO.Skye.Henchman_AdvGrl_FullAlert_030_asset,
							Audio.VO.Skye.Henchman_AdvGrl_FullAlert_040_asset,
							Audio.VO.Skye.Henchman_AdvGrl_FullAlert_050_asset,
							Audio.VO.Skye.Henchman_AdvGrl_FullAlert_060_asset
						}
					}
					FNVOType.AbilityEnd =>
					{ return array{
							Audio.VO.Skye.Henchman_AdvGrl_FullAlert_010_asset,
							Audio.VO.Skye.Henchman_AdvGrl_FullAlert_020_asset,
							Audio.VO.Skye.Henchman_AdvGrl_FullAlert_030_asset,
							Audio.VO.Skye.Henchman_AdvGrl_FullAlert_040_asset,
							Audio.VO.Skye.Henchman_AdvGrl_FullAlert_050_asset,
							Audio.VO.Skye.Henchman_AdvGrl_FullAlert_060_asset
						}
					}
					FNVOType.Ult =>
					{ return array{
							Audio.VO.Skye.Henchman_AdvGrl_FullAlert_010_asset,
							Audio.VO.Skye.Henchman_AdvGrl_FullAlert_020_asset,
							Audio.VO.Skye.Henchman_AdvGrl_FullAlert_030_asset,
							Audio.VO.Skye.Henchman_AdvGrl_FullAlert_040_asset,
							Audio.VO.Skye.Henchman_AdvGrl_FullAlert_050_asset,
							Audio.VO.Skye.Henchman_AdvGrl_FullAlert_060_asset
						}
					}
				}
			}
			FNClasses.Meowscles =>
			{
				case( VOType )
				{
					FNVOType.Idle => 
					{ return array{
							Audio.VO.Meowscles.Meowy_Idle_010_asset,
							Audio.VO.Meowscles.Meowy_Idle_020_asset,
							Audio.VO.Meowscles.Meowy_Idle_030_asset,
							Audio.VO.Meowscles.Meowy_Idle_040_asset,
							Audio.VO.Meowscles.Meowy_Idle_050_asset,
							Audio.VO.Meowscles.Meowy_Idle_060_asset,
							Audio.VO.Meowscles.Meowy_Idle_070_asset,
							Audio.VO.Meowscles.Meowy_Idle_080_asset,
							Audio.VO.Meowscles.Meowy_Idle_090_asset,
							Audio.VO.Meowscles.Meowy_Idle_100_asset
						}
					}
					FNVOType.Sus =>
					{ return array{
							Audio.VO.Meowscles.Meowy_Suspicious_010_asset,
							Audio.VO.Meowscles.Meowy_Suspicious_020_asset,
							Audio.VO.Meowscles.Meowy_Suspicious_030_asset,
							Audio.VO.Meowscles.Meowy_Suspicious_040_asset,
							Audio.VO.Meowscles.Meowy_Suspicious_050_asset,
							Audio.VO.Meowscles.Meowy_Suspicious_060_asset
						}
					}
					FNVOType.ReturnIdle =>
					{ return array{
							Audio.VO.Meowscles.Meowy_ReturnToIdle_010_asset,
							Audio.VO.Meowscles.Meowy_ReturnToIdle_020_asset,
							Audio.VO.Meowscles.Meowy_ReturnToIdle_030_asset,
							Audio.VO.Meowscles.Meowy_ReturnToIdle_040_asset,
							Audio.VO.Meowscles.Meowy_ReturnToIdle_050_asset,
							Audio.VO.Meowscles.Meowy_ReturnToIdle_060_asset

						}
					}
					FNVOType.FullAlert =>
					{ return array{
							Audio.VO.Meowscles.Meowy_Alerted_010_asset,
							Audio.VO.Meowscles.Meowy_Alerted_020_asset,
							Audio.VO.Meowscles.Meowy_Alerted_030_asset,
							Audio.VO.Meowscles.Meowy_Alerted_040_asset,
							Audio.VO.Meowscles.Meowy_Alerted_050_asset,
							Audio.VO.Meowscles.Meowy_Alerted_060_asset
						}
					}
					FNVOType.FullAlertLost =>
					{ return array{
							Audio.VO.Meowscles.Meowy_AlertLostTarget_010_asset,
							Audio.VO.Meowscles.Meowy_AlertLostTarget_020_asset,
							Audio.VO.Meowscles.Meowy_AlertLostTarget_030_asset,
							Audio.VO.Meowscles.Meowy_AlertLostTarget_040_asset,
							Audio.VO.Meowscles.Meowy_AlertLostTarget_050_asset,
							Audio.VO.Meowscles.Meowy_AlertLostTarget_060_asset
						}
					}
					FNVOType.DBNO =>
					{ return array{
							Audio.VO.Meowscles.Meowy_DBNO_Enter_010_asset,
							Audio.VO.Meowscles.Meowy_DBNO_Enter_020_asset,
							Audio.VO.Meowscles.Meowy_DBNO_Enter_030_asset,
							Audio.VO.Meowscles.Meowy_DBNO_Enter_040_asset,
							Audio.VO.Meowscles.Meowy_DBNO_Enter_050_asset,
							Audio.VO.Meowscles.Meowy_DBNO_Enter_060_asset
						}
					}
					FNVOType.Interrogation =>
					{ return array{
							Audio.VO.Meowscles.Meowy_Interrogation_010_asset,
							Audio.VO.Meowscles.Meowy_Interrogation_020_asset,
							Audio.VO.Meowscles.Meowy_Interrogation_030_asset,
							Audio.VO.Meowscles.Meowy_Interrogation_040_asset,
							Audio.VO.Meowscles.Meowy_Interrogation_050_asset,
							Audio.VO.Meowscles.Meowy_Interrogation_060_asset
						}
					}
					FNVOType.Ability =>	
					{ return array{
							Audio.VO.Meowscles.Meowy_Alerted_010_asset,
							Audio.VO.Meowscles.Meowy_Alerted_020_asset,
							Audio.VO.Meowscles.Meowy_Alerted_030_asset,
							Audio.VO.Meowscles.Meowy_Alerted_040_asset,
							Audio.VO.Meowscles.Meowy_Alerted_050_asset,
							Audio.VO.Meowscles.Meowy_Alerted_060_asset
						}
					}
					FNVOType.AbilityEnd =>
					{ return array{
							Audio.VO.Meowscles.Meowy_Alerted_010_asset,
							Audio.VO.Meowscles.Meowy_Alerted_020_asset,
							Audio.VO.Meowscles.Meowy_Alerted_030_asset,
							Audio.VO.Meowscles.Meowy_Alerted_040_asset,
							Audio.VO.Meowscles.Meowy_Alerted_050_asset,
							Audio.VO.Meowscles.Meowy_Alerted_060_asset
						}
					}
					FNVOType.Ult =>
					{ return array{
							Audio.VO.Meowscles.Meowy_Alerted_010_asset,
							Audio.VO.Meowscles.Meowy_Alerted_020_asset,
							Audio.VO.Meowscles.Meowy_Alerted_030_asset,
							Audio.VO.Meowscles.Meowy_Alerted_040_asset,
							Audio.VO.Meowscles.Meowy_Alerted_050_asset,
							Audio.VO.Meowscles.Meowy_Alerted_060_asset
						}
					}
				}
			}
			FNClasses.Jules =>
			{
				case( VOType )
				{
					FNVOType.Idle => 
					{ return array{
						   Audio.VO.Jules.TheEngineer_GibsonBoss_IdleGeneric_010_asset,
						   Audio.VO.Jules.TheEngineer_GibsonBoss_IdleGeneric_020_asset,
						   Audio.VO.Jules.TheEngineer_GibsonBoss_IdleGeneric_040_asset,
						   Audio.VO.Jules.TheEngineer_GibsonBoss_IdleGeneric_050_asset,
						   Audio.VO.Jules.TheEngineer_GibsonBoss_IdlePatrol_010_asset,
						   Audio.VO.Jules.TheEngineer_GibsonBoss_IdlePatrol_020_asset,
						   Audio.VO.Jules.TheEngineer_GibsonBoss_IdlePatrol_030_asset
						}
					}
					FNVOType.Sus =>
					{ return array{
						   Audio.VO.Jules.TheEngineer_GibsonBoss_Suspicious_010_asset,
						   Audio.VO.Jules.TheEngineer_GibsonBoss_Suspicious_020_asset,
						   Audio.VO.Jules.TheEngineer_GibsonBoss_Suspicious_021_asset,
						   Audio.VO.Jules.TheEngineer_GibsonBoss_Suspicious_030_asset,
						   Audio.VO.Jules.TheEngineer_GibsonBoss_Suspicious_031_asset,
						   Audio.VO.Jules.TheEngineer_GibsonBoss_Suspicious_040_asset
						}
					}
					FNVOType.ReturnIdle =>
					{ return array{
						   Audio.VO.Jules.TheEngineer_GibsonBoss_SuspiciousReturnToIdle_010_asset,
						   Audio.VO.Jules.TheEngineer_GibsonBoss_SuspiciousReturnToIdle_020_asset,
						   Audio.VO.Jules.TheEngineer_GibsonBoss_SuspiciousReturnToIdle_030_asset,
						   Audio.VO.Jules.TheEngineer_GibsonBoss_SuspiciousReturnToIdle_040_asset,
						   Audio.VO.Jules.TheEngineer_GibsonBoss_SuspiciousReturnToIdle_050_asset
						}
					}
					FNVOType.FullAlert =>
					{ return array{
						   Audio.VO.Jules.TheEngineer_GibsonBoss_FullAlert_010_asset,
						   Audio.VO.Jules.TheEngineer_GibsonBoss_FullAlert_011_asset,
						   Audio.VO.Jules.TheEngineer_GibsonBoss_FullAlert_021_asset,
						   Audio.VO.Jules.TheEngineer_GibsonBoss_FullAlert_030_asset,
						   Audio.VO.Jules.TheEngineer_GibsonBoss_FullAlert_041_asset
						}
					}
					FNVOType.FullAlertLost =>
					{ return array{
						   Audio.VO.Jules.TheEngineer_GibsonBoss_FullAlertLostTarget_010_asset,
						   Audio.VO.Jules.TheEngineer_GibsonBoss_FullAlertLostTarget_021_asset,
						   Audio.VO.Jules.TheEngineer_GibsonBoss_FullAlertLostTarget_031_asset,
						   Audio.VO.Jules.TheEngineer_GibsonBoss_FullAlertLostTarget_041_asset,
						   Audio.VO.Jules.TheEngineer_GibsonBoss_FullAlertLostTarget_050_asset
						}
					}
					FNVOType.DBNO =>
					{ return array{
						   Audio.VO.Jules.TheEngineer_GibsonBoss_DBNO_Enter_010_asset,
						   Audio.VO.Jules.TheEngineer_GibsonBoss_DBNO_Enter_011_asset,
						   Audio.VO.Jules.TheEngineer_GibsonBoss_DBNO_Enter_012_asset
						}
					}
					FNVOType.Interrogation =>
					{ return array{
						   Audio.VO.Jules.TheEngineer_GibsonBoss_DBNO_Enter_010_asset,
						   Audio.VO.Jules.TheEngineer_GibsonBoss_DBNO_Enter_011_asset,
						   Audio.VO.Jules.TheEngineer_GibsonBoss_DBNO_Enter_012_asset
						}
					}
					FNVOType.Ability =>	
					{ return array{
						   Audio.VO.Jules.TheEngineer_GibsonBoss_FullAlert_010_asset,
						   Audio.VO.Jules.TheEngineer_GibsonBoss_FullAlert_011_asset,
						   Audio.VO.Jules.TheEngineer_GibsonBoss_FullAlert_021_asset,
						   Audio.VO.Jules.TheEngineer_GibsonBoss_FullAlert_030_asset,
						   Audio.VO.Jules.TheEngineer_GibsonBoss_FullAlert_041_asset
						}
					}
					FNVOType.AbilityEnd =>
					{ return array{
						   Audio.VO.Jules.TheEngineer_GibsonBoss_FullAlert_010_asset,
						   Audio.VO.Jules.TheEngineer_GibsonBoss_FullAlert_011_asset,
						   Audio.VO.Jules.TheEngineer_GibsonBoss_FullAlert_021_asset,
						   Audio.VO.Jules.TheEngineer_GibsonBoss_FullAlert_030_asset,
						   Audio.VO.Jules.TheEngineer_GibsonBoss_FullAlert_041_asset
						}
					}
					FNVOType.Ult =>
					{ return array{
						   Audio.VO.Jules.TheEngineer_GibsonBoss_FullAlert_010_asset,
						   Audio.VO.Jules.TheEngineer_GibsonBoss_FullAlert_011_asset,
						   Audio.VO.Jules.TheEngineer_GibsonBoss_FullAlert_021_asset,
						   Audio.VO.Jules.TheEngineer_GibsonBoss_FullAlert_030_asset,
						   Audio.VO.Jules.TheEngineer_GibsonBoss_FullAlert_041_asset
						}
					}
				}
			}
			FNClasses.Kit =>
			{
				case( VOType )
				{
					FNVOType.Idle => 
					{ return array{
						   Audio.VO.Kit.MeowsclesJr_GibsonBoss_IdleConstruction_020_asset,
						   Audio.VO.Kit.MeowsclesJr_GibsonBoss_IdleConstruction_040_asset,
						   Audio.VO.Kit.MeowsclesJr_GibsonBoss_IdleGeneric_010_asset,
						   Audio.VO.Kit.MeowsclesJr_GibsonBoss_IdleGeneric_020_asset,
						   Audio.VO.Kit.MeowsclesJr_GibsonBoss_IdlePatrol_050_asset
						}
					}
					FNVOType.Sus =>
					{ return array{
						   Audio.VO.Kit.MeowsclesJr_GibsonBoss_Suspicious_020_asset,
						   Audio.VO.Kit.MeowsclesJr_GibsonBoss_Suspicious_021_asset,
						   Audio.VO.Kit.MeowsclesJr_GibsonBoss_Suspicious_031_asset,
						   Audio.VO.Kit.MeowsclesJr_GibsonBoss_Suspicious_041_asset,
						   Audio.VO.Kit.MeowsclesJr_GibsonBoss_Suspicious_051_asset
						}
					}
					FNVOType.ReturnIdle =>
					{ return array{
						   Audio.VO.Kit.MeowsclesJr_GibsonBoss_SuspiciousReturnToIdle_010_asset,
						   Audio.VO.Kit.MeowsclesJr_GibsonBoss_SuspiciousReturnToIdle_021_asset,
						   Audio.VO.Kit.MeowsclesJr_GibsonBoss_SuspiciousReturnToIdle_031_asset,
						   Audio.VO.Kit.MeowsclesJr_GibsonBoss_SuspiciousReturnToIdle_041_asset,
						   Audio.VO.Kit.MeowsclesJr_GibsonBoss_SuspiciousReturnToIdle_050_asset
						}
					}
					FNVOType.FullAlert =>
					{ return array{
						   Audio.VO.Kit.MeowsclesJr_GibsonBoss_FullAlert_011_asset,
						   Audio.VO.Kit.MeowsclesJr_GibsonBoss_FullAlert_021_asset,
						   Audio.VO.Kit.MeowsclesJr_GibsonBoss_FullAlert_030_asset,
						   Audio.VO.Kit.MeowsclesJr_GibsonBoss_FullAlert_031_asset,
						   Audio.VO.Kit.MeowsclesJr_GibsonBoss_FullAlert_040_asset
						}
					}
					FNVOType.FullAlertLost =>
					{ return array{
						   Audio.VO.Kit.MeowsclesJr_GibsonBoss_FullAlertLostTarget_010_asset,
						   Audio.VO.Kit.MeowsclesJr_GibsonBoss_FullAlertLostTarget_020_asset,
						   Audio.VO.Kit.MeowsclesJr_GibsonBoss_FullAlertLostTarget_041_asset,
						   Audio.VO.Kit.MeowsclesJr_GibsonBoss_FullAlertLostTarget_050_asset
						}
					}
					FNVOType.DBNO =>
					{ return array{
						   Audio.VO.Kit.MeowsclesJr_GibsonBoss_DBNO_Enter_010_asset,
						   Audio.VO.Kit.MeowsclesJr_GibsonBoss_DBNO_Enter_011_asset,
						   Audio.VO.Kit.MeowsclesJr_GibsonBoss_DBNO_Enter_012_asset
						}
					}
					FNVOType.Interrogation =>
					{ return array{
							Audio.VO.Midas.Henchman_Midas_Interrogation_010_asset,
							Audio.VO.Midas.Henchman_Midas_Interrogation_020_asset,
							Audio.VO.Midas.Henchman_Midas_Interrogation_030_asset,
							Audio.VO.Midas.Henchman_Midas_Interrogation_040_asset,
							Audio.VO.Midas.Henchman_Midas_Interrogation_050_asset,
							Audio.VO.Midas.Henchman_Midas_Interrogation_060_asset

						}
					}
					FNVOType.Ability =>	
					{ return array{
							Audio.VO.Midas.Henchman_Midas_FullAlert_010_asset,
							Audio.VO.Midas.Henchman_Midas_FullAlert_020_asset,
							Audio.VO.Midas.Henchman_Midas_FullAlert_030_asset,
							Audio.VO.Midas.Henchman_Midas_FullAlert_040_asset,
							Audio.VO.Midas.Henchman_Midas_FullAlert_050_asset,
							Audio.VO.Midas.Henchman_Midas_FullAlert_060_asset,
							Audio.VO.Midas.Henchman_Midas_FullAlert_070_asset
						}
					}
					FNVOType.AbilityEnd =>
					{ return array{
							Audio.VO.Midas.Henchman_Midas_FullAlert_010_asset,
							Audio.VO.Midas.Henchman_Midas_FullAlert_020_asset,
							Audio.VO.Midas.Henchman_Midas_FullAlert_030_asset,
							Audio.VO.Midas.Henchman_Midas_FullAlert_040_asset,
							Audio.VO.Midas.Henchman_Midas_FullAlert_050_asset,
							Audio.VO.Midas.Henchman_Midas_FullAlert_060_asset,
							Audio.VO.Midas.Henchman_Midas_FullAlert_070_asset
						}
					}
					FNVOType.Ult =>
					{ return array{
							Audio.VO.Midas.Henchman_Midas_FullAlert_010_asset,
							Audio.VO.Midas.Henchman_Midas_FullAlert_020_asset,
							Audio.VO.Midas.Henchman_Midas_FullAlert_030_asset,
							Audio.VO.Midas.Henchman_Midas_FullAlert_040_asset,
							Audio.VO.Midas.Henchman_Midas_FullAlert_050_asset,
							Audio.VO.Midas.Henchman_Midas_FullAlert_060_asset,
							Audio.VO.Midas.Henchman_Midas_FullAlert_070_asset
						}
					}
				}
			}
			FNClasses.Ocean =>
			{
				case( VOType )
				{
					FNVOType.Idle => 
					{ return array{
							Audio.VO.Ocean.Gibson_GOTHS_Idle_010_asset,
							Audio.VO.Ocean.Gibson_GOTHS_Idle_020_asset,
							Audio.VO.Ocean.Gibson_GOTHS_Idle_030_asset,
							Audio.VO.Ocean.Gibson_GOTHS_Idle_040_asset,
							Audio.VO.Ocean.Gibson_GOTHS_Idle_050_asset,
							Audio.VO.Ocean.Gibson_GOTHS_Idle_060_asset
						}
					}
					FNVOType.Sus =>
					{ return array{
							Audio.VO.Ocean.Gibson_GOTHS_Suspicious_010_asset,
							Audio.VO.Ocean.Gibson_GOTHS_Suspicious_020_asset,
							Audio.VO.Ocean.Gibson_GOTHS_Suspicious_030_asset,
							Audio.VO.Ocean.Gibson_GOTHS_Suspicious_040_asset,
							Audio.VO.Ocean.Gibson_GOTHS_Suspicious_050_asset,
							Audio.VO.Ocean.Gibson_GOTHS_Suspicious_060_asset,
							Audio.VO.Ocean.Gibson_GOTHS_Suspicious_070_asset
						}
					}
					FNVOType.ReturnIdle =>
					{ return array{
							Audio.VO.Ocean.Gibson_GOTHS_SuspiciousReturnToIdle_010_asset,
							Audio.VO.Ocean.Gibson_GOTHS_SuspiciousReturnToIdle_020_asset,
							Audio.VO.Ocean.Gibson_GOTHS_SuspiciousReturnToIdle_030_asset,
							Audio.VO.Ocean.Gibson_GOTHS_SuspiciousReturnToIdle_040_asset,
							Audio.VO.Ocean.Gibson_GOTHS_SuspiciousReturnToIdle_050_asset,
							Audio.VO.Ocean.Gibson_GOTHS_SuspiciousReturnToIdle_060_asset
						}
					}
					FNVOType.FullAlert =>
					{ return array{
							Audio.VO.Ocean.Gibson_GOTHS_FullAlert_010_asset,
							Audio.VO.Ocean.Gibson_GOTHS_FullAlert_020_asset,
							Audio.VO.Ocean.Gibson_GOTHS_FullAlert_030_asset,
							Audio.VO.Ocean.Gibson_GOTHS_FullAlert_040_asset,
							Audio.VO.Ocean.Gibson_GOTHS_FullAlert_050_asset,
							Audio.VO.Ocean.Gibson_GOTHS_FullAlert_060_asset
						}
					}
					FNVOType.FullAlertLost =>
					{ return array{
							Audio.VO.Ocean.Gibson_GOTHS_FullAlertLostTarget_010_asset,
							Audio.VO.Ocean.Gibson_GOTHS_FullAlertLostTarget_020_asset,
							Audio.VO.Ocean.Gibson_GOTHS_FullAlertLostTarget_030_asset,
							Audio.VO.Ocean.Gibson_GOTHS_FullAlertLostTarget_040_asset,
							Audio.VO.Ocean.Gibson_GOTHS_FullAlertLostTarget_050_asset
						}
					}
					FNVOType.DBNO =>
					{ return array{
							Audio.VO.Ocean.Gibson_GOTHS_DBNO_Enter_010_asset,
							Audio.VO.Ocean.Gibson_GOTHS_DBNO_Enter_020_asset,
							Audio.VO.Ocean.Gibson_GOTHS_DBNO_Enter_030_asset,
							Audio.VO.Ocean.Gibson_GOTHS_DBNO_Enter_040_asset,
							Audio.VO.Ocean.Gibson_GOTHS_DBNO_Enter_050_asset
						}
					}
					FNVOType.Interrogation =>
					{ return array{
							Audio.VO.Ocean.Gibson_GOTHS_DBNO_Enter_010_asset,
							Audio.VO.Ocean.Gibson_GOTHS_DBNO_Enter_020_asset,
							Audio.VO.Ocean.Gibson_GOTHS_DBNO_Enter_030_asset,
							Audio.VO.Ocean.Gibson_GOTHS_DBNO_Enter_040_asset,
							Audio.VO.Ocean.Gibson_GOTHS_DBNO_Enter_050_asset
						}
					}
					FNVOType.Ability =>	
					{ return array{
							Audio.VO.Ocean.Gibson_GOTHS_FullAlert_010_asset,
							Audio.VO.Ocean.Gibson_GOTHS_FullAlert_020_asset,
							Audio.VO.Ocean.Gibson_GOTHS_FullAlert_030_asset,
							Audio.VO.Ocean.Gibson_GOTHS_FullAlert_040_asset,
							Audio.VO.Ocean.Gibson_GOTHS_FullAlert_050_asset,
							Audio.VO.Ocean.Gibson_GOTHS_FullAlert_060_asset
						}
					}
					FNVOType.AbilityEnd =>
					{ return array{
							Audio.VO.Ocean.Gibson_GOTHS_FullAlert_010_asset,
							Audio.VO.Ocean.Gibson_GOTHS_FullAlert_020_asset,
							Audio.VO.Ocean.Gibson_GOTHS_FullAlert_030_asset,
							Audio.VO.Ocean.Gibson_GOTHS_FullAlert_040_asset,
							Audio.VO.Ocean.Gibson_GOTHS_FullAlert_050_asset,
							Audio.VO.Ocean.Gibson_GOTHS_FullAlert_060_asset
						}
					}
					FNVOType.Ult =>
					{ return array{
							Audio.VO.Ocean.Gibson_GOTHS_FullAlert_010_asset,
							Audio.VO.Ocean.Gibson_GOTHS_FullAlert_020_asset,
							Audio.VO.Ocean.Gibson_GOTHS_FullAlert_030_asset,
							Audio.VO.Ocean.Gibson_GOTHS_FullAlert_040_asset,
							Audio.VO.Ocean.Gibson_GOTHS_FullAlert_050_asset,
							Audio.VO.Ocean.Gibson_GOTHS_FullAlert_060_asset
						}
					}
				}
			}
			FNClasses.ShadowMidas =>
			{
				case( VOType )
				{
					FNVOType.Idle => 
					{ return array{
							Audio.VO.ShadowMidas.ShadowMidas_Idle_1_VO_asset,
							Audio.VO.ShadowMidas.ShadowMidas_Idle_2_VO_asset,
							Audio.VO.ShadowMidas.ShadowMidas_Idle_3_VO_asset,
							Audio.VO.ShadowMidas.ShadowMidas_Idle_4_VO_asset,
							Audio.VO.ShadowMidas.ShadowMidas_Idle_5_VO_asset,
							Audio.VO.ShadowMidas.ShadowMidas_Idle_6_VO_asset
						}
					}
					FNVOType.Sus =>
					{ return array{
							Audio.VO.ShadowMidas.ShadowMidas_Suspicious_1_VO_asset,
							Audio.VO.ShadowMidas.ShadowMidas_Suspicious_2_VO_asset,
							Audio.VO.ShadowMidas.ShadowMidas_Suspicious_3_VO_asset,
							Audio.VO.ShadowMidas.ShadowMidas_Suspicious_4_VO_asset,
							Audio.VO.ShadowMidas.ShadowMidas_Suspicious_5_VO_asset
						}
					}
					FNVOType.ReturnIdle =>
					{ return array{
							Audio.VO.ShadowMidas.ShadowMidas_Suspicious_to_Idle_1_VO_asset,
							Audio.VO.ShadowMidas.ShadowMidas_Suspicious_to_Idle_2_VO_asset,
							Audio.VO.ShadowMidas.ShadowMidas_Suspicious_to_Idle_3_VO_asset,
							Audio.VO.ShadowMidas.ShadowMidas_Suspicious_to_Idle_4_VO_asset
						}
					}
					FNVOType.FullAlert =>
					{ return array{
							Audio.VO.ShadowMidas.ShadowMidas_Full_Alert_1_VO_asset,
							Audio.VO.ShadowMidas.ShadowMidas_Full_Alert_2_VO_asset,
							Audio.VO.ShadowMidas.ShadowMidas_Full_Alert_3_VO_asset,
							Audio.VO.ShadowMidas.ShadowMidas_Full_Alert_4_VO_asset,
							Audio.VO.ShadowMidas.ShadowMidas_Full_Alert_5_VO_asset
						}
					}
					FNVOType.FullAlertLost =>
					{ return array{
							Audio.VO.ShadowMidas.ShadowMidas_Full_Alert_Lost_Target_1_VO_asset,
							Audio.VO.ShadowMidas.ShadowMidas_Full_Alert_Lost_Target_2_VO_asset,
							Audio.VO.ShadowMidas.ShadowMidas_Full_Alert_Lost_Target_3_VO_asset,
							Audio.VO.ShadowMidas.ShadowMidas_Full_Alert_Lost_Target_4_VO_asset,
							Audio.VO.ShadowMidas.ShadowMidas_Full_Alert_Lost_Target_5_VO_asset
						}
					}
					FNVOType.DBNO =>
					{ return array{
							Audio.VO.ShadowMidas.ShadowMidas_DBNO_Enter_1_VO_asset,
							Audio.VO.ShadowMidas.ShadowMidas_DBNO_Enter_2_VO_asset,
							Audio.VO.ShadowMidas.ShadowMidas_DBNO_Enter_3_VO_asset,
							Audio.VO.ShadowMidas.ShadowMidas_DBNO_Enter_4_VO_asset,
							Audio.VO.ShadowMidas.ShadowMidas_DBNO_Enter_5_VO_asset
						}
					}
					FNVOType.Interrogation =>
					{ return array{
							Audio.VO.ShadowMidas.ShadowMidas_Interrogation_1_VO_asset,
							Audio.VO.ShadowMidas.ShadowMidas_Interrogation_2_VO_asset,
							Audio.VO.ShadowMidas.ShadowMidas_Interrogation_4_VO_asset,
							Audio.VO.ShadowMidas.ShadowMidas_Interrogation_5_VO_asset

						}
					}
					FNVOType.Ability =>	
					{ return array{
							Audio.VO.ShadowMidas.ShadowMidas_Full_Alert_1_VO_asset,
							Audio.VO.ShadowMidas.ShadowMidas_Full_Alert_2_VO_asset,
							Audio.VO.ShadowMidas.ShadowMidas_Full_Alert_3_VO_asset,
							Audio.VO.ShadowMidas.ShadowMidas_Full_Alert_4_VO_asset,
							Audio.VO.ShadowMidas.ShadowMidas_Full_Alert_5_VO_asset
						}
					}
					FNVOType.AbilityEnd =>
					{ return array{
							Audio.VO.ShadowMidas.ShadowMidas_Full_Alert_1_VO_asset,
							Audio.VO.ShadowMidas.ShadowMidas_Full_Alert_2_VO_asset,
							Audio.VO.ShadowMidas.ShadowMidas_Full_Alert_3_VO_asset,
							Audio.VO.ShadowMidas.ShadowMidas_Full_Alert_4_VO_asset,
							Audio.VO.ShadowMidas.ShadowMidas_Full_Alert_5_VO_asset
						}
					}
					FNVOType.Ult =>
					{ return array{
							Audio.VO.ShadowMidas.ShadowMidas_Full_Alert_1_VO_asset,
							Audio.VO.ShadowMidas.ShadowMidas_Full_Alert_2_VO_asset,
							Audio.VO.ShadowMidas.ShadowMidas_Full_Alert_3_VO_asset,
							Audio.VO.ShadowMidas.ShadowMidas_Full_Alert_4_VO_asset,
							Audio.VO.ShadowMidas.ShadowMidas_Full_Alert_5_VO_asset
						}
					}
				}
			}
			FNClasses.Shadow =>
			{
				case( VOType )
				{
					FNVOType.Idle => 
					{ return array{
							Audio.VO.Shadow.ARSENIC_2020_GhostVoice_Spawn_010_asset,
							Audio.VO.Shadow.ARSENIC_2020_GhostVoice_Spawn_011_asset,
							Audio.VO.Shadow.ARSENIC_2020_GhostVoice_Spawn_012_asset,
							Audio.VO.Shadow.ARSENIC_2020_GhostVoice_Spawn_020_asset,
							Audio.VO.Shadow.ARSENIC_2020_GhostVoice_Spawn_021_asset
						}
					}
					FNVOType.Sus =>
					{ return array{
							Audio.VO.Shadow.ARSENIC_2020_GhostVoice_Spawn_010_asset,
							Audio.VO.Shadow.ARSENIC_2020_GhostVoice_Spawn_011_asset,
							Audio.VO.Shadow.ARSENIC_2020_GhostVoice_Spawn_012_asset,
							Audio.VO.Shadow.ARSENIC_2020_GhostVoice_Spawn_020_asset,
							Audio.VO.Shadow.ARSENIC_2020_GhostVoice_Spawn_021_asset
						}
					}
					FNVOType.ReturnIdle =>
					{ return array{
							Audio.VO.Shadow.ARSENIC_2020_GhostVoice_Spawn_010_asset,
							Audio.VO.Shadow.ARSENIC_2020_GhostVoice_Spawn_011_asset,
							Audio.VO.Shadow.ARSENIC_2020_GhostVoice_Spawn_012_asset,
							Audio.VO.Shadow.ARSENIC_2020_GhostVoice_Spawn_020_asset,
							Audio.VO.Shadow.ARSENIC_2020_GhostVoice_Spawn_021_asset
						}
					}
					FNVOType.FullAlert =>
					{ return array{
							Audio.VO.Shadow.ARSENIC_2020_GhostVoice_Scream_010_asset,
							Audio.VO.Shadow.ARSENIC_2020_GhostVoice_Scream_020_asset,
							Audio.VO.Shadow.ARSENIC_2020_GhostVoice_Scream_030_asset,
							Audio.VO.Shadow.ARSENIC_2020_GhostVoice_Scream_040_asset
						}
					}
					FNVOType.FullAlertLost =>
					{ return array{
							Audio.VO.Shadow.ARSENIC_2020_GhostVoice_Spawn_010_asset,
							Audio.VO.Shadow.ARSENIC_2020_GhostVoice_Spawn_011_asset,
							Audio.VO.Shadow.ARSENIC_2020_GhostVoice_Spawn_012_asset,
							Audio.VO.Shadow.ARSENIC_2020_GhostVoice_Spawn_020_asset,
							Audio.VO.Shadow.ARSENIC_2020_GhostVoice_Spawn_021_asset
						}
					}
					FNVOType.DBNO =>
					{ return array{
							Audio.VO.Shadow.ARSENIC_2020_GhostVoice_Death_011_asset,
							Audio.VO.Shadow.ARSENIC_2020_GhostVoice_Death_013_asset,
							Audio.VO.Shadow.ARSENIC_2020_GhostVoice_Death_020_asset,
							Audio.VO.Shadow.ARSENIC_2020_GhostVoice_Death_021_asset,
							Audio.VO.Shadow.ARSENIC_2020_GhostVoice_Death_022_asset
						}
					}
					FNVOType.Interrogation =>
					{ return array{
							Audio.VO.Shadow.ARSENIC_2020_GhostVoice_TakeDamage_010_asset,
							Audio.VO.Shadow.ARSENIC_2020_GhostVoice_TakeDamage_011_asset,
							Audio.VO.Shadow.ARSENIC_2020_GhostVoice_TakeDamage_012_asset,
							Audio.VO.Shadow.ARSENIC_2020_GhostVoice_TakeDamage_013_asset,
							Audio.VO.Shadow.ARSENIC_2020_GhostVoice_TakeDamage_020_asset
						}
					}
					FNVOType.Ability =>	
					{ return array{
							Audio.VO.Shadow.ARSENIC_2020_GhostVoice_Dash_010_asset,
							Audio.VO.Shadow.ARSENIC_2020_GhostVoice_Dash_011_asset,
							Audio.VO.Shadow.ARSENIC_2020_GhostVoice_Dash_012_asset,
							Audio.VO.Shadow.ARSENIC_2020_GhostVoice_Dash_013_asset,
							Audio.VO.Shadow.ARSENIC_2020_GhostVoice_Dash_020_asset
						}
					}
					FNVOType.AbilityEnd =>
					{ return array{
							Audio.VO.Shadow.ARSENIC_2020_GhostVoice_Scream_010_asset,
							Audio.VO.Shadow.ARSENIC_2020_GhostVoice_Scream_020_asset,
							Audio.VO.Shadow.ARSENIC_2020_GhostVoice_Scream_030_asset,
							Audio.VO.Shadow.ARSENIC_2020_GhostVoice_Scream_040_asset
						}
					}
					FNVOType.Ult =>
					{ return array{
							Audio.VO.Shadow.ARSENIC_2020_GhostVoice_Scream_010_asset,
							Audio.VO.Shadow.ARSENIC_2020_GhostVoice_Scream_020_asset,
							Audio.VO.Shadow.ARSENIC_2020_GhostVoice_Scream_030_asset,
							Audio.VO.Shadow.ARSENIC_2020_GhostVoice_Scream_040_asset
						}
					}
				}
			}
			FNClasses.Raz =>
			{
				case( VOType )
				{
					FNVOType.Idle => 
					{ return array{
							Audio.VO.Raz.TinyBalloon_2021_Idle_010_asset,
							Audio.VO.Raz.TinyBalloon_2021_Idle_020_asset,
							Audio.VO.Raz.TinyBalloon_2021_Idle_040_asset,
							Audio.VO.Raz.TinyBalloon_2021_Idle_050_asset
						}
					}
					FNVOType.Sus =>
					{ return array{
							Audio.VO.Raz.TinyBalloon_2021_Suspicious_010_asset,
							Audio.VO.Raz.TinyBalloon_2021_Suspicious_020_asset,
							Audio.VO.Raz.TinyBalloon_2021_Suspicious_030_asset,
							Audio.VO.Raz.TinyBalloon_2021_Suspicious_040_asset,
							Audio.VO.Raz.TinyBalloon_2021_Suspicious_050_asset
						}
					}
					FNVOType.ReturnIdle =>
					{ return array{
							Audio.VO.Raz.TinyBalloon_2021_SuspiciousReturnToIdle_010_asset,
							Audio.VO.Raz.TinyBalloon_2021_SuspiciousReturnToIdle_020_asset,
							Audio.VO.Raz.TinyBalloon_2021_SuspiciousReturnToIdle_030_asset,
							Audio.VO.Raz.TinyBalloon_2021_SuspiciousReturnToIdle_040_asset
						}
					}
					FNVOType.FullAlert =>
					{ return array{
							Audio.VO.Raz.TinyBalloon_2021_FullAlert_010_asset,
							Audio.VO.Raz.TinyBalloon_2021_FullAlert_020_asset,
							Audio.VO.Raz.TinyBalloon_2021_FullAlert_030_asset,
							Audio.VO.Raz.TinyBalloon_2021_FullAlert_040_asset,
							Audio.VO.Raz.TinyBalloon_2021_FullAlert_050_asset
						}
					}
					FNVOType.FullAlertLost =>
					{ return array{
							Audio.VO.Raz.TinyBalloon_2021_FullAlertLostTarget_020_asset,
							Audio.VO.Raz.TinyBalloon_2021_FullAlertLostTarget_030_asset,
							Audio.VO.Raz.TinyBalloon_2021_FullAlertLostTarget_040_asset,
							Audio.VO.Raz.TinyBalloon_2021_FullAlertLostTarget_050_asset,
							Audio.VO.Raz.TinyBalloon_2021_FullAlertLostTarget_051_asset
						}
					}
					FNVOType.DBNO =>
					{ return array{
							Audio.VO.Raz.TinyBalloon_2021_DBNO_Enter_010_asset,
							Audio.VO.Raz.TinyBalloon_2021_DBNO_Enter_011_asset,
							Audio.VO.Raz.TinyBalloon_2021_DBNO_Enter_012_asset,
							Audio.VO.Raz.TinyBalloon_2021_DBNO_Enter_013_asset,
							Audio.VO.Raz.TinyBalloon_2021_DBNO_Enter_014_asset
						}
					}
					FNVOType.Interrogation =>
					{ return array{
							Audio.VO.Raz.TinyBalloon_2021_DBNO_Enter_010_asset,
							Audio.VO.Raz.TinyBalloon_2021_DBNO_Enter_011_asset,
							Audio.VO.Raz.TinyBalloon_2021_DBNO_Enter_012_asset,
							Audio.VO.Raz.TinyBalloon_2021_DBNO_Enter_013_asset,
							Audio.VO.Raz.TinyBalloon_2021_DBNO_Enter_014_asset
						}
					}
					FNVOType.Ability =>	
					{ return array{
							Audio.VO.Raz.TinyBalloon_2021_WarpAbility_010_asset,
							Audio.VO.Raz.TinyBalloon_2021_WarpAbility_011_asset,
							Audio.VO.Raz.TinyBalloon_2021_WarpAbility_012_asset,
							Audio.VO.Raz.TinyBalloon_2021_WarpAbility_013_asset,
							Audio.VO.Raz.TinyBalloon_2021_WarpAbility_030_asset,
							Audio.VO.Raz.TinyBalloon_2021_WarpAbility_031_asset
						}
					}
					FNVOType.AbilityEnd =>
					{ return array{
							Audio.VO.Raz.TinyBalloon_2021_LifeDrain_Cast_020_asset,
							Audio.VO.Raz.TinyBalloon_2021_LifeDrain_Cast_021_asset,
							Audio.VO.Raz.TinyBalloon_2021_LifeDrain_Cast_022_asset,
							Audio.VO.Raz.TinyBalloon_2021_LifeDrain_Cast_024_asset
						}
					}
					FNVOType.Ult =>
					{ return array{
							Audio.VO.Raz.TinyBalloon_2021_FullAlert_010_asset,
							Audio.VO.Raz.TinyBalloon_2021_FullAlert_020_asset,
							Audio.VO.Raz.TinyBalloon_2021_FullAlert_030_asset,
							Audio.VO.Raz.TinyBalloon_2021_FullAlert_040_asset,
							Audio.VO.Raz.TinyBalloon_2021_FullAlert_050_asset
						}
					}
				}
			}
			FNClasses.Slone =>
			{
				case( VOType )
				{
					FNVOType.Idle => 
					{ return array{
							Audio.VO.Slone.Slone_2021_Idle_010_asset,
							Audio.VO.Slone.Slone_2021_Idle_020_asset,
							Audio.VO.Slone.Slone_2021_Idle_030_asset,
							Audio.VO.Slone.Slone_2021_Idle_040_asset,
							Audio.VO.Slone.Slone_2021_Idle_050_asset
						}
					}
					FNVOType.Sus =>
					{ return array{
							Audio.VO.Slone.Slone_2021_Suspicious_010_asset,
							Audio.VO.Slone.Slone_2021_Suspicious_020_asset,
							Audio.VO.Slone.Slone_2021_Suspicious_030_asset,
							Audio.VO.Slone.Slone_2021_Suspicious_040_asset,
							Audio.VO.Slone.Slone_2021_Suspicious_050_asset
						}
					}
					FNVOType.ReturnIdle =>
					{ return array{
							Audio.VO.Slone.Slone_2021_SuspiciousReturnToIdle_010_asset,
							Audio.VO.Slone.Slone_2021_SuspiciousReturnToIdle_020_asset,
							Audio.VO.Slone.Slone_2021_SuspiciousReturnToIdle_030_asset,
							Audio.VO.Slone.Slone_2021_SuspiciousReturnToIdle_040_asset,
							Audio.VO.Slone.Slone_2021_SuspiciousReturnToIdle_050_asset
						}
					}
					FNVOType.FullAlert =>
					{ return array{
							Audio.VO.Slone.Slone_2021_FullAlert_010_asset,
							Audio.VO.Slone.Slone_2021_FullAlert_020_asset,
							Audio.VO.Slone.Slone_2021_FullAlert_030_asset,
							Audio.VO.Slone.Slone_2021_FullAlert_040_asset,
							Audio.VO.Slone.Slone_2021_FullAlert_050_asset
						}
					}
					FNVOType.FullAlertLost =>
					{ return array{
							Audio.VO.Slone.Slone_2021_FullAlertLostTarget_010_asset,
							Audio.VO.Slone.Slone_2021_FullAlertLostTarget_020_asset,
							Audio.VO.Slone.Slone_2021_FullAlertLostTarget_030_asset,
							Audio.VO.Slone.Slone_2021_FullAlertLostTarget_040_asset,
							Audio.VO.Slone.Slone_2021_FullAlertLostTarget_050_asset
						}
					}
					FNVOType.DBNO =>
					{ return array{
							Audio.VO.Slone.Slone_2021_DBNO_Enter_010_asset,
							Audio.VO.Slone.Slone_2021_DBNO_Enter_011_asset,
							Audio.VO.Slone.Slone_2021_DBNO_Enter_012_asset,
							Audio.VO.Slone.Slone_2021_DBNO_Enter_013_asset,
							Audio.VO.Slone.Slone_2021_DBNO_Enter_014_asset
						}
					}
					FNVOType.Interrogation =>
					{ return array{
							Audio.VO.Slone.Slone_2021_DBNO_Enter_010_asset,
							Audio.VO.Slone.Slone_2021_DBNO_Enter_011_asset,
							Audio.VO.Slone.Slone_2021_DBNO_Enter_012_asset,
							Audio.VO.Slone.Slone_2021_DBNO_Enter_013_asset,
							Audio.VO.Slone.Slone_2021_DBNO_Enter_014_asset
						}
					}
					FNVOType.Ability =>	
					{ return array{
							Audio.VO.Slone.Slone_2021_SpecialAbility_010_asset,
							Audio.VO.Slone.Slone_2021_SpecialAbility_011_asset,
							Audio.VO.Slone.Slone_2021_SpecialAbility_012_asset,
							Audio.VO.Slone.Slone_2021_SpecialAbility_013_asset
						}
					}
					FNVOType.AbilityEnd =>
					{ return array{
							Audio.VO.Slone.Slone_2021_SpecialAbility_010_asset,
							Audio.VO.Slone.Slone_2021_SpecialAbility_011_asset,
							Audio.VO.Slone.Slone_2021_SpecialAbility_012_asset,
							Audio.VO.Slone.Slone_2021_SpecialAbility_013_asset
						}
					}
					FNVOType.Ult =>
					{ return array{
							Audio.VO.Slone.VO_Slone_Ult1_asset,
							Audio.VO.Slone.VO_Slone_Ult2_asset
						}
					}
				}
			}
			FNClasses.TheFoundation =>
			{
				case( VOType )
				{
					FNVOType.Idle => 
					{ return array{
							Audio.VO.TheFoundation.TameFoot_NPCBoss_Suspicious_010_asset,
							Audio.VO.TheFoundation.TameFoot_NPCBoss_Suspicious_011_asset,
							Audio.VO.TheFoundation.TameFoot_NPCBoss_Suspicious_012_asset,
							Audio.VO.TheFoundation.TameFoot_NPCBoss_Suspicious_013_asset
						}
					}
					FNVOType.Sus =>
					{ return array{
							Audio.VO.TheFoundation.TameFoot_NPCBoss_Suspicious_010_asset,
							Audio.VO.TheFoundation.TameFoot_NPCBoss_Suspicious_011_asset,
							Audio.VO.TheFoundation.TameFoot_NPCBoss_Suspicious_012_asset,
							Audio.VO.TheFoundation.TameFoot_NPCBoss_Suspicious_013_asset
						}
					}
					FNVOType.ReturnIdle =>
					{ return array{
							Audio.VO.TheFoundation.TameFoot_NPCBoss_Suspicious_010_asset,
							Audio.VO.TheFoundation.TameFoot_NPCBoss_Suspicious_011_asset,
							Audio.VO.TheFoundation.TameFoot_NPCBoss_Suspicious_012_asset,
							Audio.VO.TheFoundation.TameFoot_NPCBoss_Suspicious_013_asset
						}
					}
					FNVOType.FullAlert =>
					{ return array{
							Audio.VO.TheFoundation.TameFoot_NPCBoss_FullAlert_010_asset,
							Audio.VO.TheFoundation.TameFoot_NPCBoss_FullAlert_011_asset,
							Audio.VO.TheFoundation.TameFoot_NPCBoss_FullAlert_012_asset,
							Audio.VO.TheFoundation.TameFoot_NPCBoss_FullAlert_013_asset,
							Audio.VO.TheFoundation.TameFoot_NPCBoss_FullAlert_014_asset
						}
					}
					FNVOType.FullAlertLost =>
					{ return array{
							Audio.VO.TheFoundation.TameFoot_NPCBoss_LostTarget_010_asset,
							Audio.VO.TheFoundation.TameFoot_NPCBoss_LostTarget_011_asset,
							Audio.VO.TheFoundation.TameFoot_NPCBoss_LostTarget_012_asset,
							Audio.VO.TheFoundation.TameFoot_NPCBoss_LostTarget_013_asset,
							Audio.VO.TheFoundation.TameFoot_NPCBoss_LostTarget_014_asset
						}
					}
					FNVOType.DBNO =>
					{ return array{
							Audio.VO.TheFoundation.TameFoot_NPCBoss_DBNO_010_asset,
							Audio.VO.TheFoundation.TameFoot_NPCBoss_DBNO_011_asset,
							Audio.VO.TheFoundation.TameFoot_NPCBoss_DBNO_012_asset
						}
					}
					FNVOType.Interrogation =>
					{ return array{
							Audio.VO.TheFoundation.TameFoot_NPCBoss_DBNO_010_asset,
							Audio.VO.TheFoundation.TameFoot_NPCBoss_DBNO_011_asset,
							Audio.VO.TheFoundation.TameFoot_NPCBoss_DBNO_012_asset

						}
					}
					FNVOType.Ability =>	
					{ return array{
							Audio.VO.TheFoundation.TameFoot_NPCBoss_AbilityStart_010_asset,
							Audio.VO.TheFoundation.TameFoot_NPCBoss_AbilityStart_011_asset,
							Audio.VO.TheFoundation.TameFoot_NPCBoss_AbilityStart_012_asset,
							Audio.VO.TheFoundation.TameFoot_NPCBoss_AbilityStart_013_asset
						}
					}
					FNVOType.AbilityEnd =>
					{ return array{
							Audio.VO.TheFoundation.TameFoot_NPCBoss_AbilityEnd_010_asset,
							Audio.VO.TheFoundation.TameFoot_NPCBoss_AbilityEnd_011_asset,
							Audio.VO.TheFoundation.TameFoot_NPCBoss_AbilityEnd_012_asset
						}
					}
					FNVOType.Ult =>
					{ return array{
							Audio.VO.TheFoundation.TameFoot_NPCBoss_FullAlert_010_asset,
							Audio.VO.TheFoundation.TameFoot_NPCBoss_FullAlert_011_asset,
							Audio.VO.TheFoundation.TameFoot_NPCBoss_FullAlert_012_asset,
							Audio.VO.TheFoundation.TameFoot_NPCBoss_FullAlert_013_asset,
							Audio.VO.TheFoundation.TameFoot_NPCBoss_FullAlert_014_asset
						}
					}
				}
			}
			FNClasses.Gunnar =>
			{
				case( VOType )
				{
					FNVOType.Idle => 
					{ return array{
							Audio.VO.Gunnar.IOBrute_2021_Idle_010_asset,
							Audio.VO.Gunnar.IOBrute_2021_Idle_020_asset,
							Audio.VO.Gunnar.IOBrute_2021_Idle_030_asset,
							Audio.VO.Gunnar.IOBrute_2021_Idle_040_asset,
							Audio.VO.Gunnar.IOBrute_2021_Idle_050_asset
						}
					}
					FNVOType.Sus =>
					{ return array{
							Audio.VO.Gunnar.IOBrute_2021_Suspicious_010_asset,
							Audio.VO.Gunnar.IOBrute_2021_Suspicious_020_asset,
							Audio.VO.Gunnar.IOBrute_2021_Suspicious_030_asset,
							Audio.VO.Gunnar.IOBrute_2021_Suspicious_040_asset,
							Audio.VO.Gunnar.IOBrute_2021_Suspicious_050_asset
						}
					}
					FNVOType.ReturnIdle =>
					{ return array{
							Audio.VO.Gunnar.IOBrute_2021_SuspiciousReturnToIdle_010_asset,
							Audio.VO.Gunnar.IOBrute_2021_SuspiciousReturnToIdle_020_asset,
							Audio.VO.Gunnar.IOBrute_2021_SuspiciousReturnToIdle_030_asset,
							Audio.VO.Gunnar.IOBrute_2021_SuspiciousReturnToIdle_040_asset,
							Audio.VO.Gunnar.IOBrute_2021_SuspiciousReturnToIdle_050_asset
						}
					}
					FNVOType.FullAlert =>
					{ return array{
							Audio.VO.Gunnar.IOBrute_2021_FullAlert_010_asset,
							Audio.VO.Gunnar.IOBrute_2021_FullAlert_020_asset,
							Audio.VO.Gunnar.IOBrute_2021_FullAlert_030_asset,
							Audio.VO.Gunnar.IOBrute_2021_FullAlert_040_asset,
							Audio.VO.Gunnar.IOBrute_2021_FullAlert_050_asset
						}
					}
					FNVOType.FullAlertLost =>
					{ return array{
							Audio.VO.Gunnar.IOBrute_2021_FullAlertLostTarget_010_asset,
							Audio.VO.Gunnar.IOBrute_2021_FullAlertLostTarget_020_asset,
							Audio.VO.Gunnar.IOBrute_2021_FullAlertLostTarget_030_asset,
							Audio.VO.Gunnar.IOBrute_2021_FullAlertLostTarget_040_asset,
							Audio.VO.Gunnar.IOBrute_2021_FullAlertLostTarget_050_asset
						}
					}
					FNVOType.DBNO =>
					{ return array{
							Audio.VO.Gunnar.IOBrute_2021_DBNO_Enter_010_asset,
							Audio.VO.Gunnar.IOBrute_2021_DBNO_Enter_011_asset,
							Audio.VO.Gunnar.IOBrute_2021_DBNO_Enter_012_asset,
							Audio.VO.Gunnar.IOBrute_2021_DBNO_Enter_013_asset,
							Audio.VO.Gunnar.IOBrute_2021_DBNO_Enter_014_asset,
							Audio.VO.Gunnar.IOBrute_2021_DBNO_Enter_015_asset
						}
					}
					FNVOType.Interrogation =>
					{ return array{
							Audio.VO.Gunnar.IOBrute_2021_DBNO_Enter_010_asset,
							Audio.VO.Gunnar.IOBrute_2021_DBNO_Enter_011_asset,
							Audio.VO.Gunnar.IOBrute_2021_DBNO_Enter_012_asset,
							Audio.VO.Gunnar.IOBrute_2021_DBNO_Enter_013_asset,
							Audio.VO.Gunnar.IOBrute_2021_DBNO_Enter_014_asset,
							Audio.VO.Gunnar.IOBrute_2021_DBNO_Enter_015_asset

						}
					}
					FNVOType.Ability =>	
					{ return array{
							Audio.VO.Gunnar.IOBrute_2021_ChargePunchIntro_010_asset,
							Audio.VO.Gunnar.IOBrute_2021_ChargePunchIntro_011_asset,
							Audio.VO.Gunnar.IOBrute_2021_ChargePunchIntro_012_asset
						}
					}
					FNVOType.AbilityEnd =>
					{ return array{
							Audio.VO.Gunnar.IOBrute_2021_ChargePunchFinish_010_asset,
							Audio.VO.Gunnar.IOBrute_2021_ChargePunchFinish_011_asset,
							Audio.VO.Gunnar.IOBrute_2021_ChargePunchFinish_012_asset,
							Audio.VO.Gunnar.IOBrute_2021_ChargePunchFinish_013_asset,
							Audio.VO.Gunnar.IOBrute_2021_ChargePunchFinish_014_asset
						}
					}
					FNVOType.Ult =>
					{ return array{
							Audio.VO.Gunnar.IOBrute_2021_FullAlert_010_asset,
							Audio.VO.Gunnar.IOBrute_2021_FullAlert_020_asset,
							Audio.VO.Gunnar.IOBrute_2021_FullAlert_030_asset,
							Audio.VO.Gunnar.IOBrute_2021_FullAlert_040_asset,
							Audio.VO.Gunnar.IOBrute_2021_FullAlert_050_asset
						}
					}
				}
			}
			FNClasses.HuntmasterSaber =>
			{
				case( VOType )
				{
					FNVOType.Idle => 
					{ return array{
							Audio.VO.HuntmasterSaber.ThunderMeow_2022_Idle_010_asset,
							Audio.VO.HuntmasterSaber.ThunderMeow_2022_Idle_020_asset,
							Audio.VO.HuntmasterSaber.ThunderMeow_2022_Idle_030_asset,
							Audio.VO.HuntmasterSaber.ThunderMeow_2022_Idle_040_asset,
							Audio.VO.HuntmasterSaber.ThunderMeow_2022_Idle_050_asset
						}
					}
					FNVOType.Sus =>
					{ return array{
							Audio.VO.HuntmasterSaber.ThunderMeow_2022_Suspicious_010_asset,
							Audio.VO.HuntmasterSaber.ThunderMeow_2022_Suspicious_020_asset,
							Audio.VO.HuntmasterSaber.ThunderMeow_2022_Suspicious_030_asset,
							Audio.VO.HuntmasterSaber.ThunderMeow_2022_Suspicious_040_asset,
							Audio.VO.HuntmasterSaber.ThunderMeow_2022_Suspicious_050_asset
						}
					}
					FNVOType.ReturnIdle =>
					{ return array{
							Audio.VO.HuntmasterSaber.ThunderMeow_2022_SuspiciousReturnToIdle_010_asset,
							Audio.VO.HuntmasterSaber.ThunderMeow_2022_SuspiciousReturnToIdle_020_asset,
							Audio.VO.HuntmasterSaber.ThunderMeow_2022_SuspiciousReturnToIdle_030_asset,
							Audio.VO.HuntmasterSaber.ThunderMeow_2022_SuspiciousReturnToIdle_040_asset,
							Audio.VO.HuntmasterSaber.ThunderMeow_2022_SuspiciousReturnToIdle_050_asset
						}
					}
					FNVOType.FullAlert =>
					{ return array{
							Audio.VO.HuntmasterSaber.ThunderMeow_2022_FullAlert_010_asset,
							Audio.VO.HuntmasterSaber.ThunderMeow_2022_FullAlert_020_asset,
							Audio.VO.HuntmasterSaber.ThunderMeow_2022_FullAlert_030_asset,
							Audio.VO.HuntmasterSaber.ThunderMeow_2022_FullAlert_040_asset,
							Audio.VO.HuntmasterSaber.ThunderMeow_2022_FullAlert_050_asset
						}
					}
					FNVOType.FullAlertLost =>
					{ return array{
							
							Audio.VO.HuntmasterSaber.ThunderMeow_2022_FullAlertLostTarget_020_asset,
							Audio.VO.HuntmasterSaber.ThunderMeow_2022_FullAlertLostTarget_030_asset,
							Audio.VO.HuntmasterSaber.ThunderMeow_2022_FullAlertLostTarget_040_asset,
							Audio.VO.HuntmasterSaber.ThunderMeow_2022_FullAlertLostTarget_050_asset
						}
					}
					FNVOType.DBNO =>
					{ return array{
							Audio.VO.HuntmasterSaber.ThunderMeow_2022_DBNO_Enter_010_asset,
							Audio.VO.HuntmasterSaber.ThunderMeow_2022_DBNO_Enter_011_asset,
							Audio.VO.HuntmasterSaber.ThunderMeow_2022_DBNO_Enter_012_asset,
							Audio.VO.HuntmasterSaber.ThunderMeow_2022_DBNO_Enter_013_asset,
							Audio.VO.HuntmasterSaber.ThunderMeow_2022_DBNO_Enter_014_asset
						}
					}
					FNVOType.Interrogation =>
					{ return array{
							Audio.VO.HuntmasterSaber.ThunderMeow_2022_DBNO_Enter_010_asset,
							Audio.VO.HuntmasterSaber.ThunderMeow_2022_DBNO_Enter_011_asset,
							Audio.VO.HuntmasterSaber.ThunderMeow_2022_DBNO_Enter_012_asset,
							Audio.VO.HuntmasterSaber.ThunderMeow_2022_DBNO_Enter_013_asset,
							Audio.VO.HuntmasterSaber.ThunderMeow_2022_DBNO_Enter_014_asset

						}
					}
					FNVOType.Ability =>	
					{ return array{
							Audio.VO.HuntmasterSaber.ThunderMeow_2022_SpinAttackCast_010_asset,
							Audio.VO.HuntmasterSaber.ThunderMeow_2022_SpinAttackCast_011_asset,
							Audio.VO.HuntmasterSaber.ThunderMeow_2022_SpinAttackCast_012_asset,
							Audio.VO.HuntmasterSaber.ThunderMeow_2022_SpinAttackCast_013_asset,
							Audio.VO.HuntmasterSaber.ThunderMeow_2022_SpinAttackCast_014_asset,
							Audio.VO.HuntmasterSaber.ThunderMeow_2022_SpinAttackCast_015_asset
						}
					}
					FNVOType.AbilityEnd =>
					{ return array{
							Audio.VO.HuntmasterSaber.ThunderMeow_2022_SpinAttackEnd_010_asset,
							Audio.VO.HuntmasterSaber.ThunderMeow_2022_SpinAttackEnd_011_asset,
							Audio.VO.HuntmasterSaber.ThunderMeow_2022_SpinAttackEnd_012_asset,
							Audio.VO.HuntmasterSaber.ThunderMeow_2022_SpinAttackEnd_013_asset,
							Audio.VO.HuntmasterSaber.ThunderMeow_2022_SpinAttackEnd_014_asset
						}
					}
					FNVOType.Ult =>
					{ return array{
							Audio.VO.HuntmasterSaber.ThunderMeow_2022_FullAlert_010_asset,
							Audio.VO.HuntmasterSaber.ThunderMeow_2022_FullAlert_020_asset,
							Audio.VO.HuntmasterSaber.ThunderMeow_2022_FullAlert_030_asset,
							Audio.VO.HuntmasterSaber.ThunderMeow_2022_FullAlert_040_asset,
							Audio.VO.HuntmasterSaber.ThunderMeow_2022_FullAlert_050_asset
						}
					}
				}
			}
			FNClasses.TheHerald =>
			{
				case( VOType )
				{
					FNVOType.Idle => 
					{ return array{
							Audio.VO.TheHerald.HABITAT_2022_Idle_010_asset,
							Audio.VO.TheHerald.HABITAT_2022_Idle_020_asset,
							Audio.VO.TheHerald.HABITAT_2022_Idle_030_asset,
							Audio.VO.TheHerald.HABITAT_2022_Idle_040_asset,
							Audio.VO.TheHerald.HABITAT_2022_Idle_050_asset
						}
					}
					FNVOType.Sus =>
					{ return array{
							Audio.VO.TheHerald.HABITAT_2022_Suspicious_010_asset,
							Audio.VO.TheHerald.HABITAT_2022_Suspicious_020_asset,
							Audio.VO.TheHerald.HABITAT_2022_Suspicious_030_asset,
							Audio.VO.TheHerald.HABITAT_2022_Suspicious_040_asset,
							Audio.VO.TheHerald.HABITAT_2022_Suspicious_050_asset
						}
					}
					FNVOType.ReturnIdle =>
					{ return array{
							Audio.VO.TheHerald.HABITAT_2022_SuspiciousReturnToIdle_010_asset,
							Audio.VO.TheHerald.HABITAT_2022_SuspiciousReturnToIdle_020_asset,
							Audio.VO.TheHerald.HABITAT_2022_SuspiciousReturnToIdle_030_asset,
							Audio.VO.TheHerald.HABITAT_2022_SuspiciousReturnToIdle_040_asset,
							Audio.VO.TheHerald.HABITAT_2022_SuspiciousReturnToIdle_050_asset
						}
					}
					FNVOType.FullAlert =>
					{ return array{
							Audio.VO.TheHerald.HABITAT_2022_FullAlert_010_asset,
							Audio.VO.TheHerald.HABITAT_2022_FullAlert_020_asset,
							Audio.VO.TheHerald.HABITAT_2022_FullAlert_030_asset,
							Audio.VO.TheHerald.HABITAT_2022_FullAlert_040_asset,
							Audio.VO.TheHerald.HABITAT_2022_FullAlert_050_asset
						}
					}
					FNVOType.FullAlertLost =>
					{ return array{
							Audio.VO.TheHerald.HABITAT_2022_FullAlertLostTarget_010_asset,
							Audio.VO.TheHerald.HABITAT_2022_FullAlertLostTarget_020_asset,
							Audio.VO.TheHerald.HABITAT_2022_FullAlertLostTarget_030_asset,
							Audio.VO.TheHerald.HABITAT_2022_FullAlertLostTarget_040_asset,
							Audio.VO.TheHerald.HABITAT_2022_FullAlertLostTarget_050_asset
						}
					}
					FNVOType.DBNO =>
					{ return array{
							Audio.VO.TheHerald.HABITAT_2022_DBNO_011_asset,
							Audio.VO.TheHerald.HABITAT_2022_DBNO_012_asset,
							Audio.VO.TheHerald.HABITAT_2022_DBNO_013_asset,
							Audio.VO.TheHerald.HABITAT_2022_DBNO_014_asset
						}
					}
					FNVOType.Interrogation =>
					{ return array{
							Audio.VO.TheHerald.HABITAT_2022_DBNO_011_asset,
							Audio.VO.TheHerald.HABITAT_2022_DBNO_012_asset,
							Audio.VO.TheHerald.HABITAT_2022_DBNO_013_asset,
							Audio.VO.TheHerald.HABITAT_2022_DBNO_014_asset
						}
					}
					FNVOType.Ability =>	
					{ return array{
							Audio.VO.TheHerald.HABITAT_2022_LockdownShieldCast_010_asset,
							Audio.VO.TheHerald.HABITAT_2022_LockdownShieldCast_011_asset,
							Audio.VO.TheHerald.HABITAT_2022_LockdownShieldCast_012_asset,
							Audio.VO.TheHerald.HABITAT_2022_LockdownShieldCast_013_asset,
							Audio.VO.TheHerald.HABITAT_2022_LockdownShieldCast_014_asset
						}
					}
					FNVOType.AbilityEnd =>
					{ return array{
							Audio.VO.TheHerald.HABITAT_2022_LockdownShieldEnd_010_asset,
							Audio.VO.TheHerald.HABITAT_2022_LockdownShieldEnd_011_asset,
							Audio.VO.TheHerald.HABITAT_2022_LockdownShieldEnd_012_asset,
							Audio.VO.TheHerald.HABITAT_2022_LockdownShieldEnd_013_asset,
							Audio.VO.TheHerald.HABITAT_2022_LockdownShieldEnd_014_asset
						}
					}
					FNVOType.Ult =>
					{ return array{
							Audio.VO.TheHerald.HABITAT_2022_FullAlert_010_asset,
							Audio.VO.TheHerald.HABITAT_2022_FullAlert_020_asset,
							Audio.VO.TheHerald.HABITAT_2022_FullAlert_030_asset,
							Audio.VO.TheHerald.HABITAT_2022_FullAlert_040_asset,
							Audio.VO.TheHerald.HABITAT_2022_FullAlert_050_asset
						}
					}
				}
			}
			FNClasses.TheAgelessChampion =>
			{
				case( VOType )
				{
					FNVOType.Idle => 
					{ return array{
							Audio.VO.TheAgeless.EmblemKnight_2022_Idle_010_asset,
							Audio.VO.TheAgeless.EmblemKnight_2022_Idle_011_asset,
							Audio.VO.TheAgeless.EmblemKnight_2022_Idle_012_asset,
							Audio.VO.TheAgeless.EmblemKnight_2022_Idle_014_asset
						}
					}
					FNVOType.Sus =>
					{ return array{
							Audio.VO.TheAgeless.EmblemKnight_2022_Suspicious_010_asset,
							Audio.VO.TheAgeless.EmblemKnight_2022_Suspicious_011_asset,
							Audio.VO.TheAgeless.EmblemKnight_2022_Suspicious_012_asset,
							Audio.VO.TheAgeless.EmblemKnight_2022_Suspicious_013_asset,
							Audio.VO.TheAgeless.EmblemKnight_2022_Suspicious_014_asset
						}
					}
					FNVOType.ReturnIdle =>
					{ return array{
							Audio.VO.TheAgeless.EmblemKnight_2022_Suspicious_010_asset,
							Audio.VO.TheAgeless.EmblemKnight_2022_Suspicious_011_asset,
							Audio.VO.TheAgeless.EmblemKnight_2022_Suspicious_012_asset,
							Audio.VO.TheAgeless.EmblemKnight_2022_Suspicious_013_asset,
							Audio.VO.TheAgeless.EmblemKnight_2022_Suspicious_014_asset
						}
					}
					FNVOType.FullAlert =>
					{ return array{
							Audio.VO.TheAgeless.EmblemKnight_2022_FullAlert_010_asset,
							Audio.VO.TheAgeless.EmblemKnight_2022_FullAlert_011_asset,
							Audio.VO.TheAgeless.EmblemKnight_2022_FullAlert_012_asset,
							Audio.VO.TheAgeless.EmblemKnight_2022_FullAlert_013_asset,
							Audio.VO.TheAgeless.EmblemKnight_2022_FullAlert_014_asset
						}
					}
					FNVOType.FullAlertLost =>
					{ return array{
							Audio.VO.TheAgeless.EmblemKnight_2022_FullAlertLostTarget_010_asset,
							Audio.VO.TheAgeless.EmblemKnight_2022_FullAlertLostTarget_011_asset,
							Audio.VO.TheAgeless.EmblemKnight_2022_FullAlertLostTarget_012_asset,
							Audio.VO.TheAgeless.EmblemKnight_2022_FullAlertLostTarget_013_asset,
							Audio.VO.TheAgeless.EmblemKnight_2022_FullAlertLostTarget_014_asset
						}
					}
					FNVOType.DBNO =>
					{ return array{
							Audio.VO.TheAgeless.EmblemKnight_2022_Elimination_010_asset,
							Audio.VO.TheAgeless.EmblemKnight_2022_Elimination_011_asset,
							Audio.VO.TheAgeless.EmblemKnight_2022_Elimination_012_asset,
							Audio.VO.TheAgeless.EmblemKnight_2022_Elimination_013_asset
						}
					}
					FNVOType.Interrogation =>
					{ return array{
							Audio.VO.TheAgeless.EmblemKnight_2022_Elimination_010_asset,
							Audio.VO.TheAgeless.EmblemKnight_2022_Elimination_011_asset,
							Audio.VO.TheAgeless.EmblemKnight_2022_Elimination_012_asset,
							Audio.VO.TheAgeless.EmblemKnight_2022_Elimination_013_asset

						}
					}
					FNVOType.Ability =>	
					{ return array{
							Audio.VO.TheAgeless.EmblemKnight_2022_FullAlert_010_asset,
							Audio.VO.TheAgeless.EmblemKnight_2022_FullAlert_011_asset,
							Audio.VO.TheAgeless.EmblemKnight_2022_FullAlert_012_asset,
							Audio.VO.TheAgeless.EmblemKnight_2022_FullAlert_013_asset,
							Audio.VO.TheAgeless.EmblemKnight_2022_FullAlert_014_asset
						}
					}
					FNVOType.AbilityEnd =>
					{ return array{
							Audio.VO.TheAgeless.EmblemKnight_2022_FullAlert_010_asset,
							Audio.VO.TheAgeless.EmblemKnight_2022_FullAlert_011_asset,
							Audio.VO.TheAgeless.EmblemKnight_2022_FullAlert_012_asset,
							Audio.VO.TheAgeless.EmblemKnight_2022_FullAlert_013_asset,
							Audio.VO.TheAgeless.EmblemKnight_2022_FullAlert_014_asset
						}
					}
					FNVOType.Ult =>
					{ return array{
							Audio.VO.TheAgeless.EmblemKnight_2022_FullAlert_010_asset,
							Audio.VO.TheAgeless.EmblemKnight_2022_FullAlert_011_asset,
							Audio.VO.TheAgeless.EmblemKnight_2022_FullAlert_012_asset,
							Audio.VO.TheAgeless.EmblemKnight_2022_FullAlert_013_asset,
							Audio.VO.TheAgeless.EmblemKnight_2022_FullAlert_014_asset
						}
					}
				}
			}
			FNClasses.WildguardRelik =>
			{
				case( VOType )
				{
					FNVOType.Idle => 
					{ return array{
							Audio.VO.Relik.JungleBoss_S25_2023_Idle_010_asset,
							Audio.VO.Relik.JungleBoss_S25_2023_Idle_020_asset,
							Audio.VO.Relik.JungleBoss_S25_2023_Idle_040_asset,
							Audio.VO.Relik.JungleBoss_S25_2023_Idle_050_asset,
							Audio.VO.Relik.JungleBoss_S25_2023_Idle_051_asset
						}
					}
					FNVOType.Sus =>
					{ return array{
							Audio.VO.Relik.JungleBoss_S25_2023_Suspicious_011_asset,
							Audio.VO.Relik.JungleBoss_S25_2023_Suspicious_020_asset,
							Audio.VO.Relik.JungleBoss_S25_2023_Suspicious_030_asset,
							Audio.VO.Relik.JungleBoss_S25_2023_Suspicious_040_asset,
							Audio.VO.Relik.JungleBoss_S25_2023_Suspicious_042_asset
						}
					}
					FNVOType.ReturnIdle =>
					{ return array{
							Audio.VO.Relik.JungleBoss_S25_2023_Suspicious_011_asset,
							Audio.VO.Relik.JungleBoss_S25_2023_Suspicious_020_asset,
							Audio.VO.Relik.JungleBoss_S25_2023_Suspicious_030_asset,
							Audio.VO.Relik.JungleBoss_S25_2023_Suspicious_040_asset,
							Audio.VO.Relik.JungleBoss_S25_2023_Suspicious_042_asset
						}
					}
					FNVOType.FullAlert =>
					{ return array{
							Audio.VO.Relik.JungleBoss_S25_2023_Alerted_010_asset,
							Audio.VO.Relik.JungleBoss_S25_2023_Alerted_020_asset,
							Audio.VO.Relik.JungleBoss_S25_2023_Alerted_031_asset,
							Audio.VO.Relik.JungleBoss_S25_2023_Alerted_040_asset,
							Audio.VO.Relik.JungleBoss_S25_2023_Alerted_050_asset
						}
					}
					FNVOType.FullAlertLost =>
					{ return array{
							Audio.VO.Relik.JungleBoss_S25_2023_LostTarget_010_asset,
							Audio.VO.Relik.JungleBoss_S25_2023_LostTarget_020_asset,
							Audio.VO.Relik.JungleBoss_S25_2023_LostTarget_030_asset,
							Audio.VO.Relik.JungleBoss_S25_2023_LostTarget_050_asset
						}
					}
					FNVOType.DBNO =>
					{ return array{
							Audio.VO.Relik.JungleBoss_S25_2023_DBNO_010_asset,
							Audio.VO.Relik.JungleBoss_S25_2023_DBNO_012_asset,
							Audio.VO.Relik.JungleBoss_S25_2023_DBNO_013_asset
						}
					}
					FNVOType.Interrogation =>
					{ return array{
							Audio.VO.Relik.JungleBoss_S25_2023_DBNO_010_asset,
							Audio.VO.Relik.JungleBoss_S25_2023_DBNO_012_asset,
							Audio.VO.Relik.JungleBoss_S25_2023_DBNO_013_asset

						}
					}
					FNVOType.Ability =>	
					{ return array{
							Audio.VO.Relik.JungleBoss_S25_2023_Teleport_Out_Angry_010_asset,
							Audio.VO.Relik.JungleBoss_S25_2023_Teleport_Out_Angry_011_asset,
							Audio.VO.Relik.JungleBoss_S25_2023_Teleport_Out_Angry_012_asset,
							Audio.VO.Relik.JungleBoss_S25_2023_Teleport_Out_Angry_014_asset
						}
					}
					FNVOType.AbilityEnd =>
					{ return array{
							Audio.VO.Relik.JungleBoss_S25_2023_Alerted_010_asset,
							Audio.VO.Relik.JungleBoss_S25_2023_Alerted_020_asset,
							Audio.VO.Relik.JungleBoss_S25_2023_Alerted_031_asset,
							Audio.VO.Relik.JungleBoss_S25_2023_Alerted_040_asset,
							Audio.VO.Relik.JungleBoss_S25_2023_Alerted_050_asset
						}
					}
					FNVOType.Ult =>
					{ return array{
							Audio.VO.Relik.JungleBoss_S25_2023_Alerted_010_asset,
							Audio.VO.Relik.JungleBoss_S25_2023_Alerted_020_asset,
							Audio.VO.Relik.JungleBoss_S25_2023_Alerted_031_asset,
							Audio.VO.Relik.JungleBoss_S25_2023_Alerted_040_asset,
							Audio.VO.Relik.JungleBoss_S25_2023_Alerted_050_asset
						}
					}
				}
			}
			FNClasses.KadoThorne =>
			{

				case( VOType )
				{
					FNVOType.Idle => 
					{ return array{
							Audio.VO.KadoThorne.GuruBoss_S26_2023_Idle_011_asset,
							Audio.VO.KadoThorne.GuruBoss_S26_2023_Idle_042_asset,
							Audio.VO.KadoThorne.GuruBoss_S26_2023_Idle_051_asset,
							Audio.VO.KadoThorne.GuruBoss_S26_2023_Idle_052_asset,
							Audio.VO.KadoThorne.GuruBoss_S26_2023_Idle_054_asset
						}
					}
					FNVOType.Sus =>
					{ return array{
							Audio.VO.KadoThorne.GuruBoss_S26_2023_Suspicious_012_asset,
							Audio.VO.KadoThorne.GuruBoss_S26_2023_Suspicious_021_asset,
							Audio.VO.KadoThorne.GuruBoss_S26_2023_Suspicious_030_asset,
							Audio.VO.KadoThorne.GuruBoss_S26_2023_Suspicious_040_asset,
							Audio.VO.KadoThorne.GuruBoss_S26_2023_Suspicious_050_asset
						}
					}
					FNVOType.ReturnIdle =>
					{ return array{
							Audio.VO.KadoThorne.GuruBoss_S26_2023_SusReturnToIdle_010_asset,
							Audio.VO.KadoThorne.GuruBoss_S26_2023_SusReturnToIdle_021_asset,
							Audio.VO.KadoThorne.GuruBoss_S26_2023_SusReturnToIdle_030_asset,
							Audio.VO.KadoThorne.GuruBoss_S26_2023_SusReturnToIdle_040_asset,
							Audio.VO.KadoThorne.GuruBoss_S26_2023_SusReturnToIdle_050_asset
						}
					}
					FNVOType.FullAlert =>
					{ return array{
							Audio.VO.KadoThorne.GuruBoss_S26_2023_Alerted_011_asset,
							Audio.VO.KadoThorne.GuruBoss_S26_2023_Alerted_020_asset,
							Audio.VO.KadoThorne.GuruBoss_S26_2023_Alerted_032_asset,
							Audio.VO.KadoThorne.GuruBoss_S26_2023_Alerted_040_asset,
							Audio.VO.KadoThorne.GuruBoss_S26_2023_Alerted_050_asset
						}
					}
					FNVOType.FullAlertLost =>
					{ return array{
							Audio.VO.KadoThorne.GuruBoss_S26_2023_LostTarget_021_asset,
							Audio.VO.KadoThorne.GuruBoss_S26_2023_LostTarget_031_asset,
							Audio.VO.KadoThorne.GuruBoss_S26_2023_LostTarget_041_asset,
							Audio.VO.KadoThorne.GuruBoss_S26_2023_LostTarget_050_asset,
							Audio.VO.KadoThorne.GuruBoss_S26_2023_LostTarget_053_asset
						}
					}
					FNVOType.DBNO =>
					{ return array{
							Audio.VO.KadoThorne.GuruBoss_S26_2023_Elim_010_asset,
							Audio.VO.KadoThorne.GuruBoss_S26_2023_Elim_012_asset,
							Audio.VO.KadoThorne.GuruBoss_S26_2023_Elim_013_asset
						}
					}
					FNVOType.Interrogation =>
					{ return array{
							Audio.VO.KadoThorne.GuruBoss_S26_2023_Elim_010_asset,
							Audio.VO.KadoThorne.GuruBoss_S26_2023_Elim_012_asset,
							Audio.VO.KadoThorne.GuruBoss_S26_2023_Elim_013_asset

						}
					}
					FNVOType.Ability =>	
					{ return array{
							Audio.VO.KadoThorne.GuruBoss_S26_2023_Alerted_011_asset,
							Audio.VO.KadoThorne.GuruBoss_S26_2023_Alerted_020_asset,
							Audio.VO.KadoThorne.GuruBoss_S26_2023_Alerted_032_asset,
							Audio.VO.KadoThorne.GuruBoss_S26_2023_Alerted_040_asset,
							Audio.VO.KadoThorne.GuruBoss_S26_2023_Alerted_050_asset
						}
					}
					FNVOType.AbilityEnd =>
					{ return array{
							Audio.VO.KadoThorne.GuruBoss_S26_2023_Alerted_011_asset,
							Audio.VO.KadoThorne.GuruBoss_S26_2023_Alerted_020_asset,
							Audio.VO.KadoThorne.GuruBoss_S26_2023_Alerted_032_asset,
							Audio.VO.KadoThorne.GuruBoss_S26_2023_Alerted_040_asset,
							Audio.VO.KadoThorne.GuruBoss_S26_2023_Alerted_050_asset
						}
					}
					FNVOType.Ult =>
					{ return array{
							Audio.VO.KadoThorne.GuruBoss_S26_2023_Alerted_011_asset,
							Audio.VO.KadoThorne.GuruBoss_S26_2023_Alerted_020_asset,
							Audio.VO.KadoThorne.GuruBoss_S26_2023_Alerted_032_asset,
							Audio.VO.KadoThorne.GuruBoss_S26_2023_Alerted_040_asset,
							Audio.VO.KadoThorne.GuruBoss_S26_2023_Alerted_050_asset
						}
					}
				}
			}
			FNClasses.TheDealer =>
			{
				case( VOType )
				{
					FNVOType.Idle => 
					{ return array{
							Audio.VO.TheDealer.RhombCamo_S26_2023_Idle_010_asset,
							Audio.VO.TheDealer.RhombCamo_S26_2023_Idle_050_asset
						}
					}
					FNVOType.Sus =>
					{ return array{
							Audio.VO.TheDealer.RhombCamo_S26_2023_Suspicious_010_asset,
							Audio.VO.TheDealer.RhombCamo_S26_2023_Suspicious_020_asset,
							Audio.VO.TheDealer.RhombCamo_S26_2023_Suspicious_041_asset,
							Audio.VO.TheDealer.RhombCamo_S26_2023_Suspicious_042_asset,
							Audio.VO.TheDealer.RhombCamo_S26_2023_Suspicious_051_asset
						}
					}
					FNVOType.ReturnIdle =>
					{ return array{
							Audio.VO.TheDealer.RhombCamo_S26_2023_SusReturnToIdle_010_asset,
							Audio.VO.TheDealer.RhombCamo_S26_2023_SusReturnToIdle_011_asset,
							Audio.VO.TheDealer.RhombCamo_S26_2023_SusReturnToIdle_030_asset,
							Audio.VO.TheDealer.RhombCamo_S26_2023_SusReturnToIdle_040_asset,
							Audio.VO.TheDealer.RhombCamo_S26_2023_SusReturnToIdle_051_asset
						}
					}
					FNVOType.FullAlert =>
					{ return array{
							Audio.VO.TheDealer.RhombCamo_S26_2023_Alerted_010_asset,
							Audio.VO.TheDealer.RhombCamo_S26_2023_Alerted_012_asset,
							Audio.VO.TheDealer.RhombCamo_S26_2023_Alerted_022_asset,
							Audio.VO.TheDealer.RhombCamo_S26_2023_Alerted_030_asset,
							Audio.VO.TheDealer.RhombCamo_S26_2023_Alerted_041_asset,
							Audio.VO.TheDealer.RhombCamo_S26_2023_Alerted_051_asset
						}
					}
					FNVOType.FullAlertLost =>
					{ return array{
							Audio.VO.TheDealer.RhombCamo_S26_2023_LostTarget_010_asset,
							Audio.VO.TheDealer.RhombCamo_S26_2023_LostTarget_031_asset,
							Audio.VO.TheDealer.RhombCamo_S26_2023_LostTarget_040_asset,
							Audio.VO.TheDealer.RhombCamo_S26_2023_LostTarget_050_asset,
							Audio.VO.TheDealer.RhombCamo_S26_2023_LostTarget_051_asset
						}
					}
					FNVOType.DBNO =>
					{ return array{
							Audio.VO.TheDealer.RhombCamo_S26_2023_DBNO_011_asset,
							Audio.VO.TheDealer.RhombCamo_S26_2023_DBNO_013_asset,
							Audio.VO.TheDealer.RhombCamo_S26_2023_DBNO_014_asset,
							Audio.VO.TheDealer.RhombCamo_S26_2023_DBNO_015_asset
						}
					}
					FNVOType.Interrogation =>
					{ return array{
							Audio.VO.TheDealer.RhombCamo_S26_2023_DBNO_011_asset,
							Audio.VO.TheDealer.RhombCamo_S26_2023_DBNO_013_asset,
							Audio.VO.TheDealer.RhombCamo_S26_2023_DBNO_014_asset,
							Audio.VO.TheDealer.RhombCamo_S26_2023_DBNO_015_asset

						}
					}
					FNVOType.Ability =>	
					{ return array{
							Audio.VO.TheDealer.RhombCamo_S26_2023_Alerted_010_asset,
							Audio.VO.TheDealer.RhombCamo_S26_2023_Alerted_012_asset,
							Audio.VO.TheDealer.RhombCamo_S26_2023_Alerted_022_asset,
							Audio.VO.TheDealer.RhombCamo_S26_2023_Alerted_030_asset,
							Audio.VO.TheDealer.RhombCamo_S26_2023_Alerted_041_asset,
							Audio.VO.TheDealer.RhombCamo_S26_2023_Alerted_051_asset
						}
					}
					FNVOType.AbilityEnd =>
					{ return array{
							Audio.VO.TheDealer.RhombCamo_S26_2023_Alerted_010_asset,
							Audio.VO.TheDealer.RhombCamo_S26_2023_Alerted_012_asset,
							Audio.VO.TheDealer.RhombCamo_S26_2023_Alerted_022_asset,
							Audio.VO.TheDealer.RhombCamo_S26_2023_Alerted_030_asset,
							Audio.VO.TheDealer.RhombCamo_S26_2023_Alerted_041_asset,
							Audio.VO.TheDealer.RhombCamo_S26_2023_Alerted_051_asset
						}
					}
					FNVOType.Ult =>
					{ return array{
							Audio.VO.TheDealer.RhombCamo_S26_2023_Alerted_010_asset,
							Audio.VO.TheDealer.RhombCamo_S26_2023_Alerted_012_asset,
							Audio.VO.TheDealer.RhombCamo_S26_2023_Alerted_022_asset,
							Audio.VO.TheDealer.RhombCamo_S26_2023_Alerted_030_asset,
							Audio.VO.TheDealer.RhombCamo_S26_2023_Alerted_041_asset,
							Audio.VO.TheDealer.RhombCamo_S26_2023_Alerted_051_asset
						}
					}
				}
			}
			FNClasses.Valeria =>
			{
				case( VOType )
				{
					FNVOType.Idle => 
					{ return array{
							Audio.VO.Valeria.VO_MadameMoth_S28_2023_Idle_010_asset,
							Audio.VO.Valeria.VO_MadameMoth_S28_2023_Idle_030_asset,
							Audio.VO.Valeria.VO_MadameMoth_S28_2023_Idle_040_asset,
							Audio.VO.Valeria.VO_MadameMoth_S28_2023_Idle_050_asset,
							Audio.VO.Valeria.VO_MadameMoth_S28_2023_Idle_061_asset,
							Audio.VO.Valeria.VO_MadameMoth_S28_2023_Idle_073_asset
						}
					}
					FNVOType.Sus =>
					{ return array{
							Audio.VO.Valeria.VO_MadameMoth_S28_2023_Suspicious_020_asset,
							Audio.VO.Valeria.VO_MadameMoth_S28_2023_Suspicious_030_asset,
							Audio.VO.Valeria.VO_MadameMoth_S28_2023_Suspicious_040_asset,
							Audio.VO.Valeria.VO_MadameMoth_S28_2023_Suspicious_041_asset,
							Audio.VO.Valeria.VO_MadameMoth_S28_2023_Suspicious_050_asset
						}
					}
					FNVOType.ReturnIdle =>
					{ return array{
							Audio.VO.Valeria.VO_MadameMoth_S28_2023_SusReturnToIdle_012_asset,
							Audio.VO.Valeria.VO_MadameMoth_S28_2023_SusReturnToIdle_020_asset,
							Audio.VO.Valeria.VO_MadameMoth_S28_2023_SusReturnToIdle_030_asset,
							Audio.VO.Valeria.VO_MadameMoth_S28_2023_SusReturnToIdle_040_asset,
							Audio.VO.Valeria.VO_MadameMoth_S28_2023_SusReturnToIdle_050_asset
						}
					}
					FNVOType.FullAlert =>
					{ return array{
							Audio.VO.Valeria.VO_MadameMoth_S28_2023_Alerted_010_asset,
							Audio.VO.Valeria.VO_MadameMoth_S28_2023_Alerted_020_asset,
							Audio.VO.Valeria.VO_MadameMoth_S28_2023_Alerted_021_asset,
							Audio.VO.Valeria.VO_MadameMoth_S28_2023_Alerted_031_asset,
							Audio.VO.Valeria.VO_MadameMoth_S28_2023_Alerted_040_asset
						}
					}
					FNVOType.FullAlertLost =>
					{ return array{
							Audio.VO.Valeria.VO_MadameMoth_S28_2023_LostTarget_011_asset,
							Audio.VO.Valeria.VO_MadameMoth_S28_2023_LostTarget_020_asset,
							Audio.VO.Valeria.VO_MadameMoth_S28_2023_LostTarget_030_asset,
							Audio.VO.Valeria.VO_MadameMoth_S28_2023_LostTarget_041_asset,
							Audio.VO.Valeria.VO_MadameMoth_S28_2023_LostTarget_050_asset
						}
					}
					FNVOType.DBNO =>
					{ return array{
							Audio.VO.Valeria.VO_MadameMoth_S28_2023_DBNO_010_asset,
							Audio.VO.Valeria.VO_MadameMoth_S28_2023_DBNO_011_asset,
							Audio.VO.Valeria.VO_MadameMoth_S28_2023_DBNO_012_asset,
							Audio.VO.Valeria.VO_MadameMoth_S28_2023_DBNO_013_asset,
							Audio.VO.Valeria.VO_MadameMoth_S28_2023_DBNO_016_asset
						}
					}
					FNVOType.Interrogation =>
					{ return array{
							Audio.VO.Valeria.VO_MadameMoth_S28_2023_DBNO_010_asset,
							Audio.VO.Valeria.VO_MadameMoth_S28_2023_DBNO_011_asset,
							Audio.VO.Valeria.VO_MadameMoth_S28_2023_DBNO_012_asset,
							Audio.VO.Valeria.VO_MadameMoth_S28_2023_DBNO_013_asset,
							Audio.VO.Valeria.VO_MadameMoth_S28_2023_DBNO_016_asset

						}
					}
					FNVOType.Ability =>	
					{ return array{
							Audio.VO.Valeria.VO_MadameMoth_S28_2023_Alerted_010_asset,
							Audio.VO.Valeria.VO_MadameMoth_S28_2023_Alerted_020_asset,
							Audio.VO.Valeria.VO_MadameMoth_S28_2023_Alerted_021_asset,
							Audio.VO.Valeria.VO_MadameMoth_S28_2023_Alerted_031_asset,
							Audio.VO.Valeria.VO_MadameMoth_S28_2023_Alerted_040_asset
						}
					}
					FNVOType.AbilityEnd =>
					{ return array{
							Audio.VO.Valeria.VO_MadameMoth_S28_2023_Alerted_010_asset,
							Audio.VO.Valeria.VO_MadameMoth_S28_2023_Alerted_020_asset,
							Audio.VO.Valeria.VO_MadameMoth_S28_2023_Alerted_021_asset,
							Audio.VO.Valeria.VO_MadameMoth_S28_2023_Alerted_031_asset,
							Audio.VO.Valeria.VO_MadameMoth_S28_2023_Alerted_040_asset
						}
					}
					FNVOType.Ult =>
					{ return array{
							Audio.VO.Valeria.VO_MadameMoth_S28_2023_Alerted_010_asset,
							Audio.VO.Valeria.VO_MadameMoth_S28_2023_Alerted_020_asset,
							Audio.VO.Valeria.VO_MadameMoth_S28_2023_Alerted_021_asset,
							Audio.VO.Valeria.VO_MadameMoth_S28_2023_Alerted_031_asset,
							Audio.VO.Valeria.VO_MadameMoth_S28_2023_Alerted_040_asset
						}
					}
				}
			}
			FNClasses.Oscar =>
			{
				case( VOType )
				{
					FNVOType.Idle => 
					{ return array{
							Audio.VO.Oscar.VO_BengalBasher_S28_2023_Idle_010_asset,
							Audio.VO.Oscar.VO_BengalBasher_S28_2023_Idle_020_asset,
							Audio.VO.Oscar.VO_BengalBasher_S28_2023_Idle_031_asset,
							Audio.VO.Oscar.VO_BengalBasher_S28_2023_Idle_040_asset,
							Audio.VO.Oscar.VO_BengalBasher_S28_2023_Idle_050_asset,
							Audio.VO.Oscar.VO_BengalBasher_S28_2023_Idle_062_asset,
							Audio.VO.Oscar.VO_BengalBasher_S28_2023_Idle_064_asset,
							Audio.VO.Oscar.VO_BengalBasher_S28_2023_Idle_067_asset,
							Audio.VO.Oscar.VO_BengalBasher_S28_2023_Idle_068_asset
						}
					}
					FNVOType.Sus =>
					{ return array{
							Audio.VO.Oscar.VO_BengalBasher_S28_2023_Suspicious_010_asset,
							Audio.VO.Oscar.VO_BengalBasher_S28_2023_Suspicious_011_asset,
							Audio.VO.Oscar.VO_BengalBasher_S28_2023_Suspicious_021_asset,
							Audio.VO.Oscar.VO_BengalBasher_S28_2023_Suspicious_022_asset,
							Audio.VO.Oscar.VO_BengalBasher_S28_2023_Suspicious_023_asset,
							Audio.VO.Oscar.VO_BengalBasher_S28_2023_Suspicious_030_asset,
							Audio.VO.Oscar.VO_BengalBasher_S28_2023_Suspicious_043_asset,
							Audio.VO.Oscar.VO_BengalBasher_S28_2023_Suspicious_050_asset,
							Audio.VO.Oscar.VO_BengalBasher_S28_2023_Suspicious_052_asset
						}
					}
					FNVOType.ReturnIdle =>
					{ return array{
							Audio.VO.Oscar.VO_BengalBasher_S28_2023_Suspicious_010_asset,
							Audio.VO.Oscar.VO_BengalBasher_S28_2023_Suspicious_011_asset,
							Audio.VO.Oscar.VO_BengalBasher_S28_2023_Suspicious_021_asset,
							Audio.VO.Oscar.VO_BengalBasher_S28_2023_Suspicious_022_asset,
							Audio.VO.Oscar.VO_BengalBasher_S28_2023_Suspicious_023_asset,
							Audio.VO.Oscar.VO_BengalBasher_S28_2023_Suspicious_030_asset,
							Audio.VO.Oscar.VO_BengalBasher_S28_2023_Suspicious_043_asset,
							Audio.VO.Oscar.VO_BengalBasher_S28_2023_Suspicious_050_asset,
							Audio.VO.Oscar.VO_BengalBasher_S28_2023_Suspicious_052_asset
						}
					}
					FNVOType.FullAlert =>
					{ return array{
							Audio.VO.Oscar.VO_BengalBasher_S28_2023_Alerted_011_asset,
							Audio.VO.Oscar.VO_BengalBasher_S28_2023_Alerted_020_asset,
							Audio.VO.Oscar.VO_BengalBasher_S28_2023_Alerted_031_asset,
							Audio.VO.Oscar.VO_BengalBasher_S28_2023_Alerted_042_asset,
							Audio.VO.Oscar.VO_BengalBasher_S28_2023_Alerted_051_asset,
							Audio.VO.Oscar.VO_BengalBasher_S28_2023_Alerted_060_asset,
							Audio.VO.Oscar.VO_BengalBasher_S28_2023_Alerted_063_asset
						}
					}
					FNVOType.FullAlertLost =>
					{ return array{
							Audio.VO.Oscar.VO_BengalBasher_S28_2023_LostTarget_010_asset,
							Audio.VO.Oscar.VO_BengalBasher_S28_2023_LostTarget_021_asset,
							Audio.VO.Oscar.VO_BengalBasher_S28_2023_LostTarget_030_asset,
							Audio.VO.Oscar.VO_BengalBasher_S28_2023_LostTarget_041_asset,
							Audio.VO.Oscar.VO_BengalBasher_S28_2023_LostTarget_051_asset
						}
					}
					FNVOType.DBNO =>
					{ return array{
							Audio.VO.Oscar.VO_BengalBasher_S28_2023_DBNO_011_asset,
							Audio.VO.Oscar.VO_BengalBasher_S28_2023_DBNO_012_asset,
							Audio.VO.Oscar.VO_BengalBasher_S28_2023_DBNO_015_asset,
							Audio.VO.Oscar.VO_BengalBasher_S28_2023_DBNO_016_asset
						}
					}
					FNVOType.Interrogation =>
					{ return array{
							Audio.VO.Oscar.VO_BengalBasher_S28_2023_DBNO_011_asset,
							Audio.VO.Oscar.VO_BengalBasher_S28_2023_DBNO_012_asset,
							Audio.VO.Oscar.VO_BengalBasher_S28_2023_DBNO_015_asset,
							Audio.VO.Oscar.VO_BengalBasher_S28_2023_DBNO_016_asset

						}
					}
					FNVOType.Ability =>	
					{ return array{
							Audio.VO.Oscar.VO_BengalBasher_S28_2023_Alerted_011_asset,
							Audio.VO.Oscar.VO_BengalBasher_S28_2023_Alerted_020_asset,
							Audio.VO.Oscar.VO_BengalBasher_S28_2023_Alerted_031_asset,
							Audio.VO.Oscar.VO_BengalBasher_S28_2023_Alerted_042_asset,
							Audio.VO.Oscar.VO_BengalBasher_S28_2023_Alerted_051_asset,
							Audio.VO.Oscar.VO_BengalBasher_S28_2023_Alerted_060_asset,
							Audio.VO.Oscar.VO_BengalBasher_S28_2023_Alerted_063_asset
						}
					}
					FNVOType.AbilityEnd =>
					{ return array{
							Audio.VO.Oscar.VO_BengalBasher_S28_2023_Alerted_011_asset,
							Audio.VO.Oscar.VO_BengalBasher_S28_2023_Alerted_020_asset,
							Audio.VO.Oscar.VO_BengalBasher_S28_2023_Alerted_031_asset,
							Audio.VO.Oscar.VO_BengalBasher_S28_2023_Alerted_042_asset,
							Audio.VO.Oscar.VO_BengalBasher_S28_2023_Alerted_051_asset,
							Audio.VO.Oscar.VO_BengalBasher_S28_2023_Alerted_060_asset,
							Audio.VO.Oscar.VO_BengalBasher_S28_2023_Alerted_063_asset
						}
					}
					FNVOType.Ult =>
					{ return array{
							Audio.VO.Oscar.VO_BengalBasher_S28_2023_Alerted_011_asset,
							Audio.VO.Oscar.VO_BengalBasher_S28_2023_Alerted_020_asset,
							Audio.VO.Oscar.VO_BengalBasher_S28_2023_Alerted_031_asset,
							Audio.VO.Oscar.VO_BengalBasher_S28_2023_Alerted_042_asset,
							Audio.VO.Oscar.VO_BengalBasher_S28_2023_Alerted_051_asset,
							Audio.VO.Oscar.VO_BengalBasher_S28_2023_Alerted_060_asset,
							Audio.VO.Oscar.VO_BengalBasher_S28_2023_Alerted_063_asset
						}
					}
				}
			}
			FNClasses.Nisha =>
			{
				case( VOType )
				{
					FNVOType.Idle => 
					{ return array{
							Audio.VO.Nisha.VO_CeremonialGuard_S28_2023_Idle_011_asset,
							Audio.VO.Nisha.VO_CeremonialGuard_S28_2023_Idle_020_asset,
							Audio.VO.Nisha.VO_CeremonialGuard_S28_2023_Idle_030_asset,
							Audio.VO.Nisha.VO_CeremonialGuard_S28_2023_Idle_041_asset,
							Audio.VO.Nisha.VO_CeremonialGuard_S28_2023_Idle_050_asset,
							Audio.VO.Nisha.VO_CeremonialGuard_S28_2023_Idle_060_asset
						}
					}
					FNVOType.Sus =>
					{ return array{
							Audio.VO.Nisha.VO_CeremonialGuard_S28_2023_Suspicious_011_asset,
							Audio.VO.Nisha.VO_CeremonialGuard_S28_2023_Suspicious_021_asset,
							Audio.VO.Nisha.VO_CeremonialGuard_S28_2023_Suspicious_032_asset,
							Audio.VO.Nisha.VO_CeremonialGuard_S28_2023_Suspicious_040_asset,
							Audio.VO.Nisha.VO_CeremonialGuard_S28_2023_Suspicious_050_asset
						}
					}
					FNVOType.ReturnIdle =>
					{ return array{
							Audio.VO.Nisha.VO_CeremonialGuard_S28_2023_Suspicious_011_asset,
							Audio.VO.Nisha.VO_CeremonialGuard_S28_2023_Suspicious_021_asset,
							Audio.VO.Nisha.VO_CeremonialGuard_S28_2023_Suspicious_032_asset,
							Audio.VO.Nisha.VO_CeremonialGuard_S28_2023_Suspicious_040_asset,
							Audio.VO.Nisha.VO_CeremonialGuard_S28_2023_Suspicious_050_asset
						}
					}
					FNVOType.FullAlert =>
					{ return array{
							Audio.VO.Nisha.VO_CeremonialGuard_S28_2023_Alerted_010_asset,
							Audio.VO.Nisha.VO_CeremonialGuard_S28_2023_Alerted_020_asset,
							Audio.VO.Nisha.VO_CeremonialGuard_S28_2023_Alerted_030_asset,
							Audio.VO.Nisha.VO_CeremonialGuard_S28_2023_Alerted_040_asset,
							Audio.VO.Nisha.VO_CeremonialGuard_S28_2023_Alerted_053_asset
						}
					}
					FNVOType.FullAlertLost =>
					{ return array{
							Audio.VO.Nisha.VO_CeremonialGuard_S28_2023_LostTarget_010_asset,
							Audio.VO.Nisha.VO_CeremonialGuard_S28_2023_LostTarget_021_asset,
							Audio.VO.Nisha.VO_CeremonialGuard_S28_2023_LostTarget_030_asset,
							Audio.VO.Nisha.VO_CeremonialGuard_S28_2023_LostTarget_040_asset,
							Audio.VO.Nisha.VO_CeremonialGuard_S28_2023_LostTarget_050_asset
						}
					}
					FNVOType.DBNO =>
					{ return array{
							Audio.VO.Nisha.VO_CeremonialGuard_S28_2023_DBNO_010_asset,
							Audio.VO.Nisha.VO_CeremonialGuard_S28_2023_DBNO_011_asset,
							Audio.VO.Nisha.VO_CeremonialGuard_S28_2023_DBNO_012_asset,
							Audio.VO.Nisha.VO_CeremonialGuard_S28_2023_DBNO_013_asset,
							Audio.VO.Nisha.VO_CeremonialGuard_S28_2023_DBNO_015_asset,
							Audio.VO.Nisha.VO_CeremonialGuard_S28_2023_DBNO_016_asset
						}
					}
					FNVOType.Interrogation =>
					{ return array{
							Audio.VO.Nisha.VO_CeremonialGuard_S28_2023_DBNO_010_asset,
							Audio.VO.Nisha.VO_CeremonialGuard_S28_2023_DBNO_011_asset,
							Audio.VO.Nisha.VO_CeremonialGuard_S28_2023_DBNO_012_asset,
							Audio.VO.Nisha.VO_CeremonialGuard_S28_2023_DBNO_013_asset,
							Audio.VO.Nisha.VO_CeremonialGuard_S28_2023_DBNO_015_asset,
							Audio.VO.Nisha.VO_CeremonialGuard_S28_2023_DBNO_016_asset
						}
					}
					FNVOType.Ability =>	
					{ return array{
							Audio.VO.Nisha.VO_CeremonialGuard_S28_2023_Alerted_010_asset,
							Audio.VO.Nisha.VO_CeremonialGuard_S28_2023_Alerted_020_asset,
							Audio.VO.Nisha.VO_CeremonialGuard_S28_2023_Alerted_030_asset,
							Audio.VO.Nisha.VO_CeremonialGuard_S28_2023_Alerted_040_asset,
							Audio.VO.Nisha.VO_CeremonialGuard_S28_2023_Alerted_053_asset
						}
					}
					FNVOType.AbilityEnd =>
					{ return array{
							Audio.VO.Nisha.VO_CeremonialGuard_S28_2023_Alerted_010_asset,
							Audio.VO.Nisha.VO_CeremonialGuard_S28_2023_Alerted_020_asset,
							Audio.VO.Nisha.VO_CeremonialGuard_S28_2023_Alerted_030_asset,
							Audio.VO.Nisha.VO_CeremonialGuard_S28_2023_Alerted_040_asset,
							Audio.VO.Nisha.VO_CeremonialGuard_S28_2023_Alerted_053_asset
						}
					}
					FNVOType.Ult =>
					{ return array{
							Audio.VO.Nisha.VO_CeremonialGuard_S28_2023_Alerted_010_asset,
							Audio.VO.Nisha.VO_CeremonialGuard_S28_2023_Alerted_020_asset,
							Audio.VO.Nisha.VO_CeremonialGuard_S28_2023_Alerted_030_asset,
							Audio.VO.Nisha.VO_CeremonialGuard_S28_2023_Alerted_040_asset,
							Audio.VO.Nisha.VO_CeremonialGuard_S28_2023_Alerted_053_asset
						}
					}
				}
			}
			FNClasses.Montague =>
			{
				case( VOType )
				{
					FNVOType.Idle => 
					{ return array{
							Audio.VO.Montague.VO_DiamondHeart_S28_2023_Idle_010_asset,
							Audio.VO.Montague.VO_DiamondHeart_S28_2023_Idle_031_asset,
							Audio.VO.Montague.VO_DiamondHeart_S28_2023_Idle_040_asset,
							Audio.VO.Montague.VO_DiamondHeart_S28_2023_Idle_050_asset,
							Audio.VO.Montague.VO_DiamondHeart_S28_2023_Idle_060_asset
						}
					}
					FNVOType.Sus =>
					{ return array{
							Audio.VO.Montague.VO_DiamondHeart_S28_2023_Suspicious_011_asset,
							Audio.VO.Montague.VO_DiamondHeart_S28_2023_Suspicious_021_asset,
							Audio.VO.Montague.VO_DiamondHeart_S28_2023_Suspicious_031_asset,
							Audio.VO.Montague.VO_DiamondHeart_S28_2023_Suspicious_050_asset,
							Audio.VO.Montague.VO_DiamondHeart_S28_2023_Suspicious_051_asset
						}
					}
					FNVOType.ReturnIdle =>
					{ return array{
							Audio.VO.Montague.VO_DiamondHeart_S28_2023_SusReturnToIdle_012_asset,
							Audio.VO.Montague.VO_DiamondHeart_S28_2023_SusReturnToIdle_020_asset,
							Audio.VO.Montague.VO_DiamondHeart_S28_2023_SusReturnToIdle_031_asset,
							Audio.VO.Montague.VO_DiamondHeart_S28_2023_SusReturnToIdle_040_asset,
							Audio.VO.Montague.VO_DiamondHeart_S28_2023_SusReturnToIdle_050_asset
						}
					}
					FNVOType.FullAlert =>
					{ return array{
							Audio.VO.Montague.VO_DiamondHeart_S28_2023_Alerted_010_asset,
							Audio.VO.Montague.VO_DiamondHeart_S28_2023_Alerted_011_asset,
							Audio.VO.Montague.VO_DiamondHeart_S28_2023_Alerted_021_asset,
							Audio.VO.Montague.VO_DiamondHeart_S28_2023_Alerted_031_asset,
							Audio.VO.Montague.VO_DiamondHeart_S28_2023_Alerted_040_asset
						}
					}
					FNVOType.FullAlertLost =>
					{ return array{
							Audio.VO.Montague.VO_DiamondHeart_S28_2023_LostTarget_020_asset,
							Audio.VO.Montague.VO_DiamondHeart_S28_2023_LostTarget_040_asset,
							Audio.VO.Montague.VO_DiamondHeart_S28_2023_LostTarget_041_asset,
							Audio.VO.Montague.VO_DiamondHeart_S28_2023_LostTarget_050_asset,
							Audio.VO.Montague.VO_DiamondHeart_S28_2023_LostTarget_052_asset
						}
					}
					FNVOType.DBNO =>
					{ return array{
							Audio.VO.Montague.VO_DiamondHeart_S28_2023_DBNO_011_asset,
							Audio.VO.Montague.VO_DiamondHeart_S28_2023_DBNO_012_asset,
							Audio.VO.Montague.VO_DiamondHeart_S28_2023_DBNO_013_asset,
							Audio.VO.Montague.VO_DiamondHeart_S28_2023_DBNO_014_asset,
							Audio.VO.Montague.VO_DiamondHeart_S28_2023_DBNO_015_asset
						}
					}
					FNVOType.Interrogation =>
					{ return array{
							Audio.VO.Montague.VO_DiamondHeart_S28_2023_DBNO_011_asset,
							Audio.VO.Montague.VO_DiamondHeart_S28_2023_DBNO_012_asset,
							Audio.VO.Montague.VO_DiamondHeart_S28_2023_DBNO_013_asset,
							Audio.VO.Montague.VO_DiamondHeart_S28_2023_DBNO_014_asset,
							Audio.VO.Montague.VO_DiamondHeart_S28_2023_DBNO_015_asset

						}
					}
					FNVOType.Ability =>	
					{ return array{
							Audio.VO.Montague.VO_DiamondHeart_S28_2023_Alerted_010_asset,
							Audio.VO.Montague.VO_DiamondHeart_S28_2023_Alerted_011_asset,
							Audio.VO.Montague.VO_DiamondHeart_S28_2023_Alerted_021_asset,
							Audio.VO.Montague.VO_DiamondHeart_S28_2023_Alerted_031_asset,
							Audio.VO.Montague.VO_DiamondHeart_S28_2023_Alerted_040_asset
						}
					}
					FNVOType.AbilityEnd =>
					{ return array{
							Audio.VO.Montague.VO_DiamondHeart_S28_2023_Alerted_010_asset,
							Audio.VO.Montague.VO_DiamondHeart_S28_2023_Alerted_011_asset,
							Audio.VO.Montague.VO_DiamondHeart_S28_2023_Alerted_021_asset,
							Audio.VO.Montague.VO_DiamondHeart_S28_2023_Alerted_031_asset,
							Audio.VO.Montague.VO_DiamondHeart_S28_2023_Alerted_040_asset
						}
					}
					FNVOType.Ult =>
					{ return array{
							Audio.VO.Montague.VO_DiamondHeart_S28_2023_Alerted_010_asset,
							Audio.VO.Montague.VO_DiamondHeart_S28_2023_Alerted_011_asset,
							Audio.VO.Montague.VO_DiamondHeart_S28_2023_Alerted_021_asset,
							Audio.VO.Montague.VO_DiamondHeart_S28_2023_Alerted_031_asset,
							Audio.VO.Montague.VO_DiamondHeart_S28_2023_Alerted_040_asset
						}
					}
				}
			}
			FNClasses.SgtWinter =>
			{
				case( VOType )
				{
					FNVOType.Idle => 
					{ return array{
							Audio.VO.SgtWinter.SilkWalker_2021_Idle_030_asset,
							Audio.VO.SgtWinter.SilkWalker_2021_Idle_040_asset,
							Audio.VO.SgtWinter.SilkWalker_2021_Idle_050_asset,
							Audio.VO.SgtWinter.SilkWalker_2021_Idle_051_asset
						}
					}
					FNVOType.Sus =>
					{ return array{
							Audio.VO.SgtWinter.SilkWalker_2021_Greeting_010_asset,
							Audio.VO.SgtWinter.SilkWalker_2021_Greeting_020_asset,
							Audio.VO.SgtWinter.SilkWalker_2021_Greeting_030_asset,
							Audio.VO.SgtWinter.SilkWalker_2021_Greeting_040_asset,
							Audio.VO.SgtWinter.SilkWalker_2021_Greeting_050_asset
						}
					}
					FNVOType.ReturnIdle =>
					{ return array{
							Audio.VO.SgtWinter.SilkWalker_2021_Greeting_010_asset,
							Audio.VO.SgtWinter.SilkWalker_2021_Greeting_020_asset,
							Audio.VO.SgtWinter.SilkWalker_2021_Greeting_030_asset,
							Audio.VO.SgtWinter.SilkWalker_2021_Greeting_040_asset,
							Audio.VO.SgtWinter.SilkWalker_2021_Greeting_050_asset
						}
					}
					FNVOType.FullAlert =>
					{ return array{
							Audio.VO.SgtWinter.SilkWalker_2021_FullAlert_020_asset,
							Audio.VO.SgtWinter.SilkWalker_2021_FullAlert_030_asset,
							Audio.VO.SgtWinter.SilkWalker_2021_FullAlert_050_asset,
							Audio.VO.SgtWinter.SilkWalker_2021_FullAlert_060_asset,
							Audio.VO.SgtWinter.SilkWalker_2021_FullAlert_070_asset
						}
					}
					FNVOType.FullAlertLost =>
					{ return array{
							Audio.VO.Midas.Henchman_Midas_FullAlertLostTarget_010_asset,
							Audio.VO.Midas.Henchman_Midas_FullAlertLostTarget_020_asset,
							Audio.VO.Midas.Henchman_Midas_FullAlertLostTarget_030_asset,
							Audio.VO.Midas.Henchman_Midas_FullAlertLostTarget_040_asset,
							Audio.VO.Midas.Henchman_Midas_FullAlertLostTarget_050_asset,
							Audio.VO.Midas.Henchman_Midas_FullAlertLostTarget_060_asset
						}
					}
					FNVOType.DBNO =>
					{ return array{
							Audio.VO.SgtWinter.SilkWalker_2021_DBNO_Enter_010_asset,
							Audio.VO.SgtWinter.SilkWalker_2021_DBNO_Enter_011_asset,
							Audio.VO.SgtWinter.SilkWalker_2021_DBNO_Enter_012_asset
						}
					}
					FNVOType.Interrogation =>
					{ return array{
							Audio.VO.Midas.Henchman_Midas_Interrogation_010_asset,
							Audio.VO.Midas.Henchman_Midas_Interrogation_020_asset,
							Audio.VO.Midas.Henchman_Midas_Interrogation_030_asset,
							Audio.VO.Midas.Henchman_Midas_Interrogation_040_asset,
							Audio.VO.Midas.Henchman_Midas_Interrogation_050_asset,
							Audio.VO.Midas.Henchman_Midas_Interrogation_060_asset

						}
					}
					FNVOType.Ability =>	
					{ return array{
							Audio.VO.SgtWinter.SilkWalker_2021_ThrowGift_Effort_010_asset,
							Audio.VO.SgtWinter.SilkWalker_2021_ThrowGift_Effort_011_asset,
							Audio.VO.SgtWinter.SilkWalker_2021_ThrowGift_Effort_012_asset,
							Audio.VO.SgtWinter.SilkWalker_2021_ThrowGift_Effort_013_asset
						}
					}
					FNVOType.AbilityEnd =>
					{ return array{
							Audio.VO.SgtWinter.SilkWalker_2021_ThrowGift_Goodbye_010_asset,
							Audio.VO.SgtWinter.SilkWalker_2021_ThrowGift_Goodbye_020_asset,
							Audio.VO.SgtWinter.SilkWalker_2021_ThrowGift_Goodbye_030_asset,
							Audio.VO.SgtWinter.SilkWalker_2021_ThrowGift_Goodbye_040_asset
						}
					}
					FNVOType.Ult =>
					{ return array{
							Audio.VO.SgtWinter.SilkWalker_2021_FullAlert_020_asset,
							Audio.VO.SgtWinter.SilkWalker_2021_FullAlert_030_asset,
							Audio.VO.SgtWinter.SilkWalker_2021_FullAlert_050_asset,
							Audio.VO.SgtWinter.SilkWalker_2021_FullAlert_060_asset,
							Audio.VO.SgtWinter.SilkWalker_2021_FullAlert_070_asset
						}
					}
				}
			}
			FNClasses.SpireAssassin =>
			{
				case( VOType )
				{
					FNVOType.Idle => 
					{ return array{
							Audio.VO.SpireAssassin.InnerTowerBoss_2021_Idle_010_asset,
							Audio.VO.SpireAssassin.InnerTowerBoss_2021_Idle_015_asset,
							Audio.VO.SpireAssassin.InnerTowerBoss_2021_Idle_022_asset,
							Audio.VO.SpireAssassin.InnerTowerBoss_2021_Idle_040_asset,
							Audio.VO.SpireAssassin.InnerTowerBoss_2021_Idle_050_asset,
							Audio.VO.SpireAssassin.InnerTowerBoss_2021_Idle_060_asset
						}
					}
					FNVOType.Sus =>
					{ return array{
							Audio.VO.SpireAssassin.InnerTowerBoss_2021_Suspicious_010_asset,
							Audio.VO.SpireAssassin.InnerTowerBoss_2021_Suspicious_020_asset,
							Audio.VO.SpireAssassin.InnerTowerBoss_2021_Suspicious_024_asset,
							Audio.VO.SpireAssassin.InnerTowerBoss_2021_Suspicious_030_asset,
							Audio.VO.SpireAssassin.InnerTowerBoss_2021_Suspicious_040_asset,
							Audio.VO.SpireAssassin.InnerTowerBoss_2021_Suspicious_052_asset
						}
					}
					FNVOType.ReturnIdle =>
					{ return array{
							Audio.VO.SpireAssassin.InnerTowerBoss_2021_SuspiciousReturnToIdle_010_asset,
							Audio.VO.SpireAssassin.InnerTowerBoss_2021_SuspiciousReturnToIdle_020_asset,
							Audio.VO.SpireAssassin.InnerTowerBoss_2021_SuspiciousReturnToIdle_031_asset,
							Audio.VO.SpireAssassin.InnerTowerBoss_2021_SuspiciousReturnToIdle_040_asset,
							Audio.VO.SpireAssassin.InnerTowerBoss_2021_SuspiciousReturnToIdle_050_asset,
							Audio.VO.SpireAssassin.InnerTowerBoss_2021_SuspiciousReturnToIdle_061_asset
						}
					}
					FNVOType.FullAlert =>
					{ return array{
							Audio.VO.SpireAssassin.InnerTowerBoss_2021_FullAlert_020_asset,
							Audio.VO.SpireAssassin.InnerTowerBoss_2021_FullAlert_021_asset,
							Audio.VO.SpireAssassin.InnerTowerBoss_2021_FullAlert_042_asset,
							Audio.VO.SpireAssassin.InnerTowerBoss_2021_FullAlert_043_asset,
							Audio.VO.SpireAssassin.InnerTowerBoss_2021_FullAlert_050_asset,
							Audio.VO.SpireAssassin.InnerTowerBoss_2021_FullAlert_051_asset
						}
					}
					FNVOType.FullAlertLost =>
					{ return array{
							Audio.VO.SpireAssassin.InnerTowerBoss_2021_FullAlertLostTarget_010_asset,
							Audio.VO.SpireAssassin.InnerTowerBoss_2021_FullAlertLostTarget_020_asset,
							Audio.VO.SpireAssassin.InnerTowerBoss_2021_FullAlertLostTarget_030_asset,
							Audio.VO.SpireAssassin.InnerTowerBoss_2021_FullAlertLostTarget_040_asset,
							Audio.VO.SpireAssassin.InnerTowerBoss_2021_FullAlertLostTarget_050_asset,
							Audio.VO.SpireAssassin.InnerTowerBoss_2021_FullAlertLostTarget_060_asset
						}
					}
					FNVOType.DBNO =>
					{ return array{
							Audio.VO.SpireAssassin.InnerTowerBoss_2021_DBNO_Enter_010_asset,
							Audio.VO.SpireAssassin.InnerTowerBoss_2021_DBNO_Enter_011_asset,
							Audio.VO.SpireAssassin.InnerTowerBoss_2021_DBNO_Enter_012_asset,
							Audio.VO.SpireAssassin.InnerTowerBoss_2021_DBNO_Enter_013_asset,
							Audio.VO.SpireAssassin.InnerTowerBoss_2021_DBNO_Enter_014_asset,
							Audio.VO.SpireAssassin.InnerTowerBoss_2021_DBNO_Enter_015_asset
						}
					}
					FNVOType.Interrogation =>
					{ return array{
							Audio.VO.SpireAssassin.InnerTowerBoss_2021_DBNO_Enter_010_asset,
							Audio.VO.SpireAssassin.InnerTowerBoss_2021_DBNO_Enter_011_asset,
							Audio.VO.SpireAssassin.InnerTowerBoss_2021_DBNO_Enter_012_asset,
							Audio.VO.SpireAssassin.InnerTowerBoss_2021_DBNO_Enter_013_asset,
							Audio.VO.SpireAssassin.InnerTowerBoss_2021_DBNO_Enter_014_asset,
							Audio.VO.SpireAssassin.InnerTowerBoss_2021_DBNO_Enter_015_asset

						}
					}
					FNVOType.Ability =>	
					{ return array{
							Audio.VO.SpireAssassin.InnerTowerBoss_2021_WarpAbility_010_asset,
							Audio.VO.SpireAssassin.InnerTowerBoss_2021_WarpAbility_012_asset,
							Audio.VO.SpireAssassin.InnerTowerBoss_2021_WarpAbility_013_asset,
							Audio.VO.SpireAssassin.InnerTowerBoss_2021_WarpAbility_014_asset,
							Audio.VO.SpireAssassin.InnerTowerBoss_2021_WarpAbility_015_asset,
							Audio.VO.SpireAssassin.InnerTowerBoss_2021_WarpAbility_031_asset
						}
					}
					FNVOType.AbilityEnd =>
					{ return array{
							Audio.VO.SpireAssassin.InnerTowerBoss_2021_FullAlert_020_asset,
							Audio.VO.SpireAssassin.InnerTowerBoss_2021_FullAlert_021_asset,
							Audio.VO.SpireAssassin.InnerTowerBoss_2021_FullAlert_042_asset,
							Audio.VO.SpireAssassin.InnerTowerBoss_2021_FullAlert_043_asset,
							Audio.VO.SpireAssassin.InnerTowerBoss_2021_FullAlert_050_asset,
							Audio.VO.SpireAssassin.InnerTowerBoss_2021_FullAlert_051_asset
						}
					}
					FNVOType.Ult =>
					{ return array{
							Audio.VO.SpireAssassin.InnerTowerBoss_2021_FullAlert_020_asset,
							Audio.VO.SpireAssassin.InnerTowerBoss_2021_FullAlert_021_asset,
							Audio.VO.SpireAssassin.InnerTowerBoss_2021_FullAlert_042_asset,
							Audio.VO.SpireAssassin.InnerTowerBoss_2021_FullAlert_043_asset,
							Audio.VO.SpireAssassin.InnerTowerBoss_2021_FullAlert_050_asset,
							Audio.VO.SpireAssassin.InnerTowerBoss_2021_FullAlert_051_asset
						}
					}
				}
			}
			FNClasses.CubeAssassin =>
			{
				case( VOType )
				{
					FNVOType.Idle => 
					{ return array{
							Audio.VO.CubeAssassin.SmallBags_2021_Idle_010_asset,
							Audio.VO.CubeAssassin.SmallBags_2021_Idle_020_asset,
							Audio.VO.CubeAssassin.SmallBags_2021_Idle_030_asset,
							Audio.VO.CubeAssassin.SmallBags_2021_Idle_040_asset,
							Audio.VO.CubeAssassin.SmallBags_2021_Idle_050_asset
						}
					}
					FNVOType.Sus =>
					{ return array{
							Audio.VO.CubeAssassin.SmallBags_2021_Suspicious_010_asset,
							Audio.VO.CubeAssassin.SmallBags_2021_Suspicious_020_asset,
							Audio.VO.CubeAssassin.SmallBags_2021_Suspicious_030_asset,
							Audio.VO.CubeAssassin.SmallBags_2021_Suspicious_040_asset
						}
					}
					FNVOType.ReturnIdle =>
					{ return array{
							Audio.VO.CubeAssassin.SmallBags_2021_SuspiciousReturnToIdle_010_asset,
							Audio.VO.CubeAssassin.SmallBags_2021_SuspiciousReturnToIdle_020_asset,
							Audio.VO.CubeAssassin.SmallBags_2021_SuspiciousReturnToIdle_030_asset,
							Audio.VO.CubeAssassin.SmallBags_2021_SuspiciousReturnToIdle_040_asset,
							Audio.VO.CubeAssassin.SmallBags_2021_SuspiciousReturnToIdle_050_asset
						}
					}
					FNVOType.FullAlert =>
					{ return array{
							Audio.VO.CubeAssassin.SmallBags_2021_FullAlert_010_asset,
							Audio.VO.CubeAssassin.SmallBags_2021_FullAlert_020_asset,
							Audio.VO.CubeAssassin.SmallBags_2021_FullAlert_030_asset,
							Audio.VO.CubeAssassin.SmallBags_2021_FullAlert_040_asset,
							Audio.VO.CubeAssassin.SmallBags_2021_FullAlert_050_asset
						}
					}
					FNVOType.FullAlertLost =>
					{ return array{
							Audio.VO.CubeAssassin.SmallBags_2021_FullAlertLostTarget_010_asset,
							Audio.VO.CubeAssassin.SmallBags_2021_FullAlertLostTarget_020_asset,
							Audio.VO.CubeAssassin.SmallBags_2021_FullAlertLostTarget_030_asset,
							Audio.VO.CubeAssassin.SmallBags_2021_FullAlertLostTarget_040_asset,
							Audio.VO.CubeAssassin.SmallBags_2021_FullAlertLostTarget_050_asset
						}
					}
					FNVOType.DBNO =>
					{ return array{
							Audio.VO.CubeAssassin.SmallBags_2021_DBNO_Enter_010_asset,
							Audio.VO.CubeAssassin.SmallBags_2021_DBNO_Enter_011_asset,
							Audio.VO.CubeAssassin.SmallBags_2021_DBNO_Enter_012_asset,
							Audio.VO.CubeAssassin.SmallBags_2021_DBNO_Enter_013_asset,
							Audio.VO.CubeAssassin.SmallBags_2021_DBNO_Enter_014_asset
						}
					}
					FNVOType.Interrogation =>
					{ return array{
							Audio.VO.CubeAssassin.SmallBags_2021_FullAlert_010_asset,
							Audio.VO.CubeAssassin.SmallBags_2021_FullAlert_020_asset,
							Audio.VO.CubeAssassin.SmallBags_2021_FullAlert_030_asset,
							Audio.VO.CubeAssassin.SmallBags_2021_FullAlert_040_asset,
							Audio.VO.CubeAssassin.SmallBags_2021_FullAlert_050_asset

						}
					}
					FNVOType.Ability =>	
					{ return array{
							Audio.VO.CubeAssassin.SmallBags_2021_PersonalShieldCast_010_asset,
							Audio.VO.CubeAssassin.SmallBags_2021_PersonalShieldCast_011_asset,
							Audio.VO.CubeAssassin.SmallBags_2021_PersonalShieldCast_012_asset,
							Audio.VO.CubeAssassin.SmallBags_2021_PersonalShieldCast_013_asset
						}
					}
					FNVOType.AbilityEnd =>
					{ return array{
							Audio.VO.CubeAssassin.SmallBags_2021_FullAlert_010_asset,
							Audio.VO.CubeAssassin.SmallBags_2021_FullAlert_020_asset,
							Audio.VO.CubeAssassin.SmallBags_2021_FullAlert_030_asset,
							Audio.VO.CubeAssassin.SmallBags_2021_FullAlert_040_asset,
							Audio.VO.CubeAssassin.SmallBags_2021_FullAlert_050_asset
						}
					}
					FNVOType.Ult =>
					{ return array{
							Audio.VO.CubeAssassin.SmallBags_2021_FullAlert_010_asset,
							Audio.VO.CubeAssassin.SmallBags_2021_FullAlert_020_asset,
							Audio.VO.CubeAssassin.SmallBags_2021_FullAlert_030_asset,
							Audio.VO.CubeAssassin.SmallBags_2021_FullAlert_040_asset,
							Audio.VO.CubeAssassin.SmallBags_2021_FullAlert_050_asset
						}
					}
				}
			}
			FNClasses.CubeQueen =>
			{
				case( VOType )
				{
					FNVOType.Idle => 
					{ return array{
							Audio.VO.CubeQueen.CUBEQUEEN_Guava_20_014_asset,
							Audio.VO.CubeQueen.CUBEQUEEN_Guava_20_016_asset
						}
					}
					FNVOType.Sus =>
					{ return array{
							Audio.VO.CubeQueen.CUBEQUEEN_Guava_20_014_asset,
							Audio.VO.CubeQueen.CUBEQUEEN_Guava_20_016_asset
						}
					}
					FNVOType.ReturnIdle =>
					{ return array{
							Audio.VO.CubeQueen.CUBEQUEEN_Guava_20_014_asset,
							Audio.VO.CubeQueen.CUBEQUEEN_Guava_20_016_asset
						}
					}
					FNVOType.FullAlert =>
					{ return array{
							Audio.VO.CubeQueen.CUBEQUEEN_Guava_20_014_asset,
							Audio.VO.CubeQueen.CUBEQUEEN_Guava_20_016_asset
						}
					}
					FNVOType.FullAlertLost =>
					{ return array{
							Audio.VO.CubeQueen.CUBEQUEEN_Guava_20_014_asset,
							Audio.VO.CubeQueen.CUBEQUEEN_Guava_20_016_asset
						}
					}
					FNVOType.DBNO =>
					{ return array{
							Audio.VO.CubeQueen.CUBEQUEEN_Guava_51_013_asset,
							Audio.VO.CubeQueen.CUBEQUEEN_Guava_21_013_asset
						}
					}
					FNVOType.Interrogation =>
					{ return array{
						}
					}
					FNVOType.Ability =>	
					{ return array{
							Audio.VO.CubeQueen.CUBEQUEEN_Guava_35_012_asset,
							Audio.VO.CubeQueen.CUBEQUEEN_Guava_35_016_asset
						}
					}
					FNVOType.AbilityEnd =>
					{ return array{
							Audio.VO.CubeQueen.CUBEQUEEN_Guava_35_012_asset,
							Audio.VO.CubeQueen.CUBEQUEEN_Guava_35_016_asset
						}
					}
					FNVOType.Ult =>
					{ return array{
							Audio.VO.CubeQueen.CUBEQUEEN_Guava_41_018_asset,
							Audio.VO.CubeQueen.CUBEQUEEN_Guava_45_015_asset
						}
					}
				}
			}
			FNClasses.ZygAndChoppy =>
			{
				case( VOType )
				{
					FNVOType.Idle => 
					{ return array{
							Audio.VO.ZygAndChoppy.AlienLeader_2021_Idle_010_asset,
							Audio.VO.ZygAndChoppy.AlienLeader_2021_Idle_020_asset,
							Audio.VO.ZygAndChoppy.AlienLeader_2021_Idle_030_asset,
							Audio.VO.ZygAndChoppy.AlienLeader_2021_Idle_040_asset,
							Audio.VO.ZygAndChoppy.AlienLeader_2021_Idle_050_asset
						}
					}
					FNVOType.Sus =>
					{ return array{
							Audio.VO.ZygAndChoppy.AlienLeader_2021_Suspicious_010_asset,
							Audio.VO.ZygAndChoppy.AlienLeader_2021_Suspicious_041_asset,
							Audio.VO.ZygAndChoppy.AlienLeader_2021_Suspicious_050_asset,
							Audio.VO.ZygAndChoppy.AlienLeader_2021_Suspicious_051_asset
						}
					}
					FNVOType.ReturnIdle =>
					{ return array{
							Audio.VO.ZygAndChoppy.AlienLeader_2021_SuspiciousReturnToIdle_010_asset,
							Audio.VO.ZygAndChoppy.AlienLeader_2021_SuspiciousReturnToIdle_030_asset,
							Audio.VO.ZygAndChoppy.AlienLeader_2021_SuspiciousReturnToIdle_040_asset,
							Audio.VO.ZygAndChoppy.AlienLeader_2021_SuspiciousReturnToIdle_050_asset
						}
					}
					FNVOType.FullAlert =>
					{ return array{
							Audio.VO.ZygAndChoppy.AlienLeader_2021_FullAlert_010_asset,
							Audio.VO.ZygAndChoppy.AlienLeader_2021_FullAlert_020_asset,
							Audio.VO.ZygAndChoppy.AlienLeader_2021_FullAlert_030_asset,
							Audio.VO.ZygAndChoppy.AlienLeader_2021_FullAlert_040_asset
						}
					}
					FNVOType.FullAlertLost =>
					{ return array{
							Audio.VO.ZygAndChoppy.AlienLeader_2021_FullAlertLostTarget_011_asset,
							Audio.VO.ZygAndChoppy.AlienLeader_2021_FullAlertLostTarget_020_asset,
							Audio.VO.ZygAndChoppy.AlienLeader_2021_FullAlertLostTarget_031_asset,
							Audio.VO.ZygAndChoppy.AlienLeader_2021_FullAlertLostTarget_041_asset,
							Audio.VO.ZygAndChoppy.AlienLeader_2021_FullAlertLostTarget_051_asset
						}
					}
					FNVOType.DBNO =>
					{ return array{
							Audio.VO.ZygAndChoppy.AlienLeader_2021_DBNO_Enter_010_asset,
							Audio.VO.ZygAndChoppy.AlienLeader_2021_DBNO_Enter_011_asset,
							Audio.VO.ZygAndChoppy.AlienLeader_2021_DBNO_Enter_012_asset,
							Audio.VO.ZygAndChoppy.AlienLeader_2021_DBNO_Enter_013_asset
						}
					}
					FNVOType.Interrogation =>
					{ return array{
							Audio.VO.ZygAndChoppy.AlienLeader_2021_DBNO_Enter_010_asset,
							Audio.VO.ZygAndChoppy.AlienLeader_2021_DBNO_Enter_011_asset,
							Audio.VO.ZygAndChoppy.AlienLeader_2021_DBNO_Enter_012_asset,
							Audio.VO.ZygAndChoppy.AlienLeader_2021_DBNO_Enter_013_asset

						}
					}
					FNVOType.Ability =>	
					{ return array{
							Audio.VO.ZygAndChoppy.AlienLeader_2021_TendEgg_010_asset,
							Audio.VO.ZygAndChoppy.AlienLeader_2021_TendEgg_011_asset,
							Audio.VO.ZygAndChoppy.AlienLeader_2021_TendEgg_020_asset
						}
					}
					FNVOType.AbilityEnd =>
					{ return array{
							Audio.VO.ZygAndChoppy.AlienLeader_2021_TendEgg_010_asset,
							Audio.VO.ZygAndChoppy.AlienLeader_2021_TendEgg_011_asset,
							Audio.VO.ZygAndChoppy.AlienLeader_2021_TendEgg_020_asset
						}
					}
					FNVOType.Ult =>
					{ return array{
							Audio.VO.ZygAndChoppy.AlienLeader_2021_FullAlert_010_asset,
							Audio.VO.ZygAndChoppy.AlienLeader_2021_FullAlert_020_asset,
							Audio.VO.ZygAndChoppy.AlienLeader_2021_FullAlert_030_asset,
							Audio.VO.ZygAndChoppy.AlienLeader_2021_FullAlert_040_asset
						}
					}
				}
			}
			FNClasses.Zeus =>
			{
				case( VOType )
				{
					FNVOType.Idle => 
					{ return array{
							Audio.VO.Zeus.VO_Zeus_S29_2023_AlertIdle_010_asset,
							Audio.VO.Zeus.VO_Zeus_S29_2023_AlertIdle_021_asset,
							Audio.VO.Zeus.VO_Zeus_S29_2023_AlertIdle_030_asset,
							Audio.VO.Zeus.VO_Zeus_S29_2023_AlertIdle_040_asset,
							Audio.VO.Zeus.VO_Zeus_S29_2023_AlertIdle_050_asset
						}
					}
					FNVOType.Sus =>
					{ return array{
							Audio.VO.Zeus.VO_Zeus_S29_2023_Suspicious_011_asset,
							Audio.VO.Zeus.VO_Zeus_S29_2023_Suspicious_020_asset,
							Audio.VO.Zeus.VO_Zeus_S29_2023_Suspicious_030_asset,
							Audio.VO.Zeus.VO_Zeus_S29_2023_Suspicious_041_asset,
							Audio.VO.Zeus.VO_Zeus_S29_2023_Suspicious_051_asset
						}
					}
					FNVOType.ReturnIdle =>
					{ return array{
							Audio.VO.Zeus.VO_Zeus_S29_2023_SusReturnToIdle_010_asset,
							Audio.VO.Zeus.VO_Zeus_S29_2023_SusReturnToIdle_020_asset,
							Audio.VO.Zeus.VO_Zeus_S29_2023_SusReturnToIdle_040_asset,
							Audio.VO.Zeus.VO_Zeus_S29_2023_SusReturnToIdle_041_asset,
							Audio.VO.Zeus.VO_Zeus_S29_2023_SusReturnToIdle_050_asset
						}
					}
					FNVOType.FullAlert =>
					{ return array{
							Audio.VO.Zeus.VO_Zeus_S29_2023_FullAlert_Dist_021_asset,
							Audio.VO.Zeus.VO_Zeus_S29_2023_FullAlert_Dist_050_asset
						}
					}
					FNVOType.FullAlertLost =>
					{ return array{
							Audio.VO.Zeus.VO_Zeus_S29_2023_LostTarget_010_asset,
							Audio.VO.Zeus.VO_Zeus_S29_2023_LostTarget_031_asset,
							Audio.VO.Zeus.VO_Zeus_S29_2023_LostTarget_041_asset,
							Audio.VO.Zeus.VO_Zeus_S29_2023_LostTarget_050_asset,
							Audio.VO.Zeus.VO_Zeus_S29_2023_LostTarget_051_asset,
							Audio.VO.Zeus.VO_Zeus_S29_2023_LostTarget_052_asset
						}
					}
					FNVOType.DBNO =>
					{ return array{
							Audio.VO.Zeus.VO_Zeus_S29_2023_DBNO_010_asset,
							Audio.VO.Zeus.VO_Zeus_S29_2023_DBNO_011_asset,
							Audio.VO.Zeus.VO_Zeus_S29_2023_DBNO_012_asset,
							Audio.VO.Zeus.VO_Zeus_S29_2023_DBNO_013_asset,
							Audio.VO.Zeus.VO_Zeus_S29_2023_DBNO_014_asset
						}
					}
					FNVOType.Interrogation =>
					{ return array{
							Audio.VO.Zeus.VO_Zeus_S29_2023_DBNO_010_asset,
							Audio.VO.Zeus.VO_Zeus_S29_2023_DBNO_011_asset,
							Audio.VO.Zeus.VO_Zeus_S29_2023_DBNO_012_asset,
							Audio.VO.Zeus.VO_Zeus_S29_2023_DBNO_013_asset,
							Audio.VO.Zeus.VO_Zeus_S29_2023_DBNO_014_asset

						}
					}
					FNVOType.Ability =>	
					{ return array{
							Audio.VO.Zeus.VO_Zeus_S29_2023_Phase2_Lift_010_asset,
							Audio.VO.Zeus.VO_Zeus_S29_2023_Phase2_Lift_011_asset,
							Audio.VO.Zeus.VO_Zeus_S29_2023_Phase2_Lift_012_asset
						}
					}
					FNVOType.AbilityEnd =>
					{ return array{
							Audio.VO.Zeus.VO_Zeus_S29_2023_LightningBolt_Throw_010_asset,
							Audio.VO.Zeus.VO_Zeus_S29_2023_LightningBolt_Throw_011_asset,
							Audio.VO.Zeus.VO_Zeus_S29_2023_LightningBolt_Throw_012_asset,
							Audio.VO.Zeus.VO_Zeus_S29_2023_LightningBolt_Throw_013_asset
						}
					}
					FNVOType.Ult =>
					{ return array{
							Audio.VO.Zeus.VO_Zeus_S29_2023_FullAlert_Dist_021_asset,
							Audio.VO.Zeus.VO_Zeus_S29_2023_FullAlert_Dist_050_asset
						}
					}
				}
			}
			FNClasses.Ares =>
			{
				case( VOType )
				{
					FNVOType.Idle => 
					{ return array{
							Audio.VO.Ares.VO_Ares_S29_2023_Idle_010_ST_asset,
							Audio.VO.Ares.VO_Ares_S29_2023_Idle_011_ST_asset,
							Audio.VO.Ares.VO_Ares_S29_2023_Idle_012_ST_asset,
							Audio.VO.Ares.VO_Ares_S29_2023_Idle_017_ST_asset,
							Audio.VO.Ares.VO_Ares_S29_2023_Idle_020_ST_asset,
							Audio.VO.Ares.VO_Ares_S29_2023_Idle_021_ST_asset
						}
					}
					FNVOType.Sus =>
					{ return array{
							Audio.VO.Ares.VO_Ares_S29_2023_Suspicious_010_asset,
							Audio.VO.Ares.VO_Ares_S29_2023_Suspicious_020_asset,
							Audio.VO.Ares.VO_Ares_S29_2023_Suspicious_031_asset,
							Audio.VO.Ares.VO_Ares_S29_2023_Suspicious_040_asset,
							Audio.VO.Ares.VO_Ares_S29_2023_Suspicious_050_asset,
							Audio.VO.Ares.VO_Ares_S29_2023_Suspicious_051_asset
						}
					}
					FNVOType.ReturnIdle =>
					{ return array{
							Audio.VO.Ares.VO_Ares_S29_2023_SusReturnToIdle_011_asset,
							Audio.VO.Ares.VO_Ares_S29_2023_SusReturnToIdle_022_asset,
							Audio.VO.Ares.VO_Ares_S29_2023_SusReturnToIdle_030_asset,
							Audio.VO.Ares.VO_Ares_S29_2023_SusReturnToIdle_032_asset,
							Audio.VO.Ares.VO_Ares_S29_2023_SusReturnToIdle_040_asset,
							Audio.VO.Ares.VO_Ares_S29_2023_SusReturnToIdle_051_asset
						}
					}
					FNVOType.FullAlert =>
					{ return array{
							Audio.VO.Ares.VO_Ares_S29_2023_FullAlert_050_ST_asset,
							Audio.VO.Ares.VO_Ares_S29_2023_FullAlert_051_ST_asset,
							Audio.VO.Ares.VO_Ares_S29_2023_FullAlert_060_ST_asset
						}
					}
					FNVOType.FullAlertLost =>
					{ return array{
							Audio.VO.Ares.VO_Ares_S29_2023_LostTarget_011_ST_asset,
							Audio.VO.Ares.VO_Ares_S29_2023_LostTarget_021_ST_asset,
							Audio.VO.Ares.VO_Ares_S29_2023_LostTarget_030_ST_asset,
							Audio.VO.Ares.VO_Ares_S29_2023_LostTarget_041_ST_asset,
							Audio.VO.Ares.VO_Ares_S29_2023_LostTarget_050_ST_asset
						}
					}
					FNVOType.DBNO =>
					{ return array{
							Audio.VO.Ares.VO_Ares_S29_2023_DBNO_010_ST_asset,
							Audio.VO.Ares.VO_Ares_S29_2023_DBNO_011_ST_asset,
							Audio.VO.Ares.VO_Ares_S29_2023_DBNO_012_ST_asset,
							Audio.VO.Ares.VO_Ares_S29_2023_DBNO_013_ST_asset,
							Audio.VO.Ares.VO_Ares_S29_2023_DBNO_014_ST_asset,
							Audio.VO.Ares.VO_Ares_S29_2023_DBNO_015_ST_asset
						}
					}
					FNVOType.Interrogation =>
					{ return array{
							Audio.VO.Ares.VO_Ares_S29_2023_AlertIdle_010_ST_asset,
							Audio.VO.Ares.VO_Ares_S29_2023_AlertIdle_020_ST_asset,
							Audio.VO.Ares.VO_Ares_S29_2023_AlertIdle_031_ST_asset,
							Audio.VO.Ares.VO_Ares_S29_2023_AlertIdle_042_ST_asset,
							Audio.VO.Ares.VO_Ares_S29_2023_AlertIdle_052_ST_asset
						}
					}
					FNVOType.Ability =>	
					{ return array{
							Audio.VO.Ares.VO_Ares_S29_2023_FullAlert_050_ST_asset,
							Audio.VO.Ares.VO_Ares_S29_2023_FullAlert_051_ST_asset,
							Audio.VO.Ares.VO_Ares_S29_2023_FullAlert_060_ST_asset
						}
					}
					FNVOType.AbilityEnd =>
					{ return array{
							Audio.VO.Ares.VO_Ares_S29_2023_FullAlert_050_ST_asset,
							Audio.VO.Ares.VO_Ares_S29_2023_FullAlert_051_ST_asset,
							Audio.VO.Ares.VO_Ares_S29_2023_FullAlert_060_ST_asset
						}
					}
					FNVOType.Ult =>
					{ return array{
							Audio.VO.Ares.VO_Ares_S29_2023_FullAlert_050_ST_asset,
							Audio.VO.Ares.VO_Ares_S29_2023_FullAlert_051_ST_asset,
							Audio.VO.Ares.VO_Ares_S29_2023_FullAlert_060_ST_asset
						}
					}
				}
			}
			FNClasses.Cerberus =>
			{
				case( VOType )
				{
					FNVOType.Idle => 
					{ return array{
							Audio.VO.Cerberus.VO_Cerberus_S29_2023_Idle_014_asset,
							Audio.VO.Cerberus.VO_Cerberus_S29_2023_Idle_015_asset,
							Audio.VO.Cerberus.VO_Cerberus_S29_2023_Idle_021_asset,
							Audio.VO.Cerberus.VO_Cerberus_S29_2023_Idle_022_asset,
							Audio.VO.Cerberus.VO_Cerberus_S29_2023_Idle_023_asset
						}
					}
					FNVOType.Sus =>
					{ return array{
							Audio.VO.Cerberus.VO_Cerberus_S29_2023_Suspicious_010_asset,
							Audio.VO.Cerberus.VO_Cerberus_S29_2023_Suspicious_011_asset,
							Audio.VO.Cerberus.VO_Cerberus_S29_2023_Suspicious_020_asset,
							Audio.VO.Cerberus.VO_Cerberus_S29_2023_Suspicious_030_asset,
							Audio.VO.Cerberus.VO_Cerberus_S29_2023_Suspicious_041_asset,
							Audio.VO.Cerberus.VO_Cerberus_S29_2023_Suspicious_050_asset
						}
					}
					FNVOType.ReturnIdle =>
					{ return array{
							Audio.VO.Cerberus.VO_Cerberus_S29_2023_SusReturnToIdle_010_asset,
							Audio.VO.Cerberus.VO_Cerberus_S29_2023_SusReturnToIdle_012_asset,
							Audio.VO.Cerberus.VO_Cerberus_S29_2023_SusReturnToIdle_020_asset,
							Audio.VO.Cerberus.VO_Cerberus_S29_2023_SusReturnToIdle_032_asset,
							Audio.VO.Cerberus.VO_Cerberus_S29_2023_SusReturnToIdle_040_asset,
							Audio.VO.Cerberus.VO_Cerberus_S29_2023_SusReturnToIdle_052_asset
						}
					}
					FNVOType.FullAlert =>
					{ return array{
							Audio.VO.Cerberus.VO_Cerberus_S29_2023_FullAlert_010_asset,
							Audio.VO.Cerberus.VO_Cerberus_S29_2023_FullAlert_011_asset,
							Audio.VO.Cerberus.VO_Cerberus_S29_2023_FullAlert_050_asset,
							Audio.VO.Cerberus.VO_Cerberus_S29_2023_FullAlert_060_asset,
							Audio.VO.Cerberus.VO_Cerberus_S29_2023_FullAlert_061_asset
						}
					}
					FNVOType.FullAlertLost =>
					{ return array{
							Audio.VO.Cerberus.VO_Cerberus_S29_2023_LostTarget_010_asset,
							Audio.VO.Cerberus.VO_Cerberus_S29_2023_LostTarget_011_asset,
							Audio.VO.Cerberus.VO_Cerberus_S29_2023_LostTarget_021_asset,
							Audio.VO.Cerberus.VO_Cerberus_S29_2023_LostTarget_030_asset,
							Audio.VO.Cerberus.VO_Cerberus_S29_2023_LostTarget_040_asset,
							Audio.VO.Cerberus.VO_Cerberus_S29_2023_LostTarget_051_asset,
							Audio.VO.Cerberus.VO_Cerberus_S29_2023_LostTarget_052_asset
						}
					}
					FNVOType.DBNO =>
					{ return array{
							Audio.VO.Cerberus.VO_Cerberus_S29_2023_DBNO_011_asset
						}
					}
					FNVOType.Interrogation =>
					{ return array{
							Audio.VO.Cerberus.VO_Cerberus_S29_2023_TakeDamage_Lrg_010_asset,
							Audio.VO.Cerberus.VO_Cerberus_S29_2023_TakeDamage_Lrg_011_asset,
							Audio.VO.Cerberus.VO_Cerberus_S29_2023_TakeDamage_Lrg_012_asset

						}
					}
					FNVOType.Ability =>	
					{ return array{
							Audio.VO.Cerberus.VO_Cerberus_S29_2023_Jump_010_asset,
							Audio.VO.Cerberus.VO_Cerberus_S29_2023_Jump_011_asset,
							Audio.VO.Cerberus.VO_Cerberus_S29_2023_Jump_012_asset
						}
					}
					FNVOType.AbilityEnd =>
					{ return array{
							Audio.VO.Cerberus.VO_Cerberus_S29_2023_FullAlert_010_asset,
							Audio.VO.Cerberus.VO_Cerberus_S29_2023_FullAlert_011_asset,
							Audio.VO.Cerberus.VO_Cerberus_S29_2023_FullAlert_050_asset,
							Audio.VO.Cerberus.VO_Cerberus_S29_2023_FullAlert_060_asset,
							Audio.VO.Cerberus.VO_Cerberus_S29_2023_FullAlert_061_asset
						}
					}
					FNVOType.Ult =>
					{ return array{
							Audio.VO.Cerberus.VO_Cerberus_S29_2023_FullAlert_010_asset,
							Audio.VO.Cerberus.VO_Cerberus_S29_2023_FullAlert_011_asset,
							Audio.VO.Cerberus.VO_Cerberus_S29_2023_FullAlert_050_asset,
							Audio.VO.Cerberus.VO_Cerberus_S29_2023_FullAlert_060_asset,
							Audio.VO.Cerberus.VO_Cerberus_S29_2023_FullAlert_061_asset
						}
					}
				}
			}
			FNClasses.Hades =>
			{
				case( VOType )
				{
					FNVOType.Idle => 
					{ return array{
							Audio.VO.Hades.VO_Hades_S29_2023_Idle_010_asset,
							Audio.VO.Hades.VO_Hades_S29_2023_Idle_011_asset,
							Audio.VO.Hades.VO_Hades_S29_2023_Idle_012_asset,
							Audio.VO.Hades.VO_Hades_S29_2023_Idle_013_asset,
							Audio.VO.Hades.VO_Hades_S29_2023_Idle_020_asset,
							Audio.VO.Hades.VO_Hades_S29_2023_Idle_021_asset
						}
					}
					FNVOType.Sus =>
					{ return array{
							Audio.VO.Hades.VO_Hades_S29_2023_Suspicious_011_asset,
							Audio.VO.Hades.VO_Hades_S29_2023_Suspicious_020_asset,
							Audio.VO.Hades.VO_Hades_S29_2023_Suspicious_030_asset,
							Audio.VO.Hades.VO_Hades_S29_2023_Suspicious_040_asset,
							Audio.VO.Hades.VO_Hades_S29_2023_Suspicious_050_asset
						}
					}
					FNVOType.ReturnIdle =>
					{ return array{
							Audio.VO.Hades.VO_Hades_S29_2023_SusReturnToIdle_010_asset,
							Audio.VO.Hades.VO_Hades_S29_2023_SusReturnToIdle_011_asset,
							Audio.VO.Hades.VO_Hades_S29_2023_SusReturnToIdle_020_asset,
							Audio.VO.Hades.VO_Hades_S29_2023_SusReturnToIdle_040_asset,
							Audio.VO.Hades.VO_Hades_S29_2023_SusReturnToIdle_050_asset
						}
					}
					FNVOType.FullAlert =>
					{ return array{
							Audio.VO.Hades.VO_Hades_S29_2023_FullAlert_041_asset,
							Audio.VO.Hades.VO_Hades_S29_2023_FullAlert_050_asset,
							Audio.VO.Hades.VO_Hades_S29_2023_FullAlert_060_asset
						}
					}
					FNVOType.FullAlertLost =>
					{ return array{
							Audio.VO.Hades.VO_Hades_S29_2023_LostTarget_020_asset,
							Audio.VO.Hades.VO_Hades_S29_2023_LostTarget_021_asset,
							Audio.VO.Hades.VO_Hades_S29_2023_LostTarget_031_asset,
							Audio.VO.Hades.VO_Hades_S29_2023_LostTarget_050_asset,
							Audio.VO.Hades.VO_Hades_S29_2023_LostTarget_051_asset
						}
					}
					FNVOType.DBNO =>
					{ return array{
							Audio.VO.Hades.VO_Hades_S29_2023_DBNO_010_asset,
							Audio.VO.Hades.VO_Hades_S29_2023_DBNO_011_asset,
							Audio.VO.Hades.VO_Hades_S29_2023_DBNO_012_asset,
							Audio.VO.Hades.VO_Hades_S29_2023_DBNO_013_asset,
							Audio.VO.Hades.VO_Hades_S29_2023_DBNO_014_asset
						}
					}
					FNVOType.Interrogation =>
					{ return array{
							Audio.VO.Hades.VO_Hades_S29_2023_DBNO_010_asset,
							Audio.VO.Hades.VO_Hades_S29_2023_DBNO_011_asset,
							Audio.VO.Hades.VO_Hades_S29_2023_DBNO_012_asset,
							Audio.VO.Hades.VO_Hades_S29_2023_DBNO_013_asset,
							Audio.VO.Hades.VO_Hades_S29_2023_DBNO_014_asset

						}
					}
					FNVOType.Ability =>	
					{ return array{
							Audio.VO.Hades.VO_Hades_S29_2023_Levitate_010_asset,
							Audio.VO.Hades.VO_Hades_S29_2023_Levitate_011_asset
						}
					}
					FNVOType.AbilityEnd =>
					{ return array{
							Audio.VO.Hades.VO_Hades_S29_2023_LD_End_010_asset,
							Audio.VO.Hades.VO_Hades_S29_2023_LD_End_011_asset,
							Audio.VO.Hades.VO_Hades_S29_2023_LD_End_012_asset,
							Audio.VO.Hades.VO_Hades_S29_2023_LD_End_013_asset
						}
					}
					FNVOType.Ult =>
					{ return array{
							Audio.VO.Hades.VO_Hades_S29_2023_Empower_Cast_010_asset,
							Audio.VO.Hades.VO_Hades_S29_2023_Empower_Cast_011_asset,
							Audio.VO.Hades.VO_Hades_S29_2023_Empower_Cast_012_asset,
							Audio.VO.Hades.VO_Hades_S29_2023_Empower_Cast_013_asset
						}
					}
				}
			}
			FNClasses.MegaloDon =>
			{
				case( VOType )
				{
					FNVOType.Idle => 
					{ return array{
							Audio.VO.MegaloDon.VO_Megalodon_S30_2024_Idle_010_asset,
							Audio.VO.MegaloDon.VO_Megalodon_S30_2024_Idle_012_asset,
							Audio.VO.MegaloDon.VO_Megalodon_S30_2024_Idle_014_asset,
							Audio.VO.MegaloDon.VO_Megalodon_S30_2024_Idle_015_asset,
							Audio.VO.MegaloDon.VO_Megalodon_S30_2024_Idle_017_asset,
							Audio.VO.MegaloDon.VO_Megalodon_S30_2024_Idle_018_asset,
							Audio.VO.MegaloDon.VO_Megalodon_S30_2024_Idle_020_asset,
							Audio.VO.MegaloDon.VO_Megalodon_S30_2024_Idle_030_asset,
							Audio.VO.MegaloDon.VO_Megalodon_S30_2024_Idle_050_asset
						}
					}
					FNVOType.Sus =>
					{ return array{
							Audio.VO.MegaloDon.VO_Megalodon_S30_2024_Taunt_AimedAt_010_asset,
							Audio.VO.MegaloDon.VO_Megalodon_S30_2024_Taunt_AimedAt_011_asset,
							Audio.VO.MegaloDon.VO_Megalodon_S30_2024_Taunt_AimedAt_013_asset,
							Audio.VO.MegaloDon.VO_Megalodon_S30_2024_Taunt_AimedAt_020_asset,
							Audio.VO.MegaloDon.VO_Megalodon_S30_2024_Taunt_AimedAt_024_asset,
							Audio.VO.MegaloDon.VO_Megalodon_S30_2024_Taunt_AimedAt_026_asset,
							Audio.VO.MegaloDon.VO_Megalodon_S30_2024_Taunt_AimedAt_027_asset
						}
					}
					FNVOType.ReturnIdle =>
					{ return array{
							Audio.VO.MegaloDon.VO_Megalodon_S30_2024_Taunt_AimedAt_010_asset,
							Audio.VO.MegaloDon.VO_Megalodon_S30_2024_Taunt_AimedAt_011_asset,
							Audio.VO.MegaloDon.VO_Megalodon_S30_2024_Taunt_AimedAt_013_asset,
							Audio.VO.MegaloDon.VO_Megalodon_S30_2024_Taunt_AimedAt_020_asset,
							Audio.VO.MegaloDon.VO_Megalodon_S30_2024_Taunt_AimedAt_024_asset,
							Audio.VO.MegaloDon.VO_Megalodon_S30_2024_Taunt_AimedAt_026_asset,
							Audio.VO.MegaloDon.VO_Megalodon_S30_2024_Taunt_AimedAt_027_asset
						}
					}
					FNVOType.FullAlert =>
					{ return array{
							Audio.VO.MegaloDon.VO_Megalodon_S30_2024_FullAlert_010_asset,
							Audio.VO.MegaloDon.VO_Megalodon_S30_2024_FullAlert_020_asset,
							Audio.VO.MegaloDon.VO_Megalodon_S30_2024_FullAlert_030_asset,
							Audio.VO.MegaloDon.VO_Megalodon_S30_2024_FullAlert_031_asset,
							Audio.VO.MegaloDon.VO_Megalodon_S30_2024_FullAlert_050_asset
						}
					}
					FNVOType.FullAlertLost =>
					{ return array{
							Audio.VO.MegaloDon.VO_Megalodon_S30_2024_LostTarget_010_asset,
							Audio.VO.MegaloDon.VO_Megalodon_S30_2024_LostTarget_011_asset,
							Audio.VO.MegaloDon.VO_Megalodon_S30_2024_LostTarget_014_asset,
							Audio.VO.MegaloDon.VO_Megalodon_S30_2024_LostTarget_015_asset,
							Audio.VO.MegaloDon.VO_Megalodon_S30_2024_LostTarget_020_asset
						}
					}
					FNVOType.DBNO =>
					{ return array{
							Audio.VO.MegaloDon.VO_Megalodon_S30_2024_Elim_011_asset,
							Audio.VO.MegaloDon.VO_Megalodon_S30_2024_Elim_012_asset,
							Audio.VO.MegaloDon.VO_Megalodon_S30_2024_Elim_013_asset,
							Audio.VO.MegaloDon.VO_Megalodon_S30_2024_Elim_015_asset,
							Audio.VO.MegaloDon.VO_Megalodon_S30_2024_Elim_016_asset,
							Audio.VO.MegaloDon.VO_Megalodon_S30_2024_Elim_017_asset
						}
					}
					FNVOType.Interrogation =>
					{ return array{
							Audio.VO.MegaloDon.VO_Megalodon_S30_2024_TakeDamage_Lrg_010_asset,
							Audio.VO.MegaloDon.VO_Megalodon_S30_2024_TakeDamage_Lrg_011_asset,
							Audio.VO.MegaloDon.VO_Megalodon_S30_2024_TakeDamage_Lrg_012_asset,
							Audio.VO.MegaloDon.VO_Megalodon_S30_2024_TakeDamage_Lrg_013_asset,
							Audio.VO.MegaloDon.VO_Megalodon_S30_2024_TakeDamage_Lrg_014_asset,
							Audio.VO.MegaloDon.VO_Megalodon_S30_2024_TakeDamage_Lrg_015_asset,
							Audio.VO.MegaloDon.VO_Megalodon_S30_2024_TakeDamage_Lrg_016_asset

						}
					}
					FNVOType.Ability =>	
					{ return array{
							Audio.VO.MegaloDon.VO_Megalodon_S30_2024_Charge_Start_010_asset,
							Audio.VO.MegaloDon.VO_Megalodon_S30_2024_Charge_Start_011_asset,
							Audio.VO.MegaloDon.VO_Megalodon_S30_2024_Charge_Start_012_asset,
							Audio.VO.MegaloDon.VO_Megalodon_S30_2024_Charge_Start_013_asset
						}
					}
					FNVOType.AbilityEnd =>
					{ return array{
							Audio.VO.MegaloDon.VO_Megalodon_S30_2024_Charge_End_012_asset,
							Audio.VO.MegaloDon.VO_Megalodon_S30_2024_Charge_End_013_asset,
							Audio.VO.MegaloDon.VO_Megalodon_S30_2024_Charge_End_014_asset,
							Audio.VO.MegaloDon.VO_Megalodon_S30_2024_Charge_End_015_asset,
							Audio.VO.MegaloDon.VO_Megalodon_S30_2024_Charge_End_016_asset
						}
					}
					FNVOType.Ult =>
					{ return array{
							Audio.VO.MegaloDon.VO_MegaloDon_Ult_asset
						}
					}
				}
			}
			FNClasses.RingmasterScarr =>
			{
				case( VOType )
				{
					FNVOType.Idle => 
					{ return array{
							Audio.VO.RingmasterScarr.VO_NitroQueen_S30_2024_NitroDomeIdle_010_asset,
							Audio.VO.RingmasterScarr.VO_NitroQueen_S30_2024_NitroDomeIdle_020_asset,
							Audio.VO.RingmasterScarr.VO_NitroQueen_S30_2024_NitroDomeIdle_021_asset,
							Audio.VO.RingmasterScarr.VO_NitroQueen_S30_2024_NitroDomeIdle_030_asset,
							Audio.VO.RingmasterScarr.VO_NitroQueen_S30_2024_NitroDomeIdle_040_asset,
							Audio.VO.RingmasterScarr.VO_NitroQueen_S30_2024_NitroDomeIdle_050_asset
						}
					}
					FNVOType.Sus =>
					{ return array{
							Audio.VO.RingmasterScarr.VO_NitroQueen_S30_2024_Taunt_AimedAt_013_asset,
							Audio.VO.RingmasterScarr.VO_NitroQueen_S30_2024_Taunt_AimedAt_020_asset,
							Audio.VO.RingmasterScarr.VO_NitroQueen_S30_2024_Taunt_AimedAt_021_asset,
							Audio.VO.RingmasterScarr.VO_NitroQueen_S30_2024_Taunt_AimedAt_030_asset,
							Audio.VO.RingmasterScarr.VO_NitroQueen_S30_2024_Taunt_AimedAt_040_asset,
							Audio.VO.RingmasterScarr.VO_NitroQueen_S30_2024_Taunt_AimedAt_050_asset
						}
					}
					FNVOType.ReturnIdle =>
					{ return array{
							Audio.VO.RingmasterScarr.VO_NitroQueen_S30_2024_SusReturnToIdle_010_asset,
							Audio.VO.RingmasterScarr.VO_NitroQueen_S30_2024_SusReturnToIdle_020_asset,
							Audio.VO.RingmasterScarr.VO_NitroQueen_S30_2024_SusReturnToIdle_030_asset,
							Audio.VO.RingmasterScarr.VO_NitroQueen_S30_2024_SusReturnToIdle_031_asset,
							Audio.VO.RingmasterScarr.VO_NitroQueen_S30_2024_SusReturnToIdle_050_asset
						}
					}
					FNVOType.FullAlert =>
					{ return array{
							Audio.VO.RingmasterScarr.VO_NitroQueen_S30_2024_FullAlert_010_asset,
							Audio.VO.RingmasterScarr.VO_NitroQueen_S30_2024_FullAlert_011_asset,
							Audio.VO.RingmasterScarr.VO_NitroQueen_S30_2024_FullAlert_020_asset,
							Audio.VO.RingmasterScarr.VO_NitroQueen_S30_2024_FullAlert_021_asset,
							Audio.VO.RingmasterScarr.VO_NitroQueen_S30_2024_FullAlert_030_asset,
							Audio.VO.RingmasterScarr.VO_NitroQueen_S30_2024_FullAlert_040_asset,
							Audio.VO.RingmasterScarr.VO_NitroQueen_S30_2024_FullAlert_050_asset
						}
					}
					FNVOType.FullAlertLost =>
					{ return array{
							Audio.VO.RingmasterScarr.VO_NitroQueen_S30_2024_LostTarget_010_asset,
							Audio.VO.RingmasterScarr.VO_NitroQueen_S30_2024_LostTarget_011_asset,
							Audio.VO.RingmasterScarr.VO_NitroQueen_S30_2024_LostTarget_012_asset,
							Audio.VO.RingmasterScarr.VO_NitroQueen_S30_2024_LostTarget_013_asset,
							Audio.VO.RingmasterScarr.VO_NitroQueen_S30_2024_LostTarget_020_asset,
							Audio.VO.RingmasterScarr.VO_NitroQueen_S30_2024_LostTarget_040_asset,
							Audio.VO.RingmasterScarr.VO_NitroQueen_S30_2024_LostTarget_050_asset
						}
					}
					FNVOType.DBNO =>
					{ return array{
							Audio.VO.RingmasterScarr.VO_NitroQueen_S30_2024_Elim_010_asset,
							Audio.VO.RingmasterScarr.VO_NitroQueen_S30_2024_Elim_011_asset,
							Audio.VO.RingmasterScarr.VO_NitroQueen_S30_2024_Elim_012_asset,
							Audio.VO.RingmasterScarr.VO_NitroQueen_S30_2024_Elim_013_asset,
							Audio.VO.RingmasterScarr.VO_NitroQueen_S30_2024_Elim_014_asset
						}
					}
					FNVOType.Interrogation =>
					{ return array{
							Audio.VO.RingmasterScarr.VO_NitroQueen_S30_2024_Elim_010_asset,
							Audio.VO.RingmasterScarr.VO_NitroQueen_S30_2024_Elim_011_asset,
							Audio.VO.RingmasterScarr.VO_NitroQueen_S30_2024_Elim_012_asset,
							Audio.VO.RingmasterScarr.VO_NitroQueen_S30_2024_Elim_013_asset,
							Audio.VO.RingmasterScarr.VO_NitroQueen_S30_2024_Elim_014_asset

						}
					}
					FNVOType.Ability =>	
					{ return array{
							Audio.VO.RingmasterScarr.VO_NitroQueen_S30_2024_TNT_Callout_010_asset,
							Audio.VO.RingmasterScarr.VO_NitroQueen_S30_2024_TNT_Callout_021_asset,
							Audio.VO.RingmasterScarr.VO_NitroQueen_S30_2024_TNT_Callout_030_asset,
							Audio.VO.RingmasterScarr.VO_NitroQueen_S30_2024_TNT_Callout_031_asset
						}
					}
					FNVOType.AbilityEnd =>
					{ return array{
							Audio.VO.RingmasterScarr.VO_NitroQueen_S30_2024_TNT_Callout_010_asset,
							Audio.VO.RingmasterScarr.VO_NitroQueen_S30_2024_TNT_Callout_021_asset,
							Audio.VO.RingmasterScarr.VO_NitroQueen_S30_2024_TNT_Callout_030_asset,
							Audio.VO.RingmasterScarr.VO_NitroQueen_S30_2024_TNT_Callout_031_asset
						}
					}
					FNVOType.Ult =>
					{ return array{
							Audio.VO.RingmasterScarr.VO_NitroQueen_S30_2024_TNT_Callout_010_asset,
							Audio.VO.RingmasterScarr.VO_NitroQueen_S30_2024_TNT_Callout_021_asset,
							Audio.VO.RingmasterScarr.VO_NitroQueen_S30_2024_TNT_Callout_030_asset,
							Audio.VO.RingmasterScarr.VO_NitroQueen_S30_2024_TNT_Callout_031_asset
						}
					}
				}
			}
			FNClasses.TheMachinist =>
			{
				case( VOType )
				{
					FNVOType.Idle => 
					{ return array{
							Audio.VO.TheMachinist.VO_Overseer_S30_2024_Idle_014_asset,
							Audio.VO.TheMachinist.VO_Overseer_S30_2024_Idle_016_asset,
							Audio.VO.TheMachinist.VO_Overseer_S30_2024_Idle_017_asset,
							Audio.VO.TheMachinist.VO_Overseer_S30_2024_Idle_060_asset,
							Audio.VO.TheMachinist.VO_Overseer_S30_2024_Idle_070_asset
						}
					}
					FNVOType.Sus =>
					{ return array{
							Audio.VO.TheMachinist.VO_Overseer_S30_2024_Taunt_AimedAt_010_asset,
							Audio.VO.TheMachinist.VO_Overseer_S30_2024_Taunt_AimedAt_011_asset,
							Audio.VO.TheMachinist.VO_Overseer_S30_2024_Taunt_AimedAt_030_asset,
							Audio.VO.TheMachinist.VO_Overseer_S30_2024_Taunt_AimedAt_060_asset
						}
					}
					FNVOType.ReturnIdle =>
					{ return array{
							Audio.VO.TheMachinist.VO_Overseer_S30_2024_SusReturnToIdle_010_asset,
							Audio.VO.TheMachinist.VO_Overseer_S30_2024_SusReturnToIdle_020_asset,
							Audio.VO.TheMachinist.VO_Overseer_S30_2024_SusReturnToIdle_030_asset,
							Audio.VO.TheMachinist.VO_Overseer_S30_2024_SusReturnToIdle_040_asset,
							Audio.VO.TheMachinist.VO_Overseer_S30_2024_SusReturnToIdle_050_asset
						}
					}
					FNVOType.FullAlert =>
					{ return array{
							Audio.VO.TheMachinist.VO_Overseer_S30_2024_FullAlert_020_asset,
							Audio.VO.TheMachinist.VO_Overseer_S30_2024_FullAlert_030_asset,
							Audio.VO.TheMachinist.VO_Overseer_S30_2024_FullAlert_040_asset,
							Audio.VO.TheMachinist.VO_Overseer_S30_2024_FullAlert_050_asset,
							Audio.VO.TheMachinist.VO_Overseer_S30_2024_FullAlert_060_asset
						}
					}
					FNVOType.FullAlertLost =>
					{ return array{
							Audio.VO.TheMachinist.VO_Overseer_S30_2024_LostTarget_013_asset,
							Audio.VO.TheMachinist.VO_Overseer_S30_2024_LostTarget_014_asset,
							Audio.VO.TheMachinist.VO_Overseer_S30_2024_LostTarget_020_asset,
							Audio.VO.TheMachinist.VO_Overseer_S30_2024_LostTarget_030_asset,
							Audio.VO.TheMachinist.VO_Overseer_S30_2024_LostTarget_040_asset
						}
					}
					FNVOType.DBNO =>
					{ return array{
							Audio.VO.TheMachinist.VO_Overseer_S30_2024_Elim_010_asset,
							Audio.VO.TheMachinist.VO_Overseer_S30_2024_Elim_011_asset,
							Audio.VO.TheMachinist.VO_Overseer_S30_2024_Elim_012_asset,
							Audio.VO.TheMachinist.VO_Overseer_S30_2024_Elim_013_asset,
							Audio.VO.TheMachinist.VO_Overseer_S30_2024_Elim_014_asset
						}
					}
					FNVOType.Interrogation =>
					{ return array{
							Audio.VO.TheMachinist.VO_Overseer_S30_2024_Elim_010_asset,
							Audio.VO.TheMachinist.VO_Overseer_S30_2024_Elim_011_asset,
							Audio.VO.TheMachinist.VO_Overseer_S30_2024_Elim_012_asset,
							Audio.VO.TheMachinist.VO_Overseer_S30_2024_Elim_013_asset,
							Audio.VO.TheMachinist.VO_Overseer_S30_2024_Elim_014_asset

						}
					}
					FNVOType.Ability =>	
					{ return array{
							Audio.VO.TheMachinist.VO_Overseer_S30_2024_NitroState_TurretDrop_010_asset,
							Audio.VO.TheMachinist.VO_Overseer_S30_2024_NitroState_TurretDrop_020_asset,
							Audio.VO.TheMachinist.VO_Overseer_S30_2024_NitroState_TurretDrop_060_asset,
							Audio.VO.TheMachinist.VO_Overseer_S30_2024_NitroState_TurretDrop_070_asset,
							Audio.VO.TheMachinist.VO_Overseer_S30_2024_NitroState_TurretDrop_080_asset
						}
					}
					FNVOType.AbilityEnd =>
					{ return array{
							Audio.VO.TheMachinist.VO_Overseer_S30_2024_NitroState_TurretDrop_010_asset,
							Audio.VO.TheMachinist.VO_Overseer_S30_2024_NitroState_TurretDrop_020_asset,
							Audio.VO.TheMachinist.VO_Overseer_S30_2024_NitroState_TurretDrop_060_asset,
							Audio.VO.TheMachinist.VO_Overseer_S30_2024_NitroState_TurretDrop_070_asset,
							Audio.VO.TheMachinist.VO_Overseer_S30_2024_NitroState_TurretDrop_080_asset
						}
					}
					FNVOType.Ult =>
					{ return array{
							Audio.VO.TheMachinist.VO_Overseer_S30_2024_NitroState_Enter_013_asset,
							Audio.VO.TheMachinist.VO_Overseer_S30_2024_NitroState_Enter_040_asset,
							Audio.VO.TheMachinist.VO_Overseer_S30_2024_NitroState_Enter_050_asset,
							Audio.VO.TheMachinist.VO_Overseer_S30_2024_NitroState_Enter_070_asset,
							Audio.VO.TheMachinist.VO_Overseer_S30_2024_NitroState_Enter_090_asset
						}
					}
				}
			}
			FNClasses.UltimaCarver =>
			{
				case( VOType )
				{
					FNVOType.Idle => 
					{ return array{
							Audio.VO.UltimaCarver.VO_FNM_31Boss_Idle_020_asset,
							Audio.VO.UltimaCarver.VO_FNM_31Boss_Idle_030_asset,
							Audio.VO.UltimaCarver.VO_FNM_31Boss_Idle_041_asset,
							Audio.VO.UltimaCarver.VO_FNM_31Boss_Idle_050_asset,
							Audio.VO.UltimaCarver.VO_FNM_31Boss_Idle_051_asset,
							Audio.VO.UltimaCarver.VO_FNM_31Boss_Idle_052_asset,
							Audio.VO.UltimaCarver.VO_FNM_31Boss_IdleChuckle_010_asset,
							Audio.VO.UltimaCarver.VO_FNM_31Boss_IdleChuckle_012_asset,
							Audio.VO.UltimaCarver.VO_FNM_31Boss_IdleScare_010_asset,
							Audio.VO.UltimaCarver.VO_FNM_31Boss_IdleScare_011_asset
						}
					}
					FNVOType.Sus =>
					{ return array{
							Audio.VO.UltimaCarver.VO_FNM_31Boss_Suspicious_011_asset,
							Audio.VO.UltimaCarver.VO_FNM_31Boss_Suspicious_012_asset,
							Audio.VO.UltimaCarver.VO_FNM_31Boss_Suspicious_015_asset,
							Audio.VO.UltimaCarver.VO_FNM_31Boss_Suspicious_016_asset,
							Audio.VO.UltimaCarver.VO_FNM_31Boss_Suspicious_024_asset,
							Audio.VO.UltimaCarver.VO_FNM_31Boss_Suspicious_028_asset
						}
					}
					FNVOType.ReturnIdle =>
					{ return array{
							Audio.VO.UltimaCarver.VO_FNM_31Boss_SusReturnToIdle_011_asset,
							Audio.VO.UltimaCarver.VO_FNM_31Boss_SusReturnToIdle_012_asset,
							Audio.VO.UltimaCarver.VO_FNM_31Boss_SusReturnToIdle_014_asset,
							Audio.VO.UltimaCarver.VO_FNM_31Boss_SusReturnToIdle_020_asset,
							Audio.VO.UltimaCarver.VO_FNM_31Boss_SusReturnToIdle_023_asset,
							Audio.VO.UltimaCarver.VO_FNM_31Boss_SusReturnToIdle_024_asset,
							Audio.VO.UltimaCarver.VO_FNM_31Boss_SusReturnToIdle_033_asset,
							Audio.VO.UltimaCarver.VO_FNM_31Boss_SusReturnToIdle_043_asset,
							Audio.VO.UltimaCarver.VO_FNM_31Boss_SusReturnToIdle_050_asset
						}
					}
					FNVOType.FullAlert =>
					{ return array{
							Audio.VO.UltimaCarver.VO_FNM_31Boss_FullAlert_010_asset,
							Audio.VO.UltimaCarver.VO_FNM_31Boss_FullAlert_015_asset,
							Audio.VO.UltimaCarver.VO_FNM_31Boss_FullAlert_023_asset,
							Audio.VO.UltimaCarver.VO_FNM_31Boss_FullAlert_030_asset,
							Audio.VO.UltimaCarver.VO_FNM_31Boss_FullAlert_031_asset,
							Audio.VO.UltimaCarver.VO_FNM_31Boss_FullAlert_033_asset,
							Audio.VO.UltimaCarver.VO_FNM_31Boss_FullAlert_041_asset,
							Audio.VO.UltimaCarver.VO_FNM_31Boss_FullAlert_051_asset,
							Audio.VO.UltimaCarver.VO_FNM_31Boss_FullAlert_053_asset,
							Audio.VO.UltimaCarver.VO_FNM_31Boss_FullAlert_054_asset
						}
					}
					FNVOType.FullAlertLost =>
					{ return array{
							Audio.VO.UltimaCarver.VO_FNM_31Boss_AlertLostTarget_010_asset,
							Audio.VO.UltimaCarver.VO_FNM_31Boss_AlertLostTarget_011_asset,
							Audio.VO.UltimaCarver.VO_FNM_31Boss_AlertLostTarget_012_asset,
							Audio.VO.UltimaCarver.VO_FNM_31Boss_AlertLostTarget_030_asset,
							Audio.VO.UltimaCarver.VO_FNM_31Boss_AlertLostTarget_040_asset,
							Audio.VO.UltimaCarver.VO_FNM_31Boss_AlertLostTarget_050_asset
						}
					}
					FNVOType.DBNO =>
					{ return array{
							Audio.VO.UltimaCarver.VO_FNM_31Boss_Elim_011_asset,
							Audio.VO.UltimaCarver.VO_FNM_31Boss_Elim_012_asset,
							Audio.VO.UltimaCarver.VO_FNM_31Boss_Elim_013_asset,
							Audio.VO.UltimaCarver.VO_FNM_31Boss_Elim_016_asset,
							Audio.VO.UltimaCarver.VO_FNM_31Boss_Elim_017_asset
						}
					}
					FNVOType.Interrogation =>
					{ return array{
							Audio.VO.UltimaCarver.VO_FNM_31Boss_Elim_011_asset,
							Audio.VO.UltimaCarver.VO_FNM_31Boss_Elim_012_asset,
							Audio.VO.UltimaCarver.VO_FNM_31Boss_Elim_013_asset,
							Audio.VO.UltimaCarver.VO_FNM_31Boss_Elim_016_asset,
							Audio.VO.UltimaCarver.VO_FNM_31Boss_Elim_017_asset

						}
					}
					FNVOType.Ability =>	
					{ return array{
							Audio.VO.UltimaCarver.VO_FNM_31Boss_FullAlert_010_asset,
							Audio.VO.UltimaCarver.VO_FNM_31Boss_FullAlert_015_asset,
							Audio.VO.UltimaCarver.VO_FNM_31Boss_FullAlert_023_asset,
							Audio.VO.UltimaCarver.VO_FNM_31Boss_FullAlert_030_asset,
							Audio.VO.UltimaCarver.VO_FNM_31Boss_FullAlert_031_asset,
							Audio.VO.UltimaCarver.VO_FNM_31Boss_FullAlert_033_asset,
							Audio.VO.UltimaCarver.VO_FNM_31Boss_FullAlert_041_asset,
							Audio.VO.UltimaCarver.VO_FNM_31Boss_FullAlert_051_asset,
							Audio.VO.UltimaCarver.VO_FNM_31Boss_FullAlert_053_asset,
							Audio.VO.UltimaCarver.VO_FNM_31Boss_FullAlert_054_asset
						}
					}
					FNVOType.AbilityEnd =>
					{ return array{
							Audio.VO.UltimaCarver.VO_FNM_31Boss_FullAlert_010_asset,
							Audio.VO.UltimaCarver.VO_FNM_31Boss_FullAlert_015_asset,
							Audio.VO.UltimaCarver.VO_FNM_31Boss_FullAlert_023_asset,
							Audio.VO.UltimaCarver.VO_FNM_31Boss_FullAlert_030_asset,
							Audio.VO.UltimaCarver.VO_FNM_31Boss_FullAlert_031_asset,
							Audio.VO.UltimaCarver.VO_FNM_31Boss_FullAlert_033_asset,
							Audio.VO.UltimaCarver.VO_FNM_31Boss_FullAlert_041_asset,
							Audio.VO.UltimaCarver.VO_FNM_31Boss_FullAlert_051_asset,
							Audio.VO.UltimaCarver.VO_FNM_31Boss_FullAlert_053_asset,
							Audio.VO.UltimaCarver.VO_FNM_31Boss_FullAlert_054_asset
						}
					}
					FNVOType.Ult =>
					{ return array{
							Audio.VO.UltimaCarver.VO_FNM_31Boss_FullAlert_010_asset,
							Audio.VO.UltimaCarver.VO_FNM_31Boss_FullAlert_015_asset,
							Audio.VO.UltimaCarver.VO_FNM_31Boss_FullAlert_023_asset,
							Audio.VO.UltimaCarver.VO_FNM_31Boss_FullAlert_030_asset,
							Audio.VO.UltimaCarver.VO_FNM_31Boss_FullAlert_031_asset,
							Audio.VO.UltimaCarver.VO_FNM_31Boss_FullAlert_033_asset,
							Audio.VO.UltimaCarver.VO_FNM_31Boss_FullAlert_041_asset,
							Audio.VO.UltimaCarver.VO_FNM_31Boss_FullAlert_051_asset,
							Audio.VO.UltimaCarver.VO_FNM_31Boss_FullAlert_053_asset,
							Audio.VO.UltimaCarver.VO_FNM_31Boss_FullAlert_054_asset
						}
					}
				}
			}
			FNClasses.ShogunX =>
			{
				case( VOType )
				{
					FNVOType.Idle => 
					{ return array{
							Audio.VO.ShogunX.VO_FirePetal_WanderingBoss_S33_Idle_010_asset,
							Audio.VO.ShogunX.VO_FirePetal_WanderingBoss_S33_Idle_011_asset,
							Audio.VO.ShogunX.VO_FirePetal_WanderingBoss_S33_Idle_013_asset,
							Audio.VO.ShogunX.VO_FirePetal_WanderingBoss_S33_Idle_015_asset,
							Audio.VO.ShogunX.VO_FirePetal_WanderingBoss_S33_Idle_030_asset,
							Audio.VO.ShogunX.VO_FirePetal_WanderingBoss_S33_Idle_040_asset,
							Audio.VO.ShogunX.VO_FirePetal_WanderingBoss_S33_Idle_050_asset,
							Audio.VO.ShogunX.VO_FirePetal_WanderingBoss_S33_Idle_051_asset
						}
					}
					FNVOType.Sus =>
					{ return array{
							Audio.VO.ShogunX.VO_FirePetal_WanderingBoss_S33_Suspicious_010_asset,
							Audio.VO.ShogunX.VO_FirePetal_WanderingBoss_S33_Suspicious_011_asset,
							Audio.VO.ShogunX.VO_FirePetal_WanderingBoss_S33_Suspicious_012_asset,
							Audio.VO.ShogunX.VO_FirePetal_WanderingBoss_S33_Suspicious_013_asset,
							Audio.VO.ShogunX.VO_FirePetal_WanderingBoss_S33_Suspicious_014_asset,
							<#Audio.VO.ShogunX.VO_FirePetal_WanderingBoss_S33_Suspicious_015_asset,
							Audio.VO.ShogunX.VO_FirePetal_WanderingBoss_S33_Suspicious_016_asset,
							Audio.VO.ShogunX.VO_FirePetal_WanderingBoss_S33_Suspicious_020_asset,
							Audio.VO.ShogunX.VO_FirePetal_WanderingBoss_S33_Suspicious_021_asset,
							Audio.VO.ShogunX.VO_FirePetal_WanderingBoss_S33_Suspicious_022_asset,
							Audio.VO.ShogunX.VO_FirePetal_WanderingBoss_S33_Suspicious_023_asset,
							Audio.VO.ShogunX.VO_FirePetal_WanderingBoss_S33_Suspicious_024_asset,
							Audio.VO.ShogunX.VO_FirePetal_WanderingBoss_S33_Suspicious_025_asset,
							Audio.VO.ShogunX.VO_FirePetal_WanderingBoss_S33_Suspicious_026_asset,
							Audio.VO.ShogunX.VO_FirePetal_WanderingBoss_S33_Suspicious_030_asset,
							Audio.VO.ShogunX.VO_FirePetal_WanderingBoss_S33_Suspicious_031_asset,
							Audio.VO.ShogunX.VO_FirePetal_WanderingBoss_S33_Suspicious_032_asset,
							Audio.VO.ShogunX.VO_FirePetal_WanderingBoss_S33_Suspicious_033_asset,
							Audio.VO.ShogunX.VO_FirePetal_WanderingBoss_S33_Suspicious_034_asset,
							Audio.VO.ShogunX.VO_FirePetal_WanderingBoss_S33_Suspicious_035_asset,#>
							Audio.VO.ShogunX.VO_FirePetal_WanderingBoss_S33_Suspicious_036_asset
						}
					}
					FNVOType.ReturnIdle =>
					{ return array{
							Audio.VO.ShogunX.VO_FirePetal_WanderingBoss_S33_SusReturnToIdle_010_asset,
							Audio.VO.ShogunX.VO_FirePetal_WanderingBoss_S33_SusReturnToIdle_011_asset,
							Audio.VO.ShogunX.VO_FirePetal_WanderingBoss_S33_SusReturnToIdle_012_asset,
							Audio.VO.ShogunX.VO_FirePetal_WanderingBoss_S33_SusReturnToIdle_013_asset,
							Audio.VO.ShogunX.VO_FirePetal_WanderingBoss_S33_SusReturnToIdle_014_asset,
							<#Audio.VO.ShogunX.VO_FirePetal_WanderingBoss_S33_SusReturnToIdle_015_asset,
							Audio.VO.ShogunX.VO_FirePetal_WanderingBoss_S33_SusReturnToIdle_016_asset,
							Audio.VO.ShogunX.VO_FirePetal_WanderingBoss_S33_SusReturnToIdle_020_asset,
							Audio.VO.ShogunX.VO_FirePetal_WanderingBoss_S33_SusReturnToIdle_021_asset,
							Audio.VO.ShogunX.VO_FirePetal_WanderingBoss_S33_SusReturnToIdle_022_asset,
							Audio.VO.ShogunX.VO_FirePetal_WanderingBoss_S33_SusReturnToIdle_023_asset,
							Audio.VO.ShogunX.VO_FirePetal_WanderingBoss_S33_SusReturnToIdle_024_asset,
							Audio.VO.ShogunX.VO_FirePetal_WanderingBoss_S33_SusReturnToIdle_025_asset,
							Audio.VO.ShogunX.VO_FirePetal_WanderingBoss_S33_SusReturnToIdle_026_asset,
							Audio.VO.ShogunX.VO_FirePetal_WanderingBoss_S33_SusReturnToIdle_030_asset,
							Audio.VO.ShogunX.VO_FirePetal_WanderingBoss_S33_SusReturnToIdle_031_asset,
							Audio.VO.ShogunX.VO_FirePetal_WanderingBoss_S33_SusReturnToIdle_032_asset,
							Audio.VO.ShogunX.VO_FirePetal_WanderingBoss_S33_SusReturnToIdle_033_asset,
							Audio.VO.ShogunX.VO_FirePetal_WanderingBoss_S33_SusReturnToIdle_034_asset,
							Audio.VO.ShogunX.VO_FirePetal_WanderingBoss_S33_SusReturnToIdle_035_asset,
							Audio.VO.ShogunX.VO_FirePetal_WanderingBoss_S33_SusReturnToIdle_036_asset,
							Audio.VO.ShogunX.VO_FirePetal_WanderingBoss_S33_SusReturnToIdle_040_asset,
							Audio.VO.ShogunX.VO_FirePetal_WanderingBoss_S33_SusReturnToIdle_041_asset,
							Audio.VO.ShogunX.VO_FirePetal_WanderingBoss_S33_SusReturnToIdle_042_asset,
							Audio.VO.ShogunX.VO_FirePetal_WanderingBoss_S33_SusReturnToIdle_043_asset,
							Audio.VO.ShogunX.VO_FirePetal_WanderingBoss_S33_SusReturnToIdle_044_asset,
							Audio.VO.ShogunX.VO_FirePetal_WanderingBoss_S33_SusReturnToIdle_045_asset,#>
							Audio.VO.ShogunX.VO_FirePetal_WanderingBoss_S33_SusReturnToIdle_046_asset
						}
					}
					FNVOType.FullAlert =>
					{ return array{
							Audio.VO.ShogunX.VO_FirePetal_WanderingBoss_S33_FullAlert_010_asset,
							Audio.VO.ShogunX.VO_FirePetal_WanderingBoss_S33_FullAlert_011_asset,
							Audio.VO.ShogunX.VO_FirePetal_WanderingBoss_S33_FullAlert_020_asset,
							Audio.VO.ShogunX.VO_FirePetal_WanderingBoss_S33_FullAlert_022_asset,
							Audio.VO.ShogunX.VO_FirePetal_WanderingBoss_S33_FullAlert_030_asset,
							Audio.VO.ShogunX.VO_FirePetal_WanderingBoss_S33_FullAlert_040_asset,
							Audio.VO.ShogunX.VO_FirePetal_WanderingBoss_S33_FullAlert_050_asset
						}
					}
					FNVOType.FullAlertLost =>
					{ return array{
							Audio.VO.ShogunX.VO_FirePetal_WanderingBoss_S33_AlertLostTarget_010_asset,
							Audio.VO.ShogunX.VO_FirePetal_WanderingBoss_S33_AlertLostTarget_011_asset,
							Audio.VO.ShogunX.VO_FirePetal_WanderingBoss_S33_AlertLostTarget_012_asset,
							<#Audio.VO.ShogunX.VO_FirePetal_WanderingBoss_S33_AlertLostTarget_013_asset,
							Audio.VO.ShogunX.VO_FirePetal_WanderingBoss_S33_AlertLostTarget_014_asset,
							Audio.VO.ShogunX.VO_FirePetal_WanderingBoss_S33_AlertLostTarget_015_asset,
							Audio.VO.ShogunX.VO_FirePetal_WanderingBoss_S33_AlertLostTarget_016_asset,
							Audio.VO.ShogunX.VO_FirePetal_WanderingBoss_S33_AlertLostTarget_017_asset,
							Audio.VO.ShogunX.VO_FirePetal_WanderingBoss_S33_AlertLostTarget_020_asset,
							Audio.VO.ShogunX.VO_FirePetal_WanderingBoss_S33_AlertLostTarget_021_asset,
							Audio.VO.ShogunX.VO_FirePetal_WanderingBoss_S33_AlertLostTarget_030_asset,#>
							Audio.VO.ShogunX.VO_FirePetal_WanderingBoss_S33_AlertLostTarget_040_asset
						}
					}
					FNVOType.DBNO =>
					{ return array{
							Audio.VO.ShogunX.VO_FirePetal_WanderingBoss_S33_Elim_010_asset,
							Audio.VO.ShogunX.VO_FirePetal_WanderingBoss_S33_Elim_011_asset,
							Audio.VO.ShogunX.VO_FirePetal_WanderingBoss_S33_Elim_012_asset,
							Audio.VO.ShogunX.VO_FirePetal_WanderingBoss_S33_Elim_013_asset,
							Audio.VO.ShogunX.VO_FirePetal_WanderingBoss_S33_Elim_014_asset,
							Audio.VO.ShogunX.VO_FirePetal_WanderingBoss_S33_Elim_015_asset
						}
					}
					FNVOType.Interrogation =>
					{ return array{
							Audio.VO.ShogunX.VO_FirePetal_WanderingBoss_S33_TakeDamage_Lrg_010_asset,
							Audio.VO.ShogunX.VO_FirePetal_WanderingBoss_S33_TakeDamage_Lrg_011_asset,
							Audio.VO.ShogunX.VO_FirePetal_WanderingBoss_S33_TakeDamage_Lrg_012_asset,
							Audio.VO.ShogunX.VO_FirePetal_WanderingBoss_S33_TakeDamage_Lrg_013_asset,
							Audio.VO.ShogunX.VO_FirePetal_WanderingBoss_S33_TakeDamage_Lrg_014_asset,
							Audio.VO.ShogunX.VO_FirePetal_WanderingBoss_S33_TakeDamage_Lrg_015_asset,
							Audio.VO.ShogunX.VO_FirePetal_WanderingBoss_S33_TakeDamage_Lrg_016_asset,
							Audio.VO.ShogunX.VO_FirePetal_WanderingBoss_S33_TakeDamage_Lrg_017_asset,
							Audio.VO.ShogunX.VO_FirePetal_WanderingBoss_S33_TakeDamage_Lrg_018_asset,
							Audio.VO.ShogunX.VO_FirePetal_WanderingBoss_S33_TakeDamage_Lrg_019_asset

						}
					}
					FNVOType.Ability =>	
					{ return array{
							Audio.VO.ShogunX.VO_FirePetal_WanderingBoss_S33_AOE_Charge_010_asset,
							Audio.VO.ShogunX.VO_FirePetal_WanderingBoss_S33_AOE_Charge_011_asset,
							Audio.VO.ShogunX.VO_FirePetal_WanderingBoss_S33_AOE_Charge_012_asset,
							Audio.VO.ShogunX.VO_FirePetal_WanderingBoss_S33_AOE_Charge_013_asset,
							Audio.VO.ShogunX.VO_FirePetal_WanderingBoss_S33_AOE_Charge_014_asset,
							Audio.VO.ShogunX.VO_FirePetal_WanderingBoss_S33_AOE_Charge_015_asset,
							Audio.VO.ShogunX.VO_FirePetal_WanderingBoss_S33_AOE_Charge_016_asset
						}
					}
					FNVOType.AbilityEnd =>
					{ return array{
							Audio.VO.ShogunX.VO_FirePetal_WanderingBoss_S33_AOE_Cast_010_asset,
							Audio.VO.ShogunX.VO_FirePetal_WanderingBoss_S33_AOE_Cast_011_asset,
							Audio.VO.ShogunX.VO_FirePetal_WanderingBoss_S33_AOE_Cast_012_asset,
							Audio.VO.ShogunX.VO_FirePetal_WanderingBoss_S33_AOE_Cast_013_asset,
							Audio.VO.ShogunX.VO_FirePetal_WanderingBoss_S33_AOE_Cast_014_asset,
							Audio.VO.ShogunX.VO_FirePetal_WanderingBoss_S33_AOE_Cast_015_asset,
							Audio.VO.ShogunX.VO_FirePetal_WanderingBoss_S33_AOE_Cast_016_asset
						}
					}
					FNVOType.Ult =>
					{ return array{
							Audio.VO.ShogunX.VO_FirePetal_WanderingBoss_S33_Rift_Escape_010_asset,
							Audio.VO.ShogunX.VO_FirePetal_WanderingBoss_S33_Rift_Escape_011_asset,
							Audio.VO.ShogunX.VO_FirePetal_WanderingBoss_S33_Rift_Escape_012_asset,
							Audio.VO.ShogunX.VO_FirePetal_WanderingBoss_S33_Rift_Escape_013_asset,
							Audio.VO.ShogunX.VO_FirePetal_WanderingBoss_S33_Rift_Escape_014_asset,
							Audio.VO.ShogunX.VO_FirePetal_WanderingBoss_S33_Rift_Escape_015_asset,
							Audio.VO.ShogunX.VO_FirePetal_WanderingBoss_S33_Rift_Escape_016_asset
						}
					}
				}
			}
			FNClasses.NightRose =>
			{
				case( VOType )
				{
					FNVOType.Idle => 
					{ return array{
							Audio.VO.NightRose.VO_RoseDepth_S33_Idle_010_asset,
							Audio.VO.NightRose.VO_RoseDepth_S33_Idle_014_asset,
							Audio.VO.NightRose.VO_RoseDepth_S33_Idle_020_asset,
							Audio.VO.NightRose.VO_RoseDepth_S33_Idle_022_asset,
							Audio.VO.NightRose.VO_RoseDepth_S33_Idle_024_asset,
							Audio.VO.NightRose.VO_RoseDepth_S33_Idle_040_asset,
							Audio.VO.NightRose.VO_RoseDepth_S33_Idle_050_asset,
							Audio.VO.NightRose.VO_RoseDepth_S33_Idle_060_asset,
							Audio.VO.NightRose.VO_RoseDepth_S33_Idle_061_asset,
							Audio.VO.NightRose.VO_RoseDepth_S33_Idle_062_asset
						}
					}
					FNVOType.Sus =>
					{ return array{
							Audio.VO.NightRose.VO_RoseDepth_S33_Suspicious_010_asset,
							Audio.VO.NightRose.VO_RoseDepth_S33_Suspicious_020_asset,
							Audio.VO.NightRose.VO_RoseDepth_S33_Suspicious_030_asset,
							Audio.VO.NightRose.VO_RoseDepth_S33_Suspicious_031_asset,
							<#Audio.VO.NightRose.VO_RoseDepth_S33_Suspicious_040_asset,
							Audio.VO.NightRose.VO_RoseDepth_S33_Suspicious_041_asset,
							Audio.VO.NightRose.VO_RoseDepth_S33_Suspicious_050_asset,
							Audio.VO.NightRose.VO_RoseDepth_S33_Suspicious_051_asset,#>
							Audio.VO.NightRose.VO_RoseDepth_S33_Suspicious_060_asset
						}
					}
					FNVOType.ReturnIdle =>
					{ return array{
							Audio.VO.NightRose.VO_RoseDepth_S33_SusReturnToIdle_010_asset,
							Audio.VO.NightRose.VO_RoseDepth_S33_SusReturnToIdle_020_asset,
							Audio.VO.NightRose.VO_RoseDepth_S33_SusReturnToIdle_021_asset,
							Audio.VO.NightRose.VO_RoseDepth_S33_SusReturnToIdle_022_asset,
							<#Audio.VO.NightRose.VO_RoseDepth_S33_SusReturnToIdle_023_asset,
							Audio.VO.NightRose.VO_RoseDepth_S33_SusReturnToIdle_024_asset,
							Audio.VO.NightRose.VO_RoseDepth_S33_SusReturnToIdle_027_asset,
							Audio.VO.NightRose.VO_RoseDepth_S33_SusReturnToIdle_030_asset,
							Audio.VO.NightRose.VO_RoseDepth_S33_SusReturnToIdle_031_asset,
							Audio.VO.NightRose.VO_RoseDepth_S33_SusReturnToIdle_040_asset,
							Audio.VO.NightRose.VO_RoseDepth_S33_SusReturnToIdle_041_asset,
							Audio.VO.NightRose.VO_RoseDepth_S33_SusReturnToIdle_042_asset,
							Audio.VO.NightRose.VO_RoseDepth_S33_SusReturnToIdle_050_asset,#>
							Audio.VO.NightRose.VO_RoseDepth_S33_SusReturnToIdle_051_asset
						}
					}
					FNVOType.FullAlert =>
					{ return array{
							Audio.VO.NightRose.VO_RoseDepth_S33_Phase2_DBNO_012_asset,
							Audio.VO.NightRose.VO_RoseDepth_S33_Phase2_DBNO_013_asset,
							Audio.VO.NightRose.VO_RoseDepth_S33_Phase2_DBNO_015_asset,
							Audio.VO.NightRose.VO_RoseDepth_S33_Phase2_DBNO_017_asset
						}
					}
					FNVOType.FullAlertLost =>
					{ return array{
							Audio.VO.NightRose.VO_RoseDepth_S33_AlertLostTarget_010_asset,
							Audio.VO.NightRose.VO_RoseDepth_S33_AlertLostTarget_020_asset,
							Audio.VO.NightRose.VO_RoseDepth_S33_AlertLostTarget_030_asset,
							Audio.VO.NightRose.VO_RoseDepth_S33_AlertLostTarget_040_asset,
							Audio.VO.NightRose.VO_RoseDepth_S33_AlertLostTarget_050_asset,
							Audio.VO.NightRose.VO_RoseDepth_S33_AlertLostTarget_060_asset
						}
					}
					FNVOType.DBNO =>
					{ return array{
							Audio.VO.NightRose.VO_RoseDepth_S33_Elim_010_asset,
							Audio.VO.NightRose.VO_RoseDepth_S33_Elim_011_asset,
							Audio.VO.NightRose.VO_RoseDepth_S33_Elim_012_asset,
							Audio.VO.NightRose.VO_RoseDepth_S33_Elim_013_asset,
							Audio.VO.NightRose.VO_RoseDepth_S33_Elim_014_asset,
							Audio.VO.NightRose.VO_RoseDepth_S33_Elim_015_asset
						}
					}
					FNVOType.Interrogation =>
					{ return array{
							Audio.VO.NightRose.VO_RoseDepth_S33_WeakPoint_010_asset,
							Audio.VO.NightRose.VO_RoseDepth_S33_WeakPoint_011_asset,
							Audio.VO.NightRose.VO_RoseDepth_S33_WeakPoint_020_asset,
							Audio.VO.NightRose.VO_RoseDepth_S33_WeakPoint_021_asset,
							Audio.VO.NightRose.VO_RoseDepth_S33_WeakPoint_030_asset,
							Audio.VO.NightRose.VO_RoseDepth_S33_WeakPoint_040_asset,
							Audio.VO.NightRose.VO_RoseDepth_S33_WeakPoint_041_asset,
							Audio.VO.NightRose.VO_RoseDepth_S33_WeakPoint_050_asset
						}
					}
					FNVOType.Ability =>	
					{ return array{
							Audio.VO.NightRose.VO_RoseDepth_S33_DemonStrike_Swing1_010_asset,
							Audio.VO.NightRose.VO_RoseDepth_S33_DemonStrike_Swing1_011_asset,
							Audio.VO.NightRose.VO_RoseDepth_S33_DemonStrike_Swing1_012_asset,
							Audio.VO.NightRose.VO_RoseDepth_S33_DemonStrike_Swing1_013_asset,
							Audio.VO.NightRose.VO_RoseDepth_S33_DemonStrike_Swing1_014_asset
						}
					}
					FNVOType.AbilityEnd =>
					{ return array{
							Audio.VO.NightRose.VO_RoseDepth_S33_DemonStrike_Swing1_010_asset,
							Audio.VO.NightRose.VO_RoseDepth_S33_DemonStrike_Swing1_011_asset,
							Audio.VO.NightRose.VO_RoseDepth_S33_DemonStrike_Swing1_012_asset,
							Audio.VO.NightRose.VO_RoseDepth_S33_DemonStrike_Swing1_013_asset,
							Audio.VO.NightRose.VO_RoseDepth_S33_DemonStrike_Swing1_014_asset
						}
					}
					FNVOType.Ult =>
					{ return array{
							Audio.VO.NightRose.VO_RoseDepth_S33_Phase2_DBNO_012_asset,
							Audio.VO.NightRose.VO_RoseDepth_S33_Phase2_DBNO_013_asset,
							Audio.VO.NightRose.VO_RoseDepth_S33_Phase2_DBNO_015_asset,
							Audio.VO.NightRose.VO_RoseDepth_S33_Phase2_DBNO_017_asset
						}
					}
				}
			}
			FNClasses.Puppeteer =>
			{
				case( VOType )
				{
					FNVOType.Idle => 
					{ return array{
							Audio.VO.NightRose.Puppeteer.VO_FirePetal_Puppeteer_Mask_S33_Ambient_016_asset,
							Audio.VO.NightRose.Puppeteer.VO_FirePetal_Puppeteer_Mask_S33_Ambient_026_asset,
							Audio.VO.NightRose.Puppeteer.VO_FirePetal_Puppeteer_Mask_S33_Ambient_034_asset,
							Audio.VO.NightRose.Puppeteer.VO_FirePetal_Puppeteer_Mask_S33_Ambient_044_asset,
							Audio.VO.NightRose.Puppeteer.VO_FirePetal_Puppeteer_Mask_S33_Ambient_046_asset,
							Audio.VO.NightRose.Puppeteer.VO_FirePetal_Puppeteer_Mask_S33_Ambient_049_asset,
							Audio.VO.NightRose.Puppeteer.VO_FirePetal_Puppeteer_Mask_S33_Ambient_070_asset,
							Audio.VO.NightRose.Puppeteer.VO_FirePetal_Puppeteer_Mask_S33_Ambient_210_asset
						}
					}
					FNVOType.Sus =>
					{ return array{
							Audio.VO.NightRose.Puppeteer.VO_FirePetal_Puppeteer_Mask_S33_Ambient_016_asset,
							Audio.VO.NightRose.Puppeteer.VO_FirePetal_Puppeteer_Mask_S33_Ambient_026_asset,
							Audio.VO.NightRose.Puppeteer.VO_FirePetal_Puppeteer_Mask_S33_Ambient_034_asset,
							Audio.VO.NightRose.Puppeteer.VO_FirePetal_Puppeteer_Mask_S33_Ambient_044_asset,
							Audio.VO.NightRose.Puppeteer.VO_FirePetal_Puppeteer_Mask_S33_Ambient_046_asset,
							Audio.VO.NightRose.Puppeteer.VO_FirePetal_Puppeteer_Mask_S33_Ambient_049_asset,
							Audio.VO.NightRose.Puppeteer.VO_FirePetal_Puppeteer_Mask_S33_Ambient_070_asset,
							Audio.VO.NightRose.Puppeteer.VO_FirePetal_Puppeteer_Mask_S33_Ambient_210_asset
						}
					}
					FNVOType.ReturnIdle =>
					{ return array{
							Audio.VO.NightRose.Puppeteer.VO_FirePetal_Puppeteer_Mask_S33_Ambient_016_asset,
							Audio.VO.NightRose.Puppeteer.VO_FirePetal_Puppeteer_Mask_S33_Ambient_026_asset,
							Audio.VO.NightRose.Puppeteer.VO_FirePetal_Puppeteer_Mask_S33_Ambient_034_asset,
							Audio.VO.NightRose.Puppeteer.VO_FirePetal_Puppeteer_Mask_S33_Ambient_044_asset,
							Audio.VO.NightRose.Puppeteer.VO_FirePetal_Puppeteer_Mask_S33_Ambient_046_asset,
							Audio.VO.NightRose.Puppeteer.VO_FirePetal_Puppeteer_Mask_S33_Ambient_049_asset,
							Audio.VO.NightRose.Puppeteer.VO_FirePetal_Puppeteer_Mask_S33_Ambient_070_asset,
							Audio.VO.NightRose.Puppeteer.VO_FirePetal_Puppeteer_Mask_S33_Ambient_210_asset
						}
					}
					FNVOType.FullAlert =>
					{ return array{
							Audio.VO.NightRose.Puppeteer.VO_FirePetal_Puppeteer_Mask_S33_FullAlert_010_asset,
							Audio.VO.NightRose.Puppeteer.VO_FirePetal_Puppeteer_Mask_S33_FullAlert_011_asset,
							Audio.VO.NightRose.Puppeteer.VO_FirePetal_Puppeteer_Mask_S33_FullAlert_012_asset,
							Audio.VO.NightRose.Puppeteer.VO_FirePetal_Puppeteer_Mask_S33_FullAlert_013_asset,
							Audio.VO.NightRose.Puppeteer.VO_FirePetal_Puppeteer_Mask_S33_FullAlert_015_asset,
							Audio.VO.NightRose.Puppeteer.VO_FirePetal_Puppeteer_Mask_S33_FullAlert_030_asset,
							Audio.VO.NightRose.Puppeteer.VO_FirePetal_Puppeteer_Mask_S33_FullAlert_033_asset
						}
					}
					FNVOType.FullAlertLost =>
					{ return array{
							Audio.VO.NightRose.Puppeteer.VO_FirePetal_Puppeteer_Mask_S33_FullAlert_010_asset,
							Audio.VO.NightRose.Puppeteer.VO_FirePetal_Puppeteer_Mask_S33_FullAlert_011_asset,
							Audio.VO.NightRose.Puppeteer.VO_FirePetal_Puppeteer_Mask_S33_FullAlert_012_asset,
							Audio.VO.NightRose.Puppeteer.VO_FirePetal_Puppeteer_Mask_S33_FullAlert_013_asset,
							Audio.VO.NightRose.Puppeteer.VO_FirePetal_Puppeteer_Mask_S33_FullAlert_015_asset,
							Audio.VO.NightRose.Puppeteer.VO_FirePetal_Puppeteer_Mask_S33_FullAlert_030_asset,
							Audio.VO.NightRose.Puppeteer.VO_FirePetal_Puppeteer_Mask_S33_FullAlert_033_asset
						}
					}
					FNVOType.DBNO =>
					{ return array{
							Audio.VO.NightRose.Puppeteer.VO_FirePetal_Puppeteer_Mask_S33_Elim_013_asset
						}
					}
					FNVOType.Interrogation =>
					{ return array{
							Audio.VO.NightRose.Puppeteer.VO_FirePetal_Puppeteer_Mask_S33_Stunned_010_asset,
							Audio.VO.NightRose.Puppeteer.VO_FirePetal_Puppeteer_Mask_S33_Stunned_011_asset,
							Audio.VO.NightRose.Puppeteer.VO_FirePetal_Puppeteer_Mask_S33_Stunned_012_asset,
							Audio.VO.NightRose.Puppeteer.VO_FirePetal_Puppeteer_Mask_S33_Stunned_015_asset,
							Audio.VO.NightRose.Puppeteer.VO_FirePetal_Puppeteer_Mask_S33_Stunned_020_asset,
							Audio.VO.NightRose.Puppeteer.VO_FirePetal_Puppeteer_Mask_S33_Stunned_022_asset

						}
					}
					FNVOType.Ability =>	
					{ return array{
							Audio.VO.NightRose.Puppeteer.VO_FirePetal_Puppeteer_Mask_S33_LaunchProj_1_010_asset,
							Audio.VO.NightRose.Puppeteer.VO_FirePetal_Puppeteer_Mask_S33_LaunchProj_1_011_asset,
							Audio.VO.NightRose.Puppeteer.VO_FirePetal_Puppeteer_Mask_S33_LaunchProj_1_012_asset,
							Audio.VO.NightRose.Puppeteer.VO_FirePetal_Puppeteer_Mask_S33_LaunchProj_1_013_asset,
							Audio.VO.NightRose.Puppeteer.VO_FirePetal_Puppeteer_Mask_S33_LaunchProj_1_014_asset,
							Audio.VO.NightRose.Puppeteer.VO_FirePetal_Puppeteer_Mask_S33_LaunchProj_1_015_asset,
							Audio.VO.NightRose.Puppeteer.VO_FirePetal_Puppeteer_Mask_S33_LaunchProj_1_016_asset
						}
					}
					FNVOType.AbilityEnd =>
					{ return array{
							Audio.VO.NightRose.Puppeteer.VO_FirePetal_Puppeteer_Mask_S33_LaunchProj_1_010_asset,
							Audio.VO.NightRose.Puppeteer.VO_FirePetal_Puppeteer_Mask_S33_LaunchProj_1_011_asset,
							Audio.VO.NightRose.Puppeteer.VO_FirePetal_Puppeteer_Mask_S33_LaunchProj_1_012_asset,
							Audio.VO.NightRose.Puppeteer.VO_FirePetal_Puppeteer_Mask_S33_LaunchProj_1_013_asset,
							Audio.VO.NightRose.Puppeteer.VO_FirePetal_Puppeteer_Mask_S33_LaunchProj_1_014_asset,
							Audio.VO.NightRose.Puppeteer.VO_FirePetal_Puppeteer_Mask_S33_LaunchProj_1_015_asset,
							Audio.VO.NightRose.Puppeteer.VO_FirePetal_Puppeteer_Mask_S33_LaunchProj_1_016_asset
						}
					}
					FNVOType.Ult =>
					{ return array{
							Audio.VO.NightRose.Puppeteer.VO_FirePetal_Puppeteer_Mask_S33_FullAlert_010_asset,
							Audio.VO.NightRose.Puppeteer.VO_FirePetal_Puppeteer_Mask_S33_FullAlert_011_asset,
							Audio.VO.NightRose.Puppeteer.VO_FirePetal_Puppeteer_Mask_S33_FullAlert_012_asset,
							Audio.VO.NightRose.Puppeteer.VO_FirePetal_Puppeteer_Mask_S33_FullAlert_013_asset,
							Audio.VO.NightRose.Puppeteer.VO_FirePetal_Puppeteer_Mask_S33_FullAlert_015_asset,
							Audio.VO.NightRose.Puppeteer.VO_FirePetal_Puppeteer_Mask_S33_FullAlert_030_asset,
							Audio.VO.NightRose.Puppeteer.VO_FirePetal_Puppeteer_Mask_S33_FullAlert_033_asset
						}
					}
				}
			}
			FNClasses.FletcherKane =>
			{
				case( VOType )
				{
					FNVOType.Idle => 
					{ return array{
							Audio.VO.FletcherKane.VO_VividRazor_TheBaron_Idle_010_asset,
							Audio.VO.FletcherKane.VO_VividRazor_TheBaron_Idle_011_asset,
							Audio.VO.FletcherKane.VO_VividRazor_TheBaron_Idle_012_asset,
							Audio.VO.FletcherKane.VO_VividRazor_TheBaron_Idle_013_asset,
							Audio.VO.FletcherKane.VO_VividRazor_TheBaron_Idle_014_asset,
							Audio.VO.FletcherKane.VO_VividRazor_TheBaron_Idle_020_asset,
							Audio.VO.FletcherKane.VO_VividRazor_TheBaron_Idle_021_asset,
							Audio.VO.FletcherKane.VO_VividRazor_TheBaron_Idle_030_asset,
							Audio.VO.FletcherKane.VO_VividRazor_TheBaron_Idle_040_asset,
							Audio.VO.FletcherKane.VO_VividRazor_TheBaron_Idle_050_asset,
							Audio.VO.FletcherKane.VO_VividRazor_TheBaron_Idle_051_asset
						}
					}
					FNVOType.Sus =>
					{ return array{
							Audio.VO.FletcherKane.VO_VividRazor_TheBaron_Suspicious_011_asset,
							Audio.VO.FletcherKane.VO_VividRazor_TheBaron_Suspicious_020_asset,
							Audio.VO.FletcherKane.VO_VividRazor_TheBaron_Suspicious_021_asset,
							Audio.VO.FletcherKane.VO_VividRazor_TheBaron_Suspicious_040_asset,
							Audio.VO.FletcherKane.VO_VividRazor_TheBaron_Suspicious_041_asset,
							Audio.VO.FletcherKane.VO_VividRazor_TheBaron_Suspicious_042_asset
						}
					}
					FNVOType.ReturnIdle =>
					{ return array{
							Audio.VO.FletcherKane.VO_VividRazor_TheBaron_SusReturnToIdle_010_asset,
							Audio.VO.FletcherKane.VO_VividRazor_TheBaron_SusReturnToIdle_011_asset,
							Audio.VO.FletcherKane.VO_VividRazor_TheBaron_SusReturnToIdle_020_asset,
							Audio.VO.FletcherKane.VO_VividRazor_TheBaron_SusReturnToIdle_021_asset,
							Audio.VO.FletcherKane.VO_VividRazor_TheBaron_SusReturnToIdle_030_asset,
							Audio.VO.FletcherKane.VO_VividRazor_TheBaron_SusReturnToIdle_031_asset,
							Audio.VO.FletcherKane.VO_VividRazor_TheBaron_SusReturnToIdle_041_asset,
							Audio.VO.FletcherKane.VO_VividRazor_TheBaron_SusReturnToIdle_050_asset
						}
					}
					FNVOType.FullAlert =>
					{ return array{
							Audio.VO.FletcherKane.VO_VividRazor_TheBaron_FullAlert_010_asset,
							Audio.VO.FletcherKane.VO_VividRazor_TheBaron_FullAlert_011_asset,
							Audio.VO.FletcherKane.VO_VividRazor_TheBaron_FullAlert_020_asset,
							Audio.VO.FletcherKane.VO_VividRazor_TheBaron_FullAlert_030_asset,
							Audio.VO.FletcherKane.VO_VividRazor_TheBaron_FullAlert_031_asset,
							Audio.VO.FletcherKane.VO_VividRazor_TheBaron_FullAlert_032_asset,
							Audio.VO.FletcherKane.VO_VividRazor_TheBaron_FullAlert_040_asset,
							Audio.VO.FletcherKane.VO_VividRazor_TheBaron_FullAlert_041_asset
						}
					}
					FNVOType.FullAlertLost =>
					{ return array{
							Audio.VO.FletcherKane.VO_VividRazor_TheBaron_AlertLostTarget_010_asset,
							Audio.VO.FletcherKane.VO_VividRazor_TheBaron_AlertLostTarget_011_asset,
							Audio.VO.FletcherKane.VO_VividRazor_TheBaron_AlertLostTarget_012_asset,
							Audio.VO.FletcherKane.VO_VividRazor_TheBaron_AlertLostTarget_020_asset,
							Audio.VO.FletcherKane.VO_VividRazor_TheBaron_AlertLostTarget_021_asset,
							Audio.VO.FletcherKane.VO_VividRazor_TheBaron_AlertLostTarget_030_asset,
							Audio.VO.FletcherKane.VO_VividRazor_TheBaron_AlertLostTarget_040_asset,
							Audio.VO.FletcherKane.VO_VividRazor_TheBaron_AlertLostTarget_050_asset
						}
					}
					FNVOType.DBNO =>
					{ return array{
							Audio.VO.FletcherKane.VO_VividRazor_TheBaron_Elim_010_asset,
							Audio.VO.FletcherKane.VO_VividRazor_TheBaron_Elim_011_asset,
							Audio.VO.FletcherKane.VO_VividRazor_TheBaron_Elim_012_asset,
							Audio.VO.FletcherKane.VO_VividRazor_TheBaron_Elim_013_asset,
							Audio.VO.FletcherKane.VO_VividRazor_TheBaron_Elim_014_asset
						}
					}
					FNVOType.Interrogation =>
					{ return array{
							Audio.VO.FletcherKane.VO_VividRazor_TheBaron_Elim_010_asset,
							Audio.VO.FletcherKane.VO_VividRazor_TheBaron_Elim_011_asset,
							Audio.VO.FletcherKane.VO_VividRazor_TheBaron_Elim_012_asset,
							Audio.VO.FletcherKane.VO_VividRazor_TheBaron_Elim_013_asset,
							Audio.VO.FletcherKane.VO_VividRazor_TheBaron_Elim_014_asset

						}
					}
					FNVOType.Ability =>	
					{ return array{
							Audio.VO.FletcherKane.VO_VividRazor_TheBaron_TripleClawAttack_Swing3_010_asset,
							Audio.VO.FletcherKane.VO_VividRazor_TheBaron_TripleClawAttack_Swing3_011_asset,
							Audio.VO.FletcherKane.VO_VividRazor_TheBaron_TripleClawAttack_Swing3_012_asset,
							Audio.VO.FletcherKane.VO_VividRazor_TheBaron_TripleClawAttack_Swing3_013_asset
						}
					}
					FNVOType.AbilityEnd =>
					{ return array{
							Audio.VO.FletcherKane.VO_VividRazor_TheBaron_TripleClawAttack_Swing3_010_asset,
							Audio.VO.FletcherKane.VO_VividRazor_TheBaron_TripleClawAttack_Swing3_011_asset,
							Audio.VO.FletcherKane.VO_VividRazor_TheBaron_TripleClawAttack_Swing3_012_asset,
							Audio.VO.FletcherKane.VO_VividRazor_TheBaron_TripleClawAttack_Swing3_013_asset
						}
					}
					FNVOType.Ult =>
					{ return array{
							Audio.VO.FletcherKane.VO_VividRazor_TheBaron_Phase2_011_asset,
							Audio.VO.FletcherKane.VO_VividRazor_TheBaron_Phase2_020_asset,
							Audio.VO.FletcherKane.VO_VividRazor_TheBaron_Phase2_030_asset,
							Audio.VO.FletcherKane.VO_VividRazor_TheBaron_Phase2_040_asset,
							Audio.VO.FletcherKane.VO_VividRazor_TheBaron_Phase2_050_asset
						}
					}
				}
			}
			FNClasses.AlphaHeavy =>
			{
				case( VOType )
				{
					FNVOType.Idle => 
					{ return array{
							Audio.VO.AlphaHeavy.VO_S34_Henchman_1_Idle_010_asset,
							Audio.VO.AlphaHeavy.VO_S34_Henchman_1_Idle_012_asset,
							Audio.VO.AlphaHeavy.VO_S34_Henchman_1_Idle_014_asset,
							Audio.VO.AlphaHeavy.VO_S34_Henchman_1_Idle_020_asset,
							Audio.VO.AlphaHeavy.VO_S34_Henchman_1_Idle_030_asset,
							Audio.VO.AlphaHeavy.VO_S34_Henchman_1_Idle_051_asset,
							Audio.VO.AlphaHeavy.VO_S34_Henchman_1_Idle_060_asset,
							Audio.VO.AlphaHeavy.VO_S34_Henchman_1_Idle_061_asset
						}
					}
					FNVOType.Sus =>
					{ return array{
							Audio.VO.AlphaHeavy.VO_S34_Henchman_1_Suspicious_012_asset,
							Audio.VO.AlphaHeavy.VO_S34_Henchman_1_Suspicious_015_asset,
							Audio.VO.AlphaHeavy.VO_S34_Henchman_1_Suspicious_016_asset,
							Audio.VO.AlphaHeavy.VO_S34_Henchman_1_Suspicious_021_asset,
							Audio.VO.AlphaHeavy.VO_S34_Henchman_1_Suspicious_031_asset,
							Audio.VO.AlphaHeavy.VO_S34_Henchman_1_Suspicious_040_asset,
							Audio.VO.AlphaHeavy.VO_S34_Henchman_1_Suspicious_050_asset,
							Audio.VO.AlphaHeavy.VO_S34_Henchman_1_Suspicious_051_asset
						}
					}
					FNVOType.ReturnIdle =>
					{ return array{
							Audio.VO.AlphaHeavy.VO_S34_Henchman_1_SusReturnToIdle_010_asset,
							Audio.VO.AlphaHeavy.VO_S34_Henchman_1_SusReturnToIdle_011_asset,
							Audio.VO.AlphaHeavy.VO_S34_Henchman_1_SusReturnToIdle_020_asset,
							Audio.VO.AlphaHeavy.VO_S34_Henchman_1_SusReturnToIdle_030_asset,
							Audio.VO.AlphaHeavy.VO_S34_Henchman_1_SusReturnToIdle_041_asset,
							Audio.VO.AlphaHeavy.VO_S34_Henchman_1_SusReturnToIdle_051_asset
						}
					}
					FNVOType.FullAlert =>
					{ return array{
							Audio.VO.AlphaHeavy.VO_S34_Henchman_1_FullAlert_010_asset,
							Audio.VO.AlphaHeavy.VO_S34_Henchman_1_FullAlert_011_asset,
							Audio.VO.AlphaHeavy.VO_S34_Henchman_1_FullAlert_020_asset,
							Audio.VO.AlphaHeavy.VO_S34_Henchman_1_FullAlert_022_asset,
							Audio.VO.AlphaHeavy.VO_S34_Henchman_1_FullAlert_041_asset,
							Audio.VO.AlphaHeavy.VO_S34_Henchman_1_FullAlert_060_asset,
							Audio.VO.AlphaHeavy.VO_S34_Henchman_1_FullAlert_061_asset
						}
					}
					FNVOType.FullAlertLost =>
					{ return array{
							Audio.VO.AlphaHeavy.VO_S34_Henchman_1_AlertLostTarget_011_asset,
							Audio.VO.AlphaHeavy.VO_S34_Henchman_1_AlertLostTarget_012_asset,
							Audio.VO.AlphaHeavy.VO_S34_Henchman_1_AlertLostTarget_022_asset,
							Audio.VO.AlphaHeavy.VO_S34_Henchman_1_AlertLostTarget_032_asset,
							Audio.VO.AlphaHeavy.VO_S34_Henchman_1_AlertLostTarget_040_asset
						}
					}
					FNVOType.DBNO =>
					{ return array{
							Audio.VO.AlphaHeavy.VO_S34_Henchman_1_Elim_010_asset,
							Audio.VO.AlphaHeavy.VO_S34_Henchman_1_Elim_011_asset,
							Audio.VO.AlphaHeavy.VO_S34_Henchman_1_Elim_014_asset,
							Audio.VO.AlphaHeavy.VO_S34_Henchman_1_Elim_015_asset,
							Audio.VO.AlphaHeavy.VO_S34_Henchman_1_Elim_016_asset
						}
					}
					FNVOType.Interrogation =>
					{ return array{
							Audio.VO.AlphaHeavy.VO_S34_Henchman_1_Elim_010_asset,
							Audio.VO.AlphaHeavy.VO_S34_Henchman_1_Elim_011_asset,
							Audio.VO.AlphaHeavy.VO_S34_Henchman_1_Elim_014_asset,
							Audio.VO.AlphaHeavy.VO_S34_Henchman_1_Elim_015_asset,
							Audio.VO.AlphaHeavy.VO_S34_Henchman_1_Elim_016_asset

						}
					}
					FNVOType.Ability =>	
					{ return array{
							Audio.VO.AlphaHeavy.VO_S34_Henchman_1_FullAlert_010_asset,
							Audio.VO.AlphaHeavy.VO_S34_Henchman_1_FullAlert_011_asset,
							Audio.VO.AlphaHeavy.VO_S34_Henchman_1_FullAlert_020_asset,
							Audio.VO.AlphaHeavy.VO_S34_Henchman_1_FullAlert_022_asset,
							Audio.VO.AlphaHeavy.VO_S34_Henchman_1_FullAlert_041_asset,
							Audio.VO.AlphaHeavy.VO_S34_Henchman_1_FullAlert_060_asset,
							Audio.VO.AlphaHeavy.VO_S34_Henchman_1_FullAlert_061_asset
						}
					}
					FNVOType.AbilityEnd =>
					{ return array{
							Audio.VO.AlphaHeavy.VO_S34_Henchman_1_FullAlert_010_asset,
							Audio.VO.AlphaHeavy.VO_S34_Henchman_1_FullAlert_011_asset,
							Audio.VO.AlphaHeavy.VO_S34_Henchman_1_FullAlert_020_asset,
							Audio.VO.AlphaHeavy.VO_S34_Henchman_1_FullAlert_022_asset,
							Audio.VO.AlphaHeavy.VO_S34_Henchman_1_FullAlert_041_asset,
							Audio.VO.AlphaHeavy.VO_S34_Henchman_1_FullAlert_060_asset,
							Audio.VO.AlphaHeavy.VO_S34_Henchman_1_FullAlert_061_asset
						}
					}
					FNVOType.Ult =>
					{ return array{
							Audio.VO.AlphaHeavy.VO_S34_Henchman_1_FullAlert_010_asset,
							Audio.VO.AlphaHeavy.VO_S34_Henchman_1_FullAlert_011_asset,
							Audio.VO.AlphaHeavy.VO_S34_Henchman_1_FullAlert_020_asset,
							Audio.VO.AlphaHeavy.VO_S34_Henchman_1_FullAlert_022_asset,
							Audio.VO.AlphaHeavy.VO_S34_Henchman_1_FullAlert_041_asset,
							Audio.VO.AlphaHeavy.VO_S34_Henchman_1_FullAlert_060_asset,
							Audio.VO.AlphaHeavy.VO_S34_Henchman_1_FullAlert_061_asset
						}
					}
				}
			}
			FNClasses.Jones =>
			{

				case( VOType )
				{
					FNVOType.Idle => 
					{ return array{
						}
					}
					FNVOType.Sus =>
					{ return array{
						}
					}
					FNVOType.ReturnIdle =>
					{ return array{
						}
					}
					FNVOType.FullAlert =>
					{ return array{
							Audio.VO.Jones.JONES_2022_Armadillo_JonesFightsIOGuards_030_asset,
							Audio.VO.Jones.JONES_Guava_InGame_Final_PostFight_V4_050_asset,
							Audio.VO.Jones.JONES_Guava_InGame_Final_PostFight_V4_061_asset
						}
					}
					FNVOType.FullAlertLost =>
					{ return array{
							Audio.VO.Jones.JONES_2022_Armadillo_BruteDestroyed_010_asset,
							Audio.VO.Jones.JONES_2022_Armadillo_FoundationSeesBrute_020_asset,
							Audio.VO.Jones.JONES_Guava_InGame_Final_PostFight_V3_011_asset,
							Audio.VO.Jones.JONESY_2021_YogurtV3_FirstPortalClosed_GotIt_010_asset
						}
					}
					FNVOType.DBNO =>
					{ return array{
						}
					}
					FNVOType.Interrogation =>
					{ return array{
						}
					}
					FNVOType.Ability =>	
					{ return array{
						}
					}
					FNVOType.AbilityEnd =>
					{ return array{
						}
					}
					FNVOType.Ult =>
					{ return array{
						}
					}
				}
			}
			FNClasses.DoctorDoom =>
			{
				case( VOType )
				{
					FNVOType.Idle => 
					{ return array{
							Audio.VO.Doombot.VO_BossD_S31_AlertIdle_010_asset,
							Audio.VO.Doombot.VO_BossD_S31_AlertIdle_020_asset,
							Audio.VO.Doombot.VO_BossD_S31_AlertIdle_030_asset,
							Audio.VO.Doombot.VO_BossD_S31_AlertIdle_040_asset,
							Audio.VO.Doombot.VO_BossD_S31_AlertIdle_050_asset
						}
					}
					FNVOType.Sus =>
					{ return array{
							Audio.VO.Doombot.VO_BossD_S31_Suspicious_010_asset,
							Audio.VO.Doombot.VO_BossD_S31_Suspicious_020_asset,
							Audio.VO.Doombot.VO_BossD_S31_Suspicious_030_asset
						}
					}
					FNVOType.ReturnIdle =>
					{ return array{
							Audio.VO.Doombot.VO_BossD_S31_SusReturnToIdle_010_asset,
							Audio.VO.Doombot.VO_BossD_S31_SusReturnToIdle_020_asset,
							Audio.VO.Doombot.VO_BossD_S31_SusReturnToIdle_030_asset,
							Audio.VO.Doombot.VO_BossD_S31_SusReturnToIdle_040_asset,
							Audio.VO.Doombot.VO_BossD_S31_SusReturnToIdle_050_asset
						}
					}
					FNVOType.FullAlert =>
					{ return array{
							Audio.VO.Doombot.VO_BossD_S31_FullAlert_010_asset,
							Audio.VO.Doombot.VO_BossD_S31_FullAlert_020_asset,
							Audio.VO.Doombot.VO_BossD_S31_FullAlert_030_asset,
							Audio.VO.Doombot.VO_BossD_S31_FullAlert_040_asset,
							Audio.VO.Doombot.VO_BossD_S31_FullAlert_050_asset
						}
					}
					FNVOType.FullAlertLost =>
					{ return array{
							Audio.VO.Doombot.VO_BossD_S31_LostTarget_011_asset,
							Audio.VO.Doombot.VO_BossD_S31_LostTarget_013_asset,
							Audio.VO.Doombot.VO_BossD_S31_LostTarget_014_asset,
							Audio.VO.Doombot.VO_BossD_S31_LostTarget_020_asset,
							Audio.VO.Doombot.VO_BossD_S31_LostTarget_051_asset
						}
					}
					FNVOType.DBNO =>
					{ return array{
							Audio.VO.Doombot.VO_BossD_S31_Elim_020_asset,
							Audio.VO.Doombot.VO_BossD_S31_Elim_021_asset,
							Audio.VO.Doombot.VO_BossD_S31_Elim_024_asset
						}
					}
					FNVOType.Interrogation =>
					{ return array{
							Audio.VO.Doombot.VO_BossD_S31_LoseShields_010_asset,
							Audio.VO.Doombot.VO_BossD_S31_LoseShields_020_asset,
							Audio.VO.Doombot.VO_BossD_S31_LoseShields_030_asset,
							Audio.VO.Doombot.VO_BossD_S31_LoseShields_040_asset,
							Audio.VO.Doombot.VO_BossD_S31_LoseShields_050_asset
						}
					}
					FNVOType.Ability =>	
					{ return array{
							Audio.VO.Doombot.VO_BossD_S31_BigBall_010_asset,
							Audio.VO.Doombot.VO_BossD_S31_BigBall_020_asset,
							Audio.VO.Doombot.VO_BossD_S31_BigBall_030_asset,
							Audio.VO.Doombot.VO_BossD_S31_BigBall_040_asset,
							Audio.VO.Doombot.VO_BossD_S31_BigBall_050_asset
						}
					}
					FNVOType.AbilityEnd =>
					{ return array{
							Audio.VO.Doombot.VO_BossD_S31_KickAbility_010_asset,
							Audio.VO.Doombot.VO_BossD_S31_KickAbility_020_asset,
							Audio.VO.Doombot.VO_BossD_S31_KickAbility_030_asset,
							Audio.VO.Doombot.VO_BossD_S31_KickAbility_040_asset,
							Audio.VO.Doombot.VO_BossD_S31_KickAbility_050_asset
						}
					}
					FNVOType.Ult =>
					{ return array{
							Audio.VO.Doom.VO_S31_DentVibe_StandAgainst_020_asset
						}
					}
				}
			}
			FNClasses.DoctorDoomMythic =>
			{
				case( VOType )
				{
					FNVOType.Idle => 
					{ return array{
							Audio.VO.Doombot.VO_BossD_S31_AlertIdle_010_asset,
							Audio.VO.Doombot.VO_BossD_S31_AlertIdle_020_asset,
							Audio.VO.Doombot.VO_BossD_S31_AlertIdle_030_asset,
							Audio.VO.Doombot.VO_BossD_S31_AlertIdle_040_asset,
							Audio.VO.Doombot.VO_BossD_S31_AlertIdle_050_asset
						}
					}
					FNVOType.Sus =>
					{ return array{
							Audio.VO.Doombot.VO_BossD_S31_Suspicious_010_asset,
							Audio.VO.Doombot.VO_BossD_S31_Suspicious_020_asset,
							Audio.VO.Doombot.VO_BossD_S31_Suspicious_030_asset
						}
					}
					FNVOType.ReturnIdle =>
					{ return array{
							Audio.VO.Doombot.VO_BossD_S31_SusReturnToIdle_010_asset,
							Audio.VO.Doombot.VO_BossD_S31_SusReturnToIdle_020_asset,
							Audio.VO.Doombot.VO_BossD_S31_SusReturnToIdle_030_asset,
							Audio.VO.Doombot.VO_BossD_S31_SusReturnToIdle_040_asset,
							Audio.VO.Doombot.VO_BossD_S31_SusReturnToIdle_050_asset
						}
					}
					FNVOType.FullAlert =>
					{ return array{
							Audio.VO.Doombot.VO_BossD_S31_FullAlert_010_asset,
							Audio.VO.Doombot.VO_BossD_S31_FullAlert_020_asset,
							Audio.VO.Doombot.VO_BossD_S31_FullAlert_030_asset,
							Audio.VO.Doombot.VO_BossD_S31_FullAlert_040_asset,
							Audio.VO.Doombot.VO_BossD_S31_FullAlert_050_asset
						}
					}
					FNVOType.FullAlertLost =>
					{ return array{
							Audio.VO.Doombot.VO_BossD_S31_LostTarget_011_asset,
							Audio.VO.Doombot.VO_BossD_S31_LostTarget_013_asset,
							Audio.VO.Doombot.VO_BossD_S31_LostTarget_014_asset,
							Audio.VO.Doombot.VO_BossD_S31_LostTarget_020_asset,
							Audio.VO.Doombot.VO_BossD_S31_LostTarget_051_asset
						}
					}
					FNVOType.DBNO =>
					{ return array{
							Audio.VO.Doombot.VO_BossD_S31_Elim_020_asset,
							Audio.VO.Doombot.VO_BossD_S31_Elim_021_asset,
							Audio.VO.Doombot.VO_BossD_S31_Elim_024_asset
						}
					}
					FNVOType.Interrogation =>
					{ return array{
							Audio.VO.Doombot.VO_BossD_S31_LoseShields_010_asset,
							Audio.VO.Doombot.VO_BossD_S31_LoseShields_020_asset,
							Audio.VO.Doombot.VO_BossD_S31_LoseShields_030_asset,
							Audio.VO.Doombot.VO_BossD_S31_LoseShields_040_asset,
							Audio.VO.Doombot.VO_BossD_S31_LoseShields_050_asset
						}
					}
					FNVOType.Ability =>	
					{ return array{
							Audio.VO.Doombot.VO_BossD_S31_BigBall_010_asset,
							Audio.VO.Doombot.VO_BossD_S31_BigBall_020_asset,
							Audio.VO.Doombot.VO_BossD_S31_BigBall_030_asset,
							Audio.VO.Doombot.VO_BossD_S31_BigBall_040_asset,
							Audio.VO.Doombot.VO_BossD_S31_BigBall_050_asset
						}
					}
					FNVOType.AbilityEnd =>
					{ return array{
							Audio.VO.Doombot.VO_BossD_S31_KickAbility_010_asset,
							Audio.VO.Doombot.VO_BossD_S31_KickAbility_020_asset,
							Audio.VO.Doombot.VO_BossD_S31_KickAbility_030_asset,
							Audio.VO.Doombot.VO_BossD_S31_KickAbility_040_asset,
							Audio.VO.Doombot.VO_BossD_S31_KickAbility_050_asset
						}
					}
					FNVOType.Ult =>
					{ return array{
							Audio.VO.Doom.VO_S31_DentVibe_StandAgainst_020_asset
						}
					}
				}
			}
			FNClasses.Caretaker =>
			{
				case( VOType )
				{
					FNVOType.Idle => 
					{ return array{
							Audio.VO.Caretaker.Caretaker_Vox_Groan_Turn_Inv_01_asset,
							Audio.VO.Caretaker.Caretaker_Vox_Groan_Turn_Inv_02_asset,
							Audio.VO.Caretaker.Caretaker_Vox_Groan_Turn_Inv_03_asset,
							Audio.VO.Caretaker.Caretaker_Vox_Groan_Turn_Inv_04_asset
						}
					}
					FNVOType.Sus =>
					{ return array{
							Audio.VO.Caretaker.Caretaker_Vox_Groan_Turn_Inv_01_asset,
							Audio.VO.Caretaker.Caretaker_Vox_Groan_Turn_Inv_02_asset,
							Audio.VO.Caretaker.Caretaker_Vox_Groan_Turn_Inv_03_asset,
							Audio.VO.Caretaker.Caretaker_Vox_Groan_Turn_Inv_04_asset
						}
					}
					FNVOType.ReturnIdle =>
					{ return array{
							Audio.VO.Caretaker.Caretaker_Vox_Groan_Turn_Inv_01_asset,
							Audio.VO.Caretaker.Caretaker_Vox_Groan_Turn_Inv_02_asset,
							Audio.VO.Caretaker.Caretaker_Vox_Groan_Turn_Inv_03_asset,
							Audio.VO.Caretaker.Caretaker_Vox_Groan_Turn_Inv_04_asset
						}
					}
					FNVOType.FullAlert =>
					{ return array{
							Audio.VO.Caretaker.Caretaker_Vox_Groan_Turn_Inv_01_asset,
							Audio.VO.Caretaker.Caretaker_Vox_Groan_Turn_Inv_02_asset,
							Audio.VO.Caretaker.Caretaker_Vox_Groan_Turn_Inv_03_asset,
							Audio.VO.Caretaker.Caretaker_Vox_Groan_Turn_Inv_04_asset
						}
					}
					FNVOType.FullAlertLost =>
					{ return array{
							Audio.VO.Caretaker.Caretaker_Vox_Groan_Turn_Inv_01_asset,
							Audio.VO.Caretaker.Caretaker_Vox_Groan_Turn_Inv_02_asset,
							Audio.VO.Caretaker.Caretaker_Vox_Groan_Turn_Inv_03_asset,
							Audio.VO.Caretaker.Caretaker_Vox_Groan_Turn_Inv_04_asset
						}
					}
					FNVOType.DBNO =>
					{ return array{
							Audio.VO.Caretaker.Caretaker_Death_Vox_01_asset
						}
					}
					FNVOType.Interrogation =>
					{ return array{
							Audio.VO.Caretaker.Caretaker_Vox_Groan_Turn_Inv_01_asset,
							Audio.VO.Caretaker.Caretaker_Vox_Groan_Turn_Inv_02_asset,
							Audio.VO.Caretaker.Caretaker_Vox_Groan_Turn_Inv_03_asset,
							Audio.VO.Caretaker.Caretaker_Vox_Groan_Turn_Inv_04_asset

						}
					}
					FNVOType.Ability =>	
					{ return array{
							Audio.VO.Caretaker.Caretaker_Smash_Vox_01_asset,
							Audio.VO.Caretaker.Caretaker_Smash_Vox_02_asset,
							Audio.VO.Caretaker.Caretaker_Smash_Vox_03_asset,
							Audio.VO.Caretaker.Caretaker_Smash_Vox_04_asset
						}
					}
					FNVOType.AbilityEnd =>
					{ return array{
							Audio.VO.Caretaker.Caretaker_Snap_Vox_01_asset,
							Audio.VO.Caretaker.Caretaker_Snap_Vox_02_asset,
							Audio.VO.Caretaker.Caretaker_Snap_Vox_03_asset,
							Audio.VO.Caretaker.Caretaker_Snap_Vox_04_asset
						}
					}
					FNVOType.Ult =>
					{ return array{
							Audio.VO.Caretaker.Caretaker_Sideways_Vox_01_asset,
							Audio.VO.Caretaker.Caretaker_Sideways_Vox_02_asset,
							Audio.VO.Caretaker.Caretaker_Sideways_Vox_03_asset
						}
					}
				}
			}
			FNClasses.Hope =>
			{
				case( VOType )
				{
					FNVOType.Idle => 
					{ return array{
							Audio.VO.Midas.Henchman_Midas_Idle_010_asset,
							Audio.VO.Midas.Henchman_Midas_Idle_020_asset,
							Audio.VO.Midas.Henchman_Midas_Idle_030_asset,
							Audio.VO.Midas.Henchman_Midas_Idle_040_asset,
							Audio.VO.Midas.Henchman_Midas_Idle_050_asset,
							Audio.VO.Midas.Henchman_Midas_Idle_060_asset
						}
					}
					FNVOType.Sus =>
					{ return array{
							Audio.VO.Midas.Henchman_Midas_Suspicious_010_asset,
							Audio.VO.Midas.Henchman_Midas_Suspicious_020_asset,
							Audio.VO.Midas.Henchman_Midas_Suspicious_030_asset,
							Audio.VO.Midas.Henchman_Midas_Suspicious_040_asset,
							Audio.VO.Midas.Henchman_Midas_Suspicious_050_asset,
							Audio.VO.Midas.Henchman_Midas_Suspicious_060_asset
						}
					}
					FNVOType.ReturnIdle =>
					{ return array{
							Audio.VO.Midas.Henchman_Midas_SuspiciousReturnToIdle_010_asset,
							Audio.VO.Midas.Henchman_Midas_SuspiciousReturnToIdle_020_asset,
							Audio.VO.Midas.Henchman_Midas_SuspiciousReturnToIdle_030_asset,
							Audio.VO.Midas.Henchman_Midas_SuspiciousReturnToIdle_040_asset,
							Audio.VO.Midas.Henchman_Midas_SuspiciousReturnToIdle_050_asset,
							Audio.VO.Midas.Henchman_Midas_SuspiciousReturnToIdle_060_asset
						}
					}
					FNVOType.FullAlert =>
					{ return array{
							Audio.VO.Midas.Henchman_Midas_FullAlert_010_asset,
							Audio.VO.Midas.Henchman_Midas_FullAlert_020_asset,
							Audio.VO.Midas.Henchman_Midas_FullAlert_030_asset,
							Audio.VO.Midas.Henchman_Midas_FullAlert_040_asset,
							Audio.VO.Midas.Henchman_Midas_FullAlert_050_asset,
							Audio.VO.Midas.Henchman_Midas_FullAlert_060_asset,
							Audio.VO.Midas.Henchman_Midas_FullAlert_070_asset
						}
					}
					FNVOType.FullAlertLost =>
					{ return array{
							Audio.VO.Midas.Henchman_Midas_FullAlertLostTarget_010_asset,
							Audio.VO.Midas.Henchman_Midas_FullAlertLostTarget_020_asset,
							Audio.VO.Midas.Henchman_Midas_FullAlertLostTarget_030_asset,
							Audio.VO.Midas.Henchman_Midas_FullAlertLostTarget_040_asset,
							Audio.VO.Midas.Henchman_Midas_FullAlertLostTarget_050_asset,
							Audio.VO.Midas.Henchman_Midas_FullAlertLostTarget_060_asset
						}
					}
					FNVOType.DBNO =>
					{ return array{
							Audio.VO.Midas.Henchman_Midas_DBNO_Enter_010_asset,
							Audio.VO.Midas.Henchman_Midas_DBNO_Enter_020_asset,
							Audio.VO.Midas.Henchman_Midas_DBNO_Enter_030_asset,
							Audio.VO.Midas.Henchman_Midas_DBNO_Enter_040_asset,
							Audio.VO.Midas.Henchman_Midas_DBNO_Enter_050_asset
						}
					}
					FNVOType.Interrogation =>
					{ return array{
							Audio.VO.Midas.Henchman_Midas_Interrogation_010_asset,
							Audio.VO.Midas.Henchman_Midas_Interrogation_020_asset,
							Audio.VO.Midas.Henchman_Midas_Interrogation_030_asset,
							Audio.VO.Midas.Henchman_Midas_Interrogation_040_asset,
							Audio.VO.Midas.Henchman_Midas_Interrogation_050_asset,
							Audio.VO.Midas.Henchman_Midas_Interrogation_060_asset

						}
					}
					FNVOType.Ability =>	
					{ return array{
							Audio.VO.Midas.Henchman_Midas_FullAlert_010_asset,
							Audio.VO.Midas.Henchman_Midas_FullAlert_020_asset,
							Audio.VO.Midas.Henchman_Midas_FullAlert_030_asset,
							Audio.VO.Midas.Henchman_Midas_FullAlert_040_asset,
							Audio.VO.Midas.Henchman_Midas_FullAlert_050_asset,
							Audio.VO.Midas.Henchman_Midas_FullAlert_060_asset,
							Audio.VO.Midas.Henchman_Midas_FullAlert_070_asset
						}
					}
					FNVOType.AbilityEnd =>
					{ return array{
							Audio.VO.Midas.Henchman_Midas_FullAlert_010_asset,
							Audio.VO.Midas.Henchman_Midas_FullAlert_020_asset,
							Audio.VO.Midas.Henchman_Midas_FullAlert_030_asset,
							Audio.VO.Midas.Henchman_Midas_FullAlert_040_asset,
							Audio.VO.Midas.Henchman_Midas_FullAlert_050_asset,
							Audio.VO.Midas.Henchman_Midas_FullAlert_060_asset,
							Audio.VO.Midas.Henchman_Midas_FullAlert_070_asset
						}
					}
					FNVOType.Ult =>
					{ return array{
							Audio.VO.Midas.VO_Midas_Ult_asset
						}
					}
				}
			}
			FNClasses.Daigo =>
			{
				case( VOType )
				{
					FNVOType.Idle => 
					{ return array{
							Audio.VO.Daigo.VO_CherrySmoke_MaskMaker_Idle_01_asset,
							Audio.VO.Daigo.VO_CherrySmoke_MaskMaker_Idle_02_asset,
							Audio.VO.Daigo.VO_CherrySmoke_MaskMaker_Idle_03_asset,
							Audio.VO.Daigo.VO_CherrySmoke_MaskMaker_Idle_04_asset,
							Audio.VO.Daigo.VO_CherrySmoke_MaskMaker_Idle_05_asset,
							Audio.VO.Daigo.VO_CherrySmoke_MaskMaker_Idle_06_asset,
							Audio.VO.Daigo.VO_CherrySmoke_MaskMaker_Idle_07_asset,
							Audio.VO.Daigo.VO_CherrySmoke_MaskMaker_Idle_08_asset,
							Audio.VO.Daigo.VO_CherrySmoke_MaskMaker_Idle_09_asset,
							Audio.VO.Daigo.VO_CherrySmoke_MaskMaker_Idle_10_asset,
							Audio.VO.Daigo.VO_CherrySmoke_MaskMaker_Idle_11_asset,
							Audio.VO.Daigo.VO_CherrySmoke_MaskMaker_Idle_12_asset
						}
					}
					FNVOType.Sus =>
					{ return array{
							Audio.VO.Daigo.VO_CherrySmoke_MaskMaker_Suspicious_01_asset,
							Audio.VO.Daigo.VO_CherrySmoke_MaskMaker_Suspicious_02_asset,
							Audio.VO.Daigo.VO_CherrySmoke_MaskMaker_Suspicious_03_asset,
							Audio.VO.Daigo.VO_CherrySmoke_MaskMaker_Suspicious_04_asset,
							<#Audio.VO.Daigo.VO_CherrySmoke_MaskMaker_Suspicious_05_asset,
							Audio.VO.Daigo.VO_CherrySmoke_MaskMaker_Suspicious_06_asset,
							Audio.VO.Daigo.VO_CherrySmoke_MaskMaker_Suspicious_07_asset,
							Audio.VO.Daigo.VO_CherrySmoke_MaskMaker_Suspicious_08_asset,
							Audio.VO.Daigo.VO_CherrySmoke_MaskMaker_Suspicious_09_asset,
							Audio.VO.Daigo.VO_CherrySmoke_MaskMaker_Suspicious_10_asset,
							Audio.VO.Daigo.VO_CherrySmoke_MaskMaker_Suspicious_11_asset,
							Audio.VO.Daigo.VO_CherrySmoke_MaskMaker_Suspicious_12_asset,
							Audio.VO.Daigo.VO_CherrySmoke_MaskMaker_Suspicious_13_asset,
							Audio.VO.Daigo.VO_CherrySmoke_MaskMaker_Suspicious_14_asset,
							Audio.VO.Daigo.VO_CherrySmoke_MaskMaker_Suspicious_15_asset,
							Audio.VO.Daigo.VO_CherrySmoke_MaskMaker_Suspicious_16_asset,
							Audio.VO.Daigo.VO_CherrySmoke_MaskMaker_Suspicious_17_asset,
							Audio.VO.Daigo.VO_CherrySmoke_MaskMaker_Suspicious_18_asset,
							Audio.VO.Daigo.VO_CherrySmoke_MaskMaker_Suspicious_19_asset,
							Audio.VO.Daigo.VO_CherrySmoke_MaskMaker_Suspicious_20_asset,
							Audio.VO.Daigo.VO_CherrySmoke_MaskMaker_Suspicious_21_asset,
							Audio.VO.Daigo.VO_CherrySmoke_MaskMaker_Suspicious_22_asset,
							Audio.VO.Daigo.VO_CherrySmoke_MaskMaker_Suspicious_23_asset,#>
							Audio.VO.Daigo.VO_CherrySmoke_MaskMaker_Suspicious_24_asset
						}
					}
					FNVOType.ReturnIdle =>
					{ return array{
							Audio.VO.Daigo.VO_CherrySmoke_MaskMaker_SeusReturnToIdle_01_asset,
							Audio.VO.Daigo.VO_CherrySmoke_MaskMaker_SeusReturnToIdle_02_asset,
							Audio.VO.Daigo.VO_CherrySmoke_MaskMaker_SeusReturnToIdle_03_asset,
							Audio.VO.Daigo.VO_CherrySmoke_MaskMaker_SeusReturnToIdle_04_asset,
							<#Audio.VO.Daigo.VO_CherrySmoke_MaskMaker_SeusReturnToIdle_05_asset,
							Audio.VO.Daigo.VO_CherrySmoke_MaskMaker_SeusReturnToIdle_06_asset,
							Audio.VO.Daigo.VO_CherrySmoke_MaskMaker_SeusReturnToIdle_07_asset,
							Audio.VO.Daigo.VO_CherrySmoke_MaskMaker_SeusReturnToIdle_08_asset,
							Audio.VO.Daigo.VO_CherrySmoke_MaskMaker_SeusReturnToIdle_09_asset,
							Audio.VO.Daigo.VO_CherrySmoke_MaskMaker_SeusReturnToIdle_10_asset,
							Audio.VO.Daigo.VO_CherrySmoke_MaskMaker_SeusReturnToIdle_11_asset,
							Audio.VO.Daigo.VO_CherrySmoke_MaskMaker_SeusReturnToIdle_12_asset,
							Audio.VO.Daigo.VO_CherrySmoke_MaskMaker_SeusReturnToIdle_13_asset,
							Audio.VO.Daigo.VO_CherrySmoke_MaskMaker_SeusReturnToIdle_14_asset,
							Audio.VO.Daigo.VO_CherrySmoke_MaskMaker_SeusReturnToIdle_15_asset,
							Audio.VO.Daigo.VO_CherrySmoke_MaskMaker_SeusReturnToIdle_16_asset,
							Audio.VO.Daigo.VO_CherrySmoke_MaskMaker_SeusReturnToIdle_17_asset,
							Audio.VO.Daigo.VO_CherrySmoke_MaskMaker_SeusReturnToIdle_18_asset,
							Audio.VO.Daigo.VO_CherrySmoke_MaskMaker_SeusReturnToIdle_19_asset,
							Audio.VO.Daigo.VO_CherrySmoke_MaskMaker_SeusReturnToIdle_20_asset,
							Audio.VO.Daigo.VO_CherrySmoke_MaskMaker_SeusReturnToIdle_21_asset,
							Audio.VO.Daigo.VO_CherrySmoke_MaskMaker_SeusReturnToIdle_22_asset,
							Audio.VO.Daigo.VO_CherrySmoke_MaskMaker_SeusReturnToIdle_23_asset,#>
							Audio.VO.Daigo.VO_CherrySmoke_MaskMaker_SeusReturnToIdle_24_asset
						}
					}
					FNVOType.FullAlert =>
					{ return array{
							Audio.VO.Daigo.VO_CherrySmoke_MaskMaker_FullAlert_01_asset,
							Audio.VO.Daigo.VO_CherrySmoke_MaskMaker_FullAlert_02_asset,
							Audio.VO.Daigo.VO_CherrySmoke_MaskMaker_FullAlert_03_asset,
							Audio.VO.Daigo.VO_CherrySmoke_MaskMaker_FullAlert_04_asset,
							Audio.VO.Daigo.VO_CherrySmoke_MaskMaker_FullAlert_05_asset,
							Audio.VO.Daigo.VO_CherrySmoke_MaskMaker_FullAlert_06_asset,
							Audio.VO.Daigo.VO_CherrySmoke_MaskMaker_FullAlert_07_asset,
							Audio.VO.Daigo.VO_CherrySmoke_MaskMaker_FullAlert_08_asset,
							Audio.VO.Daigo.VO_CherrySmoke_MaskMaker_FullAlert_09_asset,
							Audio.VO.Daigo.VO_CherrySmoke_MaskMaker_FullAlert_10_asset,
							Audio.VO.Daigo.VO_CherrySmoke_MaskMaker_FullAlert_11_asset,
							Audio.VO.Daigo.VO_CherrySmoke_MaskMaker_FullAlert_12_asset,
							Audio.VO.Daigo.VO_CherrySmoke_MaskMaker_FullAlert_13_asset,
							Audio.VO.Daigo.VO_CherrySmoke_MaskMaker_FullAlert_14_asset,
							Audio.VO.Daigo.VO_CherrySmoke_MaskMaker_FullAlert_15_asset,
							Audio.VO.Daigo.VO_CherrySmoke_MaskMaker_FullAlert_16_asset,
							Audio.VO.Daigo.VO_CherrySmoke_MaskMaker_FullAlert_17_asset,
							Audio.VO.Daigo.VO_CherrySmoke_MaskMaker_FullAlert_18_asset,
							Audio.VO.Daigo.VO_CherrySmoke_MaskMaker_FullAlert_19_asset,
							Audio.VO.Daigo.VO_CherrySmoke_MaskMaker_FullAlert_20_asset,
							Audio.VO.Daigo.VO_CherrySmoke_MaskMaker_FullAlert_21_asset,
							Audio.VO.Daigo.VO_CherrySmoke_MaskMaker_FullAlert_22_asset,
							Audio.VO.Daigo.VO_CherrySmoke_MaskMaker_FullAlert_23_asset,
							Audio.VO.Daigo.VO_CherrySmoke_MaskMaker_FullAlert_24_asset,
							Audio.VO.Daigo.VO_CherrySmoke_MaskMaker_FullAlert_25_asset,
							Audio.VO.Daigo.VO_CherrySmoke_MaskMaker_FullAlert_26_asset,
							Audio.VO.Daigo.VO_CherrySmoke_MaskMaker_FullAlert_27_asset,
							Audio.VO.Daigo.VO_CherrySmoke_MaskMaker_FullAlert_28_asset,
							Audio.VO.Daigo.VO_CherrySmoke_MaskMaker_FullAlert_29_asset,
							Audio.VO.Daigo.VO_CherrySmoke_MaskMaker_FullAlert_30_asset,
							Audio.VO.Daigo.VO_CherrySmoke_MaskMaker_FullAlert_31_asset,
							Audio.VO.Daigo.VO_CherrySmoke_MaskMaker_FullAlert_32_asset,
							Audio.VO.Daigo.VO_CherrySmoke_MaskMaker_FullAlert_33_asset,
							Audio.VO.Daigo.VO_CherrySmoke_MaskMaker_FullAlert_34_asset,
							Audio.VO.Daigo.VO_CherrySmoke_MaskMaker_FullAlert_35_asset,
							Audio.VO.Daigo.VO_CherrySmoke_MaskMaker_FullAlert_36_asset,
							Audio.VO.Daigo.VO_CherrySmoke_MaskMaker_FullAlert_37_asset
						}
					}
					FNVOType.FullAlertLost =>
					{ return array{
							Audio.VO.Daigo.VO_CherrySmoke_MaskMaker_AlertLostTarget_01_asset,
							Audio.VO.Daigo.VO_CherrySmoke_MaskMaker_AlertLostTarget_02_asset,
							Audio.VO.Daigo.VO_CherrySmoke_MaskMaker_AlertLostTarget_03_asset,
							Audio.VO.Daigo.VO_CherrySmoke_MaskMaker_AlertLostTarget_04_asset,
							<#Audio.VO.Daigo.VO_CherrySmoke_MaskMaker_AlertLostTarget_05_asset,
							Audio.VO.Daigo.VO_CherrySmoke_MaskMaker_AlertLostTarget_06_asset,
							Audio.VO.Daigo.VO_CherrySmoke_MaskMaker_AlertLostTarget_07_asset,
							Audio.VO.Daigo.VO_CherrySmoke_MaskMaker_AlertLostTarget_08_asset,
							Audio.VO.Daigo.VO_CherrySmoke_MaskMaker_AlertLostTarget_09_asset,
							Audio.VO.Daigo.VO_CherrySmoke_MaskMaker_AlertLostTarget_10_asset,
							Audio.VO.Daigo.VO_CherrySmoke_MaskMaker_AlertLostTarget_11_asset,
							Audio.VO.Daigo.VO_CherrySmoke_MaskMaker_AlertLostTarget_12_asset,
							Audio.VO.Daigo.VO_CherrySmoke_MaskMaker_AlertLostTarget_13_asset,
							Audio.VO.Daigo.VO_CherrySmoke_MaskMaker_AlertLostTarget_14_asset,
							Audio.VO.Daigo.VO_CherrySmoke_MaskMaker_AlertLostTarget_15_asset,
							Audio.VO.Daigo.VO_CherrySmoke_MaskMaker_AlertLostTarget_16_asset,
							Audio.VO.Daigo.VO_CherrySmoke_MaskMaker_AlertLostTarget_17_asset,
							Audio.VO.Daigo.VO_CherrySmoke_MaskMaker_AlertLostTarget_18_asset,
							Audio.VO.Daigo.VO_CherrySmoke_MaskMaker_AlertLostTarget_19_asset,
							Audio.VO.Daigo.VO_CherrySmoke_MaskMaker_AlertLostTarget_20_asset,
							Audio.VO.Daigo.VO_CherrySmoke_MaskMaker_AlertLostTarget_21_asset,
							Audio.VO.Daigo.VO_CherrySmoke_MaskMaker_AlertLostTarget_22_asset,
							Audio.VO.Daigo.VO_CherrySmoke_MaskMaker_AlertLostTarget_23_asset,
							Audio.VO.Daigo.VO_CherrySmoke_MaskMaker_AlertLostTarget_24_asset,
							Audio.VO.Daigo.VO_CherrySmoke_MaskMaker_AlertLostTarget_25_asset,
							Audio.VO.Daigo.VO_CherrySmoke_MaskMaker_AlertLostTarget_26_asset,
							Audio.VO.Daigo.VO_CherrySmoke_MaskMaker_AlertLostTarget_27_asset,
							Audio.VO.Daigo.VO_CherrySmoke_MaskMaker_AlertLostTarget_28_asset,
							Audio.VO.Daigo.VO_CherrySmoke_MaskMaker_AlertLostTarget_29_asset,
							Audio.VO.Daigo.VO_CherrySmoke_MaskMaker_AlertLostTarget_30_asset,
							Audio.VO.Daigo.VO_CherrySmoke_MaskMaker_AlertLostTarget_31_asset,#>
							Audio.VO.Daigo.VO_CherrySmoke_MaskMaker_AlertLostTarget_32_asset
						}
					}
					FNVOType.DBNO =>
					{ return array{
							Audio.VO.Daigo.VO_CherrySmoke_MaskMaker_Elim_01_asset,
							Audio.VO.Daigo.VO_CherrySmoke_MaskMaker_Elim_02_asset,
							Audio.VO.Daigo.VO_CherrySmoke_MaskMaker_Elim_03_asset,
							Audio.VO.Daigo.VO_CherrySmoke_MaskMaker_Elim_04_asset,
							Audio.VO.Daigo.VO_CherrySmoke_MaskMaker_Elim_05_asset,
							Audio.VO.Daigo.VO_CherrySmoke_MaskMaker_Elim_06_asset,
							Audio.VO.Daigo.VO_CherrySmoke_MaskMaker_Elim_07_asset,
							Audio.VO.Daigo.VO_CherrySmoke_MaskMaker_Elim_08_asset,
							Audio.VO.Daigo.VO_CherrySmoke_MaskMaker_Elim_09_asset,
							Audio.VO.Daigo.VO_CherrySmoke_MaskMaker_Elim_10_asset,
							Audio.VO.Daigo.VO_CherrySmoke_MaskMaker_Elim_11_asset,
							Audio.VO.Daigo.VO_CherrySmoke_MaskMaker_Elim_12_asset,
							Audio.VO.Daigo.VO_CherrySmoke_MaskMaker_Elim_13_asset,
							Audio.VO.Daigo.VO_CherrySmoke_MaskMaker_Elim_14_asset,
							Audio.VO.Daigo.VO_CherrySmoke_MaskMaker_Elim_15_asset,
							Audio.VO.Daigo.VO_CherrySmoke_MaskMaker_Elim_16_asset,
							Audio.VO.Daigo.VO_CherrySmoke_MaskMaker_Elim_17_asset,
							Audio.VO.Daigo.VO_CherrySmoke_MaskMaker_Elim_18_asset,
							Audio.VO.Daigo.VO_CherrySmoke_MaskMaker_Elim_19_asset,
							Audio.VO.Daigo.VO_CherrySmoke_MaskMaker_Elim_20_asset
						}
					}
					FNVOType.Interrogation =>
					{ return array{
							Audio.VO.Daigo.VO_CherrySmoke_MaskMaker_TakeDmgSml_01_asset,
							Audio.VO.Daigo.VO_CherrySmoke_MaskMaker_TakeDmgSml_02_asset,
							Audio.VO.Daigo.VO_CherrySmoke_MaskMaker_TakeDmgSml_03_asset,
							Audio.VO.Daigo.VO_CherrySmoke_MaskMaker_TakeDmgSml_04_asset,
							Audio.VO.Daigo.VO_CherrySmoke_MaskMaker_TakeDmgSml_05_asset,
							Audio.VO.Daigo.VO_CherrySmoke_MaskMaker_TakeDmgSml_06_asset,
							Audio.VO.Daigo.VO_CherrySmoke_MaskMaker_TakeDmgSml_07_asset,
							Audio.VO.Daigo.VO_CherrySmoke_MaskMaker_TakeDmgSml_08_asset
						}
					}
					FNVOType.Ability =>	
					{ return array{
							Audio.VO.Daigo.VO_CherrySmoke_MaskMaker_FocusedBeam_01_asset,
							Audio.VO.Daigo.VO_CherrySmoke_MaskMaker_FocusedBeam_02_asset,
							Audio.VO.Daigo.VO_CherrySmoke_MaskMaker_FocusedBeam_03_asset,
							Audio.VO.Daigo.VO_CherrySmoke_MaskMaker_FocusedBeam_04_asset,
							Audio.VO.Daigo.VO_CherrySmoke_MaskMaker_FocusedBeam_05_asset,
							Audio.VO.Daigo.VO_CherrySmoke_MaskMaker_FocusedBeam_06_asset,
							Audio.VO.Daigo.VO_CherrySmoke_MaskMaker_FocusedBeam_07_asset,
							Audio.VO.Daigo.VO_CherrySmoke_MaskMaker_FocusedBeam_08_asset,
							Audio.VO.Daigo.VO_CherrySmoke_MaskMaker_FocusedBeam_09_asset,
							Audio.VO.Daigo.VO_CherrySmoke_MaskMaker_FocusedBeam_10_asset,
							Audio.VO.Daigo.VO_CherrySmoke_MaskMaker_FocusedBeam_11_asset,
							Audio.VO.Daigo.VO_CherrySmoke_MaskMaker_FocusedBeam_12_asset,
							Audio.VO.Daigo.VO_CherrySmoke_MaskMaker_FocusedBeam_13_asset,
							Audio.VO.Daigo.VO_CherrySmoke_MaskMaker_FocusedBeam_14_asset,
							Audio.VO.Daigo.VO_CherrySmoke_MaskMaker_FocusedBeam_15_asset,
							Audio.VO.Daigo.VO_CherrySmoke_MaskMaker_FocusedBeam_16_asset,
							Audio.VO.Daigo.VO_CherrySmoke_MaskMaker_FocusedBeam_17_asset,
							Audio.VO.Daigo.VO_CherrySmoke_MaskMaker_FocusedBeam_18_asset,
							Audio.VO.Daigo.VO_CherrySmoke_MaskMaker_FocusedBeam_19_asset,
							Audio.VO.Daigo.VO_CherrySmoke_MaskMaker_FocusedBeam_20_asset
						}
					}
					FNVOType.AbilityEnd =>
					{ return array{
							Audio.VO.Daigo.VO_CherrySmoke_MaskMaker_TeleportAoE_01_asset,
							Audio.VO.Daigo.VO_CherrySmoke_MaskMaker_TeleportAoE_02_asset,
							Audio.VO.Daigo.VO_CherrySmoke_MaskMaker_TeleportAoE_03_asset,
							Audio.VO.Daigo.VO_CherrySmoke_MaskMaker_TeleportAoE_04_asset,
							Audio.VO.Daigo.VO_CherrySmoke_MaskMaker_TeleportAoE_05_asset,
							Audio.VO.Daigo.VO_CherrySmoke_MaskMaker_TeleportAoE_06_asset,
							Audio.VO.Daigo.VO_CherrySmoke_MaskMaker_TeleportAoE_07_asset,
							Audio.VO.Daigo.VO_CherrySmoke_MaskMaker_TeleportAoE_08_asset,
							Audio.VO.Daigo.VO_CherrySmoke_MaskMaker_TeleportAoE_09_asset,
							Audio.VO.Daigo.VO_CherrySmoke_MaskMaker_TeleportAoE_10_asset,
							Audio.VO.Daigo.VO_CherrySmoke_MaskMaker_TeleportAoE_11_asset,
							Audio.VO.Daigo.VO_CherrySmoke_MaskMaker_TeleportAoE_12_asset,
							Audio.VO.Daigo.VO_CherrySmoke_MaskMaker_TeleportAoE_13_asset,
							Audio.VO.Daigo.VO_CherrySmoke_MaskMaker_TeleportAoE_14_asset,
							Audio.VO.Daigo.VO_CherrySmoke_MaskMaker_TeleportAoE_15_asset,
							Audio.VO.Daigo.VO_CherrySmoke_MaskMaker_TeleportAoE_16_asset,
							Audio.VO.Daigo.VO_CherrySmoke_MaskMaker_TeleportAoE_17_asset,
							Audio.VO.Daigo.VO_CherrySmoke_MaskMaker_TeleportAoE_18_asset,
							Audio.VO.Daigo.VO_CherrySmoke_MaskMaker_TeleportAoE_19_asset,
							Audio.VO.Daigo.VO_CherrySmoke_MaskMaker_TeleportAoE_20_asset,
							Audio.VO.Daigo.VO_CherrySmoke_MaskMaker_TeleportAoE_21_asset,
							Audio.VO.Daigo.VO_CherrySmoke_MaskMaker_TeleportAoE_22_asset,
							Audio.VO.Daigo.VO_CherrySmoke_MaskMaker_TeleportAoE_23_asset,
							Audio.VO.Daigo.VO_CherrySmoke_MaskMaker_TeleportAoE_24_asset
						}
					}
					FNVOType.Ult =>
					{ return array{
							Audio.VO.Daigo.VO_CherrySmoke_MaskMaker_SpireShardSumon_01_asset,
							Audio.VO.Daigo.VO_CherrySmoke_MaskMaker_SpireShardSumon_02_asset,
							Audio.VO.Daigo.VO_CherrySmoke_MaskMaker_SpireShardSumon_03_asset,
							Audio.VO.Daigo.VO_CherrySmoke_MaskMaker_SpireShardSumon_04_asset,
							Audio.VO.Daigo.VO_CherrySmoke_MaskMaker_SpireShardSumon_05_asset,
							Audio.VO.Daigo.VO_CherrySmoke_MaskMaker_SpireShardSumon_06_asset,
							Audio.VO.Daigo.VO_CherrySmoke_MaskMaker_SpireShardSumon_07_asset,
							Audio.VO.Daigo.VO_CherrySmoke_MaskMaker_SpireShardSumon_08_asset,
							Audio.VO.Daigo.VO_CherrySmoke_MaskMaker_SpireShardSumon_09_asset,
							Audio.VO.Daigo.VO_CherrySmoke_MaskMaker_SpireShardSumon_10_asset,
							Audio.VO.Daigo.VO_CherrySmoke_MaskMaker_SpireShardSumon_11_asset,
							Audio.VO.Daigo.VO_CherrySmoke_MaskMaker_SpireShardSumon_12_asset,
							Audio.VO.Daigo.VO_CherrySmoke_MaskMaker_SpireShardSumon_13_asset,
							Audio.VO.Daigo.VO_CherrySmoke_MaskMaker_SpireShardSumon_14_asset,
							Audio.VO.Daigo.VO_CherrySmoke_MaskMaker_SpireShardSumon_15_asset,
							Audio.VO.Daigo.VO_CherrySmoke_MaskMaker_SpireShardSumon_16_asset,
							Audio.VO.Daigo.VO_CherrySmoke_MaskMaker_SpireShardSumon_17_asset,
							Audio.VO.Daigo.VO_CherrySmoke_MaskMaker_SpireShardSumon_18_asset,
							Audio.VO.Daigo.VO_CherrySmoke_MaskMaker_SpireShardSumon_19_asset,
							Audio.VO.Daigo.VO_CherrySmoke_MaskMaker_SpireShardSumon_20_asset,
							Audio.VO.Daigo.VO_CherrySmoke_MaskMaker_SpireShardSumon_21_asset,
							Audio.VO.Daigo.VO_CherrySmoke_MaskMaker_SpireShardSumon_22_asset,
							Audio.VO.Daigo.VO_CherrySmoke_MaskMaker_SpireShardSumon_23_asset,
							Audio.VO.Daigo.VO_CherrySmoke_MaskMaker_SpireShardSumon_24_asset,
							Audio.VO.Daigo.VO_CherrySmoke_MaskMaker_SpireShardSumon_25_asset,
							Audio.VO.Daigo.VO_CherrySmoke_MaskMaker_SpireShardSumon_26_asset,
							Audio.VO.Daigo.VO_CherrySmoke_MaskMaker_SpireShardSumon_27_asset,
							Audio.VO.Daigo.VO_CherrySmoke_MaskMaker_SpireShardSumon_28_asset
						}
					}
				}
			}
			FNClasses.Kor =>
			{
				case( VOType )
				{
					FNVOType.Idle => 
					{ return array{
							Audio.VO.Kor.VO_CherrySmoke_Kor_Idle_01_asset,
							Audio.VO.Kor.VO_CherrySmoke_Kor_Idle_02_asset,
							Audio.VO.Kor.VO_CherrySmoke_Kor_Idle_03_asset,
							Audio.VO.Kor.VO_CherrySmoke_Kor_Idle_04_asset,
							Audio.VO.Kor.VO_CherrySmoke_Kor_Idle_05_asset,
							<#Audio.VO.Kor.VO_CherrySmoke_Kor_Idle_06_asset,
							Audio.VO.Kor.VO_CherrySmoke_Kor_Idle_07_asset,
							Audio.VO.Kor.VO_CherrySmoke_Kor_Idle_08_asset,
							Audio.VO.Kor.VO_CherrySmoke_Kor_Idle_09_asset,
							Audio.VO.Kor.VO_CherrySmoke_Kor_Idle_10_asset,
							Audio.VO.Kor.VO_CherrySmoke_Kor_Idle_11_asset,
							Audio.VO.Kor.VO_CherrySmoke_Kor_Idle_12_asset,
							Audio.VO.Kor.VO_CherrySmoke_Kor_Idle_13_asset,
							Audio.VO.Kor.VO_CherrySmoke_Kor_Idle_14_asset,
							Audio.VO.Kor.VO_CherrySmoke_Kor_Idle_15_asset,
							Audio.VO.Kor.VO_CherrySmoke_Kor_Idle_16_asset,
							Audio.VO.Kor.VO_CherrySmoke_Kor_Idle_17_asset,
							Audio.VO.Kor.VO_CherrySmoke_Kor_Idle_18_asset,
							Audio.VO.Kor.VO_CherrySmoke_Kor_Idle_19_asset,#>
							Audio.VO.Kor.VO_CherrySmoke_Kor_Idle_20_asset
						}
					}
					FNVOType.Sus =>
					{ return array{
							Audio.VO.Kor.VO_CherrySmoke_Kor_Suspicious_01_asset,
							Audio.VO.Kor.VO_CherrySmoke_Kor_Suspicious_02_asset,
							Audio.VO.Kor.VO_CherrySmoke_Kor_Suspicious_03_asset,
							Audio.VO.Kor.VO_CherrySmoke_Kor_Suspicious_04_asset,
							Audio.VO.Kor.VO_CherrySmoke_Kor_Suspicious_05_asset,
							<#Audio.VO.Kor.VO_CherrySmoke_Kor_Suspicious_06_asset,
							Audio.VO.Kor.VO_CherrySmoke_Kor_Suspicious_07_asset,
							Audio.VO.Kor.VO_CherrySmoke_Kor_Suspicious_08_asset,
							Audio.VO.Kor.VO_CherrySmoke_Kor_Suspicious_09_asset,
							Audio.VO.Kor.VO_CherrySmoke_Kor_Suspicious_10_asset,
							Audio.VO.Kor.VO_CherrySmoke_Kor_Suspicious_11_asset,
							Audio.VO.Kor.VO_CherrySmoke_Kor_Suspicious_12_asset,
							Audio.VO.Kor.VO_CherrySmoke_Kor_Suspicious_13_asset,
							Audio.VO.Kor.VO_CherrySmoke_Kor_Suspicious_14_asset,
							Audio.VO.Kor.VO_CherrySmoke_Kor_Suspicious_15_asset,
							Audio.VO.Kor.VO_CherrySmoke_Kor_Suspicious_16_asset,
							Audio.VO.Kor.VO_CherrySmoke_Kor_Suspicious_17_asset,
							Audio.VO.Kor.VO_CherrySmoke_Kor_Suspicious_18_asset,
							Audio.VO.Kor.VO_CherrySmoke_Kor_Suspicious_19_asset,
							Audio.VO.Kor.VO_CherrySmoke_Kor_Suspicious_20_asset,#>
							Audio.VO.Kor.VO_CherrySmoke_Kor_Suspicious_21_asset
						}
					}
					FNVOType.ReturnIdle =>
					{ return array{
							Audio.VO.Kor.VO_CherrySmoke_Kor_SusReturnToIdle_01_asset,
							Audio.VO.Kor.VO_CherrySmoke_Kor_SusReturnToIdle_02_asset,
							Audio.VO.Kor.VO_CherrySmoke_Kor_SusReturnToIdle_03_asset,
							Audio.VO.Kor.VO_CherrySmoke_Kor_SusReturnToIdle_04_asset,
							<#Audio.VO.Kor.VO_CherrySmoke_Kor_SusReturnToIdle_05_asset,
							Audio.VO.Kor.VO_CherrySmoke_Kor_SusReturnToIdle_06_asset,
							Audio.VO.Kor.VO_CherrySmoke_Kor_SusReturnToIdle_07_asset,
							Audio.VO.Kor.VO_CherrySmoke_Kor_SusReturnToIdle_08_asset,
							Audio.VO.Kor.VO_CherrySmoke_Kor_SusReturnToIdle_09_asset,
							Audio.VO.Kor.VO_CherrySmoke_Kor_SusReturnToIdle_10_asset,
							Audio.VO.Kor.VO_CherrySmoke_Kor_SusReturnToIdle_11_asset,
							Audio.VO.Kor.VO_CherrySmoke_Kor_SusReturnToIdle_12_asset,
							Audio.VO.Kor.VO_CherrySmoke_Kor_SusReturnToIdle_13_asset,
							Audio.VO.Kor.VO_CherrySmoke_Kor_SusReturnToIdle_14_asset,
							Audio.VO.Kor.VO_CherrySmoke_Kor_SusReturnToIdle_15_asset,
							Audio.VO.Kor.VO_CherrySmoke_Kor_SusReturnToIdle_16_asset,#>
							Audio.VO.Kor.VO_CherrySmoke_Kor_SusReturnToIdle_17_asset
						}
					}
					FNVOType.FullAlert =>
					{ return array{
							Audio.VO.Kor.VO_CherrySmoke_Kor_FullAlert_01_asset,
							Audio.VO.Kor.VO_CherrySmoke_Kor_FullAlert_02_asset,
							Audio.VO.Kor.VO_CherrySmoke_Kor_FullAlert_03_asset,
							Audio.VO.Kor.VO_CherrySmoke_Kor_FullAlert_04_asset,
							<#Audio.VO.Kor.VO_CherrySmoke_Kor_FullAlert_05_asset,
							Audio.VO.Kor.VO_CherrySmoke_Kor_FullAlert_06_asset,
							Audio.VO.Kor.VO_CherrySmoke_Kor_FullAlert_07_asset,
							Audio.VO.Kor.VO_CherrySmoke_Kor_FullAlert_08_asset,
							Audio.VO.Kor.VO_CherrySmoke_Kor_FullAlert_09_asset,
							Audio.VO.Kor.VO_CherrySmoke_Kor_FullAlert_10_asset,
							Audio.VO.Kor.VO_CherrySmoke_Kor_FullAlert_11_asset,
							Audio.VO.Kor.VO_CherrySmoke_Kor_FullAlert_12_asset,
							Audio.VO.Kor.VO_CherrySmoke_Kor_FullAlert_13_asset,
							Audio.VO.Kor.VO_CherrySmoke_Kor_FullAlert_14_asset,
							Audio.VO.Kor.VO_CherrySmoke_Kor_FullAlert_15_asset,
							Audio.VO.Kor.VO_CherrySmoke_Kor_FullAlert_16_asset,#>
							Audio.VO.Kor.VO_CherrySmoke_Kor_FullAlert_17_asset
						}
					}
					FNVOType.FullAlertLost =>
					{ return array{
							Audio.VO.Kor.VO_CherrySmoke_Kor_AlertLostTarget_01_asset,
							Audio.VO.Kor.VO_CherrySmoke_Kor_AlertLostTarget_02_asset,
							Audio.VO.Kor.VO_CherrySmoke_Kor_AlertLostTarget_03_asset,
							Audio.VO.Kor.VO_CherrySmoke_Kor_AlertLostTarget_04_asset,
							<#Audio.VO.Kor.VO_CherrySmoke_Kor_AlertLostTarget_05_asset,
							Audio.VO.Kor.VO_CherrySmoke_Kor_AlertLostTarget_06_asset,
							Audio.VO.Kor.VO_CherrySmoke_Kor_AlertLostTarget_07_asset,
							Audio.VO.Kor.VO_CherrySmoke_Kor_AlertLostTarget_08_asset,
							Audio.VO.Kor.VO_CherrySmoke_Kor_AlertLostTarget_09_asset,
							Audio.VO.Kor.VO_CherrySmoke_Kor_AlertLostTarget_10_asset,
							Audio.VO.Kor.VO_CherrySmoke_Kor_AlertLostTarget_11_asset,
							Audio.VO.Kor.VO_CherrySmoke_Kor_AlertLostTarget_12_asset,
							Audio.VO.Kor.VO_CherrySmoke_Kor_AlertLostTarget_13_asset,
							Audio.VO.Kor.VO_CherrySmoke_Kor_AlertLostTarget_14_asset,
							Audio.VO.Kor.VO_CherrySmoke_Kor_AlertLostTarget_15_asset,
							Audio.VO.Kor.VO_CherrySmoke_Kor_AlertLostTarget_16_asset,
							Audio.VO.Kor.VO_CherrySmoke_Kor_AlertLostTarget_17_asset,
							Audio.VO.Kor.VO_CherrySmoke_Kor_AlertLostTarget_18_asset,
							Audio.VO.Kor.VO_CherrySmoke_Kor_AlertLostTarget_19_asset,
							Audio.VO.Kor.VO_CherrySmoke_Kor_AlertLostTarget_20_asset,
							Audio.VO.Kor.VO_CherrySmoke_Kor_AlertLostTarget_21_asset,
							Audio.VO.Kor.VO_CherrySmoke_Kor_AlertLostTarget_22_asset,
							Audio.VO.Kor.VO_CherrySmoke_Kor_AlertLostTarget_23_asset,
							Audio.VO.Kor.VO_CherrySmoke_Kor_AlertLostTarget_24_asset,
							Audio.VO.Kor.VO_CherrySmoke_Kor_AlertLostTarget_25_asset,#>
							Audio.VO.Kor.VO_CherrySmoke_Kor_AlertLostTarget_26_asset
						}
					}
					FNVOType.DBNO =>
					{ return array{
							Audio.VO.Kor.VO_CherrySmoke_Kor_Elim_01_asset,
							Audio.VO.Kor.VO_CherrySmoke_Kor_Elim_02_asset,
							Audio.VO.Kor.VO_CherrySmoke_Kor_Elim_03_asset,
							Audio.VO.Kor.VO_CherrySmoke_Kor_Elim_04_asset,
							Audio.VO.Kor.VO_CherrySmoke_Kor_Elim_05_asset,
							<#Audio.VO.Kor.VO_CherrySmoke_Kor_Elim_06_asset,
							Audio.VO.Kor.VO_CherrySmoke_Kor_Elim_07_asset,
							Audio.VO.Kor.VO_CherrySmoke_Kor_Elim_08_asset,
							Audio.VO.Kor.VO_CherrySmoke_Kor_Elim_09_asset,
							Audio.VO.Kor.VO_CherrySmoke_Kor_Elim_10_asset,
							Audio.VO.Kor.VO_CherrySmoke_Kor_Elim_11_asset,
							Audio.VO.Kor.VO_CherrySmoke_Kor_Elim_12_asset,
							Audio.VO.Kor.VO_CherrySmoke_Kor_Elim_13_asset,
							Audio.VO.Kor.VO_CherrySmoke_Kor_Elim_14_asset,
							Audio.VO.Kor.VO_CherrySmoke_Kor_Elim_15_asset,
							Audio.VO.Kor.VO_CherrySmoke_Kor_Elim_16_asset,
							Audio.VO.Kor.VO_CherrySmoke_Kor_Elim_17_asset,
							Audio.VO.Kor.VO_CherrySmoke_Kor_Elim_18_asset,
							Audio.VO.Kor.VO_CherrySmoke_Kor_Elim_19_asset,
							Audio.VO.Kor.VO_CherrySmoke_Kor_Elim_20_asset,
							Audio.VO.Kor.VO_CherrySmoke_Kor_Elim_21_asset,#>
							Audio.VO.Kor.VO_CherrySmoke_Kor_Elim_22_asset
						}
					}
					FNVOType.Interrogation =>
					{ return array{
							Audio.VO.Kor.VO_CherrySmoke_Kor_HalfHealth_01_asset,
							Audio.VO.Kor.VO_CherrySmoke_Kor_HalfHealth_02_asset,
							Audio.VO.Kor.VO_CherrySmoke_Kor_HalfHealth_03_asset,
							Audio.VO.Kor.VO_CherrySmoke_Kor_HalfHealth_04_asset,
							<#Audio.VO.Kor.VO_CherrySmoke_Kor_HalfHealth_05_asset,
							Audio.VO.Kor.VO_CherrySmoke_Kor_HalfHealth_06_asset,
							Audio.VO.Kor.VO_CherrySmoke_Kor_HalfHealth_07_asset,
							Audio.VO.Kor.VO_CherrySmoke_Kor_HalfHealth_08_asset,
							Audio.VO.Kor.VO_CherrySmoke_Kor_HalfHealth_09_asset,
							Audio.VO.Kor.VO_CherrySmoke_Kor_HalfHealth_10_asset,
							Audio.VO.Kor.VO_CherrySmoke_Kor_HalfHealth_11_asset,
							Audio.VO.Kor.VO_CherrySmoke_Kor_HalfHealth_12_asset,
							Audio.VO.Kor.VO_CherrySmoke_Kor_HalfHealth_13_asset,
							Audio.VO.Kor.VO_CherrySmoke_Kor_HalfHealth_14_asset,
							Audio.VO.Kor.VO_CherrySmoke_Kor_HalfHealth_15_asset,
							Audio.VO.Kor.VO_CherrySmoke_Kor_HalfHealth_16_asset,
							Audio.VO.Kor.VO_CherrySmoke_Kor_HalfHealth_17_asset,
							Audio.VO.Kor.VO_CherrySmoke_Kor_HalfHealth_18_asset,
							Audio.VO.Kor.VO_CherrySmoke_Kor_HalfHealth_19_asset,
							Audio.VO.Kor.VO_CherrySmoke_Kor_HalfHealth_20_asset,
							Audio.VO.Kor.VO_CherrySmoke_Kor_HalfHealth_21_asset,#>
							Audio.VO.Kor.VO_CherrySmoke_Kor_HalfHealth_22_asset
						}
					}
					FNVOType.Ability =>	
					{ return array{
							Audio.VO.Kor.VO_CherrySmoke_Kor_SummonHench_01_asset,
							Audio.VO.Kor.VO_CherrySmoke_Kor_SummonHench_02_asset,
							Audio.VO.Kor.VO_CherrySmoke_Kor_SummonHench_03_asset,
							Audio.VO.Kor.VO_CherrySmoke_Kor_SummonHench_04_asset,
							<#Audio.VO.Kor.VO_CherrySmoke_Kor_SummonHench_05_asset,
							Audio.VO.Kor.VO_CherrySmoke_Kor_SummonHench_06_asset,
							Audio.VO.Kor.VO_CherrySmoke_Kor_SummonHench_07_asset,
							Audio.VO.Kor.VO_CherrySmoke_Kor_SummonHench_08_asset,
							Audio.VO.Kor.VO_CherrySmoke_Kor_SummonHench_09_asset,
							Audio.VO.Kor.VO_CherrySmoke_Kor_SummonHench_10_asset,
							Audio.VO.Kor.VO_CherrySmoke_Kor_SummonHench_11_asset,
							Audio.VO.Kor.VO_CherrySmoke_Kor_SummonHench_12_asset,
							Audio.VO.Kor.VO_CherrySmoke_Kor_SummonHench_13_asset,
							Audio.VO.Kor.VO_CherrySmoke_Kor_SummonHench_14_asset,
							Audio.VO.Kor.VO_CherrySmoke_Kor_SummonHench_15_asset,#>
							Audio.VO.Kor.VO_CherrySmoke_Kor_SummonHench_16_asset
						}				
					}
					FNVOType.AbilityEnd =>
					{ return array{
							Audio.VO.Kor.VO_CherrySmoke_Kor_LongRangeTargeting_01_asset,
							Audio.VO.Kor.VO_CherrySmoke_Kor_LongRangeTargeting_02_asset,
							Audio.VO.Kor.VO_CherrySmoke_Kor_LongRangeTargeting_03_asset,
							Audio.VO.Kor.VO_CherrySmoke_Kor_LongRangeTargeting_04_asset,
							<#Audio.VO.Kor.VO_CherrySmoke_Kor_LongRangeTargeting_05_asset,
							Audio.VO.Kor.VO_CherrySmoke_Kor_LongRangeTargeting_06_asset,
							Audio.VO.Kor.VO_CherrySmoke_Kor_LongRangeTargeting_07_asset,
							Audio.VO.Kor.VO_CherrySmoke_Kor_LongRangeTargeting_08_asset,
							Audio.VO.Kor.VO_CherrySmoke_Kor_LongRangeTargeting_09_asset,
							Audio.VO.Kor.VO_CherrySmoke_Kor_LongRangeTargeting_10_asset,
							Audio.VO.Kor.VO_CherrySmoke_Kor_LongRangeTargeting_11_asset,
							Audio.VO.Kor.VO_CherrySmoke_Kor_LongRangeTargeting_12_asset,
							Audio.VO.Kor.VO_CherrySmoke_Kor_LongRangeTargeting_13_asset,
							Audio.VO.Kor.VO_CherrySmoke_Kor_LongRangeTargeting_14_asset,
							Audio.VO.Kor.VO_CherrySmoke_Kor_LongRangeTargeting_15_asset,
							Audio.VO.Kor.VO_CherrySmoke_Kor_LongRangeTargeting_16_asset,
							Audio.VO.Kor.VO_CherrySmoke_Kor_LongRangeTargeting_17_asset,
							Audio.VO.Kor.VO_CherrySmoke_Kor_LongRangeTargeting_18_asset,
							Audio.VO.Kor.VO_CherrySmoke_Kor_LongRangeTargeting_19_asset,
							Audio.VO.Kor.VO_CherrySmoke_Kor_LongRangeTargeting_20_asset,
							Audio.VO.Kor.VO_CherrySmoke_Kor_LongRangeTargeting_21_asset,
							Audio.VO.Kor.VO_CherrySmoke_Kor_LongRangeTargeting_22_asset,
							Audio.VO.Kor.VO_CherrySmoke_Kor_LongRangeTargeting_23_asset,
							Audio.VO.Kor.VO_CherrySmoke_Kor_LongRangeTargeting_24_asset,
							Audio.VO.Kor.VO_CherrySmoke_Kor_LongRangeTargeting_25_asset,
							Audio.VO.Kor.VO_CherrySmoke_Kor_LongRangeTargeting_26_asset,
							Audio.VO.Kor.VO_CherrySmoke_Kor_LongRangeTargeting_27_asset,
							Audio.VO.Kor.VO_CherrySmoke_Kor_LongRangeTargeting_28_asset,
							Audio.VO.Kor.VO_CherrySmoke_Kor_LongRangeTargeting_29_asset,#>
							Audio.VO.Kor.VO_CherrySmoke_Kor_LongRangeTargeting_30_asset
						}
					}
					FNVOType.Ult =>
					{ return array{
							Audio.VO.Kor.VO_CherrySmoke_Kor_FullAlert_01_asset,
							Audio.VO.Kor.VO_CherrySmoke_Kor_FullAlert_02_asset,
							Audio.VO.Kor.VO_CherrySmoke_Kor_FullAlert_03_asset,
							Audio.VO.Kor.VO_CherrySmoke_Kor_FullAlert_04_asset,
							<#Audio.VO.Kor.VO_CherrySmoke_Kor_FullAlert_05_asset,
							Audio.VO.Kor.VO_CherrySmoke_Kor_FullAlert_06_asset,
							Audio.VO.Kor.VO_CherrySmoke_Kor_FullAlert_07_asset,
							Audio.VO.Kor.VO_CherrySmoke_Kor_FullAlert_08_asset,
							Audio.VO.Kor.VO_CherrySmoke_Kor_FullAlert_09_asset,
							Audio.VO.Kor.VO_CherrySmoke_Kor_FullAlert_10_asset,
							Audio.VO.Kor.VO_CherrySmoke_Kor_FullAlert_11_asset,
							Audio.VO.Kor.VO_CherrySmoke_Kor_FullAlert_12_asset,
							Audio.VO.Kor.VO_CherrySmoke_Kor_FullAlert_13_asset,
							Audio.VO.Kor.VO_CherrySmoke_Kor_FullAlert_14_asset,
							Audio.VO.Kor.VO_CherrySmoke_Kor_FullAlert_15_asset,
							Audio.VO.Kor.VO_CherrySmoke_Kor_FullAlert_16_asset,#>
							Audio.VO.Kor.VO_CherrySmoke_Kor_FullAlert_17_asset
						}
					}
				}
			}
			_ =>
			{
				return array{}
			}
		}

		#return array{}
	}

    PlayClassVO(Agent:agent, VOType:FNVOType):void=
	{
		CurrentTime := GetSimulationElapsedTime()
		if ( AbilityUsedTime := PlayersMap[Agent].flVOUsedTime, ElapsedTime := CurrentTime - PlayersMap[Agent].flVOUsedTime )
		{
			if ( VOType = FNVOType.Ult )
			{
				Agent.PlayVO(GetClassVO(GetPlayerClassIndex(Agent), VOType) ) 
			}
			else
			{
				if (ElapsedTime < 2.0)
				{
					#Print("VO on cooldown... {ElapsedTime}")
					#SFXAbilityCooldown.Play(Agent)
					return
				}

				NewCurrentTime := GetSimulationElapsedTime()
				if {set PlayersMap[Agent].flVOUsedTime = NewCurrentTime}
				Agent.PlayVO(GetClassVO(GetPlayerClassIndex(Agent), VOType) ) 
			}
		}
	}

    (Agent:agent).PlayVO(SoundArrayAsset:[]sound_wave, ?Duration:float = 10.0):void=
	{
		spawn{RemoveVO(Agent)}
		spawn{SpawnVO(Agent,SoundArrayAsset,Duration)}
	}

	SpawnVO<public>(Agent:agent,SoundArrayAsset:[]sound_wave, Duration:float)<suspends>:void=
	{
		if ( SimulationEntity := GetSimulationEntity[], FortCharacter := Agent.GetFortCharacter[] )
		{
			for ( EntPlayer : SimulationEntity.FindDescendantEntitiesWithComponent(fn_player_component), PlayerComp := EntPlayer.GetComponent[fn_player_component], PlayerComp.Agent = Agent )
			{
				SpawnTransform := FromTransform(FortCharacter.GetTransform())
				NewEnt := entity{}
				TransComp := transform_component{ Entity := NewEnt }
				SFXComp := Audio.MSS_Oneshot{ Entity := NewEnt, Sounds := SoundArrayAsset, VolumeBase := 0.0, PitchBase := 0.0, Enabled := true, AutoPlay := true }
				NewEnt.AddComponents(array{TransComp,SFXComp})
				EntPlayer.AddEntities(array{NewEnt})
				NewEnt.SetGlobalTransform(SpawnTransform)
				if {set PlayersMap[Agent].ActiveVO = option{NewEnt}}
				SFXComp.Play()
				#Sleep(Duration)
				#RemoveVO(Agent)
			}
		}
	}

	RemoveVO(Agent:agent)<suspends>:void=
	{
		if ( EntVO := PlayersMap[Agent].ActiveVO? )
		{
			EntVO.RemoveFromParent()
			if {set PlayersMap[Agent].ActiveVO = false}
		}
	}

	OnPlayerSpawned(Agent:agent):void=
	{
		Print("OnPlayerSpawned()")
		spawn{PlayerSpawn(Agent)}
	}

	AssignPlayerTeam(Agent:agent):void=
	{
		TeamCollection := GetPlayspace().GetTeamCollection()
		Teams := TeamCollection.GetTeams()
		if ( BlueCount := TeamCollection.GetAgents[Teams[0]].Length, RedCount := TeamCollection.GetAgents[Teams[1]].Length)
		{
			if ( RedCount < BlueCount )
			{
				if { TeamCollection.AddToTeam[Agent, Teams[1]] } # Red
			}
			else if ( BlueCount < RedCount )
			{
				if { TeamCollection.AddToTeam[Agent, Teams[0]] } # Blue
			}
			else # even number?, pick random team
			{
				if { TeamCollection.AddToTeam[Agent, Teams[GetRandomInt(0,1)]] }
			}
		}
	}

	PlayerEnterGame(Agent:agent)<suspends>:void=
	{
		if ( not PlayersMap[Agent] )	# new player
		{
			Print("PlayerEnterGame()")
			if ( set PlayersMap[Agent] = fn_player_class{ Agent := Agent, LinkedCreativeDevice := Self, AgentClass := FNClasses.Midas} )
			{
				if (InPlayer := player[Agent])
				{
					var OptPlayerUI :?ui_hudgameplay = false

					if (ThisPlayerUI := UIHudMap[InPlayer])
					{
						set OptPlayerUI = option{ThisPlayerUI}
					}
					else if (ThisPlayerUI := set UIHudMap[InPlayer] = ui_hudgameplay{ LinkedPlayer := InPlayer, LinkedCreativeDevice := Self } ) 
					{
						ThisPlayerUI.Init()
						set OptPlayerUI = option{ThisPlayerUI}
					}

					if (PlayerUI := OptPlayerUI?)
					{
						PlayerUI.OpenUI()
					}
				}

				<#if ( FortCharacter := Agent.GetFortCharacter[] )
				{
					AssignPlayerTeam(Agent)
					FortCharacter.Damage(1000.0)
				}#>
			}
		}
	}

	ShowHUD(Agent:agent):void= 
	{
		if (PlayerHUD := UIHudMap[Agent])
		{
			PlayerHUD.OpenUI()
		}
	}

	HideHUD(Agent:agent):void= 
	{
		if (PlayerHUD := UIHudMap[Agent])
		{
			spawn{PlayerHUD.CloseUI()}
		}
	}

	MainMenuShow(Agent:agent)<suspends>:void=
	{
		if ( UIMainMenuShow := UIMainMenuShowDevice? )
		{
			UIMainMenuShow.MainMenuShow(Agent)
		}
	}

	PlayerSpawn(Agent:agent)<suspends>:void=
	{
		# update saved level
		if ( SaveManager := SaveManagerDevice? )
		{
			SaveManager.UpdatePlayerLevel(Agent)
		}

		# main menu open up
		if ( PlayersMap[Agent].bFirstSpawn? )
		{
			spawn{MainMenuShow(Agent)}
			if {set PlayersMap[Agent].bFirstSpawn = false}
			SetPlayerClass( Agent, FNClasses.None )
			spawn{PlayerThink(Agent)}
			return
		}

		# prevent spawning naked
		PlayerClass := GetPlayerClassIndex(Agent)
		SetPlayerClass( Agent, PlayerClass )

		# open class menu
		TriggerClassMenuOpen.Trigger(Agent)

		# add brain
		spawn{PlayerThink(Agent)}
	}

	AddCond(Agent:agent, eCond:FNCond, Duration:float, ?AgentOwner:?agent = false)<suspends>:void=
	{
		if
		{
			FortCharacter := Agent.GetFortCharacter[]
			FortCharacter.IsActive[]
			set PlayersMap[Agent].CondMap[eCond] = FNCondData{ExpireTime := Duration, AgentProvider := AgentOwner}
		}
		then
		{
			Print("AddCond success")
			#AddCondEvent.Signal(Agent)
		}
	}

	InCond(Agent:agent, eCond:FNCond)<transacts><decides>: void =
	{
		var bSuccess:logic = false
		if ( PlayersMap[Agent].CondMap[eCond] )
		{
			#Print("Has Cond.")
			set bSuccess = true
			bSuccess?
		}

		bSuccess?
	}

	RemoveCond(Agent:agent, eCond:FNCond)<suspends>:void=
	{
		if ( AgentMap := PlayersMap[Agent], CondList := AgentMap.CondMap )
		{
			set AgentMap.CondMap = CondList.RemoveKey(eCond)
			if ( set PlayersMap[Agent] = AgentMap )
			{
				Print("{AgentMap.CondMap.Length}")
			}
			#RemoveCondEvent.Signal(Agent)
		}
	}

	RemoveAllCond(Agent:agent)<suspends>:void=
	{
		if { set PlayersMap[Agent].CondMap = map{} }
	}

	GetConditionDuration(Agent:agent, eCond:FNCond):float=
	{
		if ( CondExpireTime := PlayersMap[Agent].CondMap[eCond].ExpireTime )
		{
			return CondExpireTime
		}
		
		return 0.0
	}

	GetConditionProvider(Agent:agent, eCond:FNCond)<transacts><decides>:agent=
	{
		PlayersMap[Agent].CondMap[eCond].AgentProvider?
	}

	PlayerThink(Agent:agent)<suspends>:void=
	{
		if ( FortCharacter := Agent.GetFortCharacter[] )
		{
			race
			{
				block
				{
					FortCharacter.EliminatedEvent().Await()
				}
				block
				{
					loop
					{
						if ( not FortCharacter.IsActive[] )
						{
							break
						}

						if ( Agent.HasAbility[FNAbility.ShieldRegen] )
						{
							spawn{AddShield(Agent, 1.0)}
						}

						if ( Agent.HasAbility[FNAbility.HealthRegen] )
						{
							spawn{AddHealth(Agent, 1.0)}
						}

						if ( FortCharacter.IsInWater[] )
						{
							if ( Agent.HasAbility[FNAbility.TherapeuticSwim] )
							{
								FortCharacter.Heal(5.0)
							}

							case( GetPlayerClassIndex(Agent) )
							{
								FNClasses.CubeAssassin =>		FortCharacter.Damage(1.0)
								FNClasses.CubeQueen =>			FortCharacter.Damage(1.0)
								FNClasses.CubeBlue =>			FortCharacter.Damage(1.0)
								_ =>							# nothing
							}
						}

						if ( Agent.HasAbility[FNAbility.AirRaid] and FortCharacter.IsInAir[] )
						{
							spawn{DamageBuffPlayer(Agent, 1.0, 1.4)}
						}

						if ( Agent.IsAltTransform[] )
						{
							SetUltimateMeter(Agent,GetUltimateMeter(Agent) - 1.0)
						}

						<#TaggedBlueDevice := FindCreativeObjectsWithTag(volume_bluezone_tag{})
						for (FoundDevice: TaggedDevice, VolumeDevice := volume_device[FoundDevice], VolumeDevice.IsInVolume[Agent] )
						{
							# does stuff
						}#>

						<#TaggedBlueDevice := FindCreativeObjectsWithTag(volume_redzone_tag{})
						for (FoundDevice: TaggedDevice, VolumeDevice := volume_device[FoundDevice], VolumeDevice.IsInVolume[Agent] )
						{
							# does stuff
						}#>

						<#for (Key->Value : PlayersMap[Agent].CondMap)
						{
							if ( set PlayersMap[Agent].CondMap[Key].flExpireTime -= 1.0 ) {}
							Print("{Value} in ExampleMap at key {Key}")
						}#>

						<#if ( PlayersMap[Agent].CondMap[FNCond.Burning]? )
						{
							{ FortCharacter.Damage(2.0) }
						}#>

						Sleep(1.0)
					}
				}
			}

			defer
			{
				Print("Player has stopped thinking...")
			}
		}
	}

    # Event that handles when a player is eliminated
	OnPlayerEliminated(Result:elimination_result):void=
	{
		# Who got killed?
		KilledCharacter := Result.EliminatedCharacter

		if
		{
			# Let's make sure we can get the Agent of the fort_character                                                                    
			KilledAgent := KilledCharacter.GetAgent[]

			# Who did the killing? Why is there a question mark?
			# This fails if the player was killed by something other than
			# another fort_character. Ex. Damage Volumes
			KillerChar := Result.EliminatingCharacter?

            # Let's make sure we can get the Agent of the fort_character 
			KillerAgent := KillerChar.GetAgent[]
		}
        then
		{
			PlayClassVO(KilledAgent, FNVOType.DBNO)
			spawn{UpdatePlayerHUD(KilledAgent)}
			spawn{ShowKillerInfoHUD(KilledAgent, KillerAgent)}

			# Make sure they did not eliminate themself
			if(KillerAgent <> KilledAgent)
			{
				PlayClassVO(KillerAgent, FNVOType.FullAlertLost)
				AddUltimateMeter(KillerAgent, 5.0)
				AddTrackerKill(KillerAgent, 1)

				<#KillerFaction := GetPlayerClassFaction(KillerAgent)
				if ( FactionWar := FactionWarDevice? )
				{
					FactionWar.AddFactionKillCount(KillerFaction)
				}#>

				KilledAgent.AwardAssistAgent()
				AwardSpecialElimination(KillerAgent, KilledAgent)
			}

			KilledClass := GetPlayerClassIndex(KilledAgent)
			
			case( KilledClass )
			{
				FNClasses.Shadow =>
				{
					SetPlayerClass( KilledAgent, FNClasses.ShadowMidas )
				}
				FNClasses.DoctorDoomMythic =>
				{
					SetPlayerClass( KilledAgent, FNClasses.DoctorDoom )
				}
				_ =>
				{
					#nothing
				}
			}

			ClassSwitchedCleanup(KilledAgent)
		}
	}

	(DamagedAgent:agent).AwardAssistAgent():void=
	{
		if ( DamagerAgent := agent[PlayersMap[DamagedAgent].AssistAgent?], DamagerAgent <> DamagedAgent, DamagerPlayer := player[DamagerAgent] )
		{
			#Print("Awarding assist XP to recent damager...")
			XP_Assist.Award(DamagerAgent)
		}
	}

	AwardSpecialElimination(KillerAgent:agent, KilledAgent:agent):void=
	{
		#Print("CheckForAwardSpecialElimination()")
		KilledClass := GetPlayerClassIndex(KilledAgent)
		KillerClass := GetPlayerClassIndex(KillerAgent)
		KilledFaction := GetPlayerClassFaction(KilledAgent)
		KillerFaction := GetPlayerClassFaction(KillerAgent)

		if ( KilledCharacter := KilledAgent.GetFortCharacter[] )
		{
			if ( KilledCharacter.IsInWater[] )
			{
				XP_Furina.Award(KillerAgent)
			}
		}

		# very specific stuff
		case( KillerClass )
		{
			FNClasses.Midas =>
			{
				if ( KilledClass = FNClasses.Jules )
				{
					spawn{AchievementUnlock(KillerAgent, 5)}
				}
				if ( KilledClass = FNClasses.Hades )
				{
					spawn{AchievementUnlock(KillerAgent, 6)}
				}
			}
			FNClasses.SgtWinter =>
			{
				XP_SgtWinterKill.Award(KillerAgent)
				XP_SgtWinterKill.Award(KilledAgent)
			}
			FNClasses.TheFoundation =>
			{
				if ( KilledClass = FNClasses.Slone )
				{
					spawn{AchievementUnlock(KillerAgent, 7)}
				}
			}
			FNClasses.Raz =>
			{
				if ( KilledClass = FNClasses.SpireAssassin )
				{
					spawn{AchievementUnlock(KillerAgent, 8)}
				}
			}
			FNClasses.Ares =>
			{
				if ( KilledClass = FNClasses.Ares )
				{
					spawn{AchievementUnlock(KillerAgent, 9)}
				}
			}
			FNClasses.FletcherKane =>
			{
				if ( KilledClass = FNClasses.Midas or KilledClass = FNClasses.ShadowMidas )
				{
					spawn{AchievementUnlock(KillerAgent, 24)}
				}
			}
			FNClasses.AlphaHeavy =>
			{
				if ( KilledClass = FNClasses.Midas or KilledClass = FNClasses.ShadowMidas )
				{
					spawn{AchievementUnlock(KillerAgent, 24)}
				}
			}
			_ =>	#nothing
		}
	}

	OnPlayerShieldBroken(Agent:agent):void=
	{
		if
		{
			set PlayersMap[Agent].bShieldBroken = true
			PlayerAbility := PlayerAbilityDevice?
		}
		then
		{
			case( GetPlayerClassIndex(Agent) )
			{
				FNClasses.MegaloDon =>			spawn{PlayerAbility.AbilityNitroCharge(Agent)}
				FNClasses.RingmasterScarr =>	spawn{PlayerAbility.AbilityNitroCharge(Agent)}
				FNClasses.TheMachinist =>		spawn{PlayerAbility.AbilityNitroCharge(Agent)}
				_ =>							#nothing
			}

			if ( Agent.HasAbility[FNAbility.PartingGift], FortCharacter := Agent.GetFortCharacter[] )
			{
				spawn{PlayerAbility.AbilityPartingGift(Agent)}
			}
		}
	}

    # Event that handles when a player is damaged
	OnPlayerDamaged(Result:damage_result):void=
	{
		if (DamagedChar := fort_character[Result.Target], DamagedAgent := DamagedChar.GetAgent[])
		{
			if ( DamagedChar.IsActive[] )
			{
				if ( DamagedChar.GetShield() = 0.0 )
				{
					if ( not PlayersMap[DamagedAgent].bShieldBroken? )
					{
						Print("OnPlayerDamaged: Shield Broken")
						OnPlayerShieldBroken(DamagedAgent)
					}
				}

				if ( DamagedChar.GetHealth() >= (DamagedChar.GetMaxHealth() / 2.0) )
				{
					PlayClassVO(DamagedAgent, FNVOType.FullAlert)
				}
				else
				{
					# low health
					PlayClassVO(DamagedAgent, FNVOType.Interrogation)

					case( GetPlayerClassIndex(DamagedAgent) )
					{
						FNClasses.ShogunX =>
						{
							if ( PlayersMap[DamagedAgent].eAggro = FNBGMState.Alerted )
							{
								SetPlayerAggro(DamagedAgent, FNBGMState.AlertedPhase2)
							}
						}
						FNClasses.Kor =>
						{
							if ( PlayersMap[DamagedAgent].eAggro = FNBGMState.Alerted )
							{
								SetPlayerAggro(DamagedAgent, FNBGMState.AlertedPhase2)
							}
						}
						_ =>							#nothing
					}
					
					if ( DamagedAgent.HasAbility[FNAbility.Enragement], PlayersMap[DamagedAgent].eAggro = FNBGMState.Alerted )
					{
						SetPlayerAggro(DamagedAgent, FNBGMState.AlertedPhase2)
						if ( PlayerAbility := PlayerAbilityDevice? )
						{
							spawn{PlayerAbility.AbilityEnrage(DamagedAgent)}
						}
					}
				}

				<#if ( DamagedChar.GetHealth() >= 0.0 )
				{
					FortCharacter.SetHealth(1.0)
					FortCharacter.SetVulnerability( false )
					Print("Player is ded.")
				}#>
			}

			if( AttackerAgent := Result.Instigator?.GetInstigatorAgent[], DamagedAgent <> AttackerAgent )
			{
				PlayClassVO(AttackerAgent, FNVOType.FullAlert)
				AddUltimateMeter(AttackerAgent, 1.0*(Result.Amount / 25.0))
				if ( not DamagedAgent.IsAltTransform[] )
				{
					AddUltimateMeter(DamagedAgent, 0.75)
				}

				spawn{UpdateHudDamageText(AttackerAgent, Result.Amount)}
				#spawn{DamageEvent(AttackerAgent)}
				if { set PlayersMap[DamagedAgent].DamageTakenTracker += Result.Amount }

				DamagedAgent.SetAssistAgent(AttackerAgent)

				if ( DamagedChar.IsActive[], PlayersMap[DamagedAgent].eAggro = FNBGMState.Idle )
				{
					SetPlayerAggro(DamagedAgent, FNBGMState.Alerted)
				}

				if ( PlayersMap[AttackerAgent].bAltTransform?, ConditionalDaigoBlade.IsHoldingItem[AttackerAgent], PlayerUlt := PlayerUltDevice? )
				{
					spawn{PlayerUlt.UltSplinterBladeVictim(AttackerAgent,DamagedAgent)}
				}
			}
		}
	}

	(Agent:agent).SetAssistAgent(AttackerAgent:agent):void=
	{
		if
		{
			FortCharacter := Agent.GetFortCharacter[]
			FortCharacter.IsActive[]
			FortAttacker := AttackerAgent.GetFortCharacter[]
			FortAttacker.IsActive[]
			FortCharacter.GetHealth() > 0.0
			Agent <> AttackerAgent
			AttackerPlayer := player[AttackerAgent]
		}
		then
		{
			#Print("Player: Set new assist.")
			if { set PlayersMap[Agent].AssistAgent = option{AttackerAgent} }
		}
	}

	<#DamageEvent(Agent:agent)<suspends>:void=
	{
		if ( FortCharacter := Agent.GetFortCharacter[], FortCharacter.IsActive[] )
		{
			case( GetPlayerClassIndex(Agent) )
			{
				FNClasses.Shadow =>				FortCharacter.Heal(15.0)
				_ =>							#nothing
			}
		}
	}#>

	OnPlayerHealed(Result:healing_result):void=
	{
		if (HealedChar := fort_character[Result.Target], HealedAgent := HealedChar.GetAgent[])
		{
			if( HealerAgent := Result.Instigator?.GetInstigatorAgent[] , HealedAgent <> HealerAgent, HealerSource := Result.Source? )
			{
				AddUltimateMeter(HealerAgent, 1.0*(Result.Amount / 20.0))
				AddTrackerHealing(HealerAgent, Result.Amount)
			}
		}
	}

	AddTrackerHealing(Agent:agent, Amount:float):void=
	{
		if ( set PlayersMap[Agent].flHealingTracker += Amount )
		{
			if ( PlayersMap[Agent].flHealingTracker >= 100.0 )
			{
				if (set PlayersMap[Agent].flHealingTracker = 0.0 )
				{
					spawn{AwardScore(Agent, 1.0)}
					XP_Healer.Award(Agent)
				}
			}
		}
	}

	AddTrackerKill(Agent:agent, Amount:int):void=
	{
		if ( set PlayersMap[Agent].KillsTracker += Amount )
		{
			spawn{AwardScore(Agent, 1.0)}
			if ( PlayersMap[Agent].KillsTracker >= 4 )
			{
				if (set PlayersMap[Agent].KillsTracker = 0 )
				{
					# reward!!
					XP_4Kills.Award(Agent)
				}
			}
		}
	}

	SetPlayerAggro<public>(Agent:agent, BGMState:FNBGMState):void=
	{
		if { set PlayersMap[Agent].eAggro = BGMState }

		if ( ActiveBGM := PlayersMap[Agent].ActiveBGM? )
		{
			ActiveBGM.Stop(Agent)
		}

		if ( SaveManager := SaveManagerDevice?, SaveManager.IsBGMEnabled[Agent] )
		{
			case( BGMState )
			{
				FNBGMState.Idle =>
				{
					if ( Agent.IsOnTeam[Self,1], PlayersMap[Agent].ClassData.HasTeamAlt? )
					{
						if ( set PlayersMap[Agent].ActiveBGM = option{PlayersMap[Agent].ClassData.Team2_BGMIdle} ) {}
					}
					else
					{
						if ( set PlayersMap[Agent].ActiveBGM = option{PlayersMap[Agent].ClassData.BGMIdle} ) {}
					}
				}
				FNBGMState.Alerted =>
				{
					if ( Agent.IsOnTeam[Self,1], PlayersMap[Agent].ClassData.HasTeamAlt? )
					{
						if ( set PlayersMap[Agent].ActiveBGM = option{PlayersMap[Agent].ClassData.Team2_BGMPhase1} ) {}
					}
					else
					{
						if ( set PlayersMap[Agent].ActiveBGM = option{PlayersMap[Agent].ClassData.BGMPhase1} ) {}
					}
				}
				FNBGMState.AlertedPhase2 =>
				{
					if ( set PlayersMap[Agent].ActiveBGM = option{PlayersMap[Agent].ClassData.BGMPhase2} ) {}
				}
				_ =>	{}
			}

			if ( ActiveBGM := PlayersMap[Agent].ActiveBGM? )
			{
				ActiveBGM.Play(Agent)
			}
		}
	}

	RevertAltTransform(Agent:agent):void=
	{
		if ( FortCharacter := Agent.GetFortCharacter[], set PlayersMap[Agent].bAltTransform = false )
		{
			case( GetPlayerClassIndex(Agent) )
			{
				FNClasses.DoctorDoomMythic =>
				{
					SetPlayerClass( Agent, FNClasses.DoctorDoom )
					CurrentHealth := FortCharacter.GetHealth()
					CurrentShield := FortCharacter.GetShield()

					FortCharacter.SetHealth(100.0)
					FortCharacter.SetShield(0.0)
				}
				_ =>
				{
					SetPlayerClass( Agent, GetPlayerClassIndex(Agent) )
					CurrentHealth := FortCharacter.GetHealth()
					CurrentShield := FortCharacter.GetShield()

					#ClassSwitchedCleanup(Agent)

					FortCharacter.SetHealth(CurrentHealth)
					FortCharacter.SetShield(CurrentShield)
				}
			}
		}
	}

	ExpendUltimateMeter(Agent:agent):void=
	{
		case( GetPlayerClassIndex(Agent) )
		{
			FNClasses.DoctorDoom =>			SetUltimateMeter(Agent, GetMaxUlt(Agent)-1.0)
			FNClasses.DoctorDoomMythic =>	SetUltimateMeter(Agent, GetMaxUlt(Agent)-1.0)
			_ =>							SetUltimateMeter(Agent,0.0)
		}
	}

	SetUltimateMeter(Agent:agent, Amount:float):void=
	{
		NewMaxAmount : float = GetMaxUlt(Agent)
		var NewAmount : float = Amount
		if ( ( GetUltimateMeter(Agent) + NewAmount ) < 0.0 )
		{
			set NewAmount = 0.0
		}

		if ( set PlayersMap[Agent].flPlayerUltimateMeter = NewAmount )
		{
			if ( not PlayersMap[Agent].bAltTransform? )
			{
				if ( GetUltimateMeter(Agent) >= NewMaxAmount, set PlayersMap[Agent].flPlayerUltimateMeter = NewMaxAmount )
				{
					set NewAmount = NewMaxAmount
					if (set PlayersMap[Agent].bUltReady = true)
					{
						InputUlt.Register(Agent)
						#SFXUltReady.Play(Agent)
						VFXPowerUlt.Pickup(Agent)
						AddHudStatusMessage(Agent, TextLocalization.MSG_UseUlt)
					}
				}
			}
			else
			{
				if ( GetUltimateMeter(Agent) <= 0.0, set PlayersMap[Agent].flPlayerUltimateMeter = 0.0 )
				{
					set NewAmount = 0.0
					if (set PlayersMap[Agent].bUltReady = false)
					{
						RevertAltTransform(Agent)
					}
				}
			}

			if ( IntAmount := Int[(NewAmount / NewMaxAmount) * 100.0], StatUlt.SetValue[Agent, IntAmount])			
			{
				#Print("UltimateMeter: {GetUltimateMeter(Agent)}")
			}
		}
	}

	AddUltimateMeter(Agent:agent, Amount:float):void=
	{
		if ( PlayersMap[Agent].bUltReady? )
		{
			return
		}

		# Remove ult during transform
		if ( Agent.IsAltTransform[] )
		{
			AltMultiplier := 0.25
			SetUltimateMeter(Agent, GetUltimateMeter(Agent) - Clamp(Amount*AltMultiplier,0.0,GetMaxUlt(Agent)) )
			return
		}

		# Add ult
		var ClassMultiplier : float = 1.0
		case( GetPlayerClassIndex(Agent) )
		{
			FNClasses.Midas =>				set ClassMultiplier = 1.2
			FNClasses.TNTina =>				set ClassMultiplier = 1.2
			FNClasses.Brutus =>				set ClassMultiplier = 1.2
			FNClasses.Skye =>				set ClassMultiplier = 1.2
			FNClasses.Meowscles =>			set ClassMultiplier = 1.2
			FNClasses.Jules =>				set ClassMultiplier = 1.5
			FNClasses.Kit =>				set ClassMultiplier = 1.5
			FNClasses.Ocean =>				set ClassMultiplier = 1.5
			FNClasses.ShadowMidas =>		set ClassMultiplier = 1.2
			FNClasses.Shadow =>				set ClassMultiplier = 1.0
			FNClasses.Raz =>				set ClassMultiplier = 1.5
			FNClasses.Slone =>				set ClassMultiplier = 1.2
			FNClasses.TheFoundation =>		set ClassMultiplier = 1.25
			FNClasses.Gunnar =>				set ClassMultiplier = 1.2
			FNClasses.HuntmasterSaber =>	set ClassMultiplier = 1.5
			FNClasses.TheHerald =>			set ClassMultiplier = 1.5
			FNClasses.TheAgelessChampion => set ClassMultiplier = 1.5
			FNClasses.WildguardRelik =>		set ClassMultiplier = 1.4
			FNClasses.KadoThorne =>			set ClassMultiplier = 1.0
			FNClasses.TheDealer =>			set ClassMultiplier = 1.0
			FNClasses.Valeria =>			set ClassMultiplier = 1.25
			FNClasses.Oscar =>				set ClassMultiplier = 1.25
			FNClasses.Nisha =>				set ClassMultiplier = 1.25
			FNClasses.Montague =>			set ClassMultiplier = 1.25
			FNClasses.SgtWinter =>			set ClassMultiplier = 1.5
			FNClasses.SpireAssassin =>		set ClassMultiplier = 1.5
			FNClasses.CubeAssassin =>		set ClassMultiplier = 1.25
			FNClasses.CubeQueen =>			set ClassMultiplier = 1.0
			FNClasses.ZygAndChoppy =>		set ClassMultiplier = 1.2
			FNClasses.Zeus =>				set ClassMultiplier = 1.25
			FNClasses.Ares =>				set ClassMultiplier = 1.25
			FNClasses.Cerberus =>			set ClassMultiplier = 1.2
			FNClasses.Hades =>				set ClassMultiplier = 1.2
			FNClasses.MegaloDon =>			set ClassMultiplier = 1.0
			FNClasses.RingmasterScarr =>	set ClassMultiplier = 1.0
			FNClasses.TheMachinist =>		set ClassMultiplier = 1.0
			FNClasses.UltimaCarver =>		set ClassMultiplier = 1.0
			FNClasses.ShogunX =>			set ClassMultiplier = 1.0
			FNClasses.NightRose =>			set ClassMultiplier = 1.0
			FNClasses.Puppeteer =>			set ClassMultiplier = 0.25
			FNClasses.FletcherKane =>		set ClassMultiplier = 0.75
			FNClasses.AlphaHeavy =>			set ClassMultiplier = 0.75
			FNClasses.Daigo =>				set ClassMultiplier = 1.0
			FNClasses.Kor =>				set ClassMultiplier = 1.0
			FNClasses.Jones =>				set ClassMultiplier = 1.15
			FNClasses.DoctorDoom =>			set ClassMultiplier = 1.0
			FNClasses.DoctorDoomMythic =>	set ClassMultiplier = 1.0
			FNClasses.Caretaker =>			set ClassMultiplier = 1.2
			FNClasses.Hope =>				set ClassMultiplier = 1.2
			_ =>							set ClassMultiplier = 1.0
		}

		SetUltimateMeter(Agent, GetUltimateMeter(Agent) + Clamp(Amount*ClassMultiplier,0.0,GetMaxUlt(Agent)) )
	}

	GetUltimateMeter(Agent:agent)<transacts>:float=
	{
		if ( UltMeter := PlayersMap[Agent].flPlayerUltimateMeter )
		{
			return UltMeter
		}

		return 0.0
	}

	AddHealth(Agent:agent, AddAmount:float)<suspends>:void=
	{
		if ( FortCharacter := Agent.GetFortCharacter[] )
		{
			FortCharacter.SetHealth(FortCharacter.GetHealth() + AddAmount)
		}
	}

	AddShield(Agent:agent, AddAmount:float)<suspends>:void=
	{
		if ( FortCharacter := Agent.GetFortCharacter[] )
		{
			FortCharacter.SetShield(FortCharacter.GetShield() + AddAmount)
		}
	}

	<#BurnPlayer(Owner:agent, Victim:agent)<suspends>:void=
	{
		if (FortOwner := Owner.GetFortCharacter[] FortVictim := Owner.GetFortCharacter[] )
		{
			var NumberOfBurn : int = 0
			loop
			{
				Print("Burning Ambition: BURN")
				if ( FortCharacter.IsActive[], DamageInstigator := Agent.GetInstigator(), DamageSource := game_action_causer[FortCharacter] )
				{
					BurnDmg : damage_args = damage_args{Instigator := option{DamageInstigator} , Source := option{DamageSource}, Amount:=8.0}
					FortVictim.Damage(BurnDmg)
				}

				set NumberOfBurn += 1
				if ( NumberOfBurn >= 4 )
				{
					break
				}
				Sleep(1.0)
			}
		}
	}#>

    DamageBuffPlayer(Agent:agent, Duration:float, Magnitude:float)<suspends>:void=
	{
		DMGAmplifier.SetDuration(Duration)
		DMGAmplifier.SetMagnitude(Magnitude)
		VFXDMGAmplifier.SetDuration(Duration)
		if ( not DMGAmplifier.IsSpawned[] )
		{
			DMGAmplifier.Spawn()
		}
		DMGAmplifier.Pickup(Agent)
		VFXDMGAmplifier.Pickup(Agent)
	}

	GetAbilityCooldown(Agent:agent, AbilityID:FNAbility)<transacts>:float=
	{
		if ( MainAbility := GetMainAbility[Agent], MainAbility = AbilityID )
		{
			if ( AbilityTimer := PlayersMap[Agent].flAbilityCooldown )
			{
				return AbilityTimer
			}
		}

		if ( MainAbility2 := GetMainAbility2[Agent], MainAbility2 = AbilityID )
		{
			if ( AbilityTimer2 := PlayersMap[Agent].flAbilityCooldown2 )
			{
				return AbilityTimer2
			}
		}

		return 0.0
	}

	GetAbilityMaxCooldown(Agent:agent, AbilityID:FNAbility)<transacts>:float=
	{
		if ( MainAbility := GetMainAbility[Agent], MainAbility = AbilityID )
		{
			if ( AbilityTimer := PlayersMap[Agent].flAbilityCooldownMax )
			{
				return AbilityTimer
			}
		}

		if ( MainAbility2 := GetMainAbility2[Agent], MainAbility2 = AbilityID )
		{
			if ( AbilityTimer2 := PlayersMap[Agent].flAbilityCooldownMax2 )
			{
				return AbilityTimer2
			}
		}


		return 0.0
	}

	UpdatePlayerHUD(Agent:agent)<suspends>:void=
	{
		Sleep(0.1)
		if (PlayerHUD := UIHudMap[Agent])
		{
			PlayerHUD.UpdateUI()
		}
	}

	ShowKillerInfoHUD(Agent:agent, KillerAgent:agent)<suspends>:void=
	{
		if (PlayerHUD := UIHudMap[Agent])
		{
			PlayerHUD.ShowKillerInfo(KillerAgent)
		}
	}

	OnPlayerCrouched(FortChar:fort_character, IsCrounch:logic):void= 
	{
		if( IsCrounch? )
		{
			if( Agent := FortChar.GetAgent[] )
			{
				PlayClassVO(Agent, FNVOType.Idle)
			}
		}
	}

	OnPlayerSprinted(FortChar:fort_character, IsSprinted:logic):void= 
	{
		if( not IsSprinted? )
		{
			return
		}

		if( Agent := FortChar.GetAgent[] )
		{
			if (GetRandomInt(0,1) = 1)
			{
				#PlayClassVO(Agent, FNVOType.FullAlertIdle)
			}
		}
	}

	OnPlayerJumped(FortChar:fort_character):void= 
	{
		if( Agent := FortChar.GetAgent[] )
		{
			# DEBUG!
			<#if ( FactionWar := FactionWarDevice? )
			{
				KillerFaction := GetPlayerClassFaction(Agent)
				FactionWar.AddFactionKillCount(KillerFaction)
			}#>
		}
	}

	OnIntelPickedUp(Agent:agent):void= 
	{
		Print("OnIntelPickedUp")
		if (set PlayersMap[Agent].bHasTheFlag = true)
		{
			if ( Agent.HasAbility[FNAbility.LastResort], PlayerAbility := PlayerAbilityDevice? )
			{
				spawn{PlayerAbility.AbilityMovePlus(Agent)}
			}

			case( GetPlayerClassIndex(Agent) )
			{
				FNClasses.TheDealer =>
				{
					SetPlayerAggro(Agent, FNBGMState.AlertedPhase2)
				}
				_ =>
				{
					#nothing
				}
			}
		}
	}

	OnIntelDropped(Agent:agent):void= 
	{
		Print("OnIntelDropped")
		if (set PlayersMap[Agent].bHasTheFlag = false)
		{
			if ( Agent.HasAbility[FNAbility.LastResort], PlayerAbility := PlayerAbilityDevice? )
			{
				spawn{PlayerAbility.MoveNormal(Agent)}
			}

			case( GetPlayerClassIndex(Agent) )
			{
				FNClasses.TheDealer =>
				{
					SetPlayerAggro(Agent, FNBGMState.Alerted)
				}
				_ =>
				{
					#nothing
				}
			}
		}
	}

	AwardScore(Agent:agent, Amount:float)<suspends>:void=
	{
		if ( ScorePoint := Ceil[Amount] )
		{
			ScoreManager.SetScoreAward(ScorePoint)
			Print("Score Award: {ScorePoint}")
			ScoreManager.Activate(Agent)
			if { set PlayersMap[Agent].RoundScore += ScorePoint }
			spawn{AwardXP(Agent,ScorePoint)}
		}
	}

	OnPlayerWin(Agent:agent):void= 
	{
		spawn{AwardXP(Agent, 40)}

		if ( FortCharacter := Agent.GetFortCharacter[] )
		{
			FortCharacter.SetVulnerability(false)
		}
	}

	OnPlayerLose(Agent:agent):void= 
	{
		spawn{AwardXP(Agent, 20)}

		if ( FortCharacter := Agent.GetFortCharacter[] )
		{
			FortCharacter.SetVulnerability(false)
		}
	}

	VectorNormalize(v1:vector3)<transacts><decides>:vector3=
	{
		if ( Mag := VectorMagnitude[v1] )
		{
			vector3{X := v1.X/Mag, Y := v1.Y/Mag, Z:= v1.Z/Mag}
		}

		vector3{}
	}
 
	VectorMagnitude(v1:vector3)<transacts><decides>:float=
	{
		Sqrt(Pow(v1.X, 2.0) + Pow(v1.Y, 2.0) + Pow(v1.Z, 2.0))
	}
 
	# todo: add <transacts><decides>
	(Character:fort_character).IsLookingAtLocation(LocationToCheck: vector3, Threshold: float)<transacts><decides>:void=
	{
		var bSuccess:logic = false

		# Get the character's view
		ViewLocation := Character.GetViewLocation()
		ViewRotation := Character.GetViewRotation()
		# Create a direction vector from the view location to the target vector
		if ( DirectionVector := VectorNormalize[LocationToCheck - ViewLocation] )
		{
			# Convert the view rotation to a direction vector
			ViewDirection := ViewRotation.GetLocalForward()
			# Compare the direction vector and the view direction
			# Use a dot product to find the cosine of the angle between the two vectors.
			# If it's close to 1, then the vectors are nearly parallel, so the character is looking at the target.
			CosAngle := DotProduct(ViewDirection, DirectionVector)
			if( IsFacing := CosAngle > Threshold )
			{
				set bSuccess = true
				bSuccess?
			}
		}

		bSuccess?
	}

	OnPlayerBeginsEmoting(Agent:agent):void= 
	{
		if {set PlayersMap[Agent].bIsEmoting = true}
	}

	OnPlayerEndEmoting(Agent:agent):void= 
	{
		if {set PlayersMap[Agent].bIsEmoting = false}
	}

	AwardXP(Agent:agent, Amount:int)<suspends>:void=
	{
		if ( SaveManager := SaveManagerDevice? )
		{
			SaveManager.GrantXP(Agent, Amount)
		}
	}

	AchievementUnlock(Agent:agent, AchievementID:int)<suspends>:void=
	{
		if ( UIAchievement := UIAchievementDevice? )
		{
			spawn{UIAchievement.AchievementUnlock(Agent, AchievementID)}
		}
	}

	GetPassiveList(Agent:agent)<transacts><decides>:[]FNAbility=
	{
		PlayersMap[Agent].GetPassiveList[]
	}

	GetMainAbility(Agent:agent)<transacts><decides>:FNAbility=
	{
		PlayersMap[Agent].GetMainAbility[]
	}

	GetMainAbility2(Agent:agent)<transacts><decides>:FNAbility=
	{
		PlayersMap[Agent].GetMainAbility2[]
	}

	GetAbilityData(Ability:FNAbility)<transacts><decides>:FNAbilityData=
	{
		var AbilityStruct : FNAbilityData = FNAbilityData{}

		case( Ability )
		{
			FNAbility.DoubleJump =>			set AbilityStruct = {FNAbilityData{Icon := Textures.ui_ability_doublejump, Title := "Double Jump", Cooldown := 2.0}}
			FNAbility.Warp =>				set AbilityStruct = {FNAbilityData{Icon := Textures.ui_ability_dash, Title := "Warp", Cooldown := 3.0}}
			FNAbility.ShieldRegen =>		set AbilityStruct = {FNAbilityData{Icon := Textures.ui_ability_shieldregen, Title := "Shield Regen"}}
			FNAbility.Shadow =>				set AbilityStruct = {FNAbilityData{Icon := Textures.class_shadow, Title := "Shadow"}}
			FNAbility.LifeDrain =>			set AbilityStruct = {FNAbilityData{Icon := Textures.ui_ability_kado_lifedrain, Title := "Life Drain", Cooldown := 15.0}}
			FNAbility.Rock =>				set AbilityStruct = {FNAbilityData{Icon := Textures.ui_ability_rock, Title := "The Rock", Cooldown := 10.0}}
			FNAbility.Charge =>				set AbilityStruct = {FNAbilityData{Icon := Textures.ui_ability_charge, Title := "Charge", Cooldown := 10.0}}
			FNAbility.SpinSlash =>			set AbilityStruct = {FNAbilityData{Icon := Textures.ui_ability_spinslash, Title := "Spin Slash", Cooldown := 10.0}}
			FNAbility.AlienParasite =>		set AbilityStruct = {FNAbilityData{Icon := Textures.ui_ability_alienparasite, Title := "Alien Parasite", Cooldown := 10.0}}
			FNAbility.MirrorImage =>		set AbilityStruct = {FNAbilityData{Icon := Textures.ui_ability_ioguard, Title := "Mirror Image", Cooldown := 12.0}}
			FNAbility.Lockdown =>			set AbilityStruct = {FNAbilityData{Icon := Textures.ui_ability_lockdown, Title := "Lockdown", Cooldown := 12.0}}
			FNAbility.Phase2 =>				set AbilityStruct = {FNAbilityData{Icon := Textures.ui_ability_societyguard, Title := "Phase 2", Cooldown := 12.0}}
			FNAbility.Rebirth =>			set AbilityStruct = {FNAbilityData{Icon := Textures.ui_ability_rebirth, Title := "Rebirth"}}
			FNAbility.DamageBuff =>			set AbilityStruct = {FNAbilityData{Icon := Textures.ui_ability_damagebuff_ares, Title := "Damage Buff", Cooldown := 25.0}}
			FNAbility.HealingRune =>		set AbilityStruct = {FNAbilityData{Icon := Textures.ui_ability_healingrune, Title := "Healing Rune, Cooldown := 10.0"}}
			FNAbility.Laserbeam =>			set AbilityStruct = {FNAbilityData{Icon := Textures.ui_ability_beam, Title := "Beam", Cooldown := 15.0}}
			FNAbility.RiftGun =>			set AbilityStruct = {FNAbilityData{Icon := Textures.ui_ability_riftpistol, Title := "Rift Sucker", Cooldown := 12.0}}
			FNAbility.AirRaid =>			set AbilityStruct = {FNAbilityData{Icon := Textures.ui_ability_airraid, Title := "Air Raid"}}
			FNAbility.LastResort =>			set AbilityStruct = {FNAbilityData{Icon := Textures.ui_ability_flag, Title := "Last Resort"}}
			FNAbility.Engineer =>			set AbilityStruct = {FNAbilityData{Icon := Textures.ui_ability_shieldregen_machinist, Title := "Spawn Turret", Cooldown := 10.0}}
			FNAbility.PersonalShield =>		set AbilityStruct = {FNAbilityData{Icon := Textures.ui_ability_cubeassasin_shield, Title := "Spawn Shield", Cooldown := 10.0}}
			FNAbility.FootDive =>			set AbilityStruct = {FNAbilityData{Icon := Textures.ui_ability_footdive, Title := "Foot Dive", Cooldown := 7.0}}
			FNAbility.PartingGift =>		set AbilityStruct = {FNAbilityData{Icon := Textures.ui_ability_partinggift, Title := "Shield Bomb"}}
			FNAbility.Snap =>				set AbilityStruct = {FNAbilityData{Icon := Textures.ui_ability_snap, Title := "Snap", Cooldown := 10.0}}
			FNAbility.GrindVines =>			set AbilityStruct = {FNAbilityData{Icon := Textures.ui_ability_no, Title := "Grind Vines", Cooldown := 10.0}}
			FNAbility.TherapeuticSwim =>	set AbilityStruct = {FNAbilityData{Icon := Textures.ui_ability_therapeuticswim, Title := "Therapeutic Swim"}}
			FNAbility.Bombardment =>		set AbilityStruct = {FNAbilityData{Icon := Textures.ui_ability_bombardment, Title := "Eruption", Cooldown := 15.0}}
			FNAbility.DemonClaw =>			set AbilityStruct = {FNAbilityData{Icon := Textures.ui_ability_demonclaw, Title := "Demon's Slash", Cooldown := 7.0}}
			FNAbility.Puppeteer =>			set AbilityStruct = {FNAbilityData{Icon := Textures.ui_ability_puppeteer, Title := "Puppeteer"}}
			FNAbility.HealthRegen =>		set AbilityStruct = {FNAbilityData{Icon := Textures.ui_ability_shieldregen, Title := "HealthRegen"}}
			FNAbility.Enragement =>			set AbilityStruct = {FNAbilityData{Icon := Textures.ui_ability_alphaenrage, Title := "Enragement"}}
			FNAbility.ShortBeam =>			set AbilityStruct = {FNAbilityData{Icon := Textures.ui_ability_beam, Title := "Focused Beam", Cooldown := 10.0}}
			_ =>							#nothing
		}

		AbilityStruct
	}

	(Agent : agent).HasAbility(AbilityID:FNAbility)<transacts><decides>: void =
	{
		AgentHasAbility[Agent, AbilityID]
	}

	AgentHasAbility(Agent:agent, AbilityID:FNAbility)<transacts><decides>: void =
	{
		var bSuccess:logic = false
		if ( MainAbility := GetMainAbility[Agent], MainAbility = AbilityID )
		{
			set bSuccess = true
			bSuccess?
		}

		if ( MainAbility2 := GetMainAbility2[Agent], MainAbility2 = AbilityID )
		{
			set bSuccess = true
			bSuccess?
		}

		if ( PassiveList := GetPassiveList[Agent], PassiveList.Find[AbilityID] )
		{
			set bSuccess = true
			bSuccess?
		}

		bSuccess?
	}

	GetMaxHealth(Agent:agent):float=
	{
		if
		{
			ClassMaxHealth := PlayersMap[Agent].ClassData.MaxHealth
			BonusHealth := PlayersMap[Agent].HealthMult
		}
		then
		{
			return ClassMaxHealth * BonusHealth
		}

		return 100.0
	}

	GetMaxShield(Agent:agent):float=
	{
		if
		{
			ClassMaxShield := PlayersMap[Agent].ClassData.MaxShield
			BonusShield := PlayersMap[Agent].ShieldMult
		}
		then
		{
			return ClassMaxShield * BonusShield
		}

		return 100.0
	}

	GetDamageBonus(Agent:agent):float=
	{
		if
		{
			DamageBonus := PlayersMap[Agent].DamageMult
		}
		then
		{
			return DamageBonus
		}

		return 1.0
	}

	GetMaxUlt(Agent:agent):float=
	{
		if
		{
			MaxUlt := PlayersMap[Agent].ClassData.MaxUlt
		}
		then
		{
			return MaxUlt
		}

		return 100.0
	}

	SetAnimation(Agent:agent, PropAnim:creative_prop_asset, AnimDuration:float)<suspends>:void=
	{
		spawn{ModulePlayerEx.ChairAnim(Self,Agent,PropAnim,AnimDuration)}
	}

