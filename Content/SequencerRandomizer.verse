# Copyright Epic Games, Inc. All Rights Reserved.

using { /Fortnite.com/Devices }
using { /Verse.org/Random }
using { /Verse.org/Simulation }

################################################################################
## This device references an array of Cinematic Sequence Devices, and plays them randomly with a variable cooldown in between.
################################################################################
sequencer_randomization_device := class<concrete>(creative_device):

    # Array of Cinematic Sequence Devices to randomize between
    @editable   
    SequencerList : []cinematic_sequence_device = array{}

    # Minimum / Maximum time to wait before triggering another sequence
    @editable   
    CooldownMin:float = 120.0
    @editable   
    CooldownMax:float = 240.0

    # If we have more than 1 sequence to choose from, should we prevent the same sequence playing two times in a row?
    @editable   
    PreventBackToBackRepeats:logic = true 


    # OnBegin will run when the game starts
    OnBegin<override>()<suspends>:void=
        if (SequencerList.Length > 0):
            QueueSequence(GetRandomInt(0, SequencerList.Length-1))


    # Wait a random amount of time, then play the sequence
    QueueSequence<private>(Index:int)<suspends>:void=
        CooldownTime:float = GetRandomFloat(CooldownMin, CooldownMax)
        Sleep(CooldownTime)

        if (MySequencer := SequencerList[Index]):
            MySequencer.Play()

        QueueSequence(GetNextIndexToPlay(Index))

    
    # Decide the next sequence index, either completely random, or excluding the last-played sequence
    GetNextIndexToPlay<private>(LastIndex:int):int=
        if (SequencerList.Length > 1 and PreventBackToBackRepeats?):
            var IndexToReturn:int = LastIndex + GetRandomInt(1, SequencerList.Length-1)
            if(NewIndex := Mod[IndexToReturn, SequencerList.Length], set IndexToReturn = NewIndex):
                return IndexToReturn

        return GetRandomInt(0, SequencerList.Length-1)