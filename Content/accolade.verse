using { /Fortnite.com/Devices }
using { /Fortnite.com/Characters }
using { /Fortnite.com/Game }
using { /Verse.org/Simulation }
using { /Verse.org/Simulation/Tags }
using { /Verse.org/Random }
using { /UnrealEngine.com/Temporary/Diagnostics }

# A Verse-authored creative device that can be placed in a level
elimination_accolade_device := class(creative_device):

    @editable XP_Accolade : accolades_device = accolades_device{}

    # Runs when the device is started in a running game
    OnBegin<override>()<suspends>:void=
	{
        # Get all the players in the experience
        AllPlayers := GetPlayspace().GetPlayers()
        for (EliminationGamePlayer : AllPlayers)
		{
            if (FortCharacter := EliminationGamePlayer.GetFortCharacter[])
			{
                FortCharacter.EliminatedEvent().Subscribe(OnPlayerEliminated) # subscribe to eliminated event
			}
		}
	}

    # Event that handles when a player is eliminated
    OnPlayerEliminated(Result:elimination_result):void=
	{
		# Who got killed?
		KilledCharacter := Result.EliminatedCharacter

		if
		{
			# Let's make sure we can get the Agent of the fort_character                                                                    
			KilledAgent := KilledCharacter.GetAgent[]

			# Who did the killing? Why is there a question mark?
			# This fails if the player was killed by something other than
			# another fort_character. Ex. Damage Volumes
			KillerChar := Result.EliminatingCharacter?

            # Let's make sure we can get the Agent of the fort_character 
			KillerAgent := KillerChar.GetAgent[]
		}
        then
		{
			# Make sure they did not eliminate themself
			if(KillerAgent <> KilledAgent)
			{			
                XP_Accolade.Award(KillerAgent)
			}
		}
	}