using { /Verse.org }
using { /Verse.org/Native }
using { /Verse.org/Simulation }
using { /Verse.org/SpatialMath }
using { /Verse.org/Colors }
using { /Verse.org/SceneGraph }
using { /Verse.org/SceneGraph/KeyframedMovement }
using { /UnrealEngine.com/Temporary/SpatialMath }
using { /UnrealEngine.com/Temporary/Diagnostics }
using { /Fortnite.com/Characters }
using { /Fortnite.com/Teams }
using { /Fortnite.com/Game }
using { /Fortnite.com/Playspaces }
using { /Fortnite.com/FortPlayerUtilities }
using { /Fortnite.com/AI }
using { /Fortnite.com/Animation/PlayAnimation }
using { /Fortnite.com/Devices }
using { /Fortnite.com/Devices/CreativeAnimation }

# A Verse-authored component that can be added to entities
dootdevice_component<public> := class<final_super>(component):

	#@editable	var RespawnTime<public>:float = 420.0
	RespawnTime : float = 420.0
	@editable	Active : logic = true

	@editable ExplosiveDevice : explosive_device  = explosive_device{}
	@editable MapIndicator : map_indicator_device  = map_indicator_device{}
	@editable VFXPowerDoot : visual_effect_powerup_device = visual_effect_powerup_device{}

    var WeaponAgentOwner : ?agent = false
	var bFired : logic = false
	var SpawnVec : (/Verse.org/SpatialMath:)vector3 = (/Verse.org/SpatialMath:)vector3{}
	var SpawnRot : (/Verse.org/SpatialMath:)rotation = (/Verse.org/SpatialMath:)rotation{}

	BeamRange : float = 20000.0

    # Runs when the component should start simulating in a running game.
    OnBeginSimulation<override>():void =
	{
        # Run OnBeginSimulation from the parent class before
        # running this component's OnBeginSimulation logic
        (super:)OnBeginSimulation()

		if ( RoundM := Entity.GetFortRoundManager[] )
		{
			RoundM.SubscribeRoundStarted(OnRoundStarted)
			RoundM.SubscribeRoundEnded(OnRoundEnded)
		}
	}

    # Runs when the component should start simulating in a running game.
    # Can be suspended throughout the lifetime of the component. Suspensions
    # will be automatically cancelled when the component is disposed or the
    # game ends.
    OnSimulate<override>()<suspends>:void =
	{
        <#if (not Active?)
		{
			return
		}#>

		SpawnComponents()
	}

	OnRoundStarted():void =
	{
		set SpawnVec = Entity.GetGlobalTransform().Translation
		set SpawnRot = Entity.GetGlobalTransform().Rotation

		MapIndicator.Disable()
		EnableDootDevice()
	}

	OnRoundEnded():void =
	{
		Entity.RemoveFromParent()
	}

	SpawnComponents():void=
	{
		if ( InteractableComponent := Entity.GetComponent[interactable_component] )
		{
			InteractableComponent.SucceededEvent.Subscribe(OnInteractedEvent)
		}

		set SpawnVec = Entity.GetGlobalTransform().Translation
		set SpawnRot = Entity.GetGlobalTransform().Rotation
	}

	SpawnFXAt(SpawnPos: (/Verse.org/SpatialMath:)vector3)<suspends>:void=
	{
		if ( SimulationEntity := Entity.GetSimulationEntity[] )
		{
			SpawnTransform := (/Verse.org/SpatialMath:)transform{Translation:=SpawnPos}
			NE := entity{}
			TransComp := transform_component{ Entity := NE }
			VFXComp := FX.NS_DOOT_Impact_Burst{ Entity := NE, AutoPlay := true }
			NE.AddComponents(array{TransComp,VFXComp})
			SimulationEntity.AddEntities(array{NE})
			NE.SetGlobalTransform(SpawnTransform)
			Sleep(0.1)
			NE.RemoveFromParent()
		}
	}

	SpawnFXAtAgent(Agent: agent)<suspends>:void=
	{
		if (FC := Agent.GetFortCharacter[] )
		{
			PlayerT := FromVector3(FC.GetTransform().Translation)
			SpawnFXAt(PlayerT)
		}
	}

    OnInteractedEvent(Agent:agent):void =
	{
        Print("OnInteractedEvent()")
		set WeaponAgentOwner = option{Agent}
		#spawn{SpawnFXAtAgent(Agent)}
		spawn{UseDootDevice()}

		if ( InteractableComponent := Entity.GetComponent[interactable_component] )
		{
			InteractableComponent.Disable()
		}

		if ( MeshComp := Entity.GetComponent[mesh_component] )
		{
			set MeshComp.Collidable = false
			set MeshComp.Queryable = false
		}
		MapIndicator.Disable()

		for ( EntPickup : Entity.FindDescendantEntitiesWithComponent(FX.NS_DoomsdayAir), VFXPickup := EntPickup.GetComponent[FX.NS_DoomsdayAir], SFXPickup := EntPickup.GetComponent[sound_component] )
		{
			VFXPickup.Stop()
			VFXPickup.Disable()
			SFXPickup.Stop()
		}
	}

	UseDootDevice()<suspends>:void=
	{
		if (Agent := agent[WeaponAgentOwner?], FortCharacter := Agent.GetFortCharacter[] )
		{
			race
			{
				block
				{
					set bFired = false

					spawn{ AimDOOTDevice() }
					spawn{ ChargeDootDevice() }

					Sleep(5.0)
					spawn{FireDootDevice()}
					Sleep(12.25)
					spawn{StopDootDevice()}
				}
				block
				{
					FortCharacter.EliminatedEvent().Await()
					spawn{StopDootDevice()}
				}
			}
		}
	}

    ChargeDootDevice()<suspends>:void =
	{
		Print("DOOT Device: CHARGING...")
		for ( EntCharge : Entity.FindDescendantEntitiesWithComponent(FX.NS_DOOTDeviceCharge), VFXCharge := EntCharge.GetComponent[FX.NS_DOOTDeviceCharge], SFXCharge := EntCharge.GetComponent[sound_component] )
		{
			VFXCharge.Play()
			VFXCharge.Enable()
			SFXCharge.Play()
		}
	}

    FireDootDevice()<suspends>:void =
	{
		Print("DOOT Device: UNLEASHED")

		if ( Agent := WeaponAgentOwner? )
		{
			spawn{SpawnFXAtAgent(Agent)}
		}

		for ( EntCharge : Entity.FindDescendantEntitiesWithComponent(FX.NS_DOOTDeviceCharge), VFXCharge := EntCharge.GetComponent[FX.NS_DOOTDeviceCharge], SFXCharge := EntCharge.GetComponent[sound_component] )
		{
			VFXCharge.Stop()
			SFXCharge.Stop()
		}

		if ( SFXBeam := Entity.GetComponent[sound_component] )
		{
			SFXBeam.Play()
		}

		for ( EntBeam : Entity.FindDescendantEntitiesWithComponent(FX.NS_DOOTDevice_Ribbon), VFXBeam := EntBeam.GetComponent[FX.NS_DOOTDevice_Ribbon] )
		{
			#Print("DOOT Device: BEAM")
			VFXBeam.Play()
			VFXBeam.Enable()

			CollisionSphere := collision_sphere{ Radius := 100.0 <#, CollisionProfile := collision_profiles.StationaryOverlapAll#> }
			#CollisionBox := collision_box{ Extents := (/Verse.org/SpatialMath:)vector3{Forward := 100.0, Left := 100.0, Up := 100.0} }
			DebugDraw : debug_draw = debug_draw{}
			loop
			{
				if (bFired?)
				{
					break
				}
				#Print("DOOT Device: BEAM!")

				if ( Agent := agent[WeaponAgentOwner?], FortChar := Agent.GetFortCharacter[] )
				{
					SweepStart := EntBeam.GetGlobalTransform()
					SweepEnd := ( FromTransform(EntBeam.GetGlobalTransform()).Rotation.GetLocalForward() * BeamRange )

					PlayerForwardPos := ( FromVector3(EntBeam.GetGlobalTransform().Translation) + FromTransform(EntBeam.GetGlobalTransform()).Rotation.GetLocalForward() * BeamRange )
					UpdateBeam(FromVector3(PlayerForwardPos))

					#var IsHit : logic = false
					for ( SweepHit : EntBeam.FindSweepHits(FromVector3(SweepEnd), SweepStart, CollisionSphere), HitPos := SweepHit.ContactPosition, TargetComponent := SweepHit.TargetComponent, TargetEntity := TargetComponent.Entity, MeshComp := mesh_component[TargetComponent] )
					{
						#HitNormal := SweepHit.ContactFaceNormal
						UpdateBeam(HitPos <#, ?EndNormal := HitNormal#>)

						<#if ( ExplosiveDevice.TeleportTo[VFXBeam.StopLocation, (/UnrealEngine.com/Temporary/SpatialMath:)rotation{}] )
						{
							ExplosiveDevice.Explode(Agent)
						}#>

						#DebugDraw.DrawLine(FromVector3(SweepHit.SourceHitTranslation), FromVector3(HitPos), ?DrawDurationPolicy := debug_draw_duration_policy.FiniteDuration, ?Duration := 0.1)

						if ( PlayerMesh := TargetEntity.GetComponent[Meshes.PlayerBox] )
						{
							#Print("Beam Hit: Player")
							spawn{ DootAgent(TargetEntity) }
							spawn{SpawnFXAt(HitPos)}
						}
					}
				}

				Sleep(0.1)
			}
		}
	}

    DootAgent(TargetEntity:entity)<suspends>:void =
	{
		if
		{
			Agent := agent[WeaponAgentOwner?],
			FortChar := Agent.GetFortCharacter[],
			FortChar.IsActive[],
			Playspace := Entity.GetPlayspaceForEntity[]
			TeamCollection := Playspace.GetTeamCollection()
			TeamOwner := TeamCollection.GetTeam[Agent],
			DamageInstigator := Agent.GetInstigator(),
			DamageSource := game_action_causer[FortChar]
		}
		then
		{
			if ( PlayerComp := TargetEntity.GetComponent[fn_player_component], AgentVictim := PlayerComp.Agent, FortVictim := AgentVictim.GetFortCharacter[], not TeamCollection.IsOnTeam[AgentVictim, TeamOwner] )
			{
				BeamDmg := damage_args{Instigator := option{DamageInstigator}, Source := option{DamageSource}, Amount:=25.0}
				FortVictim.Damage(BeamDmg)
				VFXPowerDoot.Pickup(Agent)
			}

			if ( NPCComp := TargetEntity.GetComponent[fn_npc_component], AgentNPC := NPCComp.Agent, FortVictim := AgentNPC.GetFortCharacter[], not TeamCollection.IsOnTeam[AgentNPC, TeamOwner] )
			{
				BeamDmg := damage_args{Instigator := option{DamageInstigator}, Source := option{DamageSource}, Amount:=100.0}
				FortVictim.Damage(BeamDmg)
			}
		}
	}

    UpdateBeam(EndPos:(/Verse.org/SpatialMath:)vector3, ?EndNormal:(/Verse.org/SpatialMath:)vector3 = (/Verse.org/SpatialMath:)vector3{} ):void =
	{
		for ( EntBeam : Entity.FindDescendantEntitiesWithComponent(FX.NS_DOOTDevice_Ribbon), VFXBeam := EntBeam.GetComponent[FX.NS_DOOTDevice_Ribbon] )
		{
			#Print("DOOT Device: BEAM")

			SweepEnd := EndPos
			SweepEndNormal := EndNormal

			set VFXBeam.StopLocation = SweepEnd
			set VFXBeam.StopNormal = SweepEndNormal
		}
	}

    StopDootDevice()<suspends>:void =
	{
		Print("DOOT Device: Exhausted...")

		set bFired = true
		DisableDootDevice()

		RespawnDootDevice()
	}

    RespawnDootDevice()<suspends>:void =
	{
		Print("DOOT Device: Despawning...")
		Sleep(RespawnTime)
		Print("DOOT Device: Respawning...")
		NewTransform := (/Verse.org/SpatialMath:)transform{ Translation := SpawnVec, Rotation := SpawnRot}
		Entity.SetGlobalTransform(NewTransform)
		spawn{ ModulePropExtra.SpawnParticleVFX(FX.NS_DOOT_Impact_Burst_asset, FromVector3(Entity.GetGlobalTransform().Translation), (/UnrealEngine.com/Temporary/SpatialMath:)rotation{}, 1.2) }
		EnableDootDevice()
	}

    EnableDootDevice():void =
	{
		Print("DOOT Device: Enable")
		if ( InteractableComp := Entity.GetComponent[interactable_component] )
		{
			InteractableComp.Enable()
		}

		if ( MapIndicator.TeleportTo[FromVector3(Entity.GetGlobalTransform().Translation), (/UnrealEngine.com/Temporary/SpatialMath:)rotation{}] ) {}
		MapIndicator.Enable()

		if ( MeshComp := Entity.GetComponent[mesh_component] )
		{
			set MeshComp.Collidable = true
			set MeshComp.Queryable = true
			set MeshComp.Visible = true
		}

		for ( EntPickup : Entity.FindDescendantEntitiesWithComponent(FX.NS_DoomsdayAir), VFXPickup := EntPickup.GetComponent[FX.NS_DoomsdayAir], SFXPickup := EntPickup.GetComponent[sound_component] )
		{
			VFXPickup.Enable()
			VFXPickup.Play()
			SFXPickup.Play()
		}
	}

    DisableDootDevice():void =
	{
		Print("DOOT Device: Disable")

		if ( InteractableComp := Entity.GetComponent[interactable_component] )
		{
			InteractableComp.Disable()
		}

		if ( SFXBeam := Entity.GetComponent[sound_component] )
		{
			SFXBeam.Stop()
		}

		for ( EntBeam : Entity.FindDescendantEntitiesWithComponent(FX.NS_DOOTDevice_Ribbon), VFXBeam := EntBeam.GetComponent[FX.NS_DOOTDevice_Ribbon] )
		{
			VFXBeam.Stop()
			#VFXBeam.Disable()
		}

		if (KeyframedComp := Entity.GetComponent[keyframed_movement_component])
		{
			KeyframedComp.Stop()
		}

		for ( EntCharge : Entity.FindDescendantEntitiesWithComponent(FX.NS_DOOTDeviceCharge), VFXCharge := EntCharge.GetComponent[FX.NS_DOOTDeviceCharge], SFXCharge := EntCharge.GetComponent[sound_component] )
		{
			VFXCharge.Stop()
			SFXCharge.Stop()
		}

		for ( EntPickup : Entity.FindDescendantEntitiesWithComponent(FX.NS_DoomsdayAir), VFXPickup := EntPickup.GetComponent[FX.NS_DoomsdayAir], SFXPickup := EntPickup.GetComponent[sound_component] )
		{
			VFXPickup.Stop()
			VFXPickup.Disable()
			SFXPickup.Stop()
		}

		if ( MeshComp := Entity.GetComponent[mesh_component] )
		{
			set MeshComp.Collidable = false
			set MeshComp.Queryable = false
			set MeshComp.Visible = false
		}

		MapIndicator.Disable()

		set WeaponAgentOwner = false

		spawn{ ModulePropExtra.SpawnParticleVFX(FX.NS_DOOT_Impact_Burst_asset, FromVector3(Entity.GetGlobalTransform().Translation), (/UnrealEngine.com/Temporary/SpatialMath:)rotation{}, 1.2) }
	}

    AimDOOTDevice()<suspends>:void=
	{
		if (Agent := agent[WeaponAgentOwner?], FortCharacter := Agent.GetFortCharacter[] )
		{
			PlayerPos := FortCharacter.GetTransform().Translation + FortCharacter.GetViewRotation().GetLocalForward() * 300.0
			PlayerRot := FortCharacter.GetViewRotation()
			NewTransform := (/Verse.org/SpatialMath:)transform{ Translation := FromVector3(PlayerPos), Rotation := FromRotation(PlayerRot), Scale := (/Verse.org/SpatialMath:)vector3{Forward := 1.0, Left := 1.0, Up := 1.0}}
			Entity.SetGlobalTransform(NewTransform)

			loop
			{
				Sleep(0.4)

				if (bFired?)
				{
					break
				}

				if ( not FortCharacter.IsActive[] )
				{
					break
				}

				PlayerNewPos := FortCharacter.GetTransform().Translation + FortCharacter.GetViewRotation().GetLocalForward() * 200.0
				NewRotation := FortCharacter.GetViewRotation()
				NewKeyTransform := (/Verse.org/SpatialMath:)transform{ 
					Translation := FromVector3(PlayerNewPos) - Entity.GetGlobalTransform().Translation,
					Rotation := FromRotation(MakeComponentWiseDeltaRotation(NewRotation, FromRotation(Entity.GetGlobalTransform().Rotation))),
					Scale := (/Verse.org/SpatialMath:)vector3{Forward  := 0.0, Left := 0.0, Up := 0.0}
				}

				if (AController := Entity.GetComponent[keyframed_movement_component])
				{
					AnimKeyframes := array
					{
						keyframed_movement_delta
						{
							Transform := NewKeyTransform
							Duration := 0.4
							Easing := linear_easing_function{}
						}
					}
					AController.Stop()
					AController.SetKeyframes(AnimKeyframes, oneshot_keyframed_movement_playback_mode{})
					AController.Play()
				}
			}
		}
	}
