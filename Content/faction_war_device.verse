<#
	2AGENCY
	Faction War
#>
using { /Fortnite.com/Devices }
using { /Verse.org/Simulation }
using { ModulePlayerClasses }

# A Verse-authored creative device that can be placed in a level
faction_war_device := class(creative_device):

	var FactionKillCount: [FNFactions]int = map{}

	@editable CineCube1 : cinematic_sequence_device = cinematic_sequence_device{}
	@editable CineCube2 : cinematic_sequence_device = cinematic_sequence_device{}
	@editable CineCube3 : cinematic_sequence_device = cinematic_sequence_device{}

	@editable CineIO1 : cinematic_sequence_device = cinematic_sequence_device{}
	@editable CineIO2 : cinematic_sequence_device = cinematic_sequence_device{}

	@editable CineWasteland2 : cinematic_sequence_device = cinematic_sequence_device{}

    OnBegin<override>()<suspends>:void=
	{
		<#for (Key -> Value : FactionKillCount)
		{
			if { set FactionKillCount[Key] = 0 }
		}#>

		#if { set FactionKillCount[FNFactions.None] = 0 }
		if { set FactionKillCount[FNFactions.IO] = 0 }
		if { set FactionKillCount[FNFactions.LastReality] = 100 }
		if { set FactionKillCount[FNFactions.Wasteland] = 0 }
	}

	AddFactionKillCount<public>(ClassFaction:FNFactions):void=
	{
		Print("AddFactionKillCount()")
		if ( set FactionKillCount[ClassFaction] += 1, NewKillCount := FactionKillCount[ClassFaction] )
		{
			case( ClassFaction )
			{
				FNFactions.Spies =>			Print("Faction (Spies) Kill Count: {NewKillCount}")
				FNFactions.Seven =>			Print("Faction (Seven) Kill Count: {NewKillCount}")
				FNFactions.IO =>			Print("Faction (IO) Kill Count: {NewKillCount}")
				FNFactions.LastReality =>	Print("Faction (Last Reality) Kill Count: {NewKillCount}")
				FNFactions.Society =>		Print("Faction (Society) Kill Count: {NewKillCount}")
				FNFactions.Pantheon =>		Print("Faction (Pantheon) Kill Count: {NewKillCount}")
				FNFactions.Wasteland =>		Print("Faction (Wasteland) Kill Count: {NewKillCount}")
				_ =>						Print("Faction (None) Kill Count: {NewKillCount}")
			}

			if ( NewKillCount = 50 )
			{
				case( ClassFaction )
				{
					FNFactions.IO =>
					{
						Print("IO reached 50 Kills!")
						UpdateFactionOnMap(FNFactions.IO, 1)
					}
					FNFactions.LastReality =>
					{
						Print("Last Reality reached 50 Kills!")
						UpdateFactionOnMap(FNFactions.LastReality, 1)
					}
					_ =>						#nothing
				}
			}
			else if ( NewKillCount = 75 )
			{
				case( ClassFaction )
				{
					FNFactions.IO =>
					{
						Print("IO reached 75 Kills!")
						UpdateFactionOnMap(FNFactions.IO, 2)
					}
					FNFactions.LastReality =>
					{
						Print("Last Reality reached 75 Kills!")
						UpdateFactionOnMap(FNFactions.LastReality, 2)
					}
					FNFactions.Wasteland =>
					{
						Print("Wasteland reached 75 Kills!")
						UpdateFactionOnMap(FNFactions.Wasteland, 2)
					}
					_ =>						#nothing
				}
			}
			else if ( NewKillCount = 100 )
			{
				case( ClassFaction )
				{
					FNFactions.IO =>
					{
						Print("IO reached 100 Kills!")
						UpdateFactionOnMap(FNFactions.IO, 3)
					}
					FNFactions.LastReality =>
					{
						Print("Last Reality reached 100 Kills!")
						UpdateFactionOnMap(FNFactions.LastReality, 3)
					}
					FNFactions.Wasteland =>
					{
						Print("Wasteland reached 100 Kills!")
						UpdateFactionOnMap(FNFactions.Wasteland, 3)
					}
					_ =>						#nothing
				}
			}
		}
	}

	UpdateFactionOnMap(ClassFaction:FNFactions, FactionStage:int):void=
	{
		case( ClassFaction )
		{
			FNFactions.Spies =>			Print("Map Update: Spies Stage {FactionStage}")
			FNFactions.Seven =>			Print("Map Update: The Seven Stage {FactionStage}")
			FNFactions.IO =>
			{
				Print("Map Update: IO Stage {FactionStage}")
				case( FactionStage )
				{
					1 =>			CineIO1.Play()
					2 =>			CineIO2.Play()
					_ =>			#nothing
				}
			}
			FNFactions.LastReality =>
			{
				Print("Map Update: Last Reality Stage {FactionStage}")
				case( FactionStage )
				{
					1 =>			CineCube1.Play()
					2 =>			CineCube2.Play()
					3 =>			CineCube3.Play()
					_ =>			#nothing
				}
			}
			FNFactions.Society =>		Print("Map Update: Society Stage {FactionStage}")
			FNFactions.Pantheon =>		Print("Map Update: Pantheon Stage {FactionStage}")
			FNFactions.Wasteland =>
			{
				Print("Map Update: Wasteland Stage {FactionStage}")
				case( FactionStage )
				{
					2 =>			CineWasteland2.Play()
					_ =>			#nothing
				}
			}
			_ =>						#nothing
		}
	}
