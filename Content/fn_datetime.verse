<#
	Real Time Clock
#>
using { /Fortnite.com/Devices }
using { /Verse.org/Simulation }
using { /UnrealEngine.com/Temporary/Diagnostics }
using { /Verse.org/Random }
 
date_time := struct:
    year<public>: int
    month<public>: int
    day<public>: int
    hour<public>: int
    minute<public>: int
    second<public>: int
 
Date_Time_Provider := class:  
    #Made By Himez 
    #Discord : Himez
    #Twitter : @HimezUEFN
    #This class is used to get the current date and time from the seconds since epoch
 
    # Constants
    SecondsInMinute: int= 60
    SecondsInHour: int= 3600
    SecondsInDay: int= 86400
    SecondsInYear: int= 31536000
    SecondsInLeapYear: int= 31622400
 
    # Init date and time variables
    var Year: int= 1970
    var Month: int= 1
    var Day: int= 1
    var Hour: int= 0
    var Minute: int= 0
    var Second: int= 0
    var EpochSeconds: int= 0
    
    # Get the current seconds since epoch
    GetMonthName<public>(MonthInt:int):string=
        Months : []string = array{"January", "February", "March", "April", "May", "June", "July", "August", "September", "October", "November", "December"}
        if. MonthName := Months[MonthInt - 1] then. return MonthName  else. return "Invalid Month"
 
    #Get Date and Time method
    GetDateAndTime<public>():date_time=
        ResetDateAndTime()
        CurrentEpochsSeconds  :=  GetSecondsSinceEpoch()
        if.  Temp := Floor[CurrentEpochsSeconds] then. set EpochSeconds = Temp
        CalculateDateAndTime()
        return date_time{year:= Year, month:= Month, day:= Day, hour:= Hour, minute:= Minute, second:= Second}
 
    # caculate day and time
    CalculateDateAndTime<private>():void=
        CaculateYear()
        CalculateDay()
        CalculateMonth()
        CalculateHour()
        CalculateMinute()
        CalculateSecond()
 
    # Reset date and time
    ResetDateAndTime<private>():void=
       set Year = 1970
       set Month = 1
       set Day = 1
       set Hour = 0
       set Minute = 0
       set Second = 0
       set EpochSeconds = 0
 
    # Determine the year
    CaculateYear<private>():void =
        loop:
            LeapYear := IsLeapYear(Year)
            SecondsInCurrentYear := if. LeapYear? then. SecondsInLeapYear else. SecondsInYear
            if (EpochSeconds >= SecondsInCurrentYear):
                set EpochSeconds -= SecondsInCurrentYear
                set Year += 1
            else:
                break
     
    #Helper function to determine if a year is a leap year
    IsLeapYear<private>(YearInt: int):logic=
        if. Mod[YearInt , 4 ] = 0 and Mod[YearInt ,100 ] <> 0 or Mod[ YearInt , 400] = 0 then. return true else. return false
    
    # Calculate day
    CalculateDay<private>(): void =
        set Day = if. Temp := Floor(EpochSeconds / SecondsInDay) then. Temp + 1 else. 1
        if. Temp := Mod[EpochSeconds, SecondsInDay] then. set EpochSeconds = Temp  # Calculate  remaining seconds for time of day 
 
    # Calculate month
    CalculateMonth<private>(): void = 
        var DaysInMonth: []int = array{31, 28, 31, 30, 31, 30, 31, 31, 30, 31, 30, 31}
        LeapYear := IsLeapYear(Year) # Check if it is a leap year
        if. LeapYear? then. if. set DaysInMonth[1] = 29  # Adjust February for leap year
        loop:
            if (Day > DaysInMonth[Month - 1]):
                if. set Day -= DaysInMonth[Month - 1]
                set Month += 1       
            else:
                break
            
    # Calculate hour
    CalculateHour<private>():void=
        set Hour = if. Temp := Floor(EpochSeconds / SecondsInHour) then. Temp else. 0
        set EpochSeconds = if . Temp := Mod[EpochSeconds, SecondsInHour] then. Temp else. 0 # Calculate remaining secondsq
 
    # Calculate minute
    CalculateMinute<private>():void=
        set Minute = if. Temp := Floor(EpochSeconds / SecondsInMinute) then. Temp else. 0
        set EpochSeconds = if. Temp := Mod[EpochSeconds, SecondsInMinute] then. Temp else. 0 # Calculate remaining seconds
 
    # Remaining seconds
    CalculateSecond<private>():void=
        set Second += EpochSeconds
        set EpochSeconds = 0
 
# Example of how to use the Date_Time_Provider class --------------------------
<#
DateAndTime:= class(creative_device):
    S2M<localizes>(Value:string):message =  "{Value}"
    @editable Date : billboard_device = billboard_device{}
    @editable Time : billboard_device = billboard_device{}
    DateTimeProvider : Date_Time_Provider = Date_Time_Provider{}
    OnBegin<override>()<suspends>:void=    
             loop:
                DateTime := DateTimeProvider.GetDateAndTime()
                MonthName := DateTimeProvider.GetMonthName(DateTime.month) 
                Date.SetText(S2M("{MonthName} {DateTime.day}, {DateTime.year}"))
                Time.SetText(S2M("UTC: {DateTime.hour}:{DateTime.minute}:{DateTime.second}"))
                Sleep(1.0)
#>

snowprogress_device:= class(creative_device):
    #S2M<localizes>(Value:string):message =  "{Value}"

    DateTimeProvider : Date_Time_Provider = Date_Time_Provider{}

	@editable CineSnow1 : cinematic_sequence_device = cinematic_sequence_device{}	# work
	@editable CineSnow2 : cinematic_sequence_device = cinematic_sequence_device{}	# not work
	@editable CineSnow3 : cinematic_sequence_device = cinematic_sequence_device{}	# not work
	@editable CineSnow4 : cinematic_sequence_device = cinematic_sequence_device{}	# not work
	@editable CineSnow5 : cinematic_sequence_device = cinematic_sequence_device{}	# not work

    OnBegin<override>()<suspends>:void=    
	{
		DateTime := DateTimeProvider.GetDateAndTime()
		MonthName := DateTimeProvider.GetMonthName(DateTime.month)
		Print("{MonthName} {DateTime.day}, {DateTime.year}")

		case( MonthName )
		{
			"December" =>
			{
				if ( DateTime.day >= 11 )
				{
					Print("Snow Stage: 5")
					CineSnow5.Play()
				}
				else
				{
					case( DateTime.day )
					{
						7 =>
						{
							Print("Snow Stage: 1")
							CineSnow1.Play()
						}
						8 =>
						{
							Print("Snow Stage: 2")
							CineSnow2.Play()
						}
						9 =>
						{
							Print("Snow Stage: 3")
							CineSnow3.Play()
						}
						10 =>
						{
							Print("Snow Stage: 4")
							CineSnow4.Play()
						}
						_ =>
						{
							Print("Snow Stage: 0")
						}
					}
				}
			}
			"January" =>
			{
				if ( DateTime.day <= 7 )
				{
					Print("Snow Stage: 5")
					CineSnow5.Play()
				}
				else
				{
					case( DateTime.day )
					{
						8 =>
						{
							Print("Snow Stage: 4")
							CineSnow4.Play()
						}
						9 =>
						{
							Print("Snow Stage: 3")
							CineSnow3.Play()
						}
						10 =>
						{
							Print("Snow Stage: 2")
							CineSnow2.Play()
						}
						11 =>
						{
							Print("Snow Stage: 1")
							CineSnow1.Play()
						}
						_ =>
						{
							Print("Snow Stage: 0")
						}
					}
				}
			}
			_ =>	#nothing
		}
	}
 