<#
	2AGENCY
	Game Core, Manager, Rules and Settings.
#>
using { /Fortnite.com/Game }
using { /Fortnite.com/Devices }
using { /Fortnite.com/Characters }
using { /Verse.org/Simulation }
using { /Verse.org/Simulation/Tags }
using { ModulePlayerClasses }

player_spawner_tag := class(tag){}

# A Verse-authored creative device that can be placed in a level
fn_gamerules_device := class(creative_device):

	@editable PlayerClassesDevice : ?player_classes_manager_device = false
	@editable PlayerAbilityDevice : ?player_ability_device = false
	@editable PlayerUltDevice : ?player_ability_ultimate_device = false
	@editable UIClassMenuDevice : ?ui_classmenu_manager_device = false
	@editable UIVoiceMenuDevice : ?ui_voicemenu_manager_device = false
	@editable UIMainMenuDevice : ?ui_mainmenu_device = false
	@editable UIAchievementDevice : ?ui_achievements_device = false
	#@editable LoadoutDevice : ?loadout_device = false
	@editable SaveManagerDevice : ?save_manager_device = false
	@editable TeamDeathmatchDevice : ?fn_game_tdm = false

	@editable GameplayZone : mutator_zone_device = mutator_zone_device{}
	@editable InputFire : input_trigger_device = input_trigger_device{}

    # Runs when the device is started in a running game
    OnBegin<override>()<suspends>:void=
	{
		Print("Initialize Gamerules...")

		GetPlayspace().PlayerAddedEvent().Subscribe(OnPlayerJoined)
		GetPlayspace().PlayerRemovedEvent().Subscribe(OnPlayerRemoved)

        # Get all the players in the experience
		#AllPlayers := GetPlayspace().GetPlayers()
		AllPlayers := GetPlayspace().GetParticipants()
        for (NewGamePlayer : AllPlayers)
		{
			OnPlayerJoined(NewGamePlayer)
		}

        TaggedSpawners := FindCreativeObjectsWithTag(player_spawner_tag{})
        for (FoundSpawners: TaggedSpawners, PlayerSpawners := player_spawner_device[FoundSpawners])
		{
			PlayerSpawners.SpawnedEvent.Subscribe(OnPlayerSpawned)
		}

		GameplayZone.AgentBeginsEmotingEvent.Subscribe(OnPlayerBeginsEmoting)
		GameplayZone.AgentEndsEmotingEvent.Subscribe(OnPlayerEndEmoting)

		InputFire.PressedEvent.Subscribe(InputPressedFire)
	}

    OnPlayerJoined(Agent: agent): void =
	{
		if (FortCharacter := Agent.GetFortCharacter[])
		{
			FortCharacter.EliminatedEvent().Subscribe(OnPlayerEliminated)
			FortCharacter.DamagedEvent().Subscribe(OnPlayerDamaged)
			FortCharacter.DamagedShieldEvent().Subscribe(OnPlayerDamaged)
			FortCharacter.HealedEvent().Subscribe(OnPlayerHealed)
			FortCharacter.HealedShieldEvent().Subscribe(OnPlayerHealed)
			FortCharacter.CrouchedEvent().Subscribe(OnPlayerCrouched)
			FortCharacter.SprintedEvent().Subscribe(OnPlayerSprinted)
			FortCharacter.JumpedEvent().Subscribe(OnPlayerJumped)
		}

		if ( PlayerClasses := PlayerClassesDevice? )
		{
			PlayerClasses.OnPlayerJoined(Agent)
		}

		#UIClassMenuDevice.UIPlayerJoined(Agent)

		if (Player := player[Agent], SaveManager := SaveManagerDevice?)
		{
			SaveManager.InitializePlayerSaves(Player)
		}

		if ( TeamDeathmatch := TeamDeathmatchDevice? )
		{
			TeamDeathmatch.OnPlayerJoined(Agent)
		}
	}

    OnPlayerRemoved(Agent: agent): void =
	{
		if ( PlayerClasses := PlayerClassesDevice? )
		{
			PlayerClasses.OnPlayerRemoved(Agent)
		}

		if ( UIClassMenu := UIClassMenuDevice? )
		{
			UIClassMenu.UIPlayerRemoved(Agent)
		}

		if ( UIVoiceMenu := UIVoiceMenuDevice? )
		{
			UIVoiceMenu.UIPlayerRemoved(Agent)
		}

		if ( UIMainMenu := UIMainMenuDevice? )
		{
			UIMainMenu.UIPlayerRemoved(Agent)
		}

		if ( UIAchievement := UIAchievementDevice? )
		{
			UIAchievement.UIPlayerRemoved(Agent)
		}

		<#if ( UILoadout := LoadoutDevice? )
		{
			UILoadout.UIPlayerRemoved(Agent)
		}#>

		if ( TeamDeathmatch := TeamDeathmatchDevice? )
		{
			TeamDeathmatch.OnPlayerRemoved(Agent)
		}
	}

	OnPlayerSpawned(Agent:agent):void=
	{
		if ( PlayerClasses := PlayerClassesDevice? )
		{
			PlayerClasses.OnPlayerSpawned(Agent)
		}

		<#if ( PlayerLoadout := LoadoutDevice? )
		{
			PlayerLoadout.OnPlayerSpawned(Agent)
		}#>
	}

	OnPlayerEliminated(Result:elimination_result):void=
	{
		if ( PlayerClasses := PlayerClassesDevice? )
		{
			PlayerClasses.OnPlayerEliminated(Result)
		}

		if ( PlayerAbility := PlayerAbilityDevice? )
		{
			PlayerAbility.OnPlayerEliminated(Result)
		}

		if ( TeamDeathmatch := TeamDeathmatchDevice? )
		{
			TeamDeathmatch.OnPlayerEliminated(Result)
		}
	}

	OnPlayerDamaged(Result:damage_result):void=
	{
		if ( PlayerClasses := PlayerClassesDevice? )
		{
			PlayerClasses.OnPlayerDamaged(Result)
		}
	}

	OnPlayerShieldDamaged(Result:damage_result):void=
	{
		if ( PlayerClasses := PlayerClassesDevice? )
		{
			PlayerClasses.OnPlayerDamaged(Result)
		}
	}

	OnPlayerHealed(Result:healing_result):void=
	{
		if ( PlayerClasses := PlayerClassesDevice? )
		{
			PlayerClasses.OnPlayerHealed(Result)
		}
	}

	OnPlayerCrouched(FortChar:fort_character, IsCrounch:logic):void= 
	{
		if ( PlayerClasses := PlayerClassesDevice? )
		{
			PlayerClasses.OnPlayerCrouched(FortChar, IsCrounch)
		}
	}

	OnPlayerSprinted(FortChar:fort_character, IsSprinted:logic):void= 
	{
		if ( PlayerClasses := PlayerClassesDevice? )
		{
			PlayerClasses.OnPlayerSprinted(FortChar, IsSprinted)
		}
	}

	OnPlayerJumped(FortChar:fort_character):void= 
	{
		if ( PlayerClasses := PlayerClassesDevice? )
		{
			PlayerClasses.OnPlayerJumped(FortChar)
		}
	}

	OnPlayerBeginsEmoting(Agent:agent):void= 
	{
		if ( PlayerClasses := PlayerClassesDevice? )
		{
			PlayerClasses.OnPlayerBeginsEmoting(Agent)
		}
		if ( PlayerUltimate := PlayerUltDevice? )
		{
			PlayerUltimate.OnPlayerBeginsEmoting(Agent)
		}
	}

	OnPlayerEndEmoting(Agent:agent):void= 
	{
		if ( PlayerClasses := PlayerClassesDevice? )
		{
			PlayerClasses.OnPlayerEndEmoting(Agent)
		}
	}

	InputPressedFire(Agent : agent) : void =
	{
		if ( PlayerClasses := PlayerClassesDevice? )
		{
			PlayerClasses.InputPressedFire(Agent)
		}
	
		if ( PlayerAbility := PlayerAbilityDevice? )
		{
			PlayerAbility.InputPressedFireAbility(Agent)
		}

		if ( UIClassMenu := UIClassMenuDevice? )
		{
			UIClassMenu.InputPressedFireUI(Agent)
		}

		<#if ( UILoadout := LoadoutDevice? )
		{
			UILoadout.InputPressedFireUI(Agent)
		}#>
	}
