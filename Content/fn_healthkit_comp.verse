using { /Verse.org }
using { /Verse.org/Native }
using { /Verse.org/SceneGraph }
using { /Verse.org/Simulation }
using { /Fortnite.com/Characters }

# A Verse-authored component that can be added to entities
fn_healthkit_comp<public> := class<final_super>(component):

    # A custom variable you can expose to the editor
    @editable
    var HealthPercent<public>:float = 50.0
  
    # Runs when the component should start simulating in a running game.
    OnBeginSimulation<override>():void =
	{
        # Run OnBeginSimulation from the parent class before
        # running this component's OnBeginSimulation logic
        (super:)OnBeginSimulation()

        # TODO: Place logic to run when the component starts simulating here
        Print("OnBeginSimulation")
	}

    # Runs when the component should start simulating in a running game.
    # Can be suspended throughout the lifetime of the component. Suspensions
    # will be automatically cancelled when the component is disposed or the
    # game ends.
    OnSimulate<override>()<suspends>:void =
	{
        # TODO: Place simple suspends logic to run for this component here
        Print("OnSimulate")

		if ( MeshComp := Entity.GetComponent[mesh_component] )
		{
			MeshComp.EntityEnteredEvent.Subscribe(OnEntityEnteredEvent)
			MeshComp.EntityExitedEvent.Subscribe(OnEntityExitedEvent)
		}
	}


    OnEntityEnteredEvent(HitEnt:entity):void =
	{
        Print("HealthKit: OnEntityEnteredEvent")
		<#if ( PlayerComponent := HitEnt.GetComponent[fn_player_comp], Agent := PlayerComponent.Agent, FortCharacter := Agent.GetFortCharacter[], FortCharacter.IsActive[] )
		{
			FortCharacter.Heal(HealthPercent)
		}
		}#>
	}

    OnEntityExitedEvent(HitEnt:entity):void =
	{
        Print("HealthKit: OnEntityExitedEvent")
	}
