<#
	NPC's custom scene graph component
	use "FabricInteractable" for mesh's collision
#>
using { /Verse.org }
using { /Verse.org/Native }
using { /Verse.org/Simulation }
using { /Verse.org/SpatialMath }
using { /Verse.org/SceneGraph }
using { /Verse.org/SceneGraph/KeyframedMovement }
using { /UnrealEngine.com/Temporary/SpatialMath }
using { /UnrealEngine.com/Temporary/Diagnostics }
using { /Fortnite.com/Game }
using { /Fortnite.com/Teams }
using { /Fortnite.com/Characters }
using { /Fortnite.com/Devices }
using { /Fortnite.com/Playspaces }

# A Verse-authored component that can be added to entities
fn_npc_component<public> := class<final_super>(component):

	var Agent<public> : agent
  
    # Runs when the component should start simulating in a running game.
    OnBeginSimulation<override>():void =
	{
		(super:)OnBeginSimulation()
		#Print("Initialize NPC Component...")

		if ( RoundM := Entity.GetFortRoundManager[] )
		{
			RoundM.SubscribeRoundStarted(OnRoundStarted)
			RoundM.SubscribeRoundEnded(OnRoundEnded)
		}
	}

    # Runs when the component should start simulating in a running game.
    # Can be suspended throughout the lifetime of the component. Suspensions
    # will be automatically cancelled when the component is disposed or the
    # game ends.
    OnSimulate<override>()<suspends>:void =
    {
		#TickEvents.PrePhysics.Subscribe(OnPrePhysics)
		spawn{AgentThink()}
	}

	OnRoundStarted():void =
	{

	}

	OnRoundEnded():void =
	{
		Entity.RemoveFromParent()
	}

    GetAgent()<decides>:void=
	{
		Agent
	}

    FollowAgent():void=
	{
		if ( FortCharacter := Agent.GetFortCharacter[], FortCharacter.IsActive[] )
		{
			NPCPos := FortCharacter.GetTransform().Translation
			NewTransform := (/UnrealEngine.com/Temporary/SpatialMath:)transform{ Translation := NPCPos, Rotation := (/UnrealEngine.com/Temporary/SpatialMath:)rotation{}, Scale := FromVector3(Entity.GetGlobalTransform().Scale)}
			Entity.SetGlobalTransform(FromTransform(NewTransform))

			<#NewKeyTransform := (/Verse.org/SpatialMath:)transform{ 
				Translation := FromVector3(NPCPos) - Entity.GetGlobalTransform().Translation,
				Rotation := (/Verse.org/SpatialMath:)rotation{},
				Scale := (/Verse.org/SpatialMath:)vector3{Forward  := 0.0, Left := 0.0, Up := 0.0}
			}

			if (AController := Entity.GetComponent[keyframed_movement_component])
			{
				AnimKeyframes := array
				{
					keyframed_movement_delta
					{
						Transform := NewKeyTransform
						Duration := 0.1
						Easing := linear_easing_function{}
					}
				}
				AController.Stop()
				AController.SetKeyframes(AnimKeyframes, oneshot_keyframed_movement_playback_mode{})
				AController.Play()
			}#>

			DebugDraw := debug_draw{}
			DebugDraw.DrawBox(FromVector3(Entity.GetGlobalTransform().Translation), (/UnrealEngine.com/Temporary/SpatialMath:)rotation{}, ?Extent := (/UnrealEngine.com/Temporary/SpatialMath:)vector3{ X := 75.0,  Y := 75.0, Z := 150.0}, ?DrawDurationPolicy := debug_draw_duration_policy.FiniteDuration, ?Duration := 0.15)
		}
	}

    AgentThink()<suspends>:void=
	{
		if (<#NPC := npc[Agent],#> FortCharacter := Agent.GetFortCharacter[])
		{
			race
			{
				<#block
				{
					if (RoundManager := Entity.GetFortRoundManager[])
					{
						RoundManager.AwaitRoundEnd()
						Entity.RemoveFromParent()
					}
				}#>
				block
				{
					FortCharacter.EliminatedEvent().Await()
					Entity.RemoveFromParent()
				}
				block
				{
					loop
					{
						Sleep(0.5)
						if ( IsSimulating[], IsInScene[], FortCharacter.IsActive[] ) 
						{
							FollowAgent()
						}
					}
				}
			}
		}
	}

	# crash game alot
	<#OnPrePhysics(DeltaTime:float):void =
	{
		if (NPC := npc[Agent], NPC.IsActive[], FortCharacter := Agent.GetFortCharacter[], FortCharacter.IsActive[], IsSimulating[], IsInScene[])
		{
			FollowAgent()
		}
	}#>

	<#OnRemovingFromScene<override>()<suspends>:void =
	{
		Print("OnRemovingFromScene")
	}#>
