using { /Verse.org }
using { /Verse.org/Native }
using { /Verse.org/Simulation }
using { /Verse.org/Simulation/Tags }
using { /Verse.org/SpatialMath }
using { /Verse.org/SceneGraph }
using { /Fortnite.com/Game }
using { /Fortnite.com/Devices }
using { /Fortnite.com/Characters }
using { /Fortnite.com/Playspaces }

# A Verse-authored component that can be added to entities
fn_npc_manager_component<public> := class<final_super>(component):

	@editable PlayerClassesDevice : ?player_classes_manager_device = false
	@editable CreatureSpawners: []creature_spawner_device = array{}

    # Runs when the component should start simulating in a running game.
    OnBeginSimulation<override>():void =
	{
        (super:)OnBeginSimulation()
        Print("Initialize NPC Manager...")

		if ( RoundM := Entity.GetFortRoundManager[] )
		{
			RoundM.SubscribeRoundStarted(OnRoundStarted)
			RoundM.SubscribeRoundEnded(OnRoundEnded)
		}
	}

    # Runs when the component should start simulating in a running game.
    # Can be suspended throughout the lifetime of the component. Suspensions
    # will be automatically cancelled when the component is disposed or the
    # game ends.
    OnSimulate<override>()<suspends>:void =
	{
		#
	}

	OnRemovingFromScene<override>():void =
	{
        (super:)OnRemovingFromScene()
		for ( EntNPC : Entity.FindDescendantEntitiesWithComponent(fn_npc_component), NPCComp := EntNPC.GetComponent[fn_npc_component] )
		{
			EntNPC.RemoveFromParent()
			Print("NPC Component Manager: Removed All NPC.")
		}
	}

	OnRoundStarted():void =
	{
		for( AllCreatureSpawners: CreatureSpawners )
		{
			AllCreatureSpawners.SpawnedEvent.Subscribe(OnNPCSpawned)
		}
	}

	OnRoundEnded():void =
	{
		for ( EntNPC : Entity.FindDescendantEntitiesWithComponent(fn_npc_component), NPCComp := EntNPC.GetComponent[fn_npc_component] )
		{
			EntNPC.RemoveFromParent()
			Print("NPC Component Manager: Removed All NPC.")
		}
		Entity.RemoveFromParent()
	}

    InitializeNPC(Agent: agent)<suspends>:void =
	{
		#Sleep(1.0) # gotta wait or server crash
		AddNPC(Agent)
		AwaitCreatureElimination(Agent)
		if ( FortCharacter := Agent.GetFortCharacter[] )
		{
			FortCharacter.DamagedEvent().Subscribe(OnNPCDamaged)
		}
	}

    AddNPC(Agent: agent): void =
	{
		if ( HasNPCComponent[Agent] )
		{
			Print("NPC Component Manager: Agent already has a NPC Component.")
			return
		}

		if ( SimulationEntity := Entity.GetSimulationEntity[] )
		{
			SpawnTransform := (/Verse.org/SpatialMath:)transform{Translation:=(/Verse.org/SpatialMath:)vector3{Forward := 0.0, Left := 0.0, Up := 0.0}}
			NE := Ent_NPC{}
			#NE := entity{}
			NPCComp := fn_npc_component{ Entity := NE, Agent := Agent }
			NE.AddComponents(array{NPCComp})
			SimulationEntity.AddEntities(array{NE})
			NE.SetGlobalTransform(SpawnTransform)
			Print("NPC Component Manager: Added a NPC.")
			CountNPCComponent()
		}

		<#if (FortCharacter := Agent.GetFortCharacter[])
		{
			FortCharacter.EliminatedEvent().Subscribe(OnNPCEliminated)
		}#>
	}

	HasNPCComponent(Agent:agent)<transacts><decides>: void =
	{
		var bSuccess:logic = false
		if ( SimulationEntity := Entity.GetSimulationEntity[] )
		{
			for ( EntNPC : SimulationEntity.FindDescendantEntitiesWithComponent(fn_npc_component), NPCComp := EntNPC.GetComponent[fn_npc_component], NPCComp.Agent = Agent )
			{
				set bSuccess = true
				bSuccess?
			}
		}

		bSuccess?
	}

	CountNPCComponent(): void =
	{
		var NPCCount : int = 0
		if ( SimulationEntity := Entity.GetSimulationEntity[] )
		{
			for ( EntNPC : SimulationEntity.FindDescendantEntitiesWithComponent(fn_npc_component), NPCComp := EntNPC.GetComponent[fn_npc_component] )
			{
				set NPCCount += 1
			}
		}

		Print("Ent_NPC Count: {NPCCount}")
	}

    RemoveNPC(Agent: agent): void =
	{
		if ( SimulationEntity := Entity.GetSimulationEntity[] )
		{
			for ( EntNPC : SimulationEntity.FindDescendantEntitiesWithComponent(fn_npc_component), NPCComp := EntNPC.GetComponent[fn_npc_component], NPCComp.Agent = Agent )
			{
				EntNPC.RemoveFromParent()
				Print("NPC Component Manager: Removed a NPC.")
				CountNPCComponent()
			}
		}
	}

    OnNPCRemoved(Agent: agent): void =
	{
		#Print("OnNPCRemoved()")
		RemoveNPC(Agent)
	}

	OnNPCSpawned(Agent:agent):void=
	{
		Print("OnNPCSpawned()")
		spawn{InitializeNPC(Agent)}
	}

	<#OnNPCEliminated(Result:elimination_result):void=
	{
		#Print("OnNPCEliminated()")
		KilledCharacter := Result.EliminatedCharacter
		if ( KilledAgent := KilledCharacter.GetAgent[] )
		{
			#RemoveNPC(KilledAgent)
		}
	}#>

	AwaitCreatureElimination(Agent:agent)<suspends>:void=
	{
		if ( FortCharacter := Agent.GetFortCharacter[] )
		{
			#FortCharacter.EliminatedEvent().Await() # does not work
			loop
			{
				if (FortCharacter.GetHealth() <= 0.0)
				{
					OnNPCRemoved(Agent)
					break
				}
				FortCharacter.DamagedEvent().Await()
			}
		}
	}

	OnNPCDamaged(Result:damage_result):void=
	{
		if (DamagedChar := fort_character[Result.Target], DamagedAgent := DamagedChar.GetAgent[], PlayerClasses := PlayerClassesDevice?)
		{
			if( AttackerAgent := Result.Instigator?.GetInstigatorAgent[], DamagedAgent <> AttackerAgent )
			{
				PlayerClasses.PlayClassVO(AttackerAgent, ModulePlayerClasses.FNVOType.FullAlert)

				PlayerClasses.AddUltimateMeter(AttackerAgent, 1.0*(Result.Amount / 25.0))

				spawn{PlayerClasses.UpdateHudDamageText(AttackerAgent, Result.Amount)}

				<#if ( PlayerClasses.AgentHasAbility[AttackerAgent,FNAbility.AirRaid], FortAttacker := AttackerAgent.GetFortCharacter[], FortAttacker.IsInAir[] )
				{
					spawn{FortAttacker.AirRaidDamage(DamagedChar, Result.Amount / 2.5)}
				}#>
			}
		}
	}
