using { /Verse.org }
using { /Verse.org/Native }
using { /Verse.org/Simulation }
using { /Verse.org/Simulation/Tags }
using { /Verse.org/SpatialMath }
using { /Verse.org/SceneGraph }
using { /Fortnite.com/Game }
using { /Fortnite.com/Devices }
using { /Fortnite.com/Characters }
using { /Fortnite.com/Playspaces }

#player_spawner_tag := class(tag){}

# NOTES:
# 
#

# A Verse-authored component that can be added to entities
fn_player_manager_component<public> := class<final_super>(component):

	#@editable var MyCustomInt<public>:int = 10
  
    # Runs when the component should start simulating in a running game.
    OnBeginSimulation<override>():void =
	{
        (super:)OnBeginSimulation()
        Print("Initialize Player Manager...")

		if ( RoundM := Entity.GetFortRoundManager[] )
		{
			RoundM.SubscribeRoundStarted(OnRoundStarted)
			RoundM.SubscribeRoundEnded(OnRoundEnded)
		}
	}

    # Runs when the component should start simulating in a running game.
    # Can be suspended throughout the lifetime of the component. Suspensions
    # will be automatically cancelled when the component is disposed or the
    # game ends.
    OnSimulate<override>()<suspends>:void =
	{
		#
	}

	OnRemovingFromScene<override>():void =
	{
        (super:)OnRemovingFromScene()
		for ( EntPlayer : Entity.FindDescendantEntitiesWithComponent(fn_player_component), PlayerComp := EntPlayer.GetComponent[fn_player_component] )
		{
			EntPlayer.RemoveFromParent()
			Print("Player Component Manager: Removed All Player.")
		}
	}

	OnRoundStarted():void =
	{
        if (PlaySpace := Entity.GetPlayspaceForEntity[])
		{
			PlaySpace.ParticipantAddedEvent().Subscribe(OnPlayerJoined)
			PlaySpace.ParticipantRemovedEvent().Subscribe(OnPlayerRemoved)

			# Get all the players in the experience
			AllPlayers := PlaySpace.GetParticipants()
			for (NewGamePlayer : AllPlayers)
			{
				OnPlayerJoined(NewGamePlayer)
			}
        }
		else
		{
			Print("Error: No PlaySpace.")
		}

        TaggedSpawners := Entity.FindCreativeObjectsWithTag(player_spawner_tag{})
        for (FoundSpawners: TaggedSpawners, PlayerSpawners := player_spawner_device[FoundSpawners])
		{
			PlayerSpawners.SpawnedEvent.Subscribe(OnPlayerSpawned)
		}
	}

	OnRoundEnded():void =
	{
		for ( EntPlayer : Entity.FindDescendantEntitiesWithComponent(fn_player_component), PlayerComp := EntPlayer.GetComponent[fn_player_component] )
		{
			EntPlayer.RemoveFromParent()
			Print("Player Component Manager: Removed All Player.")
		}
		Entity.RemoveFromParent()
	}

    InitializePlayer(Agent: agent)<suspends>:void =
	{
		Sleep(4.0) # gotta wait or server crash
		AddPlayer(Agent)
	}

    AddPlayer(Agent: agent): void =
	{
		if ( HasPlayerComponent[Agent] )
		{
			Print("Player Component Manager: Agent already has a Player Component.")
			return
		}

		if ( SimulationEntity := Entity.GetSimulationEntity[] )
		{
			SpawnTransform := (/Verse.org/SpatialMath:)transform{Translation:=(/Verse.org/SpatialMath:)vector3{Forward := 0.0, Left := 0.0, Up := 0.0}}
			NE := Ent_Player{}
			#NE := entity{}
			PlayerComp := fn_player_component{ Entity := NE, Agent := Agent }
			NE.AddComponents(array{PlayerComp})
			SimulationEntity.AddEntities(array{NE})
			NE.SetGlobalTransform(SpawnTransform)
			Print("Player Component Manager: Added a Player.")
		}

		<#if (FortCharacter := Agent.GetFortCharacter[])
		{
			FortCharacter.EliminatedEvent().Subscribe(OnPlayerEliminated)
		}#>
	}

	HasPlayerComponent(Agent:agent)<transacts><decides>: void =
	{
		var bSuccess:logic = false
		if ( SimulationEntity := Entity.GetSimulationEntity[] )
		{
			for ( EntPlayer : SimulationEntity.FindDescendantEntitiesWithComponent(fn_player_component), PlayerComp := EntPlayer.GetComponent[fn_player_component], PlayerComp.Agent = Agent )
			{
				set bSuccess = true
				bSuccess?
			}
		}

		bSuccess?
	}

    RemovePlayer(Agent: agent): void =
	{
		if ( SimulationEntity := Entity.GetSimulationEntity[] )
		{
			for ( EntPlayer : SimulationEntity.FindDescendantEntitiesWithComponent(fn_player_component), PlayerComp := EntPlayer.GetComponent[fn_player_component], PlayerComp.Agent = Agent )
			{
				EntPlayer.RemoveFromParent()
				Print("Player Component Manager: Removed a Player.")
			}
		}
	}

    OnPlayerJoined(Agent: agent): void =
	{
		#Print("OnPlayerJoined()")
		spawn{InitializePlayer(Agent)}
	}

    OnPlayerRemoved(Agent: agent): void =
	{
		#Print("OnPlayerRemoved()")
		RemovePlayer(Agent)
	}

	OnPlayerSpawned(Agent:agent):void=
	{
		Print("OnPlayerSpawned()")
		#AddPlayer(Agent)
	}

	<#OnPlayerEliminated(Result:elimination_result):void=
	{
		#Print("OnPlayerEliminated()")
		KilledCharacter := Result.EliminatedCharacter
		if ( KilledAgent := KilledCharacter.GetAgent[] )
		{
			#RemovePlayer(KilledAgent)
		}
	}#>
