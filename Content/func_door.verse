<#
	sliding door
#>
using { /Fortnite.com/Devices }
using { /Fortnite.com/Characters }
using { /Verse.org/Simulation }
using { /UnrealEngine.com/Temporary/SpatialMath }
using { ModulePropExtra }

# A Verse-authored creative device that can be placed in a level
func_door := class(creative_device):

	@editable DoorProp : creative_prop = creative_prop{}
	@editable DoorArea : capture_area_device = capture_area_device{}

	@editable MoveDirection  : vector3 = vector3{}

	@editable SFXStart : audio_player_device = audio_player_device{}
	@editable SFXStop : audio_player_device = audio_player_device{}

	var DoorHomePos : vector3 = vector3{}

    # Runs when the device is started in a running game
	OnBegin<override>()<suspends>:void=
	{
		Print("Initialize func_door")
		DoorArea.FirstAgentEntersEvent.Subscribe(OnAreaEnter)
		DoorArea.LastAgentExitsEvent.Subscribe(OnAreaExit)

		set DoorHomePos = DoorProp.GetTransform().Translation
	}

	OnAreaEnter(Agent:agent):void=
	{
		spawn{AnimateOpenDoor()}
	}

	OnAreaExit(Agent:agent):void=
	{
		spawn{AnimateCloseDoor()}
	}

	AnimateOpenDoor()<suspends>:void=
	{
		if ( not DoorProp.IsValid[] )
		{
			return
		}

		SFXStart.Play()
		NewPos := DoorProp.GetTransform().Translation + MoveDirection
		DoorProp.SmoothMoveTo(NewPos, 0.25)
		SFXStart.Stop()
		SFXStop.Play()
	}

	AnimateCloseDoor()<suspends>:void=
	{
		if ( not DoorProp.IsValid[] )
		{
			return
		}

		SFXStart.Play()
		DoorProp.SmoothMoveTo(DoorHomePos, 0.25)
		SFXStart.Stop()
		SFXStop.Play()
	}
