<#
	2AGENCY
	Team Deathmatch.
#>
using { /Fortnite.com/Game }
using { /Fortnite.com/Devices }
using { /Fortnite.com/Teams }
using { /Fortnite.com/Characters }
using { /Fortnite.com/UI }
using { /Verse.org/Simulation }
using { /Verse.org/Colors }
using { /UnrealEngine.com/Temporary/Diagnostics }
using { /UnrealEngine.com/Temporary/UI }
using { /UnrealEngine.com/Temporary/SpatialMath }
using { SortingAlgorithms }


# Structure for passing a player and their stats as arguments.
player_and_stats<public> := struct
{
	Player<public>:player
	RoundScore<public>:int
}

ui_hudtdm := class() 
{
    LinkedCreativeDevice<public> : fn_game_tdm
    LinkedPlayer<public> : player

    var IsInitialized<private> : logic = false
    var MainCanvas<private> : canvas = canvas{}

	UITextMessage<localizes>(Value : string) : message = "{Value}"

	UIBlueScore<private>: text_block = text_block{DefaultTextColor := NamedColors.White, DefaultShadowColor := NamedColors.Black, DefaultShadowOffset := option{vector2{X := 2.0, Y := 2.0}}, DefaultJustification := text_justification.Center}
	UIRedScore<private>: text_block = text_block{DefaultTextColor := NamedColors.White, DefaultShadowColor := NamedColors.Black, DefaultShadowOffset := option{vector2{X := 2.0, Y := 2.0}}, DefaultJustification := text_justification.Center}
	UITextClock<private>: text_block = text_block{DefaultTextColor := NamedColors.White, DefaultShadowColor := NamedColors.Black, DefaultShadowOffset := option{vector2{X := 2.0, Y := 2.0}}, DefaultJustification := text_justification.Center}

    Init<public>() : void =
	{
        if (IsInitialized?)
		{
            return
		}

		set MainCanvas = CreateUI()
		set IsInitialized = true
    }

    OpenUI<public>() : void =
	{
		if (PlayerUI := GetPlayerUI[LinkedPlayer])
		{
			PlayerUI.AddWidget(GetCanvas(), player_ui_slot{InputMode := ui_input_mode.None})
		}
	}

    CloseUI<public>()<suspends> : void =
	{
        if (PlayerUI := GetPlayerUI[LinkedPlayer])
		{
			PlayerUI.RemoveWidget(GetCanvas())
		}
    }

    UpdateTeamScore(BlueScore:int, RedScore:int) : void =
	{
		UIBlueScore.SetText(UITextMessage("{BlueScore}"))
		UIRedScore.SetText(UITextMessage("{RedScore}"))
	}

    UpdateTimeDisplay(InText:message) : void =
	{
		UITextClock.SetText(InText)
	}

    GetCanvas<public>() : canvas = MainCanvas  # Exposes the canvas to public

	CreateUI<private>() : canvas =
	{
		MatBGGold := Materials.UI.M_UI_BG{Opacity:=0.75}
		MatBGWhite := Materials.UI.MI_UI_BG_White{Opacity:=0.75}

		UICanvas : canvas = canvas{Slots := array{
		canvas_slot
		{
			ZOrder := {0}
			Anchors := anchors{Minimum := vector2{X := 0.5, Y := 0.0}, Maximum := vector2{X := 0.5, Y := 0.0}}
			Offsets := margin{ Top:=0.0, Left:= 0.0, Right:=0.0, Bottom := 0.0 }
			Alignment := vector2{X := 0.5, Y := 0.0}
			SizeToContent := true
			Widget := stack_box
			{
				Orientation := orientation.Vertical
				Slots := array
				{
					stack_box_slot
					{
						Widget := overlay
						{
							Slots := array
							{
								overlay_slot
								{
									VerticalAlignment := vertical_alignment.Fill
									HorizontalAlignment := horizontal_alignment.Fill
									Widget := material_block
									{
										DefaultImage := MatBGGold
										DefaultDesiredSize := vector2{X := 500.0, Y := 50.0}
										#DefaultOpacity := 0.5
									}
								}
								overlay_slot
								{
									VerticalAlignment := vertical_alignment.Center
									HorizontalAlignment := horizontal_alignment.Fill
									Padding := margin{Top := 0.0, Left := 10.0, Right := 10.0, Bottom := 0.0}
									Widget := text_block
									{
										DefaultText := UITextMessage("1666-3570-8041")
										DefaultTextColor := NamedColors.Gold
										DefaultTextOpacity := 0.6
										DefaultShadowColor := NamedColors.White
										DefaultShadowOffset := option{vector2{X := 0.0, Y := 0.0}}
										DefaultJustification := text_justification.Center
									}
								}
							}
						}
					}
					stack_box_slot
					{
						Widget := stack_box
						{
							Orientation := orientation.Horizontal
							Slots := array
							{
								stack_box_slot
								{
									Widget := overlay
									{
										Slots := array
										{
											overlay_slot
											{
												VerticalAlignment := vertical_alignment.Fill
												HorizontalAlignment := horizontal_alignment.Fill
												Widget := material_block
												{
													DefaultImage := MatBGWhite
													DefaultTint := NamedColors.CornflowerBlue
													DefaultDesiredSize := vector2{X := 150.0, Y := 50.0}
													#DefaultOpacity := 0.5
												}
											}
											overlay_slot
											{
												VerticalAlignment := vertical_alignment.Center
												HorizontalAlignment := horizontal_alignment.Fill
												Padding := margin{Top := 5.0, Left := 5.0, Right := 5.0, Bottom := 5.0}
												Widget := UIBlueScore
											}
										}
									}
								}
								stack_box_slot
								{
									Widget := overlay
									{
										Slots := array
										{
											overlay_slot
											{
												VerticalAlignment := vertical_alignment.Fill
												HorizontalAlignment := horizontal_alignment.Fill
												Widget := material_block
												{
													DefaultImage := MatBGGold
													DefaultDesiredSize := vector2{X := 200.0, Y := 50.0}
													#DefaultOpacity := 0.5
												}
											}
											overlay_slot
											{
												VerticalAlignment := vertical_alignment.Center
												HorizontalAlignment := horizontal_alignment.Fill
												Padding := margin{Top := 5.0, Left := 5.0, Right := 5.0, Bottom := 5.0}
												Widget := UITextClock
											}
										}
									}
								}
								stack_box_slot
								{
									Widget := overlay
									{
										Slots := array
										{
											overlay_slot
											{
												VerticalAlignment := vertical_alignment.Fill
												HorizontalAlignment := horizontal_alignment.Fill
												Widget := material_block
												{
													DefaultImage := MatBGWhite
													DefaultTint := NamedColors.DarkRed
													DefaultDesiredSize := vector2{X := 150.0, Y := 50.0}
													#DefaultOpacity := 0.5
												}
											}
											overlay_slot
											{
												VerticalAlignment := vertical_alignment.Center
												HorizontalAlignment := horizontal_alignment.Fill
												Padding := margin{Top := 5.0, Left := 5.0, Right := 5.0, Bottom := 5.0}
												Widget := UIRedScore
											}
										}
									}
								}
							}
						}
					}
				}
			}
		}
        }}

		UIBlueScore.SetText(UITextMessage("0"))
		UIRedScore.SetText(UITextMessage("0"))

		return UICanvas
	}
}


# A Verse-authored creative device that can be placed in a level
fn_game_tdm := class(creative_device):

	var TeamKillCount: [int]int = map{}
	TeamBlue : int = 0
	TeamRed : int = 1

	var UIHudMap : [player]ui_hudtdm = map{}

	#var PlayerLeaderboard: []agent = array{}

	@editable PlayerClassesDevice : ?player_classes_manager_device = false

	@editable TeamSettingsBlue : team_settings_and_inventory_device = team_settings_and_inventory_device{}
	@editable TeamSettingsRed : team_settings_and_inventory_device = team_settings_and_inventory_device{}

	#@editable TimerMinutes : int = 1
	#@editable TimerSeconds : int = 30
	@editable TimerLength: int = 90		# in seconds

	var BreakTimerLoop : logic = false

	CountdownTimerToUI<localizes>(Value : string)<computes>: message = "{Value}"
	UITextMessage<localizes>(Value : string) : message = "{Value}"
	UIPlayerName<localizes>(Agent:agent)<computes>:message = "{Agent}"
	#UIPlayerNameScore<localizes>(Agent:agent,Value : string):message = "{Agent} - {Value}\n"

	@editable HudMsg_Victory : hud_message_device = hud_message_device{}
	@editable HudMsg_VictoryStats : hud_message_device = hud_message_device{}

	@editable SFX_TimeTick	: audio_player_device = audio_player_device{}
	@editable MX_Victory	: audio_player_device = audio_player_device{}
	@editable MX_Lose		: audio_player_device = audio_player_device{}

    # Runs when the device is started in a running game
    OnBegin<override>()<suspends>:void=
	{
		Print("Initialize Team Deathmatch...")

		if { set TeamKillCount[TeamBlue] = 0 }
		if { set TeamKillCount[TeamRed] = 0 }

		UpdateHudTeamScore()
		StartCountdownTimer()
	}

	OnPlayerJoined(Agent: agent): void =
	{
		if (InPlayer := player[Agent])
		{
			var OptPlayerUI :?ui_hudtdm = false

			if (ThisPlayerUI := UIHudMap[InPlayer])
			{
				set OptPlayerUI = option{ThisPlayerUI}
			}
			else if (ThisPlayerUI := set UIHudMap[InPlayer] = ui_hudtdm{ LinkedPlayer := InPlayer, LinkedCreativeDevice := Self } ) 
			{
				ThisPlayerUI.Init()
				set OptPlayerUI = option{ThisPlayerUI}
			}

			if (PlayerUI := OptPlayerUI?)
			{
				PlayerUI.OpenUI()
			}
		}
	}

	# Removes an element from the given map and returns a new map without that element
	(Map:[keytype]valuetype where keytype:subtype(comparable), valuetype:type).RemoveKey(ElementToRemove:keytype):[keytype]valuetype=
	{
		var NewMap:[keytype]valuetype = map{}
		for (Key -> Value : Map, Key <> ElementToRemove)
		{
			set NewMap = ConcatenateMaps(NewMap, map{Key => Value})
		}

		return NewMap
	}

    OnPlayerRemoved(Agent: agent): void =
	{
		if (AgentPlayer := player[Agent])
		{
			set UIHudMap = UIHudMap.RemoveKey(AgentPlayer)
		}
	}

	AddTeamKillCount<public>(TeamID:int, KillCount:int):void=
	{
		#Print("AddTeamKillCount()")
		if ( set TeamKillCount[TeamID] += KillCount, NewKillCount := TeamKillCount[TeamID] )
		{
			<#case( TeamID )
			{
				0 =>			Print("Team (BLUE) Kill Count: {NewKillCount}")
				1 =>			Print("Team (RED) Kill Count: {NewKillCount}")
				_ =>			Print("Team (None) Kill Count: {NewKillCount}")
			}#>
			UpdateHudTeamScore()
		}
	}

	AddTeamKillCount<public>(TeamName:team, KillCount:int):void=
	{
		TeamCollection := GetPlayspace().GetTeamCollection()
		if( VerseTeamIndex := TeamCollection.GetTeams().Find[TeamName], TeamIndex := VerseTeamIndex)
		{
			AddTeamKillCount(TeamIndex, KillCount)
		}
	}

    StartCountdownTimer():void=
	{
		set BreakTimerLoop = false
		spawn{OnCountdownTimer()}
	}

    StopCountdownTimer():void=
	{
		set BreakTimerLoop = true
	}

    OnCountdownTimer()<suspends>:void=
	{
		var TimeLeft : int = TimerLength
		loop
		{
			Sleep(1.0)
			set TimeLeft -= 1

			# set our label
			var Hours : int = 0
			var Minutes : int = 0
			var Seconds : int = 0

			if ( TimeLeft <= 0 )
			{
				set Hours = 0
				set Minutes = 0
				set Seconds = 0
				set BreakTimerLoop = true
			}
			else
			{
				if (NewHours := TimeLeft / 3600, NewMinutes := Mod[TimeLeft, 3600] / 60, NewSeconds := Mod[TimeLeft, 60])
				{
					set Hours = Floor(NewHours)
					set Minutes = Floor(NewMinutes)
					set Seconds = NewSeconds
				}

				if ( TimeLeft <= 30 ) # beep beep
				{
					SFX_TimeTick.Play()
				}
			}				

			if ( Hours > 0 )
			{
				#UpdateHudTime(CountdownTimerToUI("{Hours}:{Minutes}:{Seconds}"))
				if ( Seconds >= 10 )
				{
					UpdateHudTime(CountdownTimerToUI("{Hours}:{Minutes}:{Seconds}"))
				}
				else
				{
					UpdateHudTime(CountdownTimerToUI("{Hours}:{Minutes}:0{Seconds}"))
				}
			}
			else
			{
				#UpdateHudTime(CountdownTimerToUI("{Minutes}:{Seconds}"))
				if ( Seconds >= 10 )
				{
					UpdateHudTime(CountdownTimerToUI("{Minutes}:{Seconds}"))
				}
				else
				{
					UpdateHudTime(CountdownTimerToUI("{Minutes}:0{Seconds}"))
				}
			}

			if ( BreakTimerLoop? )
			{
				break
			}
		}
		OnCountdownEnded()
	}

    OnCountdownEnded():void=
	{
		Print("OnCountdownEnded()")

		if ( BlueKillCount := TeamKillCount[TeamBlue], RedKillCount := TeamKillCount[TeamRed] )
		{
			Print("Blue: {BlueKillCount} vs. Red: {RedKillCount}")
			if ( BlueKillCount > RedKillCount )
			{
				Print("Blue Wins!")
				TeamVictory(TeamBlue)
			}
			else
			{
				Print("Red Wins!")
				TeamVictory(TeamRed)
			}
		}
	}

    TeamVictory(TeamID:int):void=
	{
		spawn{VictorySequence(TeamID)}
	}

    VictorySequence(TeamID:int)<suspends> : void =
	{
		TeamCollection := GetPlayspace().GetTeamCollection()
		Teams := TeamCollection.GetTeams()
		AllPlayers := GetPlayspace().GetPlayers()
		for (Player : AllPlayers)
		{
			if ( TeamCollection.IsOnTeam[Player, Teams[TeamID]] )
			{
				if ( PlayerClasses := PlayerClassesDevice? )
				{
					PlayerClasses.OnPlayerWin(Player)
				}
				HudMsg_Victory.Show(Player, UITextMessage("VICTORY!"), ?DisplayTime := 5.0)
				MX_Victory.Play(Player)
			}
			else
			{
				if ( PlayerClasses := PlayerClassesDevice? )
				{
					PlayerClasses.OnPlayerLose(Player)
				}
				HudMsg_Victory.Show(Player, UITextMessage("LOSE"), ?DisplayTime := 5.0)
				MX_Lose.Play(Player)
			}
		}
		Sleep(5.0)
		case( TeamID )
		{
			0 => TeamSettingsBlue.EndRound()
			1 => TeamSettingsRed.EndRound()
			_ => #uhh
		}
    }

	OnPlayerEliminated(Result:elimination_result):void=
	{
		# Who got killed?
		KilledCharacter := Result.EliminatedCharacter

		if
		{
			# Let's make sure we can get the Agent of the fort_character                                                                    
			KilledAgent := KilledCharacter.GetAgent[]

			# Who did the killing? Why is there a question mark?
			# This fails if the player was killed by something other than
			# another fort_character. Ex. Damage Volumes
			KillerChar := Result.EliminatingCharacter?

            # Let's make sure we can get the Agent of the fort_character 
			KillerAgent := KillerChar.GetAgent[]
		}
        then
		{
			# Make sure they did not eliminate themself
			if(KillerAgent <> KilledAgent)
			{
				TeamCollection := GetPlayspace().GetTeamCollection()
				Teams := TeamCollection.GetTeams()
				if ( TeamCollection.IsOnTeam[KillerAgent, Teams[TeamBlue]] )
				{
					AddTeamKillCount(TeamBlue, 1)
				}
				else if ( TeamCollection.IsOnTeam[KillerAgent, Teams[TeamRed]] )
				{
					AddTeamKillCount(TeamRed, 1)
				}
			}

			# TEST
			#UpdatePlayerLeaderboard()
		}
	}

	ShowHUD(Agent:agent):void= 
	{
		if (PlayerHUD := UIHudMap[Agent])
		{
			PlayerHUD.OpenUI()
		}
	}

	HideHUD(Agent:agent):void= 
	{
		if (PlayerHUD := UIHudMap[Agent])
		{
			spawn{PlayerHUD.CloseUI()}
		}
	}

    UpdateHudTime(InText:message) : void =
	{
		AllPlayers := GetPlayspace().GetPlayers()
		for (Player : AllPlayers, PlayerHUD := UIHudMap[Player])
		{
			PlayerHUD.UpdateTimeDisplay(InText)
		}
	}

    UpdateHudTeamScore() : void =
	{
		if ( BlueScore := TeamKillCount[TeamBlue], RedScore := TeamKillCount[TeamRed] )
		{
			AllPlayers := GetPlayspace().GetPlayers()
			for (Player : AllPlayers, PlayerHUD := UIHudMap[Player])
			{
				PlayerHUD.UpdateTeamScore(BlueScore,RedScore)
			}
		}
	}

	# Returns Left if Left has greater Points than Right.
	MorePointsComparison(Left:player_and_stats, Right:player_and_stats)<decides><transacts>:player_and_stats=
	{
		Left.RoundScore > Right.RoundScore
		Left
	}

	# Returns Left if Left has less Points than Right.
	LessPointsComparison(Left:player_and_stats, Right:player_and_stats)<decides><transacts>:player_and_stats=
	{
		Left.RoundScore < Right.RoundScore
		Left
	}

    UpdatePlayerLeaderboard():void=
	{
		AllPlayers := GetPlayspace().GetPlayers()
        PlayerAndStatsArray :=
		{
            for ( Player : AllPlayers, PlayerClasses := PlayerClassesDevice?, PlayerFromMap := PlayerClasses.PlayersMap[Player])
            {
                player_and_stats
				{
                    Player := Player
                    RoundScore := PlayerFromMap.RoundScore
				}
			}
		}

		var VictoryText : string = ""

		SortedPlayersAndStats := SortingAlgorithms.MergeSort(PlayerAndStatsArray, MorePointsComparison) # this thing wont FUCKING WORK
		for( PlayerIndex -> PlayerAndStats : SortedPlayersAndStats, Agent := agent[PlayerAndStats.Player] )
		{
			Print("{PlayerIndex}")
			PlayerScore := PlayerAndStats.RoundScore
			PlayerName := Localize(UIPlayerName(Agent))
			set VictoryText += "{PlayerName} - {PlayerScore}\n"
		}

		#AllPlayers := GetPlayspace().GetPlayers()
		for ( Player : AllPlayers )
		{
			HudMsg_VictoryStats.Show(Player, UITextMessage(VictoryText), ?DisplayTime := 5.0) #it show up as anonymous???
		}
	}