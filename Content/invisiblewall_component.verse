<#
	scene graph invisible wall component
	
#>
using { /Verse.org }
using { /Verse.org/Native }
using { /Verse.org/Simulation }
using { /Verse.org/SpatialMath }
using { /Verse.org/SceneGraph }
using { /UnrealEngine.com/Temporary/SpatialMath }
using { /UnrealEngine.com/Temporary/Diagnostics }
using { /Fortnite.com/Game }

# A Verse-authored component that can be added to entities
invisiblewall_component<public> := class<final_super>(component):

    # Runs when the component should start simulating in a running game.
    OnBeginSimulation<override>():void =
	{
		(super:)OnBeginSimulation()
		#Print("Initialize InvisibleWall Component...")

		if ( RoundM := Entity.GetFortRoundManager[] )
		{
			RoundM.SubscribeRoundStarted(OnRoundStarted)
			RoundM.SubscribeRoundEnded(OnRoundEnded)
		}
	}

    # Runs when the component should start simulating in a running game.
    # Can be suspended throughout the lifetime of the component. Suspensions
    # will be automatically cancelled when the component is disposed or the
    # game ends.
    OnSimulate<override>()<suspends>:void =
    {
		if ( MeshComp := Entity.GetComponent[mesh_component] )
		{
			set MeshComp.Visible = false
		}
	}

	OnRoundStarted():void =
	{
		if ( MeshComp := Entity.GetComponent[mesh_component] )
		{
			set MeshComp.Visible = false
		}
	}

	OnRoundEnded():void =
	{
		Entity.RemoveFromParent()
	}
