<#
	2AGENCY
	random stuff on the map
#>
using { /Fortnite.com/Devices }
using { /Fortnite.com/Characters }
using { /Verse.org/Simulation }
using { /UnrealEngine.com/Temporary/SpatialMath }
using { ModulePlayerClasses }
using { ModulePlayerEx }

# A Verse-authored creative device that can be placed in a level
map_interactions_device := class(creative_device):

	@editable PlayerClassesDevice : ?player_classes_manager_device = false

	@editable GrillMoonChair : chair_device = chair_device{}

	#@editable Analytic_Tea : analytics_device = analytics_device{}
	@editable ConversationTea : conversation_device = conversation_device{}
	@editable SFXTeaSus : audio_player_device = audio_player_device{}
	@editable SFXTeaGive : audio_player_device = audio_player_device{}

	var AgentDaigo : ?agent = false

    # Runs when the device is started in a running game
    OnBegin<override>()<suspends>:void=
	{
		Print("Initialize Secret Map Interactions...")

		GrillMoonChair.SeatedEvent.Subscribe(OnGrillMoonEnter)
		GrillMoonChair.ExitedEvent.Subscribe(OnGrillMoonExit)

		ConversationTea.OnConversationEvent.Subscribe(OnConversationTea)
	}

	# GrillMoon Event
	OnGrillMoonEnter(Agent:agent):void= 
	{
		if ( FortCharacter := Agent.GetFortCharacter[], FortCharacter.IsActive[] )
		{
			spawn{GrillMoonEventFail(Agent)}
		}
	}

	GrillMoonEventFail(Agent:agent)<suspends>:void= 
	{
		if ( FortCharacter := Agent.GetFortCharacter[], FortCharacter.IsActive[], PlayerClasses := PlayerClassesDevice?  )
		{
			race
			{
				block
				{
					FortCharacter.EliminatedEvent().Await()
				}
				block
				{
					loop
					{
						QuitAgent := GetPlayspace().PlayerRemovedEvent().Await()
						if ( Agent = QuitAgent )
						{
							break
						}
					}
				}
				block
				{
					FortCharacter.SetVulnerability(false)
					PlayerClasses.SetPlayerAggro(Agent, FNBGMState.None)
					PlayerClasses.PlayClassVO(Agent, FNVOType.Sus)
					Sleep(4.5)
					PlayerClasses.PlayClassVO(Agent, FNVOType.Interrogation)
					FortCharacter.SetHealth(FortCharacter.GetHealth() - 1.0)
					Sleep(1.5)
					PlayerClasses.PlayClassVO(Agent, FNVOType.DBNO)
					FortCharacter.SetHealth(1.0)
				}
			}
		}
	}

	# GrillMoon Event
	OnGrillMoonExit(Agent:agent):void= 
	{
		if ( FortCharacter := Agent.GetFortCharacter[], FortCharacter.IsActive[] )
		{
			if ( FortCharacter.TeleportTo[vector3{X := 20832.0,Y := -17504.0, Z := 3072.0 }, rotation{}] ) {}
			FortCharacter.SetVulnerability(true)
			FortCharacter.SetShield(0.0)
			FortCharacter.SetHealth(0.0)
			FortCharacter.Damage(1.0)
		}
		GrillMoonChair.Disable()
	}

	# Tea Event
	OnConversationTea(Agent:agent,EventNumber:int):void= 
	{
		if ( FortCharacter := Agent.GetFortCharacter[], FortCharacter.IsActive[] )
		{
			case( EventNumber )
			{
				3 =>
				{
					SFXTeaGive.Play()
					PlayerPos := FortCharacter.GetTransform().Translation + vector3{X:= 0.0, Y:= 0.0, Z:= 50.0}
					spawn{ ModulePropExtra.SpawnParticleVFX(FX.NS_ShieldRegen_asset, PlayerPos, rotation{}, 1.5) }
					FortCharacter.Heal(25.0)
				}
				2 =>
				{
					ConversationTea.EndConversation(Agent)
				}
				1 =>
				{
					SFXTeaSus.Play()
				}
				_ =>
				{
					ConversationTea.EndConversation(Agent)
				}
			}
		}
	}
