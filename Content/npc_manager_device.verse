<#
	2AGENCY
	NPC Manager
#>
using { /Fortnite.com/Devices }
using { /Fortnite.com/Characters }
using { /Fortnite.com/Game }
using { /Verse.org/Simulation }

# A Verse-authored creative device that can be placed in a level
npc_manager_device := class(creative_device):

	var aNPCs : []fort_character = array{}

    # Runs when the device is started in a running game
    OnBegin<override>()<suspends>:void=
	{
		Print("Initialize NPC Manager Device...")
	}

	AddNPCCount(FortCharacter : fort_character) : void =
	{
		set aNPCs += array{FortCharacter}
		#Print("NPC Count: {aNPCs.Length}")
	}

	RemoveNPCCount(FortCharacter : fort_character) : void =
	{
		if (NewArray := aNPCs.RemoveFirstElement[FortCharacter]) {set aNPCs = NewArray}
		#Print("NPC Count: {aNPCs.Length}")
	}
