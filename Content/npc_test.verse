using { /Fortnite.com/AI }
using { /Fortnite.com/Characters }
using { /Verse.org/Simulation }
using { /Fortnite.com/Playspaces }
using { /Verse.org/Random }
using { /UnrealEngine.com/Temporary/SpatialMath }

# A Verse-authored NPC Behavior that can be used within an NPC Character Definition or an NPC Spawner device's NPC Behavior Script Override.
npc_test := class(npc_behavior):

    # This function runs when the NPC is spawned in the world and ready to follow a behavior.
    OnBegin<override>()<suspends>:void=
	{
		#Print("Spawn: npc_test.")
		if
		{
			Agent := GetAgent[],
			FortCharacter := Agent.GetFortCharacter[],
			Navigatable := FortCharacter.GetNavigatable[]
			Leashable := FortCharacter.GetFortLeashable[]
		}
		then
		{
			Sleep(5.0)
			TestPos := vector3{X:=20736.0,Y:=-2272.0,Z:=3710.0}
			TestRadius := 1000.0


			spawn{RandomTeleport(TestPos)}

			NavigationTarget : navigation_target = MakeNavigationTarget(TestPos)
			Navigatable.NavigateTo(NavigationTarget, ?ReachRadius := TestRadius)
			Leashable.SetLeashPosition(TestPos, InnerRadius := TestRadius / 1.5, OuterRadius := TestRadius)
			FortCharacter.SetVulnerability(true)

			RandomPlayerClass()
		}
	}

    RandomTeleport(TeleportPos:vector3)<suspends>:void=
	{
        Print("Spawn: npc_test.")
		<#if
		{
			Agent := GetAgent[]
			FortCharacter := Agent.GetFortCharacter[]
			Entity := GetEntity[]						# this GetEntity is EXPERIMENTAL, delete this before memory calculate!!!
			Playspace := Entity.GetPlayspaceForEntity[]
		}
		then
		{
			TeamCollection := Playspace.GetTeamCollection()
			Teams := TeamCollection.GetTeams()
			if ( TeamCollection.IsOnTeam[Agent, Teams[0]] )
			{
				return
			}

			Sleep(GetRandomFloat(1.0,5.0))
			if ( FortCharacter.TeleportTo[TeleportPos, rotation{}] ) {}
		}#>

		if
		{
			Agent := GetAgent[]
			FortCharacter := Agent.GetFortCharacter[]
		}
		then
		{
			Sleep(GetRandomFloat(1.0,5.0))
			if ( FortCharacter.TeleportTo[TeleportPos, rotation{}] ) {}
		}
	}

    RandomPlayerClass():void=
	{
		if
		{
			Agent := GetAgent[],
			FortCharacter := Agent.GetFortCharacter[]
			#PlayerClasses := PlayerClassesDevice?
		}
		then
		{
			#PlayerClasses.SetPlayerClass( Agent, FNClasses.Midas )
			FortCharacter.SetMaxHealth(400.0)
			FortCharacter.SetHealth(400.0)
			FortCharacter.SetShield(100.0)
		}
	}

    # This function runs when the NPC is despawned or eliminated from the world.
    OnEnd<override>():void=
	{
		
	}