<#
	2AGENCY
	Player Abilites (Down But Not Out)
#>
using { /Fortnite.com/Devices }
using { /Fortnite.com/Characters }
using { /Verse.org/Simulation }
using { /UnrealEngine.com/Temporary/SpatialMath }
using { ModulePlayerClasses }
using { ModulePlayerEx }

# A Verse-authored creative device that can be placed in a level
player_ability_dbno_device := class(creative_device):

	@editable PlayerClassesDevice : ?player_classes_manager_device = false
	@editable DBNODevice : down_but_not_out_device = down_but_not_out_device{}

	@editable MoveModDoubleJump : movement_modulator_device  = movement_modulator_device{}

	@editable{Categories:=array{CategoryVFXPowerup}} VFXPowerZeusRebirth : visual_effect_powerup_device = visual_effect_powerup_device{}
	@editable{Categories:=array{CategoryVFXPowerup}} VFXPowerZeusRebirthLoop : visual_effect_powerup_device = visual_effect_powerup_device{}
	@editable{Categories:=array{CategoryVFXPowerup}} VFXDBNONightRose : visual_effect_powerup_device = visual_effect_powerup_device{}

	@editable ItemRemoverZeus : item_remover_device = item_remover_device{}
	@editable ItemGranterZeus : item_granter_device = item_granter_device{}

	@editable{Categories:=array{CategoryAnim}} AnimRebirth : creative_prop_asset = DefaultCreativePropAsset
	@editable{Categories:=array{CategoryAnim}} AnimPuppeteer : creative_prop_asset = DefaultCreativePropAsset

    # Runs when the device is started in a running game
    OnBegin<override>()<suspends>:void=
	{
		Print("Initialize Player Abilities (DBNO)...")
		DBNODevice.AgentDownedEvent.Subscribe(OnPlayerDBNO)
		DBNODevice.AgentRevivedEvent.Subscribe(OnPlayerRevived)
	}

	OnPlayerDBNO(Agent:agent):void= 
	{
		Print("OnPlayerDBNO()")
		if ( PlayerClasses := PlayerClassesDevice? )
		{
			PlayerClasses.PlayClassVO(Agent, FNVOType.DBNO)
			if ( PlayerClasses.AgentHasAbility[Agent, FNAbility.Shadow] )
			{
				spawn{AbilityBecomeShadow(Agent)}
			}
			else if ( PlayerClasses.AgentHasAbility[Agent, FNAbility.Rebirth] )
			{
				spawn{AbilityRebirth(Agent)}
			}
			else if ( PlayerClasses.AgentHasAbility[Agent, FNAbility.Puppeteer] )
			{
				spawn{AbilityPuppeteer(Agent)}
			}
			else
			{
				if (FortCharacter := Agent.GetFortCharacter[])
				{
					FortCharacter.SetHealth(0.0)
				}
			}
		}
	}

	OnPlayerRevived(Agent:agent):void= 
	{
		Print("OnPlayerRevived()")
	}

    ReviveAgent(Agent:agent)<suspends>:void =
	{
		if ( FortCharacter := Agent.GetFortCharacter[], FortCharacter.IsActive[], FortCharacter.IsDownButNotOut[], Player := player[Agent], Player.IsActive[] )
		{
			DBNODevice.Revive(Agent) # source of all evil and errors
		}
	}

    AbilityBecomeShadow(Agent:agent)<suspends>:void=
	{
		if (FortCharacter := Agent.GetFortCharacter[], PlayerClasses := PlayerClassesDevice? )
		{
			if ( PlayerClasses.PlayersMap[Agent].bRevived? )
			{
				return
			}

			defer
			{
				FortCharacter.SetVulnerability( true )
			}

			Print("Joining Midas' Shadow Army...")
			race
			{
				block
				{
					FortCharacter.EliminatedEvent().Await()
					Print("Shadow Transform cancel (Player Died)...")
				}
				block
				{
					loop
					{
						QuitAgent := GetPlayspace().PlayerRemovedEvent().Await()
						if ( Agent = QuitAgent )
						{
							Print("Shadow Transform cancel (Player Disconnected)...")
							break
						}
					}
				}
				block
				{
					Sleep(0.5)
					FortCharacter.SetVulnerability( false )
					PlayerClasses.UpdateHudDisableAbility(Agent, FNAbility.Shadow)
					if {set PlayerClasses.PlayersMap[Agent].bRevived = true}
					spawn{ReviveAgent(Agent)}
					PlayerClasses.SetPlayerClass( Agent, FNClasses.Shadow )
					Sleep(1.5)
				}
			}
		}
	}

    AbilityRebirth(Agent:agent)<suspends>:void=
	{
		if (FortCharacter := Agent.GetFortCharacter[], PlayerClasses := PlayerClassesDevice? )
		{
			if ( PlayerClasses.PlayersMap[Agent].bRevived? )
			{
				return
			}

			defer
			{
				FortCharacter.SetVulnerability( true )
				FortCharacter.ReleaseFromStasis()
			}

			Print("Zeus Phase 3 Transform...")
			race
			{
				block
				{
					FortCharacter.EliminatedEvent().Await()
					Print("Zeus Phase 3 cancel (Player Died)...")
				}
				block
				{
					loop
					{
						QuitAgent := GetPlayspace().PlayerRemovedEvent().Await()
						if ( Agent = QuitAgent )
						{
							Print("Zeus Phase 3 cancel (Player Disconnected)...")
							break
						}
					}
				}
				block
				{
					Print("Zeus Phase 3...")
					Sleep(0.2)
					VFXPowerZeusRebirth.Pickup(Agent)
					ItemRemoverZeus.Remove(Agent) # Prevent verse internal error!
					PlayerClasses.SetPlayerAggro(Agent, FNBGMState.AlertedPhase2)
					FortCharacter.SetVulnerability( false )
					if {set PlayerClasses.PlayersMap[Agent].bRevived = true}
					Sleep(4.0)
					if ( FortCharacter.IsActive[] )
					{
						MoveModDoubleJump.Activate(Agent)
					}
					Sleep(0.5)
					FortCharacter.PutInStasis(stasis_args{AllowTurning := false, AllowFalling := false, AllowEmotes := true})
					if ( FortCharacter.IsActive[] )
					{
						ItemRemoverZeus.Remove(Agent) # IT STILL HAPPENNING
						spawn{ReviveAgent(Agent)} # revive + Mythic Zeus Thunderbolt = verse internal error
						spawn{ChairAnim(Self,Agent,AnimRebirth,1.4)}
					}
					FortCharacter.SetHealth(400.0)
					VFXPowerZeusRebirthLoop.Pickup(Agent)
					spawn{PlayerClasses.UpdatePlayerHUD(Agent)}
					Sleep(1.5)
					ItemGranterZeus.GrantItem(Agent)
					PlayerClasses.UpdateHudDisableAbility(Agent, FNAbility.Rebirth)
				}
			}
		}
	}

    AbilityPuppeteer(Agent:agent)<suspends>:void=
	{
		if (FortCharacter := Agent.GetFortCharacter[], PlayerClasses := PlayerClassesDevice? )
		{
			if ( PlayerClasses.PlayersMap[Agent].bRevived? )
			{
				return
			}

			defer
			{
				FortCharacter.SetVulnerability( true )
				FortCharacter.ReleaseFromStasis()
			}

			Print("Puppeteer Takeover...")
			race
			{
				block
				{
					FortCharacter.EliminatedEvent().Await()
					Print("Puppeteer cancel (Player Died)...")
				}
				block
				{
					loop
					{
						QuitAgent := GetPlayspace().PlayerRemovedEvent().Await()
						if ( Agent = QuitAgent )
						{
							Print("Puppeteer cancel (Player Disconnected)...")
							break
						}
					}
				}
				block
				{
					PlayerClasses.UpdateHudDisableAbility(Agent, FNAbility.Puppeteer)
					FortCharacter.SetVulnerability( false )
					if {set PlayerClasses.PlayersMap[Agent].bRevived = true}
					VFXDBNONightRose.Pickup(Agent)
					if ( FortCharacter.IsActive[] )
					{
						spawn{ReviveAgent(Agent)}
						spawn{ChairAnim(Self,Agent,AnimPuppeteer,2.6)}
						Sleep(1.2)
						PlayerClasses.PlayClassVO(Agent, FNVOType.Idle)
						Sleep(1.4)
						VFXPos := FortCharacter.GetTransform().Translation + vector3{X:= 0.0, Y:= 0.0, Z:= 50.0}
						spawn{ ModulePropExtra.SpawnParticleVFX(FX.NS_Warp_NightRose_asset, VFXPos, rotation{}, 1.2) }
					}
					PlayerClasses.SetPlayerClass( Agent, FNClasses.Puppeteer )
					PlayerClasses.SetPlayerAggro(Agent, FNBGMState.AlertedPhase2)
				}
			}
		}
	}
