<#
	2AGENCY
	Player Abilites
#>
using { /Fortnite.com/Devices }
using { /Fortnite.com/Devices/CreativeAnimation }
using { /Fortnite.com/Devices/CreativeAnimation/InterpolationTypes }
using { /Fortnite.com/Characters }
using { /Fortnite.com/Teams }
using { /Fortnite.com/Game }
using { /Fortnite.com/FortPlayerUtilities }
using { /Fortnite.com/AI }
using { /Fortnite.com/Animation/PlayAnimation }
using { /Verse.org/SceneGraph }
using { /Verse.org/Simulation }
using { /Verse.org/Simulation/Tags }
using { /UnrealEngine.com/Assets }
using { /UnrealEngine.com/Temporary/Diagnostics }
using { /UnrealEngine.com/Temporary/SpatialMath }
using { ModulePlayerClasses }
using { ModulePlayerEx }
using { ModulePropExtra }

turret_tag := class(tag){}
chair_footdive_tag := class(tag){}
chairanim_tag := class(tag){}
spawned_device_tag := class(tag){}

volume_bluezone_tag := class(tag){}
volume_redzone_tag := class(tag){}

CategorySFX<localizes>:message="Sound Effects"
CategoryProp<localizes>:message="Props"
CategoryVFXPowerup<localizes>:message="VFX Powerups"
CategoryAnim<localizes>:message="Animations"

# A Verse-authored creative device that can be placed in a level
player_ability_device := class(creative_device):

	@editable PlayerClassesDevice : ?player_classes_manager_device = false
	@editable NPCManagerDevice : ?npc_manager_device = false

	@editable CreatureSpawners: []creature_spawner_device = array{}

	@editable InputSkill : input_trigger_device = input_trigger_device{}
	@editable InputJump : input_trigger_device = input_trigger_device{}
	@editable InputCrouch : input_trigger_device = input_trigger_device{}
	@editable InputFire : input_trigger_device = input_trigger_device{}
	@editable InputSprint : input_trigger_device = input_trigger_device{}

	@editable ConditionalHealingBeam : conditional_button_device = conditional_button_device{}

	@editable{Categories:=array{CategorySFX}} SFXAbilityCooldown	: audio_player_device = audio_player_device{}

	@editable ClassSelectorDefault : class_and_team_selector_device = class_and_team_selector_device{}
	@editable ClassSelectorLowGrav : class_and_team_selector_device = class_and_team_selector_device{}

	@editable MoveModWarp : movement_modulator_device  = movement_modulator_device{}
	@editable MoveModWarpAir : movement_modulator_device  = movement_modulator_device{}
	@editable{Categories:=array{CategorySFX}} SFXWarp : audio_player_device = audio_player_device{}
	@editable{Categories:=array{CategorySFX}} SFXWarpZP : audio_player_device = audio_player_device{}
	@editable{Categories:=array{CategorySFX}} SFXWarpChrome : audio_player_device = audio_player_device{}
	@editable{Categories:=array{CategorySFX}} SFXWarpStyx : audio_player_device = audio_player_device{}

	@editable{Categories:=array{CategoryVFXPowerup}} VFXPowerWind : visual_effect_powerup_device = visual_effect_powerup_device{}
	@editable{Categories:=array{CategoryVFXPowerup}} VFXPowerWarp : visual_effect_powerup_device = visual_effect_powerup_device{}
	@editable{Categories:=array{CategoryVFXPowerup}} VFXPowerWarpZP : visual_effect_powerup_device = visual_effect_powerup_device{}
	@editable{Categories:=array{CategoryVFXPowerup}} VFXPowerWarpChrome : visual_effect_powerup_device = visual_effect_powerup_device{}
	@editable{Categories:=array{CategoryVFXPowerup}} VFXPowerWarpStyx : visual_effect_powerup_device = visual_effect_powerup_device{}
	@editable{Categories:=array{CategoryVFXPowerup}} VFXPowerLeapZeus : visual_effect_powerup_device = visual_effect_powerup_device{}
	@editable{Categories:=array{CategoryVFXPowerup}} VFXPowerLeapKor : visual_effect_powerup_device = visual_effect_powerup_device{}

	@editable{Categories:=array{CategoryProp}} VFXPropRockStart : creative_prop_asset = DefaultCreativePropAsset
	@editable{Categories:=array{CategoryProp}} VFXPropRockEnd : creative_prop_asset = DefaultCreativePropAsset
	@editable{Categories:=array{CategoryProp}} VFXPropDoomBomb : creative_prop_asset = DefaultCreativePropAsset

	@editable MoveModNormal : movement_modulator_device  = movement_modulator_device{}
	@editable MoveModStop : movement_modulator_device  = movement_modulator_device{}
	@editable MoveModPlus : movement_modulator_device  = movement_modulator_device{}
	@editable MoveModHighJump : movement_modulator_device  = movement_modulator_device{}
	@editable MoveModDoubleJump : movement_modulator_device  = movement_modulator_device{}
	@editable MoveModLeap : movement_modulator_device  = movement_modulator_device{}
	@editable MoveModStraight : movement_modulator_device  = movement_modulator_device{}
	@editable ExplosiveDeviceHigh : explosive_device  = explosive_device{}
	@editable ExplosiveDeviceLow : explosive_device  = explosive_device{}

	@editable MoveModGroundPound : movement_modulator_device  = movement_modulator_device{}

	@editable{Categories:=array{CategoryProp}} VFXPropSpinSlashStart : creative_prop_asset = DefaultCreativePropAsset
	@editable{Categories:=array{CategoryProp}} VFXPropCerberusSlashStart : creative_prop_asset = DefaultCreativePropAsset

	@editable MirrorImageSpawner : guard_spawner_device = guard_spawner_device{}
	@editable SocietySpawner : guard_spawner_device = guard_spawner_device{}
	@editable DemonSpawner : guard_spawner_device = guard_spawner_device{}

	#var LockdownSpawnerHomePos : vector3 = vector3{}
	@editable LockdownSpawner : wildlife_spawner_device = wildlife_spawner_device{}

	@editable{Categories:=array{CategorySFX}} SFXAbilityLifeDrainStart : audio_player_device = audio_player_device{}
	@editable{Categories:=array{CategorySFX}} SFXAbilityLifeDrainImpact : audio_player_device = audio_player_device{}
	@editable{Categories:=array{CategoryProp}} VFXPropLifeDrainStart : creative_prop_asset = DefaultCreativePropAsset
	@editable{Categories:=array{CategoryProp}} VFXPropLifeDrainKadoStart : creative_prop_asset = DefaultCreativePropAsset
	@editable{Categories:=array{CategoryProp}} VFXPropLifeDrainHadesStart : creative_prop_asset = DefaultCreativePropAsset

	@editable ChargeMovementMod : movement_modulator_device  = movement_modulator_device{}
	@editable{Categories:=array{CategorySFX}} SFXAbilityChargeStart : audio_player_device = audio_player_device{}
	@editable{Categories:=array{CategorySFX}} SFXAbilityChargeImpact : audio_player_device = audio_player_device{}
	@editable{Categories:=array{CategorySFX}} SFXAbilityChargeEnd : audio_player_device = audio_player_device{}
	@editable{Categories:=array{CategoryProp}} VFXPropChargeEnd : creative_prop_asset = DefaultCreativePropAsset

	@editable{Categories:=array{CategorySFX}} SFXAbilityCubeBeamEnd : audio_player_device = audio_player_device{}
	@editable{Categories:=array{CategorySFX}} SFXAbilityLaserBeamHit : audio_player_device = audio_player_device{}
	@editable{Categories:=array{CategoryProp}} VFXPropCubeBeamEnd : creative_prop_asset = DefaultCreativePropAsset
	@editable{Categories:=array{CategoryProp}} VFXPropGenoBeamEnd : creative_prop_asset = DefaultCreativePropAsset
	@editable{Categories:=array{CategoryProp}} VFXPropDoomBeamEnd : creative_prop_asset = DefaultCreativePropAsset
	@editable CineAbilityPP : cinematic_sequence_device = cinematic_sequence_device{}

	@editable{Categories:=array{CategoryProp}} VFXPropSentry : creative_prop_asset = DefaultCreativePropAsset
	@editable{Categories:=array{CategoryProp}} VFXPropCubeAShield : creative_prop_asset = DefaultCreativePropAsset
	#@editable{Categories:=array{CategoryProp}} VFXPropGrindVines : creative_prop_asset = DefaultCreativePropAsset

	@editable{Categories:=array{CategoryProp}} VFXPropJonesRift : creative_prop_asset = DefaultCreativePropAsset
	@editable{Categories:=array{CategoryProp}} VFXPropJonesRiftProj : creative_prop_asset = DefaultCreativePropAsset

	@editable{Categories:=array{CategoryVFXPowerup}} VFXPowerJetDash : visual_effect_powerup_device = visual_effect_powerup_device{}
	@editable{Categories:=array{CategoryVFXPowerup}} VFXPowerJetDashMegaloDon : visual_effect_powerup_device = visual_effect_powerup_device{}
	@editable{Categories:=array{CategoryVFXPowerup}} VFXPowerGenoBeamStart : visual_effect_powerup_device = visual_effect_powerup_device{}
	@editable{Categories:=array{CategoryVFXPowerup}} VFXPowerCubeBeamStart : visual_effect_powerup_device = visual_effect_powerup_device{}
	@editable{Categories:=array{CategoryVFXPowerup}} VFXPowerCaretakerBeamStart : visual_effect_powerup_device = visual_effect_powerup_device{}
	@editable{Categories:=array{CategoryVFXPowerup}} VFXPowerDoomBeamStart : visual_effect_powerup_device = visual_effect_powerup_device{}
	@editable{Categories:=array{CategoryVFXPowerup}} VFXPowerGroundPound : visual_effect_powerup_device = visual_effect_powerup_device{}
	@editable{Categories:=array{CategoryVFXPowerup}} VFXPowerSpireSlam : visual_effect_powerup_device = visual_effect_powerup_device{}
	@editable{Categories:=array{CategoryVFXPowerup}} VFXPowerSpireRush : visual_effect_powerup_device = visual_effect_powerup_device{}
	@editable{Categories:=array{CategoryVFXPowerup}} VFXPowerAlien : visual_effect_powerup_device = visual_effect_powerup_device{}

	@editable{Categories:=array{CategorySFX}} SFXAlienPAttach : audio_player_device = audio_player_device{}
	@editable{Categories:=array{CategorySFX}} SFXAlienPDetach : audio_player_device = audio_player_device{}

	#@editable{Categories:=array{CategoryVFXPowerup}} VFXNitroCharge : visual_effect_powerup_device = visual_effect_powerup_device{}
	@editable{Categories:=array{CategoryVFXPowerup}} VFXPowerFootDiveCharge : visual_effect_powerup_device = visual_effect_powerup_device{}
	@editable{Categories:=array{CategoryVFXPowerup}} VFXPowerFootDive : visual_effect_powerup_device = visual_effect_powerup_device{}
	@editable{Categories:=array{CategoryVFXPowerup}} VFXPowerCaretakerSnap : visual_effect_powerup_device = visual_effect_powerup_device{}
	@editable{Categories:=array{CategoryVFXPowerup}} VFXPowerShogunBombardment : visual_effect_powerup_device = visual_effect_powerup_device{}
	@editable{Categories:=array{CategoryVFXPowerup}} VFXKaneEnrage : visual_effect_powerup_device = visual_effect_powerup_device{}
	@editable{Categories:=array{CategoryAnim}} AnimFootDive : creative_prop_asset = DefaultCreativePropAsset
	@editable{Categories:=array{CategoryAnim}} AnimFootDiveLoop : creative_prop_asset = DefaultCreativePropAsset
	@editable{Categories:=array{CategoryAnim}} AnimSpinSlash : creative_prop_asset = DefaultCreativePropAsset
	@editable{Categories:=array{CategoryAnim}} AnimCaretakerSnap : creative_prop_asset = DefaultCreativePropAsset
	@editable{Categories:=array{CategoryAnim}} AnimNightRoseClaw : creative_prop_asset = DefaultCreativePropAsset
	@editable{Categories:=array{CategoryAnim}} AnimKaneClaw : creative_prop_asset = DefaultCreativePropAsset
	@editable{Categories:=array{CategoryAnim}} AnimKaneEnrage : creative_prop_asset = DefaultCreativePropAsset
	@editable{Categories:=array{CategoryAnim}} AnimShogunBombardment : creative_prop_asset = DefaultCreativePropAsset
	@editable{Categories:=array{CategoryAnim}} AnimDaigoBeam : creative_prop_asset = DefaultCreativePropAsset
	@editable{Categories:=array{CategoryVFXPowerup}} VFXPowerDaigoBeam : visual_effect_powerup_device = visual_effect_powerup_device{}

    # Runs when the device is started in a running game
    OnBegin<override>()<suspends>:void=
	{
		Print("Initialize Player Abilities...")

		InputSkill.PressedEvent.Subscribe(InputPressedSkill)
		InputJump.PressedEvent.Subscribe(InputPressedJump)
		InputCrouch.PressedEvent.Subscribe(InputPressedCrouch)

		MirrorImageSpawner.SpawnedEvent.Subscribe(OnMirrorImageSpawned)
		MirrorImageSpawner.EliminatedEvent.Subscribe(OnMirrorImageEliminated)
		MirrorImageSpawner.HiredEvent.Subscribe(OnMirrorImageHired)
		MirrorImageSpawner.DismissedEvent.Subscribe(OnMirrorImageDismissed)

		SocietySpawner.SpawnedEvent.Subscribe(OnMirrorImageSpawned)
		SocietySpawner.EliminatedEvent.Subscribe(OnMirrorImageEliminated)
		SocietySpawner.HiredEvent.Subscribe(OnMirrorImageHired)
		SocietySpawner.DismissedEvent.Subscribe(OnMirrorImageDismissed)

		DemonSpawner.SpawnedEvent.Subscribe(OnDemonSpawned)
		DemonSpawner.EliminatedEvent.Subscribe(OnMirrorImageEliminated)
		DemonSpawner.HiredEvent.Subscribe(OnMirrorImageHired)
		DemonSpawner.DismissedEvent.Subscribe(OnMirrorImageDismissed)

		LockdownSpawner.SpawnedEvent.Subscribe(OnLockdownSpawned)
		LockdownSpawner.TamedEvent.Subscribe(OnLockdownTamed)
		LockdownSpawner.UntamedEvent.Subscribe(OnLockdownUntamed)
		LockdownSpawner.RiddenEvent.Subscribe(OnLockdownRidden)
		LockdownSpawner.DismountedEvent.Subscribe(OnLockdownDismounted)
		LockdownSpawner.EliminatedEvent.Subscribe(OnLockdownEliminated)
		#set LockdownSpawnerHomePos = LockdownSpawner.GetTransform().Translation

		for( AllCreatureSpawners: CreatureSpawners )
		{
			AllCreatureSpawners.SpawnedEvent.Subscribe(OnCreatureSpawned)
		}
	}

	(FortCharacter : fort_character).DrainHealth(Amount:float):void =
	{
		FortCharacter.SetHealth(FortCharacter.GetHealth() - Amount)
	}

	<#
	(FortCharacter : fort_character).AddShield(Amount:float):void =
	{
		FortCharacter.SetShield(FortCharacter.GetShield() + Amount)
	}
	#>

    # Event that handles when a player is eliminated
	OnPlayerEliminated(Result:elimination_result):void=
	{
		# Who got killed?
		KilledCharacter := Result.EliminatedCharacter

		if
		{
			PlayerClasses := PlayerClassesDevice?
			KilledAgent := KilledCharacter.GetAgent[]
			KillerChar := Result.EliminatingCharacter?
			KillerAgent := KillerChar.GetAgent[]
		}
        then
		{
			if ( PlayerClasses.AgentHasAbility[KilledAgent,FNAbility.PartingGift] )
			{
				spawn{KilledAgent.Explode(800.0, 75.0)}
			}
		}
	}

	GetAllPlayersInVectorDistance( CenterPos : vector3, DistanceRange: float)<transacts>:[]fort_character=
	{
		var ClosestPlayers: []fort_character = array{}

		if ( Entity := GetSimulationEntity[] )
		{
			CollisionSphere := collision_sphere{ Radius := DistanceRange <#, CollisionProfile := collision_profiles.StationaryOverlapAll#> }
			CollisionTransform := FromTransform(transform{Translation := CenterPos})

			for
			{
				Overlap : Entity.FindOverlapHits(CollisionTransform, CollisionSphere)
				PlayerMesh := Meshes.PlayerBox[Overlap.TargetComponent]
				TargetEntity := PlayerMesh.Entity
			}
			do
			{
				if ( PlayerComp := TargetEntity.GetComponent[fn_player_component], Agent := PlayerComp.Agent, FortChar := Agent.GetFortCharacter[] )
				{
					set ClosestPlayers += array{FortChar}
				}

				if ( NPCComp := TargetEntity.GetComponent[fn_npc_component], AgentNPC := NPCComp.Agent, FortNPC := AgentNPC.GetFortCharacter[] )
				{
					set ClosestPlayers += array{FortNPC}
				}
			}
		}


		Drawer := debug_draw{}
		Drawer.DrawSphere(CenterPos, ?Radius := DistanceRange, ?DrawDurationPolicy := debug_draw_duration_policy.FiniteDuration, ?Duration:= float(1.5))
        return ClosestPlayers
	}

	<#GetAllPlayerAndNPC()<transacts>:[]fort_character=
	{
		var ClosestPlayers: []fort_character = array{}
		AllPlayers := GetPlayspace().GetPlayers()
		for (Player : AllPlayers, FortVictim := Player.GetFortCharacter[], FortVictim.IsActive[] )
		{
			set ClosestPlayers += array{FortVictim}
		}

		if ( SimulationEntity := Entity.GetSimulationEntity[] )
		{
			for ( EntNPC : SimulationEntity.FindDescendantEntitiesWithComponent(fn_npc_component), NPCComp := EntNPC.GetComponent[fn_npc_component], AgentNPC := NPCComp.Agent, FortNPC := AgentNPC.GetFortCharacter[] )
			{
				set ClosestPlayers += array{FortNPC}
			}
		}

        return ClosestPlayers
	}#>

	<#GetAllPlayersInPropSphere(PropCenter : creative_prop, DistanceRange: float)<transacts>:[]fort_character=
	{
		var ClosestPlayers: []fort_character = array{}
		if ( PropCenter.IsValid[] )
		{
			Print("GetAllPlayersInPropSphere()")
			Sphere:sphere3 = sphere3{Center := vector3{X := 0.0, Y := 0.0, Z := 0.0}, Radius := DistanceRange}

			for (EntityInArea : PropCenter.GetSimulationEntity[].FindEntitiesOriginInside(Sphere) , FortCharacter := fort_character[EntityInArea])
			{
				Print("Found Entity!")
				set ClosestPlayers += array{FortCharacter}
			}
		}

        return ClosestPlayers
	}#>

	GetAllPlayersInDistance(PlayerCenter : fort_character, DistanceRange: float)<transacts>:[]fort_character=
	{
		CenterCurrentPos := PlayerCenter.GetTransform().Translation
        return GetAllPlayersInVectorDistance(CenterCurrentPos, DistanceRange)
	}

	GetAllPlayersInPropDistance(PropCenter : creative_prop, DistanceRange: float)<transacts>:[]fort_character=
	{
		CenterCurrentPos := PropCenter.GetTransform().Translation
        return GetAllPlayersInVectorDistance(CenterCurrentPos, DistanceRange)
	}

	GetNearestPlayer(CharacterCenter:fort_character, MaxDistance:float)<transacts><decides>:fort_character =
	{
		var MaybeTarget : ?fort_character = false
		var NearestLocation : float = MaxDistance
		AllPlayers := GetPlayspace().GetParticipants()
		if ( Agent := CharacterCenter.GetAgent[] )
		{
			for (Target : AllPlayers, FortCharacter := Target.GetFortCharacter[], TargetAgent := agent[Target], TargetAgent <> Agent)
			{
				if (DistancePlayerToTarget := Distance(FortCharacter.GetTransform().Translation, CharacterCenter.GetTransform().Translation) < NearestLocation, not TargetAgent.IsInSpawnRoom[Self] )
				{
					set MaybeTarget = option{FortCharacter}
					set NearestLocation = DistancePlayerToTarget
				}
			}

			if ( NPCManager := NPCManagerDevice? )
			{
				for (FortNPC : NPCManager.aNPCs, TargetAgent := FortNPC.GetAgent[])
				{
					if (DistancePlayerToTarget := Distance(FortNPC.GetTransform().Translation, CharacterCenter.GetTransform().Translation) < NearestLocation)
					{
						set MaybeTarget = option{FortNPC}
						set NearestLocation = DistancePlayerToTarget
					}
				}
			}

			if ( SimulationEntity := GetSimulationEntity[] )
			{
				for ( EntNPC : SimulationEntity.FindDescendantEntitiesWithComponent(fn_npc_component), NPCComp := EntNPC.GetComponent[fn_npc_component], AgentNPC := NPCComp.Agent, FortNPC := AgentNPC.GetFortCharacter[] )
				{
					if (DistancePlayerToTarget := Distance(FortNPC.GetTransform().Translation, CharacterCenter.GetTransform().Translation) < NearestLocation)
					{
						set MaybeTarget = option{FortNPC}
						set NearestLocation = DistancePlayerToTarget
					}
				}
			}
		}

		MaybeTarget?
	}

	GetNearestEnemy(CharacterCenter:fort_character, MaxDistance:float)<transacts><decides>:fort_character =
	{
		var MaybeTarget : ?fort_character = false
		var NearestLocation : float = MaxDistance
		AllPlayers := GetPlayspace().GetParticipants()
		TeamCollection := GetPlayspace().GetTeamCollection()
		if ( Agent := CharacterCenter.GetAgent[], TeamOwner := TeamCollection.GetTeam[Agent] )
		{
			for (Target : AllPlayers, FortCharacter := Target.GetFortCharacter[], TargetAgent := agent[Target], TargetAgent <> Agent, not TeamCollection.IsOnTeam[TargetAgent, TeamOwner])
			{
				if (DistancePlayerToTarget := Distance(FortCharacter.GetTransform().Translation, CharacterCenter.GetTransform().Translation) < NearestLocation, not TargetAgent.IsInSpawnRoom[Self] )
				{
					set MaybeTarget = option{FortCharacter}
					set NearestLocation = DistancePlayerToTarget
				}
			}

			if ( SimulationEntity := GetSimulationEntity[] )
			{
				for ( EntNPC : SimulationEntity.FindDescendantEntitiesWithComponent(fn_npc_component), NPCComp := EntNPC.GetComponent[fn_npc_component], AgentNPC := NPCComp.Agent, FortNPC := AgentNPC.GetFortCharacter[], not TeamCollection.IsOnTeam[AgentNPC, TeamOwner] )
				{
					if (DistancePlayerToTarget := Distance(FortNPC.GetTransform().Translation, CharacterCenter.GetTransform().Translation) < NearestLocation)
					{
						set MaybeTarget = option{FortNPC}
						set NearestLocation = DistancePlayerToTarget
					}
				}
			}
		}

		MaybeTarget?
	}

	GetNearestTeammate(CharacterCenter : fort_character)<transacts><decides>:fort_character =
	{
		var MaybeTarget : ?fort_character = false
		var NearestLocation : float = 1500.0
		AllPlayers := GetPlayspace().GetParticipants()
		TeamCollection := GetPlayspace().GetTeamCollection()
		if ( Agent := CharacterCenter.GetAgent[], TeamOwner := TeamCollection.GetTeam[Agent] )
		{
			#Print("GetNearestTeammate().")
			if ( SimulationEntity := GetSimulationEntity[] )
			{
				for ( EntNPC : SimulationEntity.FindDescendantEntitiesWithComponent(fn_npc_component), NPCComp := EntNPC.GetComponent[fn_npc_component], AgentNPC := NPCComp.Agent, FortNPC := AgentNPC.GetFortCharacter[], TeamCollection.IsOnTeam[AgentNPC, TeamOwner] )
				{
					if (DistancePlayerToTarget := Distance(FortNPC.GetTransform().Translation, CharacterCenter.GetTransform().Translation) < NearestLocation)
					{
						set MaybeTarget = option{FortNPC}
						set NearestLocation = DistancePlayerToTarget
					}
				}
			}

			#var MapDistance : [fort_character]float = map{}
			for (Target : AllPlayers, FortCharacter := Target.GetFortCharacter[], TargetAgent := agent[Target], TargetAgent <> Agent, TeamCollection.IsOnTeam[Target, TeamOwner])
			{
				if (DistancePlayerToTarget := Distance(FortCharacter.GetTransform().Translation, CharacterCenter.GetTransform().Translation) < NearestLocation)
				{
					Print("Found nearest teammate.")
					set MaybeTarget = option{FortCharacter}
					set NearestLocation = DistancePlayerToTarget
				}
			}
		}

		NearestTarget := MaybeTarget?
	}

	GetNearestTeammateToHeal(CharacterCenter : fort_character)<transacts><decides>:fort_character =
	{
		var MaybeTarget : ?fort_character = false
		if ( ClosestTeammate := GetNearestTeammate[CharacterCenter], ClosestTeammate.IsActive[] )
		{
			set MaybeTarget = option{ClosestTeammate}
		}

		NearestTarget := MaybeTarget?
	}

    InputPressedJump(Agent : agent) : void =
	{
		if (FortCharacter := Agent.GetFortCharacter[], FortCharacter.IsActive[], PlayerClasses := PlayerClassesDevice?)
		{
			if ( FortCharacter.IsOnGround[] )
			{
				if (set PlayerClasses.PlayersMap[Agent].bDoubleJumped = false) {}
			}

			if ( FortCharacter.IsInAir[] and ( not FortCharacter.IsGliding[] ) )
			{
				if ( PlayerClasses.AgentHasAbility[Agent, FNAbility.DoubleJump] )
				{
					PlayerCooldown := PlayerClasses.GetAbilityCooldown(Agent, FNAbility.DoubleJump)
					if ( PlayerCooldown > 0.0 )
					{
						SFXAbilityCooldown.Play(Agent)
						return
					}

					case( PlayerClasses.GetPlayerClassIndex(Agent) )
					{
						FNClasses.TheFoundation =>		spawn{ AbilityHighJump(Agent) }
						FNClasses.KadoThorne =>			spawn{ AbilityHighJump(Agent) }
						FNClasses.Zeus =>				spawn{ AbilityLeap(Agent) }
						FNClasses.Kor =>				spawn{ AbilityLeap(Agent) }
						FNClasses.Daigo =>
						{
							if ( InputSprint.IsHeld[Agent] )
							{
								spawn{ AbilityCharge(Agent) }
							}
							else
							{
								spawn{ AbilityHighJump(Agent) }
							}
						}
						_ =>							spawn{ AbilityDoubleJump(Agent) }
					}
					spawn{PlayerClasses.AbilityCooldown(Agent, FNAbility.DoubleJump)}
				}

				if ( PlayerClasses.AgentHasAbility[Agent, FNAbility.Warp] )
				{
					PlayerCooldown := PlayerClasses.GetAbilityCooldown(Agent, FNAbility.Warp)
					if ( PlayerCooldown > 0.0 )
					{
						SFXAbilityCooldown.Play(Agent)
						return
					}

					case( PlayerClasses.GetPlayerClassIndex(Agent) )
					{
						FNClasses.SpireAssassin =>		return
						FNClasses.Shadow =>				spawn{ AbilityWarp(Agent, 1) }
						FNClasses.Raz =>				spawn{ AbilityWarp(Agent, 1) }
						FNClasses.CubeAssassin =>		spawn{ AbilityWarp(Agent, 1) }
						FNClasses.HuntmasterSaber =>	spawn{ AbilityWarp(Agent, 0) }
						FNClasses.WildguardRelik =>		spawn{ AbilityWarp(Agent, 0) }
						FNClasses.TheHerald =>			spawn{ AbilityWarp(Agent, 2) }
						FNClasses.Cerberus =>			spawn{ AbilityWarp(Agent, 3) }
						_ =>							spawn{ AbilityWarp(Agent, 0) }
					}
					spawn{PlayerClasses.AbilityCooldown(Agent, FNAbility.Warp)}
				}

				if ( PlayerClasses.AgentHasAbility[Agent, FNAbility.FootDive] )
				{
					spawn{ AbilityFootDive(Agent) }
				}
			}
		}
	}

	<#HoldSkill(Agent:agent)<suspends>:void=
	{
		if (FortCharacter := Agent.GetFortCharacter[], PlayerClasses := PlayerClassesDevice?, PlayerClasses.AgentIsAltTransform[Agent] )
		{
			race
			{
				block # Tab
				{
					InputSkill.ReleasedEvent.Await()
					spawn{ AbilityRock(Agent) }
				}
				block # Hold 
				{
					Sleep(0.5)
					spawn{AbilityLaserBeam(Agent)}
				}
			}
		}
	}#>

    InputPressedCrouch(Agent : agent) : void =
	{
		if (FortCharacter := Agent.GetFortCharacter[], FortCharacter.IsActive[], FortCharacter.IsInAir[], PlayerClasses := PlayerClassesDevice?, PlayerClasses.PlayersMap[Agent].bDoubleJumped? )
		{
			if (PlayerClasses.PlayersMap[Agent].bGroundPounding?)
			{
				return
			}

			case( PlayerClasses.GetPlayerClassIndex(Agent) )
			{
				FNClasses.KadoThorne =>			spawn{ HoldCrouch(Agent) }
				FNClasses.TheFoundation =>		spawn{ HoldCrouch(Agent) }
				FNClasses.Zeus =>				spawn{ HoldCrouch(Agent) }
				FNClasses.Daigo =>				spawn{ HoldCrouch(Agent) }
				_ =>							#nothing
			}
		}
	}

	HoldCrouch(Agent:agent)<suspends>:void=
	{
		if (FortCharacter := Agent.GetFortCharacter[] )
		{
			race
			{
				block
				{
					InputCrouch.ReleasedEvent.Await()
					Print("Crouch Cancel.")
				}
				block
				{
					Sleep(0.5)
					spawn{AbilityGroundPound(Agent)}
				}
			}
		}
	}

	AbilityGroundPound(Agent:agent)<suspends>:void=
	{
		if (FortCharacter := Agent.GetFortCharacter[], PlayerClasses := PlayerClassesDevice? )
		{
			if ( FortCharacter.IsOnGround[] )
			{
				return
			}

			Print("GroundPound: Activate.")
			if {set PlayerClasses.PlayersMap[Agent].bGroundPounding = true}

			case( PlayerClasses.GetPlayerClassIndex(Agent) )
			{
				FNClasses.Daigo =>				VFXPowerSpireSlam.Pickup(Agent)
				_ =>							VFXPowerGroundPound.Pickup(Agent)
			}

			Agent.MoveMod(MoveModGroundPound)
			race
			{
				block
				{
					loop
					{
						if ( not FortCharacter.IsInAir[] )
						{
							Print("GroundPound: Landed.")
							break
						}
						Sleep(0.1)
					}
				}
				block
				{
					FortCharacter.EliminatedEvent().Await()
					Print("GroundPound: Failed.")
				}
			}
			SpawnPos := FortCharacter.GetTransform().Translation

			case( PlayerClasses.GetPlayerClassIndex(Agent) )
			{
				FNClasses.Daigo =>				spawn{SpawnParticleVFX(FX.NS_Daigo_SpireSlam_asset, SpawnPos, rotation{}, 4.0)}
				_ =>							spawn{SpawnPropVFX(VFXPropRockEnd,SpawnPos,rotation{},1.2)}
			}

			if ( ExplosiveDeviceLow.TeleportTo[SpawnPos, rotation{}] )
			{
				ExplosiveDeviceLow.Explode(Agent)
			}

			if {set PlayerClasses.PlayersMap[Agent].bGroundPounding = false}

			TeamCollection := GetPlayspace().GetTeamCollection()
			if ( TeamOwner := TeamCollection.GetTeam[Agent] )
			{
				AllVictim := GetAllPlayersInVectorDistance(SpawnPos, 800.0)
				for ( ClosestVictim : AllVictim, ClosestAgent := ClosestVictim.GetAgent[], not TeamCollection.IsOnTeam[ClosestAgent, TeamOwner], FortCharacter.IsActive[] )
				{
					var DmgMultiplier : float = 1.0
					if ( ClosestPlayer := player[ClosestVictim] ) # if player
					{
						if ( ClosestPlayer.IsActive[] )
						{
							set DmgMultiplier = 1.0
							ClosestAgent.MoveMod(MoveModDoubleJump)
						}
					}
					else
					{
						set DmgMultiplier = 2.5
					}
					if ( FortCharacter.IsActive[], DamageInstigator := Agent.GetInstigator(), DamageSource := game_action_causer[FortCharacter] )
					{
						RockDmg : damage_args = damage_args{Instigator := option{DamageInstigator}, Source := option{DamageSource}, Amount:=(35.0*DmgMultiplier)*PlayerClasses.GetDamageBonus(Agent)}
						ClosestVictim.Damage(RockDmg)
					}
				}
			}
		}
	}

	# bosses ability
	InputPressedSkill(Agent : agent) : void =
	{
		if (FortCharacter := Agent.GetFortCharacter[], FortCharacter.IsActive[], not FortCharacter.IsGliding[], PlayerClasses := PlayerClassesDevice? )
		{
			<#if ( GetUltimateMeter(Agent) = 100.0 )
			{
				UseUltimate(Agent)
				return
			}#>

			if ( Agent.IsInSpawnRoom[Self] )
			{
				Print("Cannot use Ability in spawn room.")
				PlayerClasses.AddHudStatusMessage(Agent, TextLocalization.MSG_NoAbilityInSpawn)
				SFXAbilityCooldown.Play(Agent)
				return
			}

			UseAbility(Agent)
		}
	}

    UseAbility(Agent : agent) : void =
	{
		if ( PlayerClasses := PlayerClassesDevice?, NewAbility := PlayerClasses.GetMainAbility[Agent] )
		{
			PlayerCooldown := PlayerClasses.GetAbilityCooldown(Agent, NewAbility)
			if ( PlayerCooldown > 0.0 )
			{
				Print("Ability on cooldown...")
				PlayerClasses.AddHudStatusMessage(Agent, TextLocalization.MSG_AbilityCool)
				SFXAbilityCooldown.Play(Agent)
				return
			}

			case( NewAbility )
			{
				FNAbility.Warp =>
				{
					case( PlayerClasses.GetPlayerClassIndex(Agent) )
					{
						FNClasses.SpireAssassin =>		spawn{ AbilityWarp(Agent, 1) }
						FNClasses.Shadow =>				spawn{ AbilityWarp(Agent, 1) }
						FNClasses.CubeAssassin =>		spawn{ AbilityWarp(Agent, 1) }
						FNClasses.WildguardRelik =>		spawn{ AbilityWarp(Agent, 0) }
						_ =>							spawn{ AbilityWarp(Agent, 0) }
					}
				}
				FNAbility.LifeDrain =>
				{
					case( PlayerClasses.GetPlayerClassIndex(Agent) )
					{
						FNClasses.Hades =>				spawn{ AbilityLifeDrain(Agent, true) }
						_ =>							spawn{ AbilityLifeDrain(Agent, false) }
					}
				}
				FNAbility.Rock =>
				{
					spawn{ AbilityRock(Agent) }
				}
				FNAbility.Charge =>				spawn{ AbilityCharge(Agent) }
				FNAbility.SpinSlash =>			spawn{ AbilitySpinSlash(Agent) }
				FNAbility.AlienParasite =>		spawn{ AlienParasitePlayer(Agent, 10.0) }
				FNAbility.MirrorImage =>
				{
					case( PlayerClasses.GetPlayerClassIndex(Agent) )
					{
						FNClasses.Slone =>				spawn{ AbilityMirrorImage(Agent) }
						FNClasses.Kor =>				spawn{ AbilityDemonSpawner(Agent) }
						_ =>							spawn{ AbilityMirrorImage(Agent) }
					}
				}
				FNAbility.Lockdown =>			spawn{ AbilityLockdownSummon(Agent) }
				FNAbility.Phase2 =>				spawn{ AbilityShieldRegen(Agent, false) }
				FNAbility.DamageBuff =>			spawn{ DamageBuffPlayer(Agent, 5.0, 1.2) }
				FNAbility.HealingRune =>		spawn{ PlayerClasses.BlueCubeHeal(Agent) }
				FNAbility.Laserbeam =>			spawn{ AbilityLaserBeam(Agent) }
				FNAbility.RiftGun =>			spawn{ AbilityRiftGun(Agent) }
				FNAbility.Engineer =>			spawn{ AbilityBuildingSentry(Agent) }
				FNAbility.PersonalShield =>		spawn{ AbilityObjShield(Agent) }
				FNAbility.FootDive =>			spawn{ AbilityFootDive(Agent) }
				FNAbility.Snap =>				spawn{ AbilitySnap(Agent) }
				FNAbility.GrindVines =>			spawn{ AbilityGrindVines(Agent) }
				FNAbility.Bombardment =>		spawn{ AbilityBombardment(Agent) }
				FNAbility.DemonClaw =>			spawn{ AbilityMelee(Agent) }
				FNAbility.ShortBeam =>			spawn{ AbilityShortBeam(Agent) }
				_ =>							return
			}

			spawn{PlayerClasses.AbilityCooldown(Agent, NewAbility)}
		}
	}

	#
	# warp ability
	# 
	AbilityWarp(Agent:agent, WarpType:int)<suspends>:void=
	{
		if
		{
			FortCharacter := Agent.GetFortCharacter[]
			FortCharacter.IsActive[]
			PlayerClasses := PlayerClassesDevice?
			not PlayerClasses.PlayersMap[Agent].bDoubleJumped?
			set PlayerClasses.PlayersMap[Agent].bDoubleJumped = true
		}
		then
		{
			race # Race to prevent VERSE INTERNAL ERROR.
			{
				block
				{
					Print("Warp: Activated")
					PlayerClasses.PlayClassVO(Agent, FNVOType.Ability)
					Agent.MoveMod(MoveModStop)

					PlayerPos := FortCharacter.GetTransform().Translation
					case( WarpType )
					{
						1 =>
						{
							VFXPowerWarpZP.Pickup(Agent)
							SFXWarpZP.Play(Agent)
						}
						2 =>
						{
							VFXPowerWarpChrome.Pickup(Agent)
							SFXWarpChrome.Play(Agent)
						}
						3 =>
						{
							VFXPowerWarpStyx.Pickup(Agent)
							SFXWarpStyx.Play(Agent)
						}
						_ =>
						{
							VFXPowerWarp.Pickup(Agent)
							SFXWarp.Play(Agent)
						}
					}

					Sleep(0.3)
					#PlayerNewPos := FortCharacter.GetTransform().Translation + FortCharacter.GetViewRotation().GetLocalForward() * 750.0 # MaxDistance = 1250
					#PlayerNewRotation := FortCharacter.GetViewRotation()
					VFXPowerWind.Pickup(Agent)
					if ( not FortCharacter.IsInAir[] )
					{
						Agent.MoveMod(MoveModWarp)
					}
					else
					{
						Agent.MoveMod(MoveModWarpAir)
					}
					#if ( FortCharacter.TeleportTo[PlayerNewPos, PlayerNewRotation] ) {}

					Agent.MoveMod(MoveModNormal)
				}
				block # slow af, require IsActive[] in code above
				{
					FortCharacter.EliminatedEvent().Await()
					Print("Successfully stopped reality collapse event!!!")
				}
			}
		}
	}

	#
	# raz + kado + hades' ability
	# 
    AbilityLifeDrain(Agent:agent, bHealTeammate:logic)<suspends>:void=
	{
		if ( FortCharacter := Agent.GetFortCharacter[], PlayerClasses := PlayerClassesDevice? )
		{
			Print("Life Drain: Activated")
			var NumberOfDrains : int = 0
			var DrainRange : float = 1280.0 #Kado = 2600.0, Raz = 1280

			SFXAbilityLifeDrainStart.Play(Agent)
			PlayerClasses.PlayClassVO(Agent, FNVOType.Ability)

			if ( PlayAnimController := FortCharacter.GetPlayAnimationController[] )
			{
				PlayAnimController.Play(Animations.Hades_Idle)
			}

			Sleep(1.75) # 2.0

			var VFXType : creative_prop_asset = VFXPropLifeDrainKadoStart
			var VFXHeight : float = 128.0
			case( PlayerClasses.GetPlayerClassIndex(Agent) )
			{
				FNClasses.Raz =>
				{
					set VFXType = VFXPropLifeDrainStart
					set VFXHeight = 128.0
				}
				FNClasses.KadoThorne =>
				{
					set VFXType = VFXPropLifeDrainKadoStart
					set VFXHeight = 50.0
				}
				FNClasses.Hades =>
				{
					set VFXType = VFXPropLifeDrainHadesStart
					set VFXHeight = 128.0
				}
				_ =>							#nothing
			}

			PlayerCurrentPos := FortCharacter.GetTransform().Translation + vector3{X:= 0.0, Y:= 0.0, Z:= VFXHeight}

			SpawnVFX := SpawnProp(VFXType, PlayerCurrentPos, rotation{})

			Agent.MoveMod(MoveModStop)
			PlayerClasses.PlayClassVO(Agent, FNVOType.AbilityEnd)

			branch
			{
				if (SpawnedVFX:= SpawnVFX(0)?)
				{
					defer
					{
						SpawnedVFX.Dispose()
					}
					loop
					{
						if ( NumberOfDrains >= 4 or not FortCharacter.IsActive[] )
						{
							break
						}

						PlayerNewPos := FortCharacter.GetTransform().Translation + vector3{X:= 0.0, Y:= 0.0, Z:= VFXHeight}
						#SpawnedVFX.MoveTo(PlayerNewPos, rotation{}, 0.15)
						spawn{SpawnedVFX.SmoothMoveTo(PlayerNewPos, 0.1)}
						Sleep(0.1)
					}
				}
			}
			TeamCollection := GetPlayspace().GetTeamCollection()
			if ( TeamOwner := TeamCollection.GetTeam[Agent] )
			{
				loop
				{
					AllVictim := GetAllPlayersInDistance(FortCharacter, DrainRange)
					for ( ClosestVictim : AllVictim,ClosestAgent := ClosestVictim.GetAgent[], ( ClosestAgent <> Agent ) )
					{
						if ( not TeamCollection.IsOnTeam[ClosestAgent, TeamOwner] )
						{
							spawn{Agent.LifeDrain(ClosestAgent)}
						}
						else
						{
							if ( bHealTeammate? )
							{
								Print("Life Drain: Heal Teammate")
								if ( FortCharacter.IsActive[], HealInstigator := Agent.GetInstigator(), HealSource := game_action_causer[FortCharacter] )
								{
									DrainHeal : healing_args = healing_args{Instigator := option{HealInstigator}, Source := option{HealSource}, Amount:=10.0}
									ClosestVictim.Heal(DrainHeal)
								}
								VictimPos := ClosestVictim.GetTransform().Translation + vector3{X:= 0.0, Y:= 0.0, Z:= 75.0}
								spawn{ SpawnParticleVFX(FX.NS_LifeDrain_Impact_Burst_asset, VictimPos, rotation{}, 1.2) }
							}
						}

						FortCharacter.Heal(35.0)
					}
					FortCharacter.DrainHealth(1.0)

					ExplosivePos := FortCharacter.GetTransform().Translation
					if ( ExplosiveDeviceLow.TeleportTo[ExplosivePos, rotation{}] ) {}
					ExplosiveDeviceLow.Explode(Agent)

					set NumberOfDrains += 1
					if ( NumberOfDrains >= 4 or not FortCharacter.IsActive[] )
					{
						break
					}

					Sleep(1.5)
				}
				Agent.MoveMod(MoveModNormal)
			}
		}
	}

	(AgentOwner:agent).LifeDrain(Agent:agent)<suspends>:void=
	{
		if ( FortOwner := AgentOwner.GetFortCharacter[], FortVictim := Agent.GetFortCharacter[], PlayerClasses := PlayerClassesDevice? )
		{		
			Print("Life Drain: Damage Victim")
			if ( FortOwner.IsActive[], DamageInstigator := AgentOwner.GetInstigator(), DamageSource := game_action_causer[FortOwner] )
			{
				DrainDmg : damage_args = damage_args {Instigator := option{DamageInstigator} , Source := option{DamageSource}, Amount:=20.0*PlayerClasses.GetDamageBonus(Agent)}
				FortVictim.Damage(DrainDmg)
			}

			VictimPos := FortVictim.GetTransform().Translation + vector3{X:= 0.0, Y:= 0.0, Z:= 50.0}
			spawn{ SpawnParticleVFX(FX.NS_LifeDrain_Impact_Burst_asset, VictimPos, rotation{}, 1.2) }
			SFXAbilityLifeDrainImpact.Play(Agent)

			if ( SimulationEntity := GetSimulationEntity[] )
			{
				OwnerTransform := FromTransform(transform{Translation := FortOwner.GetTransform().Translation + vector3{X:= 0.0, Y:= 0.0, Z:= 50.0}})
				NE := entity{}
				TransComp := transform_component{ Entity := NE }

				case( PlayerClasses.GetPlayerClassIndex(AgentOwner) )
				{
					FNClasses.KadoThorne =>
					{
						VFXComp := FX.NS_LifeDrain_Ribbon_Kado{ Entity := NE, Enabled := true, AutoPlay := true, StopLocation := FromVector3(VictimPos) }
						NE.AddComponents(array{TransComp,VFXComp})
					}
					FNClasses.Hades =>
					{
						VFXComp := FX.NS_LifeDrain_Ribbon_Hades{ Entity := NE, Enabled := true, AutoPlay := true, StopLocation := FromVector3(VictimPos) }
						NE.AddComponents(array{TransComp,VFXComp})
					}
					_ =>
					{
						VFXComp := FX.NS_LifeDrain_Ribbon{ Entity := NE, Enabled := true, AutoPlay := true, StopLocation := FromVector3(VictimPos) }
						NE.AddComponents(array{TransComp,VFXComp})
					}
				}

				SimulationEntity.AddEntities(array{NE})
				NE.SetGlobalTransform(OwnerTransform)

				Sleep(1.0)
				NE.RemoveFromParent()
			}
		}
	}

	#
	# slone's ability
	# 
	OnMirrorImageSpawned(GuardAgent : agent) : void =
	{
		if ( FortCharacter := GuardAgent.GetFortCharacter[] )
		{
			spawn{FortCharacterLifeTime(FortCharacter,15.0)}
			SpawnPos := FortCharacter.GetTransform().Translation
			spawn{ SpawnParticleVFX(FX.NS_MirrorImage_PreTell_asset, SpawnPos, rotation{}, 1.5) }
			FortCharacter.DamagedEvent().Subscribe(OnNPCDamaged)
		}
	}

	OnDemonSpawned(GuardAgent : agent) : void =
	{
		if ( FortCharacter := GuardAgent.GetFortCharacter[] )
		{
			spawn{FortCharacterLifeTime(FortCharacter,15.0)}
			SpawnPos := FortCharacter.GetTransform().Translation
			spawn{ SpawnParticleVFX(FX.NS_Warp_NightRose_asset, SpawnPos, rotation{}, 2.0) }
			FortCharacter.DamagedEvent().Subscribe(OnNPCDamaged)
		}
	}

	OnMirrorImageEliminated(Result:device_ai_interaction_result) : void =
	{
		if ( Guard := Result.Target?, FortCharacter := Guard.GetFortCharacter[] )
		{
			#NPCManager.RemoveNPCCount(FortCharacter)
		}
	}

	# become henchmen
    OnMirrorImageHired(Result:device_ai_interaction_result) : void =
	{
		if (Hirer := Result.Source?, HirerFort := Hirer.GetFortCharacter[], PlayerClasses := PlayerClassesDevice?)
		{
			if (Guard := Result.Target?, GuardFort := Guard.GetFortCharacter[])
			{
				if (set PlayerClasses.PlayersMap[Hirer].ActiveSummon = option{GuardFort}) {}
				HirerPos := HirerFort.GetTransform().Translation
				HirerRotation := HirerFort.GetViewRotation()
				if ( GuardFort.TeleportTo[HirerPos, HirerRotation] )
				{
					if ( LeashInterface := GuardFort.GetFortLeashable[] )
					{
						LeashInterface.SetLeashAgent(Hirer, 200.0, 500.0)
					}
				}
			}
		}
	}

    OnMirrorImageDismissed(Result:device_ai_interaction_result) : void =
	{
		if (Hirer := Result.Source?, PlayerClasses := PlayerClassesDevice?)
		{
			if (set PlayerClasses.PlayersMap[Hirer].ActiveSummon = false) {}
		}
	}

    AbilityMirrorImage(Agent:agent)<suspends>:void=
	{
		if ( PlayerClasses := PlayerClassesDevice? )
		{
			if ( FortSummon := PlayerClasses.PlayersMap[Agent].ActiveSummon?, FortSummon.IsActive[] )
			{
				MirrorImageSpawner.DismissAgentHiredGuards(Agent)
			}

			if ( FortCharacter := Agent.GetFortCharacter[] )
			{
				race
				{
					block
					{
						PlayerClasses.PlayClassVO(Agent, FNVOType.Ability)
						Agent.MoveMod(MoveModStop)

						PlayerPos := FortCharacter.GetTransform().Translation + vector3{X:= 0.0, Y:= 0.0, Z:= 50.0}
						spawn{ SpawnParticleVFX(FX.NS_MirrorImage_PreTell_asset, PlayerPos, rotation{}, 1.6) }

						Sleep(1.5)
						Agent.MoveMod(MoveModNormal)
						MirrorImageSpawner.Spawn(Agent)
					}
					block
					{
						FortCharacter.EliminatedEvent().Await()
						Print("Successfully stopped reality collapse event!!!")
					}
				}
			}
		}
	}

    AbilityDemonSpawner(Agent:agent)<suspends>:void=
	{
		if ( PlayerClasses := PlayerClassesDevice? )
		{
			if ( FortSummon := PlayerClasses.PlayersMap[Agent].ActiveSummon?, FortSummon.IsActive[] )
			{
				DemonSpawner.DismissAgentHiredGuards(Agent)
			}

			if ( FortCharacter := Agent.GetFortCharacter[] )
			{
				race
				{
					block
					{
						PlayerClasses.PlayClassVO(Agent, FNVOType.Ability)
						Agent.MoveMod(MoveModStop)

						PlayerPos := FortCharacter.GetTransform().Translation + vector3{X:= 0.0, Y:= 0.0, Z:= 50.0}
						spawn{ SpawnParticleVFX(FX.NS_Warp_NightRose_asset, PlayerPos, rotation{}, 1.6) }

						Sleep(1.5)
						Agent.MoveMod(MoveModNormal)
						DemonSpawner.Spawn(Agent)
					}
					block
					{
						FortCharacter.EliminatedEvent().Await()
						Print("Successfully stopped reality collapse event!!!")
					}
				}
			}
		}
	}

	OnNPCDamaged(Result:damage_result):void=
	{
		if (DamagedChar := fort_character[Result.Target], DamagedAgent := DamagedChar.GetAgent[], PlayerClasses := PlayerClassesDevice?)
		{
			if( AttackerAgent := Result.Instigator?.GetInstigatorAgent[], DamagedAgent <> AttackerAgent )
			{
				PlayerClasses.PlayClassVO(AttackerAgent, FNVOType.FullAlert)

				PlayerClasses.AddUltimateMeter(AttackerAgent, 1.0*(Result.Amount / 25.0))

				spawn{PlayerClasses.UpdateHudDamageText(AttackerAgent, Result.Amount)}

				<#if ( PlayerClasses.AgentHasAbility[AttackerAgent,FNAbility.AirRaid], FortAttacker := AttackerAgent.GetFortCharacter[], FortAttacker.IsInAir[] )
				{
					spawn{FortAttacker.AirRaidDamage(DamagedChar, Result.Amount / 2.5)}
				}#>
			}
		}
	}

	<#(FortCharacter : fort_character).AirRaidDamage(FortVictim:fort_character, DmgAmount:float)<suspends>:void=
	{
		if ( FortCharacter.IsActive[], Agent := FortCharacter.GetAgent[], DamageInstigator := Agent.GetInstigator(), DamageSource := game_action_causer[FortCharacter] )
		{
			Print("AirRaidDamage()")
			BeamDmg : damage_args = damage_args {Instigator := option{DamageInstigator}, Source := option{DamageSource}, Amount:=DmgAmount}
			FortVictim.Damage(BeamDmg)
		}
	}#>

	#
	# foundation's ability
	# 
    AbilityRock(Agent:agent)<suspends>:void=
	{
		if ( FortCharacter := Agent.GetFortCharacter[], PlayerClasses := PlayerClassesDevice? )
		{
			Print("The Rock: Activate")
			PlayerClasses.PlayClassVO(Agent, FNVOType.Ability)
			Agent.MoveMod(MoveModStop)

			PlayerPos := FortCharacter.GetTransform().Translation + vector3{ X := 0.0, Y := 0.0, Z := 128.0}
			PlayerRotation := FortCharacter.GetViewRotation()

			var VFXType : creative_prop_asset = VFXPropRockStart
			var PropScale : vector3 = vector3{X:= 1.0, Y:= 1.0, Z:= 1.0}
			ClassIndex := PlayerClasses.GetPlayerClassIndex(Agent)
			if ( ClassIndex = FNClasses.DoctorDoom or ClassIndex = FNClasses.DoctorDoomMythic )
			{
				set VFXType = VFXPropDoomBomb
			}

			if ( PlayerClasses.AgentIsAltTransform[Agent] )
			{
				set PropScale = vector3{X:= 3.0, Y:= 3.0, Z:= 3.0}
			}

			NewTransform := transform{Scale := PropScale, Rotation := PlayerRotation, Translation := PlayerPos}

			StartVFX := SpawnProp(VFXType, NewTransform)

			var bRockReady : logic = false
			branch
			{
				if (StartedVFX:= StartVFX(0)?)
				{
					loop
					{
						if ( bRockReady? )
						{
							break
						}
						PlayerNewPos := FortCharacter.GetTransform().Translation + vector3{ X := 0.0, Y := 0.0, Z := 128.0}
						#StartedVFX.MoveTo(PlayerNewPos, rotation{}, 0.15)
						spawn{StartedVFX.SmoothMoveTo(PlayerNewPos, 0.1)}
						Sleep(0.1)
					}
				}
			}
			Sleep(1.25) #2.25
			set bRockReady = true
			if (StartedVFX:= StartVFX(0)?)
			{
				Print("The Rock: Yeet")
				if ( ClassIndex <> FNClasses.DoctorDoom )
				{
					PlayerClasses.PlayClassVO(Agent, FNVOType.AbilityEnd)
				}
				PlayerForwardPos := FortCharacter.GetTransform().Translation + FortCharacter.GetViewRotation().GetLocalForward() * 400.0 # PrjDistanceToTarget = 1500.0
				#StartedVFX.MoveTo(PlayerForwardPos, rotation{}, 0.15)

				if (AController := StartedVFX.GetAnimationController[])
				{
					AnimKeyframes := array
					{
						keyframe_delta
						{
							DeltaLocation := PlayerForwardPos - StartedVFX.GetTransform().Translation
							DeltaRotation := rotation{}
							Time := 0.15
							Interpolation := EaseOut
						}
					}
					AController.SetAnimation(AnimKeyframes, ?Mode:=animation_mode.OneShot)
					AController.Play()
				}
				Sleep(0.15)

				StartedVFXPos := StartedVFX.GetTransform().Translation
				StartedVFX.Dispose()
				EndVFX := SpawnProp(VFXPropRockEnd, StartedVFXPos, rotation{})

				if ( EndedVFX:= EndVFX(0)? )
				{
					EndedVFXPos := EndedVFX.GetTransform().Translation
					if ( ExplosiveDeviceHigh.TeleportTo[EndedVFXPos, rotation{}] ) {}
					ExplosiveDeviceHigh.Explode(Agent)
					
					if ( PlayerClasses.AgentIsAltTransform[Agent] )
					{
						spawn{SpawnParticleVFX(FX.NS_Doom_Gigabomb_Explo_asset,EndedVFXPos,rotation{},2.0)}
					}

					TeamCollection := GetPlayspace().GetTeamCollection()
					if ( TeamOwner := TeamCollection.GetTeam[Agent] )
					{
						var RockRadius : float = 800.0
						if ( PlayerClasses.AgentIsAltTransform[Agent] )
						{
							set RockRadius = 1000.0
						}
						AllVictim := GetAllPlayersInPropDistance(EndedVFX, RockRadius) #1250.0
						for ( ClosestVictim : AllVictim, ClosestAgent := ClosestVictim.GetAgent[], not TeamCollection.IsOnTeam[ClosestAgent, TeamOwner] )
						{
							Print("The Rock: Damage Victim")
							var DmgMultiplier : float = 1.0
							var DmgBase : float = 90.0
							if ( PlayerClasses.AgentIsAltTransform[Agent] )
							{
								set DmgBase = 180.0
							}

							if ( ClosestPlayer := player[ClosestVictim] ) # if player
							{
								#MoveModKnockback.Activate(ClosestAgent)
							}
							else
							{
								set DmgMultiplier = 2.5
							}

							if ( FortCharacter.IsActive[], DamageInstigator := Agent.GetInstigator(), DamageSource := game_action_causer[FortCharacter] )
							{
								RockDmg : damage_args = damage_args {Instigator := option{DamageInstigator}, Source := option{DamageSource}, Amount:=(DmgBase*DmgMultiplier)*PlayerClasses.GetDamageBonus(Agent)}
								ClosestVictim.Damage(RockDmg)
							}
						}
					}
				}

				Sleep(1.0)
				if (EndedVFX:= EndVFX(0)?) {EndedVFX.Dispose() }
			}

			Agent.MoveMod(MoveModNormal)
		}
	}

	#
	# huntmaster saber's ability
	# 
    AbilitySpinSlash(Agent:agent)<suspends>:void=
	{
		if ( FortCharacter := Agent.GetFortCharacter[], PlayerClasses := PlayerClassesDevice? )
		{
			Print("Spin Slash: Activate")
			PlayerClasses.PlayClassVO(Agent, FNVOType.Ability)
			Agent.MoveMod(MoveModStop)

			PlayerPos := FortCharacter.GetTransform().Translation + vector3{ X := 0.0, Y := 0.0, Z := 0.0}
			PlayerRotation := FortCharacter.GetViewRotation()

			var VFXType : creative_prop_asset = VFXPropSpinSlashStart
			ClassIndex := PlayerClasses.GetPlayerClassIndex(Agent)
			if ( ClassIndex = FNClasses.Cerberus )
			{
				set VFXType = VFXPropCerberusSlashStart
			}
			StartVFX := SpawnProp(VFXType, PlayerPos, PlayerRotation)

			Sleep(0.2) #SpinSlash.Spin.TimeBeforeSpin

			ExplosivePos := FortCharacter.GetTransform().Translation
			if ( ExplosiveDeviceLow.TeleportTo[ExplosivePos, rotation{}] ) {}
			ExplosiveDeviceLow.Explode(Agent)

			spawn{ChairAnim(Self,Agent,AnimSpinSlash,1.2)}
			FortCharacter.SetVulnerability(true) # prevent PutInStasis invincible bug!
			var NumberOfSlashes : int = 0
			if (StartedVFX:= StartVFX(0)?)
			{
				TeamCollection := GetPlayspace().GetTeamCollection()
				if ( TeamOwner := TeamCollection.GetTeam[Agent] )
				{
					loop
					{
						AllVictim := GetAllPlayersInDistance(FortCharacter, 500.0)
						for ( ClosestVictim : AllVictim, ClosestAgent := ClosestVictim.GetAgent[], ClosestAgent <> Agent and ( not TeamCollection.IsOnTeam[ClosestAgent, TeamOwner] ) )
						{
							Print("Spin Slash: Damage Victim")
							var DmgMultiplier : float = 1.0
							if ( ClosestPlayer := player[ClosestVictim] ) # if player
							{
								#MoveModKnockback.Activate(ClosestAgent)
							}
							else
							{
								set DmgMultiplier = 4.0
							}
							if ( FortCharacter.IsActive[], DamageInstigator := Agent.GetInstigator(), DamageSource := game_action_causer[FortCharacter] )
							{
								SpinDmg : damage_args = damage_args {Instigator := option{DamageInstigator}, Source := option{DamageSource}, Amount:=(30.0*DmgMultiplier)*PlayerClasses.GetDamageBonus(Agent)} #MajorDamagePawn
								ClosestVictim.Damage(SpinDmg)
							}
						}

						set NumberOfSlashes += 1
						if ( NumberOfSlashes >= 4 )
						{
							break
						}

						Sleep(0.2)
					}

					PlayerClasses.PlayClassVO(Agent, FNVOType.AbilityEnd)
					FortCharacter.ReleaseFromStasis()
					Sleep(0.3)
					AllFinalVictim := GetAllPlayersInDistance(FortCharacter, 500.0)
					for ( ClosestFinalVictim : AllFinalVictim, ClosestFinalAgent := ClosestFinalVictim.GetAgent[], ClosestFinalAgent <> Agent and ( not TeamCollection.IsOnTeam[ClosestFinalAgent, TeamOwner] ) )
					{
						Print("Spin Slash: Damage Victim")
						if ( FortCharacter.IsActive[], DamageInstigator := Agent.GetInstigator(), DamageSource := game_action_causer[FortCharacter] )
						{
							SpinFinalDmg : damage_args = damage_args {Instigator := option{DamageInstigator}, Source := option{DamageSource}, Amount:=(25.0)*PlayerClasses.GetDamageBonus(Agent)} #MajorDamagePawn
							ClosestFinalVictim.Damage(SpinFinalDmg)
						}
					}

					ExplosivePosFinal := FortCharacter.GetTransform().Translation
					if ( ExplosiveDeviceHigh.TeleportTo[ExplosivePosFinal, rotation{}] ) {}
						ExplosiveDeviceHigh.Explode(Agent)

					StartedVFX.Dispose()
				}
			}
			Agent.MoveMod(MoveModNormal)
		}
	}

	#
	# gunnar's ability
	# 
	AbilityCharge(Agent:agent)<suspends>:void=
	{
		if ( FortCharacter := Agent.GetFortCharacter[], PlayerClasses := PlayerClassesDevice? )
		{
			race
			{
				block
				{
					FortCharacter.EliminatedEvent().Await()
				}
				block
				{
					Print("Charge: Activated")
					PlayerClasses.PlayClassVO(Agent, FNVOType.Ability)
					FortCharacter.PutInStasis(stasis_args{AllowTurning := false, AllowFalling := false, AllowEmotes := true})

					case( PlayerClasses.GetPlayerClassIndex(Agent) )
					{
						FNClasses.MegaloDon =>
						{
							VFXPowerJetDashMegaloDon.Pickup(Agent)
						}
						FNClasses.Daigo =>
						{
							VFXPowerSpireRush.Pickup(Agent)
							if {set PlayerClasses.PlayersMap[Agent].bGroundPounding = true}
						}
						_ =>						VFXPowerJetDash.Pickup(Agent)
					}			

					Sleep(1.0)
					FortCharacter.ReleaseFromStasis()

					PlayerPos := FortCharacter.GetTransform().Translation
					PlayerRotation := FortCharacter.GetTransform().Rotation #FortCharacter.GetViewRotation()

					SFXAbilityChargeStart.Play(Agent)
					var NumberOfCharges : int = 0
					loop
					{
						if ( FortCharacter.IsActive[] )
						{
							ChargeMovementMod.Activate(Agent)
						}

						set NumberOfCharges += 1
						if ( NumberOfCharges >= 12 or not FortCharacter.IsActive[] )
						{
							break
						}
						Sleep(0.1)
					}

					TeamCollection := GetPlayspace().GetTeamCollection()
					if ( TeamOwner := TeamCollection.GetTeam[Agent] )
					{
						AllVictim := GetAllPlayersInDistance(FortCharacter, 400.0)
						for ( ClosestVictim : AllVictim, ClosestAgent := ClosestVictim.GetAgent[], ClosestAgent <> Agent and ( not TeamCollection.IsOnTeam[ClosestAgent, TeamOwner] ) )
						{
							Print("Charge Punch: Damage Victim")
							var DmgMultiplier : float = 1.0
							if ( ClosestPlayer := player[ClosestVictim] ) # if player
							{
								ClosestAgent.MoveMod(MoveModDoubleJump)
							}
							else
							{
								set DmgMultiplier = 4.0
							}
							if ( FortCharacter.IsActive[], DamageInstigator := Agent.GetInstigator(), DamageSource := game_action_causer[FortCharacter] )
							{
								PunchDmg : damage_args = damage_args {Instigator := option{DamageInstigator}, Source := option{DamageSource}, Amount:=(100.0*DmgMultiplier)*PlayerClasses.GetDamageBonus(Agent)} #Abilities.Charge.Dmg.Player = 36
								ClosestVictim.Damage(PunchDmg)
							}
							SFXAbilityChargeImpact.Play(ClosestAgent)
						}
					}

					PlayerFinalPos := FortCharacter.GetTransform().Translation
					case( PlayerClasses.GetPlayerClassIndex(Agent) )
					{
						FNClasses.Daigo =>
						{
							spawn{SpawnParticleVFX(FX.NS_Daigo_Aoe_asset,PlayerFinalPos,rotation{},2.0)}
							if {set PlayerClasses.PlayersMap[Agent].bGroundPounding = false}
						}
						_ =>
						{
							spawn{SpawnParticleVFX(FX.NS_ChargeAttack_Dash_Impact_asset,PlayerFinalPos,rotation{},1.0)}
							SFXAbilityChargeEnd.Play(Agent)
						}
					}			


					if ( ExplosiveDeviceHigh.TeleportTo[PlayerFinalPos, rotation{}] ) {}
					ExplosiveDeviceHigh.Explode(Agent)

					PlayerClasses.PlayClassVO(Agent, FNVOType.AbilityEnd)
				}
			}
		}
	}

	OnCreatureSpawned(Agent : agent) : void =
	{
		if ( FortCharacter := Agent.GetFortCharacter[] )
		{
			spawn{ AwaitCreatureElimination(FortCharacter) }
			FortCharacter.DamagedEvent().Subscribe(OnNPCDamaged)
		}
	}

	AwaitCreatureElimination(FortCreature: fort_character)<suspends>:void=
	{
		#FortCreature.EliminatedEvent().Await() # does not work

		loop
		{
			if (FortCreature.GetHealth() <= 0.0)
			{
				#NPCManager.RemoveNPCCount(FortCreature)
				break
			}
			FortCreature.DamagedEvent().Await()
		}
	}

	<#OnCreatureEliminated(AgentPlayer : agent, ElimByPlayer: logic ) : void =
	{
		# useless
	}#>

    FortCharacterLifeTime(FortCharacter:fort_character, LifeTime:float)<suspends>:void=
	{
		Sleep(LifeTime)
		FortCharacter.Damage(5000.0)
	}

	#
	# herald's summons
	# 
	OnLockdownSpawned(WildlifeAgent : agent) : void =
	{
		if ( FortCharacter := WildlifeAgent.GetFortCharacter[] )
		{
			SpawnPos := FortCharacter.GetTransform().Translation
			spawn{ SpawnParticleVFX(FX.NS_CreatureRise_Summon_Spawn_asset, SpawnPos, rotation{}, 2.0) }
			spawn{FortCharacterLifeTime(FortCharacter,15.0)}
		}
	}

	OnLockdownEliminated(Result:device_ai_interaction_result) : void =
	{
		Print("OnLockdownEliminated")
		if ( Wildlife := Result.Target?, FortCharacter := Wildlife.GetFortCharacter[] )
		{
			#NPCManager.RemoveNPCCount(FortCharacter)
		}
	}

    RideWildLife(Agent:agent, WildAgent:agent)<suspends>:void =
	{
		if (TamerFort := Agent.GetFortCharacter[], WolfFort := WildAgent.GetFortCharacter[])
		{
			TamerPos := TamerFort.GetTransform().Translation
			TamerRotation := TamerFort.GetViewRotation()
			if ( WolfFort.TeleportTo[TamerPos, TamerRotation] )
			{
				<#if ( TamerFort.IsActive[], WolfFort.IsActive[] )
				{
					LockdownSpawner.Ride(Agent) # little shit that keep causing all the errors
				}#>
			}
		}
	}

    OnLockdownTamed(Result:device_ai_interaction_result) : void =
	{
		Print("OnLockdownTamed()")
		if (Tamer := Result.Source?, TamerFort := Tamer.GetFortCharacter[], PlayerClasses := PlayerClassesDevice?)
		{
			if (Wolf := Result.Target?, WolfFort := Wolf.GetFortCharacter[])
			{
				TeamCollection := GetPlayspace().GetTeamCollection()
				if ( TeamOwner := TeamCollection.GetTeam[Tamer] )
				{
					if { TeamCollection.AddToTeam[Wolf, TeamOwner] }
				}

				if (set PlayerClasses.PlayersMap[Tamer].ActiveSummon = option{WolfFort}) {}
				spawn{RideWildLife(Tamer,Wolf)} # still causing verse internal errors.
			}
		}
	}

    OnLockdownUntamed(Result:device_ai_interaction_result) : void =
	{
		if (Tamer := Result.Source?, PlayerClasses := PlayerClassesDevice?)
		{
			if (set PlayerClasses.PlayersMap[Tamer].ActiveSummon = false) {}
		}

		if ( Wildlife := Result.Target?, FortCharacter := Wildlife.GetFortCharacter[] )
		{
			FortCharacter.Damage(1000.0)
		}
	}

    OnLockdownRidden(Result:device_ai_interaction_result) : void =
	{
		Print("OnLockdownRidden()")
	}

    OnLockdownDismounted(Result:device_ai_interaction_result) : void =
	{
		if ( Wildlife := Result.Target?, FortCharacter := Wildlife.GetFortCharacter[] )
		{
			FortCharacter.Damage(1000.0)
		}
	}

    AbilityLockdownSummon(Agent:agent)<suspends>:void=
	{
		if ( PlayerClasses := PlayerClassesDevice? )
		{
			if ( FortSummon := PlayerClasses.PlayersMap[Agent].ActiveSummon?, FortSummon.IsActive[] )
			{
				LockdownSpawner.Dismount(Agent)
			}

			if (FortCharacter := Agent.GetFortCharacter[] )
			{
				Print("Lockdown Summon: Activated")
				race
				{
					block
					{
						PlayerClasses.PlayClassVO(Agent, FNVOType.AbilityEnd)
						Agent.MoveMod(MoveModStop)

						PlayerPos := FortCharacter.GetTransform().Translation
						PlayerRotation := FortCharacter.GetViewRotation()
						spawn{ SpawnParticleVFX(FX.NS_CreatureRise_Summon_asset, PlayerPos, PlayerRotation, 2.0) }
						
						Sleep(0.75)

						Sleep(1.2)
						#if ( LockdownSpawner.TeleportTo[PlayerPos, PlayerRotation] )
						#{
							LockdownSpawner.Spawn()
						#}
						Sleep(0.1) # gotta wait cuz taming immediately wont work.
						if ( FortCharacter.IsActive[] ) 
						{
							LockdownSpawner.Tame(Agent)
						}
						#if ( LockdownSpawner.TeleportTo[LockdownSpawnerHomePos, rotation{}] ) {}
						Agent.MoveMod(MoveModNormal)
					}
					block
					{
						FortCharacter.EliminatedEvent().Await()
						Print("Successfully stopped reality collapse event!!!")
					}
				}
			}
		}
	}

	#
	# the society bosses ability
	# 
    AbilityShieldRegen(Agent:agent, bIsUlt:logic)<suspends>:void=
	{
		if ( PlayerClasses := PlayerClassesDevice? )
		{
			if ( FortSummon := PlayerClasses.PlayersMap[Agent].ActiveSummon?, FortSummon.IsActive[] )
			{
				SocietySpawner.DismissAgentHiredGuards(Agent)
			}

			if ( FortCharacter := Agent.GetFortCharacter[] )
			{
				race
				{
					block
					{
						PlayerClasses.PlayClassVO(Agent, FNVOType.Ability)
						Agent.MoveMod(MoveModStop)

						PlayerPos := FortCharacter.GetTransform().Translation + vector3{X:= 0.0, Y:= 0.0, Z:= 50.0}
						spawn{ SpawnParticleVFX(FX.NS_MirrorImage_PreTell_asset, PlayerPos, rotation{}, 1.5) }

						Sleep(1.0)
						Agent.MoveMod(MoveModNormal)
						FortCharacter.Heal(45.0)

						if ( bIsUlt? )
						{
							FortCharacter.SetShield(FortCharacter.GetShield() + FortCharacter.GetMaxShield())
						}
						else
						{
							FortCharacter.SetShield(FortCharacter.GetShield() + 15.0)
						}

						SocietySpawner.Spawn(Agent)
					}
					block
					{
						FortCharacter.EliminatedEvent().Await()
						Print("Successfully stopped reality collapse event!!!")
					}
				}
			}
		}
	}

	#
	# unique beam ability
	# 
	AbilityLaserBeam(Agent:agent)<suspends>:void=
	{
		if ( FortCharacter := Agent.GetFortCharacter[], PlayerClasses := PlayerClassesDevice?, SimulationEntity := GetSimulationEntity[] )
		{
			race
			{
				block
				{
					FortCharacter.EliminatedEvent().Await()
				}
				block
				{
					Print("Laser Beam: Activate")
					PlayerClasses.PlayClassVO(Agent, FNVOType.Ability)
					Agent.MoveMod(MoveModStop)

					CollisionSphere := collision_sphere{ Radius := 100.0 }
					EntBeam := entity{}
					TransComp := transform_component{ Entity := EntBeam }

					var VFXType : visual_effect_powerup_device = VFXPowerCubeBeamStart
					var VFXEndType : particle_system_component = FX.NS_Daigo_Beam{Entity := EntBeam, Enabled := true, AutoPlay := false}
					var BeamRange : float = 1000.0
					var NumberOfBeamMax : int = 45
					case( PlayerClasses.GetPlayerClassIndex(Agent) )
					{
						FNClasses.CubeQueen =>
						{
							set VFXType = VFXPowerCubeBeamStart
							set VFXEndType = FX.NS_CubeQueen_Beam{Entity := EntBeam, Enabled := true, AutoPlay := false}
							set BeamRange = 1000.0
							set NumberOfBeamMax = 45
						}
						FNClasses.TheAgelessChampion =>
						{
							set VFXType = VFXPowerGenoBeamStart
							set VFXEndType = FX.NS_Geno_Beam{Entity := EntBeam, Enabled := true, AutoPlay := false}
							set BeamRange = 1000.0
							set NumberOfBeamMax = 45
						}
						FNClasses.Caretaker =>
						{
							set VFXType = VFXPowerCaretakerBeamStart
							set VFXEndType = FX.NS_CubeQueen_Beam{Entity := EntBeam, Enabled := true, AutoPlay := false}
							set BeamRange = 5000.0
							set NumberOfBeamMax = 45
						}
						FNClasses.DoctorDoomMythic =>
						{
							set VFXType = VFXPowerDoomBeamStart
							set VFXEndType = FX.NS_Doom_Beam{Entity := EntBeam, Enabled := true, AutoPlay := false}
							set BeamRange = 15000.0
							set NumberOfBeamMax = 55
						}
						_ =>
						{
							set VFXType = VFXPowerCubeBeamStart
							#set VFXEndType = FX.NS_Daigo_Beam{Entity := EntBeam, Enabled := true, AutoPlay := false}
							set BeamRange = 5000.0
							set NumberOfBeamMax = 45
						}
					}

					VFXComp := VFXEndType
					EntBeam.AddComponents(array{TransComp,VFXComp})
					SimulationEntity.AddEntities(array{EntBeam})
					defer
					{
						VFXComp.Stop()
						EntBeam.RemoveFromParent()
					}

					CineAbilityPP.Play(Agent)
					VFXType.Pickup(Agent)

					Sleep(1.75)

					var NumberOfBeam : int = 0

					TeamCollection := GetPlayspace().GetTeamCollection()
					if ( TeamOwner := TeamCollection.GetTeam[Agent] )
					{
						VFXComp.Play()
						loop
						{
							OwnerTransform := FromTransform(transform{Translation := FortCharacter.GetTransform().Translation + vector3{X:= 0.0, Y:= 0.0, Z:= 60.0} + FortCharacter.GetViewRotation().GetLocalForward() * 140.0})
							PlayerForwardPos := FortCharacter.GetTransform().Translation + FortCharacter.GetViewRotation().GetLocalForward() * BeamRange

							EntBeam.SetGlobalTransform(OwnerTransform)

							if ( VFXDaigo := FX.NS_Daigo_Beam[VFXComp] )
							{
								set VFXDaigo.StopLocation = FromVector3(PlayerForwardPos)
							}
							if ( VFXCube := FX.NS_CubeQueen_Beam[VFXComp] )
							{
								set VFXCube.StopLocation = FromVector3(PlayerForwardPos)
							}
							if ( VFXGeno := FX.NS_Geno_Beam[VFXComp] )
							{
								set VFXGeno.StopLocation = FromVector3(PlayerForwardPos)
							}
							if ( VFXDoom := FX.NS_Doom_Beam[VFXComp] )
							{
								set VFXDoom.StopLocation = FromVector3(PlayerForwardPos)
							}

							SweepStart := OwnerTransform
							SweepEnd := FromVector3(FortCharacter.GetViewRotation().GetLocalForward() * BeamRange)

							#var IsHit : logic = false # breaking for loop.
							for ( SweepHit : EntBeam.FindSweepHits(SweepEnd, SweepStart, CollisionSphere), HitPos := SweepHit.ContactPosition, TargetComponent := SweepHit.TargetComponent, MeshComp := mesh_component[TargetComponent], TargetEntity := TargetComponent.Entity <#, not IsHit?#> )
							{
								#set IsHit = true
								if ( VFXDaigo := FX.NS_Daigo_Beam[VFXComp] )
								{
									set VFXDaigo.StopLocation = HitPos
								}
								if ( VFXCube := FX.NS_CubeQueen_Beam[VFXComp] )
								{
									set VFXCube.StopLocation = HitPos
								}
								if ( VFXGeno := FX.NS_Geno_Beam[VFXComp] )
								{
									set VFXGeno.StopLocation = HitPos
								}
								if ( VFXDoom := FX.NS_Doom_Beam[VFXComp] )
								{
									set VFXDoom.StopLocation = HitPos
								}

								<#if ( ExplosiveDeviceLow.TeleportTo[FromVector3(HitPos), rotation{}] )
								{
									ExplosiveDeviceLow.Explode(Agent)
								}#>

								if ( PlayerMesh := TargetEntity.GetComponent[Meshes.PlayerBox] )
								{
									if ( PlayerComp := TargetEntity.GetComponent[fn_player_component], AgentVictim := PlayerComp.Agent, not AgentVictim.IsInSpawnRoom[Self] )
									{
										spawn{ Agent.Damage(AgentVictim, 20.0) }
										SFXAbilityLaserBeamHit.Play(AgentVictim)
									}

									if ( NPCComp := TargetEntity.GetComponent[fn_npc_component], AgentNPC := NPCComp.Agent )
									{
										spawn{ Agent.Damage(AgentNPC, 40.0) }
									}
								}
								#return
							}

							set NumberOfBeam += 1
							if ( NumberOfBeam >= NumberOfBeamMax or ( not FortCharacter.IsActive[] ) )
							{
								if ( ExplosiveDeviceHigh.TeleportTo[PlayerForwardPos, rotation{}] ) {}
								ExplosiveDeviceHigh.Explode(Agent)
								break
							}

							Sleep(0.1)
						}
					}

					Agent.MoveMod(MoveModNormal)
					SFXAbilityCubeBeamEnd.Play(Agent)
					CineAbilityPP.Stop(Agent)

					#PlayerClasses.PlayClassVO(Agent, FNVOType.AbilityEnd)
				}
			}
		}
	}

	AbilityShortBeam(Agent:agent)<suspends>:void=
	{
		if ( FortCharacter := Agent.GetFortCharacter[], PlayerClasses := PlayerClassesDevice?, SimulationEntity := GetSimulationEntity[] )
		{
			race
			{
				block
				{
					FortCharacter.EliminatedEvent().Await()
				}
				block
				{
					Print("Demon Beam: Activate")
					PlayerClasses.PlayClassVO(Agent, FNVOType.Ability)

					var VFXType : visual_effect_powerup_device = VFXPowerDaigoBeam
					VFXType.Pickup(Agent)
					spawn{ModulePlayerEx.ChairAnimFollow(Self,Agent,AnimDaigoBeam,2.5)}

					Agent.MoveMod(MoveModStop)

					BeamRange := 15000.0
					CollisionSphere := collision_sphere{ Radius := 100.0 }

					PreEntBeam := entity{}
					PreTransComp := transform_component{ Entity := PreEntBeam }
					PreVFXComp := FX.NS_Daigo_Beam_Pre{ Entity := PreEntBeam, Enabled := true, AutoPlay := false }
					PreEntBeam.AddComponents(array{PreTransComp,PreVFXComp})
					SimulationEntity.AddEntities(array{PreEntBeam})

					EntBeam := entity{}
					TransComp := transform_component{ Entity := EntBeam }
					VFXComp := FX.NS_Daigo_Beam{ Entity := EntBeam, Enabled := true, AutoPlay := false }
					EntBeam.AddComponents(array{TransComp,VFXComp})
					SimulationEntity.AddEntities(array{EntBeam})

					defer
					{
						PreEntBeam.RemoveFromParent()
						EntBeam.RemoveFromParent()
					}

					Sleep(1.0)

					var IsPreBeamEnd : logic = false
					branch
					{
						PreVFXComp.Play()
						loop
						{
							if ( IsPreBeamEnd? )
							{
								break
							}
							PreLoopOwnerTransform := FromTransform(transform{Translation := FortCharacter.GetTransform().Translation + vector3{X:= 0.0, Y:= 0.0, Z:= 60.0} + FortCharacter.GetViewRotation().GetLocalForward() * 140.0})
							PreLoopPlayerForwardPos := FortCharacter.GetTransform().Translation + FortCharacter.GetViewRotation().GetLocalForward() * BeamRange

							PreEntBeam.SetGlobalTransform(PreLoopOwnerTransform)

							set PreVFXComp.StopLocation = FromVector3(PreLoopPlayerForwardPos)

							PreSweepStart := PreLoopOwnerTransform
							PreSweepEnd := FromVector3(FortCharacter.GetViewRotation().GetLocalForward() * BeamRange)
							var PreIsHit : logic = false # breaking for loop.
							for ( SweepHit : PreEntBeam.FindSweepHits(PreSweepEnd, PreSweepStart, CollisionSphere), HitPos := SweepHit.ContactPosition, MeshComp := mesh_component[SweepHit.TargetComponent], not IsAlmostEqual[FromVector3(PreLoopOwnerTransform.Translation),FromVector3(HitPos),0.2] <#, not PreIsHit?#> )
							{
								set PreIsHit = true
								set PreVFXComp.StopLocation = HitPos
								#Print("FindSweepHits: {FromVector3(HitPos)}")
							}
							Sleep(0.025)
						}
					}
					Sleep(1.0)
					set IsPreBeamEnd = true
					PreVFXComp.Stop()
					VFXComp.Stop()

					OwnerTransform := FromTransform(transform{Translation := FortCharacter.GetTransform().Translation + vector3{X:= 0.0, Y:= 0.0, Z:= 60.0} + FortCharacter.GetViewRotation().GetLocalForward() * 140.0})
					PlayerForwardPos := FortCharacter.GetTransform().Translation + FortCharacter.GetViewRotation().GetLocalForward() * BeamRange

					EntBeam.SetGlobalTransform(OwnerTransform)

					set VFXComp.StopLocation = FromVector3(PlayerForwardPos)

					SweepStart := OwnerTransform
					SweepEnd := FromVector3(FortCharacter.GetViewRotation().GetLocalForward() * BeamRange)

					var IsHit : logic = false # breaking for loop.
					for ( SweepHit : EntBeam.FindSweepHits(SweepEnd, SweepStart, CollisionSphere), HitPos := SweepHit.ContactPosition, TargetComponent := SweepHit.TargetComponent, MeshComp := mesh_component[TargetComponent], TargetEntity := TargetComponent.Entity <#, not IsHit?#> )
					{
						if ( not IsAlmostEqual[FromVector3(OwnerTransform.Translation),FromVector3(HitPos),0.2] )
						{
							set VFXComp.StopLocation = HitPos
						}

						#Print("FindSweepHits: {FromVector3(HitPos)}")
						if ( ExplosiveDeviceLow.TeleportTo[FromVector3(HitPos), rotation{}] )
						{
							ExplosiveDeviceLow.Explode(Agent)
						}

						if ( PlayerMesh := TargetEntity.GetComponent[Meshes.PlayerBox] )
						{
							if ( PlayerComp := TargetEntity.GetComponent[fn_player_component], AgentVictim := PlayerComp.Agent, not AgentVictim.IsInSpawnRoom[Self] )
							{
								#Print("Short Beam Hit: Player")
								spawn{ Agent.Damage(AgentVictim, 50.0) }
							}

							if ( NPCComp := TargetEntity.GetComponent[fn_npc_component], AgentNPC := NPCComp.Agent )
							{
								#Print("Short Beam Hit: Player")
								spawn{ Agent.Damage(AgentNPC, 200.0) }
							}
						}
						set IsHit = true
					}

					Sleep(1.0)
					VFXComp.Stop()
				}
			}
		}

		Agent.MoveMod(MoveModNormal)
	}

	(AgentOwner:agent).Damage(Agent:agent,DMGAmount:float)<suspends>:void=
	{
		TeamCollection := GetPlayspace().GetTeamCollection()
		if ( FortOwner := AgentOwner.GetFortCharacter[], FortVictim := Agent.GetFortCharacter[], PlayerClasses := PlayerClassesDevice?, TeamOwner := TeamCollection.GetTeam[AgentOwner], not TeamCollection.IsOnTeam[Agent, TeamOwner] )
		{		
			if ( FortOwner.IsActive[], DamageInstigator := AgentOwner.GetInstigator(), DamageSource := game_action_causer[FortOwner] )
			{
				DrainDmg : damage_args = damage_args {Instigator := option{DamageInstigator} , Source := option{DamageSource}, Amount:=DMGAmount*PlayerClasses.GetDamageBonus(AgentOwner)}
				FortVictim.Damage(DrainDmg)
			}
		}
	}

	#
	# medic's medigun
	# 
	AbilityHealingBeam(Agent:agent)<suspends>:void=
	{
		if ( FortCharacter := Agent.GetFortCharacter[], PlayerClasses := PlayerClassesDevice? )
		{
			Print("Healing Beam: Activate")
			Sleep(0.2)

			if ( ClosestMate := GetNearestTeammateToHeal[FortCharacter], SimulationEntity := GetSimulationEntity[] )
			{
				PlayerClasses.PlayClassVO(Agent, FNVOType.Ability)
				PlayerPos := FortCharacter.GetTransform().Translation + vector3{ X := 0.0, Y := 0.0, Z := 75.0}
				PlayerRotation := FortCharacter.GetViewRotation()

				OwnerTransform := FromTransform(transform{Translation := PlayerPos})
				PlayerForwardPos := FortCharacter.GetTransform().Translation + FortCharacter.GetViewRotation().GetLocalForward() * 500.0
				EntBeam := entity{}
				TransComp := transform_component{ Entity := EntBeam }
				VFXComp := FX.NS_Hades_HealingBeam{ Entity := EntBeam, Enabled := true, AutoPlay := true, StopLocation := FromVector3(PlayerForwardPos) }
				EntBeam.AddComponents(array{TransComp,VFXComp})
				SimulationEntity.AddEntities(array{EntBeam})
				EntBeam.SetGlobalTransform(OwnerTransform)

				race
				{
					block
					{
						FortCharacter.EliminatedEvent().Await()
					}
					block
					{
						ClosestMate.EliminatedEvent().Await()
					}
					block
					{
						InputFire.ReleasedEvent.Await()
					}
					block
					{
						loop
						{
							QuitAgent := GetPlayspace().PlayerRemovedEvent().Await()
							if ( Agent = QuitAgent )
							{
								break
							}
						}
					}
					block
					{
						loop
						{
							if ( not FortCharacter.IsActive[] )
							{
								Print("Healing Beam: Died")
								break
							}

							if ( not ConditionalHealingBeam.IsHoldingItem[Agent] )
							{
								Print("Healing Beam: Switch Item")
								break
							}

							if ( not InputFire.IsHeld[Agent] )
							{
								Print("Healing Beam: Off")
								break
							}

							HealerPos := FortCharacter.GetTransform().Translation
							TargetPos := ClosestMate.GetTransform().Translation
							EDistance := Distance(HealerPos, TargetPos)
							if (EDistance > 1500.0)
							{
								Print("Healing Beam: Too Far")
								break
							}

							HealerTransform := FromTransform(transform{Translation := HealerPos})
							EntBeam.SetGlobalTransform(HealerTransform)
							set VFXComp.StopLocation = FromVector3(TargetPos)

							if ( FortCharacter.IsActive[], DamageInstigator := Agent.GetInstigator(), DamageSource := game_action_causer[FortCharacter] )
							{
								BeamHeal : healing_args = healing_args{Instigator := option{DamageInstigator}, Source := option{DamageSource}, Amount:=5.0}
								ClosestMate.Heal(BeamHeal)
								#spawn{ SpawnParticleVFX(FX.NS_ShieldRegen_asset, TargetPos + vector3{ X := 0.0, Y := 0.0, Z := 75.0}, rotation{}, 0.1) }
								Print("Healing Beam: Heal {ClosestMate.GetHealth()}")
							}

							Sleep(0.1)
						}
					}
				}

				EntBeam.RemoveFromParent()
			}
			else
			{
				PlayerClasses.AddHudStatusMessage(Agent, TextLocalization.MSG_NoTeammateFound)
				SFXAbilityCooldown.Play(Agent)
			}

			NewCurrentTime := GetSimulationElapsedTime()
			if {set PlayerClasses.PlayersMap[Agent].flAbilityUsedTime = NewCurrentTime}
		}
	}

	#
	# unique double jump ability
	# 
	AbilityDoubleJump(Agent:agent)<suspends>:void=
	{
		if
		{
			FortCharacter := Agent.GetFortCharacter[]
			FortCharacter.IsActive[]
			PlayerClasses := PlayerClassesDevice?
			not PlayerClasses.PlayersMap[Agent].bDoubleJumped?
			set PlayerClasses.PlayersMap[Agent].bDoubleJumped = true
		}
		then
		{
			VFXPowerWind.Pickup(Agent)
			MoveModDoubleJump.Activate(Agent)
		}
	}

	#
	# unique high jump ability
	# 
	AbilityHighJump(Agent:agent)<suspends>:void=
	{
		if
		{
			FortCharacter := Agent.GetFortCharacter[]
			FortCharacter.IsActive[]
			PlayerClasses := PlayerClassesDevice?
			not PlayerClasses.PlayersMap[Agent].bDoubleJumped?
			set PlayerClasses.PlayersMap[Agent].bDoubleJumped = true
		}
		then
		{
			Print("High Jump: Activated")
			PlayerClasses.PlayClassVO(Agent, FNVOType.Ability)

			VFXPowerWind.Pickup(Agent)
			MoveModHighJump.Activate(Agent)
		}
	}

	#
	# unique high jump ability
	# 
	AbilityLeap(Agent:agent)<suspends>:void=
	{
		if
		{
			FortCharacter := Agent.GetFortCharacter[]
			FortCharacter.IsActive[]
			PlayerClasses := PlayerClassesDevice?
			not PlayerClasses.PlayersMap[Agent].bDoubleJumped?
			set PlayerClasses.PlayersMap[Agent].bDoubleJumped = true
		}
		then
		{
			Print("Leap: Activated")
			PlayerClasses.PlayClassVO(Agent, FNVOType.AbilityEnd)

			case( PlayerClasses.GetPlayerClassIndex(Agent) )
			{
				FNClasses.Zeus =>				VFXPowerLeapZeus.Pickup(Agent)
				FNClasses.Kor =>				VFXPowerLeapKor.Pickup(Agent)
				_ =>							VFXPowerWind.Pickup(Agent)
			}

			MoveModLeap.Activate(Agent)
		}
	}

	Explode(Agent:agent, ExploPos:vector3, ExploRadius:float, ExploDMG:float)<suspends>:void=
	{
		if (FortCharacter := Agent.GetFortCharacter[], FortCharacter.IsActive[], Player := player[Agent], Player.IsActive[], PlayerClasses := PlayerClassesDevice? )
		{
			spawn{SpawnPropVFX(VFXPropRockEnd,ExploPos,rotation{},1.2)}
			if ( ExplosiveDeviceLow.TeleportTo[ExploPos, rotation{}] )
			{
				ExplosiveDeviceLow.Explode(Agent)
			}

			TeamCollection := GetPlayspace().GetTeamCollection()
			if ( TeamOwner := TeamCollection.GetTeam[Agent] )
			{
				AllVictim := GetAllPlayersInVectorDistance(ExploPos, ExploRadius)
				for ( ClosestVictim : AllVictim, ClosestAgent := ClosestVictim.GetAgent[], not TeamCollection.IsOnTeam[ClosestAgent, TeamOwner], FortCharacter.IsActive[], Player.IsActive[] )
				{
					var DmgMultiplier : float = 1.0
					if ( ClosestPlayer := player[ClosestVictim] ) # if player
					{
						if ( ClosestPlayer.IsActive[] )
						{
							set DmgMultiplier = 1.0
							ClosestAgent.MoveMod(MoveModDoubleJump)
						}
					}
					else
					{
						set DmgMultiplier = 2.5
					}
					if ( FortCharacter.IsActive[], DamageInstigator := Agent.GetInstigator(), DamageSource := game_action_causer[FortCharacter] )
					{
						RockDmg : damage_args = damage_args{Instigator := option{DamageInstigator}, Source := option{DamageSource}, Amount:=(ExploDMG*DmgMultiplier)*PlayerClasses.GetDamageBonus(Agent)}
						ClosestVictim.Damage(RockDmg)
					}
				}
			}
		}
	}

	(Agent:agent).Explode(ExploRadius:float, ExploDMG:float)<suspends>:void=
	{
		if (FortCharacter := Agent.GetFortCharacter[], FortCharacter.IsActive[] )
		{
			ExploPos := FortCharacter.GetTransform().Translation
			Explode(Agent, ExploPos, ExploRadius, ExploDMG)
		}
	}

	#
	# FOOT DIVE
	AbilityFootDive(Agent:agent)<suspends>:void=
	{
		if (FortCharacter := Agent.GetFortCharacter[], FortCharacter.IsActive[], PlayerClasses := PlayerClassesDevice? )
		{
			PlayerCooldown := PlayerClasses.GetAbilityCooldown(Agent, FNAbility.FootDive)
			if ( PlayerCooldown > 0.0 )
			{
				SFXAbilityCooldown.Play(Agent)
				return
			}

			if ( PlayerClasses.PlayersMap[Agent].bDoubleJumped? )
			{
				##Default.BossD.DiveKick.MaxDistance = 4500.0
				if ( NearestTarget := GetNearestEnemy[FortCharacter, 3000.0] ) #Default.BossD.DiveKick.MaxDistance = 4500.0
				{
					Print("Foot Dive: DIVE")
					spawn{PlayerClasses.AbilityCooldown(Agent, FNAbility.FootDive)}
					PlayerDivePos := FortCharacter.GetTransform().Translation
					PlayerDiveRot := FortCharacter.GetViewRotation()
					if (PlayerDiveRotYaw := PlayerDiveRot.GetYawPitchRollDegrees()[0], PlayerDiveRotRoll := PlayerDiveRot.GetYawPitchRollDegrees()[2])
					{
						race
						{
							block
							{
								FortCharacter.EliminatedEvent().Await()
								Print("Foot Dive: Failed.")
							}
							block
							{
								loop
								{
									QuitAgent := GetPlayspace().PlayerRemovedEvent().Await()
									if ( Agent = QuitAgent )
									{
										Print("Foot Dive: Disconnected")
										break
									}
								}
							}
							block
							{
								AnimDiveRot := MakeRotationFromYawPitchRollDegrees(PlayerDiveRotYaw, 0.0, PlayerDiveRotRoll) 
								NewDiveTransform := transform{Scale := vector3{X:= 0.01, Y:= 0.01, Z:= 0.01}, Rotation := AnimDiveRot, Translation := PlayerDivePos}
								SpawnVFX := SpawnProp(AnimFootDive, NewDiveTransform)
								if (SpawnedVFX:= SpawnVFX(0)?)
								{
									defer{SpawnedVFX.Dispose()}

									PropPos := SpawnedVFX.GetTransform().Translation

									TaggedDevice := FindCreativeObjectsWithTag(chair_footdive_tag{})
									for (FoundDevice: TaggedDevice, ChairDevice := chair_device[FoundDevice])
									{
										DevicePos := FoundDevice.GetTransform().Translation
										if ( IsAlmostEqual[PropPos,DevicePos,0.2] )
										{
											ChairDevice.Seat(Agent)
											ChairDevice.DisableExit()
											FortCharacter.PutInStasis(stasis_args{AllowTurning := false, AllowFalling := false, AllowEmotes := true})
											Sleep(0.4)
											ChairDevice.Eject(Agent)
										}
									}
								}

								SpawnVFX2 := SpawnProp(AnimFootDiveLoop, NewDiveTransform)
								if (SpawnedVFX2:= SpawnVFX2(0)?)
								{
									defer
									{
										FortCharacter.ReleaseFromStasis()
										SpawnedVFX2.Dispose()
									}

									PropPos := SpawnedVFX2.GetTransform().Translation

									TaggedDevice := FindCreativeObjectsWithTag(chair_footdive_tag{})
									for (FoundDevice: TaggedDevice, ChairDevice := chair_device[FoundDevice])
									{
										DevicePos := FoundDevice.GetTransform().Translation
										if ( IsAlmostEqual[PropPos,DevicePos,0.2] )
										{
											ChairDevice.Seat(Agent)
											ChairDevice.DisableExit()
											VFXPowerFootDive.Pickup(Agent)

											TargetPos := NearestTarget.GetTransform().Translation
											ChairDevice.MoveTo(TargetPos, AnimDiveRot, 0.4)

											<#if ( Entity := GetSimulationEntity[] )
											{
												Print("Get Entity...")
												LookRot :[]float= FortCharacter.GetViewRotation().GetYawPitchRollDegrees()
												if ( YawRadian := (/UnrealEngine.com/Temporary/SpatialMath:)DegreesToRadians(LookRot[0]), PitchRadian := (/UnrealEngine.com/Temporary/SpatialMath:)DegreesToRadians(LookRot[1]) )
												{
													LookDirection := (/UnrealEngine.com/Temporary/SpatialMath:)vector3{X := Cos(YawRadian)*Cos(PitchRadian), Y := Sin(YawRadian)*Cos(PitchRadian), Z := Sin(PitchRadian)}
													# Set collision check distance (cm; use very large value for infinite range)
													Range := 4500.0
													DisplacementRange := LookDirection*Range
													# Start position for collision check
													ViewLoc := FromVector3(FortCharacter.GetViewLocation())
													ViewTrans := transform{Translation := FromVector3(ViewLoc)}
													# Collision (HitResult occurs when this collides)
													Collision : collision_sphere = collision_sphere{Radius := 100.0}
													# Check if sweep collision hits anything
													# HitResult occurs if collision detected
													HitResult := Entity.FindSweepHits(FromVector3(DisplacementRange), FromTransform(ViewTrans), Collision)
													Print("FindSweepHits...")
													DebugDraw := debug_draw{}
													DebugDraw.DrawLine(FortCharacter.GetViewLocation(), FortCharacter.GetViewLocation()+DisplacementRange, ?Duration:= 2.0)

													var IsHit : logic = false
													for(Result : HitResult)
													{
														#ParentEntity := Result.TargetComponent.Entity
														# Treat as wall hit if no raycast_component
														# Optionally, differentiate walls by using a specific component
														if(not IsHit?)
														{
															# Perform hit detection only once (when IsHit is false)
															set IsHit = true
															HitPos := FromVector3(Result.ContactPosition)
															HitNormal := FromVector3(Result.ContactNormal)
															# Set return value
															#set ReturnValue = (option{HitPos}, option{HitNormal})
															# Debug
															DebugDraw.DrawSphere(HitPos, ?Radius := 10.0, ?Duration:= 2.0)
															DebugDraw.DrawLine(HitPos, HitPos + HitNormal*100.0, ?Duration:= 2.0)
															Print("Foot Dive Hit : {HitPos}")
															ChairDevice.MoveTo(HitPos, AnimDiveRot, 0.4)
														}
													}
												}
											}#>

											ChairDevice.Eject(Agent)
										}
									}

									if ( FortCharacter.IsActive[] ) # extra check cuz it keep making ErrRuntime_Internal
									{
										FinalPos := FortCharacter.GetTransform().Translation
										spawn{Explode(Agent, FinalPos, 800.0, 40.0)}
									}
								}
							}
						}
					}
				}
			}
			else
			{
				if (set PlayerClasses.PlayersMap[Agent].bDoubleJumped = true)
				{
					Print("Foot Dive: Activated")
					PlayerClasses.PlayClassVO(Agent, FNVOType.AbilityEnd)

					VFXPowerFootDiveCharge.Pickup(Agent)
					Agent.MoveMod(MoveModHighJump)
				}
			}
		}
	}

	<#AbilityFootDiveOld(Agent:agent)<suspends>:void=
	{
		if (FortCharacter := Agent.GetFortCharacter[], FortCharacter.IsActive[], PlayerClasses := PlayerClassesDevice? )
		{
			Print("Foot Dive: Activated")
			PlayerClasses.PlayClassVO(Agent, FNVOType.Ability)

			VFXPowerFootDive.Pickup(Agent)
			Agent.MoveMod(MoveModHighJump)

			#Sleep(1.5)
			if ( NearestTarget := GetNearestPlayer[FortCharacter, 4500.0] ) #Default.BossD.DiveKick.MaxDistance
			{
				TargetPos := NearestTarget.GetTransform().Translation
				spawn{Agent.ApplyImpulse(TargetPos)}
				race
				{
					block
					{
						loop
						{
							if ( not FortCharacter.IsInAir[] )
							{
								Print("Foot Dive: Landed.")
								break
							}
							Sleep(0.1)
						}
					}
					block
					{
						FortCharacter.EliminatedEvent().Await()
						Print("Foot Dive: Failed.")
					}
				}
				SpawnPos := FortCharacter.GetTransform().Translation
				spawn{Explode(Agent, SpawnPos, 800.0, 50.0)}
			}
		}
	}#>

	(Agent:agent).ApplyImpulse(TargetPos:vector3)<suspends>:void=
	{
		if (FortCharacter := Agent.GetFortCharacter[], FortCharacter.IsActive[] )
		{
			PlayerPos := FortCharacter.GetTransform().Translation
			if (LookDirection := (TargetPos - PlayerPos).MakeUnitVector[])
			{
				Yaw := RadiansToDegrees(ArcTan(LookDirection.Y, LookDirection.X))
				Pitch := RadiansToDegrees(ArcTan(LookDirection.Z, Sqrt((LookDirection.X * LookDirection.X) + (LookDirection.Y * LookDirection.Y))))
				Roll := 0.0
				NewRotation := MakeRotationFromYawPitchRollDegrees(Yaw, Pitch, Roll)

				NonePos := MoveModStraight.GetTransform().Translation
				if{ MoveModStraight.TeleportTo[NonePos, NewRotation] }
				MoveModStraight.Activate(Agent)
			}
		}
	}

	AlienParasitePlayer(Agent:agent, Duration:float)<suspends>:void=
	{
		if ( FortCharacter := Agent.GetFortCharacter[], PlayerClasses := PlayerClassesDevice?, ClassMaxHealth := PlayerClasses.PlayersMap[Agent].ClassData.MaxHealth, ClassMaxShield := PlayerClasses.PlayersMap[Agent].ClassData.MaxShield )
		{
			race
			{
				block
				{
					CurrentHP := FortCharacter.GetHealth()
					CurrentShield := FortCharacter.GetShield()
					ClassSelectorLowGrav.ChangeClass( Agent )
					FortCharacter.SetMaxHealth(ClassMaxHealth/1.5)
					FortCharacter.SetMaxShield(ClassMaxShield)
					FortCharacter.SetHealth(CurrentHP)
					FortCharacter.SetShield(CurrentShield)
					SFXAlienPAttach.Play(Agent)
					VFXPowerAlien.Pickup(Agent)

					Sleep(Duration)
					NewCurrentHP := FortCharacter.GetHealth()
					NewCurrentShield := FortCharacter.GetShield()
					if ( AgentClassSelector := PlayerClasses.PlayersMap[Agent].ClassData.ClassSelector )
					{
						AgentClassSelector.ChangeClass( Agent )
					}

					FortCharacter.SetMaxHealth(ClassMaxHealth)
					FortCharacter.SetMaxShield(ClassMaxShield)
					FortCharacter.SetHealth(NewCurrentHP)
					FortCharacter.SetShield(NewCurrentShield)
					SFXAlienPDetach.Play(Agent)
				}
				block
				{
					FortCharacter.EliminatedEvent().Await()
					Print("Successfully stopped reality collapse event!!!")
				}
			}
		}
	}

    DamageBuffPlayer(Agent:agent, Duration:float, Magnitude:float)<suspends>:void=
	{
		if ( PlayerClasses := PlayerClassesDevice? )
		{
			spawn{PlayerClasses.DamageBuffPlayer(Agent, Duration, Magnitude)}
		}
	}

	InputPressedFireAbility(Agent : agent) : void =
	{
		if (FortCharacter := Agent.GetFortCharacter[], FortCharacter.IsActive[], PlayerClasses := PlayerClassesDevice? )
		{
			if ( ConditionalHealingBeam.IsHoldingItem[Agent] )
			{
				CurrentTime := GetSimulationElapsedTime()
				if ( AbilityUsedTime := PlayerClasses.PlayersMap[Agent].flAbilityUsedTime, ElapsedTime := CurrentTime - PlayerClasses.PlayersMap[Agent].flAbilityUsedTime )
				{
					if (ElapsedTime < 2.0)
					{
						Print("Ability on cooldown... {ElapsedTime}")
						SFXAbilityCooldown.Play(Agent)
						return
					}

					case( PlayerClasses.GetPlayerClassIndex(Agent) )
					{
						FNClasses.Hades =>				spawn{ AbilityHealingBeam(Agent) }
						_ =>							# does nothing
					}
				}
			}
		}
	}

	#
	# jones' ability
	# 
    AbilityRiftGun(Agent:agent)<suspends>:void=
	{
		if ( FortCharacter := Agent.GetFortCharacter[], PlayerClasses := PlayerClassesDevice? )
		{
			Print("Rift Gun: Shoot")
			PlayerClasses.PlayClassVO(Agent, FNVOType.Ability)

			PlayerPos := FortCharacter.GetTransform().Translation + vector3{ X := 0.0, Y := 0.0, Z := 50.0}
			PlayerRotation := FortCharacter.GetViewRotation()

			PlayerClasses.PlayClassVO(Agent, FNVOType.AbilityEnd)

			StartVFX := SpawnProp(VFXPropJonesRiftProj, PlayerPos, PlayerRotation)
			if (StartedVFX:= StartVFX(0)?)
			{
				if (AController := StartedVFX.GetAnimationController[])
				{
					PlayerForwardPos := FortCharacter.GetTransform().Translation + FortCharacter.GetViewRotation().GetLocalForward() * 400.0 # PrjDistanceToTarget = 1500.0
					AnimKeyframes := array
					{
						keyframe_delta
						{
							DeltaLocation := PlayerForwardPos - StartedVFX.GetTransform().Translation
							DeltaRotation := rotation{}
							Time := 0.4
							Interpolation := EaseOut
						}
					}
					AController.SetAnimation(AnimKeyframes, ?Mode:=animation_mode.OneShot)
					AController.Play()
				}
				Sleep(0.4)

				StartedVFXPos := StartedVFX.GetTransform().Translation
				StartedVFX.Dispose()
				EndVFX := SpawnProp(VFXPropJonesRift, StartedVFXPos, rotation{})

				var NumberOfDMG : int = 0
				if (EndedVFX:= EndVFX(0)?)
				{
					defer
					{
						EndedVFX.Dispose()
					}
					ExplosivePos := EndedVFX.GetTransform().Translation
					if ( ExplosiveDeviceLow.TeleportTo[ExplosivePos, rotation{}] ) {}
					ExplosiveDeviceLow.Explode(Agent)

					TeamCollection := GetPlayspace().GetTeamCollection()
					if ( TeamOwner := TeamCollection.GetTeam[Agent] )
					{
						loop
						{
							AllVictim := GetAllPlayersInPropDistance(EndedVFX, 500.0)
							for ( ClosestVictim : AllVictim, ClosestAgent := ClosestVictim.GetAgent[], ClosestAgent <> Agent and ( not TeamCollection.IsOnTeam[ClosestAgent, TeamOwner] ) )
							{
								Print("Rift: Damage Victim")
								var DmgMultiplier : float = PlayerClasses.GetDamageBonus(Agent)
								if ( FortCharacter.IsActive[], DamageInstigator := Agent.GetInstigator(), DamageSource := game_action_causer[FortCharacter] )
								{
									SpinDmg : damage_args = damage_args {Instigator := option{DamageInstigator}, Source := option{DamageSource}, Amount:=30.0*DmgMultiplier}
									ClosestVictim.Damage(SpinDmg)
								}
							}

							set NumberOfDMG += 1
							if ( NumberOfDMG >= 4 )
							{
								break
							}

							Sleep(0.2)
						}
					}
					Sleep(0.5)
				}
			}
		}
	}

	#
	# heist ability
	# 
	AbilityMovePlus(Agent:agent)<suspends>:void=
	{
		Agent.MoveMod(MoveModPlus)
	}

	MoveNormal(Agent:agent)<suspends>:void=
	{
		Agent.MoveMod(MoveModNormal)
	}

	#
	# 
	# 
	AbilityNitroCharge(Agent:agent)<suspends>:void=
	{
		Print("AbilityNitroCharge")
		<#if (PlayerClasses := PlayerClassesDevice? )
		{
			PlayerClasses.PlayClassVO(Agent, FNVOType.Ability)
			VFXNitroCharge.Pickup(Agent)
		}#>
	}

    AbilityPartingGift(Agent:agent)<suspends>:void=
	{
		if ( FortCharacter := Agent.GetFortCharacter[], PlayerClasses := PlayerClassesDevice? )
		{
			ExploPos := FortCharacter.GetTransform().Translation
			spawn{Explode(Agent, ExploPos, 800.0, 75.0)}
			PlayerClasses.UpdateHudDisableAbility(Agent, FNAbility.PartingGift)
		}
	}

	#
	# engineer gaming
	# 
    AbilityBuildingSentry(Agent:agent)<suspends>:void=
	{
		<#if ( FortCharacter := Agent.GetFortCharacter[], PlayerClasses := PlayerClassesDevice? )
		{
			if ( OldProp := PlayerClasses.PlayersMap[Agent].ActiveBuilding?, OldProp.IsValid[] )
			{
				return
				#OldProp.Dispose()
				#if {set PlayerClasses.PlayersMap[Agent].ActiveBuilding = false}
			}

			PlayerPos := FortCharacter.GetTransform().Translation
			PlayerRot := FortCharacter.GetViewRotation()
			if (PlayerRotYaw := PlayerRot.GetYawPitchRollDegrees()[0], PlayerRotRoll := PlayerRot.GetYawPitchRollDegrees()[2])
			{
				BuildingPos := PlayerPos + FortCharacter.GetViewRotation().GetLocalForward() * 400.0
				BuildingRot := MakeRotationFromYawPitchRollDegrees(PlayerRotYaw , 0.0, PlayerRotRoll) 
				SpawnVFX := SpawnProp(VFXPropSentry, BuildingPos, BuildingRot)
				if (SpawnedVFX:= SpawnVFX(0)?)
				{
					PropPos := SpawnedVFX.GetTransform().Translation

					TaggedDevice := FindCreativeObjectsWithTag(turret_tag{})
					for (FoundDevice: TaggedDevice, BuildingDevice := automated_turret_device[FoundDevice])
					{
						DevicePos := FoundDevice.GetTransform().Translation
						if ( IsAlmostEqual[PropPos,DevicePos,0.2], set PlayerClasses.PlayersMap[Agent].ActiveBuilding = option{SpawnedVFX} )
						{
							PlayerClasses.PlayClassVO(Agent, FNVOType.Ability)
							Print("Engineer: Gaming")
							BuildingDevice.ActivatedEvent.Subscribe(OnSentryActivated)		# the only sub that works,
							BuildingDevice.DamagedEvent.Subscribe(OnSentryDamaged)			#not working
							BuildingDevice.DestroyedEvent.Subscribe(OnSentryDestroyed)		#not working
							BuildingDevice.TargetFoundEvent.Subscribe(OnSentryTargetFound)	#not working
							BuildingDevice.TargetLostEvent.Subscribe(OnSentryTargetLost)	#not working
							BuildingDevice.SetTeam(Agent)
							BuildingDevice.SetMaxHealth(500.0)
							BuildingDevice.SetHealth(500.0)
							BuildingDevice.SetActivationRange(20.0)
							BuildingDevice.SetTargetRange(20.0)
							race
							{
								block
								{
									FortCharacter.EliminatedEvent().Await()
									Print("Engineer: FortCharacter EliminatedEvent")
								}
								block
								{
									BuildingDevice.DestroyedEvent.Await()
									Print("Engineer: DestroyedEvent")
								}
							}
							Print("Engineer: Sentry Down!")
							if ( SpawnedVFX.IsValid[] )
							{
								SpawnedVFX.Dispose()
								if {set PlayerClasses.PlayersMap[Agent].ActiveBuilding = false}
							}
						}
					}
				}
			}
		}#>
	}

	<#OnSentryActivated(MaybeAgent:?agent):void=
	{
		Print("OnSentryActivated")
	}

	OnSentryDamaged(MaybeAgent:?agent):void=
	{
		Print("OnSentryDamaged")
	}

	OnSentryDestroyed(MaybeAgent:?agent):void=
	{
		Print("OnSentryDestroyed")
	}

	OnSentryTargetFound(Agent:agent):void=
	{
		Print("OnSentryTargetFound")
	}

	OnSentryTargetLost(Agent:agent):void=
	{
		Print("OnSentryTargetLost")
	}#>

	#
	# breakable prop
	# 
    AbilityObjShield(Agent:agent)<suspends>:void=
	{
		if ( FortCharacter := Agent.GetFortCharacter[], PlayerClasses := PlayerClassesDevice? )
		{
			if ( OldProp := PlayerClasses.PlayersMap[Agent].ActiveBuilding?, OldProp.IsValid[] )
			{
				return
				#OldProp.Dispose()
				#if {set PlayerClasses.PlayersMap[Agent].ActiveBuilding = false}
			}

			PlayerPos := FortCharacter.GetTransform().Translation
			PlayerRot := FortCharacter.GetViewRotation()
			if (PlayerRotYaw := PlayerRot.GetYawPitchRollDegrees()[0], PlayerRotRoll := PlayerRot.GetYawPitchRollDegrees()[2])
			{
				BuildingPos := PlayerPos - vector3{ X := 0.0, Y := 0.0, Z := 100.0}
				BuildingRot := MakeRotationFromYawPitchRollDegrees(PlayerRotYaw , 0.0, PlayerRotRoll)
				spawn{ SpawnParticleVFX(FX.NS_CubeAssassin_Shield_asset, BuildingPos, BuildingRot, 0.5) }
				Sleep(0.4)
				SpawnVFX := SpawnProp(VFXPropCubeAShield, BuildingPos, BuildingRot)
				if (SpawnedVFX:= SpawnVFX(0)?, set PlayerClasses.PlayersMap[Agent].ActiveBuilding = option{SpawnedVFX})
				{
					if ( SpawnedVFX.IsValid[] )
					{
						set SpawnedVFX.CanBeDamaged = true
					}

					Print("Obj: Shield")
					PlayerClasses.PlayClassVO(Agent, FNVOType.Ability)
					race
					{
						block
						{
							FortCharacter.EliminatedEvent().Await()
							Print("Shield: FortCharacter EliminatedEvent...")
						}
						block
						{
							Sleep(5.0)
							Print("Shield: Lifetime Ending...")
						}
					}
					Print("Shield Down!")
					if ( SpawnedVFX.IsValid[] )
					{
						spawn{ SpawnParticleVFX(FX.NS_CubeAssassin_Shield_Destroy_asset, BuildingPos, BuildingRot, 0.5) }
						SpawnedVFX.Dispose()
						if {set PlayerClasses.PlayersMap[Agent].ActiveBuilding = false}
					}
				}
			}
		}
	}

	#
	# caretaker's snap ability
	# 
    AbilitySnap(Agent:agent)<suspends>:void=
	{
		if ( FortCharacter := Agent.GetFortCharacter[], PlayerClasses := PlayerClassesDevice? )
		{
			CurrentTime := GetSimulationElapsedTime()
			if ( AbilityUsedTime := PlayerClasses.PlayersMap[Agent].flAbilityUsedTime, ElapsedTime := CurrentTime - PlayerClasses.PlayersMap[Agent].flAbilityUsedTime, ElapsedTime < 1.45 )
			{
				SFXAbilityCooldown.Play(Agent)
				return
			}

			Print("Snap: Activated")
			spawn{ChairAnim(Self,Agent,AnimCaretakerSnap,2.25)}
			FortCharacter.SetVulnerability(true) # prevent PutInStasis invincible bug!
			VFXPowerCaretakerSnap.Pickup(Agent)

			Sleep(1.9)
			TeamCollection := GetPlayspace().GetTeamCollection()
			if ( TeamOwner := TeamCollection.GetTeam[Agent] )
			{
				AllVictim := GetAllPlayersInDistance(FortCharacter, 1280.0)
				for ( ClosestVictim : AllVictim,ClosestAgent := ClosestVictim.GetAgent[], ( ClosestAgent <> Agent ), not TeamCollection.IsOnTeam[ClosestAgent, TeamOwner] )
				{
					Print("Snap: Damage Victim")
					if ( FortCharacter.IsActive[], DamageInstigator := Agent.GetInstigator(), DamageSource := game_action_causer[FortCharacter] )
					{
						DrainDmg : damage_args = damage_args {Instigator := option{DamageInstigator} , Source := option{DamageSource}, Amount:=70.0*PlayerClasses.GetDamageBonus(Agent)}
						ClosestVictim.Damage(DrainDmg)
					}
					VictimPos := ClosestVictim.GetTransform().Translation + vector3{X:= 0.0, Y:= 0.0, Z:= 50.0}
					spawn{ SpawnParticleVFX(FX.NS_Caretaker_Snap_Hit_asset, VictimPos, rotation{}, 1.2) }
				}
			}
		}
	}

	#
	# ehhh
	# 
    AbilityGrindVines(Agent:agent)<suspends>:void=
	{
		<#
		if ( FortCharacter := Agent.GetFortCharacter[], PlayerClasses := PlayerClassesDevice? )
		{
			if ( OldProp := PlayerClasses.PlayersMap[Agent].ActiveBuilding?, OldProp.IsValid[] )
			{
				return
				#OldProp.Dispose()
				#if {set PlayerClasses.PlayersMap[Agent].ActiveBuilding = false}
			}

			PlayerPos := FortCharacter.GetTransform().Translation
			PlayerRot := FortCharacter.GetViewRotation()
			if (PlayerRotYaw := PlayerRot.GetYawPitchRollDegrees()[0], PlayerRotRoll := PlayerRot.GetYawPitchRollDegrees()[2])
			{
				BuildingPos := PlayerPos - vector3{ X := 0.0, Y := 0.0, Z := 100.0}
				BuildingRot := MakeRotationFromYawPitchRollDegrees(PlayerRotYaw , 0.0, PlayerRotRoll)
				#spawn{ SpawnParticleVFX(FX.NS_CubeAssassin_Shield_asset, BuildingPos, BuildingRot, 0.5) }
				Sleep(0.4)
				SpawnVFX := SpawnProp(VFXPropGrindVines, BuildingPos, BuildingRot)
				if (SpawnedVFX:= SpawnVFX(0)?, set PlayerClasses.PlayersMap[Agent].ActiveBuilding = option{SpawnedVFX})
				{
					Print("Grind Vines: Spawn")
					PlayerClasses.PlayClassVO(Agent, FNVOType.Ability)
					race
					{
						block
						{
							FortCharacter.EliminatedEvent().Await()
							Print("Shield: FortCharacter EliminatedEvent...")
						}
						block
						{
							Sleep(5.0)
							Print("Grind Vines: Lifetime Ending...")
						}
					}
					Print("Grind Vines: Removed")
					if ( SpawnedVFX.IsValid[] )
					{
						SpawnedVFX.Dispose()
						if {set PlayerClasses.PlayersMap[Agent].ActiveBuilding = false}
					}
				}
			}
		}
		#>
	}

	AbilityMelee(Agent:agent)<suspends>:void=
	{
		if ( FortCharacter := Agent.GetFortCharacter[], PlayerClasses := PlayerClassesDevice? )
		{
			Print("Ability: Melee")

			case( PlayerClasses.GetPlayerClassIndex(Agent) )
			{
				FNClasses.FletcherKane =>		spawn{ChairAnim(Self,Agent,AnimKaneClaw,0.65)}
				_ =>							spawn{ChairAnim(Self,Agent,AnimNightRoseClaw,0.65)}
			}

			#VFXPowerMelee.Pickup(Agent)
			PlayerForwardPos := FortCharacter.GetTransform().Translation + FortCharacter.GetViewRotation().GetLocalForward() * 300.0
			PlayerClasses.PlayClassVO(Agent, FNVOType.Ability)

			Sleep(0.4)

			TeamCollection := GetPlayspace().GetTeamCollection()
			if ( TeamOwner := TeamCollection.GetTeam[Agent] )
			{
				#spawn{ SpawnParticleVFX(FX.NS_NightRose_Melee_asset, PlayerForwardPos, rotation{}, 1.0) }

				case( PlayerClasses.GetPlayerClassIndex(Agent) )
				{
					FNClasses.FletcherKane =>		spawn{ SpawnParticleVFX(FX.NS_FletcherKane_Melee_asset, PlayerForwardPos, rotation{}, 1.0) }
					_ =>							spawn{ SpawnParticleVFX(FX.NS_NightRose_Melee_asset, PlayerForwardPos, rotation{}, 1.0) }
				}

				AllVictim := GetAllPlayersInVectorDistance(PlayerForwardPos, 250.0)
				for ( ClosestVictim : AllVictim, ClosestAgent := ClosestVictim.GetAgent[], ClosestAgent <> Agent and ( not TeamCollection.IsOnTeam[ClosestAgent, TeamOwner] ) )
				{
					Print("NightRose: Claw Victim")
					var DmgMultiplier : float = PlayerClasses.GetDamageBonus(Agent)

					if ( FortCharacter.IsActive[], DamageInstigator := Agent.GetInstigator(), DamageSource := game_action_causer[FortCharacter] )
					{
						BeamDmg : damage_args = damage_args {Instigator := option{DamageInstigator}, Source := option{DamageSource}, Amount:=75.0*DmgMultiplier}
						ClosestVictim.Damage(BeamDmg)
						VictimPos := ClosestVictim.GetTransform().Translation + vector3{X:= 0.0, Y:= 0.0, Z:= 40.0}
						spawn{ SpawnParticleVFX(FX.NS_Impact_Red_asset, VictimPos, rotation{}, 0.25) }
					}
				}

				if ( ExplosiveDeviceLow.TeleportTo[PlayerForwardPos, rotation{}] )
				{
					ExplosiveDeviceLow.Explode(Agent)
				}
			}
		}
	}

	AbilityBombardment(Agent:agent)<suspends>:void=
	{
		if ( FortCharacter := Agent.GetFortCharacter[], PlayerClasses := PlayerClassesDevice? )
		{
			Print("ShogunX: Bombardment")
			#spawn{ChairAnim(Self,Agent,AnimShogunBombardment,3.0)}
			Agent.MoveMod(MoveModStop)
			VFXPowerShogunBombardment.Pickup(Agent)
			PlayerClasses.PlayClassVO(Agent, FNVOType.Ability)

			race
			{
				block
				{
					FortCharacter.EliminatedEvent().Await()
					Print("Successfully stopped reality collapse event!!!")
				}
				block
				{
					Sleep(3.0)
					TeamCollection := GetPlayspace().GetTeamCollection()
					if ( TeamOwner := TeamCollection.GetTeam[Agent] )
					{
						PlayerForwardPos := FortCharacter.GetTransform().Translation
						spawn{ SpawnParticleVFX(FX.NS_ShogunX_Bombardment_Bomb_asset, PlayerForwardPos, rotation{}, 1.0) }
						PlayerClasses.PlayClassVO(Agent, FNVOType.AbilityEnd)

						AllVictim := GetAllPlayersInVectorDistance(PlayerForwardPos, 600.0)
						for ( ClosestVictim : AllVictim, ClosestAgent := ClosestVictim.GetAgent[], ClosestAgent <> Agent and ( not TeamCollection.IsOnTeam[ClosestAgent, TeamOwner] ) )
						{
							Print("ShogunX: Bomb Victim")

							var DmgMultiplier : float = PlayerClasses.GetDamageBonus(Agent)

							if ( FortCharacter.IsActive[], DamageInstigator := Agent.GetInstigator(), DamageSource := game_action_causer[FortCharacter] )
							{
								BeamDmg : damage_args = damage_args {Instigator := option{DamageInstigator}, Source := option{DamageSource}, Amount:=120.0*DmgMultiplier}
								ClosestVictim.Damage(BeamDmg)
							}
						}

						if ( ExplosiveDeviceHigh.TeleportTo[PlayerForwardPos, rotation{}] )
						{
							ExplosiveDeviceHigh.Explode(Agent)
						}
						Agent.MoveMod(MoveModNormal)
					}
				}
			}
		}
	}

	AbilityEnrage(Agent:agent)<suspends>:void=
	{
		if ( FortCharacter := Agent.GetFortCharacter[], PlayerClasses := PlayerClassesDevice? )
		{
			spawn{ChairAnimInvulnerable(Self,Agent,AnimKaneEnrage,0.8)}
			PlayerClasses.PlayClassVO(Agent, FNVOType.Ult)
			Agent.MoveMod(MoveModPlus)
			PlayerPos := FortCharacter.GetTransform().Translation
			spawn{ SpawnParticleVFX(FX.NS_Aura_FletcherKane_Intro_asset, PlayerPos, rotation{}, 2.0) }
			VFXKaneEnrage.Pickup(Agent)
			PlayerClasses.UpdateHudDisableAbility(Agent, FNAbility.Enragement)
			#FortCharacter.Heal(30.0)
			spawn{Agent.SlowHeal(30.0)}
		}
	}

	DoomPrimaryAttack(Agent:agent)<suspends>:void=
	{
		if ( FortCharacter := Agent.GetFortCharacter[], PlayerClasses := PlayerClassesDevice?, SimulationEntity := GetSimulationEntity[] )
		{
			race
			{
				block
				{
					FortCharacter.EliminatedEvent().Await()
					Print("Successfully stopped reality collapse event!!!")
				}
				block
				{
					Print("Wanderer: Arcane Attack")
					CurrentTime := GetSimulationElapsedTime()
					if {set PlayerClasses.PlayersMap[Agent].flAbilityUsedTime = CurrentTime}

					NewEnt := Ent_Projectile_Doom{}
					NewEnt2 := Ent_Projectile_Doom{}
					NewEnt3 := Ent_Projectile_Doom{}
					SimulationEntity.AddEntities(array{NewEnt})

					#EntTransform := NewEnt.GetGlobalTransform().Scale
					OwnerTransform := FromTransform(transform{Translation := FortCharacter.GetTransform().Translation + FortCharacter.GetViewRotation().GetLocalForward() * 150.0, Rotation := FortCharacter.GetViewRotation(), Scale := FromVector3(NewEnt.GetGlobalTransform().Scale)})

					NewEnt.SetGlobalTransform(OwnerTransform)
					if ( ProjComp := NewEnt.GetComponent[projectile_component] )
					{
						ProjComp.FireProjectile(Agent)
					}
					Sleep(0.08)
					OwnerTransform2 := FromTransform(transform{Translation := FortCharacter.GetTransform().Translation + FortCharacter.GetViewRotation().GetLocalForward() * 150.0, Rotation := FortCharacter.GetViewRotation(), Scale := FromVector3(NewEnt.GetGlobalTransform().Scale)})
					SimulationEntity.AddEntities(array{NewEnt2})
					NewEnt2.SetGlobalTransform(OwnerTransform2)
					if ( ProjComp2 := NewEnt2.GetComponent[projectile_component] )
					{
						ProjComp2.FireProjectile(Agent)
					}
					Sleep(0.08)
					OwnerTransform3 := FromTransform(transform{Translation := FortCharacter.GetTransform().Translation + FortCharacter.GetViewRotation().GetLocalForward() * 150.0, Rotation := FortCharacter.GetViewRotation(), Scale := FromVector3(NewEnt.GetGlobalTransform().Scale)})
					SimulationEntity.AddEntities(array{NewEnt3})
					NewEnt3.SetGlobalTransform(OwnerTransform3)
					if ( ProjComp3 := NewEnt3.GetComponent[projectile_component] )
					{
						ProjComp3.FireProjectile(Agent)
					}
				}
			}
		}
	}
