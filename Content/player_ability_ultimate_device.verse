<#
	2AGENCY
	Player Abilites (Ultimate)
#>
using { /Fortnite.com/Devices }
using { /Fortnite.com/Devices/CreativeAnimation }
using { /Fortnite.com/Devices/CreativeAnimation/InterpolationTypes }
using { /Fortnite.com/Characters }
using { /Fortnite.com/Teams }
using { /Fortnite.com/Game }
using { /Fortnite.com/FortPlayerUtilities }
using { /Fortnite.com/AI }
using { /Fortnite.com/Animation/PlayAnimation }
using { /Verse.org/Assets }
using { /Verse.org/SceneGraph }
using { /Verse.org/Simulation }
using { /Verse.org/Simulation/Tags }
using { /Verse.org/Random }
using { /Verse.org/Colors }
using { /UnrealEngine.com/Assets }
using { /UnrealEngine.com/Temporary/Diagnostics }
using { /UnrealEngine.com/Temporary/SpatialMath }
using { ModulePlayerClasses }
using { ModulePropExtra }
using { ModuleAchievements }

#spawned_ult_device_tag := class(tag){}

# A Verse-authored creative device that can be placed in a level
player_ability_ultimate_device := class(creative_device):

	@editable PlayerClassesDevice : ?player_classes_manager_device = false

	@editable InputUlt : input_trigger_device = input_trigger_device{}

	@editable XP_Ultimate : accolades_device = accolades_device{}

	@editable ExplosiveDeviceUlt : explosive_device  = explosive_device{}

	@editable ItemGranterUltSgtWinter : item_granter_device = item_granter_device{}

	@editable CamUltOrbit			: gameplay_camera_orbit_device = gameplay_camera_orbit_device{}
	@editable CineUltPurple			: cinematic_sequence_device = cinematic_sequence_device{}
	@editable PlayerRef_UltPurple	: player_reference_device = player_reference_device{}
	@editable DMGAmplifier			: damage_amplifier_powerup_device = damage_amplifier_powerup_device{}
	@editable{Categories:=array{CategoryVFXPowerup}} VFXDMGAmplifier		: visual_effect_powerup_device = visual_effect_powerup_device{}
	@editable{Categories:=array{CategoryProp}} VFXUltMidas				: creative_prop_asset = DefaultCreativePropAsset
	@editable{Categories:=array{CategoryProp}} VFXUltValeria			: creative_prop_asset = DefaultCreativePropAsset
	@editable{Categories:=array{CategoryProp}} VFXUltSlone				: creative_prop_asset = DefaultCreativePropAsset
	@editable{Categories:=array{CategoryProp}} VFXUltCubeAssassin		: creative_prop_asset = DefaultCreativePropAsset
	@editable{Categories:=array{CategoryProp}} VFXUltCubeBlue			: creative_prop_asset = DefaultCreativePropAsset
	@editable{Categories:=array{CategoryVFXPowerup}} VFXPowerCubeInvulnerable : visual_effect_powerup_device = visual_effect_powerup_device{}
	@editable{Categories:=array{CategoryProp}} VFXUltCubeQueen			: creative_prop_asset = DefaultCreativePropAsset
	@editable{Categories:=array{CategoryProp}} VFXUltTheFoundation		: creative_prop_asset = DefaultCreativePropAsset
	@editable{Categories:=array{CategoryProp}} VFXUltTheHerald			: creative_prop_asset = DefaultCreativePropAsset
	@editable{Categories:=array{CategoryProp}} VFXUltTheAgeless			: creative_prop_asset = DefaultCreativePropAsset
	@editable{Categories:=array{CategoryProp}} VFXUltZeus				: creative_prop_asset = DefaultCreativePropAsset
	@editable{Categories:=array{CategoryProp}} VFXUltJones				: creative_prop_asset = DefaultCreativePropAsset
	@editable{Categories:=array{CategoryProp}} VFXPurple				: creative_prop_asset = DefaultCreativePropAsset
	@editable{Categories:=array{CategoryProp}} VFXPurpleBall			: creative_prop_asset = DefaultCreativePropAsset
	@editable{Categories:=array{CategoryProp}} VFXUltOcean				: creative_prop_asset = DefaultCreativePropAsset
	@editable{Categories:=array{CategorySFX}} SFXUltPurple				: audio_player_device = audio_player_device{}
	@editable{Categories:=array{CategoryProp}} VFXUltKadoThorne			: creative_prop_asset = DefaultCreativePropAsset
	@editable{Categories:=array{CategorySFX}} SFXUltKadoThorne			: audio_player_device = audio_player_device{}
	@editable{Categories:=array{CategoryVFXPowerup}} VFXPowerUltKadoThorne : visual_effect_powerup_device = visual_effect_powerup_device{}
	@editable{Categories:=array{CategoryProp}} VFXUltTheIceKing			: creative_prop_asset = DefaultCreativePropAsset
	@editable{Categories:=array{CategorySFX}} SFXUltTheIceKing			: audio_player_device = audio_player_device{}
	@editable{Categories:=array{CategoryProp}} VFXUltNightRose			: creative_prop_asset = DefaultCreativePropAsset
	@editable{Categories:=array{CategorySFX}} SFXUltNightRose			: audio_player_device = audio_player_device{}
	@editable{Categories:=array{CategoryVFXPowerup}} VFXPowerUltNightRose : visual_effect_powerup_device = visual_effect_powerup_device{}
	@editable{Categories:=array{CategoryAnim}} AnimPuppetIdle			: creative_prop_asset = DefaultCreativePropAsset
	@editable{Categories:=array{CategoryVFXPowerup}} VFXPowerUltCubeAssassin : visual_effect_powerup_device = visual_effect_powerup_device{}
	@editable{Categories:=array{CategoryVFXPowerup}} VFXPowerUltMegaloDon : visual_effect_powerup_device = visual_effect_powerup_device{}
	@editable{Categories:=array{CategoryProp}} VFXUltMegaloDon			: creative_prop_asset = DefaultCreativePropAsset
	@editable MoveModCharge : movement_modulator_device  = movement_modulator_device{}
	@editable{Categories:=array{CategoryAnim}} AnimDoomTransform		: creative_prop_asset = DefaultCreativePropAsset
	@editable{Categories:=array{CategoryVFXPowerup}} VFXPowerUltDoomCharge	: visual_effect_powerup_device = visual_effect_powerup_device{}
	@editable{Categories:=array{CategoryProp}} VFXUltShogunX			: creative_prop_asset = DefaultCreativePropAsset
	@editable{Categories:=array{CategoryAnim}} AnimShogunXBlade			: creative_prop_asset = DefaultCreativePropAsset
	@editable{Categories:=array{CategoryAnim}} AnimUltCubeQueen			: creative_prop_asset = DefaultCreativePropAsset
	@editable{Categories:=array{CategoryVFXPowerup}} VFXPowerUltCubeQueen : visual_effect_powerup_device = visual_effect_powerup_device{}
	@editable{Categories:=array{CategoryAnim}} AnimUltIceKing			: creative_prop_asset = DefaultCreativePropAsset
	@editable{Categories:=array{CategoryVFXPowerup}} VFXPowerUltIceKing : visual_effect_powerup_device = visual_effect_powerup_device{}
	@editable{Categories:=array{CategoryVFXPowerup}} VFXPowerUltZeus : visual_effect_powerup_device = visual_effect_powerup_device{}
	@editable{Categories:=array{CategoryAnim}} AnimUltHerald			: creative_prop_asset = DefaultCreativePropAsset
	@editable{Categories:=array{CategoryAnim}} AnimUltGeno				: creative_prop_asset = DefaultCreativePropAsset
	@editable{Categories:=array{CategoryAnim}} AnimUltZeus				: creative_prop_asset = DefaultCreativePropAsset
	@editable{Categories:=array{CategoryAnim}} AnimUltCaretaker			: creative_prop_asset = DefaultCreativePropAsset
	@editable{Categories:=array{CategoryAnim}} AnimUltKane				: creative_prop_asset = DefaultCreativePropAsset
	@editable{Categories:=array{CategoryProp}} VFXUltKane				: creative_prop_asset = DefaultCreativePropAsset
	@editable{Categories:=array{CategoryVFXPowerup}} VFXPowerUltKane 	: visual_effect_powerup_device = visual_effect_powerup_device{}
	@editable{Categories:=array{CategoryAnim}} AnimUltDaigo				: creative_prop_asset = DefaultCreativePropAsset
	@editable{Categories:=array{CategoryVFXPowerup}} VFXPowerUltDaigo 	: visual_effect_powerup_device = visual_effect_powerup_device{}
	@editable{Categories:=array{CategoryAnim}} AnimUltDaigoVictim				: creative_prop_asset = DefaultCreativePropAsset
	@editable{Categories:=array{CategoryVFXPowerup}} VFXPowerUltDaigoVictim 	: visual_effect_powerup_device = visual_effect_powerup_device{}

    # Runs when the device is started in a running game
    OnBegin<override>()<suspends>:void=
	{
		Print("Initialize Player Abilities (Ultimate)...")

		InputUlt.PressedEvent.Subscribe(InputPressedUlt)
	}

	GetUltimateMeter(Agent:agent)<transacts>:float=
	{
		if ( PlayerClasses := PlayerClassesDevice? )
		{
			return PlayerClasses.GetUltimateMeter(Agent)
		}

		return 0.0
	}

	GetMaxUlt(Agent:agent):float=
	{
		if ( PlayerClasses := PlayerClassesDevice? )
		{
			MaxUlt : float = PlayerClasses.GetMaxUlt(Agent)
			return MaxUlt
		}

		return 100.0
	}

	InputPressedUlt(Agent : agent) : void =
	{
		if (FortCharacter := Agent.GetFortCharacter[], FortCharacter.IsActive[], not FortCharacter.IsGliding[], PlayerClasses := PlayerClassesDevice? )
		{
			MaxUlt := GetMaxUlt(Agent)
			if ( GetUltimateMeter(Agent) = MaxUlt )
			{
				UseUltimate(Agent)
				return
			}
		}
	}

	# Emoting causes fucked up sprint and slide....?!?!
	# ...Emoting again after the bug will fix the bug
	OnPlayerBeginsEmoting(Agent:agent):void= 
	{
		MaxUlt := GetMaxUlt(Agent)
		if ( GetUltimateMeter(Agent) = MaxUlt )
		{
			UseUltimate(Agent)
			return
		}

		if ( PlayerClasses := PlayerClassesDevice? )
		{
			PlayerClasses.PlayClassVO(Agent, FNVOType.Idle)
		}
	}

	GetAllPlayersInVectorDistance( CenterPos : vector3, DistanceRange: float)<transacts>:[]fort_character=
	{
		var ClosestPlayers: []fort_character = array{}

		if ( Entity := GetSimulationEntity[] )
		{
			CollisionSphere := collision_sphere{ Radius := DistanceRange <#, CollisionProfile := collision_profiles.StationaryOverlapAll#> }
			CollisionTransform := FromTransform(transform{Translation := CenterPos})

			for
			{
				Overlap : Entity.FindOverlapHits(CollisionTransform, CollisionSphere)
				PlayerMesh := Meshes.PlayerBox[Overlap.TargetComponent]
				TargetEntity := PlayerMesh.Entity
			}
			do
			{
				if ( PlayerComp := TargetEntity.GetComponent[fn_player_component], Agent := PlayerComp.Agent, FortChar := Agent.GetFortCharacter[] )
				{
					set ClosestPlayers += array{FortChar}
				}

				if ( NPCComp := TargetEntity.GetComponent[fn_npc_component], AgentNPC := NPCComp.Agent, FortNPC := AgentNPC.GetFortCharacter[] )
				{
					set ClosestPlayers += array{FortNPC}
				}
			}
		}


		Drawer := debug_draw{}
		Drawer.DrawSphere(CenterPos, ?Radius := DistanceRange, ?DrawDurationPolicy := debug_draw_duration_policy.FiniteDuration, ?Duration:= float(1.5))
        return ClosestPlayers
	}

	GetAllPlayersInDistance(PlayerCenter : fort_character, DistanceRange: float)<transacts>:[]fort_character=
	{
		CenterCurrentPos := PlayerCenter.GetTransform().Translation
        return GetAllPlayersInVectorDistance(CenterCurrentPos, DistanceRange)
	}

	GetAllPlayersInPropDistance(PropCenter : creative_prop, DistanceRange: float)<transacts>:[]fort_character=
	{
		CenterCurrentPos := PropCenter.GetTransform().Translation
        return GetAllPlayersInVectorDistance(CenterCurrentPos, DistanceRange)
	}

	UseUltimate(Agent:agent):void=
	{
		if ( PlayerClasses := PlayerClassesDevice? )
		{
			InputUlt.Unregister(Agent)

			if ( FortCharacter := Agent.GetFortCharacter[], FortCharacter.IsActive[] )
			{
				# FORCE player out of emote to fix sprinting issues.
				MoveModCharge.Activate(Agent)

				# a little gift
				FortCharacter.Heal(50.0)
			}

			#Print("Ultimate: ")

			PlayerClasses.PlayClassVO(Agent, FNVOType.Ult)
			spawn{PlayerClasses.AwardScore(Agent, 1.0)}
			case( PlayerClasses.GetPlayerClassIndex(Agent) )
			{
				FNClasses.Midas =>
				{
					spawn{ UltContinuousDMG(Agent, VFXUltMidas, 10.0, 0.7) }
					spawn{PlayerClasses.AchievementUnlock(Agent, 10)}
				}
				FNClasses.TNTina =>	
				{
					spawn{ UltContinuousDMG(Agent, VFXUltMidas, 10.0, 0.7) }
					spawn{PlayerClasses.AchievementUnlock(Agent, 10)}
				}
				FNClasses.Brutus =>	
				{
					spawn{ UltContinuousDMG(Agent, VFXUltMidas, 10.0, 0.7) }
					spawn{PlayerClasses.AchievementUnlock(Agent, 10)}
				}
				FNClasses.Skye =>
				{
					spawn{ UltContinuousDMG(Agent, VFXUltMidas, 10.0, 0.7) }
					spawn{PlayerClasses.AchievementUnlock(Agent, 10)}
				}
				FNClasses.Meowscles =>
				{
					spawn{ UltContinuousDMG(Agent, VFXUltMidas, 10.0, 0.7) }
					spawn{PlayerClasses.AchievementUnlock(Agent, 10)}
				}
				FNClasses.Jules =>
				{
					spawn{ UltFlood(Agent) }
				}
				FNClasses.Kit =>
				{
					spawn{ UltFlood(Agent) }
				}
				FNClasses.Ocean =>
				{
					spawn{ UltFlood(Agent) }
				}
				FNClasses.ShadowMidas =>
				{
					spawn{ UltContinuousDMG(Agent, VFXUltMidas, 10.0, 0.7) }
					spawn{PlayerClasses.AchievementUnlock(Agent, 10)}
				}
				FNClasses.Shadow =>
				{
					spawn{ DamageBuffPlayer(Agent, 10.0, 2.0, ?IsUlt:=true) }
				}
				FNClasses.TheIceKing =>
				{
					spawn{ UltTheWorld(Agent, FNClasses.TheIceKing) }
					spawn{PlayerClasses.AchievementUnlock(Agent, 18)}
				}
				FNClasses.Raz =>				spawn{ DamageBuffPlayer(Agent, 10.0, 2.0) }
				FNClasses.Slone =>
				{
					spawn{ UltNuke(Agent, FNClasses.Slone) }
					spawn{PlayerClasses.AchievementUnlock(Agent, 11)}
				}
				FNClasses.CubeBlue =>
				{
					spawn{ UltCubeBlueInvulnerable(Agent) }
					spawn{PlayerClasses.AchievementUnlock(Agent, 13)}
				}
				FNClasses.CubeQueen =>
				{
					spawn{ UltNuke(Agent, FNClasses.CubeQueen) }
					spawn{PlayerClasses.AchievementUnlock(Agent, 12)}
				}
				FNClasses.TheFoundation =>		spawn{ UltNuke(Agent, FNClasses.TheFoundation) }
				FNClasses.Gunnar =>				spawn{ DamageBuffPlayer(Agent, 10.0, 2.0, ?IsUlt:=true) }
				FNClasses.HuntmasterSaber =>	spawn{ DamageBuffPlayer(Agent, 10.0, 2.0, ?IsUlt:=true) }
				FNClasses.TheHerald =>
				{
					spawn{ UltNuke(Agent, FNClasses.TheHerald) }
					spawn{PlayerClasses.AchievementUnlock(Agent, 14)}
				}
				FNClasses.TheAgelessChampion =>
				{
					spawn{ UltNuke(Agent, FNClasses.TheAgelessChampion) }
					spawn{PlayerClasses.AchievementUnlock(Agent, 15)}
				}
				FNClasses.WildguardRelik =>
				{
					spawn{ UltHollowPurple(Agent, FNClasses.WildguardRelik) }
					spawn{PlayerClasses.AchievementUnlock(Agent, 16)}
				}
				FNClasses.KadoThorne =>
				{
					spawn{ UltTheWorld(Agent, FNClasses.KadoThorne) }
					spawn{PlayerClasses.AchievementUnlock(Agent, 17)}
				}
				FNClasses.TheDealer =>
				{
					spawn{ UltTheWorld(Agent, FNClasses.KadoThorne) }
					spawn{PlayerClasses.AchievementUnlock(Agent, 17)}
				}
				FNClasses.Valeria =>
				{
					spawn{ UltContinuousDMG(Agent, VFXUltValeria, 10.0, 0.7) }
					spawn{PlayerClasses.AchievementUnlock(Agent, 19)}
				}
				FNClasses.Oscar =>
				{
					spawn{ UltContinuousDMG(Agent, VFXUltValeria, 10.0, 0.7) }
					spawn{PlayerClasses.AchievementUnlock(Agent, 19)}
				}
				FNClasses.Nisha =>
				{
					spawn{ UltContinuousDMG(Agent, VFXUltValeria, 10.0, 0.7) }
					spawn{PlayerClasses.AchievementUnlock(Agent, 19)}
				}
				FNClasses.Montague =>
				{
					spawn{ UltContinuousDMG(Agent, VFXUltValeria, 10.0, 0.7) }
					spawn{PlayerClasses.AchievementUnlock(Agent, 19)}
				}
				FNClasses.SgtWinter =>			ItemGranterUltSgtWinter.GrantItem(Agent)
				FNClasses.CubeAssassin =>		spawn{ UltSidewaysExpansion(Agent) }
				FNClasses.SpireAssassin =>		spawn{ DamageBuffPlayer(Agent, 10.0, 2.0, ?IsUlt:=true) }
				FNClasses.ZygAndChoppy =>		spawn{ DamageBuffPlayer(Agent, 10.0, 2.0, ?IsUlt:=true) }
				FNClasses.Zeus =>
				{
					spawn{ UltNuke(Agent, FNClasses.Zeus) }
					spawn{PlayerClasses.AchievementUnlock(Agent, 20)}
				}
				FNClasses.Ares =>
				{
					spawn{ UltNuke(Agent, FNClasses.Zeus) }
					spawn{PlayerClasses.AchievementUnlock(Agent, 20)}
				}
				FNClasses.Cerberus =>
				{
					spawn{ UltNuke(Agent, FNClasses.Zeus) }
					spawn{PlayerClasses.AchievementUnlock(Agent, 20)}
				}
				FNClasses.Hades =>
				{
					spawn{ UltNuke(Agent, FNClasses.Zeus) }
					spawn{PlayerClasses.AchievementUnlock(Agent, 20)}
				}
				FNClasses.Jones =>
				{
					spawn{ UltNuke(Agent, FNClasses.Jones) }
					spawn{PlayerClasses.AchievementUnlock(Agent, 21)}
				}
				FNClasses.MegaloDon =>
				{
					spawn{ UltMegaCharge(Agent) }
					spawn{PlayerClasses.AchievementUnlock(Agent, 22)}
				}
				FNClasses.RingmasterScarr =>
				{
					spawn{ UltMegaCharge(Agent) }
					spawn{PlayerClasses.AchievementUnlock(Agent, 22)}
				}
				FNClasses.TheMachinist =>
				{
					spawn{ UltMegaCharge(Agent) }
					spawn{PlayerClasses.AchievementUnlock(Agent, 22)}
				}
				FNClasses.UltimaCarver =>
				{
					spawn{ DamageBuffPlayer(Agent, 10.0, 2.0, ?IsUlt:=true) }
				}
				FNClasses.DoctorDoom =>
				{
					spawn{ UltDoom(Agent) }
				}
				FNClasses.ShogunX =>
				{
					#spawn{ DamageBuffPlayer(Agent, 10.0, 2.0, ?IsUlt:=true) }
					spawn{ UltContinuousDMG(Agent, VFXUltShogunX, 3.0, 0.25) }
					spawn{PlayerClasses.AchievementUnlock(Agent, 23)}
				}
				FNClasses.NightRose =>
				{
					spawn{ UltTheWorld(Agent, FNClasses.NightRose) }
				}
				FNClasses.Puppeteer =>
				{
					spawn{ DamageBuffPlayer(Agent, 10.0, 2.0, ?IsUlt:=true) }
				}
				FNClasses.FletcherKane =>
				{
					spawn{ UltHollowPurple(Agent, FNClasses.FletcherKane) }
				}
				FNClasses.AlphaHeavy =>
				{
					spawn{ UltHollowPurple(Agent, FNClasses.FletcherKane) }
				}
				FNClasses.Caretaker =>
				{
					spawn{ UltSidewaysExpansion(Agent) }
				}
				FNClasses.Hope =>
				{
					spawn{ UltNuke(Agent, FNClasses.Hope) }
				}
				FNClasses.Daigo =>
				{
					#spawn{ UltSplinterBlade(Agent) }
					spawn{ UltHollowPurple(Agent, FNClasses.FletcherKane) }
				}
				FNClasses.Kor =>
				{
					#spawn{ DamageBuffPlayer(Agent, 10.0, 2.0, ?IsUlt:=true) }
					spawn{ UltHollowPurple(Agent, FNClasses.FletcherKane) }
				}
				_ =>							# nothing
			}

			XP_Ultimate.Award(Agent)
			PlayerClasses.ExpendUltimateMeter(Agent)
			#if {set PlayerClasses.PlayersMap[Agent].bUltReady = false}
		}
	}

    AddCamUlt(Agent:agent, CamType:int, Duration:float)<suspends>:void=
	{
		if ( FortCharacter := Agent.GetFortCharacter[] )
		{

		}
	}

    PlayCineUlt(Agent:agent, Duration:float)<suspends>:void=
	{
		if ( FortCharacter := Agent.GetFortCharacter[] )
		{
			FortCharacter.PutInStasis(stasis_args{AllowTurning := false, AllowFalling := false, AllowEmotes := true})
			PlayerRef_UltPurple.Clear()
			PlayerRef_UltPurple.Register(Agent)
			PlayerRef_UltPurple.Activate()
			CineUltPurple.Play(Agent)
			Sleep(Duration)
			FortCharacter.ReleaseFromStasis()
			CineUltPurple.Stop(Agent)
		}
	}

    UltNuke(Agent:agent, ClassIndex:FNClasses)<suspends>:void=
	{
		if ( FortCharacter := Agent.GetFortCharacter[], PlayerClasses := PlayerClassesDevice? )
		{
			Print("Nuke: Activate")
			PlayerPos := FortCharacter.GetTransform().Translation
			PlayerRot := FortCharacter.GetViewRotation()
			var CamRot : rotation = PlayerRot
			if (PlayerRotYaw := PlayerRot.GetYawPitchRollDegrees()[0])
			{
				set CamRot = MakeRotationFromYawPitchRollDegrees(PlayerRotYaw, 0.0, 0.0) 
			}

			var VFXType : creative_prop_asset = VFXUltSlone
			case( ClassIndex )
			{
				FNClasses.CubeQueen =>
				{
					set VFXType = VFXUltCubeQueen
					VFXPowerUltCubeQueen.Pickup(Agent)
					spawn{ModulePlayerEx.ChairAnimInvulnerable(Self,Agent,AnimUltCubeQueen,5.2)}
					#spawn{AddCamUlt(Agent,1,1.5)}
				}
				FNClasses.TheFoundation =>
				{
					set VFXType = VFXUltTheFoundation
					#spawn{AddCamUlt(Agent,0,1.0)}
				}
				FNClasses.TheHerald =>
				{
					set VFXType = VFXUltTheHerald
					spawn{ModulePlayerEx.ChairAnimInvulnerable(Self,Agent,AnimUltHerald,5.2)}
					#spawn{AddCamUlt(Agent,0,1.0)}
				}
				FNClasses.TheAgelessChampion =>
				{
					set VFXType = VFXUltTheAgeless
					#spawn{ModulePlayerEx.ChairAnimInvulnerable(Self,Agent,AnimUltGeno,5.2)}
					#spawn{AddCamUlt(Agent,0,1.0)}
					#spawn{PlayerUltAnimation(Agent,FNClasses.TheAgelessChampion)}
				}
				FNClasses.Zeus =>
				{
					set VFXType = VFXUltZeus
					VFXPowerUltZeus.Pickup(Agent)
					spawn{ModulePlayerEx.ChairAnimInvulnerable(Self,Agent,AnimUltZeus,5.2)}
					#spawn{AddCamUlt(Agent,1,1.5)}
				}
				FNClasses.Jones =>
				{
					set VFXType = VFXUltJones
					#spawn{AddCamUlt(Agent,1,1.5)}
				}
				FNClasses.Hope =>
				{
					set VFXType = VFXUltJones
					#spawn{AddCamUlt(Agent,1,1.5)}
				}
				_ =>	#nothing
			}

			StartVFX := SpawnProp(VFXType, PlayerPos, CamRot)
			Sleep(5.3)
			if (StartedVFX:= StartVFX(0)?)
			{
				defer
				{
					StartedVFX.Dispose()
				}

				if ( FortCharacter.IsActive[] )
				{
					TeamCollection := GetPlayspace().GetTeamCollection()
					if ( TeamOwner := TeamCollection.GetTeam[Agent] )
					{
						AllVictim := GetAllPlayersInVectorDistance(PlayerPos, 2000.0)
						for ( ClosestVictim : AllVictim, ClosestAgent := ClosestVictim.GetAgent[], ClosestAgent <> Agent and ( not TeamCollection.IsOnTeam[ClosestAgent, TeamOwner] ) )
						{
							Print("Nuke: Nuke Victim")
							var DmgMultiplier : float = 1.0
							if ( ClosestPlayer := player[ClosestVictim] ) # if player
							{
							}
							else
							{
								set DmgMultiplier = 5.0
							}
							if ( FortCharacter.IsActive[], DamageInstigator := Agent.GetInstigator(), DamageSource := game_action_causer[FortCharacter] )
							{
								NukeDmg : damage_args = damage_args {Instigator := option{DamageInstigator}, <#Source := option{DamageSource},#> Amount:=500.0*DmgMultiplier}
								ClosestVictim.Damage(NukeDmg)
							}
						}
					}
					if ( ExplosiveDeviceUlt.TeleportTo[PlayerPos, rotation{}] ) {}
					ExplosiveDeviceUlt.Explode(Agent)
					Sleep(4.2)
				}
			}
			if {set PlayerClasses.PlayersMap[Agent].bUltReady = false}
		}
	}

    <#PlayerUltAnimation(Agent:agent, ClassIndex:FNClasses)<suspends>:void=
	{
		if ( FortCharacter := Agent.GetFortCharacter[], PlayerClasses := PlayerClassesDevice? )
		{
			PlayerPos := FortCharacter.GetTransform().Translation
			PlayerRot := FortCharacter.GetViewRotation()
			var CamRot : rotation = PlayerRot
			if (PlayerRotYaw := PlayerRot.GetYawPitchRollDegrees()[0])
			{
				set CamRot = MakeRotationFromYawPitchRollDegrees(PlayerRotYaw, 0.0, 0.0) 
			}

			case( ClassIndex )
			{
				FNClasses.TheAgelessChampion =>
				{
					Print("Anim: Ageless")
					NewTransform := transform{Scale := vector3{X:= 0.01, Y:= 0.01, Z:= 0.01}, Rotation := CamRot, Translation := PlayerPos}
					SpawnVFX := SpawnProp(AnimUltGeno, NewTransform)
					if (SpawnedVFX:= SpawnVFX(0)?)
					{
						Print("Anim Ageless: Spawn Prop")
						defer
						{
							FortCharacter.ReleaseFromStasis()
							SpawnedVFX.Dispose()
						}
						PropPos := SpawnedVFX.GetTransform().Translation
						TaggedDevice := FindCreativeObjectsWithTag(chairanim_tag{})
						for (FoundDevice: TaggedDevice, ChairDevice := chair_device[FoundDevice])
						{
							Print("Anim Ageless: Lookin for Chair")
							DevicePos := FoundDevice.GetTransform().Translation
							if ( IsAlmostEqual[PropPos,DevicePos,1.0] )
							{
								Print("Anim Ageless: Found Chair")
								ChairDevice.Seat(Agent)
								FortCharacter.PutInStasis(stasis_args{AllowTurning := false, AllowFalling := false, AllowEmotes := true})
								NewPos := FortCharacter.GetTransform().Translation + vector3{X:= -500.0, Y:= 0.0, Z:= 0.0}
								Sleep(3.0)
								ChairDevice.MoveTo(NewPos, CamRot, 2.5)
								Sleep(2.5)
								ChairDevice.Eject(Agent)
							}
						}

						FortCharacter.SetVulnerability(true)
					}
				}
				_ =>	#nothing
			}
		}
	}#>

    UltContinuousDMG(Agent:agent, VFXType:creative_prop_asset, UltDuration:float, UltDelay:float)<suspends>:void=
	{
		if ( FortCharacter := Agent.GetFortCharacter[], PlayerClasses := PlayerClassesDevice? )
		{
			Print("Continous AoE: Activate")
			PlayerPos := FortCharacter.GetTransform().Translation
			PlayerRot := FortCharacter.GetViewRotation()
			var CamRot : rotation = PlayerRot
			if (PlayerRotYaw := PlayerRot.GetYawPitchRollDegrees()[0])
			{
				set CamRot = MakeRotationFromYawPitchRollDegrees(PlayerRotYaw, 0.0, 0.0) 
			}

			ClassIndex := PlayerClasses.GetPlayerClassIndex(Agent)
			if ( ClassIndex = FNClasses.ShogunX )
			{
				spawn{ModulePlayerEx.ChairAnimInvulnerable(Self,Agent,AnimShogunXBlade,1.9)}
			}
			else
			{
				#spawn{AddCamUlt(Agent,0,1.0)}
			}

			StartVFX := SpawnProp(VFXType, PlayerPos, CamRot)
			Sleep(2.0)
			if (StartedVFX:= StartVFX(0)?)
			{
				defer
				{
					StartedVFX.Dispose()
					if ( ClassIndex = FNClasses.ShogunX )
					{
						spawn{ SpawnParticleVFX(FX.NS_Warp_NightRose_asset, PlayerPos + vector3{X:= 0.0, Y:= 0.0, Z:= 50.0}, rotation{}, 1.2) }
						FortCharacter.ReleaseFromStasis()
						FortCharacter.Show()
					}
				}

				if ( FortCharacter.IsActive[] )
				{
					if ( ClassIndex = FNClasses.ShogunX )
					{
						#spawn{AddCamUlt(Agent,0,3.0)}
						spawn{ SpawnParticleVFX(FX.NS_Warp_NightRose_asset, PlayerPos + vector3{X:= 0.0, Y:= 0.0, Z:= 50.0}, rotation{}, 1.2) }
						FortCharacter.Hide()
						FortCharacter.PutInStasis(stasis_args{AllowTurning := false, AllowFalling := false, AllowEmotes := true})
						FortCharacter.SetVulnerability(true)
					}

					if ( ExplosiveDeviceUlt.TeleportTo[PlayerPos, rotation{}] ) {}
					ExplosiveDeviceUlt.Explode(Agent)
					var bDoneDMG : logic = false
					branch
					{
						TeamCollection := GetPlayspace().GetTeamCollection()
						if ( TeamOwner := TeamCollection.GetTeam[Agent] )
						{
							loop
							{
								if ( not FortCharacter.IsActive[] )
								{
									break
								}

								if ( bDoneDMG? )
								{
									break
								}
								AllVictim := GetAllPlayersInVectorDistance(PlayerPos, 2000.0)
								for ( ClosestVictim : AllVictim, ClosestAgent := ClosestVictim.GetAgent[], ClosestAgent <> Agent and ( not TeamCollection.IsOnTeam[ClosestAgent, TeamOwner] ) )
								{
									Print("Continous AoE: DMG Victim")
									var DmgMultiplier : float = 1.0
									if ( ClosestPlayer := player[ClosestVictim] ) # if player
									{
										set DmgMultiplier = 1.0
									}
									else
									{
										set DmgMultiplier = 5.0
									}
									if ( FortCharacter.IsActive[], DamageInstigator := Agent.GetInstigator(), DamageSource := game_action_causer[FortCharacter] )
									{
										NukeDmg : damage_args = damage_args {Instigator := option{DamageInstigator}, <#Source := option{DamageSource},#> Amount:=50.0*DmgMultiplier}
										ClosestVictim.Damage(NukeDmg)
									}
								}
								Sleep(UltDelay)
							}
						}
					}
					Sleep(UltDuration)
					set bDoneDMG = true
				}
			}
			if {set PlayerClasses.PlayersMap[Agent].bUltReady = false}
		}
	}

    UltFlood(Agent:agent)<suspends>:void=
	{
		if ( FortCharacter := Agent.GetFortCharacter[], PlayerClasses := PlayerClassesDevice? )
		{
			Print("Flood: Activate")
			PlayerPos := FortCharacter.GetTransform().Translation
			PlayerRot := FortCharacter.GetViewRotation()
			var CamRot : rotation = PlayerRot
			if (PlayerRotYaw := PlayerRot.GetYawPitchRollDegrees()[0])
			{
				set CamRot = MakeRotationFromYawPitchRollDegrees(PlayerRotYaw, 0.0, 0.0) 
			}

			StartVFX := SpawnProp(VFXUltOcean, PlayerPos, CamRot)
			if (SpawnedVFX := StartVFX(0)?)
			{
				defer
				{
					SpawnedVFX.Dispose()
				}
				<#PropPos := SpawnedVFX.GetTransform().Translation
				TaggedDevice := FindCreativeObjectsWithTag(spawned_ult_device_tag{})
				for (FoundDevice: TaggedDevice, WaterDevice := water_device[FoundDevice])
				{
					DevicePos := FoundDevice.GetTransform().Translation
					if ( IsAlmostEqual[PropPos,DevicePos,0.1] )
					{
						Print("found water device!")
						WaterDevice.AgentEntersWaterEvent.Subscribe(OnEnterFlood)
						WaterDevice.BeginVerticalFilling()
						Sleep(6.0)
						WaterDevice.BeginVerticalEmptying()
					}
				}#>
				Sleep(6.0)
			}
			if {set PlayerClasses.PlayersMap[Agent].bUltReady = false}
		}
	}

	OnEnterFlood(Agent:agent):void=
	{
		Print("OnEnterFlood")
		if ( FortEnter := Agent.GetFortCharacter[] )
		{
			FortEnter.Damage(1.0)
		}
	}

    UltHollowPurple(Agent:agent, ClassIndex:FNClasses)<suspends>:void=
	{
		if ( FortCharacter := Agent.GetFortCharacter[], PlayerClasses := PlayerClassesDevice? )
		{
			Print("PURPLE: Activate")

			var VFXType : creative_prop_asset = VFXPurpleBall
			if ( ClassIndex = FNClasses.FletcherKane )
			{
				spawn{ModulePlayerEx.ChairAnimFollow(Self,Agent,AnimUltKane,4.2)}
				VFXPowerUltKane.Pickup(Agent)
				set VFXType = VFXUltKane
				#PlayerPos := FortCharacter.GetTransform().Translation
				#spawn{ SpawnParticleVFX(FX.NS_Ult_FletcherKane_Ex_asset, PlayerPos, rotation{}, 5.0) }
			}
			else
			{
				SFXUltPurple.Play(Agent)
				spawn{ PlayCineUlt(Agent, 1.5) }
				spawn{ SpawnPropVFXFollowPlayer(VFXPurple,FortCharacter,2.7) }
				set VFXType = VFXPurpleBall
			}

			FortCharacter.SetVulnerability( false )

			Sleep(1.5)
			FortCharacter.SetVulnerability( true )
			Sleep(1.1)
			PlayerForwardPos := FortCharacter.GetTransform().Translation + FortCharacter.GetViewRotation().GetLocalForward() * 400.0
			PlayerRotation := FortCharacter.GetViewRotation()
			StartVFX := SpawnProp(VFXType, PlayerForwardPos, PlayerRotation)
			if (StartedVFX:= StartVFX(0)?)
			{
				if ( ExplosiveDeviceUlt.TeleportTo[PlayerForwardPos, rotation{}] ) {}
				ExplosiveDeviceUlt.Explode(Agent)
				ProjectileForwardPos := FortCharacter.GetTransform().Translation + FortCharacter.GetViewRotation().GetLocalForward() * 7500.0
				#spawn{ StartedVFX.MoveTo(ProjectileForwardPos, PlayerRotation, 4.0) }

				if (AController := StartedVFX.GetAnimationController[])
				{
					AnimKeyframes := array
					{
						keyframe_delta
						{
							DeltaLocation := ProjectileForwardPos - StartedVFX.GetTransform().Translation
							DeltaRotation := MakeComponentWiseDeltaRotation(PlayerRotation, StartedVFX.GetTransform().Rotation)
							Time := 4.0
							Interpolation := EaseIn
						}
					}
					AController.SetAnimation(AnimKeyframes, ?Mode:=animation_mode.OneShot)
					AController.Play()
				}

				branch
				{
					loop
					{
						if ( not StartedVFX.IsValid[] )
						{
							Print("StartedVFX not valid, stopping Purple...")
							break
						}

						TeamCollection := GetPlayspace().GetTeamCollection()
						if ( TeamOwner := TeamCollection.GetTeam[Agent] )
						{
							AllVictim := GetAllPlayersInPropDistance(StartedVFX, 400.0)
							for ( ClosestVictim : AllVictim, ClosestAgent := ClosestVictim.GetAgent[], ClosestAgent <> Agent and ( not TeamCollection.IsOnTeam[ClosestAgent, TeamOwner] ) )
							{
								Print("PURPLE: Hollow Victim")
								if ( FortCharacter.IsActive[], DamageInstigator := Agent.GetInstigator(), DamageSource := game_action_causer[FortCharacter] )
								{
									ProjectileDMG : damage_args = damage_args{Instigator := option{DamageInstigator}, <#Source := option{DamageSource},#> Amount:=220.0}
									ClosestVictim.Damage(ProjectileDMG)
								}
							}
						}

						Sleep(0.25)
					}
				}
				block
				{
					Sleep(4.0)
					EndPos := StartedVFX.GetTransform().Translation
					spawn{ SpawnParticleVFX(FX.NS_Ult_FletcherKane_Projectile_End_asset, EndPos, rotation{}, 7.0) }
					StartedVFX.Dispose()
				}
			}
			if {set PlayerClasses.PlayersMap[Agent].bUltReady = false}
		}
	}

	FreezePlayer(Agent:agent, Duration:float)<suspends>:void=
	{
		if ( FortCharacter := Agent.GetFortCharacter[] )
		{
			FortCharacter.PutInStasis(stasis_args{AllowTurning := false, AllowFalling := false, AllowEmotes := true})
			FortCharacter.SetVulnerability(true) # prevent PutInStasis invincible bug!
			Sleep(Duration)
			FortCharacter.ReleaseFromStasis()
		}
	}

	PuppetPlayer(Agent:agent, Duration:float)<suspends>:void=
	{
		if ( FortCharacter := Agent.GetFortCharacter[] )
		{
			PlayerPos := FortCharacter.GetTransform().Translation + vector3{X:= 0.0, Y:= 0.0, Z:= 50.0}
			spawn{ SpawnParticleVFX(FX.NS_NightRose_Puppet_Hit_asset, PlayerPos, rotation{}, 1.2) }

			if ( ClosestPlayer := player[Agent] ) # if player
			{
				spawn{ModulePlayerEx.ChairAnim(Self,Agent,AnimPuppetIdle,Duration)}
				FortCharacter.SetVulnerability(true) # prevent PutInStasis invincible bug!
			}
			else
			{
				FortCharacter.PutInStasis(stasis_args{AllowTurning := false, AllowFalling := false, AllowEmotes := true})
				if ( PlayAnimController := FortCharacter.GetPlayAnimationController[] )
				{
					AnimationInstance := PlayAnimController.Play(Animations.Puppeteer_Hanging_Idle)
				}
				Sleep(Duration)
				FortCharacter.ReleaseFromStasis()
			}
		}
	}

    UltTheWorld(Agent:agent, ClassIndex:FNClasses)<suspends>:void=
	{
		if ( FortCharacter := Agent.GetFortCharacter[], PlayerClasses := PlayerClassesDevice? )
		{
			Print("THE WORLD: Activate")

			var VFXType : creative_prop_asset = VFXUltKadoThorne
			case( ClassIndex )
			{
				FNClasses.TheIceKing =>
				{
					spawn{ModulePlayerEx.ChairAnimInvulnerable(Self,Agent,AnimUltIceKing,3.5)}
					SFXUltTheIceKing.Play(Agent)
					VFXPowerUltIceKing.Pickup(Agent)
					set VFXType = VFXUltTheIceKing
				}
				FNClasses.NightRose =>
				{
					SFXUltNightRose.Play(Agent)
					VFXPowerUltNightRose.Pickup(Agent)
					set VFXType = VFXUltNightRose
				}
				_ =>
				{
					SFXUltKadoThorne.Play(Agent)
					VFXPowerUltKadoThorne.Pickup(Agent)
					set VFXType = VFXUltKadoThorne
				}
			}

			Sleep(3.0)

			PlayerPos := FortCharacter.GetTransform().Translation + vector3{ X := 0.0, Y := 0.0, Z := 0.0}
			StartVFX := SpawnProp(VFXType, PlayerPos, rotation{})
			if (StartedVFX:= StartVFX(0)?)
			{
				defer
				{
					StartedVFX.Dispose()
				}
				TeamCollection := GetPlayspace().GetTeamCollection()
				if ( TeamOwner := TeamCollection.GetTeam[Agent] )
				{
					AllVictim := GetAllPlayersInVectorDistance(PlayerPos, 2000.0)
					for ( ClosestVictim : AllVictim, ClosestAgent := ClosestVictim.GetAgent[], ClosestAgent <> Agent and ( not TeamCollection.IsOnTeam[ClosestAgent, TeamOwner] ) )
					{
						Print("THE WORLD: Stop Victim")
						case( ClassIndex )
						{
							FNClasses.NightRose =>
							{
								spawn{ PuppetPlayer( ClosestAgent, 5.0 ) }
							}
							_ =>
							{
								spawn{ FreezePlayer( ClosestAgent, 5.0 ) }
							}
						}
					}
					Sleep(5.0)
				}
			}
			if {set PlayerClasses.PlayersMap[Agent].bUltReady = false}
		}
	}

    DamageBuffPlayer(Agent:agent, Duration:float, Magnitude:float, ?IsUlt:logic=false)<suspends>:void=
	{
		DMGAmplifier.SetDuration(Duration)
		DMGAmplifier.SetMagnitude(Magnitude)
		VFXDMGAmplifier.SetDuration(Duration)
		DMGAmplifier.Spawn()
		DMGAmplifier.Pickup(Agent)
		VFXDMGAmplifier.Pickup(Agent)

		if ( IsUlt?, PlayerClasses := PlayerClassesDevice? )
		{
			Sleep(Duration)
			if {set PlayerClasses.PlayersMap[Agent].bUltReady = false}
		}
	}

    UltSidewaysExpansion(Agent:agent)<suspends>:void=
	{
		if ( FortCharacter := Agent.GetFortCharacter[], PlayerClasses := PlayerClassesDevice? )
		{
			Print("DOMAIN EXPANSION: Activate")
			VFXPowerUltCubeAssassin.Pickup(Agent)
			spawn{ModulePlayerEx.ChairAnimInvulnerable(Self,Agent,AnimUltCaretaker,3.0)}
			Sleep(2.0)

			PlayerPos := FortCharacter.GetTransform().Translation + vector3{ X := 0.0, Y := 0.0, Z := 0.0}
			StartVFX := SpawnProp(VFXUltCubeAssassin, PlayerPos, rotation{})

			if (StartedVFX:= StartVFX(0)?)
			{
				defer
				{
					StartedVFX.Dispose()
					spawn{ SpawnParticleVFX(FX.NS_Ult_CubeAssassin_Destroy_asset, PlayerPos, rotation{}, 1.0) }
				}
				TeamCollection := GetPlayspace().GetTeamCollection()
				if ( TeamOwner := TeamCollection.GetTeam[Agent] )
				{
					AllVictim := GetAllPlayersInVectorDistance(PlayerPos, 2000.0)
					for ( ClosestVictim : AllVictim, ClosestAgent := ClosestVictim.GetAgent[], TeamCollection.IsOnTeam[ClosestAgent, TeamOwner] )
					{
						Print("SIDEWAYS DOMAIN: Buff Teammates")
						if ( ClosestPlayer := player[ClosestAgent] ) # if player
						{
							spawn{DamageBuffPlayer(ClosestAgent, 10.0, 2.0)}
						}
					}

					Sleep(10.0)
				}
			}
			if {set PlayerClasses.PlayersMap[Agent].bUltReady = false}
		}
	}

    CubeInvulnerablePlayer(Agent:agent)<suspends>:void=
	{
		if ( FortCharacter := Agent.GetFortCharacter[], PlayerClasses := PlayerClassesDevice? )
		{
			defer
			{
				FortCharacter.SetVulnerability( true )
				if {set PlayerClasses.PlayersMap[Agent].bUltReady = false}
			}
			FortCharacter.SetVulnerability( false )
			VFXPowerCubeInvulnerable.Pickup(Agent)
			Sleep(10.0)
		}
	}

    UltCubeBlueInvulnerable(Agent:agent)<suspends>:void=
	{
		if ( FortCharacter := Agent.GetFortCharacter[], PlayerClasses := PlayerClassesDevice? )
		{
			Print("UBERCHARGE: Activate")

			Sleep(1.0)

			PlayerPos := FortCharacter.GetTransform().Translation + vector3{ X := 0.0, Y := 0.0, Z := 0.0}
			StartVFX := SpawnProp(VFXUltCubeBlue, PlayerPos, rotation{})

			if (StartedVFX:= StartVFX(0)?)
			{
				defer
				{
					StartedVFX.Dispose()
				}
				TeamCollection := GetPlayspace().GetTeamCollection()
				if ( TeamOwner := TeamCollection.GetTeam[Agent] )
				{
					AllVictim := GetAllPlayersInVectorDistance(PlayerPos, 2000.0)
					for ( ClosestVictim : AllVictim, ClosestAgent := ClosestVictim.GetAgent[], TeamCollection.IsOnTeam[ClosestAgent, TeamOwner] )
					{
						Print("Auf Wiedersehen, Dummkopfs!")
						if ( ClosestPlayer := player[ClosestAgent] ) # if player
						{
							spawn{CubeInvulnerablePlayer(ClosestAgent)}
						}
					}

					Sleep(1.0)
				}
			}
			if {set PlayerClasses.PlayersMap[Agent].bUltReady = false}
		}
	}

	#
	# leviathan event
	# 
	UltMegaCharge(Agent:agent)<suspends>:void=
	{
		if ( FortCharacter := Agent.GetFortCharacter[], PlayerClasses := PlayerClassesDevice? )
		{
			Print("Leviathan: Activated")
			VFXPowerUltMegaloDon.Pickup(Agent)
			FortCharacter.SetVulnerability( false )

			Sleep(2.0)
			FortCharacter.SetVulnerability( true )

			Sleep(1.0)
			CamUltOrbit.AddTo(Agent)
			spawn{ SpawnPropVFXFollowPlayer(VFXUltMegaloDon,FortCharacter,5.0,?IgnoreZ := true) }

			TeamCollection := GetPlayspace().GetTeamCollection()
			if ( TeamOwner := TeamCollection.GetTeam[Agent] )
			{
				var UltStartTime : float = GetSimulationElapsedTime()
				loop
				{
					if ( not FortCharacter.IsActive[] )
					{
						break
					}

					CurrentTime := GetSimulationElapsedTime()
					if ( ElapsedTime := CurrentTime - UltStartTime, ElapsedTime > 5.0)
					{
						break
					}

					if ( FortCharacter.IsActive[] ) { MoveModCharge.Activate(Agent) }
					AllVictim := GetAllPlayersInDistance(FortCharacter, 1280.0)
					for ( ClosestVictim : AllVictim, ClosestAgent := ClosestVictim.GetAgent[], ClosestAgent <> Agent and ( not TeamCollection.IsOnTeam[ClosestAgent, TeamOwner] ) )
					{
						Print("Leviathan: Devour Victim")
						var DmgMultiplier : float = 1.0
						if ( ClosestPlayer := player[ClosestVictim] ) # if player
						{
						}
						else
						{
							set DmgMultiplier = 4.0
						}

						if ( FortCharacter.IsActive[], DamageInstigator := Agent.GetInstigator(), DamageSource := game_action_causer[FortCharacter] )
						{
							PunchDmg : damage_args = damage_args {Instigator := option{DamageInstigator}, Source := option{DamageSource}, Amount:=250.0*DmgMultiplier}
							ClosestVictim.Damage(PunchDmg)
						}
					}

					PlayerNewPos := FortCharacter.GetTransform().Translation
					if ( ExplosiveDeviceUlt.TeleportTo[PlayerNewPos, rotation{}] )
					{
						ExplosiveDeviceUlt.Explode(Agent)
					}

					Sleep(0.1)
				}
			}

			CamUltOrbit.RemoveFrom(Agent)
			if {set PlayerClasses.PlayersMap[Agent].bUltReady = false}
		}
	}

	#
	# doom transform
	# 
	UltDoom(Agent:agent)<suspends>:void=
	{
		if ( FortCharacter := Agent.GetFortCharacter[], PlayerClasses := PlayerClassesDevice? )
		{
			race
			{
				block
				{
					FortCharacter.EliminatedEvent().Await()
					Print("Doom Ult: Failed.")
				}
				block
				{
					Print("Doom Ult: Activated")
					FortCharacter.SetVulnerability(false)

					PlayerPos := FortCharacter.GetTransform().Translation
					PlayerRot := FortCharacter.GetViewRotation()
					if (PlayerRotYaw := PlayerRot.GetYawPitchRollDegrees()[0], PlayerRotRoll := PlayerRot.GetYawPitchRollDegrees()[2])
					{
						AnimRot := MakeRotationFromYawPitchRollDegrees(PlayerRotYaw, 0.0, PlayerRotRoll) 
						NewTransform := transform{Scale := vector3{X:= 0.01, Y:= 0.01, Z:= 0.01}, Rotation := AnimRot, Translation := PlayerPos}
						SpawnVFX := SpawnProp(AnimDoomTransform, NewTransform)
						if (SpawnedVFX:= SpawnVFX(0)?)
						{
							defer
							{
								FortCharacter.ReleaseFromStasis()
								SpawnedVFX.Dispose()
							}
							PropPos := SpawnedVFX.GetTransform().Translation
							TaggedDevice := FindCreativeObjectsWithTag(chairanim_tag{})
							for (FoundDevice: TaggedDevice, ChairDevice := chair_device[FoundDevice])
							{
								DevicePos := FoundDevice.GetTransform().Translation
								if ( IsAlmostEqual[PropPos,DevicePos,0.1] )
								{
									ChairDevice.Seat(Agent)
									FortCharacter.PutInStasis(stasis_args{AllowTurning := false, AllowFalling := false, AllowEmotes := true})
									VFXPowerUltDoomCharge.Pickup(Agent)
									#spawn{AddCamUlt(Agent,1,6.0)}
									NewPos := FortCharacter.GetTransform().Translation + vector3{X:= 0.0, Y:= 0.0, Z:= 12000.0}
									Sleep(0.5)
									if ( ChairDevice.TeleportTo[NewPos, AnimRot] ) {}
									Sleep(6.5)
									ChairDevice.Eject(Agent)
								}
							}

							FortCharacter.SetVulnerability(true)
							PlayerClasses.SetPlayerClass( Agent, FNClasses.DoctorDoomMythic )
							PlayerClasses.ExpendUltimateMeter(Agent)
							PlayerClasses.SetPlayerAggro(Agent, FNBGMState.AlertedPhase2)
							if {set PlayerClasses.PlayersMap[Agent].bAltTransform = true}
							if {set PlayerClasses.PlayersMap[Agent].bUltReady = false}
						}
					}
				}
			}
		}
	}

	#
	# daigo's splinter blade merge
	# 
	UltSplinterBlade(Agent:agent)<suspends>:void=
	{
		if ( FortCharacter := Agent.GetFortCharacter[], PlayerClasses := PlayerClassesDevice? )
		{
			Print("Splinter Blade: Merge")
			if {set PlayerClasses.PlayersMap[Agent].bAltTransform = true}
			if {set PlayerClasses.PlayersMap[Agent].bUltReady = true}
			VFXPowerUltDaigo.Pickup(Agent)
			<#race
			{
				block
				{
					FortCharacter.EliminatedEvent().Await()
					Print("Splinter Blade: Failed.")
				}
				block
				{
					Print("Splinter Blade: Activated")
					spawn{ModulePlayerEx.ChairAnimInvulnerable(Self,Agent,AnimUltDaigo,4.0)}
					if {set PlayerClasses.PlayersMap[Agent].bAltTransform = true}
					if {set PlayerClasses.PlayersMap[Agent].bUltReady = true}
					PlayerPos := FortCharacter.GetTransform().Translation
					spawn{ SpawnParticleVFX(FX.NS_Ult_Daigo_asset, PlayerPos, rotation{}, 5.0) }

					Sleep(2.25)
					VFXPowerUltDaigo.Pickup(Agent)
					Sleep(1.25)
				}
			}#>
		}
	}

	UltSplinterBladeVictim(Agent:agent, AgentVictim:agent)<suspends>:void=
	{
		if ( FortVictim := AgentVictim.GetFortCharacter[], PlayerClasses := PlayerClassesDevice? )
		{
			race
			{
				block
				{
					FortVictim.EliminatedEvent().Await()
					Print("Splinter Blade: Elimination Cancelled.")
				}
				block
				{
					Print("Splinter Blade: Eliminating...")
					FortVictim.SetVulnerability(false)
					#if ( ClosestPlayer := player[AgentVictim] ) # if player
					#{
						spawn{ModulePlayerEx.ChairAnim(Self,AgentVictim,AnimUltDaigoVictim,4.0)}
					<#}
					else
					{
						FortVictim.PutInStasis(stasis_args{AllowTurning := false, AllowFalling := false, AllowEmotes := true})
						if ( PlayAnimController := FortVictim.GetPlayAnimationController[] )
						{
							AnimationInstance := PlayAnimController.Play(Animations.Ult.Ult_Daigo_Victim)
						}
					}#>
					VFXPowerUltDaigoVictim.Pickup(AgentVictim)
					Sleep(1.0)
					PlayerClasses.PlayClassVO(Agent, FNVOType.Interrogation)
					Sleep(3.0)
					FortVictim.SetVulnerability(true)
					if ( FortCharacter := Agent.GetFortCharacter[], FortCharacter.IsActive[], DamageInstigator := Agent.GetInstigator(), DamageSource := game_action_causer[FortCharacter] )
					{
						NukeDmg : damage_args = damage_args {Instigator := option{DamageInstigator}, <#Source := option{DamageSource},#> Amount:=5000.0}
						FortVictim.Damage(NukeDmg)
					}
					if {set PlayerClasses.PlayersMap[Agent].bAltTransform = false}
					if {set PlayerClasses.PlayersMap[Agent].bUltReady = false}
					PlayerClasses.SetPlayerClass( Agent, FNClasses.Daigo )
				}
			}
		}
	}
