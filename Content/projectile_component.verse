<#
	scene graph component
	projectiles
#>
using { /Verse.org }
using { /Verse.org/Assets }
using { /Verse.org/Native }
using { /Verse.org/Simulation }
using { /Verse.org/SpatialMath }
using { /Verse.org/SceneGraph }
using { /Verse.org/SceneGraph/KeyframedMovement }
using { /UnrealEngine.com/Temporary/Diagnostics }
using { /UnrealEngine.com/Temporary/SpatialMath }
using { /Fortnite.com/Game }
using { /Fortnite.com/Teams }
using { /Fortnite.com/Characters }
using { /Fortnite.com/Devices }
using { /Fortnite.com/Playspaces }

# A Verse-authored component that can be added to entities
projectile_component<public> := class<final_super>(component):

	var AgentOwner<public> : ?agent = false

	@editable ExplodeRange<public>:float = 280.0
	@editable ExplodeDamage<public>:float = 10.0

	@editable ExplodeVFX<public>:particle_system = FX.NS_Doom_Impact_asset

	@editable Speed<public>:float = 6000.0
	@editable LifeTime<public>:float = 2.0	# 1.5
	var MoveDistance : float = 250.0
	var MoveGravity : float = 0.0

	#@editable ExplodeVFX<public>:particle_system_component = particle_system_component{Entity := Entity, AutoPlay := true}

    # Runs when the component should start simulating in a running game.
    OnBeginSimulation<override>():void =
	{
		(super:)OnBeginSimulation()
		#Print("Initialize Projectile Component...")

		if ( RoundM := Entity.GetFortRoundManager[] )
		{
			RoundM.SubscribeRoundStarted(OnRoundStarted)
			RoundM.SubscribeRoundEnded(OnRoundEnded)
		}
	}

    # Runs when the component should start simulating in a running game.
    # Can be suspended throughout the lifetime of the component. Suspensions
    # will be automatically cancelled when the component is disposed or the
    # game ends.
    OnSimulate<override>()<suspends>:void =
    {
        if ( Mesh := Entity.GetComponent[mesh_component] )
		{
            Mesh.EntityEnteredEvent.Subscribe(OnEntityEntered)
		}
	}

	OnRoundStarted():void =
	{
	}

	OnRoundEnded():void =
	{
		Entity.RemoveFromParent()
	}

    GetAgentOwner()<decides>:agent=
	{
		AgentOwn := agent[AgentOwner?]
	}

    OnEntityEntered(OtherEntity:entity):void =
	{
		if ( Mesh := OtherEntity.GetComponent[mesh_component] )
		{
			if ( InvisMesh := OtherEntity.GetComponent[Meshes.PlayerBox] )
			{
				Print("Projectile:OnEntityEntered = Player")
				spawn{Explode()}
			}

			if ( InvisMesh := OtherEntity.GetComponent[Meshes.InvisibleWall_Cube] )
			{
				Print("Projectile:OnEntityEntered = Invisible Wall")
				spawn{Explode()}
			}
		}
	}

    FireProjectile(Agent:agent):void=
	{
		set AgentOwner = option{Agent}

        if ( Mesh := Entity.GetComponent[mesh_component] )
		{
            Mesh.EntityEnteredEvent.Subscribe(OnEntityEntered)
		}

		spawn{ProjectileThink()}
	}

    ProjectileThink()<suspends>:void=
	{
		# simulate fake physics like falling n stuff.

		spawn{ProjectileLifeTime()}
		race
		{
			block
			{
				if ( Mesh := Entity.GetComponent[mesh_component] )
				{
					Mesh.EntityEnteredEvent.Await()
				}
			}
			block
			{
				loop
				{
					if ( not IsSimulating[] )
					{
						#Print("Projectile stopped simulating...")
						break
					}

					spawn{ProjectileMove()}
					#CheckCollision()
					Duration : float = LifeTime # MoveDistance / Speed
					Sleep(0.1)
				}
			}
		}
	}

    ProjectileLifeTime()<suspends>:void=
	{
		Sleep(LifeTime)
		if ( IsSimulating[], IsInScene[] )
		{
			#Print("Projectile:ProjectileLifeTime()")
			Explode()
		}
	}

    ProjectileMove()<suspends>:void=
	{
		if (MoveComp := Entity.GetComponent[keyframed_movement_component])
		{
			set MoveGravity -= 6.0

			Duration : float = LifeTime # MoveDistance / Speed
			CurrentTrans := Entity.GetGlobalTransform().Translation
			DesTrans := FromVector3((FromTransform(Entity.GetGlobalTransform()).Rotation.GetLocalForward()) * MoveDistance) + (/Verse.org/SpatialMath:)vector3{Forward  := 0.0, Left := 0.0, Up := MoveGravity}

			NewKeyTransform := (/Verse.org/SpatialMath:)transform{ Translation := DesTrans, Scale := (/Verse.org/SpatialMath:)vector3{Forward  := 0.0, Left := 0.0, Up := 0.0}}

			AnimKeyframes := array
			{
				keyframed_movement_delta
				{
					Transform := NewKeyTransform
					Duration := 0.1
					Easing := linear_easing_function{}
				}
			}
			MoveComp.Stop()
			MoveComp.SetKeyframes(AnimKeyframes, oneshot_keyframed_movement_playback_mode{})
			MoveComp.Play()
			#Print("ProjectileMove()")
		}
	}

    CheckCollision():void=
	{
		CollisionSphere := collision_sphere{ Radius := 25.0 <#, CollisionProfile := collision_profiles.StationaryOverlapAll#> }
		CollisionTransform := Entity.GetGlobalTransform()
		for
		{
			Overlap : Entity.FindOverlapHits(CollisionTransform, CollisionSphere)
			Mesh := mesh_component[Overlap.TargetComponent]
		}
		do
		{
			if ( PlayerMesh := Meshes.PlayerBox[Overlap.TargetComponent] )
			{
				#Print("Projectile:CheckCollision = player")
				spawn{Explode()}
				return
			}

			if ( WallMesh := Meshes.InvisibleWall_Cube[Overlap.TargetComponent] )
			{
				#Print("Projectile:CheckCollision() = walls")
				spawn{Explode()}
				return
			}
		}
	}

    Explode()<suspends>:void=
	{
		if ( AgentOwn := agent[AgentOwner?], FortCharacter := AgentOwn.GetFortCharacter[], MoveComp := Entity.GetComponent[keyframed_movement_component] )
		{
			#Print("Projectile:Explode()")
			MoveComp.Stop()
			spawn{ImpactVFX()}

			var ClosestVictims: []fort_character = array{}

			CollisionSphere := collision_sphere{ Radius := ExplodeRange <#, CollisionProfile := collision_profiles.StationaryOverlapAll#> }
			CollisionTransform := Entity.GetGlobalTransform()
			for
			{
				Overlap : Entity.FindOverlapHits(CollisionTransform, CollisionSphere)
				PlayerMesh := Meshes.PlayerBox[Overlap.TargetComponent]
				TargetEntity := PlayerMesh.Entity
			}
			do
			{
				if ( PlayerComp := TargetEntity.GetComponent[fn_player_component], Agent := PlayerComp.Agent, Agent <> AgentOwn, FortChar := Agent.GetFortCharacter[] )
				{
					#set ClosestVictims += array{FortChar}
					DamageAgent(Agent)
				}
				else if ( NPCComp := TargetEntity.GetComponent[fn_npc_component], AgentNPC := NPCComp.Agent, FortNPC := AgentNPC.GetFortCharacter[] )
				{
					#set ClosestVictims += array{FortNPC}
					DamageAgent(AgentNPC)
				}
			}

			#for ( VictimChar : ClosestVictims, FortCharacter.IsActive[], VictimChar.IsActive[], DamageInstigator := AgentOwn.GetInstigator(), DamageSource := game_action_causer[FortCharacter] )
			#{
			#	ProjDmg : damage_args = damage_args{Instigator := option{DamageInstigator}, Source := option{DamageSource}, Amount:=ExplodeDamage}
			#	VictimChar.Damage(ProjDmg)
			#}

			#Drawer := debug_draw{}
			#Drawer.DrawSphere(Entity.GetGlobalTransform().Translation, ?Radius := ExplodeRange, ?DrawDurationPolicy := debug_draw_duration_policy.FiniteDuration, ?Duration:= float(1.5))
		}
		Entity.RemoveFromParent()
	}

    DamageAgent(AgentVictim:agent)<suspends>:void =
	{
		if
		{
			Agent := agent[AgentOwner?]
			FortChar := Agent.GetFortCharacter[]
			FortVictim := AgentVictim.GetFortCharacter[]
			FortChar.IsActive[]
			FortVictim.IsActive[]
			Playspace := Entity.GetPlayspaceForEntity[]
			TeamCollection := Playspace.GetTeamCollection()
			TeamOwner := TeamCollection.GetTeam[Agent]
			not TeamCollection.IsOnTeam[AgentVictim, TeamOwner]
			DamageInstigator := Agent.GetInstigator()
			DamageSource := game_action_causer[FortChar]
		}
		then
		{
			ProjDmg := damage_args{Instigator := option{DamageInstigator}, Source := option{DamageSource}, Amount:=ExplodeDamage}
			FortVictim.Damage(ProjDmg)
		}
	}

    ImpactVFX()<suspends>:void=
	{
		<#if ( SimulationEntity := Entity.GetSimulationEntity[] )
		{
			SpawnPos := (/Verse.org/SpatialMath:)transform{Translation:=Entity.GetGlobalTransform().Translation}
			NE := entity{}
			TransComp := transform_component{ Entity := NE }
			VFXComp := FX.NS_Doom_Impact{ Entity := NE, AutoPlay := true }
			NE.AddComponents(array{TransComp,VFXComp})
			SimulationEntity.AddEntities(array{NE})
			NE.SetGlobalTransform(SpawnPos)
			Sleep(1.0)
			NE.RemoveFromParent()
		}#>

		SpawnPos := FromVector3(Entity.GetGlobalTransform().Translation)
		spawn{ ModulePropExtra.SpawnParticleVFX(ExplodeVFX, SpawnPos, (/UnrealEngine.com/Temporary/SpatialMath:)rotation{}, 1.2) }
	}
