<#
	Left 4 Elim
	Persistent Save Manager
#>
using { /Fortnite.com/Devices }
using { /Verse.org/Simulation }
using { /Verse.org/Random }
using { /UnrealEngine.com/Temporary/Diagnostics }
using { ModuleAchievements }
using { ModuleAchievements }

player_profile_data := class<final><persistable>:
	Version : int = 0
	Level : int = 1
	XP : int = 0
	Achievements : []FNAchievements = array{} # DEPRECATED
	AchieveIDs : []int = array{}
	SettingBGM : logic = true
	SettingVO : logic = false
	SettingMouse : logic = false
	SettingHat : logic = true
	SettingCamFP : logic = false

MakePlayerProfileData<constructor>(Src:player_profile_data)<transacts> := player_profile_data:
	Level := Src.Level
	XP := Src.XP
	Achievements := Src.Achievements # DEPRECATED
	AchieveIDs := Src.AchieveIDs
	SettingBGM := Src.SettingBGM
	SettingVO := Src.SettingVO
	SettingMouse := Src.SettingMouse
	SettingHat := Src.SettingHat
	SettingCamFP := Src.SettingCamFP

var PlayerProfileDataMap:weak_map(player, player_profile_data) = map{}

# A Verse-authored creative device that can be placed in a level
save_manager_device := class(creative_device):

	@editable HudMsg_LevelUp : hud_message_device = hud_message_device{}
	@editable MSGFeed_LevelUp 	: elimination_feed_device = elimination_feed_device{}
	@editable Analytic_LevelUp : analytics_device = analytics_device{}
	@editable StatCreator_LevelUp : stat_creator_device = stat_creator_device{}
	UITextMessage<localizes>(Value : string) : message = "{Value}"

	MaxXP : int = 100

    # Runs when the device is started in a running game
    OnBegin<override>()<suspends>:void=
	{
		Print("Initialize Save Manager...")
	}

    InitializePlayerSaves(Player : player) : void =
	{
		if ( Player.IsActive[] )
		{
			if ( PlayerProfileDataMap[Player] )
			{
				Print("Saves already exist for this player!")
				<#if ( SaveData := PlayerProfileDataMap[Player] )
				{
					DeleteEnumAchievement( SaveData )
				}#>
			}
			else
			{
				if ( set PlayerProfileDataMap[Player] = player_profile_data{} )
				{
					Print("created new saves!")
				}
			}
		}
	}

	CheckSaveDataForPlayer(Agent:agent)<decides><transacts>:void=
	{
		if(Player := player[Agent], Player.IsActive[])
		{
			if(PlayerProfileDataMap[Player])
			{
				# no-op
			}
			else
			{
				set PlayerProfileDataMap[Player] = player_profile_data{}
			}
		}
	}

	GrantXP(Agent:agent, GrantedXP:int):void=
	{
		if
		{
			CheckSaveDataForPlayer[Agent]
			Player := player[Agent]
			Player.IsActive[]
			SourceData := PlayerProfileDataMap[Player]
	 
			set PlayerProfileDataMap[Player] = player_profile_data:
				MakePlayerProfileData<constructor>(SourceData)
				XP := SourceData.XP + GrantedXP
		}
		then
		{
			ShowPlayerXP(Agent)

			if ( UpdatedSaveData := PlayerProfileDataMap[Player], NewXP := UpdatedSaveData.XP, NewXP >= MaxXP )
			{
				GrantLevelUp(Agent)
			}
		}
	}

	GrantLevelUp(Agent:agent):void=
	{
		if
		{
			CheckSaveDataForPlayer[Agent]
			Player := player[Agent]
			Player.IsActive[]
			SaveData := PlayerProfileDataMap[Player]
			CurLevel := SaveData.Level
			CurLevel < 150

			set PlayerProfileDataMap[Player] = player_profile_data:
				MakePlayerProfileData<constructor>(SaveData)
				Level := SaveData.Level + 1
				XP := 0
		}
		then
		{
			ShowPlayerLevel(Agent)
		}
	}

	GetPlayerLevel(Agent:agent):int=
	{
		if
		{
			CheckSaveDataForPlayer[Agent]
			Player := player[Agent]
			Player.IsActive[]
			SaveData := PlayerProfileDataMap[Player]
		}
		then
		{
			return SaveData.Level
		}

		return 0
	}

	ShowPlayerLevel(Agent:agent):void=
	{
		if
		{
			CheckSaveDataForPlayer[Agent]
			Player := player[Agent]
			Player.IsActive[]
			SaveData := PlayerProfileDataMap[Player]
		}
		then
		{
			Print("Player's Level: {SaveData.Level}")
			HudMsg_LevelUp.Show(Agent, UITextMessage("{SaveData.Level}"), ?DisplayTime := 3.0)
			MSGFeed_LevelUp.Activate(Agent)
			Analytic_LevelUp.Submit(Agent)
			UpdatePlayerLevel(Agent)
		}
	}

	UpdatePlayerLevel(Agent:agent):void=
	{
		if
		{
			CheckSaveDataForPlayer[Agent]
			Player := player[Agent]
			Player.IsActive[]
			SaveData := PlayerProfileDataMap[Player]
		}
		then
		{
			if { StatCreator_LevelUp.SetValue[Agent, SaveData.Level] }
		}
	}

	GetPlayerXP(Agent:agent):int=
	{
		if
		{
			CheckSaveDataForPlayer[Agent]
			Player := player[Agent]
			Player.IsActive[]
			SaveData := PlayerProfileDataMap[Player]
		}
		then
		{
			return SaveData.XP
		}

		return 0
	}

	ShowPlayerXP(Agent:agent):void=
	{
		if
		{
			CheckSaveDataForPlayer[Agent]
			Player := player[Agent]
			Player.IsActive[]
			SaveData := PlayerProfileDataMap[Player]
		}
		then
		{
			Print("Player's XP: {SaveData.XP}")
		}
	}

	DeleteEnumAchievement(Src:player_profile_data)<transacts>:player_profile_data =
	{
		NewData:player_profile_data = player_profile_data:
			MakePlayerProfileData<constructor>(Src)
			Achievements := array{}
	}

	SetAchievementHistory(Src:player_profile_data, NewAchievements:[]int)<transacts>:player_profile_data =
	{
		NewData:player_profile_data = player_profile_data:
			MakePlayerProfileData<constructor>(Src)
			AchieveIDs := NewAchievements
	}

	RecordAchievementsHistory(Agent:agent, NewAchievement:int)<decides><transacts>:void=
	{
		var bFailure:logic = false
		if
		{
			CheckSaveDataForPlayer[Agent]
			Player := player[Agent]
			Player.IsActive[]
			SaveData := PlayerProfileDataMap[Player]
		}
		then
		{
			# check unlocked?
			if ( SaveData.AchieveIDs.Find[NewAchievement] )
			{
				Print("Achievement completed.")
				#ShowAchievementsHistory(Agent)
				bFailure?
			}

			if
			{
				var UpdatedSaveData:player_profile_data = SaveData
				set UpdatedSaveData = SetAchievementHistory(UpdatedSaveData, array{NewAchievement} + SaveData.AchieveIDs)
				FitsInPlayerMap[UpdatedSaveData]
				set PlayerProfileDataMap[Player] = player_profile_data:
					MakePlayerProfileData<constructor>(SaveData)
					AchieveIDs := UpdatedSaveData.AchieveIDs
			}
			else # the new data is too big to fit in the player map, update the data until it now fits.
			{
				var TestUpdatedSaveData:player_profile_data = SaveData
				set TestUpdatedSaveData = SetAchievementHistory(TestUpdatedSaveData, array{})
				var UpdatedSaveData:player_profile_data = TestUpdatedSaveData

				ProposedAchievementHistory:[]int = array{NewAchievement} + SaveData.AchieveIDs
				var DataFits:logic = true
				for (CurAchievementHistory : ProposedAchievementHistory, DataFits?)
				{
					set TestUpdatedSaveData = SetAchievementHistory(TestUpdatedSaveData, array{CurAchievementHistory})
					if (FitsInPlayerMap[TestUpdatedSaveData])
					{
						set UpdatedSaveData = SetAchievementHistory(UpdatedSaveData, TestUpdatedSaveData.AchieveIDs)
					}

					if
					{
						SourceData := PlayerProfileDataMap[Player]
						set PlayerProfileDataMap[Player] = player_profile_data:
							MakePlayerProfileData<constructor>(SourceData)
							AchieveIDs := UpdatedSaveData.AchieveIDs
					}
					set DataFits = false
				}
			}

			set bFailure = true
			bFailure?
		}
	}

	GetAchievementsHistory(Agent:agent):[]int=
	{
		if
		{
			CheckSaveDataForPlayer[Agent]
			Player := player[Agent]
			Player.IsActive[]
			SaveData := PlayerProfileDataMap[Player]
		}
		then
		{
			return SaveData.AchieveIDs
		}
		
		return array{}
	}

	ToggleSettingBGM(Agent:agent, NewSetting:logic):void=
	{
		if
		{
			CheckSaveDataForPlayer[Agent]
			Player := player[Agent]
			Player.IsActive[]
			SourceData := PlayerProfileDataMap[Player]
	 
			set PlayerProfileDataMap[Player] = player_profile_data:
				MakePlayerProfileData<constructor>(SourceData)
				SettingBGM := NewSetting
		}
	}

    IsBGMEnabled(Agent:agent)<transacts><decides>:void =
	{
		#if ( CheckSaveDataForPlayer[Agent], Player := player[Agent], Player.IsActive[], SaveData := PlayerProfileDataMap[Player] )
		if ( Player := player[Agent], Player.IsActive[], CheckSaveDataForPlayer[Agent], SaveData := PlayerProfileDataMap[Player] )
		{
			SaveData.SettingBGM?
		}
	}

	ToggleSettingVOMenu(Agent:agent, NewSetting:logic):void=
	{
		if
		{
			CheckSaveDataForPlayer[Agent]
			Player := player[Agent]
			Player.IsActive[]
			SourceData := PlayerProfileDataMap[Player]
	 
			set PlayerProfileDataMap[Player] = player_profile_data:
				MakePlayerProfileData<constructor>(SourceData)
				SettingVO := NewSetting
		}
	}

    IsVOMenuEnabled(Agent:agent)<transacts><decides>:void =
	{
		if ( CheckSaveDataForPlayer[Agent], Player := player[Agent], Player.IsActive[], SaveData := PlayerProfileDataMap[Player] )
		{
			SaveData.SettingVO?
		}
	}

	ToggleSettingMouse(Agent:agent, NewSetting:logic):void=
	{
		if
		{
			CheckSaveDataForPlayer[Agent]
			Player := player[Agent]
			Player.IsActive[]
			SourceData := PlayerProfileDataMap[Player]
	 
			set PlayerProfileDataMap[Player] = player_profile_data:
				MakePlayerProfileData<constructor>(SourceData)
				SettingMouse := NewSetting
		}
	}

    IsMouseEnabled(Agent:agent)<transacts><decides>:void =
	{
		if ( CheckSaveDataForPlayer[Agent], Player := player[Agent], Player.IsActive[], SaveData := PlayerProfileDataMap[Player] )
		{
			SaveData.SettingMouse?
		}
	}

	ToggleSettingHat(Agent:agent, NewSetting:logic):void=
	{
		if
		{
			CheckSaveDataForPlayer[Agent]
			Player := player[Agent]
			Player.IsActive[]
			SourceData := PlayerProfileDataMap[Player]
	 
			set PlayerProfileDataMap[Player] = player_profile_data:
				MakePlayerProfileData<constructor>(SourceData)
				SettingHat := NewSetting
		}
	}

    IsHatEnabled(Agent:agent)<transacts><decides>:void =
	{
		if ( Player := player[Agent], Player.IsActive[], CheckSaveDataForPlayer[Agent], SaveData := PlayerProfileDataMap[Player] )
		{
			SaveData.SettingHat?
		}
	}

	ToggleSettingCamFP(Agent:agent, NewSetting:logic):void=
	{
		if
		{
			CheckSaveDataForPlayer[Agent]
			Player := player[Agent]
			Player.IsActive[]
			SourceData := PlayerProfileDataMap[Player]
	 
			set PlayerProfileDataMap[Player] = player_profile_data:
				MakePlayerProfileData<constructor>(SourceData)
				SettingCamFP := NewSetting
		}
	}

    IsCamFPEnabled(Agent:agent)<transacts><decides>:void =
	{
		if ( Player := player[Agent], Player.IsActive[], CheckSaveDataForPlayer[Agent], SaveData := PlayerProfileDataMap[Player] )
		{
			SaveData.SettingCamFP?
		}
	}
