using { /Fortnite.com/Game }
using { /Fortnite.com/Devices }
using { /Fortnite.com/Devices/CreativeAnimation }
using { /Fortnite.com/Devices/CreativeAnimation/InterpolationTypes }
using { /Verse.org/Simulation }
using { /UnrealEngine.com/Temporary/Diagnostics }
using { /Fortnite.com/Characters }
using { /UnrealEngine.com/Temporary/SpatialMath }
using { ModuleAchievements }

# A Verse-authored creative device that can be placed in a level
remote_signal_doot_device := class(creative_device):

	@editable UIAchievementDevice : ?ui_achievements_device = false

	@editable SignalRemoteLegendary : signal_remote_manager_device = signal_remote_manager_device{}
	@editable DOOTMesh : creative_prop = creative_prop{}
	#@editable VFXFire : vfx_spawner_device = vfx_spawner_device{}
	#@editable VFXCharge : vfx_spawner_device = vfx_spawner_device{}
	@editable AudioFire : audio_player_device = audio_player_device{}
	@editable AudioMusic : audio_player_device = audio_player_device{}
	@editable DamageDevice : damage_volume_device = damage_volume_device{}
	@editable CinematicSequence : cinematic_sequence_device = cinematic_sequence_device{}
	@editable AudioHit : audio_player_device = audio_player_device{}
	@editable ItemRemover : item_remover_device = item_remover_device{}
	var HomePos : vector3 = vector3{}
	var HomeRot : rotation = rotation{}

    @editable VolumeZP : mutator_zone_device  = mutator_zone_device{}
	@editable CineZP : cinematic_sequence_device = cinematic_sequence_device{}
    #@editable AmbZP : audio_player_device = audio_player_device{}

    @editable VolumeTheDevice : mutator_zone_device  = mutator_zone_device{}
	@editable CineTheDevice : cinematic_sequence_device = cinematic_sequence_device{}

    var DootOwner : ?player = false
	var bFired : logic = false

    @editable XP_Accolade : accolades_device = accolades_device{}

    # Runs when the device is started in a running game
    OnBegin<override>()<suspends>:void=
	{
        SignalRemoteLegendary.PrimarySignalEvent.Subscribe(HandlePrimaryButton)
        SignalRemoteLegendary.SecondarySignalEvent.Subscribe(HandleSecondaryButton)
        DamageDevice.AgentEntersEvent.Subscribe(HandleDamageEnters)
        DamageDevice.AgentExitsEvent.Subscribe(HandleDamageExits)
        CineZP.StoppedEvent.Subscribe(OnCineStoppedZP)

		set HomePos = DOOTMesh.GetTransform().Translation
		set HomeRot = DOOTMesh.GetTransform().Rotation
	}

    AimDOOTDevice()<suspends>:void=
	{
		if (Agent := agent[DootOwner?], FortCharacter := Agent.GetFortCharacter[] )
		{
			PlayerPos := FortCharacter.GetTransform().Translation + FortCharacter.GetViewRotation().GetLocalForward() * 450.0
			PlayerRot := FortCharacter.GetViewRotation().ApplyYaw(5.0)
			if ( DOOTMesh.TeleportTo[PlayerPos, PlayerRot]) {}

			loop
			{
				Sleep(0.25)

				if (bFired?)
				{
					break
				}

				if ( not FortCharacter.IsActive[] )
				{
					break
				}

				PlayerNewPos := FortCharacter.GetTransform().Translation + FortCharacter.GetViewRotation().GetLocalForward() * 450.0
				NewRotation := FortCharacter.GetViewRotation().ApplyYaw(5.0)
				if (AController := DOOTMesh.GetAnimationController[])
				{
					AnimKeyframes := array
					{
						keyframe_delta
						{
							DeltaLocation := PlayerNewPos - DOOTMesh.GetTransform().Translation
							DeltaRotation := MakeComponentWiseDeltaRotation(NewRotation, DOOTMesh.GetTransform().Rotation)
							Time := 0.25
							Interpolation := Linear
						}
					}
					AController.SetAnimation(AnimKeyframes, ?Mode:=animation_mode.OneShot)
					AController.Play()
				}
			}
		}
	}

    UseDOOTDevice()<suspends>:void=
	{
		if (Agent := agent[DootOwner?], FortCharacter := Agent.GetFortCharacter[] )
		{
			race
			{
				block
				{
					set bFired = false

					DOOTMesh.Show()
					#VFXCharge.Enable()
					CinematicSequence.Play()

					spawn{ AimDOOTDevice() }

					XP_Accolade.Award(Agent)
					if ( UIAchievement := UIAchievementDevice? )
					{
						spawn{UIAchievement.AchievementUnlock(Agent, 1)}
					}

					AudioMusic.Play()
					AudioFire.Play()

					Sleep(5.0)

					#VFXCharge.Disable()
					#VFXFire.Enable()
					#DamageDevice.UpdateSelectedTeam(Agent)
					DamageDevice.Enable()
					#set bFired = true

					<#EDistance := Distance(FortCharacter.GetTransform().Translation, PurpleCube.GetTransform().Translation)
					if (EDistance < 3000.0)
					{
						Print("DOOT Distance to Kevin: {EDistance}")
						Print("Event: DOOT vs Kevin")
						CinePurpleCube.Play()
						if ( UIAchievement := UIAchievementDevice? )
						{
							spawn{UIAchievement.AchievementUnlock(Agent, 2)}
						}
						set bFired = true
					}#>

					if ( VolumeZP.IsInVolume[Agent] )
					{
						Print("Event: DOOT vs ZP")
						VolumeZP.Disable()
						CineZP.Play()
						AudioHit.Play(Agent)
						if ( UIAchievement := UIAchievementDevice? )
						{
							spawn{UIAchievement.AchievementUnlock(Agent, 3)}
						}
						set bFired = true
					}

					if ( VolumeTheDevice.IsInVolume[Agent] )
					{
						Print("Event: DOOT vs ZP")
						VolumeTheDevice.Disable()
						CineTheDevice.Play()
						AudioHit.Play(Agent)
						if ( UIAchievement := UIAchievementDevice? )
						{
							spawn{UIAchievement.AchievementUnlock(Agent, 4)}
						}
						set bFired = true
					}

					Sleep(10.0)
				}
				block
				{
					FortCharacter.EliminatedEvent().Await()
				}
			}
			spawn{StopDOOTDevice()}
		}
	}

    StopDOOTDevice()<suspends>:void=
	{
		if (Agent := agent[DootOwner?], FortCharacter := Agent.GetFortCharacter[] )
		{
			set bFired = true
			DamageDevice.Disable()
			CinematicSequence.Stop()
			AudioFire.Stop()
			AudioMusic.Stop()
			DOOTMesh.Hide()
			if ( DOOTMesh.TeleportTo[HomePos, HomeRot]) {}
		}
	}

    HurtArea(Agent:agent)<suspends>:void=
	{
		if (FortCharacter := Agent.GetFortCharacter[], DootAgent := agent[DootOwner?] )
		{
			loop
			{
				if ( DootFort := DootAgent.GetFortCharacter[], DootFort.IsActive[], DootInstigator := DootAgent.GetInstigator(), DootSource := game_action_causer[DootOwner?] )
				{
					DootDmg : damage_args = damage_args {Instigator := option{DootInstigator}, Source := option{DootSource}, Amount:=400.0}
					FortCharacter.Damage(DootDmg)
				}

				if ( not DamageDevice.IsInVolume[Agent] )
				{
					break
				}

				Sleep(0.25)
			}
		}
	}

    HandlePrimaryButton(Agent : agent) : void =
	{
		if (InPlayer := player[Agent])
		{
			set DootOwner = option{InPlayer}

			#SignalRemoteLegendary.Disable()
			ItemRemover.Remove(Agent)
			spawn{UseDOOTDevice()}
		}

	}

    HandleSecondaryButton(Agent : agent) : void =
	{
		HandlePrimaryButton(Agent)
	}

    HandleDamageEnters(Agent : agent) : void =
	{
		if (InPlayer := player[Agent])
		{
			AudioHit.Play(Agent)
			spawn{HurtArea(Agent)}
		}

	}

    HandleDamageExits(Agent : agent) : void =
	{
		if (InPlayer := player[Agent])
		{
		
		}
	}

    OnCineStoppedZP() : void =
	{
		#AmbZP.Play()
	}