using { /Fortnite.com/Devices }
using { /Fortnite.com/Characters }
using { /Fortnite.com/UI }
using { /Verse.org/Simulation }
using { /UnrealEngine.com/Temporary/Diagnostics }
using { /Verse.org/Colors }
using { /UnrealEngine.com/Temporary/UI }
using { /UnrealEngine.com/Temporary/SpatialMath }

# A Verse-authored creative device that can be placed in a level
remote_signal_vo_device := class(creative_device):

	@editable SignalRemoteVoice : signal_remote_manager_device = signal_remote_manager_device{}
	@editable VOMidasDBNO : audio_player_device = audio_player_device{}
	@editable VOMidasFullAlert : audio_player_device = audio_player_device{}
	@editable VOMidasFullAlertLostTarget : audio_player_device = audio_player_device{}
	@editable VOMidasIdle : audio_player_device = audio_player_device{}
	@editable VOMidasInterrogation : audio_player_device = audio_player_device{}
	@editable VOMidasSus : audio_player_device = audio_player_device{}
	@editable VOMidasSusToIdle : audio_player_device = audio_player_device{}

	ColorBG: color_block = color_block {DefaultColor := NamedColors.Black, DefaultOpacity := 0.75}
    #ExitButton : button_quiet = button_quiet{}
    VO1Button : button_quiet = button_quiet{}
    VO2Button : button_quiet = button_quiet{}
    VO3Button : button_quiet = button_quiet{}
    VO4Button : button_quiet = button_quiet{}
    VO5Button : button_quiet = button_quiet{}
    VO6Button : button_quiet = button_quiet{}
    VO7Button : button_quiet = button_quiet{}
    VO8Button : button_quiet = button_quiet{}

    UITextMessage<localizes>(Value : string) : message = "{Value}"

	TextVoiceMenu<localizes> : message = "Voice Menu"

	var VoiceMenuCanvas : [player]?canvas = map{}

	CreateVoiceMenuUI() : canvas =
	{
		UICanvas : canvas = canvas{}

        #ExitButton.SetText(UITextMessage("Exit"))
        #UICanvas.AddWidget(canvas_slot{
        #    Anchors := anchors{ Minimum:= vector2{ X := 0.11, Y := 0.4}, Maximum := vector2 { X := 0.11, Y :=  0.4 }}
        #    Widget := ExitButton
        #})

        UICanvas.AddWidget(canvas_slot{
			ZOrder := {0}
			Anchors := anchors{Minimum := vector2{X := 0.5, Y := 0.5}, Maximum := vector2{X := 0.5, Y := 0.5}}
			Offsets := margin{ Top:=0.0, Left:= 0.0, Right:=0.0, Bottom := 0.0 }
			Alignment := vector2{X := 0.5, Y := 0.5}
			SizeToContent := true
			Widget := ColorBG
        })

        UICanvas.AddWidget(canvas_slot{
			ZOrder := {1}
			Anchors := anchors{Minimum := vector2{X := 0.5, Y := 0.5}, Maximum := vector2{X := 0.5, Y := 0.5}}
			Offsets := margin{ Top:=0.0, Left:= 0.0, Right:=0.0, Bottom := 0.0 }
			Alignment := vector2{X := 0.5, Y := 0.5}
			SizeToContent := true
			Widget := stack_box{
				Orientation := orientation.Vertical
				Slots := array{
					stack_box_slot{
						Widget := VO1Button
					}
					stack_box_slot{
						Widget := VO2Button
						}
					stack_box_slot{
						Widget := VO3Button
						}
					stack_box_slot{
						Widget := VO4Button
						}
					stack_box_slot{
						Widget := VO5Button
						}
					stack_box_slot{
						Widget := VO6Button
						}
					stack_box_slot{
						Widget := VO7Button
						}
					stack_box_slot{
						Widget := VO8Button
						}
					}
				}
        })

        VO1Button.SetText(UITextMessage("Ouch!")) 
        VO2Button.SetText(UITextMessage("Alert"))
        VO3Button.SetText(UITextMessage("Where?"))
        VO4Button.SetText(UITextMessage("Idle"))
        VO5Button.SetText(UITextMessage("Interrogation"))
        VO6Button.SetText(UITextMessage("Sus"))
        VO7Button.SetText(UITextMessage("Not Sus"))
        VO8Button.SetText(UITextMessage("..."))

		#ExitButton.OnClick().Subscribe(HandleSelectedExitButton)

		VO1Button.OnClick().Subscribe(OnSelectedVO1Button)
		VO2Button.OnClick().Subscribe(OnSelectedVO2Button)
		VO3Button.OnClick().Subscribe(OnSelectedVO3Button)
		VO4Button.OnClick().Subscribe(OnSelectedVO4Button)
		VO5Button.OnClick().Subscribe(OnSelectedVO5Button)
		VO6Button.OnClick().Subscribe(OnSelectedVO6Button)
		VO7Button.OnClick().Subscribe(OnSelectedVO7Button)
		VO8Button.OnClick().Subscribe(OnSelectedVO8Button)

		return UICanvas
	}

	<#HandleSelectedExitButton(Message : widget_message) : void =
	{
		if (PlayerUI := GetPlayerUI[Message.Player], MyUI := VoiceMenuCanvas[Message.Player]?, SelectedButton := text_button_base[Message.Source])
		{
			ExitUI(Message.Player)
		}
	}#>

	OnSelectedVO1Button(Message : widget_message) : void =
	{
		if (PlayerUI := GetPlayerUI[Message.Player], MyUI := VoiceMenuCanvas[Message.Player]?, SelectedButton := text_button_base[Message.Source], FortCharacter := Message.Player.GetFortCharacter[])
		{
			VOMidasDBNO.Play( Message.Player )

			ExitUI(Message.Player)
		}
	}

	OnSelectedVO2Button(Message : widget_message) : void =
	{
		if (PlayerUI := GetPlayerUI[Message.Player], MyUI := VoiceMenuCanvas[Message.Player]?, SelectedButton := text_button_base[Message.Source], FortCharacter := Message.Player.GetFortCharacter[])
		{
			VOMidasFullAlert.Play( Message.Player )

			ExitUI(Message.Player)
		}
	}

	OnSelectedVO3Button(Message : widget_message) : void =
	{
		if (PlayerUI := GetPlayerUI[Message.Player], MyUI := VoiceMenuCanvas[Message.Player]?, SelectedButton := text_button_base[Message.Source], FortCharacter := Message.Player.GetFortCharacter[])
		{
			VOMidasFullAlertLostTarget.Play( Message.Player )

			ExitUI(Message.Player)
		}
	}

	OnSelectedVO4Button(Message : widget_message) : void =
	{
		if (PlayerUI := GetPlayerUI[Message.Player], MyUI := VoiceMenuCanvas[Message.Player]?, SelectedButton := text_button_base[Message.Source], FortCharacter := Message.Player.GetFortCharacter[])
		{
			VOMidasIdle.Play( Message.Player )

			ExitUI(Message.Player)
		}
	}

	OnSelectedVO5Button(Message : widget_message) : void =
	{
		if (PlayerUI := GetPlayerUI[Message.Player], MyUI := VoiceMenuCanvas[Message.Player]?, SelectedButton := text_button_base[Message.Source], FortCharacter := Message.Player.GetFortCharacter[])
		{
			VOMidasInterrogation.Play( Message.Player )

			ExitUI(Message.Player)
		}
	}

	OnSelectedVO6Button(Message : widget_message) : void =
	{
		if (PlayerUI := GetPlayerUI[Message.Player], MyUI := VoiceMenuCanvas[Message.Player]?, SelectedButton := text_button_base[Message.Source], FortCharacter := Message.Player.GetFortCharacter[])
		{
			VOMidasSus.Play( Message.Player )

			ExitUI(Message.Player)
		}
	}

	OnSelectedVO7Button(Message : widget_message) : void =
	{
		if (PlayerUI := GetPlayerUI[Message.Player], MyUI := VoiceMenuCanvas[Message.Player]?, SelectedButton := text_button_base[Message.Source], FortCharacter := Message.Player.GetFortCharacter[])
		{
			VOMidasSusToIdle.Play( Message.Player )

			ExitUI(Message.Player)
		}
	}

	OnSelectedVO8Button(Message : widget_message) : void =
	{
		if (PlayerUI := GetPlayerUI[Message.Player], MyUI := VoiceMenuCanvas[Message.Player]?, SelectedButton := text_button_base[Message.Source])
		{
			ExitUI(Message.Player)
		}
	}

	ExitUI(Player : player) : void =
	{
		if (PlayerUI := GetPlayerUI[Player], VoiceMenu := VoiceMenuCanvas[Player]?)
		{
			PlayerUI.RemoveWidget(VoiceMenu)
				if (set VoiceMenuCanvas[Player] = false) {}
		}
	}

    HandlePrimaryButton(Agent : agent) : void =
	{
		if (InPlayer := player[Agent], PlayerUI := GetPlayerUI[InPlayer])
		{
			 if (VoiceMenu := VoiceMenuCanvas[InPlayer]?)
			 {
				PlayerUI.RemoveWidget(VoiceMenu)
				if (set VoiceMenuCanvas[InPlayer] = false) {}
			}
			else
			{
				NewVoiceMenu := CreateVoiceMenuUI()
				PlayerUI.AddWidget(NewVoiceMenu, player_ui_slot{InputMode := ui_input_mode.All})
				if (set VoiceMenuCanvas[InPlayer] = option{NewVoiceMenu}) {}
			}
		}
	}

    HandleSecondaryButton(Agent : agent) : void =
	{
		if (InPlayer := player[Agent], PlayerUI := GetPlayerUI[InPlayer])
		{
			if (VoiceMenu := VoiceMenuCanvas[InPlayer]?)
			{
				PlayerUI.RemoveWidget(VoiceMenu)
					if (set VoiceMenuCanvas[InPlayer] = false) {}
			}
		}
	}


    # Runs when the device is started in a running game
    OnBegin<override>()<suspends>:void=
	{
        SignalRemoteVoice.PrimarySignalEvent.Subscribe(HandlePrimaryButton)
        SignalRemoteVoice.SecondarySignalEvent.Subscribe(HandleSecondaryButton)
	}