<#
	TeuAgency
	UI Achievement List
#>
using { /Fortnite.com/Devices }
using { /Fortnite.com/Characters }
using { /Fortnite.com/Teams }
using { /Fortnite.com/UI }
using { /Fortnite.com/Game }
using { /Verse.org/Simulation }
using { /Verse.org/Simulation/Tags }
using { /Verse.org/Colors }
using { /Verse.org/Assets }
using { /UnrealEngine.com/Temporary/Diagnostics }
using { /UnrealEngine.com/Temporary/UI }
using { /UnrealEngine.com/Temporary/SpatialMath }
using { ModuleAchievements }

# Notes!!!
# AddWidget, RemoveWidget is evil and can cause big lags.

ui_achievements := class()
{
    LinkedCreativeDevice<public> : ui_achievements_device
    LinkedPlayer<public> : player 

    var IsInitialized<private> : logic = false
    var MainCanvas<private> : canvas = canvas{}

	UITextMessage<localizes>(Value : string) : message = "{Value}"

	UIBG: color_block = color_block{DefaultColor := NamedColors.Black, DefaultOpacity := 0.85}
	UITitleBG: color_block = color_block{DefaultColor := NamedColors.Black, DefaultOpacity := 1.0 <#, DefaultDesiredSize := vector2{X := 0.0, Y := 120.0}#>}
	UITitleText: text_block = text_block{DefaultTextColor := NamedColors.White, DefaultShadowColor := NamedColors.Black, DefaultShadowOffset := option{vector2{X := 2.0, Y := 2.0}} }

    BtnExit : button_loud = button_loud{}
    BtnScrollUp : button_regular = button_regular{}
    BtnScrollDown : button_regular = button_regular{}
	var ScrollOffset : float = 0.0
    StackBoxAchievementList : stack_box = stack_box{Orientation := orientation.Vertical}
	var AchievementArray : []overlay = array{}

    Init<public>() : void =
	{
        if (IsInitialized?)
		{
            return
		}

		set MainCanvas = CreateUI()
		set IsInitialized = true
    }

    OpenUI<public>() : void =
	{
        if (PlayerUI := GetPlayerUI[LinkedPlayer])
		{
            PlayerUI.AddWidget(GetCanvas(), player_ui_slot{InputMode := ui_input_mode.All})
			set ScrollOffset = 0.0
			UpdateUI()
		}
	}

    CloseUI<public>()<suspends> : void =
	{
        if (PlayerUI := GetPlayerUI[LinkedPlayer])
		{
			PlayerUI.RemoveWidget(GetCanvas())
		}
    }

    GetCanvas<public>() : canvas = MainCanvas

	CreateUI<private>() : canvas =
	{
		UICanvas : canvas = canvas{Slots := array{
		canvas_slot
		{
			ZOrder := {0}
			Anchors := anchors{Minimum := vector2{X := 0.0, Y := 0.0}, Maximum := vector2{X := 1.0, Y := 1.0}}
			Offsets := margin{Top := 0.0, Left := 0.0, Right := 0.0, Bottom := 0.0}
			Alignment := vector2{X := 0.0, Y := 0.0}
			SizeToContent := true
			Widget := UIBG
        }
		canvas_slot
		{
			ZOrder := {2}
			Anchors := anchors{Minimum := vector2{X := 0.0, Y := 0.0}, Maximum := vector2{X := 1.0, Y := 0.1}}
			Offsets := margin{Top := 0.0, Left := 0.0, Right := 0.0, Bottom := 0.0}
			Alignment := vector2{X := 0.0, Y := 0.0}
			SizeToContent := true
			Widget := UITitleBG
        }
		canvas_slot
		{
			ZOrder := {2}
			Anchors := anchors{Minimum := vector2{X := 0.5, Y := 0.025}, Maximum := vector2{X := 0.5, Y := 0.025} }
			Offsets := margin{Top := 0.0, Left := 0.0, Right := 0.0, Bottom := 0.0}
			Alignment := vector2{X := 0.5, Y := 0.0}
			SizeToContent := true
			Widget := UITitleText
        }
		canvas_slot
		{
			ZOrder := {1}
			Anchors := anchors{Minimum := vector2{X := 0.5, Y := 0.1}, Maximum := vector2{X := 0.5, Y := 0.1}}
			Offsets := margin{ Top:=ScrollOffset, Left:= 0.0, Right:=0.0, Bottom := 0.0 }
			Alignment := vector2{X := 0.5, Y := 0.0}
			SizeToContent := true
			Widget := StackBoxAchievementList
        }
		canvas_slot
		{
			ZOrder := {2}
			Anchors := anchors{Minimum := vector2{X := 0.9, Y := 0.9}, Maximum := vector2{X := 0.98, Y := 0.98} }
			Offsets := margin{Top := 0.0, Left := 0.0, Right := 0.0, Bottom := 0.0}
			Alignment := vector2{X := 1.0, Y := 1.0}
			SizeToContent := true
			Widget := stack_box{
				Orientation := orientation.Horizontal
				Slots := array{
					stack_box_slot
					{
						VerticalAlignment := vertical_alignment.Fill
						HorizontalAlignment := horizontal_alignment.Fill
						#Padding := margin{Top := 0.0, Left := 0.0, Right := 0.0, Bottom := 0.0}
						Widget := BtnExit
					}
				}
			}
		}
		canvas_slot
		{
			ZOrder := {2}
			Anchors := anchors{Minimum := vector2{X := 0.9, Y := 0.5}, Maximum := vector2{X := 0.9, Y := 0.5} }
			Offsets := margin{Top := 0.0, Left := 0.0, Right := 0.0, Bottom := 0.0}
			Alignment := vector2{X := 1.0, Y := 0.5}
			SizeToContent := true
			Widget := stack_box{
				Orientation := orientation.Vertical
				Slots := array{
					stack_box_slot
					{
						VerticalAlignment := vertical_alignment.Fill
						HorizontalAlignment := horizontal_alignment.Fill
						#Padding := margin{Top := 0.0, Left := 0.0, Right := 0.0, Bottom := 0.0}
						Widget := BtnScrollUp
					}
					stack_box_slot
					{
						VerticalAlignment := vertical_alignment.Fill
						HorizontalAlignment := horizontal_alignment.Fill
						#Padding := margin{Top := 0.0, Left := 0.0, Right := 0.0, Bottom := 0.0}
						Widget := BtnScrollDown
					}
				}
			}
		}
		}}

		UITitleText.SetText(TextLocalization.Achievements)

		BtnExit.SetText(TextLocalization.Exit)
		BtnExit.OnClick().Subscribe(OnBtnExit)
		BtnScrollUp.SetText(UITextMessage("↑"))
		BtnScrollUp.OnClick().Subscribe(OnBtnScrollUp)
		BtnScrollDown.SetText(UITextMessage("↓"))
		BtnScrollDown.OnClick().Subscribe(OnBtnScrollDown)

		return UICanvas
	}

	UpdateUI<public>() : void =
	{
		Print("Achievement: Updating UI...")
		for (OldAchievement : AchievementArray)
		{
			StackBoxAchievementList.RemoveWidget(OldAchievement)
		}
		set AchievementArray = array{}

		GetCanvas().RemoveWidget(StackBoxAchievementList)
		GetCanvas().AddWidget(canvas_slot
								{
									ZOrder := {1}
									Anchors := anchors{Minimum := vector2{X := 0.5, Y := 0.1}, Maximum := vector2{X := 0.5, Y := 0.1}}
									Offsets := margin{ Top:=ScrollOffset, Left:= 0.0, Right:=0.0, Bottom := 0.0 }
									Alignment := vector2{X := 0.5, Y := 0.0}
									SizeToContent := true
									Widget := StackBoxAchievementList
								})

		var UnlockedAchievementProgress : int = 0
		AllAchievementList := LinkedCreativeDevice.GetAllAchievements()
		AchievementList := LinkedCreativeDevice.GetAchievementsHistory(LinkedPlayer)
		for (AllAchievement : AllAchievementList)
		{
			AchievementTitle := LinkedCreativeDevice.GetAchievementsTitle(AllAchievement)
			AchievementDesc := LinkedCreativeDevice.GetAchievementsDesc(AllAchievement)
			AchievementIcon := LinkedCreativeDevice.GetAchievementsIcon(AllAchievement)

			if ( UnlockedAchievement := AchievementList.Find[AllAchievement] )
			{
				NewUnlockedOverlay :=	overlay
										{
											Slots := array
											{
												overlay_slot
												{
													VerticalAlignment := vertical_alignment.Fill
													HorizontalAlignment := horizontal_alignment.Fill
													Widget := material_block
													{
														DefaultImage := Materials.UI.M_UI_BG{}
														DefaultTint := NamedColors.White
														DefaultDesiredSize := vector2{X := 900.0, Y := 120.0}
													}
												}
												overlay_slot
												{
													VerticalAlignment := vertical_alignment.Top
													HorizontalAlignment := horizontal_alignment.Left
													Widget := texture_block
													{
														DefaultImage := AchievementIcon
														DefaultDesiredSize := vector2{X := 120.0, Y := 120.0}
													}
												}
												overlay_slot
												{
													VerticalAlignment := vertical_alignment.Fill
													HorizontalAlignment := horizontal_alignment.Fill
													Padding := margin{Top := 0.0, Left := 125.0, Right := 0.0, Bottom := 0.0}
													Widget := stack_box
													{
														Orientation := orientation.Vertical
														Slots := array
														{
															stack_box_slot
															{
																VerticalAlignment := vertical_alignment.Top
																HorizontalAlignment := horizontal_alignment.Fill
																Padding := margin{Top := 5.0, Left := 0.0, Right := 0.0, Bottom := 0.0}
																Widget := text_block
																{
																	DefaultText := UITextMessage(AchievementTitle)
																	DefaultTextColor := NamedColors.White
																	DefaultShadowColor := NamedColors.White
																	DefaultShadowOffset := option{vector2{X := 0.0, Y := 0.0}}
																	DefaultJustification := text_justification.Left
																}
															}
															stack_box_slot
															{
																VerticalAlignment := vertical_alignment.Bottom
																HorizontalAlignment := horizontal_alignment.Fill
																Padding := margin{Top := 0.0, Left := 10.0, Right := 0.0, Bottom := 5.0}
																Widget := text_block
																{
																	DefaultText := UITextMessage(AchievementDesc)
																	DefaultTextColor := NamedColors.White
																	DefaultTextOpacity := 0.5
																	DefaultShadowColor := NamedColors.White
																	DefaultShadowOffset := option{vector2{X := 0.0, Y := 0.0}}
																	DefaultJustification := text_justification.Left
																}
															}
														}
													}
												}
											}
										}

				set AchievementArray += array{NewUnlockedOverlay}
				set UnlockedAchievementProgress += 1
				StackBoxAchievementList.AddWidget(
							stack_box_slot
							{
								Padding := margin{Top := 2.0, Left := 0.0, Right := 0.0, Bottom := 5.0}
								Widget := NewUnlockedOverlay
							}
				)
			}
			else
			{
				NewOverlay :=	overlay
									{
										Slots := array
										{
											overlay_slot
											{
												VerticalAlignment := vertical_alignment.Fill
												HorizontalAlignment := horizontal_alignment.Fill
												Widget := material_block
												{
													DefaultImage := Materials.UI.MI_UI_BG_White{}
													DefaultTint := NamedColors.White
													DefaultDesiredSize := vector2{X := 900.0, Y := 120.0}
												}
											}
											overlay_slot
											{
												VerticalAlignment := vertical_alignment.Top
												HorizontalAlignment := horizontal_alignment.Left
												Widget := texture_block
												{
													DefaultImage := AchievementIcon
													DefaultDesiredSize := vector2{X := 120.0, Y := 120.0}
												}
											}
											overlay_slot
											{
												VerticalAlignment := vertical_alignment.Fill
												HorizontalAlignment := horizontal_alignment.Fill
												Padding := margin{Top := 0.0, Left := 125.0, Right := 0.0, Bottom := 0.0}
												Widget := stack_box
												{
													Orientation := orientation.Vertical
													Slots := array
													{
														stack_box_slot
														{
															VerticalAlignment := vertical_alignment.Top
															HorizontalAlignment := horizontal_alignment.Fill
															Padding := margin{Top := 5.0, Left := 0.0, Right := 0.0, Bottom := 0.0}
															Widget := text_block
															{
																DefaultText := UITextMessage(AchievementTitle)
																DefaultTextColor := NamedColors.White
																DefaultShadowColor := NamedColors.White
																DefaultShadowOffset := option{vector2{X := 0.0, Y := 0.0}}
																DefaultJustification := text_justification.Left
															}
														}
														stack_box_slot
														{
															VerticalAlignment := vertical_alignment.Bottom
															HorizontalAlignment := horizontal_alignment.Fill
															Padding := margin{Top := 0.0, Left := 10.0, Right := 0.0, Bottom := 5.0}
															Widget := text_block
															{
																DefaultText := UITextMessage(AchievementDesc)
																DefaultTextColor := NamedColors.White
																DefaultTextOpacity := 0.5
																DefaultShadowColor := NamedColors.White
																DefaultShadowOffset := option{vector2{X := 0.0, Y := 0.0}}
																DefaultJustification := text_justification.Left
															}
														}
													}
												}
											}
											overlay_slot
											{
												VerticalAlignment := vertical_alignment.Fill
												HorizontalAlignment := horizontal_alignment.Fill
												Widget := color_block
												{
													DefaultColor := NamedColors.Black
													DefaultDesiredSize := vector2{X := 900.0, Y := 120.0}
													DefaultOpacity := 0.65
												}
											}
										}
									}

				set AchievementArray += array{NewOverlay}
				StackBoxAchievementList.AddWidget(
							stack_box_slot
							{
								Padding := margin{Top := 2.0, Left := 0.0, Right := 0.0, Bottom := 5.0}
								Widget := NewOverlay
							}
				)
			}
		}

		UITitleText.SetText(UITextMessage("Achievements Progress: {UnlockedAchievementProgress}/{AchievementArray.Length}"))
	}

	OnBtnExit(Message : widget_message) : void =
	{
		if ( PlayerUI := GetPlayerUI[Message.Player], Agent := agent[Message.Player])
		{
			spawn{LinkedCreativeDevice.CloseUIMenu(Agent)}
		}
	}

	OnBtnScrollUp(Message : widget_message) : void =
	{
		if ( ScrollOffset = 0.0 )
		{
			return
		}
		set ScrollOffset += ( 120.0 * 7 )
		UpdateUI()
	}

	OnBtnScrollDown(Message : widget_message) : void =
	{
		set ScrollOffset -= ( 120.0 * 7 )
		#if ( ScrollOffset < -(120.0*(AchievementArray.Length - 1)) )
		if ( ScrollOffset < -(120.0*AchievementArray.Length) )
		{
			set ScrollOffset = 0.0
		}
		UpdateUI()
	}
}

# A Verse-authored creative device that can be placed in a level
ui_achievements_device := class(creative_device):

	var UIMapAchievement : [player]ui_achievements = map{}

	@editable SaveManagerDevice : ?save_manager_device = false

	@editable PopupMainMenu : popup_dialog_device = popup_dialog_device{}
	@editable AnalyticAchievement : analytics_device = analytics_device{}

	UITextMessage<localizes>(Value : string) : message = "{Value}"
	@editable HudMsg_Achievement : hud_message_device = hud_message_device{}
	@editable MSGFeed_Achievement 	: elimination_feed_device = elimination_feed_device{}

    # Runs when the device is started in a running game
    OnBegin<override>()<suspends>:void=
	{
		Print("Initialize UI Achievement Menu...")
	}

	# Removes an element from the given map and returns a new map without that element
	(Map:[keytype]valuetype where keytype:subtype(comparable), valuetype:type).RemoveKey(ElementToRemove:keytype):[keytype]valuetype=
	{
		var NewMap:[keytype]valuetype = map{}
		for (Key -> Value : Map, Key <> ElementToRemove)
		{
			set NewMap = ConcatenateMaps(NewMap, map{Key => Value})
		}

		return NewMap
	}

    UIPlayerRemoved(Agent: agent): void =
	{
		Print("UIPlayerRemoved")
		if (AgentPlayer := player[Agent])
		{
			set UIMapAchievement = UIMapAchievement.RemoveKey(AgentPlayer)
		}
	}

    OpenUIMenu(Agent:agent):void=
	{
		if (InPlayer := player[Agent])
		{
			var OptPlayerUI :?ui_achievements = false

			if (ThisPlayerUI := UIMapAchievement[InPlayer])
			{
				set OptPlayerUI = option{ThisPlayerUI}
			}
			else if (ThisPlayerUI := set UIMapAchievement[InPlayer] = ui_achievements{ LinkedPlayer := InPlayer, LinkedCreativeDevice := Self } ) 
			{
				ThisPlayerUI.Init()
				set OptPlayerUI = option{ThisPlayerUI}
			}

			if (PlayerUI := OptPlayerUI?)
			{
				PlayerUI.OpenUI()
			}
			else
			{
				Print("Error opening ui_achievements")
			}
		}
	}

    CloseUIMenu(Agent:agent)<suspends>:void=
	{
		if ( InPlayer := player[Agent], UIMenu := UIMapAchievement[InPlayer] )
		{
			spawn{UIMenu.CloseUI()}
			PopupMainMenu.Show(Agent)
		}
	}

	GetAchievementsHistory(Agent:agent):[]int=
	{
		if ( SaveManager := SaveManagerDevice? )
		{
			AchievementList := SaveManager.GetAchievementsHistory(Agent)
			return AchievementList
		}
		return array{}
	}

	GetAllAchievements():[]int=
	{
		var AllAchievementArray : []int = array{}

		for (AchievementID := 1..24)
		{
			set AllAchievementArray += array{AchievementID}
		}

		return AllAchievementArray
	}

	GetAchievementsTitle(Achievement:int):string=
	{
		var AchievementTitle : string = ""
		case( Achievement )
		{
			0 =>		set AchievementTitle = "First Achievement"
			1 =>		set AchievementTitle = "Doot 🎵~"
			2 =>		set AchievementTitle = "Butterfly Event?"
			3 =>		set AchievementTitle = "Zero Event?"
			4 =>		set AchievementTitle = "The Device Event?"
			5 =>		set AchievementTitle = "Myths of Midas"
			6 =>		set AchievementTitle = "Ascendant Midas"
			7 =>		set AchievementTitle = "I got over it"
			8 =>		set AchievementTitle = "The Spire Quest"
			9 =>		set AchievementTitle = "God of War™"
			10 =>		set AchievementTitle = "It's not just a storm?"
			11 =>		set AchievementTitle = "I won't bring you home"
			12 =>		set AchievementTitle = "The Last Reality is at hand!"
			13 =>		set AchievementTitle = "Are you sure this will work?"
			14 =>		set AchievementTitle = "Fracture"
			15 =>		set AchievementTitle = "It's always been Geno."
			16 =>		set AchievementTitle = "Imaginary Technique: Purple"
			17 =>		set AchievementTitle = "Za Warudo!"
			18 =>		set AchievementTitle = "Ice Storm"
			19 =>		set AchievementTitle = "Societal Collapse"
			20 =>		set AchievementTitle = "Destruction of Olympus"
			21 =>		set AchievementTitle = "Big Bang"
			22 =>		set AchievementTitle = "Your fate is sealed Wastelanders"
			23 =>		set AchievementTitle = "鬼"
			24 =>		set AchievementTitle = "Durr Burger gift card..."

			## =>		set AchievementTitle = "I shall set the seas ablaze!"
			## =>		set AchievementTitle = "Sinner's Finale"
			_ =>									#nothing
		}

		return AchievementTitle
	}

	GetAchievementsDesc(Achievement:int):string=
	{
		var AchievementTitle : string = ""
		case( Achievement )
		{
			0 =>			set AchievementTitle = "Play the game for the first time."
			1 =>			set AchievementTitle = "Use the DOOT Device."
			2 =>			set AchievementTitle = "DOOT the Cube during Fortnite Chapter 5."
			3 =>			set AchievementTitle = "DOOT the Zero Point."
			4 =>			set AchievementTitle = "DOOT the Device."
			5 =>			set AchievementTitle = "A golden king transformed his daughter gold."
			6 =>			set AchievementTitle = "Eliminate Hades with Midas."
			7 =>			set AchievementTitle = "Eliminate Slone with the Foundation."
			8 =>			set AchievementTitle = "Eliminate Spire Assassin with Raz."
			9 =>			set AchievementTitle = "Eliminate the God of War."
			10 =>			set AchievementTitle = "Use Midas' Crew Ultimate."
			11 =>			set AchievementTitle = "Use Doctor Slone's Ultimate."
			12 =>			set AchievementTitle = "Use Cube Queen's Ultimate."
			13 =>			set AchievementTitle = "Use Blue Cube's Ultimate."
			14 =>			set AchievementTitle = "Use the Herald's Ultimate."
			15 =>			set AchievementTitle = "Use the Ageless' Ultimate."
			16 =>			set AchievementTitle = "Use the Relik's Ultimate."
			17 =>			set AchievementTitle = "Use Kado Thorne's Ultimate."
			18 =>			set AchievementTitle = "Use Ice King's Ultimate."
			19 =>			set AchievementTitle = "Use the Society's Ultimate."
			20 =>			set AchievementTitle = "Use God's Ultimate."
			21 =>			set AchievementTitle = "Use Jones' Ultimate."
			22 =>			set AchievementTitle = "Use Wasteland Warriors Ultimate."
			23 =>			set AchievementTitle = "Use Shogun X's Ultimate."
			24 =>			set AchievementTitle = "Eliminate Midas with Alpha Boss."
			## =>			set AchievementTitle = "Witness the Pandora's Box Implosion."
			## =>			set AchievementTitle = "Getting killed by a sword while dancing."
			_ =>									#nothing
		}

		return AchievementTitle
	}

	GetAchievementsIcon(Achievement:int):texture=
	{
		var AchievementIcon : texture = Textures.midas_cursed1
		case( Achievement )
		{
			0 =>			set AchievementIcon = Textures.midas_cursed1
			1 =>			set AchievementIcon = Textures.midas_cursed1
			2 =>			set AchievementIcon = Textures.midas_cursed1
			3 =>			set AchievementIcon = Textures.midas_cursed1
			4 =>			set AchievementIcon = Textures.midas_cursed1
			5 =>			set AchievementIcon = Textures.jules_cursed
			6 =>			set AchievementIcon = Textures.class_hades
			7 =>			set AchievementIcon = Textures.class_thefoundation
			8 =>			set AchievementIcon = Textures.class_glyphmasterraz
			9 =>			set AchievementIcon = Textures.class_ares
			10 =>			set AchievementIcon = Textures.midas_cursed1
			11 =>			set AchievementIcon = Textures.class_doctorslone
			12 =>			set AchievementIcon = Textures.class_cubequeen
			13 =>			set AchievementIcon = Textures.class_cubeblue
			14 =>			set AchievementIcon = Textures.class_theherald
			15 =>			set AchievementIcon = Textures.class_theageless
			16 =>			set AchievementIcon = Textures.class_relik
			17 =>			set AchievementIcon = Textures.class_kadothorne
			18 =>			set AchievementIcon = Textures.class_iceking
			19 =>			set AchievementIcon = Textures.class_valeria
			20 =>			set AchievementIcon = Textures.class_zeus
			21 =>			set AchievementIcon = Textures.class_jones
			22 =>			set AchievementIcon = Textures.class_megalodon
			23 =>			set AchievementIcon = Textures.class_shogunx
			24 =>			set AchievementIcon = Textures.class_fletcherkane
			_ =>									#nothing
		}

		return AchievementIcon
	}

	AchievementUnlock(Agent:agent, AchievementID:int)<suspends>:void=
	{
		if ( SaveManager := SaveManagerDevice?, SaveManager.RecordAchievementsHistory[Agent, AchievementID] )
		{
			AchievementTitle := GetAchievementsTitle(AchievementID)
			HudMsg_Achievement.Show(Agent, UITextMessage(AchievementTitle), ?DisplayTime := 4.0)
			MSGFeed_Achievement.Activate(Agent)
			SaveManager.GrantXP(Agent, 30)
			AnalyticAchievement.Submit(Agent)
		}
	}

	AchievementUnlockAllPlayer(AchievementID:int)<suspends>:void=
	{
		AllPlayers := GetPlayspace().GetPlayers()
		for (Player : AllPlayers, Agent := agent[Player] )
		{
			spawn{AchievementUnlock(Agent,AchievementID)}
		}
	}