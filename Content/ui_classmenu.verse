<#
	2AGENCY
	UI - Class Selection
#>
using { /Fortnite.com/Devices }
using { /Fortnite.com/Characters }
using { /Fortnite.com/Teams }
using { /Fortnite.com/UI }
using { /Fortnite.com/Game }
using { /Verse.org/Assets }
using { /Verse.org/Simulation }
using { /Verse.org/Simulation/Tags }
using { /Verse.org/Colors }
using { /UnrealEngine.com/Temporary/Diagnostics }
using { /UnrealEngine.com/Temporary/UI }
using { /UnrealEngine.com/Temporary/SpatialMath }
using { ModulePlayerClasses }
using { ModulePlayerEx }
using { ModuleUIExtra }

# Notes!!!
# Todo: class count.


button_fnclass := struct
{
    ButtonQuiet<public> : button_quiet
    ButtonMat<public> : material_block
    ClassData<public> : fn_class_data
}

ClassDescBox := struct
{
	DescString<public> : string = ""
	DescIcon<public> : texture = Textures.ui_ability_no
	DescBGColor<public> : color = NamedColors.SaddleBrown
}

ui_classmenu := class() 
{
    LinkedCreativeDevice<public> : ui_classmenu_manager_device
    LinkedPlayer<public> : player 

    var IsInitialized<private> : logic = false
    var MainCanvas<private> : canvas = canvas{}

    var IsOpen<public> : logic = false

	UITextMessage<localizes>(Value : string) : message = "{Value}"

	ClassSelectBG: material_block = material_block{DefaultImage := Materials.UI.MI_UI_BG_White{}, DefaultTint := NamedColors.SaddleBrown}
	ClassSelectText: text_block = text_block{DefaultTextColor := NamedColors.White, DefaultShadowColor := NamedColors.Gray, DefaultShadowOffset := option{vector2{X := 2.0, Y := 2.0}} }
	ClassSelectTextDesc: text_block = text_block{DefaultTextColor := NamedColors.White, DefaultShadowColor := NamedColors.Orange, DefaultShadowOffset := option{vector2{X := 2.0, Y := 2.0}} }
    StackBoxClassList : stack_box = stack_box{Orientation := orientation.Vertical}
    StackBoxClassDesc : stack_box = stack_box{Orientation := orientation.Vertical}
	var ClassDescArray : []overlay = array{}

    BtnChangeTeam : button_regular = button_regular{}
    BtnMainMenu : button_regular = button_regular{}
    BtnBGM : button_regular = button_regular{}

    var iMaxColumn : int = 6
    var iMaxRow : int = 7
    var iSelectedBtnColumn : int = 0
    var iSelectedBtnRow : int = 0
	var ClassBtnMap: [int][]button_fnclass = map{}

    Init<public>() : void =
	{
        if (IsInitialized?)
		{
            return
		}

		set iSelectedBtnColumn = 5
		set iSelectedBtnRow = 5

		set MainCanvas = CreateClassSelectionUI()
		set IsInitialized = true
    }

    OpenUI<public>() : void =
	{
        if (PlayerUI := GetPlayerUI[LinkedPlayer])
		{
			if ( SaveManager := LinkedCreativeDevice.SaveManagerDevice?, Agent := agent[LinkedPlayer] )
			{
				if ( SaveManager.IsMouseEnabled[Agent] )
				{
					PlayerUI.AddWidget(GetCanvas(), player_ui_slot{InputMode := ui_input_mode.All})
				}
				else
				{
					PlayerUI.AddWidget(GetCanvas(), player_ui_slot{InputMode := ui_input_mode.None})
				}
			}
			else
			{
				PlayerUI.AddWidget(GetCanvas(), player_ui_slot{InputMode := ui_input_mode.None})
			}

			UpdateUI()
			set IsOpen = true
		}
	}

    CloseUI<public>()<suspends> : void =
	{
        if (PlayerUI := GetPlayerUI[LinkedPlayer])
		{
			PlayerUI.RemoveWidget(GetCanvas())
			set IsOpen = false
			
			# extra precaution to prevent bad internet + ui stuck on screen.
			Sleep(0.1)
			PlayerUI.RemoveWidget(GetCanvas())
		}
    }

    GetCanvas<public>() : canvas = MainCanvas  # Exposes the class Canvas to public

	CreateClassSelectionUI<private>() : canvas =
	{
		UIClassSelectionCanvas : canvas = canvas{Slots := array{
		canvas_slot
		{
			ZOrder := {0}
			Anchors := anchors{Minimum := vector2{X := 0.0, Y := 0.0}, Maximum := vector2{X := 0.5, Y := 1.0}}
			Offsets := margin{Top := 0.0, Left := 0.0, Right := 0.0, Bottom := 0.0}
			Alignment := vector2{X := 0.0, Y := 0.0}
			SizeToContent := true
			Widget := ClassSelectBG
        }
		canvas_slot
		{
			ZOrder := {1}
			Anchors := anchors{Minimum := vector2{X := 0.25, Y := 0.025}, Maximum := vector2{X := 0.25, Y := 0.025} }
			Offsets := margin{Top := 0.0, Left := 0.0, Right := 0.0, Bottom := 0.0}
			Alignment := vector2{X := 0.5, Y := 0.0}
			SizeToContent := true
			Widget := ClassSelectText
        }
		canvas_slot
		{
			ZOrder := {1}
			Anchors := anchors{Minimum := vector2{X := 0.75, Y := 0.05}, Maximum := vector2{X := 0.75, Y := 0.05} }
			Offsets := margin{Top := 0.0, Left := 0.0, Right := 0.0, Bottom := 0.0}
			Alignment := vector2{X := 0.5, Y := 0.0}
			SizeToContent := true
			Widget := ClassSelectTextDesc
        }
		canvas_slot
		{
			ZOrder := {1}
			Anchors := anchors{Minimum := vector2{X := 0.1, Y := 0.075}, Maximum := vector2{X := 0.1, Y := 0.075}}
			Offsets := margin{ Top:=0.0, Left:= 0.0, Right:=0.0, Bottom := 0.0 }
			Alignment := vector2{X := 0.0, Y := 0.0}
			SizeToContent := true
			Widget := StackBoxClassList
        }
		canvas_slot
		{
			ZOrder := {1}
			Anchors := anchors{Minimum := vector2{X := 1.0, Y := 0.4}, Maximum := vector2{X := 1.0, Y := 0.4}}
			Offsets := margin{ Top:=0.0, Left:= 0.0, Right:=0.0, Bottom := 0.0 }
			Alignment := vector2{X := 1.0, Y := 0.0}
			SizeToContent := true
			Widget := StackBoxClassDesc
        }
		canvas_slot
		{
			ZOrder := {1}
			Anchors := anchors{Minimum := vector2{X := 0.0, Y := 1.0}, Maximum := vector2{X := 0.0, Y := 1.0} }
			Offsets := margin{Top := 0.0, Left := 0.0, Right := 0.0, Bottom := 0.0}
			Alignment := vector2{X := 0.0, Y := 1.0}
			SizeToContent := true
			Widget := stack_box{
				Orientation := orientation.Horizontal
				Slots := array{
					stack_box_slot
					{
						VerticalAlignment := vertical_alignment.Fill
						HorizontalAlignment := horizontal_alignment.Fill
						#Padding := margin{Top := 0.0, Left := 0.0, Right := 0.0, Bottom := 0.0}
						Widget := BtnMainMenu
					}
					stack_box_slot
					{
						VerticalAlignment := vertical_alignment.Fill
						HorizontalAlignment := horizontal_alignment.Fill
						#Padding := margin{Top := 0.0, Left := 0.0, Right := 0.0, Bottom := 0.0}
						Widget := BtnBGM
					}
				}
			}
		}
		canvas_slot
		{
			ZOrder := {1}
			Anchors := anchors{Minimum := vector2{X := 1.0, Y := 1.0}, Maximum := vector2{X := 1.0, Y := 1.0} }
			Offsets := margin{Top := 0.0, Left := 0.0, Right := 0.0, Bottom := 0.0}
			Alignment := vector2{X := 1.0, Y := 1.0}
			SizeToContent := true
			Widget := stack_box{
				Orientation := orientation.Horizontal
				Slots := array{
					stack_box_slot
					{
						VerticalAlignment := vertical_alignment.Fill
						HorizontalAlignment := horizontal_alignment.Fill
						#Padding := margin{Top := 0.0, Left := 0.0, Right := 0.0, Bottom := 50.0}
						Widget := BtnChangeTeam
					}
				}
			}
		}
		}}

		ClassSelectText.SetText(TextLocalization.SelectClass)
		#ClassSelectTextDesc.SetText(UITextMessage(""))

		BtnChangeTeam.SetText(TextLocalization.ChangeTeam)
		BtnMainMenu.SetText(TextLocalization.MainMenu)
		BtnBGM.SetText(TextLocalization.Setting_MusicON)

		ClassSelectText.SetShadowOpacity(1.0)
		ClassSelectTextDesc.SetShadowOpacity(1.0)

		BtnChangeTeam.OnClick().Subscribe(OnBtnChangeTeam)

		BtnMainMenu.OnClick().Subscribe(OnBtnMainMenu)
		BtnBGM.OnClick().Subscribe(OnBtnBGM)

		# Max Column = 7
		RowStackBoxes : []stack_box = for ( I := 0..iMaxColumn ) {stack_box{Orientation := orientation.Horizontal}}
		var StackRow : int = 0
		var StackColumn : int = 0
		var BtnArray : []button_fnclass = array{}

		# create button for each classes
		AllClassesList := LinkedCreativeDevice.GetAllClasses()
		for (NewClass : AllClassesList, ClassImage := NewClass.ClassImage, ClassID := NewClass.FNClassID)
		{
			NewBtn :=	button_quiet{}
			NewMat := Materials.UI.M_UI_ClassPortrait{IconTexture:=ClassImage}
			NewImage :=	material_block{DefaultImage := NewMat, DefaultDesiredSize := vector2{X := 85.0, Y := 125.0}}
			NewOverlay :=		overlay
								{
									Slots := array
									{
										overlay_slot
										{
											VerticalAlignment := vertical_alignment.Center
											Widget := NewImage
										}
										<#VerticalAlignment := vertical_alignment.Bottom
										HorizontalAlignment := horizontal_alignment.Right
										Padding := margin{Top := 0.0, Left := 0.0, Right := 0.0, Bottom := 0.0}
										Widget := text_block
										{
											DefaultText := UITextMessage("0")
											DefaultTextColor := NamedColors.White
											DefaultTextOpacity := 0.9
											DefaultShadowColor := NamedColors.Black
											DefaultShadowOffset := option{vector2{X := 0.0, Y := 0.0}}
											DefaultJustification := text_justification.Right
										}#>
										overlay_slot
										{
											VerticalAlignment := vertical_alignment.Fill
											HorizontalAlignment := horizontal_alignment.Fill
											Padding := margin{Top := 0.0, Left := 0.0, Right := 0.0, Bottom := 0.0}
											Widget := NewBtn
										}
									}
								}

			# new class!
			if (ClassID = FNClasses.Daigo or ClassID = FNClasses.Kor)
			{
				NewOverlay.AddWidget(
									overlay_slot
									{
										VerticalAlignment := vertical_alignment.Fill
										HorizontalAlignment := horizontal_alignment.Fill
										Widget := texture_block
										{
											DefaultImage := Textures.ui_class_new
											DefaultDesiredSize := vector2{X := 85.0, Y := 125.0}
										}
									}

				)
			}

			# updated class!
			if (ClassID = FNClasses.DoctorDoom)
			{
				NewOverlay.AddWidget(
									overlay_slot
									{
										VerticalAlignment := vertical_alignment.Fill
										HorizontalAlignment := horizontal_alignment.Fill
										Widget := texture_block
										{
											DefaultImage := Textures.ui_class_updated
											DefaultDesiredSize := vector2{X := 85.0, Y := 125.0}
										}
									}

				)
			}

			if (StackBox := RowStackBoxes[StackColumn])
			{
				StackBox.AddWidget(
								stack_box_slot
								{
									VerticalAlignment := vertical_alignment.Top
									HorizontalAlignment := horizontal_alignment.Left
									Padding := margin{Top := 0.0, Left := 0.0, Right := 1.0, Bottom := 0.0}
									Widget := NewOverlay
								}
				)
			}

			NewBtn.OnClick().SubscribeWidget(OnClassSelectedBtn, ClassID)
			NewFNBtn := button_fnclass{ ButtonQuiet := NewBtn, ButtonMat := NewImage, ClassData := NewClass}

			set BtnArray += array{NewFNBtn}
			if { set ClassBtnMap[StackColumn] = BtnArray }

			set StackRow += 1
			if ( StackRow >= iMaxRow ) # number of buttons per column
			{
				# new column
				set StackColumn += 1
				set BtnArray = array{}
				set StackRow = 0
			}
		}

		# add stuff into column
		for (RowStackBox : RowStackBoxes)
		{
			StackBoxClassList.AddWidget(
								stack_box_slot
								{
									VerticalAlignment := vertical_alignment.Top
									HorizontalAlignment := horizontal_alignment.Left
									Padding := margin{Top := 0.0, Left := 0.0, Right := 0.0, Bottom := 0.0}
									Widget := RowStackBox
								}
			)
		}

		return UIClassSelectionCanvas
	}

	MoveClassBtn(BtnMovement:string) : void =
	{
		case( BtnMovement )
		{
			"Up" =>
			{
				set iSelectedBtnColumn -= 1
			}
			"Down" =>
			{
				set iSelectedBtnColumn += 1
			}
			"Left" =>
			{
				set iSelectedBtnRow -= 1
			}
			"Right" =>
			{
				set iSelectedBtnRow += 1
			}
			_ =>
			{
				set iSelectedBtnColumn = 0
				set iSelectedBtnRow = 0
			}
		}

		if ( iSelectedBtnColumn = ( ClassBtnMap.Length - 1 ) )
		{
			AllClassesList := LinkedCreativeDevice.GetAllClasses()
			LastColumnMax := ( AllClassesList.Length -1 - (iMaxRow*iMaxColumn) )
			if ( iSelectedBtnRow > LastColumnMax ) # max row = 0 to 4
			{
				set iSelectedBtnRow = 4
			}

			if ( iSelectedBtnRow < 0 )
			{
				set iSelectedBtnRow = 0
			}
		}
		else
		{
			if ( iSelectedBtnRow > ( iMaxRow -1 ) )
			{
				set iSelectedBtnRow -= 1
			}

			if ( iSelectedBtnRow < 0 )
			{
				set iSelectedBtnRow = 0
			}
		}

		if ( iSelectedBtnColumn > ClassBtnMap.Length )
		{
			set iSelectedBtnColumn -= 1
		}

		if ( iSelectedBtnColumn < 0 )
		{
			set iSelectedBtnColumn = 0
		}

		SelectClassBtn()
	}

	SelectClassBtn() : void =
	{
		if ( Agent := agent[LinkedPlayer], SaveManager := LinkedCreativeDevice.SaveManagerDevice?, SaveManager.IsMouseEnabled[Agent] )
		{
			for ( BtnColumn->ClassBtn : ClassBtnMap )
			{
				for (BtnFNClass : ClassBtn, BtnQuiet := BtnFNClass.ButtonQuiet )
				{
					BtnQuiet.SetEnabled(true)
				}
			}

			BtnMainMenu.SetEnabled(true)
			BtnBGM.SetEnabled(true)
			BtnChangeTeam.SetEnabled(true)

			return
		}

		for ( BtnColumn->ClassBtn : ClassBtnMap )
		{
			for (BtnFNClass : ClassBtn, BtnQuiet := BtnFNClass.ButtonQuiet )
			{
				BtnQuiet.SetEnabled(false)

				BtnMatRaw := BtnFNClass.ButtonMat.GetImage()
				if ( BtnMat := Materials.UI.M_UI_ClassPortrait[BtnMatRaw] )
				{
					set BtnMat.IsSelected = 0.0
				}
			}
		}

		BtnMainMenu.SetEnabled(false)
		BtnBGM.SetEnabled(false)
		BtnChangeTeam.SetEnabled(false)

		if ( SelectedBtns := ClassBtnMap[iSelectedBtnColumn], BtnFNClass := SelectedBtns[iSelectedBtnRow], BtnQuiet := BtnFNClass.ButtonQuiet, BtnClassData := BtnFNClass.ClassData )
		{
			BtnQuiet.SetEnabled(true)
			SetClassDesc(BtnClassData)

			BtnMatRaw := BtnFNClass.ButtonMat.GetImage()
			if ( BtnMat := Materials.UI.M_UI_ClassPortrait[BtnMatRaw] )
			{
				set BtnMat.IsSelected = 1.0
			}
		}

		if ( iSelectedBtnColumn = ClassBtnMap.Length )
		{
			case( iSelectedBtnRow )
			{
				0 => BtnMainMenu.SetEnabled(true)
				1 => BtnBGM.SetEnabled(true)
				2 => BtnChangeTeam.SetEnabled(true)
				3 => BtnChangeTeam.SetEnabled(true)
				4 => BtnChangeTeam.SetEnabled(true)
				5 => BtnChangeTeam.SetEnabled(true)
				_ =>
			}
		}

		if ( Agent := agent[LinkedPlayer] )
		{
			LinkedCreativeDevice.SFX_Select.Play(Agent)
		}
	}

	SelectButton() : void =
	{
		if ( Agent := agent[LinkedPlayer] )
		{
			if ( iSelectedBtnColumn = ClassBtnMap.Length )
			{
				case( iSelectedBtnRow )
				{
					0 =>	ClickMainMenu(Agent)
					1 =>	ToggleBGM(Agent)
					2 =>	ClickChangeTeam(Agent)
					3 =>	ClickChangeTeam(Agent)
					4 =>	ClickChangeTeam(Agent)
					5 =>	ClickChangeTeam(Agent)
					_ =>	
				}
			}
			else
			{
				if ( SelectedBtns := ClassBtnMap[iSelectedBtnColumn], BtnFNClass := SelectedBtns[iSelectedBtnRow], BtnClassData := BtnFNClass.ClassData, ClassID := BtnClassData.FNClassID )
				{
					SetPlayerClass(ClassID)
					spawn{LinkedCreativeDevice.CloseClassMenu(Agent)}
					LinkedCreativeDevice.PlayClassStinger(Agent, ClassID)
				}
			}
		}
	}

	SetPlayerClass(ClassID:FNClasses) : void =
	{
		if ( Agent := agent[LinkedPlayer] )
		{
			LinkedCreativeDevice.SetPlayerClass( Agent, ClassID )
		}
	}

	SetClassDesc(ClassData:fn_class_data) : void =
	{
		ClassID := ClassData.FNClassID
		var ClassName : message =  ClassData.FNClassName

		if ( Agent := agent[LinkedPlayer], Agent.IsOnTeam[LinkedCreativeDevice, 1], ClassData.HasTeamAlt? )
		{
			set ClassName = ClassData.Team2_FNClassName
		}

		ClassSelectTextDesc.SetText(ClassName)

		var NewClassDescStructs : []ClassDescBox = array{}

		case( ClassID )
		{
				FNClasses.Midas =>
				{
					set NewClassDescStructs += array{ClassDescBox{DescIcon := Textures.ui_ability_doublejump, DescBGColor := NamedColors.Gold, DescString := "ABILITY: Double Jump"}}
					set NewClassDescStructs += array{ClassDescBox{DescIcon := Textures.ui_ultimate, DescBGColor := NamedColors.Gold, DescString := "ULTIMATE: The Device\nSummon the storm, dealing\ncontinuous AoE damage."}}
				}
				FNClasses.TNTina =>
				{
					set NewClassDescStructs += array{ClassDescBox{DescIcon := Textures.ui_ability_doublejump, DescBGColor := NamedColors.Gold, DescString := "ABILITY: Double Jump"}}
					set NewClassDescStructs += array{ClassDescBox{DescIcon := Textures.ui_ultimate, DescBGColor := NamedColors.Gold, DescString := "ULTIMATE: The Device\nSummon the storm, dealing\ncontinuous AoE damage."}}
					set NewClassDescStructs += array{ClassDescBox{DescIcon := Textures.ui_ability_partinggift, DescBGColor := NamedColors.Gold, DescString := "PASSIVE: Shield Bomb\nWhen shield breaks,\nyou explode."}}
				}
				FNClasses.Brutus =>
				{
					set NewClassDescStructs += array{ClassDescBox{DescIcon := Textures.ui_ability_doublejump, DescBGColor := NamedColors.Gold, DescString := "ABILITY: Double Jump"}}
					set NewClassDescStructs += array{ClassDescBox{DescIcon := Textures.ui_ultimate, DescBGColor := NamedColors.Gold, DescString := "ULTIMATE: The Device\nSummon the storm, dealing\ncontinuous AoE damage."}}
				}
				FNClasses.Skye =>
				{
					set NewClassDescStructs += array{ClassDescBox{DescIcon := Textures.ui_ability_doublejump, DescBGColor := NamedColors.Gold, DescString := "ABILITY: Double Jump"}}
					set NewClassDescStructs += array{ClassDescBox{DescIcon := Textures.ui_ultimate, DescBGColor := NamedColors.Gold, DescString := "ULTIMATE: The Device\nSummon the storm, dealing\ncontinuous AoE damage."}}
				}
				FNClasses.Meowscles =>
				{
					set NewClassDescStructs += array{ClassDescBox{DescIcon := Textures.ui_ability_doublejump, DescBGColor := NamedColors.Gold, DescString := "ABILITY: Double Jump"}}
					set NewClassDescStructs += array{ClassDescBox{DescIcon := Textures.ui_ultimate, DescBGColor := NamedColors.Gold, DescString := "ULTIMATE: The Device\nSummon the storm, dealing\ncontinuous AoE damage."}}
				}
				FNClasses.Jules =>
				{
					set NewClassDescStructs += array{ClassDescBox{DescIcon := Textures.ui_ultimate, DescBGColor := NamedColors.Gold, DescString := "ULTIMATE: Flood"}}
					set NewClassDescStructs += array{ClassDescBox{DescIcon := Textures.ui_ability_therapeuticswim, DescBGColor := NamedColors.Gold, DescString := "PASSIVE: Therapeutic Swim\nRegenerate health while swimming."}}
				}
				FNClasses.Kit =>
				{
					set NewClassDescStructs += array{ClassDescBox{DescIcon := Textures.ui_ultimate, DescBGColor := NamedColors.Gold, DescString := "ULTIMATE: Flood"}}
					set NewClassDescStructs += array{ClassDescBox{DescIcon := Textures.ui_ability_airraid, DescBGColor := NamedColors.Gold, DescString := "PASSIVE: Air Raid\nDamage Buff x1.5 while airborne\nfor longer than a second."}}
					set NewClassDescStructs += array{ClassDescBox{DescIcon := Textures.ui_ability_therapeuticswim, DescBGColor := NamedColors.Gold, DescString := "PASSIVE: Therapeutic Swim\nRegenerate health while swimming."}}
				}
				FNClasses.Ocean =>
				{
					set NewClassDescStructs += array{ClassDescBox{DescIcon := Textures.ui_ultimate, DescBGColor := NamedColors.Gold, DescString := "ULTIMATE: Flood"}}
					set NewClassDescStructs += array{ClassDescBox{DescIcon := Textures.ui_ability_therapeuticswim, DescBGColor := NamedColors.Gold, DescString := "PASSIVE: Therapeutic Swim\nRegenerate health while swimming."}}
				}
				FNClasses.ShadowMidas =>
				{
					set NewClassDescStructs += array{ClassDescBox{DescIcon := Textures.class_shadow, DescBGColor := NamedColors.Purple, DescString := "ABILITY: Join Shadow\nRespawn as Shadow after dying."}}
					set NewClassDescStructs += array{ClassDescBox{DescIcon := Textures.ui_ultimate, DescBGColor := NamedColors.Purple, DescString := "ULTIMATE: The Device\nSummon the storm, dealing\ncontinuous AoE damage."}}
				}
				FNClasses.Raz =>
				{
					set NewClassDescStructs += array{ClassDescBox{DescIcon := Textures.ui_ability_raz_lifedrain, DescBGColor := NamedColors.Purple, DescString := "ABILITY: Life Drain\nsiphon nearby enemies hp."}}
					set NewClassDescStructs += array{ClassDescBox{DescIcon := Textures.ui_ability_dash, DescBGColor := NamedColors.Purple, DescString := "ABILITY: Warp\nDash forward."}}
					set NewClassDescStructs += array{ClassDescBox{DescIcon := Textures.ui_ultimate, DescBGColor := NamedColors.Purple, DescString := "ULTIMATE: Damage x2."}}
				}
				FNClasses.Slone =>
				{
					set NewClassDescStructs += array{ClassDescBox{DescIcon := Textures.ui_ability_ioguard, DescBGColor := NamedColors.DarkGray,  DescString := "ABILITY: Mirror Image\nSummon an IO Guard."}}
					set NewClassDescStructs += array{ClassDescBox{DescIcon := Textures.ui_ultimate, DescBGColor := NamedColors.DarkGray,  DescString := "ULTIMATE: Operation: Sky Fire\nnuke all enemies within radius."}}
				}
				FNClasses.TheFoundation =>
				{
					set NewClassDescStructs += array{ClassDescBox{DescIcon := Textures.ui_ability_doublejump, DescString := "ABILITY: Double Jump\nHold Crouch midair to\nperform ground pound."}}
					set NewClassDescStructs += array{ClassDescBox{DescIcon := Textures.ui_ability_rock, DescString := "ABILITY: The Rock\nThrow an Explosive Rock."}}
					set NewClassDescStructs += array{ClassDescBox{DescIcon := Textures.ui_ultimate, DescString := "ULTIMATE: Meteor\nSummon the Meteor and \nnuke all enemies within its impact radius."}}
				}
				FNClasses.Gunnar =>
				{
					set NewClassDescStructs += array{ClassDescBox{DescIcon := Textures.ui_ability_charge, DescBGColor := NamedColors.DarkGray,  DescString := "ABILITY: Dash forward and explode."}}
					set NewClassDescStructs += array{ClassDescBox{DescIcon := Textures.ui_ultimate, DescBGColor := NamedColors.DarkGray,  DescString := "ULTIMATE: Damage x2."}}
				}
				FNClasses.HuntmasterSaber =>
				{
					set NewClassDescStructs += array{ClassDescBox{DescIcon := Textures.ui_ability_dash, DescBGColor := NamedColors.DarkGray,  DescString := "ABILITY: Warp\nDash forward."}}
					set NewClassDescStructs += array{ClassDescBox{DescIcon := Textures.ui_ability_spinslash, DescBGColor := NamedColors.DarkGray,  DescString := "ABILITY: Spin Slash\ndeals continuous AoE damage."}}
					set NewClassDescStructs += array{ClassDescBox{DescIcon := Textures.ui_ultimate, DescBGColor := NamedColors.DarkGray,  DescString := "ULTIMATE: Damage x2."}}
				}
				FNClasses.TheHerald =>
				{
					set NewClassDescStructs += array{ClassDescBox{DescIcon := Textures.ui_ability_dash, DescBGColor := NamedColors.Purple, DescString := "ABILITY: Warp\nDash forward."}}
					set NewClassDescStructs += array{ClassDescBox{DescIcon := Textures.ui_ability_lockdown, DescBGColor := NamedColors.Purple, DescString := "ABILITY: Chrome Summon\nSummon a Wolf."}}
					set NewClassDescStructs += array{ClassDescBox{DescIcon := Textures.ui_ultimate, DescBGColor := NamedColors.Purple, DescString := "ULTIMATE: Fracture Event\nnuke all enemies within radius."}}
				}
				FNClasses.TheAgelessChampion =>
				{
					set NewClassDescStructs += array{ClassDescBox{DescIcon := Textures.ui_ability_beam, DescBGColor := NamedColors.DarkGray, DescString := "ABILITY: Geno Blast\nShoot laser beam, dealing \ncontinuous damage."}}
					set NewClassDescStructs += array{ClassDescBox{DescIcon := Textures.ui_ultimate, DescBGColor := NamedColors.DarkGray, DescString := "ULTIMATE: Collision Event\nnuke all enemies within radius."}}
				}
				FNClasses.WildguardRelik =>
				{
					set NewClassDescStructs += array{ClassDescBox{DescIcon := Textures.ui_ability_dash, DescBGColor := NamedColors.DarkGreen, DescString := "ABILITY: Warp\nDash forward."}}
					set NewClassDescStructs += array{ClassDescBox{DescIcon := Textures.ui_ultimate, DescBGColor := NamedColors.DarkGreen, DescString := "ULTIMATE: Hollow Purple\nShoot Hollow Purple forwards, dealing \ncontinuous damage to all enemies in its path."}}
				}
				FNClasses.KadoThorne =>
				{
					set NewClassDescStructs += array{ClassDescBox{DescIcon := Textures.ui_ability_doublejump, DescBGColor := NamedColors.DarkRed, DescString := "ABILITY: Double Jump\nHold Crouch midair to\nperform ground pound."}}
					set NewClassDescStructs += array{ClassDescBox{DescIcon := Textures.ui_ability_kado_lifedrain, DescBGColor := NamedColors.DarkRed, DescString := "ABILITY: Life Drain\nsiphon nearby enemies hp."}}
					set NewClassDescStructs += array{ClassDescBox{DescIcon := Textures.ui_ultimate, DescBGColor := NamedColors.DarkRed, DescString := "ULTIMATE: Stop Time\nSummon the Time Machine and \nStun all enemies within range."}}
				}
				FNClasses.TheDealer =>
				{
					set NewClassDescStructs += array{ClassDescBox{DescIcon := Textures.ui_ability_flag, DescBGColor := NamedColors.DarkRed, DescString := "PASSIVE: Last Resort\nMove faster while carrying\na capture flag."}}
					set NewClassDescStructs += array{ClassDescBox{DescIcon := Textures.ui_ultimate, DescBGColor := NamedColors.DarkRed, DescString := "ULTIMATE: Stop Time\nSummon the Time Machine and \nStun all enemies within range."}}
				}
				FNClasses.Valeria =>
				{
					set NewClassDescStructs += array{ClassDescBox{DescIcon := Textures.ui_ability_societyguard, DescBGColor := NamedColors.Goldenrod,  DescString := "ABILITY: Phase 2\nSummon a Society Guard."}}
					set NewClassDescStructs += array{ClassDescBox{DescIcon := Textures.ui_ultimate, DescBGColor := NamedColors.Goldenrod,  DescString := "ULTIMATE: Pandora's Box Opening\nSummon the Pandora's Box, dealing\ncontinuous AoE damage."}}
					set NewClassDescStructs += array{ClassDescBox{DescIcon := Textures.ui_ability_shieldregen_valeria, DescBGColor := NamedColors.Goldenrod,  DescString := "PASSIVE: Shield Regen"}}
				}
				FNClasses.Oscar =>
				{
					set NewClassDescStructs += array{ClassDescBox{DescIcon := Textures.ui_ability_societyguard, DescBGColor := NamedColors.Goldenrod,  DescString := "ABILITY: Phase 2\nSummon a Society Guard."}}
					set NewClassDescStructs += array{ClassDescBox{DescIcon := Textures.ui_ultimate, DescBGColor := NamedColors.Goldenrod,  DescString := "ULTIMATE: Pandora's Box Opening\nSummon the Pandora's Box, dealing\ncontinuous AoE damage."}}
					set NewClassDescStructs += array{ClassDescBox{DescIcon := Textures.ui_ability_shieldregen_oscar, DescBGColor := NamedColors.Goldenrod,  DescString := "PASSIVE: Shield Regen"}}
				}
				FNClasses.Nisha =>
				{
					set NewClassDescStructs += array{ClassDescBox{DescIcon := Textures.ui_ability_societyguard, DescBGColor := NamedColors.Goldenrod,  DescString := "ABILITY: Phase 2\nSummon a Society Guard."}}
					set NewClassDescStructs += array{ClassDescBox{DescIcon := Textures.ui_ultimate, DescBGColor := NamedColors.Goldenrod,  DescString := "ULTIMATE: Pandora's Box Opening\nSummon the Pandora's Box, dealing\ncontinuous AoE damage."}}
					set NewClassDescStructs += array{ClassDescBox{DescIcon := Textures.ui_ability_shieldregen_nisha, DescBGColor := NamedColors.Goldenrod,  DescString := "PASSIVE: Shield Regen"}}
				}
				FNClasses.Montague =>
				{
					set NewClassDescStructs += array{ClassDescBox{DescIcon := Textures.ui_ability_societyguard, DescBGColor := NamedColors.Goldenrod, DescString := "ABILITY: Phase 2\nSummon a Society Guard."}}
					set NewClassDescStructs += array{ClassDescBox{DescIcon := Textures.ui_ultimate, DescBGColor := NamedColors.Goldenrod, DescString := "ULTIMATE: Pandora's Box Opening\nSummon the Pandora's Box, dealing\ncontinuous AoE damage."}}
					set NewClassDescStructs += array{ClassDescBox{DescIcon := Textures.ui_ability_shieldregen_montague, DescBGColor := NamedColors.Goldenrod, DescString := "PASSIVE: Shield Regen"}}
				}
				FNClasses.SgtWinter =>
				{
					set NewClassDescStructs += array{ClassDescBox{DescIcon := Textures.ui_ultimate, DescBGColor := NamedColors.DodgerBlue, DescString := "ULTIMATE: Grant Present."}}
				}
				FNClasses.TheIceKing =>
				{
					set NewClassDescStructs += array{ClassDescBox{DescIcon := Textures.ui_ultimate, DescBGColor := NamedColors.DodgerBlue,  DescString := "ULTIMATE: Ice Storm\nStun all enemies within range."}}
				}
				FNClasses.SpireAssassin =>
				{
					set NewClassDescStructs += array{ClassDescBox{DescIcon := Textures.ui_ability_dash, DescBGColor := NamedColors.Purple, DescString := "ABILITY: Warp\nDash forward."}}
					set NewClassDescStructs += array{ClassDescBox{DescIcon := Textures.ui_ultimate, DescBGColor := NamedColors.Purple, DescString := "ULTIMATE: Damage x2."}}
				}
				FNClasses.CubeAssassin =>
				{
					set NewClassDescStructs += array{ClassDescBox{DescIcon := Textures.ui_ability_dash, DescBGColor := NamedColors.Purple, DescString := "ABILITY: Warp\nDash forward."}}
					set NewClassDescStructs += array{ClassDescBox{DescIcon := Textures.ui_ability_cubeassasin_shield, DescBGColor := NamedColors.Purple, DescString := "ABILITY: Spawn Shield\nPlace a temporary shield."}}
					set NewClassDescStructs += array{ClassDescBox{DescIcon := Textures.ui_ultimate, DescBGColor := NamedColors.Purple, DescString := "ULTIMATE: Sideways Domain\nProvides 'Damage x2' to you\nand nearby teammates."}}
					set NewClassDescStructs += array{ClassDescBox{DescIcon := Textures.ui_ability_no, DescBGColor := NamedColors.Purple, DescString := "PASSIVE: Can't Swim"}}
				}
				FNClasses.CubeBlue =>
				{
					set NewClassDescStructs += array{ClassDescBox{DescIcon := Textures.ui_ability_cubelightning, DescString := "Primary Attack: Lightning\nAbsorb all damage taken\nand turn it into extra damage."}}
					set NewClassDescStructs += array{ClassDescBox{DescIcon := Textures.ui_ability_healingrune, DescString := "ABILITY: Healing Rune\nSpawn a rune that can heal\nall teammate within its radius."}}
					set NewClassDescStructs += array{ClassDescBox{DescIcon := Textures.ui_ultimate, DescString := "ULTIMATE: Invincibility\nGranting invincibility to all\n teammate within radius."}}
					set NewClassDescStructs += array{ClassDescBox{DescIcon := Textures.ui_ability_no, DescString := "PASSIVE: Can't Swim"}}
				}
				FNClasses.CubeQueen =>
				{
					set NewClassDescStructs += array{ClassDescBox{DescIcon := Textures.ui_ability_beam, DescBGColor := NamedColors.Purple, DescString := "ABILITY: Cube Blast\nShoot laser beam, dealing \ncontinuous damage."}}
					set NewClassDescStructs += array{ClassDescBox{DescIcon := Textures.ui_ultimate, DescBGColor := NamedColors.Purple, DescString := "ULTIMATE: The End Event\nSummon the Mothership and \nnuke all enemies under it."}}
					set NewClassDescStructs += array{ClassDescBox{DescIcon := Textures.ui_ability_no, DescBGColor := NamedColors.Purple, DescString := "PASSIVE: Can't Swim"}}
				}
				FNClasses.ZygAndChoppy =>
				{
					set NewClassDescStructs += array{ClassDescBox{DescIcon := Textures.ui_ability_alienparasite, DescBGColor := NamedColors.Purple, DescString := "ABILITY: Alien Parasite\nConsume 20% HP and \ngrants low gravity effects."}}
					set NewClassDescStructs += array{ClassDescBox{DescIcon := Textures.ui_ultimate, DescBGColor := NamedColors.Purple, DescString := "ULTIMATE: Damage x2."}}
				}
				FNClasses.Zeus =>
				{
					set NewClassDescStructs += array{ClassDescBox{DescIcon := Textures.ui_ability_doublejump, DescBGColor := NamedColors.DarkBlue, DescString := "ABILITY: Leap\nDouble jump to leap.\nHold Crouch midair to ground pound."}}
					set NewClassDescStructs += array{ClassDescBox{DescIcon := Textures.ui_ultimate, DescBGColor := NamedColors.DarkBlue, DescString := "ULTIMATE: Olympus Event\nSummon Olympus' Statue and \nnuke all enemies within radius."}}
				}
				FNClasses.Ares =>
				{
					set NewClassDescStructs += array{ClassDescBox{DescIcon := Textures.ui_ability_damagebuff_ares, DescBGColor := NamedColors.DarkBlue, DescString := "ABILITY: Damage Buff\nIncreased all damage by 5%\n for 5 seconds."}}
					set NewClassDescStructs += array{ClassDescBox{DescIcon := Textures.ui_ultimate, DescBGColor := NamedColors.DarkBlue, DescString := "ULTIMATE: Olympus Event\nSummon Olympus' Statue and \nnuke all enemies within radius."}}
				}
				FNClasses.Cerberus =>
				{
					set NewClassDescStructs += array{ClassDescBox{DescIcon := Textures.ui_ability_dash, DescBGColor := NamedColors.DarkBlue, DescString := "ABILITY: Warp\nDash forward."}}
					set NewClassDescStructs += array{ClassDescBox{DescIcon := Textures.ui_ability_spinslash, DescBGColor := NamedColors.DarkBlue, DescString := "ABILITY: Spin Slash\ndeals continuous AoE damage."}}
					set NewClassDescStructs += array{ClassDescBox{DescIcon := Textures.ui_ultimate, DescBGColor := NamedColors.DarkBlue, DescString := "ULTIMATE: Olympus Event\nSummon Olympus' Statue and \nnuke all enemies within radius."}}
				}
				FNClasses.Hades =>
				{
					set NewClassDescStructs += array{ClassDescBox{DescIcon := Textures.ui_ability_hades_lifedrain, DescBGColor := NamedColors.DarkBlue, DescString := "ABILITY: Life Drain\nSiphon nearby enemies hp\nto you and your teammates."}}
					set NewClassDescStructs += array{ClassDescBox{DescIcon := Textures.ui_wpn_cardauthority, DescBGColor := NamedColors.DarkBlue, DescString := "CUSTOM ITEM: Card\nPress and hold to heal\nyour nearest teammate."}}
					set NewClassDescStructs += array{ClassDescBox{DescIcon := Textures.ui_ultimate, DescBGColor := NamedColors.DarkBlue, DescString := "ULTIMATE: Olympus Event\nSummon Olympus' Statue and \nnuke all enemies within radius"}}
					set NewClassDescStructs += array{ClassDescBox{DescIcon := Textures.ui_ability_siphon_hades, DescBGColor := NamedColors.DarkBlue, DescString := "PASSIVE: Siphon+"}}
				}
				FNClasses.MegaloDon =>
				{
					set NewClassDescStructs += array{ClassDescBox{DescIcon := Textures.ui_ability_charge, DescBGColor := NamedColors.DarkRed, DescString := "ABILITY: Charge\nDash forward and explode."}}
					set NewClassDescStructs += array{ClassDescBox{DescIcon := Textures.ui_ultimate, DescBGColor := NamedColors.DarkRed, DescString := "ULTIMATE: Leviathan Event\nDrive the Leviathan and\nnuke everything in its path"}}
				}
				FNClasses.RingmasterScarr =>
				{
					set NewClassDescStructs += array{ClassDescBox{DescIcon := Textures.ui_ability_damagebuff_scarr, DescBGColor := NamedColors.DarkRed, DescString := "ABILITY: Damage Buff\nIncreased all damage by 5%\n for 5 seconds."}}
					set NewClassDescStructs += array{ClassDescBox{DescIcon := Textures.ui_ultimate, DescBGColor := NamedColors.DarkRed, DescString := "ULTIMATE: Leviathan Event\nDrive the Leviathan and\nnuke everything in its path"}}
				}
				FNClasses.TheMachinist =>
				{
					set NewClassDescStructs += array{ClassDescBox{DescIcon := Textures.ui_ultimate, DescBGColor := NamedColors.DarkRed, DescString := "ULTIMATE: Leviathan Event\nDrive the Leviathan and\nnuke everything in its path"}}
					#set NewClassDescStructs += array{ClassDescBox{DescIcon := Textures.ui_ability_sentry, DescBGColor := NamedColors.DarkRed, DescString := "ABILITY: Spawn Turret"}}
					set NewClassDescStructs += array{ClassDescBox{DescIcon := Textures.ui_ability_shieldregen_machinist, DescBGColor := NamedColors.DarkRed, DescString := "PASSIVE: Shield Regen"}}
				}
				FNClasses.UltimaCarver =>
				{
					set NewClassDescStructs += array{ClassDescBox{DescIcon := Textures.ui_ultimate, DescBGColor := NamedColors.Purple, DescString := "ULTIMATE: Damage x2"}}
					#set NewClassDescStructs += array{ClassDescBox{DescIcon := Textures.ui_ability_no, DescBGColor := NamedColors.Purple, DescString := "ABILITY: Spawn Grind Vines"}}
					#set NewClassDescStructs += array{ClassDescBox{DescIcon := Textures.ui_ability_no, DescBGColor := NamedColors.Purple, DescString := "PASSIVE: Pumpkin"}}
				}
				FNClasses.ShogunX =>
				{
					set NewClassDescStructs += array{ClassDescBox{DescIcon := Textures.ui_ultimate, DescBGColor := NamedColors.DarkRed, DescString := "ULTIMATE: Spirit Slashes\nUnleashes slashes, dealing\ncontinuous AoE damage."}}
					set NewClassDescStructs += array{ClassDescBox{DescIcon := Textures.ui_ability_bombardment, DescBGColor := NamedColors.DarkRed, DescString := "ABILITY: Eruption\nCharge up a fiery AoE attack."}}
					#set NewClassDescStructs += array{ClassDescBox{DescIcon := Textures.ui_ability_no, DescBGColor := NamedColors.DarkRed, DescString := "PASSIVE: PLACEHOLDER"}}
				}
				FNClasses.NightRose =>
				{
					set NewClassDescStructs += array{ClassDescBox{DescIcon := Textures.ui_ability_demonclaw, DescBGColor := NamedColors.DarkRed, DescString := "ABILITY: Demon's Slash\nShort range melee attack."}}
					set NewClassDescStructs += array{ClassDescBox{DescIcon := Textures.ui_ability_puppeteer, DescBGColor := NamedColors.DarkRed, DescString := "ABILITY: Puppeteer\nRespawn as Puppeteer after dying."}}
					set NewClassDescStructs += array{ClassDescBox{DescIcon := Textures.ui_ultimate, DescBGColor := NamedColors.DarkRed, DescString := "ULTIMATE: Puppeteer's Wrath\nStun all enemies within range,\nstopping them from moving."}}
					#set NewClassDescStructs += array{ClassDescBox{DescIcon := Textures.ui_ability_no, DescBGColor := NamedColors.DarkRed, DescString := "PASSIVE: PLACEHOLDER"}}
				}
				FNClasses.FletcherKane =>
				{
					set NewClassDescStructs += array{ClassDescBox{DescIcon := Textures.ui_ability_demonclaw, DescBGColor := NamedColors.Gold, DescString := "ABILITY: Vicious Slash\nShort range melee attack."}}
					set NewClassDescStructs += array{ClassDescBox{DescIcon := Textures.ui_ultimate, DescBGColor := NamedColors.Gold, DescString := "ULTIMATE: The Mask\nShoot dark magic forwards, dealing \ncontinuous damage to all enemies in its path."}}
					set NewClassDescStructs += array{ClassDescBox{DescIcon := Textures.ui_ability_alphaenrage, DescBGColor := NamedColors.Gold, DescString := "PASSIVE: Unstoppable\nWhen low health,\nGrant x2 move speed."}}
				}
				FNClasses.AlphaHeavy =>
				{
					#set NewClassDescStructs += array{ClassDescBox{DescIcon := Textures.ui_ability_demonclaw, DescBGColor := NamedColors.Gold, DescString := "ABILITY: TBD\n......"}}
					set NewClassDescStructs += array{ClassDescBox{DescIcon := Textures.ui_ultimate, DescBGColor := NamedColors.Gold, DescString := "ULTIMATE: The Mask\nShoot dark magic forwards, dealing \ncontinuous damage to all enemies in its path."}}
					set NewClassDescStructs += array{ClassDescBox{DescIcon := Textures.class_alphaheavy, DescBGColor := NamedColors.Gold, DescString := "PASSIVE: Alpha Heavy\nGravity is a little heavier."}}
				}
				FNClasses.Kor =>
				{
					set NewClassDescStructs += array{ClassDescBox{DescIcon := Textures.ui_ability_demongrunt, DescBGColor := NamedColors.DarkRed, DescString := "ABILITY: Lieutenant\nSummon a Demon Grunt."}}
					set NewClassDescStructs += array{ClassDescBox{DescIcon := Textures.ui_ability_doublejump, DescBGColor := NamedColors.DarkRed, DescString := "ABILITY: Leap\nDouble jump to leap."}}
					set NewClassDescStructs += array{ClassDescBox{DescIcon := Textures.ui_ultimate, DescBGColor := NamedColors.DarkRed, DescString := "ULTIMATE: The Mask\nShoot dark magic forwards, dealing \ncontinuous damage to all enemies in its path."}}
					set NewClassDescStructs += array{ClassDescBox{DescIcon := Textures.ui_ability_kormedal, DescBGColor := NamedColors.DarkRed,  DescString := "PASSIVE: Shrouded Striker\nIncreased jump height."}}
				}
				FNClasses.Daigo =>
				{
					set NewClassDescStructs += array{ClassDescBox{DescIcon := Textures.ui_ability_beam, DescBGColor := NamedColors.DarkRed, DescString := "ABILITY: Focused Beam\nFires a long range laser beam."}}
					set NewClassDescStructs += array{ClassDescBox{DescIcon := Textures.ui_ability_doublejump, DescBGColor := NamedColors.DarkRed, DescString := "ABILITY: Double Jump\nHold Crouch midair to Spire Strike.\n"}}
					set NewClassDescStructs += array{ClassDescBox{DescIcon := Textures.ui_ability_dash, DescBGColor := NamedColors.DarkRed, DescString := "ABILITY: Cursed Rush\nHold Sprint and Jump midair\nto dash forward, dealing an AoE damage."}}
					#set NewClassDescStructs += array{ClassDescBox{DescIcon := Textures.ui_ultimate, DescBGColor := NamedColors.DarkRed, DescString := "ULTIMATE: Zero Blade\nTurn Splinter Blades into\nan instant-kill sword."}}
					set NewClassDescStructs += array{ClassDescBox{DescIcon := Textures.ui_ultimate, DescBGColor := NamedColors.DarkRed, DescString := "ULTIMATE: The Mask\nShoot dark magic forwards, dealing \ncontinuous damage to all enemies in its path."}}
					#set NewClassDescStructs += array{ClassDescBox{DescIcon := Textures.ui_ability_daigomedal, DescBGColor := NamedColors.DarkRed,  DescString := "PASSIVE: Infernal Defenses\nShield Regen."}}
				}
				FNClasses.Jones =>
				{
					set NewClassDescStructs += array{ClassDescBox{DescIcon := Textures.ui_ability_riftpistol, DescString := "ABILITY: Rift Sucker\nShoot a Rift, dealing \ncontinuous AoE damage."}}
					set NewClassDescStructs += array{ClassDescBox{DescIcon := Textures.ui_ultimate, DescString := "ULTIMATE: Big Bang\nSummon the Seven and \nnuke all enemies within radius."}}
				}
				FNClasses.DoctorDoom =>
				{
					set NewClassDescStructs += array{ClassDescBox{DescIcon := Textures.ui_ability_footdive, DescBGColor := NamedColors.DarkGreen, DescString := "ABILITY: Foot Dive\nTriple Jump to dive kick\nthe closest enemy."}}
					set NewClassDescStructs += array{ClassDescBox{DescIcon := Textures.ui_ability_mysticalbomb, DescBGColor := NamedColors.DarkGreen, DescString := "ABILITY: Mystical Bomb\nThrow an Explosive Bomb."}}
					set NewClassDescStructs += array{ClassDescBox{DescIcon := Textures.ui_ultimate, DescBGColor := NamedColors.DarkGreen, DescString := "ULTIMATE: Absolute Doom\nTab Ability - Gigabomb.\nHold Ability - Scorch Beam.\nPrimary Attack - Arcane Barrage."}}
					set NewClassDescStructs += array{ClassDescBox{DescIcon := Textures.ui_ability_siphon_doom, DescBGColor := NamedColors.DarkGreen, DescString := "PASSIVE: Siphon+"}}
				}
				FNClasses.Caretaker =>
				{
					set NewClassDescStructs += array{ClassDescBox{DescIcon := Textures.ui_ability_smash, DescBGColor := NamedColors.Purple, DescString := "Primary Attack: SMASH\nSmash the ground, dealing\n120 AoE damage."}}
					set NewClassDescStructs += array{ClassDescBox{DescIcon := Textures.ui_ability_snap, DescBGColor := NamedColors.Purple, DescString := "Alt Attack: Snap\nDeals 70 AoE damage."}}
					set NewClassDescStructs += array{ClassDescBox{DescIcon := Textures.ui_ability_beam, DescBGColor := NamedColors.Purple, DescString := "ABILITY: Cube Blast\nShoot laser beam, dealing \ncontinuous damage."}}
					set NewClassDescStructs += array{ClassDescBox{DescIcon := Textures.ui_ultimate, DescBGColor := NamedColors.Purple, DescString := "ULTIMATE: Sideways Domain\nProvides 'Damage x2' to you\nand nearby teammates."}}
				}
				<#FNClasses.Hope =>
				{
					set NewClassDescStructs += array{ClassDescBox{DescIcon := Textures.ui_ability_no, DescBGColor := NamedColors.Goldenrod, DescString := "ABILITY: Hopeful Heal\nProvides an AoE healing."}}
					set NewClassDescStructs += array{ClassDescBox{DescIcon := Textures.ui_ultimate, DescBGColor := NamedColors.Goldenrod, DescString := "ULTIMATE: Pandora's Chosen\nSummon Rift Beacons, nuking\nall enemeis within the beacons."}}
					#set NewClassDescStructs += array{ClassDescBox{DescIcon := Textures.ui_ability_no, DescBGColor := NamedColors.Goldenrod, DescString := "PASSIVE: Hope\nProvides healing."}}
				}#>
				_ =>
				{
				}
		}

		SetPlayerClass(ClassID)

		for (OldClassDesc : ClassDescArray)
		{
			StackBoxClassDesc.RemoveWidget(OldClassDesc)
		}
		set ClassDescArray = array{}

		for (NewClassDesc : NewClassDescStructs)
		{
			NewOverlay :=		overlay
								{
									Slots := array
									{
										overlay_slot
										{
											VerticalAlignment := vertical_alignment.Fill
											HorizontalAlignment := horizontal_alignment.Fill
											Widget := material_block
											{
												DefaultImage := Materials.UI.MI_UI_BG_White{}
												DefaultDesiredSize := vector2{X := 500.0, Y := 120.0}
												DefaultTint := NewClassDesc.DescBGColor
											}
										}
										overlay_slot
										{
											VerticalAlignment := vertical_alignment.Fill
											HorizontalAlignment := horizontal_alignment.Fill
											Padding := margin{Top := 0.0, Left := 0.0, Right := 0.0, Bottom := 0.0}
											Widget := stack_box
											{
												Orientation := orientation.Horizontal
												Slots := array
												{
													stack_box_slot
													{
														VerticalAlignment := vertical_alignment.Fill
														HorizontalAlignment := horizontal_alignment.Left
														Widget := texture_block
														{
															DefaultImage := NewClassDesc.DescIcon
															DefaultDesiredSize := vector2{X := 115.0, Y := 115.0}
														}
													}
													stack_box_slot
													{
														VerticalAlignment := vertical_alignment.Top
														HorizontalAlignment := horizontal_alignment.Left
														Padding := margin{Top := 0.0, Left := 10.0, Right := 9.0, Bottom := 0.0}
														Widget := text_block
														{
															DefaultText := UITextMessage(NewClassDesc.DescString)
															DefaultTextColor := NamedColors.White
															DefaultTextOpacity := 1.0
															DefaultShadowColor := NamedColors.Black
															DefaultShadowOffset := option{vector2{X := 0.0, Y := 0.0}}
															DefaultJustification := text_justification.Left
															DefaultOverflowPolicy := text_overflow_policy.Ellipsis
														}
													}
												}
											}
										}
									}
								}


			set ClassDescArray += array{NewOverlay}

			StackBoxClassDesc.AddWidget(
								stack_box_slot
								{
									Padding := margin{Top := 0.0, Left := 0.0, Right := 0.0, Bottom := 25.0}
									VerticalAlignment := vertical_alignment.Fill
									HorizontalAlignment := horizontal_alignment.Fill
									Widget := NewOverlay
								}
			)
		}

	}

	OnClassSelectedBtn(Message : widget_message, ClassID : FNClasses) : void =
	{
		if ( PlayerUI := GetPlayerUI[Message.Player], Agent := agent[Message.Player])
		{
			SetPlayerClass( ClassID )
			spawn{LinkedCreativeDevice.CloseClassMenu(Agent)}
			LinkedCreativeDevice.PlayClassStinger(Agent, ClassID)
		}
	}

	OnBtnChangeTeam(Message : widget_message) : void =
	{
		if ( PlayerUI := GetPlayerUI[Message.Player], Agent := agent[Message.Player])
		{
			ClickChangeTeam(Agent)
		}
	}

	ClickChangeTeam(Agent : agent) : void =
	{
		Print("ClickChangeTeam()")
		LinkedCreativeDevice.SwitchPlayerTeam( Agent )
		MoveClassBtn("Up")
	}

	OnBtnMainMenu(Message : widget_message) : void =
	{
		if ( PlayerUI := GetPlayerUI[Message.Player], Agent := agent[Message.Player])
		{
			ClickMainMenu(Agent)
		}
	}

	ClickMainMenu(Agent : agent) : void =
	{
		Print("ClickMainMenu()")
		spawn{LinkedCreativeDevice.CloseClassMenu(Agent)}
		if ( PlayerClasses := LinkedCreativeDevice.PlayerClassesDevice? )
		{
			spawn{PlayerClasses.MainMenuShow(Agent)}
		}
	}

	OnBtnBGM(Message : widget_message) : void =
	{
		if ( PlayerUI := GetPlayerUI[Message.Player], Agent := agent[Message.Player])
		{
			ToggleBGM(Agent)
		}
	}

	ToggleBGM(Agent : agent) : void =
	{
		if ( SaveManager := LinkedCreativeDevice.SaveManagerDevice? )
		{
			Print("ToggleBGM()")
			if ( SaveManager.IsBGMEnabled[Agent] )
			{
				LinkedCreativeDevice.SetBGMEnabled(Agent, false)
				UpdateUI()
				LinkedCreativeDevice.SetPlayerAggro(Agent, FNBGMState.None)
				Print("BGM: Enable")
			}
			else
			{
				LinkedCreativeDevice.SetBGMEnabled(Agent, true)
				UpdateUI()
				LinkedCreativeDevice.SetPlayerAggro(Agent, FNBGMState.Idle)
				Print("BGM: Disable")
			}
		}
	}

	UpdateUI<public>() : void =
	{
		if ( Agent := agent[LinkedPlayer] )
		{
			BtnMainMenu.SetEnabled(false)
			BtnBGM.SetEnabled(false)
			BtnChangeTeam.SetEnabled(false)

			if ( Agent.IsOnTeam[LinkedCreativeDevice, 0] )
			{
				ClassSelectBG.SetTint(NamedColors.CornflowerBlue)

				for ( BtnColumn->ClassBtn : ClassBtnMap )
				{
					for (BtnFNClass : ClassBtn, BtnClassData := BtnFNClass.ClassData, ClassID := BtnClassData.FNClassID )
					{
						BtnMatRaw := BtnFNClass.ButtonMat.GetImage()
						if ( BtnMat := Materials.UI.M_UI_ClassPortrait[BtnMatRaw] )
						{
							case( ClassID )
							{
								FNClasses.TNTina =>		set BtnMat.IconTexture = Textures.tntina_cursed
								FNClasses.Meowscles =>	set BtnMat.IconTexture = Textures.class_meowscles
								_ =>					#nothing
							}
						}
					}
				}
			}
			else
			{
				ClassSelectBG.SetTint(NamedColors.DarkRed)

				for ( BtnColumn->ClassBtn : ClassBtnMap )
				{
					for (BtnFNClass : ClassBtn, BtnClassData := BtnFNClass.ClassData, ClassID := BtnClassData.FNClassID )
					{
						BtnMatRaw := BtnFNClass.ButtonMat.GetImage()
						if ( BtnMat := Materials.UI.M_UI_ClassPortrait[BtnMatRaw] )
						{
							case( ClassID )
							{
								FNClasses.TNTina =>		set BtnMat.IconTexture = Textures.class_dynamotntina
								FNClasses.Meowscles =>	set BtnMat.IconTexture = Textures.class_meowdas
								_ =>					#nothing
							}
						}
					}
				}
			}

			if ( SaveManager := LinkedCreativeDevice.SaveManagerDevice?, SaveManager.IsBGMEnabled[Agent] )
			{
				BtnBGM.SetText(TextLocalization.Setting_MusicON)
			}
			else
			{
				BtnBGM.SetText(TextLocalization.Setting_MusicOFF)
			}
		}

		SelectClassBtn()
    }
}

# A Verse-authored creative device that can be placed in a level
ui_classmenu_manager_device := class(creative_device):

	# Class Selection UI
	var UIClassSelectionMap : [player]ui_classmenu = map{}

	@editable PlayerClassesDevice : ?player_classes_manager_device = false
	@editable SaveManagerDevice : ?save_manager_device = false

	@editable TriggerClassMenuOpen : trigger_device = trigger_device{}
	@editable TriggerClassMenuClose : trigger_device = trigger_device{}

	@editable SFX_Select : audio_player_device = audio_player_device{}
	@editable VO_SelectLoadout : audio_player_device = audio_player_device{}

	@editable InputForward : input_trigger_device = input_trigger_device{}
	@editable InputBackward : input_trigger_device = input_trigger_device{}
	@editable InputLeft : input_trigger_device = input_trigger_device{}
	@editable InputRight : input_trigger_device = input_trigger_device{}

	@editable CamClass			: gameplay_camera_orbit_device = gameplay_camera_orbit_device{}

	@editable Music_ClassSelectedC2S2	: audio_player_device = audio_player_device{}
	@editable Music_ClassSelectedC2S3	: audio_player_device = audio_player_device{}
	@editable Music_ClassSelectedC2S8	: audio_player_device = audio_player_device{}
	@editable Music_ClassSelectedIO		: audio_player_device = audio_player_device{}
	@editable Music_ClassSelectedC3S4	: audio_player_device = audio_player_device{}
	@editable Music_ClassSelectedC4S1	: audio_player_device = audio_player_device{}
	@editable Music_ClassSelectedC4S3	: audio_player_device = audio_player_device{}
	@editable Music_ClassSelectedC4S4	: audio_player_device = audio_player_device{}
	@editable Music_ClassSelectedOG		: audio_player_device = audio_player_device{}
	@editable Music_ClassSelectedC5S1	: audio_player_device = audio_player_device{}
	@editable Music_ClassSelectedC5S2	: audio_player_device = audio_player_device{}
	@editable Music_ClassSelectedC5S3	: audio_player_device = audio_player_device{}
	@editable Music_ClassSelectedC5S4	: audio_player_device = audio_player_device{}
	@editable Music_ClassSelectedC6S1	: audio_player_device = audio_player_device{}

	@editable MSGFeedJoinBlue 	: elimination_feed_device = elimination_feed_device{}
	@editable MSGFeedJoinRed 	: elimination_feed_device = elimination_feed_device{}

	@editable CamFP : gameplay_camera_first_person_device = gameplay_camera_first_person_device{}

    # Runs when the device is started in a running game
    OnBegin<override>()<suspends>:void=
	{
		Print("Initialize UI Class Menu...")

        TriggerClassMenuOpen.TriggeredEvent.Subscribe(OnClassSelectionOpened)

		InputForward.PressedEvent.Subscribe(InputPressedForward)
		InputBackward.PressedEvent.Subscribe(InputPressedBackward)
		InputLeft.PressedEvent.Subscribe(InputPressedLeft)
		InputRight.PressedEvent.Subscribe(InputPressedRight)
	}

	AwaitPlayerInUI(Agent:agent)<suspends>:void=
	{
		if (FortCharacter := Agent.GetFortCharacter[])
		{
			#Print("AwaitPlayerInUI")
			race
			{
				block
				{
					Result := FortCharacter.EliminatedEvent().Await()
					UIPlayerEliminated(Result)
				}
				block
				{
					loop
					{
						MaybeAgent := TriggerClassMenuClose.TriggeredEvent.Await()
						if (CloseAgent := MaybeAgent?, CloseAgent = Agent)
						{
							break
						}
					}
					Print("AwaitPlayerInUI Cancel.")
				}
				<#block
				{
					Sleep(60.0)
					Print("AwaitPlayerInUI Time Out!")
					spawn{CloseClassMenu(Agent)}
				}#>
			}
		}
	}

	# Removes an element from the given map and returns a new map without that element
	(Map:[keytype]valuetype where keytype:subtype(comparable), valuetype:type).RemoveKey(ElementToRemove:keytype):[keytype]valuetype=
	{
		var NewMap:[keytype]valuetype = map{}
		for (Key -> Value : Map, Key <> ElementToRemove)
		{
			set NewMap = ConcatenateMaps(NewMap, map{Key => Value})
		}

		return NewMap
	}

	<#UIPlayerJoined(Agent: agent): void =
	{
		Print("UIPlayerJoined")
		if (FortCharacter := Agent.GetFortCharacter[])
		{
			FortCharacter.EliminatedEvent().Subscribe(UIPlayerEliminated)
		}
	}#>

    UIPlayerRemoved(Agent: agent): void =
	{
		Print("UIPlayerRemoved")
		if (AgentPlayer := player[Agent])
		{
			set UIClassSelectionMap = UIClassSelectionMap.RemoveKey(AgentPlayer)
		}
	}

	UIPlayerEliminated(Result:elimination_result):void=
	{
		Print("UIPlayerEliminated")
		KilledCharacter := Result.EliminatedCharacter
		if ( KilledAgent := KilledCharacter.GetAgent[] )
		{
			spawn{CloseClassMenu(KilledAgent)}
		}
	}

    OpenClassMenu(Agent:agent):void=
	{
		if (InPlayer := player[Agent])
		{
			var OptPlayerUI :?ui_classmenu = false

			if (ThisPlayerUI := UIClassSelectionMap[InPlayer])
			{
				set OptPlayerUI = option{ThisPlayerUI}
			}
			else if (ThisPlayerUI := set UIClassSelectionMap[InPlayer] = ui_classmenu{ LinkedPlayer := InPlayer, LinkedCreativeDevice := Self } ) 
			{
				ThisPlayerUI.Init()
				set OptPlayerUI = option{ThisPlayerUI}
			}

			if (PlayerUI := OptPlayerUI?)
			{
				PlayerUI.OpenUI()

				if ( SaveManager := SaveManagerDevice?, SaveManager.IsCamFPEnabled[Agent] )
				{
					CamFP.RemoveFrom(Agent)
				}

				InputForward.Register(Agent)
				InputBackward.Register(Agent)
				InputLeft.Register(Agent)
				InputRight.Register(Agent)
				CamClass.AddTo(Agent)

				if ( PlayerClasses := PlayerClassesDevice? )
				{
					PlayerClasses.HideHUD(Agent)
				}

				if ( FortCharacter := Agent.GetFortCharacter[] )
				{
					FortCharacter.PutInStasis(stasis_args{AllowTurning := false, AllowFalling := false, AllowEmotes := true})
				}
				#spawn{AwaitPlayerInUI(Agent)} #not needed anymore
			}
			else
			{
				Print("Error opening ui_classmenu")
			}

			VO_SelectLoadout.Play(Agent)
		}
	}

    CloseClassMenu(Agent:agent)<suspends>:void=
	{
		if ( InPlayer := player[Agent] )
		{
			if ( UIClassMenu := UIClassSelectionMap[InPlayer] )
			{
				spawn{UIClassMenu.CloseUI()}
			}

			InputForward.Unregister(Agent)
			InputBackward.Unregister(Agent)
			InputLeft.Unregister(Agent)
			InputRight.Unregister(Agent)
			CamClass.RemoveFrom(Agent)

			if ( PlayerClasses := PlayerClassesDevice? )
			{
				PlayerClasses.ShowHUD(Agent)
			}

			if ( FortCharacter := Agent.GetFortCharacter[] )
			{
				FortCharacter.ReleaseFromStasis()

				if ( FortCharacter.IsActive[], InPlayer.IsActive[] )
				{
					TriggerClassMenuClose.Trigger(Agent) # can cause verse internal error.
				}
			}

			if ( SaveManager := SaveManagerDevice? )
			{
				if ( SaveManager.IsCamFPEnabled[Agent] )
				{
					CamFP.AddTo(Agent)
				}
				else
				{
					CamFP.RemoveFrom(Agent)
				}
			}
		}
	}

    OnClassSelectionOpened(MaybeAgent:?agent):void=
	{
		if (Agent := MaybeAgent?)
		{
			OpenClassMenu(Agent)
		}
	}

	SwitchPlayerTeam(Agent:agent):void=
	{
		spawn{CloseClassMenu(Agent)}
		TeamCollection := GetPlayspace().GetTeamCollection()
		Teams := TeamCollection.GetTeams()
		if ( TeamCollection.IsOnTeam[Agent, Teams[0]] )
		{
			if ( TeamCollection.AddToTeam[Agent, Teams[1]] )
			{
				if (FortCharacter := Agent.GetFortCharacter[])
				{
					FortCharacter.SetVulnerability(true)
					FortCharacter.SetShield(0.0)
					FortCharacter.SetHealth(0.0)
					FortCharacter.Damage(1.0)
				}
				MSGFeedJoinRed.Activate(Agent)
			}
		}
		else
		{
			if ( TeamCollection.AddToTeam[Agent, Teams[0]] )
			{
				if (FortCharacter := Agent.GetFortCharacter[])
				{
					FortCharacter.SetVulnerability(true)
					FortCharacter.SetShield(0.0)
					FortCharacter.SetHealth(0.0)
					FortCharacter.Damage(1.0)
				}
				MSGFeedJoinBlue.Activate(Agent)
			}
		}
	}

	SetPlayerClass(Agent:agent, eClass:FNClasses):void=
	{
		if ( PlayerClasses := PlayerClassesDevice? )
		{
			PlayerClasses.SetPlayerClass(Agent,eClass)
		}
	}

	SetPlayerAggro(Agent:agent, BGMState:FNBGMState):void=
	{
		if ( PlayerClasses := PlayerClassesDevice? )
		{
			PlayerClasses.SetPlayerAggro(Agent,BGMState)
		}
	}

	SetBGMEnabled(Agent:agent, bEnable:logic):void=
	{
		if ( SaveManager := SaveManagerDevice? )
		{
			SaveManager.ToggleSettingBGM(Agent, bEnable)
		}
	}

	GetAllClasses():[]fn_class_data=
	{
		var AllClassesArray : []fn_class_data = array{}
		if ( PlayerClasses := PlayerClassesDevice? )
		{
			set AllClassesArray = PlayerClasses.GetAllClasses()
		}
		return AllClassesArray
	}

	InputPressedForward(Agent : agent) : void =
	{
		if ( InPlayer := player[Agent], UIClassMenu := UIClassSelectionMap[InPlayer], UIClassMenu.IsOpen? )
		{
			UIClassMenu.MoveClassBtn("Up")
		}
	}

	InputPressedBackward(Agent : agent) : void =
	{
		if ( InPlayer := player[Agent], UIClassMenu := UIClassSelectionMap[InPlayer], UIClassMenu.IsOpen? )
		{
			UIClassMenu.MoveClassBtn("Down")
		}
	}

	InputPressedLeft(Agent : agent) : void =
	{
		if ( InPlayer := player[Agent], UIClassMenu := UIClassSelectionMap[InPlayer], UIClassMenu.IsOpen? )
		{
			UIClassMenu.MoveClassBtn("Left")
		}
	}

	InputPressedRight(Agent : agent) : void =
	{
		if ( InPlayer := player[Agent], UIClassMenu := UIClassSelectionMap[InPlayer], UIClassMenu.IsOpen? )
		{
			UIClassMenu.MoveClassBtn("Right")
		}
	}

	InputPressedFireUI(Agent : agent) : void =
	{
		if ( InPlayer := player[Agent], UIClassMenu := UIClassSelectionMap[InPlayer], UIClassMenu.IsOpen? )
		{
			UIClassMenu.SelectButton()
		}
	}

	PlayClassStinger(Agent:agent, eClass:FNClasses):void=
	{
		case( eClass )
		{
			FNClasses.Midas =>
			{
				Music_ClassSelectedC2S2.Play(Agent)
			}
			FNClasses.TNTina =>
			{
				Music_ClassSelectedC2S2.Play(Agent)
			}
			FNClasses.Brutus =>
			{
				Music_ClassSelectedC2S2.Play(Agent)
			}
			FNClasses.Skye =>
			{
				Music_ClassSelectedC2S2.Play(Agent)
			}
			FNClasses.Meowscles =>
			{
				Music_ClassSelectedC2S2.Play(Agent)
			}
			FNClasses.Jules =>
			{
				Music_ClassSelectedC2S3.Play(Agent)
			}
			FNClasses.Kit =>
			{
				Music_ClassSelectedC2S3.Play(Agent)
			}
			FNClasses.Ocean =>
			{
				Music_ClassSelectedC2S3.Play(Agent)
			}
			FNClasses.ShadowMidas =>
			{
				Music_ClassSelectedC2S8.Play(Agent)
			}
			FNClasses.WildguardRelik =>
			{
				Music_ClassSelectedC4S3.Play(Agent)
			}
			FNClasses.TheAgelessChampion =>
			{
				Music_ClassSelectedC4S1.Play(Agent)
			}
			FNClasses.KadoThorne =>
			{
				Music_ClassSelectedC4S4.Play(Agent)
			}
			FNClasses.TheDealer =>
			{
				Music_ClassSelectedC4S4.Play(Agent)
			}
			FNClasses.Raz =>
			{
				Music_ClassSelectedC2S8.Play(Agent)
			}
			FNClasses.Slone =>
			{
				Music_ClassSelectedIO.Play(Agent)
			}
			FNClasses.TheFoundation =>
			{
				Music_ClassSelectedIO.Play(Agent)
			}
			FNClasses.Gunnar =>
			{
				Music_ClassSelectedIO.Play(Agent)
			}
			FNClasses.HuntmasterSaber =>
			{
				Music_ClassSelectedIO.Play(Agent)
			}
			FNClasses.TheHerald =>
			{
				Music_ClassSelectedC3S4.Play(Agent)
			}
			FNClasses.TheIceKing =>
			{
				Music_ClassSelectedOG.Play(Agent)
			}
			FNClasses.CubeQueen =>
			{
				Music_ClassSelectedC2S8.Play(Agent)
			}
			FNClasses.Valeria =>
			{
				Music_ClassSelectedC5S1.Play(Agent)
			}
			FNClasses.Oscar =>
			{
				Music_ClassSelectedC5S1.Play(Agent)
			}
			FNClasses.Nisha =>
			{
				Music_ClassSelectedC5S1.Play(Agent)
			}
			FNClasses.Montague =>
			{
				Music_ClassSelectedC5S1.Play(Agent)
			}
			FNClasses.SgtWinter =>
			{
				Music_ClassSelectedOG.Play(Agent)
			}
			FNClasses.CubeAssassin =>
			{
				Music_ClassSelectedC2S8.Play(Agent)
			}
			FNClasses.SpireAssassin =>
			{
				Music_ClassSelectedC2S8.Play(Agent)
			}
			FNClasses.ZygAndChoppy =>
			{
				Music_ClassSelectedC2S8.Play(Agent)
			}
			FNClasses.CubeBlue =>
			{
				Music_ClassSelectedC2S8.Play(Agent)
			}
			FNClasses.Zeus =>
			{
				Music_ClassSelectedC5S2.Play(Agent)
			}
			FNClasses.Ares =>
			{
				Music_ClassSelectedC5S2.Play(Agent)
			}
			FNClasses.Cerberus =>
			{
				Music_ClassSelectedC5S2.Play(Agent)
			}
			FNClasses.Hades =>
			{
				Music_ClassSelectedC5S2.Play(Agent)
			}
			FNClasses.MegaloDon =>
			{
				Music_ClassSelectedC5S3.Play(Agent)
			}
			FNClasses.RingmasterScarr =>
			{
				Music_ClassSelectedC5S3.Play(Agent)
			}
			FNClasses.TheMachinist =>
			{
				Music_ClassSelectedC5S3.Play(Agent)
			}
			FNClasses.DoctorDoom =>
			{
				Music_ClassSelectedC5S4.Play(Agent)
			}
			FNClasses.UltimaCarver =>
			{
				Music_ClassSelectedC2S8.Play(Agent)
			}
			FNClasses.ShogunX =>
			{
				Music_ClassSelectedC6S1.Play(Agent)
			}
			FNClasses.NightRose =>
			{
				Music_ClassSelectedC6S1.Play(Agent)
			}
			FNClasses.FletcherKane =>
			{
				#Music_ClassSelectedC6S2.Play(Agent)
			}
			FNClasses.AlphaHeavy =>
			{
				#Music_ClassSelectedC6S2.Play(Agent)
			}
			FNClasses.Daigo =>
			{
				Music_ClassSelectedC6S1.Play(Agent)
			}
			FNClasses.Kor =>
			{
				Music_ClassSelectedC6S1.Play(Agent)
			}
			FNClasses.Jones =>
			{
				Music_ClassSelectedOG.Play(Agent)
			}
			_ =>
			{
				#nothing
			}
		}
	}
