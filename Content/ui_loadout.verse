<#
	2AGENCY
	UI - Loadout
#>
using { /Fortnite.com/Devices }
using { /Fortnite.com/Characters }
using { /Fortnite.com/Teams }
using { /Fortnite.com/UI }
using { /Fortnite.com/Game }
using { /Verse.org }
using { /Verse.org/Assets }
using { /Verse.org/Simulation }
using { /Verse.org/Simulation/Tags }
using { /Verse.org/Colors }
using { /Verse.org/Random }
using { /UnrealEngine.com/Temporary/Diagnostics }
using { /UnrealEngine.com/Temporary/UI }
using { /UnrealEngine.com/Temporary/SpatialMath }
using { ModuleUIExtra }

# Notes!!!
# AddWidget, RemoveWidget is evil and can cause big lags.

	FNRarity<public> := enum
	{
		Common,
		Uncommon,
		Rare,
		Epic,
		Legendary,
		Mythic
	}

button_itemslot := struct
{
    ButtonQuiet<public> : button_quiet
    ItemData<public> : item_box_data
}

item_box_data := struct
{
	ItemID<public> : int = 0
	Icon<public> : texture = Textures.midas_cursed1
	Rarity<public> : FNRarity = FNRarity.Common
	VFXDevice<public> : visual_effect_powerup_device = visual_effect_powerup_device{}
}

ui_loadout := class() 
{
    LinkedCreativeDevice<public> : loadout_device
    LinkedPlayer<public> : player 

    var IsInitialized<private> : logic = false
    var MainCanvas<private> : canvas = canvas{}

    var IsOpen<public> : logic = false

	UITextMessage<localizes>(Value : string) : message = "{Value}"

	UIBG: color_block = color_block{DefaultColor := NamedColors.Gray, DefaultOpacity := 0.95}
	UITitleText: text_block = text_block{DefaultTextColor := NamedColors.White, DefaultShadowColor := NamedColors.Gray, DefaultShadowOffset := option{vector2{X := 2.0, Y := 2.0}} }
	UITitleTextDesc: text_block = text_block{DefaultTextColor := NamedColors.White, DefaultShadowColor := NamedColors.Orange, DefaultShadowOffset := option{vector2{X := 2.0, Y := 2.0}} }
    StackBoxItemList : stack_box = stack_box{Orientation := orientation.Vertical}

    BtnExit : button_loud = button_loud{}

    var iSelectedBtnColumn : int = 0
    var iSelectedBtnRow : int = 0
	var ItemBtnMap: [int][]button_itemslot = map{}

    Init<public>() : void =
	{
        if (IsInitialized?)
		{
            return
		}

		set iSelectedBtnColumn = 0
		set iSelectedBtnRow = 0

		set MainCanvas = CreateUI()
		set IsInitialized = true
    }

    OpenUI<public>() : void =
	{
        if (PlayerUI := GetPlayerUI[LinkedPlayer])
		{
			PlayerUI.AddWidget(GetCanvas(), player_ui_slot{InputMode := ui_input_mode.None})

			set iSelectedBtnColumn = 0
			set iSelectedBtnRow = 0

			UpdateUI()
			set IsOpen = true
		}
	}

    CloseUI<public>()<suspends> : void =
	{
        if (PlayerUI := GetPlayerUI[LinkedPlayer])
		{
			PlayerUI.RemoveWidget(GetCanvas())
			set IsOpen = false
		}
    }

    GetCanvas<public>() : canvas = MainCanvas  # Exposes the class Canvas to public

	CreateUI<private>() : canvas =
	{
		UICanvas : canvas = canvas{Slots := array{
		canvas_slot
		{
			ZOrder := {0}
			Anchors := anchors{Minimum := vector2{X := 0.0, Y := 0.0}, Maximum := vector2{X := 0.5, Y := 1.0}}
			Offsets := margin{Top := 0.0, Left := 0.0, Right := 0.0, Bottom := 0.0}
			Alignment := vector2{X := 0.0, Y := 0.0}
			SizeToContent := true
			Widget := UIBG
        }
		canvas_slot
		{
			ZOrder := {1}
			Anchors := anchors{Minimum := vector2{X := 0.25, Y := 0.025}, Maximum := vector2{X := 0.25, Y := 0.025} }
			Offsets := margin{Top := 0.0, Left := 0.0, Right := 0.0, Bottom := 0.0}
			Alignment := vector2{X := 0.5, Y := 0.0}
			SizeToContent := true
			Widget := UITitleText
        }
		canvas_slot
		{
			ZOrder := {1}
			Anchors := anchors{Minimum := vector2{X := 0.75, Y := 0.05}, Maximum := vector2{X := 0.75, Y := 0.05} }
			Offsets := margin{Top := 0.0, Left := 0.0, Right := 0.0, Bottom := 0.0}
			Alignment := vector2{X := 0.5, Y := 0.0}
			SizeToContent := true
			Widget := UITitleTextDesc
        }
		canvas_slot
		{
			ZOrder := {1}
			Anchors := anchors{Minimum := vector2{X := 0.0, Y := 0.1}, Maximum := vector2{X := 0.0, Y := 0.1}}
			Offsets := margin{ Top:=0.0, Left:= 50.0, Right:=0.0, Bottom := 0.0 }
			Alignment := vector2{X := 0.0, Y := 0.0}
			SizeToContent := true
			Widget := StackBoxItemList
        }
		canvas_slot
		{
			ZOrder := {1}
			Anchors := anchors{Minimum := vector2{X := 0.0, Y := 1.0}, Maximum := vector2{X := 0.0, Y := 1.0} }
			Offsets := margin{Top := 0.0, Left := 0.0, Right := 0.0, Bottom := 0.0}
			Alignment := vector2{X := 0.0, Y := 1.0}
			SizeToContent := true
			Widget := stack_box{
				Orientation := orientation.Horizontal
				Slots := array{
					stack_box_slot
					{
						VerticalAlignment := vertical_alignment.Fill
						HorizontalAlignment := horizontal_alignment.Fill
						#Padding := margin{Top := 0.0, Left := 0.0, Right := 0.0, Bottom := 0.0}
						Widget := BtnExit
					}
				}
			}
		}
		}}

		UITitleText.SetText(UITextMessage("Loadout: Hat Aura"))
		UITitleTextDesc.SetText(UITextMessage(""))

		BtnExit.SetText(UITextMessage("Exit"))

		UITitleText.SetShadowOpacity(1.0)
		UITitleTextDesc.SetShadowOpacity(1.0)

		BtnExit.OnClick().Subscribe(OnBtnExit)

		RowStackBoxes : []stack_box = for ( I := 0..6 ) {stack_box{Orientation := orientation.Horizontal}}
		var StackRow : int = 0
		var StackColumn : int = 0
		var BtnArray : []button_itemslot = array{}

		var RarityColor : color = NamedColors.Gray

		# create button for each items
		AllItemList := LinkedCreativeDevice.GetAllItems()
		for (NewItem : AllItemList, ItemID := NewItem.ItemID )
		{
			case( NewItem.Rarity )
			{
				FNRarity.Common =>		set RarityColor = NamedColors.Gray 
				FNRarity.Uncommon =>	set RarityColor = NamedColors.LightGreen 
				FNRarity.Rare =>		set RarityColor = NamedColors.Aqua 
				FNRarity.Epic =>		set RarityColor = NamedColors.Purple 
				FNRarity.Legendary =>	set RarityColor = NamedColors.Orange 
				FNRarity.Mythic =>		set RarityColor = NamedColors.Gold 
			}
			
			
			
			NewBtn :=	button_quiet{}
			NewOverlay :=			overlay
									{
										Slots := array
										{
											overlay_slot
											{
												VerticalAlignment := vertical_alignment.Fill
												HorizontalAlignment := horizontal_alignment.Fill
												Widget := color_block
												{
													DefaultColor := RarityColor
													DefaultDesiredSize := vector2{X := 140.0, Y := 160.0}
													DefaultOpacity := 0.85
												}
											}
											overlay_slot
											{
												VerticalAlignment := vertical_alignment.Center
												Widget := texture_block
												{
													DefaultImage := NewItem.Icon
													DefaultDesiredSize := vector2{X := 140.0, Y := 140.0}
												}
											}
											overlay_slot
											{
												VerticalAlignment := vertical_alignment.Bottom
												HorizontalAlignment := horizontal_alignment.Fill
												Padding := margin{Top := 0.0, Left := 10.0, Right := 10.0, Bottom := 0.0}
												Widget := text_block
												{
													DefaultText := UITextMessage(LinkedCreativeDevice.GetItemName(ItemID))
													DefaultTextColor := NamedColors.Black
													DefaultTextOpacity := 1.0
													DefaultShadowColor := NamedColors.White
													DefaultShadowOffset := option{vector2{X := 0.0, Y := 0.0}}
													DefaultJustification := text_justification.Center
												}
											}
											overlay_slot
											{
												VerticalAlignment := vertical_alignment.Fill
												HorizontalAlignment := horizontal_alignment.Fill
												Widget := NewBtn
											}
										}
									}

			if (StackBox := RowStackBoxes[StackColumn])
			{
				StackBox.AddWidget(
									stack_box_slot
									{
										VerticalAlignment := vertical_alignment.Top
										HorizontalAlignment := horizontal_alignment.Left
										Padding := margin{Top := 0.0, Left := 0.0, Right := 0.0, Bottom := 0.0}
										Widget := NewOverlay
									}

				)
			}

			NewBtn.OnClick().SubscribeWidget(OnItemSelectedBtn, ItemID)
			NewFNBtn := button_itemslot{ ButtonQuiet := NewBtn, ItemData := NewItem}

			set BtnArray += array{NewFNBtn}
			if { set ItemBtnMap[StackColumn] = BtnArray }

			set StackRow += 1
			if ( StackRow >= 6 ) # number of buttons per column
			{
				# new column
				set StackColumn += 1
				set BtnArray = array{}
				set StackRow = 0
			}
		}

		# add stuff into column
		for (RowStackBox : RowStackBoxes)
		{
			StackBoxItemList.AddWidget(
								stack_box_slot
								{
									VerticalAlignment := vertical_alignment.Top
									HorizontalAlignment := horizontal_alignment.Left
									Padding := margin{Top := 0.0, Left := 0.0, Right := 0.0, Bottom := 0.0}
									Widget := RowStackBox
								}
			)
		}

		return UICanvas
	}

	MoveItemBtn(BtnMovement:string) : void =
	{
		case( BtnMovement )
		{
			"Up" =>
			{
				set iSelectedBtnColumn -= 1
			}
			"Down" =>
			{
				set iSelectedBtnColumn += 1
			}
			"Left" =>
			{
				set iSelectedBtnRow -= 1
			}
			"Right" =>
			{
				set iSelectedBtnRow += 1
			}
			_ =>
			{
				set iSelectedBtnColumn = 0
				set iSelectedBtnRow = 0
			}
		}

		if ( iSelectedBtnColumn = ItemBtnMap.Length )
		{
			set iSelectedBtnRow = 0
		}
		else
		{
			if ( iSelectedBtnRow > ( 6 -1 ) )
			{
				set iSelectedBtnRow -= 1
			}

			if ( iSelectedBtnRow < 0 )
			{
				set iSelectedBtnRow = 0
			}
		}

		if ( iSelectedBtnColumn > ItemBtnMap.Length )
		{
			set iSelectedBtnColumn -= 1
		}

		if ( iSelectedBtnColumn < 0 )
		{
			set iSelectedBtnColumn = 0
		}

		SelectItemBtn()
	}

	SelectItemBtn() : void =
	{
		if ( Agent := agent[LinkedPlayer] )
		{
			for ( BtnColumn->ItemBtn : ItemBtnMap )
			{
				for (BtnFNClass : ItemBtn, BtnQuiet := BtnFNClass.ButtonQuiet )
				{
					BtnQuiet.SetEnabled(false)
				}
			}

			BtnExit.SetEnabled(false)

			if ( SelectedBtns := ItemBtnMap[iSelectedBtnColumn], BtnFNClass := SelectedBtns[iSelectedBtnRow], BtnQuiet := BtnFNClass.ButtonQuiet, BtnItemData := BtnFNClass.ItemData, BtnItemID := BtnItemData.ItemID )
			{
				BtnQuiet.SetEnabled(true)
				spawn{LinkedCreativeDevice.AddUnusual(Agent, BtnItemID)}
				UITitleTextDesc.SetText(UITextMessage(LinkedCreativeDevice.GetItemName(BtnItemID)))
			}

			if ( iSelectedBtnColumn = ItemBtnMap.Length )
			{
				case( iSelectedBtnRow )
				{
					0 => BtnExit.SetEnabled(true)
					_ =>
				}
			}

			LinkedCreativeDevice.SFX_Select.Play(Agent)
		}
	}

	SelectButton() : void =
	{
		if ( Agent := agent[LinkedPlayer] )
		{
			if ( iSelectedBtnColumn = ItemBtnMap.Length )
			{
				case( iSelectedBtnRow )
				{
					0 =>	ClickExit(Agent)
					_ =>	
				}
			}
			else
			{
				if ( SelectedBtns := ItemBtnMap[iSelectedBtnColumn], BtnFNClass := SelectedBtns[iSelectedBtnRow], BtnItemData := BtnFNClass.ItemData, ItemID := BtnItemData.ItemID )
				{
					AddUnusual(ItemID)
					ClickExit(Agent)
				}
			}
		}
	}

	AddUnusual(ItemID:int) : void =
	{
		if ( Agent := agent[LinkedPlayer] )
		{
			spawn{LinkedCreativeDevice.AddUnusual(Agent, ItemID)}
		}
	}

	OnItemSelectedBtn(Message : widget_message, ItemID : int) : void =
	{
		if ( PlayerUI := GetPlayerUI[Message.Player], Agent := agent[Message.Player])
		{
			AddUnusual( ItemID )
			ClickExit(Agent)
		}
	}

	OnBtnExit(Message : widget_message) : void =
	{
		if ( PlayerUI := GetPlayerUI[Message.Player], Agent := agent[Message.Player])
		{
			ClickExit(Agent)
		}
	}

	ClickExit(Agent : agent) : void =
	{
		Print("ClickExit()")
		spawn{LinkedCreativeDevice.CloseUI(Agent)}
		if ( PlayerClasses := LinkedCreativeDevice.PlayerClassesDevice? )
		{
			spawn{PlayerClasses.MainMenuShow(Agent)}
		}
	}

	UpdateUI<public>() : void =
	{
		SelectItemBtn()
    }
}

# A Verse-authored creative device that can be placed in a level
loadout_device := class(creative_device):

	var PlayerUIMap : [player]ui_loadout = map{}

	@editable PlayerClassesDevice : ?player_classes_manager_device = false
	@editable SaveManagerDevice : ?save_manager_device = false

	var ItemsWeight : float = 0.0

	@editable SFX_Select : audio_player_device = audio_player_device{}

	@editable InputForward : input_trigger_device = input_trigger_device{}
	@editable InputBackward : input_trigger_device = input_trigger_device{}
	@editable InputLeft : input_trigger_device = input_trigger_device{}
	@editable InputRight : input_trigger_device = input_trigger_device{}

	@editable CamClass			: gameplay_camera_orbit_device = gameplay_camera_orbit_device{}

	@editable AnalyticDropUncommon : analytics_device = analytics_device{}
	@editable AnalyticDropRare : analytics_device = analytics_device{}
	@editable AnalyticDropEpic : analytics_device = analytics_device{}
	@editable AnalyticDropLegendary : analytics_device = analytics_device{}
	@editable AnalyticDropMythic : analytics_device = analytics_device{}

	@editable VFXUnusual0 : visual_effect_powerup_device = visual_effect_powerup_device{}
	@editable VFXUnusual1 : visual_effect_powerup_device = visual_effect_powerup_device{}
	@editable VFXUnusual2 : visual_effect_powerup_device = visual_effect_powerup_device{}
	@editable VFXUnusual3 : visual_effect_powerup_device = visual_effect_powerup_device{}
	@editable VFXUnusual4 : visual_effect_powerup_device = visual_effect_powerup_device{}
	@editable VFXUnusual5 : visual_effect_powerup_device = visual_effect_powerup_device{}
	@editable VFXUnusual6 : visual_effect_powerup_device = visual_effect_powerup_device{}
	@editable VFXUnusual7 : visual_effect_powerup_device = visual_effect_powerup_device{}
	@editable VFXUnusual8 : visual_effect_powerup_device = visual_effect_powerup_device{}
	@editable VFXUnusual9 : visual_effect_powerup_device = visual_effect_powerup_device{}
	@editable VFXUnusual10 : visual_effect_powerup_device = visual_effect_powerup_device{}

    # Runs when the device is started in a running game
    OnBegin<override>()<suspends>:void=
	{
		Print("Initialize UI Loadout...")

		AllItemList := GetAllItems()
		set ItemsWeight = 0.0
		for (NewItem : AllItemList)
		{
			set ItemsWeight += GetItemWeight(NewItem)
		}

		InputForward.PressedEvent.Subscribe(InputPressedForward)
		InputBackward.PressedEvent.Subscribe(InputPressedBackward)
		InputLeft.PressedEvent.Subscribe(InputPressedLeft)
		InputRight.PressedEvent.Subscribe(InputPressedRight)
	}

	# Removes an element from the given map and returns a new map without that element
	(Map:[keytype]valuetype where keytype:subtype(comparable), valuetype:type).RemoveKey(ElementToRemove:keytype):[keytype]valuetype=
	{
		var NewMap:[keytype]valuetype = map{}
		for (Key -> Value : Map, Key <> ElementToRemove)
		{
			set NewMap = ConcatenateMaps(NewMap, map{Key => Value})
		}

		return NewMap
	}

    UIPlayerRemoved(Agent: agent): void =
	{
		Print("UIPlayerRemoved")
		if (AgentPlayer := player[Agent])
		{
			set PlayerUIMap = PlayerUIMap.RemoveKey(AgentPlayer)
		}
	}

	UIPlayerEliminated(Result:elimination_result):void=
	{
		Print("UIPlayerEliminated")
		KilledCharacter := Result.EliminatedCharacter
		if ( KilledAgent := KilledCharacter.GetAgent[] )
		{
			spawn{CloseUI(KilledAgent)}
		}
	}

    OpenUI(Agent:agent):void=
	{
		if (InPlayer := player[Agent])
		{
			var OptPlayerUI :?ui_loadout = false

			if (ThisPlayerUI := PlayerUIMap[InPlayer])
			{
				set OptPlayerUI = option{ThisPlayerUI}
			}
			else if (ThisPlayerUI := set PlayerUIMap[InPlayer] = ui_loadout{ LinkedPlayer := InPlayer, LinkedCreativeDevice := Self } ) 
			{
				ThisPlayerUI.Init()
				set OptPlayerUI = option{ThisPlayerUI}
			}

			if (PlayerUI := OptPlayerUI?)
			{
				PlayerUI.OpenUI()

				InputForward.Register(Agent)
				InputBackward.Register(Agent)
				InputLeft.Register(Agent)
				InputRight.Register(Agent)
				CamClass.AddTo(Agent)

				if ( PlayerClasses := PlayerClassesDevice? )
				{
					PlayerClasses.HideHUD(Agent)
				}

				if ( FortCharacter := Agent.GetFortCharacter[] )
				{
					FortCharacter.PutInStasis(stasis_args{AllowTurning := false, AllowFalling := false, AllowEmotes := true})
				}
			}
			else
			{
				Print("Error opening ui_loadout")
			}
		}
	}

    CloseUI(Agent:agent)<suspends>:void=
	{
		if ( InPlayer := player[Agent] )
		{
			if ( UILoadout := PlayerUIMap[InPlayer] )
			{
				spawn{UILoadout.CloseUI()}
			}

			InputForward.Unregister(Agent)
			InputBackward.Unregister(Agent)
			InputLeft.Unregister(Agent)
			InputRight.Unregister(Agent)
			CamClass.RemoveFrom(Agent)

			if ( PlayerClasses := PlayerClassesDevice? )
			{
				PlayerClasses.ShowHUD(Agent)
			}

			if ( FortCharacter := Agent.GetFortCharacter[] )
			{
				FortCharacter.ReleaseFromStasis()
			}
		}
	}

	GetAllItems():[]item_box_data=
	{
		var AllArray : []item_box_data = array{}
		set AllArray += array{item_box_data{ItemID := 0, Icon := Textures.ui_ability_no, Rarity := FNRarity.Common, VFXDevice := VFXUnusual0}}
		set AllArray += array{item_box_data{ItemID := 1, Icon := Textures.item_hat_halo, Rarity := FNRarity.Uncommon, VFXDevice := VFXUnusual1}}
		set AllArray += array{item_box_data{ItemID := 2, Icon := Textures.item_hat_firefly, Rarity := FNRarity.Legendary, VFXDevice := VFXUnusual2}}
		set AllArray += array{item_box_data{ItemID := 3, Icon := Textures.item_hat_furina, Rarity := FNRarity.Rare, VFXDevice := VFXUnusual3}}
		set AllArray += array{item_box_data{ItemID := 4, Icon := Textures.item_hat_cubed, Rarity := FNRarity.Epic, VFXDevice := VFXUnusual4}}
		set AllArray += array{item_box_data{ItemID := 5, Icon := Textures.item_hat_device, Rarity := FNRarity.Legendary, VFXDevice := VFXUnusual5}}
		set AllArray += array{item_box_data{ItemID := 6, Icon := Textures.item_hat_galaxy, Rarity := FNRarity.Epic, VFXDevice := VFXUnusual6}}
		set AllArray += array{item_box_data{ItemID := 7, Icon := Textures.item_hat_burning, Rarity := FNRarity.Uncommon, VFXDevice := VFXUnusual7}}
		set AllArray += array{item_box_data{ItemID := 8, Icon := Textures.item_hat_smoke, Rarity := FNRarity.Uncommon, VFXDevice := VFXUnusual8}}
		set AllArray += array{item_box_data{ItemID := 9, Icon := Textures.item_hat_glow, Rarity := FNRarity.Uncommon, VFXDevice := VFXUnusual9}}
		set AllArray += array{item_box_data{ItemID := 10, Icon := Textures.item_hat_glow, Rarity := FNRarity.Uncommon, VFXDevice := VFXUnusual10}}
		set AllArray += array{item_box_data{ItemID := 11, Icon := Textures.item_hat_glow, Rarity := FNRarity.Uncommon, VFXDevice := VFXUnusual10}}

		return AllArray
	}

	GetItemName(ItemID:int):string=
	{
		var StringTitle : string = ""
		case( ItemID )
		{
			0 =>	set StringTitle = ""
			1 =>	set StringTitle = "Halo"
			2 =>	set StringTitle = "Firefly"
			3 =>	set StringTitle = "Water?"
			4 =>	set StringTitle = "Cubed"
			5 =>	set StringTitle = "Device"
			6 =>	set StringTitle = "Galaxy"
			7 =>	set StringTitle = "Burning"
			8 =>	set StringTitle = "Smoke"
			9 =>	set StringTitle = "Glow"
			10 =>	set StringTitle = "Spark"
			11 =>	set StringTitle = "Flare"
			_ =>									#nothing
		}

		return StringTitle
	}

	GetItemWeight(ItemData:item_box_data):float=
	{
		var RarityWeight : float = 10.0
		case( ItemData.Rarity )
		{
			FNRarity.Common => # unobtainable
			{
				set RarityWeight = 0.0
			}
			FNRarity.Uncommon =>
			{
				set RarityWeight = 25.0
			}
			FNRarity.Rare =>
			{
				set RarityWeight = 10.0
			}
			FNRarity.Epic =>
			{
				set RarityWeight = 5.0
			}
			FNRarity.Legendary =>
			{
				set RarityWeight = 1.0
			}
			FNRarity.Mythic =>
			{
				set RarityWeight = 0.01
			}
		}

		return RarityWeight
	}

	GetRandomGachaItem():item_box_data=
	{
		#ItemsWeight := AllItemList.Length - 1

		var diceRoll : float = GetRandomFloat(0.0, ItemsWeight)
		var RandomItem : item_box_data = item_box_data{}

		AllItemList := GetAllItems()
		for (NewItem : AllItemList <#, NewItem.Rarity <> FNRarity.Common#>)
		{
			RarityWeight := GetItemWeight(NewItem)
			var RarityString : string = ""
			case( NewItem.Rarity )
			{
				FNRarity.Common => # unobtainable
				{
					set RarityString = "Common"
				}
				FNRarity.Uncommon =>
				{
					set RarityString = "Uncommon"
				}
				FNRarity.Rare =>
				{
					set RarityString = "Rare"
				}
				FNRarity.Epic =>
				{
					set RarityString = "Epic"
				}
				FNRarity.Legendary =>
				{
					set RarityString = "Legendary"
				}
				FNRarity.Mythic =>
				{
					set RarityString = "Mythic"
				}
			}

			# If item weight is greater (or equal) than our diceRoll, we take that item and return
			if (RarityWeight >= diceRoll)
			{
				set RandomItem = NewItem
				Print("{diceRoll} Reward: {RarityString} {GetItemName(NewItem.ItemID)}")
				return RandomItem
			}

			# If we didn't return, we substract the weight to our diceRoll and cycle to the next item
			set diceRoll -= RarityWeight
		}

		return RandomItem
	}

	InputPressedForward(Agent : agent) : void =
	{
		if ( InPlayer := player[Agent], UILoadout := PlayerUIMap[InPlayer], UILoadout.IsOpen? )
		{
			UILoadout.MoveItemBtn("Up")
		}
	}

	InputPressedBackward(Agent : agent) : void =
	{
		if ( InPlayer := player[Agent], UILoadout := PlayerUIMap[InPlayer], UILoadout.IsOpen? )
		{
			UILoadout.MoveItemBtn("Down")
		}
	}

	InputPressedLeft(Agent : agent) : void =
	{
		if ( InPlayer := player[Agent], UILoadout := PlayerUIMap[InPlayer], UILoadout.IsOpen? )
		{
			UILoadout.MoveItemBtn("Left")
		}
	}

	InputPressedRight(Agent : agent) : void =
	{
		if ( InPlayer := player[Agent], UILoadout := PlayerUIMap[InPlayer], UILoadout.IsOpen? )
		{
			UILoadout.MoveItemBtn("Right")
		}
	}

	InputPressedFireUI(Agent : agent) : void =
	{
		if ( InPlayer := player[Agent], UILoadout := PlayerUIMap[InPlayer], UILoadout.IsOpen? )
		{
			UILoadout.SelectButton()
		}
	}

	OnPlayerSpawned(Agent:agent):void=
	{
		Print("OnPlayerSpawned()")
		#spawn{AddUnusual(Agent)}
	}

	AddUnusual(Agent:agent, ItemID:int)<suspends>:void=
	{
		if ( ItemID = 0 )
		{
			# Clear FXs
			VFXUnusual0.Pickup(Agent)
			return
		}

		AllItemList := GetAllItems()
		for (NewItem : AllItemList, NewItem.ItemID = ItemID)
		{
			#Print("AddUnusual({ItemID})")
			NewItem.VFXDevice.Pickup(Agent)
		}
	}

	RewardRandomItem(Agent:agent)<suspends>:void=
	{
		NewItem := GetRandomGachaItem()
		ItemID := NewItem.ItemID
		if ( ItemID = 0 )
		{
			return
		}
	}