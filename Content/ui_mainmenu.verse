<#
	TeuAgency
	UI Main Menu, Settings.
#>
using { /Fortnite.com/Devices }
using { /Fortnite.com/Characters }
using { /Fortnite.com/Teams }
using { /Fortnite.com/UI }
using { /Fortnite.com/Game }
using { /Verse.org/Simulation }
using { /Verse.org/Simulation/Tags }
using { /Verse.org/Colors }
using { /Verse.org/Assets }
using { /UnrealEngine.com/Temporary/Diagnostics }
using { /UnrealEngine.com/Temporary/UI }
using { /UnrealEngine.com/Temporary/SpatialMath }
using { ModulePlayerClasses }

ui_settings := class()
{
    LinkedCreativeDevice<public> : ui_mainmenu_device
    LinkedPlayer<public> : player 

    var IsInitialized<private> : logic = false
    var MainCanvas<private> : canvas = canvas{}

	UITextMessage<localizes>(Value : string) : message = "{Value}"

	UIBG: material_block = material_block{DefaultImage := Materials.UI.M_UI_BG{}}
	UITitleBG: material_block = material_block{DefaultImage := Materials.UI.M_UI_BG{} <#, DefaultDesiredSize := vector2{X := 0.0, Y := 120.0}#>}
	UITitleText: text_block = text_block{DefaultTextColor := NamedColors.White, DefaultShadowColor := NamedColors.Black, DefaultShadowOffset := option{vector2{X := 2.0, Y := 2.0}} }

    BtnExit : button_loud = button_loud{}
    BtnBGM : button_regular = button_regular{}
    BtnVOMenu : button_regular = button_regular{}
    BtnMouseCursor : button_regular = button_regular{}
    BtnHat : button_regular = button_regular{}
    BtnFPS : button_regular = button_regular{}
	var MainMenuArray : []overlay = array{}

    Init<public>() : void =
	{
        if (IsInitialized?)
		{
            return
		}

		set MainCanvas = CreateUI()
		set IsInitialized = true
    }

    OpenUI<public>() : void =
	{
        if (PlayerUI := GetPlayerUI[LinkedPlayer])
		{
            PlayerUI.AddWidget(GetCanvas(), player_ui_slot{InputMode := ui_input_mode.All})
			UpdateUI()
		}
	}

    CloseUI<public>()<suspends> : void =
	{
        if (PlayerUI := GetPlayerUI[LinkedPlayer])
		{
			PlayerUI.RemoveWidget(GetCanvas())
		}
    }

    GetCanvas<public>() : canvas = MainCanvas

	CreateUI<private>() : canvas =
	{
		UICanvas : canvas = canvas{Slots := array{
		canvas_slot
		{
			ZOrder := {0}
			Anchors := anchors{Minimum := vector2{X := 0.0, Y := 0.0}, Maximum := vector2{X := 1.0, Y := 1.0}}
			Offsets := margin{Top := 0.0, Left := 0.0, Right := 0.0, Bottom := 0.0}
			Alignment := vector2{X := 0.0, Y := 0.0}
			SizeToContent := true
			Widget := UIBG
        }
		canvas_slot
		{
			ZOrder := {2}
			Anchors := anchors{Minimum := vector2{X := 0.0, Y := 0.0}, Maximum := vector2{X := 1.0, Y := 0.1}}
			Offsets := margin{Top := 0.0, Left := 0.0, Right := 0.0, Bottom := 0.0}
			Alignment := vector2{X := 0.0, Y := 0.0}
			SizeToContent := true
			Widget := UITitleBG
        }
		canvas_slot
		{
			ZOrder := {2}
			Anchors := anchors{Minimum := vector2{X := 0.5, Y := 0.025}, Maximum := vector2{X := 0.5, Y := 0.025} }
			Offsets := margin{Top := 0.0, Left := 0.0, Right := 0.0, Bottom := 0.0}
			Alignment := vector2{X := 0.5, Y := 0.0}
			SizeToContent := true
			Widget := UITitleText
        }
		canvas_slot
		{
			ZOrder := {1}
			Anchors := anchors{Minimum := vector2{X := 0.5, Y := 0.1}, Maximum := vector2{X := 0.5, Y := 0.1}}
			Offsets := margin{ Top:=0.0, Left:= 0.0, Right:=0.0, Bottom := 0.0 }
			Alignment := vector2{X := 0.5, Y := 0.0}
			SizeToContent := true
			Widget := stack_box{
				Orientation := orientation.Vertical
				Slots := array
				{
					stack_box_slot
					{
						VerticalAlignment := vertical_alignment.Fill
						HorizontalAlignment := horizontal_alignment.Fill
						Widget := BtnBGM
					}
					stack_box_slot
					{
						VerticalAlignment := vertical_alignment.Fill
						HorizontalAlignment := horizontal_alignment.Fill
						Widget := BtnVOMenu
					}
					stack_box_slot
					{
						VerticalAlignment := vertical_alignment.Fill
						HorizontalAlignment := horizontal_alignment.Fill
						Widget := BtnMouseCursor
					}
					stack_box_slot
					{
						VerticalAlignment := vertical_alignment.Fill
						HorizontalAlignment := horizontal_alignment.Fill
						Widget := BtnHat
					}
					stack_box_slot
					{
						VerticalAlignment := vertical_alignment.Fill
						HorizontalAlignment := horizontal_alignment.Fill
						Widget := BtnFPS
					}
				}
			}
        }
		canvas_slot
		{
			ZOrder := {2}
			Anchors := anchors{Minimum := vector2{X := 0.9, Y := 0.9}, Maximum := vector2{X := 0.98, Y := 0.98} }
			Offsets := margin{Top := 0.0, Left := 0.0, Right := 0.0, Bottom := 0.0}
			Alignment := vector2{X := 1.0, Y := 1.0}
			SizeToContent := true
			Widget := stack_box{
				Orientation := orientation.Horizontal
				Slots := array{
					stack_box_slot
					{
						VerticalAlignment := vertical_alignment.Fill
						HorizontalAlignment := horizontal_alignment.Fill
						#Padding := margin{Top := 0.0, Left := 0.0, Right := 0.0, Bottom := 0.0}
						Widget := BtnExit
					}
				}
			}
		}
		}}

		UITitleText.SetText(TextLocalization.Settings)

		BtnExit.SetText(TextLocalization.Exit)
		BtnExit.OnClick().Subscribe(OnBtnExit)

		BtnBGM.SetText(TextLocalization.Setting_MusicON)
		BtnVOMenu.SetText(TextLocalization.Setting_VoiceMenuOFF)
		BtnMouseCursor.SetText(TextLocalization.Setting_MouseCursorOFF)
		BtnHat.SetText(TextLocalization.Setting_HatON)
		BtnFPS.SetText(TextLocalization.Setting_FirstPersonOFF)

		BtnBGM.OnClick().Subscribe(OnBtnBGM)
		BtnVOMenu.OnClick().Subscribe(OnBtnVOMenu)
		BtnMouseCursor.OnClick().Subscribe(OnBtnMouse)
		BtnHat.OnClick().Subscribe(OnBtnHat)
		BtnFPS.OnClick().Subscribe(OnBtnFPS)

		return UICanvas
	}

	OnBtnExit(Message : widget_message) : void =
	{
		if ( PlayerUI := GetPlayerUI[Message.Player], Agent := agent[Message.Player])
		{
			spawn{LinkedCreativeDevice.CloseUIMenu(Agent)}
		}
	}

	OnBtnBGM(Message : widget_message) : void =
	{
		if ( PlayerUI := GetPlayerUI[Message.Player], Agent := agent[Message.Player])
		{
			ToggleBGM(Agent)
		}
	}

	ToggleBGM(Agent : agent) : void =
	{
		if ( SaveManager := LinkedCreativeDevice.SaveManagerDevice? )
		{
			Print("ToggleBGM()")
			if ( SaveManager.IsBGMEnabled[Agent] )
			{
				LinkedCreativeDevice.SetBGMEnabled(Agent, false)
				UpdateUI()
				LinkedCreativeDevice.SetPlayerAggro(Agent, FNBGMState.None)
				Print("BGM: Disable")
			}
			else
			{
				LinkedCreativeDevice.SetBGMEnabled(Agent, true)
				UpdateUI()
				LinkedCreativeDevice.SetPlayerAggro(Agent, FNBGMState.Idle)
				Print("BGM: Enable")
			}
		}
	}

	OnBtnVOMenu(Message : widget_message) : void =
	{
		if ( PlayerUI := GetPlayerUI[Message.Player], Agent := agent[Message.Player])
		{
			ToggleVOMenu(Agent)
		}
	}

	ToggleVOMenu(Agent : agent) : void =
	{
		if ( SaveManager := LinkedCreativeDevice.SaveManagerDevice? )
		{
			Print("ToggleVOMenu()")
			if ( SaveManager.IsVOMenuEnabled[Agent] )
			{
				LinkedCreativeDevice.SetVOMenuEnabled(Agent, false)
				UpdateUI()
				Print("VO Menu: Disable")
			}
			else
			{
				LinkedCreativeDevice.SetVOMenuEnabled(Agent, true)
				UpdateUI()
				Print("VO Menu: Enable")
			}
		}
	}

	OnBtnMouse(Message : widget_message) : void =
	{
		if ( PlayerUI := GetPlayerUI[Message.Player], Agent := agent[Message.Player])
		{
			ToggleMouseClick(Agent)
		}
	}

	ToggleMouseClick(Agent : agent) : void =
	{
		if ( SaveManager := LinkedCreativeDevice.SaveManagerDevice? )
		{
			Print("ToggleMouseClick()")
			if ( SaveManager.IsMouseEnabled[Agent] )
			{
				LinkedCreativeDevice.SetMouseEnabled(Agent, false)
				UpdateUI()
				Print("Mouse Click: Disable")
			}
			else
			{
				LinkedCreativeDevice.SetMouseEnabled(Agent, true)
				UpdateUI()
				Print("Mouse Click: Enable")
			}
		}
	}

	OnBtnHat(Message : widget_message) : void =
	{
		if ( PlayerUI := GetPlayerUI[Message.Player], Agent := agent[Message.Player])
		{
			ToggleHat(Agent)
		}
	}

	ToggleHat(Agent : agent) : void =
	{
		if ( SaveManager := LinkedCreativeDevice.SaveManagerDevice? )
		{
			Print("ToggleHat()")
			if ( SaveManager.IsHatEnabled[Agent] )
			{
				LinkedCreativeDevice.SetHatEnabled(Agent, false)
				UpdateUI()
				Print("Hat: Disable")
			}
			else
			{
				LinkedCreativeDevice.SetHatEnabled(Agent, true)
				UpdateUI()
				Print("Hat: Enable")
			}
		}
	}

	OnBtnFPS(Message : widget_message) : void =
	{
		if ( PlayerUI := GetPlayerUI[Message.Player], Agent := agent[Message.Player])
		{
			ToggleFPS(Agent)
		}
	}

	ToggleFPS(Agent : agent) : void =
	{
		if ( SaveManager := LinkedCreativeDevice.SaveManagerDevice? )
		{
			Print("ToggleHat()")
			if ( SaveManager.IsCamFPEnabled[Agent] )
			{
				LinkedCreativeDevice.SetFPSEnabled(Agent, false)
				UpdateUI()
				Print("FPS: Disable")
			}
			else
			{
				LinkedCreativeDevice.SetFPSEnabled(Agent, true)
				UpdateUI()
				Print("Hat: Enable")
			}
		}
	}

	UpdateUI<public>() : void =
	{
		if ( Agent := agent[LinkedPlayer], SaveManager := LinkedCreativeDevice.SaveManagerDevice? )
		{
			if ( SaveManager.IsBGMEnabled[Agent] )
			{
				BtnBGM.SetText(TextLocalization.Setting_MusicON)
			}
			else
			{
				BtnBGM.SetText(TextLocalization.Setting_MusicOFF)
			}

			if ( SaveManager.IsVOMenuEnabled[Agent] )
			{
				BtnVOMenu.SetText(TextLocalization.Setting_VoiceMenuON)
			}
			else
			{
				BtnVOMenu.SetText(TextLocalization.Setting_VoiceMenuOFF)
			}

			if ( SaveManager.IsMouseEnabled[Agent] )
			{
				BtnMouseCursor.SetText(TextLocalization.Setting_MouseCursorON)
			}
			else
			{
				BtnMouseCursor.SetText(TextLocalization.Setting_MouseCursorOFF)
			}

			if ( SaveManager.IsHatEnabled[Agent] )
			{
				BtnHat.SetText(TextLocalization.Setting_HatON)
			}
			else
			{
				BtnHat.SetText(TextLocalization.Setting_HatOFF)
			}

			if ( SaveManager.IsCamFPEnabled[Agent] )
			{
				BtnFPS.SetText(TextLocalization.Setting_FirstPersonON)
			}
			else
			{
				BtnFPS.SetText(TextLocalization.Setting_FirstPersonOFF)
			}
		}
    }
}

# A Verse-authored creative device that can be placed in a level
ui_mainmenu_device := class(creative_device):

	var UIMapSettings : [player]ui_settings = map{}

	@editable PlayerClassesDevice : ?player_classes_manager_device = false
	@editable SaveManagerDevice : ?save_manager_device = false
	@editable UIClassMenuDevice : ?ui_classmenu_manager_device = false
	@editable UIAchievementDevice : ?ui_achievements_device = false
	@editable LoadoutDevice : ?loadout_device = false
	@editable TeamDeathmatchDevice : ?fn_game_tdm = false

	@editable PopupMainMenu : popup_dialog_device = popup_dialog_device{}
	#@editable PopupSettings : popup_dialog_device = popup_dialog_device{}

	UITextMessage<localizes>(Value : string) : message = "{Value}"
	#UIPlayerName<localizes>(Agent:agent):message = "{Agent}"

	@editable CineHatOn : cinematic_sequence_device = cinematic_sequence_device{}
	@editable CineHatOff : cinematic_sequence_device = cinematic_sequence_device{}

    # Runs when the device is started in a running game
    OnBegin<override>()<suspends>:void=
	{
		Print("Initialize UI Main Menu...")

		PopupMainMenu.RespondingButtonEvent.Subscribe(OnMainMenuBtn)
		PopupMainMenu.ShownEvent.Subscribe(OnMainMenuShow)
		PopupMainMenu.DismissedEvent.Subscribe(OnMainMenuHide)

		<#PopupSettings.SetTitleText(UITextMessage(""))
		PopupSettings.SetDescriptionText(UITextMessage(""))

		PopupSettings.RespondingButtonEvent.Subscribe(OnSettingBtn)
		PopupSettings.ShownEvent.Subscribe(OnSettingShow)
		PopupSettings.DismissedEvent.Subscribe(OnSettingHide)#>
	}

	# Removes an element from the given map and returns a new map without that element
	(Map:[keytype]valuetype where keytype:subtype(comparable), valuetype:type).RemoveKey(ElementToRemove:keytype):[keytype]valuetype=
	{
		var NewMap:[keytype]valuetype = map{}
		for (Key -> Value : Map, Key <> ElementToRemove)
		{
			set NewMap = ConcatenateMaps(NewMap, map{Key => Value})
		}

		return NewMap
	}

    UIPlayerRemoved(Agent: agent): void =
	{
		Print("UIPlayerRemoved")
		if (AgentPlayer := player[Agent])
		{
			set UIMapSettings = UIMapSettings.RemoveKey(AgentPlayer)
		}
	}

    OpenUIMenu(Agent:agent):void=
	{
		if (InPlayer := player[Agent])
		{
			var OptPlayerUI :?ui_settings = false

			if (ThisPlayerUI := UIMapSettings[InPlayer])
			{
				set OptPlayerUI = option{ThisPlayerUI}
			}
			else if (ThisPlayerUI := set UIMapSettings[InPlayer] = ui_settings{ LinkedPlayer := InPlayer, LinkedCreativeDevice := Self } ) 
			{
				ThisPlayerUI.Init()
				set OptPlayerUI = option{ThisPlayerUI}
			}

			if (PlayerUI := OptPlayerUI?)
			{
				PlayerUI.OpenUI()
			}
			else
			{
				Print("Error opening ui_settings")
			}
		}
	}

    CloseUIMenu(Agent:agent)<suspends>:void=
	{
		if ( InPlayer := player[Agent], UIMenu := UIMapSettings[InPlayer] )
		{
			spawn{UIMenu.CloseUI()}
			PopupMainMenu.Show(Agent)
		}
	}

	OpenClassMenu(Agent:agent):void= 
	{
		if ( UIClassMenu := UIClassMenuDevice? )
		{
			UIClassMenu.OpenClassMenu(Agent)
		}

		if ( SaveManager := SaveManagerDevice?, SaveManager.IsHatEnabled[Agent] )
		{
			CineHatOn.Play(Agent)
		}
		else
		{
			CineHatOff.Play(Agent)
		}
	}

	OnMainMenuBtn(Agent:agent, BtnIndex:int):void= 
	{
		if ( Player := player[Agent], Player.IsActive[] )
		{
			case( BtnIndex )
			{
				0 =>
				{
					PopupMainMenu.Hide(Agent)
					Print("Main Menu: Play")
					spawn{MainMenuEnterGame(Agent)}
				}
				1 =>
				{
					PopupMainMenu.Hide(Agent)
					if ( UIAchievement := UIAchievementDevice? )
					{
						Print("Main Menu: Achievement")
						UIAchievement.OpenUIMenu(Agent)
					}
				}
				2 =>
				{
					PopupMainMenu.Hide(Agent)
					Print("Main Menu: Settings")
					OpenUIMenu(Agent)
				}
				3 =>
				{
					PopupMainMenu.Hide(Agent)
					if ( UILoadout := LoadoutDevice? )
					{
						Print("Main Menu: Loadout")
						UILoadout.OpenUI(Agent)
					}
				}
				_ =>
				{
					PopupMainMenu.Hide(Agent)
				}
			}
		}
	}

	OnMainMenuShow(Agent:agent):void= 
	{
		if ( PlayerClasses := PlayerClassesDevice? )
		{
			PlayerClasses.HideHUD(Agent)
		}

		if ( TeamDeathmatch := TeamDeathmatchDevice? )
		{
			TeamDeathmatch.HideHUD(Agent)
		}
	}

	OnMainMenuHide(Agent:agent):void= 
	{
		if ( PlayerClasses := PlayerClassesDevice? )
		{
			PlayerClasses.ShowHUD(Agent)
		}

		if ( TeamDeathmatch := TeamDeathmatchDevice? )
		{
			TeamDeathmatch.ShowHUD(Agent)
		}
	}

	MainMenuEnterGame(Agent:agent)<suspends>:void=
	{
		Sleep(0.1)
		OpenClassMenu(Agent)
	}

	MainMenuShow(Agent:agent)<suspends>:void=
	{
		PopupMainMenu.SetTitleText(TextLocalization.UIPlayerName(Agent))

		if ( SaveManager := SaveManagerDevice? )
		{
			PlayerLevel := SaveManager.GetPlayerLevel(Agent)
			PlayerXP := SaveManager.GetPlayerXP(Agent)
			PopupMainMenu.SetDescriptionText(UITextMessage("Lv. {PlayerLevel} [XP - {PlayerXP}/100]"))
		}
		else
		{
			PopupMainMenu.SetDescriptionText(UITextMessage("Lv. "))
		}

		PopupMainMenu.Show(Agent)
	}


	<#OnSettingBtn(Agent:agent, BtnIndex:int):void= 
	{
		if ( Player := player[Agent], Player.IsActive[] )
		{
			case( BtnIndex )
			{
				0 =>
				{
					Print("Setting: Music")
				}
				1 =>
				{
					Print("Setting: VO Menu")
				}
				2 =>
				{
					Print("Setting: Mouse")
				}
				5 =>
				{
					Print("Setting: Exit")
				}
				_ =>
				{
					PopupSettings.Hide(Agent)
				}
			}
		}
	}

	OnSettingShow(Agent:agent):void= 
	{
		if ( PlayerClasses := PlayerClassesDevice? )
		{
			PlayerClasses.HideHUD(Agent)
		}

		if ( TeamDeathmatch := TeamDeathmatchDevice? )
		{
			TeamDeathmatch.HideHUD(Agent)
		}
	}

	OnSettingHide(Agent:agent):void= 
	{
		if ( PlayerClasses := PlayerClassesDevice? )
		{
			PlayerClasses.ShowHUD(Agent)
		}

		if ( TeamDeathmatch := TeamDeathmatchDevice? )
		{
			TeamDeathmatch.ShowHUD(Agent)
		}
	}#>


	SetBGMEnabled(Agent:agent, bEnable:logic):void=
	{
		if ( SaveManager := SaveManagerDevice? )
		{
			SaveManager.ToggleSettingBGM(Agent, bEnable)
		}
	}

	SetHatEnabled(Agent:agent, bEnable:logic):void=
	{
		if ( SaveManager := SaveManagerDevice? )
		{
			SaveManager.ToggleSettingHat(Agent, bEnable)
		}
	}

	SetFPSEnabled(Agent:agent, bEnable:logic):void=
	{
		if ( SaveManager := SaveManagerDevice? )
		{
			SaveManager.ToggleSettingCamFP(Agent, bEnable)
		}
	}

	SetVOMenuEnabled(Agent:agent, bEnable:logic):void=
	{
		if ( SaveManager := SaveManagerDevice? )
		{
			SaveManager.ToggleSettingVOMenu(Agent, bEnable)
		}
	}

	SetMouseEnabled(Agent:agent, bEnable:logic):void=
	{
		if ( SaveManager := SaveManagerDevice? )
		{
			SaveManager.ToggleSettingMouse(Agent, bEnable)
		}
	}

	SetPlayerAggro(Agent:agent, BGMState:FNBGMState):void=
	{
		if ( PlayerClasses := PlayerClassesDevice? )
		{
			PlayerClasses.SetPlayerAggro(Agent,BGMState)
		}
	}
