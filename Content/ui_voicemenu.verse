<#
	2AGENCY
	UI - Voice Menu
#>
using { /Fortnite.com/Devices }
using { /Fortnite.com/UI }
using { /Fortnite.com/Game }
using { /Verse.org/Simulation }
using { /Verse.org/Simulation/Tags }
using { /Verse.org/Colors }
using { /UnrealEngine.com/Temporary/Diagnostics }
using { /UnrealEngine.com/Temporary/UI }
using { /UnrealEngine.com/Temporary/SpatialMath }
using { ModulePlayerClasses }

# Notes!!!
# nothing

ui_voicemenu := class() 
{
    LinkedCreativeDevice<public> : ui_voicemenu_manager_device
    LinkedPlayer<public> : player

    var IsInitialized<private> : logic = false
    var MainCanvas<private> : canvas = canvas{}

	UITextMessage<localizes>(Value : string) : message = "{Value}"

	ColorBG: color_block = color_block {DefaultColor := NamedColors.Black, DefaultOpacity := 0.75}
    VO1Button : button_quiet = button_quiet{}
    VO2Button : button_quiet = button_quiet{}
    VO3Button : button_quiet = button_quiet{}
    VO4Button : button_quiet = button_quiet{}
    VO5Button : button_quiet = button_quiet{}
    VO6Button : button_quiet = button_quiet{}
    VO7Button : button_quiet = button_quiet{}
    VO8Button : button_quiet = button_quiet{}

    var iScrollSlot : int = 0

    Init<public>() : void =
	{
        if (IsInitialized?)
		{
            return
		}

		set MainCanvas = CreateUI()
		set IsInitialized = true
    }

    OpenUI<public>() : void =
	{
		if (PlayerUI := GetPlayerUI[LinkedPlayer])
		{
			PlayerUI.AddWidget(GetCanvas(), player_ui_slot{InputMode := ui_input_mode.None})
		}
	}

	<#UITimeOut()<suspends> : void =
	{
		Sleep(4.0)
		if ( PlayerUI := GetPlayerUI[LinkedPlayer])
		{
			PlayerUI.RemoveWidget(GetCanvas())
		}
	}#>

    CloseUI<public>(PlayerUI: player_ui)<suspends> : void =
	{
		PlayerUI.RemoveWidget(GetCanvas())
    }

    ScrollButton(bUp:logic) : void =
	{
		if ( bUp? )
		{
			set iScrollSlot -= 1
			if ( iScrollSlot <= 0 ) { set iScrollSlot = 0 }
		}
		else
		{
			set iScrollSlot += 1
			if ( iScrollSlot >= 7 ) { set iScrollSlot = 7 }
		}

        UpdateScrollButton()
	}

    UpdateScrollButton() : void =
	{
        VO1Button.SetEnabled(false)
        VO2Button.SetEnabled(false)
        VO3Button.SetEnabled(false)
        VO4Button.SetEnabled(false)
        VO5Button.SetEnabled(false)
        VO6Button.SetEnabled(false)
        VO7Button.SetEnabled(false)
        VO8Button.SetEnabled(false)
		case( iScrollSlot )
		{
			0 =>		VO1Button.SetEnabled(true)
			1 =>		VO2Button.SetEnabled(true)
			2 =>		VO3Button.SetEnabled(true)
			3 =>		VO4Button.SetEnabled(true)
			4 =>		VO5Button.SetEnabled(true)
			5 =>		VO6Button.SetEnabled(true)
			6 =>		VO7Button.SetEnabled(true)
			7 =>		VO8Button.SetEnabled(true)
			_ =>		# nothing
		}
	}

    GetCanvas<public>() : canvas = MainCanvas  # Exposes the canvas to public

	CreateUI<private>() : canvas =
	{
		UICanvas : canvas = canvas{Slots := array{
		canvas_slot
		{
			ZOrder := {0}
			Anchors := anchors{Minimum := vector2{X := 0.15, Y := 0.25}, Maximum := vector2{X := 0.35, Y := 0.75}}
			Offsets := margin{ Top:=0.0, Left:= 0.0, Right:=0.0, Bottom := 0.0 }
			Alignment := vector2{X := 0.5, Y := 0.5}
			SizeToContent := true
			Widget := ColorBG
        }
		canvas_slot{
			ZOrder := {1}
			Anchors := anchors{Minimum := vector2{X := 0.25, Y := 0.5}, Maximum := vector2{X := 0.25, Y := 0.5}}
			Offsets := margin{ Top:=0.0, Left:= 0.0, Right:=0.0, Bottom := 0.0 }
			Alignment := vector2{X := 0.5, Y := 0.5}
			SizeToContent := true
			Widget := stack_box{
				Orientation := orientation.Vertical
				Slots := array{
					stack_box_slot{
						VerticalAlignment := vertical_alignment.Fill
						HorizontalAlignment := horizontal_alignment.Fill
						Padding := margin{Top := 0.0, Left := 0.0, Right := 0.0, Bottom := 0.0}
						Widget := VO1Button
					}
					stack_box_slot{
						VerticalAlignment := vertical_alignment.Fill
						HorizontalAlignment := horizontal_alignment.Fill
						Padding := margin{Top := 0.0, Left := 0.0, Right := 0.0, Bottom := 0.0}
						Widget := VO2Button
						}
					stack_box_slot{
						VerticalAlignment := vertical_alignment.Fill
						HorizontalAlignment := horizontal_alignment.Fill
						Padding := margin{Top := 0.0, Left := 0.0, Right := 0.0, Bottom := 0.0}
						Widget := VO3Button
						}
					stack_box_slot{
						VerticalAlignment := vertical_alignment.Fill
						HorizontalAlignment := horizontal_alignment.Fill
						Padding := margin{Top := 0.0, Left := 0.0, Right := 0.0, Bottom := 0.0}
						Widget := VO4Button
						}
					stack_box_slot{
						VerticalAlignment := vertical_alignment.Fill
						HorizontalAlignment := horizontal_alignment.Fill
						Padding := margin{Top := 0.0, Left := 0.0, Right := 0.0, Bottom := 0.0}
						Widget := VO5Button
						}
					stack_box_slot{
						VerticalAlignment := vertical_alignment.Fill
						HorizontalAlignment := horizontal_alignment.Fill
						Padding := margin{Top := 0.0, Left := 0.0, Right := 0.0, Bottom := 0.0}
						Widget := VO6Button
						}
					stack_box_slot{
						VerticalAlignment := vertical_alignment.Fill
						HorizontalAlignment := horizontal_alignment.Fill
						Padding := margin{Top := 0.0, Left := 0.0, Right := 0.0, Bottom := 0.0}
						Widget := VO7Button
						}
					stack_box_slot{
						VerticalAlignment := vertical_alignment.Fill
						HorizontalAlignment := horizontal_alignment.Fill
						Padding := margin{Top := 0.0, Left := 0.0, Right := 0.0, Bottom := 0.0}
						Widget := VO8Button
						}
					}
				}
        }}}

        VO1Button.SetText(UITextMessage("Ouch!")) 
        VO2Button.SetText(UITextMessage("Alert!"))
        VO3Button.SetText(UITextMessage("Where?"))
        VO4Button.SetText(UITextMessage("~🎵"))
        VO5Button.SetText(UITextMessage("?!"))
        VO6Button.SetText(UITextMessage("Sus"))
        VO7Button.SetText(UITextMessage("Not Sus"))
        VO8Button.SetText(UITextMessage("..."))

        UpdateScrollButton()

		return UICanvas
	}

	PlaySelectedVO() : void =
	{
		if ( PlayerUI := GetPlayerUI[LinkedPlayer] )
		{
			spawn{CloseUI(PlayerUI)}
		}
		if ( Agent := agent[LinkedPlayer] )
		{
			case( iScrollSlot )
			{
				0 =>		LinkedCreativeDevice.PlayClassVO(Agent, FNVOType.DBNO)
				1 =>		LinkedCreativeDevice.PlayClassVO(Agent, FNVOType.FullAlert)
				2 =>		LinkedCreativeDevice.PlayClassVO(Agent, FNVOType.FullAlertLost)
				3 =>		LinkedCreativeDevice.PlayClassVO(Agent, FNVOType.Idle)
				4 =>		LinkedCreativeDevice.PlayClassVO(Agent, FNVOType.Interrogation)
				5 =>		LinkedCreativeDevice.PlayClassVO(Agent, FNVOType.Sus)
				6 =>		LinkedCreativeDevice.PlayClassVO(Agent, FNVOType.ReturnIdle)
				7 =>		# nothing
				_ =>		# nothing
			}
		}
	}
}

# A Verse-authored creative device that can be placed in a level
ui_voicemenu_manager_device := class(creative_device):

	@editable PlayerClassesDevice : ?player_classes_manager_device = false
	@editable SaveManagerDevice : ?save_manager_device = false

	# Voice Menu UI
	var UIVoiceMenuMap : [player]ui_voicemenu = map{}
	@editable InputVoiceMenu : input_trigger_device = input_trigger_device{}
	@editable InputScrollUp : input_trigger_device = input_trigger_device{}
	@editable InputScrollDown : input_trigger_device = input_trigger_device{}

	<#@editable MSGFeedVO1 	: elimination_feed_device = elimination_feed_device{}
	@editable MSGFeedVO2 	: elimination_feed_device = elimination_feed_device{}
	@editable MSGFeedVO3 	: elimination_feed_device = elimination_feed_device{}
	@editable MSGFeedVO4 	: elimination_feed_device = elimination_feed_device{}
	@editable MSGFeedVO5 	: elimination_feed_device = elimination_feed_device{}
	@editable MSGFeedVO6 	: elimination_feed_device = elimination_feed_device{}
	@editable MSGFeedVO7 	: elimination_feed_device = elimination_feed_device{}#>

    # Runs when the device is started in a running game
    OnBegin<override>()<suspends>:void=
	{
		Print("Initialize UI Voice Menu...")

		InputVoiceMenu.PressedEvent.Subscribe(InputPressedVoiceMenu)
		InputVoiceMenu.ReleasedEvent.Subscribe(InputReleasedVoiceMenu)
		InputScrollUp.PressedEvent.Subscribe(InputPressedScrollUp)
		InputScrollDown.PressedEvent.Subscribe(InputPressedScrollDown)
	}

	# voice menu
    InputPressedVoiceMenu(Agent : agent) : void =
	{
		if ( not IsVOMenuEnabled[Agent] )
		{
			return
		}

		if ( InPlayer := player[Agent])
		{
			var OptPlayerUI :?ui_voicemenu = false
			if (ThisPlayerUI := UIVoiceMenuMap[InPlayer])
			{
				set OptPlayerUI = option{ThisPlayerUI}
			}
			else if (ThisPlayerUI := set UIVoiceMenuMap[InPlayer] = ui_voicemenu{ LinkedPlayer := InPlayer, LinkedCreativeDevice := Self } ) 
			{
				ThisPlayerUI.Init()
				set OptPlayerUI = option{ThisPlayerUI}
			}

			if (PlayerUI := OptPlayerUI?)
			{
				PlayerUI.OpenUI()
				InputScrollUp.Register(Agent)
				InputScrollDown.Register(Agent)
			}
			else
			{
				Print("Error opening ui_voicemenu")
			}
		}
	}

    InputReleasedVoiceMenu(Agent:agent, fHeldDuration:float) : void =
	{
		if (UIMenu := UIVoiceMenuMap[Agent])
		{
			UIMenu.PlaySelectedVO()
		}
		InputScrollUp.Unregister(Agent)
		InputScrollDown.Unregister(Agent)
	}

    InputPressedScrollUp(Agent : agent) : void =
	{
		if (UIMenu := UIVoiceMenuMap[Agent])
		{
			UIMenu.ScrollButton(true)
		}
	}

    InputPressedScrollDown(Agent : agent) : void =
	{
		if (UIMenu := UIVoiceMenuMap[Agent])
		{
			UIMenu.ScrollButton(false)
		}
	}

	# Removes an element from the given map and returns a new map without that element
	(Map:[keytype]valuetype where keytype:subtype(comparable), valuetype:type).RemoveKey(ElementToRemove:keytype):[keytype]valuetype=
	{
		var NewMap:[keytype]valuetype = map{}
		for (Key -> Value : Map, Key <> ElementToRemove)
		{
			set NewMap = ConcatenateMaps(NewMap, map{Key => Value})
		}

		return NewMap
	}

    UIPlayerRemoved(Agent: agent): void =
	{
		Print("UIPlayerRemoved")
		if (AgentPlayer := player[Agent])
		{
			set UIVoiceMenuMap = UIVoiceMenuMap.RemoveKey(AgentPlayer)
		}
	}

    PlayClassVO(Agent:agent, VOType:FNVOType):void=
	{
		if ( PlayerClasses := PlayerClassesDevice? )
		{
			PlayerClasses.PlayClassVO(Agent,VOType)
			<#case( VOType )
			{
				FNVOType.DBNO =>			MSGFeedVO1.Activate(Agent)
				FNVOType.FullAlert =>		MSGFeedVO2.Activate(Agent)
				FNVOType.FullAlertLost =>	MSGFeedVO3.Activate(Agent)
				FNVOType.Idle =>			MSGFeedVO4.Activate(Agent)
				FNVOType.Interrogation =>	MSGFeedVO5.Activate(Agent)
				FNVOType.Sus =>				MSGFeedVO6.Activate(Agent)
				FNVOType.ReturnIdle =>		MSGFeedVO7.Activate(Agent)
				_ =>		# nothing
			}#>
		}
	}

    IsVOMenuEnabled(Agent:agent)<transacts><decides>:void =
	{
		if ( SaveManager := SaveManagerDevice? )
		{
			SaveManager.IsVOMenuEnabled[Agent]
		}
	}
